using { /Fortnite.com/Devices }
using { /Fortnite.com/Characters }
using { /Fortnite.com/UI }
using { /Fortnite.com/Game }
using { /Verse.org/Simulation }
using { /Verse.org/Colors }
using { /Verse.org/Random }
using { /UnrealEngine.com/Temporary/Diagnostics }
using { /UnrealEngine.com/Temporary/SpatialMath }
using { /UnrealEngine.com/Temporary/UI }
using { /Verse.org/Verse }
using { /Verse.org/Assets }
using { /Verse.org/Colors/NamedColors }

Custom_Player := class<unique>():

    MyAgentObj:agent

    #LOAD
    var GameLoaded:int = 0

    var TextBlockPetNbr :text_block = text_block{}
    var TotalPetsActive:int = 0
    
    var FollowActiveL1:int = 0
    var FollowActiveL2:int = 0
    var FollowActiveL3:int = 0
    var FollowActiveL4:int = 0
    var FollowActiveL5:int = 0
    var FollowActiveL6:int = 0
    var FollowActiveL7:int = 0
    var FollowActiveL8:int = 0
    var FollowActiveL9:int = 0
    var FollowActiveL10:int = 0
    var FollowActiveL11:int = 0
    var FollowActiveL12:int = 0
    var FollowActiveL13:int = 0
    var FollowActiveL14:int = 0
    var FollowActiveL15:int = 0
    var FollowActiveL16:int = 0
    var FollowActiveL17:int = 0
    var FollowActiveL18:int = 0
    var FollowActiveL19:int = 0
    var FollowActiveL20:int = 0
    var FollowActiveL21:int = 0
    var FollowActiveL22:int = 0
    var FollowActiveL23:int = 0
    var FollowActiveL24:int = 0
    var FollowActiveL25:int = 0
    var FollowActiveL26:int = 0
    var FollowActiveL27:int = 0
    var FollowActiveL28:int = 0
    var FollowActiveL29:int = 0
    var FollowActiveL30:int = 0

    var FollowActiveL31:int = 0
    var FollowActiveL32:int = 0
    var FollowActiveL33:int = 0
    var FollowActiveL34:int = 0

    #Purchase Check

    var ButtonL1Pressed:int = 0
    var ButtonL2Pressed:int = 0
    var ButtonL3Pressed:int = 0
    var ButtonL4Pressed:int = 0
    var ButtonL5Pressed:int = 0
    var ButtonL6Pressed:int = 0
    var ButtonL7Pressed:int = 0
    var ButtonL8Pressed:int = 0
    var ButtonL9Pressed:int = 0
    var ButtonL10Pressed:int = 0
    var ButtonL11Pressed:int = 0
    var ButtonL12Pressed:int = 0
    var ButtonL13Pressed:int = 0
    var ButtonL14Pressed:int = 0
    var ButtonL15Pressed:int = 0
    var ButtonL16Pressed:int = 0
    var ButtonL17Pressed:int = 0
    var ButtonL18Pressed:int = 0
    var ButtonL19Pressed:int = 0
    var ButtonL20Pressed:int = 0
    var ButtonL21Pressed:int = 0
    var ButtonL22Pressed:int = 0
    var ButtonL23Pressed:int = 0
    var ButtonL24Pressed:int = 0
    var ButtonL25Pressed:int = 0
    var ButtonL26Pressed:int = 0
    var ButtonL27Pressed:int = 0
    var ButtonL28Pressed:int = 0
    var ButtonL29Pressed:int = 0
    var ButtonL30Pressed:int = 0

    var ButtonL31Pressed:int = 0
    var ButtonL32Pressed:int = 0
    var ButtonL33Pressed:int = 0
    var ButtonL34Pressed:int = 0

    #Pets Owned
    var Ownership_L1_Pet1:int = 0
    var Ownership_L1_Pet2:int = 0
    var Ownership_L1_Pet3:int = 0
    var Ownership_L1_Pet4:int = 0
    
    var Ownership_L2_Pet1:int = 0
    var Ownership_L2_Pet2:int = 0
    var Ownership_L2_Pet3:int = 0
    var Ownership_L2_Pet4:int = 0

    var Ownership_L3_Pet1:int = 0
    var Ownership_L3_Pet2:int = 0
    var Ownership_L3_Pet3:int = 0
    var Ownership_L3_Pet4:int = 0

    var Ownership_L4_Pet1:int = 0
    var Ownership_L4_Pet2:int = 0
    var Ownership_L4_Pet3:int = 0
    var Ownership_L4_Pet4:int = 0

    var Ownership_L5_Pet1:int = 0
    var Ownership_L5_Pet2:int = 0
    var Ownership_L5_Pet3:int = 0
    var Ownership_L5_Pet4:int = 0

    var Ownership_L6_Pet1:int = 0
    var Ownership_L6_Pet2:int = 0
    var Ownership_L6_Pet3:int = 0
    var Ownership_L6_Pet4:int = 0

    var Ownership_L7_Pet1:int = 0
    var Ownership_L7_Pet2:int = 0
    var Ownership_L7_Pet3:int = 0
    var Ownership_L7_Pet4:int = 0

    var Ownership_L8_Pet1:int = 0
    var Ownership_L8_Pet2:int = 0
    var Ownership_L8_Pet3:int = 0
    var Ownership_L8_Pet4:int = 0

    var Ownership_L9_Pet1:int = 0
    var Ownership_L9_Pet2:int = 0
    var Ownership_L9_Pet3:int = 0
    var Ownership_L9_Pet4:int = 0

    var Ownership_L10_Pet1:int = 0
    var Ownership_L10_Pet2:int = 0
    var Ownership_L10_Pet3:int = 0
    var Ownership_L10_Pet4:int = 0

    var Ownership_L11_Pet1:int = 0
    var Ownership_L11_Pet2:int = 0
    var Ownership_L11_Pet3:int = 0
    var Ownership_L11_Pet4:int = 0

    var Ownership_L12_Pet1:int = 0
    var Ownership_L12_Pet2:int = 0
    var Ownership_L12_Pet3:int = 0
    var Ownership_L12_Pet4:int = 0

    var Ownership_L13_Pet1:int = 0
    var Ownership_L13_Pet2:int = 0
    var Ownership_L13_Pet3:int = 0
    var Ownership_L13_Pet4:int = 0

    var Ownership_L14_Pet1:int = 0
    var Ownership_L14_Pet2:int = 0
    var Ownership_L14_Pet3:int = 0
    var Ownership_L14_Pet4:int = 0

    var Ownership_L15_Pet1:int = 0
    var Ownership_L15_Pet2:int = 0
    var Ownership_L15_Pet3:int = 0
    var Ownership_L15_Pet4:int = 0

    var Ownership_L16_Pet1:int = 0
    var Ownership_L16_Pet2:int = 0
    var Ownership_L16_Pet3:int = 0
    var Ownership_L16_Pet4:int = 0

    var Ownership_L17_Pet1:int = 0
    var Ownership_L17_Pet2:int = 0
    var Ownership_L17_Pet3:int = 0
    var Ownership_L17_Pet4:int = 0
    
    var Ownership_L18_Pet1:int = 0
    var Ownership_L18_Pet2:int = 0
    var Ownership_L18_Pet3:int = 0
    var Ownership_L18_Pet4:int = 0
    
    var Ownership_L19_Pet1:int = 0
    var Ownership_L19_Pet2:int = 0
    var Ownership_L19_Pet3:int = 0
    var Ownership_L19_Pet4:int = 0

    var Ownership_L20_Pet1:int = 0
    var Ownership_L20_Pet2:int = 0
    var Ownership_L20_Pet3:int = 0
    var Ownership_L20_Pet4:int = 0

    var Ownership_L21_Pet1:int = 0
    var Ownership_L21_Pet2:int = 0
    var Ownership_L21_Pet3:int = 0
    var Ownership_L21_Pet4:int = 0

    var Ownership_L22_Pet1:int = 0
    var Ownership_L22_Pet2:int = 0
    var Ownership_L22_Pet3:int = 0
    var Ownership_L22_Pet4:int = 0

    var Ownership_L23_Pet1:int = 0
    var Ownership_L23_Pet2:int = 0
    var Ownership_L23_Pet3:int = 0
    var Ownership_L23_Pet4:int = 0

    var Ownership_L24_Pet1:int = 0
    var Ownership_L24_Pet2:int = 0
    var Ownership_L24_Pet3:int = 0
    var Ownership_L24_Pet4:int = 0

    var Ownership_L25_Pet1:int = 0
    var Ownership_L25_Pet2:int = 0
    var Ownership_L25_Pet3:int = 0
    var Ownership_L25_Pet4:int = 0

    var Ownership_L26_Pet1:int = 0
    var Ownership_L26_Pet2:int = 0
    var Ownership_L26_Pet3:int = 0
    var Ownership_L26_Pet4:int = 0

    var Ownership_L27_Pet1:int = 0
    var Ownership_L27_Pet2:int = 0
    var Ownership_L27_Pet3:int = 0
    var Ownership_L27_Pet4:int = 0

    var Ownership_L28_Pet1:int = 0
    var Ownership_L28_Pet2:int = 0
    var Ownership_L28_Pet3:int = 0
    var Ownership_L28_Pet4:int = 0

    var Ownership_L29_Pet1:int = 0
    var Ownership_L29_Pet2:int = 0
    var Ownership_L29_Pet3:int = 0
    var Ownership_L29_Pet4:int = 0

    var Ownership_L30_Pet1:int = 0
    var Ownership_L30_Pet2:int = 0
    var Ownership_L30_Pet3:int = 0
    var Ownership_L30_Pet4:int = 0

    var Ownership_Secret_Pet1:int = 0
    var Ownership_Secret_Pet2:int = 0
    var Ownership_Secret_Pet3:int = 0
    var Ownership_Secret_Pet4:int = 0
    var Ownership_Secret_Pet5:int = 0

    var Ownership_Total_Pets:int = 0

    #Pets Unlocked

    var Pet_Unlocked_L1:int = 0  #1,2,3,4
    var Pet_Unlocked_L2:int = 0
    var Pet_Unlocked_L3:int = 0
    var Pet_Unlocked_L4:int = 0
    var Pet_Unlocked_L5:int = 0
    var Pet_Unlocked_L6:int = 0
    var Pet_Unlocked_L7:int = 0
    var Pet_Unlocked_L8:int = 0
    var Pet_Unlocked_L9:int = 0
    var Pet_Unlocked_L10:int = 0
    var Pet_Unlocked_L11:int = 0
    var Pet_Unlocked_L12:int = 0
    var Pet_Unlocked_L13:int = 0
    var Pet_Unlocked_L14:int = 0
    var Pet_Unlocked_L15:int = 0
    var Pet_Unlocked_L16:int = 0
    var Pet_Unlocked_L17:int = 0
    var Pet_Unlocked_L18:int = 0
    var Pet_Unlocked_L19:int = 0
    var Pet_Unlocked_L20:int = 0
    var Pet_Unlocked_L21:int = 0
    var Pet_Unlocked_L22:int = 0
    var Pet_Unlocked_L23:int = 0
    var Pet_Unlocked_L24:int = 0
    var Pet_Unlocked_L25:int = 0
    var Pet_Unlocked_L26:int = 0
    var Pet_Unlocked_L27:int = 0
    var Pet_Unlocked_L28:int = 0
    var Pet_Unlocked_L29:int = 0
    var Pet_Unlocked_L30:int = 0

    #Barrier

    var TextBlockPetMax :text_block = text_block{}
    var TotalPetsBuyable:int = 1

    var B1_Unlocked:int = 0
    var B2_Unlocked:int = 0
    var B3_Unlocked:int = 0
    var B4_Unlocked:int = 0
    var B5_Unlocked:int = 0
    var B6_Unlocked:int = 0
    var B7_Unlocked:int = 0
    var B8_Unlocked:int = 0
    var B9_Unlocked:int = 0
    var B10_Unlocked:int = 0
    var B11_Unlocked:int = 0
    var B12_Unlocked:int = 0
    var B13_Unlocked:int = 0
    var B14_Unlocked:int = 0
    var B15_Unlocked:int = 0
    var B16_Unlocked:int = 0
    var B17_Unlocked:int = 0
    var B18_Unlocked:int = 0
    var B19_Unlocked:int = 0
    var B20_Unlocked:int = 0
    var B21_Unlocked:int = 0
    var B22_Unlocked:int = 0
    var B23_Unlocked:int = 0
    var B24_Unlocked:int = 0
    var B25_Unlocked:int = 0
    var B26_Unlocked:int = 0
    var B27_Unlocked:int = 0
    var B28_Unlocked:int = 0
    var B29_Unlocked:int = 0
    var B30_Unlocked:int = 0
    var B31_Unlocked:int = 0
    var B32_Unlocked:int = 0
    var B33_Unlocked:int = 0
    var B34_Unlocked:int = 0

    #Pet Power

    var TextBlockPower :text_block = text_block{}
    var Pet_Power_Value:float = 0.0

    var Pet_Power_ValueL1:int = 0
    var Pet_Power_ValueL2:int = 0
    var Pet_Power_ValueL3:int = 0
    var Pet_Power_ValueL4:int = 0
    var Pet_Power_ValueL5:int = 0
    var Pet_Power_ValueL6:int = 0
    var Pet_Power_ValueL7:int = 0
    var Pet_Power_ValueL8:int = 0
    var Pet_Power_ValueL9:int = 0
    var Pet_Power_ValueL10:int = 0
    var Pet_Power_ValueL11:int = 0
    var Pet_Power_ValueL12:int = 0
    var Pet_Power_ValueL13:int = 0
    var Pet_Power_ValueL14:int = 0
    var Pet_Power_ValueL15:int = 0
    var Pet_Power_ValueL16:int = 0
    var Pet_Power_ValueL17:int = 0
    var Pet_Power_ValueL18:int = 0
    var Pet_Power_ValueL19:int = 0
    var Pet_Power_ValueL20:int = 0
    var Pet_Power_ValueL21:int = 0
    var Pet_Power_ValueL22:int = 0
    var Pet_Power_ValueL23:int = 0
    var Pet_Power_ValueL24:int = 0
    var Pet_Power_ValueL25:int = 0
    var Pet_Power_ValueL26:int = 0
    var Pet_Power_ValueL27:int = 0
    var Pet_Power_ValueL28:int = 0
    var Pet_Power_ValueL29:int = 0
    var Pet_Power_ValueL30:int = 0

    var BonusPowerMultiplicator:float=1.0 #Increase with rebirth etc

    #Pet Crit
    var TextBlockCrit :text_block = text_block{}
    var Pet_Crit_Value:int = 0

    #Pet Speed
    var TextBlockSpeed :text_block = text_block{}
    var Pet_Speed_Value:float = 1.0

    #Player Speed
    var TextBlockPlayerSpeed :text_block = text_block{}
    var Player_Speed_Value:float = 1.0

    #Pet Luck
    var TextBlockLuck :text_block = text_block{}
    var Pet_Luck_Value:int = 0
    var Pet_Luck_Value_Final:int = 0

    #Player Power
    var Power_Player:float = 100.0
    
    #gold
    var TextBlockGold:text_block = text_block{}
    var PlayerGold:int = 0

    #diamond
    var TextBlockDiamond:text_block = text_block{}
    var PlayerDiamond:int = 0

    #Area

    var PlayerInTreasureAreaFlag:int = 0

    #Rebirth

    var RebirthCount:int = 0

    #Inventory

    var Potion_Owned_Luck1:int = 0
    var Potion_Owned_Luck2:int = 0
    var Potion_Owned_Luck3:int = 0

    var Potion_Owned_Gold1:int = 0
    var Potion_Owned_Gold2:int = 0
    var Potion_Owned_Gold3:int = 0

    var Potion_Owned_Pet_Power1:int = 0
    var Potion_Owned_Pet_Power2:int = 0
    var Potion_Owned_Pet_Power3:int = 0

    var Potion_Owned_Player_Power1:int = 0
    var Potion_Owned_Player_Power2:int = 0
    var Potion_Owned_Player_Power3:int = 0

    #Item used Remaining Seconds

    var Potion_Active_Luck1_Sec:int = 0
    var Potion_Active_Luck2_Sec:int = 0
    var Potion_Active_Luck3_Sec:int = 0

    var Potion_Active_Gold1_Sec:int = 0
    var Potion_Active_Gold2_Sec:int = 0
    var Potion_Active_Gold3_Sec:int = 0

    var Potion_Active_PetPower1_Sec:int = 0
    var Potion_Active_PetPower2_Sec:int = 0
    var Potion_Active_PetPower3_Sec:int = 0

    var Potion_Active_PlayerPower1_Sec:int = 0
    var Potion_Active_PlayerPower2_Sec:int = 0
    var Potion_Active_PlayerPower3_Sec:int = 0

    #Item bonus value

    Potion_Bonus_Luck1:int = 1 # addition
    Potion_Bonus_Luck2:int = 3
    Potion_Bonus_Luck3:int = 5

    var Bonus_Luck:int = 0

    Potion_Bonus_Gold1:float = 1.05 #multiplicator
    Potion_Bonus_Gold2:float = 1.1
    Potion_Bonus_Gold3:float = 1.2

    var Bonus_Gold:float = 1.0

    var Bonus_Gold_Perma:float = 1.0

    Potion_Bonus_PetPower1:float = 1.15 #multiplicator
    Potion_Bonus_PetPower2:float = 1.3
    Potion_Bonus_PetPower3:float = 1.5

    var Bonus_Pet_Power:float = 1.0

    Potion_Bonus_PlayerPower1:float = 1.15 #multiplicator
    Potion_Bonus_PlayerPower2:float = 1.3
    Potion_Bonus_PlayerPower3:float = 1.5

    var Bonus_Player_Power:float = 1.0

    #Code input from map 1
    var InputCode0:int= 0
    var InputCode1:int= 0
    var InputCode2:int= 0
    var InputCode3:int= 0
    var InputCode4:int= 0
    var InputCode5:int= 0
    var InputCode6:int= 0
    var InputCode7:int= 0
    var InputCode8:int= 0
    var InputCode9:int= 0

    var Code1Opened:int= 0
    var Code2Opened:int= 0
    var Code3Opened:int= 0
    var Code4Opened:int= 0

    #UI Var

        #Potions Owned
        var TextBlockOwnedLuck1:text_block = text_block{}
        var TextBlockOwnedLuck2:text_block = text_block{}
        var TextBlockOwnedLuck3:text_block = text_block{}

        var TextBlockOwnedGold1:text_block = text_block{}
        var TextBlockOwnedGold2:text_block = text_block{}
        var TextBlockOwnedGold3:text_block = text_block{}

        var TextBlockOwnedPetPower1:text_block = text_block{}
        var TextBlockOwnedPetPower2:text_block = text_block{}
        var TextBlockOwnedPetPower3:text_block = text_block{}

        var TextBlockOwnedPlayerPower1:text_block = text_block{}
        var TextBlockOwnedPlayerPower2:text_block = text_block{}
        var TextBlockOwnedPlayerPower3:text_block = text_block{}

        #Potions Seconds left

        var TextBlockPotion_Active_Luck1_Sec:text_block = text_block{}
        var TextBlockPotion_Active_Luck2_Sec:text_block = text_block{}
        var TextBlockPotion_Active_Luck3_Sec:text_block = text_block{}

        var TextBlockPotion_Active_Gold1_Sec:text_block = text_block{}
        var TextBlockPotion_Active_Gold2_Sec:text_block = text_block{}
        var TextBlockPotion_Active_Gold3_Sec:text_block = text_block{}

        var TextBlockPotion_Active_PetPower1_Sec:text_block = text_block{}
        var TextBlockPotion_Active_PetPower2_Sec:text_block = text_block{}
        var TextBlockPotion_Active_PetPower3_Sec:text_block = text_block{}

        var TextBlockPotion_Active_PlayerPower1_Sec:text_block = text_block{}
        var TextBlockPotion_Active_PlayerPower2_Sec:text_block = text_block{}
        var TextBlockPotion_Active_PlayerPower3_Sec:text_block = text_block{}

        #Petdex

        var Pet1Texture:texture=HUD_PetDex.pet_locked
        var Pet2Texture:texture=HUD_PetDex.pet_locked
        var Pet3Texture:texture=HUD_PetDex.pet_locked
        var Pet4Texture:texture=HUD_PetDex.pet_locked
        var Pet5Texture:texture=HUD_PetDex.pet_locked
        var Pet6Texture:texture=HUD_PetDex.pet_locked
        var Pet7Texture:texture=HUD_PetDex.pet_locked
        var Pet8Texture:texture=HUD_PetDex.pet_locked
        var Pet9Texture:texture=HUD_PetDex.pet_locked
        var Pet10Texture:texture=HUD_PetDex.pet_locked
        var Pet11Texture:texture=HUD_PetDex.pet_locked
        var Pet12Texture:texture=HUD_PetDex.pet_locked
        var Pet13Texture:texture=HUD_PetDex.pet_locked
        var Pet14Texture:texture=HUD_PetDex.pet_locked
        var Pet15Texture:texture=HUD_PetDex.pet_locked
        var Pet16Texture:texture=HUD_PetDex.pet_locked
        var Pet17Texture:texture=HUD_PetDex.pet_locked
        var Pet18Texture:texture=HUD_PetDex.pet_locked
        var Pet19Texture:texture=HUD_PetDex.pet_locked
        var Pet20Texture:texture=HUD_PetDex.pet_locked
        var Pet21Texture:texture=HUD_PetDex.pet_locked
        var Pet22Texture:texture=HUD_PetDex.pet_locked
        var Pet23Texture:texture=HUD_PetDex.pet_locked
        var Pet24Texture:texture=HUD_PetDex.pet_locked
        var Pet25Texture:texture=HUD_PetDex.pet_locked
        var Pet26Texture:texture=HUD_PetDex.pet_locked
        var Pet27Texture:texture=HUD_PetDex.pet_locked
        var Pet28Texture:texture=HUD_PetDex.pet_locked
        var Pet29Texture:texture=HUD_PetDex.pet_locked
        var Pet30Texture:texture=HUD_PetDex.pet_locked
        var Pet31Texture:texture=HUD_PetDex.pet_locked
        var Pet32Texture:texture=HUD_PetDex.pet_locked
        var Pet33Texture:texture=HUD_PetDex.pet_locked
        var Pet34Texture:texture=HUD_PetDex.pet_locked
        var Pet35Texture:texture=HUD_PetDex.pet_locked
        var Pet36Texture:texture=HUD_PetDex.pet_locked
        var Pet37Texture:texture=HUD_PetDex.pet_locked
        var Pet38Texture:texture=HUD_PetDex.pet_locked
        var Pet39Texture:texture=HUD_PetDex.pet_locked
        var Pet40Texture:texture=HUD_PetDex.pet_locked
        var Pet41Texture:texture=HUD_PetDex.pet_locked
        var Pet42Texture:texture=HUD_PetDex.pet_locked
        var Pet43Texture:texture=HUD_PetDex.pet_locked
        var Pet44Texture:texture=HUD_PetDex.pet_locked
        var Pet45Texture:texture=HUD_PetDex.pet_locked
        var Pet46Texture:texture=HUD_PetDex.pet_locked
        var Pet47Texture:texture=HUD_PetDex.pet_locked
        var Pet48Texture:texture=HUD_PetDex.pet_locked
        var Pet49Texture:texture=HUD_PetDex.pet_locked
        var Pet50Texture:texture=HUD_PetDex.pet_locked
        var Pet51Texture:texture=HUD_PetDex.pet_locked
        var Pet52Texture:texture=HUD_PetDex.pet_locked
        var Pet53Texture:texture=HUD_PetDex.pet_locked
        var Pet54Texture:texture=HUD_PetDex.pet_locked
        var Pet55Texture:texture=HUD_PetDex.pet_locked
        var Pet56Texture:texture=HUD_PetDex.pet_locked
        var Pet57Texture:texture=HUD_PetDex.pet_locked
        var Pet58Texture:texture=HUD_PetDex.pet_locked
        var Pet59Texture:texture=HUD_PetDex.pet_locked
        var Pet60Texture:texture=HUD_PetDex.pet_locked
        var Pet61Texture:texture=HUD_PetDex.pet_locked
        var Pet62Texture:texture=HUD_PetDex.pet_locked
        var Pet63Texture:texture=HUD_PetDex.pet_locked
        var Pet64Texture:texture=HUD_PetDex.pet_locked
        var Pet65Texture:texture=HUD_PetDex.pet_locked
        var Pet66Texture:texture=HUD_PetDex.pet_locked
        var Pet67Texture:texture=HUD_PetDex.pet_locked
        var Pet68Texture:texture=HUD_PetDex.pet_locked
        var Pet69Texture:texture=HUD_PetDex.pet_locked
        var Pet70Texture:texture=HUD_PetDex.pet_locked
        var Pet71Texture:texture=HUD_PetDex.pet_locked
        var Pet72Texture:texture=HUD_PetDex.pet_locked
        var Pet73Texture:texture=HUD_PetDex.pet_locked
        var Pet74Texture:texture=HUD_PetDex.pet_locked
        var Pet75Texture:texture=HUD_PetDex.pet_locked
        var Pet76Texture:texture=HUD_PetDex.pet_locked
        var Pet77Texture:texture=HUD_PetDex.pet_locked
        var Pet78Texture:texture=HUD_PetDex.pet_locked
        var Pet79Texture:texture=HUD_PetDex.pet_locked
        var Pet80Texture:texture=HUD_PetDex.pet_locked
        var Pet81Texture:texture=HUD_PetDex.pet_locked
        var Pet82Texture:texture=HUD_PetDex.pet_locked
        var Pet83Texture:texture=HUD_PetDex.pet_locked
        var Pet84Texture:texture=HUD_PetDex.pet_locked
        var Pet85Texture:texture=HUD_PetDex.pet_locked
        var Pet86Texture:texture=HUD_PetDex.pet_locked
        var Pet87Texture:texture=HUD_PetDex.pet_locked
        var Pet88Texture:texture=HUD_PetDex.pet_locked
        var Pet89Texture:texture=HUD_PetDex.pet_locked
        var Pet90Texture:texture=HUD_PetDex.pet_locked
        var Pet91Texture:texture=HUD_PetDex.pet_locked
        var Pet92Texture:texture=HUD_PetDex.pet_locked
        var Pet93Texture:texture=HUD_PetDex.pet_locked
        var Pet94Texture:texture=HUD_PetDex.pet_locked
        var Pet95Texture:texture=HUD_PetDex.pet_locked
        var Pet96Texture:texture=HUD_PetDex.pet_locked
        var Pet97Texture:texture=HUD_PetDex.pet_locked
        var Pet98Texture:texture=HUD_PetDex.pet_locked
        var Pet99Texture:texture=HUD_PetDex.pet_locked
        var Pet100Texture:texture=HUD_PetDex.pet_locked
        var Pet101Texture:texture=HUD_PetDex.pet_locked
        var Pet102Texture:texture=HUD_PetDex.pet_locked
        var Pet103Texture:texture=HUD_PetDex.pet_locked
        var Pet104Texture:texture=HUD_PetDex.pet_locked
        var Pet105Texture:texture=HUD_PetDex.pet_locked
        var Pet106Texture:texture=HUD_PetDex.pet_locked
        var Pet107Texture:texture=HUD_PetDex.pet_locked
        var Pet108Texture:texture=HUD_PetDex.pet_locked
        var Pet109Texture:texture=HUD_PetDex.pet_locked
        var Pet110Texture:texture=HUD_PetDex.pet_locked
        var Pet111Texture:texture=HUD_PetDex.pet_locked
        var Pet112Texture:texture=HUD_PetDex.pet_locked
        var Pet113Texture:texture=HUD_PetDex.pet_locked
        var Pet114Texture:texture=HUD_PetDex.pet_locked
        var Pet115Texture:texture=HUD_PetDex.pet_locked
        var Pet116Texture:texture=HUD_PetDex.pet_locked
        var Pet117Texture:texture=HUD_PetDex.pet_locked
        var Pet118Texture:texture=HUD_PetDex.pet_locked
        var Pet119Texture:texture=HUD_PetDex.pet_locked
        var Pet120Texture:texture=HUD_PetDex.pet_locked
        var Pet121Texture:texture=HUD_PetDex.pet_locked
        var Pet122Texture:texture=HUD_PetDex.pet_locked
        var Pet123Texture:texture=HUD_PetDex.pet_locked
        var Pet124Texture:texture=HUD_PetDex.pet_locked
        var Pet125Texture:texture=HUD_PetDex.pet_locked

        var TextNumberOfPets:text_block = text_block{}

        var bShopOpen : logic = false
        var MaybeShopCanvas : ?canvas = false

    #Functions needed later
    StringToMessage<localizes>(value:string)<computes> : message = "{value}"

    FindCharIndex(SearchChar:char, Str:[]char)<transacts><decides>:int=
        var FoundIndex:int = 0
        var Found:logic = false
        for (X -> Char : Str, not Found?):
            if (Char = SearchChar):
                set FoundIndex = X
        FoundIndex
    
    ToTruncatedString(Number:float, Decimals:int):[]char=
        var Str:[]char = ToString(Number)
        if:
            DotIndex := FindCharIndex['.', Str]
            StopIndex := if (Decimals > 0) then Min(DotIndex+Decimals+1,Str.Length) else DotIndex
            Value:=Str.Slice[0, StopIndex] 
            set Str = Value
        Str

    ToStopWatchFormat<public>(F:float)<decides><transacts>:[]char=

        F2 := F*1.00000001

        Hours := Floor[F2/3600.0]
        Minutes := Mod[Floor[F2/60.0], 60]
        Seconds := Mod[Floor[F2], 60]
        var MinutesString:[]char = "{Minutes}"
        var SecondsString:[]char = "{Seconds}"
        var HoursString:[]char = "{Hours}"
        if ( Minutes < 10 ):
            set MinutesString = "0{Minutes}"
        if ( Seconds < 10 ):
            set SecondsString = "0{Seconds}"
        if ( Hours < 10 ):
            set HoursString = "0{Hours}"
        if ( Minutes = 1 ):
            set MinutesString = "00"
        return "{HoursString}:{MinutesString}:{SecondsString}"

    #Textures for pets: locked / unlocked
    EnumToTexturePet1(Ownershiptype : int)<transacts>:texture=
        case(Ownershiptype):
            0 => HUD_PetDex.pet_locked
            1 => HUD_PetDex.pet_unlocked1
            _ => HUD_PetDex.pet_locked

    EnumToTexturePet2(Ownershiptype : int)<transacts>:texture=
        case(Ownershiptype):
            0 => HUD_PetDex.pet_locked
            1 => HUD_PetDex.pet_unlocked2
            _ => HUD_PetDex.pet_locked

    EnumToTexturePet3(Ownershiptype : int)<transacts>:texture=
        case(Ownershiptype):
            0 => HUD_PetDex.pet_locked
            1 => HUD_PetDex.pet_unlocked3
            _ => HUD_PetDex.pet_locked

    EnumToTexturePet4(Ownershiptype : int)<transacts>:texture=
        case(Ownershiptype):
            0 => HUD_PetDex.pet_locked
            1 => HUD_PetDex.pet_unlocked4
            _ => HUD_PetDex.pet_locked

    EnumToTexturePet5(Ownershiptype : int)<transacts>:texture=
        case(Ownershiptype):
            0 => HUD_PetDex.pet_locked
            1 => HUD_PetDex.pet_unlocked5
            _ => HUD_PetDex.pet_locked

    EnumToTexturePet6(Ownershiptype : int)<transacts>:texture=
        case(Ownershiptype):
            0 => HUD_PetDex.pet_locked
            1 => HUD_PetDex.pet_unlocked6
            _ => HUD_PetDex.pet_locked

    EnumToTexturePet7(Ownershiptype : int)<transacts>:texture=
        case(Ownershiptype):
            0 => HUD_PetDex.pet_locked
            1 => HUD_PetDex.pet_unlocked7
            _ => HUD_PetDex.pet_locked

    EnumToTexturePet8(Ownershiptype : int)<transacts>:texture=
        case(Ownershiptype):
            0 => HUD_PetDex.pet_locked
            1 => HUD_PetDex.pet_unlocked8
            _ => HUD_PetDex.pet_locked

    EnumToTexturePet9(Ownershiptype : int)<transacts>:texture=
        case(Ownershiptype):
            0 => HUD_PetDex.pet_locked
            1 => HUD_PetDex.pet_unlocked9
            _ => HUD_PetDex.pet_locked

    EnumToTexturePet10(Ownershiptype : int)<transacts>:texture=
        case(Ownershiptype):
            0 => HUD_PetDex.pet_locked
            1 => HUD_PetDex.pet_unlocked10
            _ => HUD_PetDex.pet_locked

    EnumToTexturePet11(Ownershiptype : int)<transacts>:texture=
        case(Ownershiptype):
            0 => HUD_PetDex.pet_locked
            1 => HUD_PetDex.pet_unlocked11
            _ => HUD_PetDex.pet_locked

    EnumToTexturePet12(Ownershiptype : int)<transacts>:texture=
        case(Ownershiptype):
            0 => HUD_PetDex.pet_locked
            1 => HUD_PetDex.pet_unlocked12
            _ => HUD_PetDex.pet_locked

    EnumToTexturePet13(Ownershiptype : int)<transacts>:texture=
        case(Ownershiptype):
            0 => HUD_PetDex.pet_locked
            1 => HUD_PetDex.pet_unlocked13
            _ => HUD_PetDex.pet_locked

    EnumToTexturePet14(Ownershiptype : int)<transacts>:texture=
        case(Ownershiptype):
            0 => HUD_PetDex.pet_locked
            1 => HUD_PetDex.pet_unlocked14
            _ => HUD_PetDex.pet_locked

    EnumToTexturePet15(Ownershiptype : int)<transacts>:texture=
        case(Ownershiptype):
            0 => HUD_PetDex.pet_locked
            1 => HUD_PetDex.pet_unlocked15
            _ => HUD_PetDex.pet_locked

    EnumToTexturePet16(Ownershiptype : int)<transacts>:texture=
        case(Ownershiptype):
            0 => HUD_PetDex.pet_locked
            1 => HUD_PetDex.pet_unlocked16
            _ => HUD_PetDex.pet_locked

    EnumToTexturePet17(Ownershiptype : int)<transacts>:texture=
        case(Ownershiptype):
            0 => HUD_PetDex.pet_locked
            1 => HUD_PetDex.pet_unlocked17
            _ => HUD_PetDex.pet_locked

    EnumToTexturePet18(Ownershiptype : int)<transacts>:texture=
        case(Ownershiptype):
            0 => HUD_PetDex.pet_locked
            1 => HUD_PetDex.pet_unlocked18
            _ => HUD_PetDex.pet_locked

    EnumToTexturePet19(Ownershiptype : int)<transacts>:texture=
        case(Ownershiptype):
            0 => HUD_PetDex.pet_locked
            1 => HUD_PetDex.pet_unlocked19
            _ => HUD_PetDex.pet_locked

    EnumToTexturePet20(Ownershiptype : int)<transacts>:texture=
        case(Ownershiptype):
            0 => HUD_PetDex.pet_locked
            1 => HUD_PetDex.pet_unlocked20
            _ => HUD_PetDex.pet_locked

    EnumToTexturePet21(Ownershiptype : int)<transacts>:texture=
        case(Ownershiptype):
            0 => HUD_PetDex.pet_locked
            1 => HUD_PetDex.pet_unlocked21
            _ => HUD_PetDex.pet_locked

    EnumToTexturePet22(Ownershiptype : int)<transacts>:texture=
        case(Ownershiptype):
            0 => HUD_PetDex.pet_locked
            1 => HUD_PetDex.pet_unlocked22
            _ => HUD_PetDex.pet_locked

    EnumToTexturePet23(Ownershiptype : int)<transacts>:texture=
        case(Ownershiptype):
            0 => HUD_PetDex.pet_locked
            1 => HUD_PetDex.pet_unlocked23
            _ => HUD_PetDex.pet_locked

    EnumToTexturePet24(Ownershiptype : int)<transacts>:texture=
        case(Ownershiptype):
            0 => HUD_PetDex.pet_locked
            1 => HUD_PetDex.pet_unlocked24
            _ => HUD_PetDex.pet_locked

    EnumToTexturePet25(Ownershiptype : int)<transacts>:texture=
        case(Ownershiptype):
            0 => HUD_PetDex.pet_locked
            1 => HUD_PetDex.pet_unlocked25
            _ => HUD_PetDex.pet_locked

    EnumToTexturePet26(Ownershiptype : int)<transacts>:texture=
        case(Ownershiptype):
            0 => HUD_PetDex.pet_locked
            1 => HUD_PetDex.pet_unlocked26
            _ => HUD_PetDex.pet_locked

    EnumToTexturePet27(Ownershiptype : int)<transacts>:texture=
        case(Ownershiptype):
            0 => HUD_PetDex.pet_locked
            1 => HUD_PetDex.pet_unlocked27
            _ => HUD_PetDex.pet_locked

    EnumToTexturePet28(Ownershiptype : int)<transacts>:texture=
        case(Ownershiptype):
            0 => HUD_PetDex.pet_locked
            1 => HUD_PetDex.pet_unlocked28
            _ => HUD_PetDex.pet_locked

    EnumToTexturePet29(Ownershiptype : int)<transacts>:texture=
        case(Ownershiptype):
            0 => HUD_PetDex.pet_locked
            1 => HUD_PetDex.pet_unlocked29
            _ => HUD_PetDex.pet_locked

    EnumToTexturePet30(Ownershiptype : int)<transacts>:texture=
        case(Ownershiptype):
            0 => HUD_PetDex.pet_locked
            1 => HUD_PetDex.pet_unlocked30
            _ => HUD_PetDex.pet_locked

    EnumToTexturePet31(Ownershiptype : int)<transacts>:texture=
        case(Ownershiptype):
            0 => HUD_PetDex.pet_locked
            1 => HUD_PetDex.pet_unlocked31
            _ => HUD_PetDex.pet_locked

    EnumToTexturePet32(Ownershiptype : int)<transacts>:texture=
        case(Ownershiptype):
            0 => HUD_PetDex.pet_locked
            1 => HUD_PetDex.pet_unlocked32
            _ => HUD_PetDex.pet_locked

    EnumToTexturePet33(Ownershiptype : int)<transacts>:texture=
        case(Ownershiptype):
            0 => HUD_PetDex.pet_locked
            1 => HUD_PetDex.pet_unlocked33
            _ => HUD_PetDex.pet_locked

    EnumToTexturePet34(Ownershiptype : int)<transacts>:texture=
        case(Ownershiptype):
            0 => HUD_PetDex.pet_locked
            1 => HUD_PetDex.pet_unlocked34
            _ => HUD_PetDex.pet_locked

    EnumToTexturePet35(Ownershiptype : int)<transacts>:texture=
        case(Ownershiptype):
            0 => HUD_PetDex.pet_locked
            1 => HUD_PetDex.pet_unlocked35
            _ => HUD_PetDex.pet_locked

    EnumToTexturePet36(Ownershiptype : int)<transacts>:texture=
        case(Ownershiptype):
            0 => HUD_PetDex.pet_locked
            1 => HUD_PetDex.pet_unlocked36
            _ => HUD_PetDex.pet_locked

    EnumToTexturePet37(Ownershiptype : int)<transacts>:texture=
        case(Ownershiptype):
            0 => HUD_PetDex.pet_locked
            1 => HUD_PetDex.pet_unlocked37
            _ => HUD_PetDex.pet_locked

    EnumToTexturePet38(Ownershiptype : int)<transacts>:texture=
        case(Ownershiptype):
            0 => HUD_PetDex.pet_locked
            1 => HUD_PetDex.pet_unlocked38
            _ => HUD_PetDex.pet_locked

    EnumToTexturePet39(Ownershiptype : int)<transacts>:texture=
        case(Ownershiptype):
            0 => HUD_PetDex.pet_locked
            1 => HUD_PetDex.pet_unlocked39
            _ => HUD_PetDex.pet_locked

    EnumToTexturePet40(Ownershiptype : int)<transacts>:texture=
        case(Ownershiptype):
            0 => HUD_PetDex.pet_locked
            1 => HUD_PetDex.pet_unlocked40
            _ => HUD_PetDex.pet_locked

    EnumToTexturePet41(Ownershiptype : int)<transacts>:texture=
        case(Ownershiptype):
            0 => HUD_PetDex.pet_locked
            1 => HUD_PetDex.pet_unlocked41
            _ => HUD_PetDex.pet_locked

    EnumToTexturePet42(Ownershiptype : int)<transacts>:texture=
        case(Ownershiptype):
            0 => HUD_PetDex.pet_locked
            1 => HUD_PetDex.pet_unlocked42
            _ => HUD_PetDex.pet_locked

    EnumToTexturePet43(Ownershiptype : int)<transacts>:texture=
        case(Ownershiptype):
            0 => HUD_PetDex.pet_locked
            1 => HUD_PetDex.pet_unlocked43
            _ => HUD_PetDex.pet_locked

    EnumToTexturePet44(Ownershiptype : int)<transacts>:texture=
        case(Ownershiptype):
            0 => HUD_PetDex.pet_locked
            1 => HUD_PetDex.pet_unlocked44
            _ => HUD_PetDex.pet_locked

    EnumToTexturePet45(Ownershiptype : int)<transacts>:texture=
        case(Ownershiptype):
            0 => HUD_PetDex.pet_locked
            1 => HUD_PetDex.pet_unlocked45
            _ => HUD_PetDex.pet_locked

    EnumToTexturePet46(Ownershiptype : int)<transacts>:texture=
        case(Ownershiptype):
            0 => HUD_PetDex.pet_locked
            1 => HUD_PetDex.pet_unlocked46
            _ => HUD_PetDex.pet_locked

    EnumToTexturePet47(Ownershiptype : int)<transacts>:texture=
        case(Ownershiptype):
            0 => HUD_PetDex.pet_locked
            1 => HUD_PetDex.pet_unlocked47
            _ => HUD_PetDex.pet_locked

    EnumToTexturePet48(Ownershiptype : int)<transacts>:texture=
        case(Ownershiptype):
            0 => HUD_PetDex.pet_locked
            1 => HUD_PetDex.pet_unlocked48
            _ => HUD_PetDex.pet_locked

    EnumToTexturePet49(Ownershiptype : int)<transacts>:texture=
        case(Ownershiptype):
            0 => HUD_PetDex.pet_locked
            1 => HUD_PetDex.pet_unlocked49
            _ => HUD_PetDex.pet_locked

    EnumToTexturePet50(Ownershiptype : int)<transacts>:texture=
        case(Ownershiptype):
            0 => HUD_PetDex.pet_locked
            1 => HUD_PetDex.pet_unlocked50
            _ => HUD_PetDex.pet_locked

    EnumToTexturePet51(Ownershiptype : int)<transacts>:texture=
        case(Ownershiptype):
            0 => HUD_PetDex.pet_locked
            1 => HUD_PetDex.pet_unlocked51
            _ => HUD_PetDex.pet_locked

    EnumToTexturePet52(Ownershiptype : int)<transacts>:texture=
        case(Ownershiptype):
            0 => HUD_PetDex.pet_locked
            1 => HUD_PetDex.pet_unlocked52
            _ => HUD_PetDex.pet_locked

    EnumToTexturePet53(Ownershiptype : int)<transacts>:texture=
        case(Ownershiptype):
            0 => HUD_PetDex.pet_locked
            1 => HUD_PetDex.pet_unlocked53
            _ => HUD_PetDex.pet_locked

    EnumToTexturePet54(Ownershiptype : int)<transacts>:texture=
        case(Ownershiptype):
            0 => HUD_PetDex.pet_locked
            1 => HUD_PetDex.pet_unlocked54
            _ => HUD_PetDex.pet_locked

    EnumToTexturePet55(Ownershiptype : int)<transacts>:texture=
        case(Ownershiptype):
            0 => HUD_PetDex.pet_locked
            1 => HUD_PetDex.pet_unlocked55
            _ => HUD_PetDex.pet_locked

    EnumToTexturePet56(Ownershiptype : int)<transacts>:texture=
        case(Ownershiptype):
            0 => HUD_PetDex.pet_locked
            1 => HUD_PetDex.pet_unlocked56
            _ => HUD_PetDex.pet_locked

    EnumToTexturePet57(Ownershiptype : int)<transacts>:texture=
        case(Ownershiptype):
            0 => HUD_PetDex.pet_locked
            1 => HUD_PetDex.pet_unlocked57
            _ => HUD_PetDex.pet_locked

    EnumToTexturePet58(Ownershiptype : int)<transacts>:texture=
        case(Ownershiptype):
            0 => HUD_PetDex.pet_locked
            1 => HUD_PetDex.pet_unlocked58
            _ => HUD_PetDex.pet_locked

    EnumToTexturePet59(Ownershiptype : int)<transacts>:texture=
        case(Ownershiptype):
            0 => HUD_PetDex.pet_locked
            1 => HUD_PetDex.pet_unlocked59
            _ => HUD_PetDex.pet_locked

    EnumToTexturePet60(Ownershiptype : int)<transacts>:texture=
        case(Ownershiptype):
            0 => HUD_PetDex.pet_locked
            1 => HUD_PetDex.pet_unlocked60
            _ => HUD_PetDex.pet_locked

    EnumToTexturePet61(Ownershiptype : int)<transacts>:texture=
        case(Ownershiptype):
            0 => HUD_PetDex.pet_locked
            1 => HUD_PetDex.pet_unlocked61
            _ => HUD_PetDex.pet_locked

    EnumToTexturePet62(Ownershiptype : int)<transacts>:texture=
        case(Ownershiptype):
            0 => HUD_PetDex.pet_locked
            1 => HUD_PetDex.pet_unlocked62
            _ => HUD_PetDex.pet_locked

    EnumToTexturePet63(Ownershiptype : int)<transacts>:texture=
        case(Ownershiptype):
            0 => HUD_PetDex.pet_locked
            1 => HUD_PetDex.pet_unlocked63
            _ => HUD_PetDex.pet_locked

    EnumToTexturePet64(Ownershiptype : int)<transacts>:texture=
        case(Ownershiptype):
            0 => HUD_PetDex.pet_locked
            1 => HUD_PetDex.pet_unlocked64
            _ => HUD_PetDex.pet_locked

    EnumToTexturePet65(Ownershiptype : int)<transacts>:texture=
        case(Ownershiptype):
            0 => HUD_PetDex.pet_locked
            1 => HUD_PetDex.pet_unlocked65
            _ => HUD_PetDex.pet_locked

    EnumToTexturePet66(Ownershiptype : int)<transacts>:texture=
        case(Ownershiptype):
            0 => HUD_PetDex.pet_locked
            1 => HUD_PetDex.pet_unlocked66
            _ => HUD_PetDex.pet_locked

    EnumToTexturePet67(Ownershiptype : int)<transacts>:texture=
        case(Ownershiptype):
            0 => HUD_PetDex.pet_locked
            1 => HUD_PetDex.pet_unlocked67
            _ => HUD_PetDex.pet_locked

    EnumToTexturePet68(Ownershiptype : int)<transacts>:texture=
        case(Ownershiptype):
            0 => HUD_PetDex.pet_locked
            1 => HUD_PetDex.pet_unlocked68
            _ => HUD_PetDex.pet_locked

    EnumToTexturePet69(Ownershiptype : int)<transacts>:texture=
        case(Ownershiptype):
            0 => HUD_PetDex.pet_locked
            1 => HUD_PetDex.pet_unlocked69
            _ => HUD_PetDex.pet_locked

    EnumToTexturePet70(Ownershiptype : int)<transacts>:texture=
        case(Ownershiptype):
            0 => HUD_PetDex.pet_locked
            1 => HUD_PetDex.pet_unlocked70
            _ => HUD_PetDex.pet_locked

    EnumToTexturePet71(Ownershiptype : int)<transacts>:texture=
        case(Ownershiptype):
            0 => HUD_PetDex.pet_locked
            1 => HUD_PetDex.pet_unlocked71
            _ => HUD_PetDex.pet_locked

    EnumToTexturePet72(Ownershiptype : int)<transacts>:texture=
        case(Ownershiptype):
            0 => HUD_PetDex.pet_locked
            1 => HUD_PetDex.pet_unlocked72
            _ => HUD_PetDex.pet_locked

    EnumToTexturePet73(Ownershiptype : int)<transacts>:texture=
        case(Ownershiptype):
            0 => HUD_PetDex.pet_locked
            1 => HUD_PetDex.pet_unlocked73
            _ => HUD_PetDex.pet_locked

    EnumToTexturePet74(Ownershiptype : int)<transacts>:texture=
        case(Ownershiptype):
            0 => HUD_PetDex.pet_locked
            1 => HUD_PetDex.pet_unlocked74
            _ => HUD_PetDex.pet_locked

    EnumToTexturePet75(Ownershiptype : int)<transacts>:texture=
        case(Ownershiptype):
            0 => HUD_PetDex.pet_locked
            1 => HUD_PetDex.pet_unlocked75
            _ => HUD_PetDex.pet_locked

    EnumToTexturePet76(Ownershiptype : int)<transacts>:texture=
        case(Ownershiptype):
            0 => HUD_PetDex.pet_locked
            1 => HUD_PetDex.pet_unlocked76
            _ => HUD_PetDex.pet_locked

    EnumToTexturePet77(Ownershiptype : int)<transacts>:texture=
        case(Ownershiptype):
            0 => HUD_PetDex.pet_locked
            1 => HUD_PetDex.pet_unlocked77
            _ => HUD_PetDex.pet_locked

    EnumToTexturePet78(Ownershiptype : int)<transacts>:texture=
        case(Ownershiptype):
            0 => HUD_PetDex.pet_locked
            1 => HUD_PetDex.pet_unlocked78
            _ => HUD_PetDex.pet_locked

    EnumToTexturePet79(Ownershiptype : int)<transacts>:texture=
        case(Ownershiptype):
            0 => HUD_PetDex.pet_locked
            1 => HUD_PetDex.pet_unlocked79
            _ => HUD_PetDex.pet_locked

    EnumToTexturePet80(Ownershiptype : int)<transacts>:texture=
        case(Ownershiptype):
            0 => HUD_PetDex.pet_locked
            1 => HUD_PetDex.pet_unlocked80
            _ => HUD_PetDex.pet_locked

    EnumToTexturePet81(Ownershiptype : int)<transacts>:texture=
        case(Ownershiptype):
            0 => HUD_PetDex.pet_locked
            1 => HUD_PetDex.pet_unlocked81
            _ => HUD_PetDex.pet_locked

    EnumToTexturePet82(Ownershiptype : int)<transacts>:texture=
        case(Ownershiptype):
            0 => HUD_PetDex.pet_locked
            1 => HUD_PetDex.pet_unlocked82
            _ => HUD_PetDex.pet_locked

    EnumToTexturePet83(Ownershiptype : int)<transacts>:texture=
        case(Ownershiptype):
            0 => HUD_PetDex.pet_locked
            1 => HUD_PetDex.pet_unlocked83
            _ => HUD_PetDex.pet_locked

    EnumToTexturePet84(Ownershiptype : int)<transacts>:texture=
        case(Ownershiptype):
            0 => HUD_PetDex.pet_locked
            1 => HUD_PetDex.pet_unlocked84
            _ => HUD_PetDex.pet_locked

    EnumToTexturePet85(Ownershiptype : int)<transacts>:texture=
        case(Ownershiptype):
            0 => HUD_PetDex.pet_locked
            1 => HUD_PetDex.pet_unlocked85
            _ => HUD_PetDex.pet_locked

    EnumToTexturePet86(Ownershiptype : int)<transacts>:texture=
        case(Ownershiptype):
            0 => HUD_PetDex.pet_locked
            1 => HUD_PetDex.pet_unlocked86
            _ => HUD_PetDex.pet_locked

    EnumToTexturePet87(Ownershiptype : int)<transacts>:texture=
        case(Ownershiptype):
            0 => HUD_PetDex.pet_locked
            1 => HUD_PetDex.pet_unlocked87
            _ => HUD_PetDex.pet_locked

    EnumToTexturePet88(Ownershiptype : int)<transacts>:texture=
        case(Ownershiptype):
            0 => HUD_PetDex.pet_locked
            1 => HUD_PetDex.pet_unlocked88
            _ => HUD_PetDex.pet_locked

    EnumToTexturePet89(Ownershiptype : int)<transacts>:texture=
        case(Ownershiptype):
            0 => HUD_PetDex.pet_locked
            1 => HUD_PetDex.pet_unlocked89
            _ => HUD_PetDex.pet_locked

    EnumToTexturePet90(Ownershiptype : int)<transacts>:texture=
        case(Ownershiptype):
            0 => HUD_PetDex.pet_locked
            1 => HUD_PetDex.pet_unlocked90
            _ => HUD_PetDex.pet_locked

    EnumToTexturePet91(Ownershiptype : int)<transacts>:texture=
        case(Ownershiptype):
            0 => HUD_PetDex.pet_locked
            1 => HUD_PetDex.pet_unlocked91
            _ => HUD_PetDex.pet_locked

    EnumToTexturePet92(Ownershiptype : int)<transacts>:texture=
        case(Ownershiptype):
            0 => HUD_PetDex.pet_locked
            1 => HUD_PetDex.pet_unlocked92
            _ => HUD_PetDex.pet_locked

    EnumToTexturePet93(Ownershiptype : int)<transacts>:texture=
        case(Ownershiptype):
            0 => HUD_PetDex.pet_locked
            1 => HUD_PetDex.pet_unlocked93
            _ => HUD_PetDex.pet_locked

    EnumToTexturePet94(Ownershiptype : int)<transacts>:texture=
        case(Ownershiptype):
            0 => HUD_PetDex.pet_locked
            1 => HUD_PetDex.pet_unlocked94
            _ => HUD_PetDex.pet_locked

    EnumToTexturePet95(Ownershiptype : int)<transacts>:texture=
        case(Ownershiptype):
            0 => HUD_PetDex.pet_locked
            1 => HUD_PetDex.pet_unlocked95
            _ => HUD_PetDex.pet_locked

    EnumToTexturePet96(Ownershiptype : int)<transacts>:texture=
        case(Ownershiptype):
            0 => HUD_PetDex.pet_locked
            1 => HUD_PetDex.pet_unlocked96
            _ => HUD_PetDex.pet_locked

    EnumToTexturePet97(Ownershiptype : int)<transacts>:texture=
        case(Ownershiptype):
            0 => HUD_PetDex.pet_locked
            1 => HUD_PetDex.pet_unlocked97
            _ => HUD_PetDex.pet_locked

    EnumToTexturePet98(Ownershiptype : int)<transacts>:texture=
        case(Ownershiptype):
            0 => HUD_PetDex.pet_locked
            1 => HUD_PetDex.pet_unlocked98
            _ => HUD_PetDex.pet_locked

    EnumToTexturePet99(Ownershiptype : int)<transacts>:texture=
        case(Ownershiptype):
            0 => HUD_PetDex.pet_locked
            1 => HUD_PetDex.pet_unlocked99
            _ => HUD_PetDex.pet_locked

    EnumToTexturePet100(Ownershiptype : int)<transacts>:texture=
        case(Ownershiptype):
            0 => HUD_PetDex.pet_locked
            1 => HUD_PetDex.pet_unlocked100
            _ => HUD_PetDex.pet_locked

    EnumToTexturePet101(Ownershiptype : int)<transacts>:texture=
        case(Ownershiptype):
            0 => HUD_PetDex.pet_locked
            1 => HUD_PetDex.pet_unlocked101
            _ => HUD_PetDex.pet_locked

    EnumToTexturePet102(Ownershiptype : int)<transacts>:texture=
        case(Ownershiptype):
            0 => HUD_PetDex.pet_locked
            1 => HUD_PetDex.pet_unlocked102
            _ => HUD_PetDex.pet_locked

    EnumToTexturePet103(Ownershiptype : int)<transacts>:texture=
        case(Ownershiptype):
            0 => HUD_PetDex.pet_locked
            1 => HUD_PetDex.pet_unlocked103
            _ => HUD_PetDex.pet_locked

    EnumToTexturePet104(Ownershiptype : int)<transacts>:texture=
        case(Ownershiptype):
            0 => HUD_PetDex.pet_locked
            1 => HUD_PetDex.pet_unlocked104
            _ => HUD_PetDex.pet_locked

    EnumToTexturePet105(Ownershiptype : int)<transacts>:texture=
        case(Ownershiptype):
            0 => HUD_PetDex.pet_locked
            1 => HUD_PetDex.pet_unlocked105
            _ => HUD_PetDex.pet_locked

    EnumToTexturePet106(Ownershiptype : int)<transacts>:texture=
        case(Ownershiptype):
            0 => HUD_PetDex.pet_locked
            1 => HUD_PetDex.pet_unlocked106
            _ => HUD_PetDex.pet_locked

    EnumToTexturePet107(Ownershiptype : int)<transacts>:texture=
        case(Ownershiptype):
            0 => HUD_PetDex.pet_locked
            1 => HUD_PetDex.pet_unlocked107
            _ => HUD_PetDex.pet_locked

    EnumToTexturePet108(Ownershiptype : int)<transacts>:texture=
        case(Ownershiptype):
            0 => HUD_PetDex.pet_locked
            1 => HUD_PetDex.pet_unlocked108
            _ => HUD_PetDex.pet_locked

    EnumToTexturePet109(Ownershiptype : int)<transacts>:texture=
        case(Ownershiptype):
            0 => HUD_PetDex.pet_locked
            1 => HUD_PetDex.pet_unlocked109
            _ => HUD_PetDex.pet_locked

    EnumToTexturePet110(Ownershiptype : int)<transacts>:texture=
        case(Ownershiptype):
            0 => HUD_PetDex.pet_locked
            1 => HUD_PetDex.pet_unlocked110
            _ => HUD_PetDex.pet_locked

    EnumToTexturePet111(Ownershiptype : int)<transacts>:texture=
        case(Ownershiptype):
            0 => HUD_PetDex.pet_locked
            1 => HUD_PetDex.pet_unlocked111
            _ => HUD_PetDex.pet_locked

    EnumToTexturePet112(Ownershiptype : int)<transacts>:texture=
        case(Ownershiptype):
            0 => HUD_PetDex.pet_locked
            1 => HUD_PetDex.pet_unlocked112
            _ => HUD_PetDex.pet_locked

    EnumToTexturePet113(Ownershiptype : int)<transacts>:texture=
        case(Ownershiptype):
            0 => HUD_PetDex.pet_locked
            1 => HUD_PetDex.pet_unlocked113
            _ => HUD_PetDex.pet_locked

    EnumToTexturePet114(Ownershiptype : int)<transacts>:texture=
        case(Ownershiptype):
            0 => HUD_PetDex.pet_locked
            1 => HUD_PetDex.pet_unlocked114
            _ => HUD_PetDex.pet_locked

    EnumToTexturePet115(Ownershiptype : int)<transacts>:texture=
        case(Ownershiptype):
            0 => HUD_PetDex.pet_locked
            1 => HUD_PetDex.pet_unlocked115
            _ => HUD_PetDex.pet_locked

    EnumToTexturePet116(Ownershiptype : int)<transacts>:texture=
        case(Ownershiptype):
            0 => HUD_PetDex.pet_locked
            1 => HUD_PetDex.pet_unlocked116
            _ => HUD_PetDex.pet_locked

    EnumToTexturePet117(Ownershiptype : int)<transacts>:texture=
        case(Ownershiptype):
            0 => HUD_PetDex.pet_locked
            1 => HUD_PetDex.pet_unlocked117
            _ => HUD_PetDex.pet_locked

    EnumToTexturePet118(Ownershiptype : int)<transacts>:texture=
        case(Ownershiptype):
            0 => HUD_PetDex.pet_locked
            1 => HUD_PetDex.pet_unlocked118
            _ => HUD_PetDex.pet_locked

    EnumToTexturePet119(Ownershiptype : int)<transacts>:texture=
        case(Ownershiptype):
            0 => HUD_PetDex.pet_locked
            1 => HUD_PetDex.pet_unlocked119
            _ => HUD_PetDex.pet_locked

    EnumToTexturePet120(Ownershiptype : int)<transacts>:texture=
        case(Ownershiptype):
            0 => HUD_PetDex.empty_image
            1 => HUD_PetDex.pet_unlocked120
            _ => HUD_PetDex.empty_image

    EnumToTexturePet121(Ownershiptype : int)<transacts>:texture=
        case(Ownershiptype):
            0 => HUD_PetDex.empty_image
            1 => HUD_PetDex.pet_unlocked121
            _ => HUD_PetDex.empty_image

    EnumToTexturePet122(Ownershiptype : int)<transacts>:texture=
        case(Ownershiptype):
            0 => HUD_PetDex.empty_image
            1 => HUD_PetDex.pet_unlocked122
            _ => HUD_PetDex.empty_image

    EnumToTexturePet123(Ownershiptype : int)<transacts>:texture=
        case(Ownershiptype):
            0 => HUD_PetDex.empty_image
            1 => HUD_PetDex.pet_unlocked123
            _ => HUD_PetDex.empty_image

    EnumToTexturePet124(Ownershiptype : int)<transacts>:texture=
        case(Ownershiptype):
            0 => HUD_PetDex.empty_image
            1 => HUD_PetDex.pet_unlocked124
            _ => HUD_PetDex.empty_image

    EnumToTexturePet125(Ownershiptype : int)<transacts>:texture=
        case(Ownershiptype):
            0 => HUD_PetDex.empty_image
            1 => HUD_PetDex.pet_unlocked125
            _ => HUD_PetDex.empty_image

    InitiateUI():void= #STATS + INVENTORY

        #Stats related
        set TextBlockPower = text_block{DefaultText := StringToMessage("{ToTruncatedString(Pet_Power_Value,0)}"),DefaultTextColor:=NamedColors.Black}
        TextBlockPower.SetShadowOpacity(1.0)

        set TextBlockCrit = text_block{DefaultText := StringToMessage("+{Pet_Crit_Value}%"),DefaultTextColor:=NamedColors.Black}
        TextBlockCrit.SetShadowOpacity(1.0)

        set TextBlockSpeed = text_block{DefaultText := StringToMessage("x{ToTruncatedString(Pet_Speed_Value,1)}"),DefaultTextColor:=NamedColors.Black}
        TextBlockSpeed.SetShadowOpacity(1.0)

        set TextBlockPlayerSpeed = text_block{DefaultText := StringToMessage("x{ToTruncatedString(Player_Speed_Value,1)}"),DefaultTextColor:=NamedColors.Black}
        TextBlockPlayerSpeed.SetShadowOpacity(1.0)

        set TextBlockLuck = text_block{DefaultText := StringToMessage("+{Pet_Luck_Value_Final}%"),DefaultTextColor:=NamedColors.Black}
        TextBlockLuck.SetShadowOpacity(1.0)
        
        set TextBlockGold = text_block{DefaultText := StringToMessage("{PlayerGold}"),DefaultTextColor:=NamedColors.Black}
        TextBlockGold.SetShadowOpacity(1.0)

        set TextBlockDiamond = text_block{DefaultText := StringToMessage("{PlayerDiamond}"),DefaultTextColor:=NamedColors.Black}
        TextBlockDiamond.SetShadowOpacity(1.0)

        #Inventory related

        #Pet power
        set TextBlockOwnedPetPower1 = text_block{DefaultText := StringToMessage("{Potion_Owned_Pet_Power1}"),DefaultTextColor:=NamedColors.White,DefaultShadowColor:=Black,DefaultShadowOpacity:=1.0,DefaultShadowOffset := option{ vector2{ X := 2.0, Y := 2.0 }}}
        set TextBlockOwnedPetPower2 = text_block{DefaultText := StringToMessage("{Potion_Owned_Pet_Power2}"),DefaultTextColor:=NamedColors.White,DefaultShadowColor:=Black,DefaultShadowOpacity:=1.0,DefaultShadowOffset := option{ vector2{ X := 2.0, Y := 2.0 }}}
        set TextBlockOwnedPetPower3 = text_block{DefaultText := StringToMessage("{Potion_Owned_Pet_Power3}"),DefaultTextColor:=NamedColors.White,DefaultShadowColor:=Black,DefaultShadowOpacity:=1.0,DefaultShadowOffset := option{ vector2{ X := 2.0, Y := 2.0 }}}

        set TextBlockPotion_Active_PetPower1_Sec = text_block{DefaultText := StringToMessage("{Potion_Active_PetPower1_Sec}"),DefaultTextColor:=NamedColors.White,DefaultShadowColor:=Black,DefaultShadowOpacity:=1.0,DefaultShadowOffset := option{ vector2{ X := 2.0, Y := 2.0 }}}
        set TextBlockPotion_Active_PetPower2_Sec = text_block{DefaultText := StringToMessage("{Potion_Active_PetPower2_Sec}"),DefaultTextColor:=NamedColors.White,DefaultShadowColor:=Black,DefaultShadowOpacity:=1.0,DefaultShadowOffset := option{ vector2{ X := 2.0, Y := 2.0 }}}
        set TextBlockPotion_Active_PetPower3_Sec = text_block{DefaultText := StringToMessage("{Potion_Active_PetPower3_Sec}"),DefaultTextColor:=NamedColors.White,DefaultShadowColor:=Black,DefaultShadowOpacity:=1.0,DefaultShadowOffset := option{ vector2{ X := 2.0, Y := 2.0 }}}

        #Player power
        set TextBlockOwnedPlayerPower1 = text_block{DefaultText := StringToMessage("{Potion_Owned_Player_Power1}"),DefaultTextColor:=NamedColors.White,DefaultShadowColor:=Black,DefaultShadowOpacity:=1.0,DefaultShadowOffset := option{ vector2{ X := 2.0, Y := 2.0 }}}
        set TextBlockOwnedPlayerPower2 = text_block{DefaultText := StringToMessage("{Potion_Owned_Player_Power2}"),DefaultTextColor:=NamedColors.White,DefaultShadowColor:=Black,DefaultShadowOpacity:=1.0,DefaultShadowOffset := option{ vector2{ X := 2.0, Y := 2.0 }}}
        set TextBlockOwnedPlayerPower3 = text_block{DefaultText := StringToMessage("{Potion_Owned_Player_Power3}"),DefaultTextColor:=NamedColors.White,DefaultShadowColor:=Black,DefaultShadowOpacity:=1.0,DefaultShadowOffset := option{ vector2{ X := 2.0, Y := 2.0 }}}

        set TextBlockPotion_Active_PlayerPower1_Sec = text_block{DefaultText := StringToMessage("{Potion_Active_PlayerPower1_Sec}"),DefaultTextColor:=NamedColors.White,DefaultShadowColor:=Black,DefaultShadowOpacity:=1.0,DefaultShadowOffset := option{ vector2{ X := 2.0, Y := 2.0 }}}
        set TextBlockPotion_Active_PlayerPower2_Sec = text_block{DefaultText := StringToMessage("{Potion_Active_PlayerPower2_Sec}"),DefaultTextColor:=NamedColors.White,DefaultShadowColor:=Black,DefaultShadowOpacity:=1.0,DefaultShadowOffset := option{ vector2{ X := 2.0, Y := 2.0 }}}
        set TextBlockPotion_Active_PlayerPower3_Sec = text_block{DefaultText := StringToMessage("{Potion_Active_PlayerPower3_Sec}"),DefaultTextColor:=NamedColors.White,DefaultShadowColor:=Black,DefaultShadowOpacity:=1.0,DefaultShadowOffset := option{ vector2{ X := 2.0, Y := 2.0 }}}

        #Gold
        set TextBlockOwnedGold1 = text_block{DefaultText := StringToMessage("{Potion_Owned_Gold1}"),DefaultTextColor:=NamedColors.White,DefaultShadowColor:=Black,DefaultShadowOpacity:=1.0,DefaultShadowOffset := option{ vector2{ X := 2.0, Y := 2.0 }}}
        set TextBlockOwnedGold2 = text_block{DefaultText := StringToMessage("{Potion_Owned_Gold2}"),DefaultTextColor:=NamedColors.White,DefaultShadowColor:=Black,DefaultShadowOpacity:=1.0,DefaultShadowOffset := option{ vector2{ X := 2.0, Y := 2.0 }}}
        set TextBlockOwnedGold3 = text_block{DefaultText := StringToMessage("{Potion_Owned_Gold3}"),DefaultTextColor:=NamedColors.White,DefaultShadowColor:=Black,DefaultShadowOpacity:=1.0,DefaultShadowOffset := option{ vector2{ X := 2.0, Y := 2.0 }}}

        set TextBlockPotion_Active_Gold1_Sec = text_block{DefaultText := StringToMessage("{Potion_Active_Gold1_Sec}"),DefaultTextColor:=NamedColors.White,DefaultShadowColor:=Black,DefaultShadowOpacity:=1.0,DefaultShadowOffset := option{ vector2{ X := 2.0, Y := 2.0 }}}
        set TextBlockPotion_Active_Gold2_Sec = text_block{DefaultText := StringToMessage("{Potion_Active_Gold2_Sec}"),DefaultTextColor:=NamedColors.White,DefaultShadowColor:=Black,DefaultShadowOpacity:=1.0,DefaultShadowOffset := option{ vector2{ X := 2.0, Y := 2.0 }}}
        set TextBlockPotion_Active_Gold3_Sec = text_block{DefaultText := StringToMessage("{Potion_Active_Gold3_Sec}"),DefaultTextColor:=NamedColors.White,DefaultShadowColor:=Black,DefaultShadowOpacity:=1.0,DefaultShadowOffset := option{ vector2{ X := 2.0, Y := 2.0 }}}

        #Luck
        set TextBlockOwnedLuck1 = text_block{DefaultText := StringToMessage("{Potion_Owned_Luck1}"),DefaultTextColor:=NamedColors.White,DefaultShadowColor:=Black,DefaultShadowOpacity:=1.0,DefaultShadowOffset := option{ vector2{ X := 2.0, Y := 2.0 }}}
        set TextBlockOwnedLuck2 = text_block{DefaultText := StringToMessage("{Potion_Owned_Luck2}"),DefaultTextColor:=NamedColors.White,DefaultShadowColor:=Black,DefaultShadowOpacity:=1.0,DefaultShadowOffset := option{ vector2{ X := 2.0, Y := 2.0 }}}
        set TextBlockOwnedLuck3 = text_block{DefaultText := StringToMessage("{Potion_Owned_Luck3}"),DefaultTextColor:=NamedColors.White,DefaultShadowColor:=Black,DefaultShadowOpacity:=1.0,DefaultShadowOffset := option{ vector2{ X := 2.0, Y := 2.0 }}}

        set TextBlockPotion_Active_Luck1_Sec = text_block{DefaultText := StringToMessage("{Potion_Active_Luck1_Sec}"),DefaultTextColor:=NamedColors.White,DefaultShadowColor:=Black,DefaultShadowOpacity:=1.0,DefaultShadowOffset := option{ vector2{ X := 2.0, Y := 2.0 }}}
        set TextBlockPotion_Active_Luck2_Sec = text_block{DefaultText := StringToMessage("{Potion_Active_Luck2_Sec}"),DefaultTextColor:=NamedColors.White,DefaultShadowColor:=Black,DefaultShadowOpacity:=1.0,DefaultShadowOffset := option{ vector2{ X := 2.0, Y := 2.0 }}}
        set TextBlockPotion_Active_Luck3_Sec = text_block{DefaultText := StringToMessage("{Potion_Active_Luck3_Sec}"),DefaultTextColor:=NamedColors.White,DefaultShadowColor:=Black,DefaultShadowOpacity:=1.0,DefaultShadowOffset := option{ vector2{ X := 2.0, Y := 2.0 }}}

        PlayerScoreWidget:canvas = CreateUserUI()
        PlayerScoreWidget2:canvas = CreateUserUI2()
        if(PlayerObj:=player[MyAgentObj], PlayerUI := GetPlayerUI[PlayerObj]):
            PlayerUI.AddWidget(PlayerScoreWidget)
            PlayerUI.AddWidget(PlayerScoreWidget2)

    Remove_UI():void= #STATS + INVENTORY

        PlayerScoreWidget:canvas = CreateUserUI()
        PlayerScoreWidget2:canvas = CreateUserUI2()
        if(PlayerObj:=player[MyAgentObj], PlayerUI := GetPlayerUI[PlayerObj]):
            PlayerUI.RemoveWidget(PlayerScoreWidget)
            PlayerUI.RemoveWidget(PlayerScoreWidget2)

    InitiateUI_PetDex():void=
        #L1
        if(Ownership_L1_Pet4=1):
            set Pet1Texture = EnumToTexturePet1(1)
            set Pet2Texture = EnumToTexturePet2(1)
            set Pet3Texture = EnumToTexturePet3(1)
            set Pet4Texture = EnumToTexturePet4(1)
        else if(Ownership_L1_Pet3=1):
            set Pet1Texture = EnumToTexturePet1(1)
            set Pet2Texture = EnumToTexturePet2(1)
            set Pet3Texture = EnumToTexturePet3(1)
        else if(Ownership_L1_Pet2=1):
            set Pet1Texture = EnumToTexturePet1(1)
            set Pet2Texture = EnumToTexturePet2(1)
        else if(Ownership_L1_Pet1=1):
            set Pet1Texture = EnumToTexturePet1(1)
        #L2
        if(Ownership_L2_Pet4=1):
            set Pet5Texture = EnumToTexturePet5(1)
            set Pet6Texture = EnumToTexturePet6(1)
            set Pet7Texture = EnumToTexturePet7(1)
            set Pet8Texture = EnumToTexturePet8(1)
        else if(Ownership_L2_Pet3=1):
            set Pet5Texture = EnumToTexturePet5(1)
            set Pet6Texture = EnumToTexturePet6(1)
            set Pet7Texture = EnumToTexturePet7(1)
        else if(Ownership_L2_Pet2=1):
            set Pet5Texture = EnumToTexturePet5(1)
            set Pet6Texture = EnumToTexturePet6(1)
        else if(Ownership_L2_Pet1=1):
            set Pet5Texture = EnumToTexturePet5(1)
        #L3
        if(Ownership_L3_Pet4=1):
            set Pet9Texture = EnumToTexturePet9(1)
            set Pet10Texture = EnumToTexturePet10(1)
            set Pet11Texture = EnumToTexturePet11(1)
            set Pet12Texture = EnumToTexturePet12(1)
        if(Ownership_L3_Pet3=1):
            set Pet9Texture = EnumToTexturePet9(1)
            set Pet10Texture = EnumToTexturePet10(1)
            set Pet11Texture = EnumToTexturePet11(1)
        if(Ownership_L3_Pet2=1):
            set Pet9Texture = EnumToTexturePet9(1)
            set Pet10Texture = EnumToTexturePet10(1)
        if(Ownership_L3_Pet1=1):
            set Pet9Texture = EnumToTexturePet9(1)

        #L4
        if(Ownership_L4_Pet4=1):
            set Pet13Texture = EnumToTexturePet13(1)
            set Pet14Texture = EnumToTexturePet14(1)
            set Pet15Texture = EnumToTexturePet15(1)
            set Pet16Texture = EnumToTexturePet16(1)
        if(Ownership_L4_Pet3=1):
            set Pet13Texture = EnumToTexturePet13(1)
            set Pet14Texture = EnumToTexturePet14(1)
            set Pet15Texture = EnumToTexturePet15(1)
        if(Ownership_L4_Pet2=1):
            set Pet13Texture = EnumToTexturePet13(1)
            set Pet14Texture = EnumToTexturePet14(1)
        if(Ownership_L4_Pet1=1):
            set Pet13Texture = EnumToTexturePet13(1)

        #L5
        if(Ownership_L5_Pet4=1):
            set Pet17Texture = EnumToTexturePet17(1)
            set Pet18Texture = EnumToTexturePet18(1)
            set Pet19Texture = EnumToTexturePet19(1)
            set Pet20Texture = EnumToTexturePet20(1)
        if(Ownership_L5_Pet3=1):
            set Pet17Texture = EnumToTexturePet17(1)
            set Pet18Texture = EnumToTexturePet18(1)
            set Pet19Texture = EnumToTexturePet19(1)
        if(Ownership_L5_Pet2=1):
            set Pet17Texture = EnumToTexturePet17(1)
            set Pet18Texture = EnumToTexturePet18(1)
        if(Ownership_L5_Pet1=1):
            set Pet17Texture = EnumToTexturePet17(1)

        #L6
        if(Ownership_L6_Pet4=1):
            set Pet21Texture = EnumToTexturePet21(1)
            set Pet22Texture = EnumToTexturePet22(1)
            set Pet23Texture = EnumToTexturePet23(1)
            set Pet24Texture = EnumToTexturePet24(1)
        if(Ownership_L6_Pet3=1):
            set Pet21Texture = EnumToTexturePet21(1)
            set Pet22Texture = EnumToTexturePet22(1)
            set Pet23Texture = EnumToTexturePet23(1)
        if(Ownership_L6_Pet2=1):
            set Pet21Texture = EnumToTexturePet21(1)
            set Pet22Texture = EnumToTexturePet22(1)
        if(Ownership_L6_Pet1=1):
            set Pet21Texture = EnumToTexturePet21(1)

        #L7
        if(Ownership_L7_Pet4=1):
            set Pet25Texture = EnumToTexturePet25(1)
            set Pet26Texture = EnumToTexturePet26(1)
            set Pet27Texture = EnumToTexturePet27(1)
            set Pet28Texture = EnumToTexturePet28(1)
        if(Ownership_L7_Pet3=1):
            set Pet25Texture = EnumToTexturePet25(1)
            set Pet26Texture = EnumToTexturePet26(1)
            set Pet27Texture = EnumToTexturePet27(1)
        if(Ownership_L7_Pet2=1):
            set Pet25Texture = EnumToTexturePet25(1)
            set Pet26Texture = EnumToTexturePet26(1)
        if(Ownership_L7_Pet1=1):
            set Pet25Texture = EnumToTexturePet25(1)

        #L8
        if(Ownership_L8_Pet4=1):
            set Pet29Texture = EnumToTexturePet29(1)
            set Pet30Texture = EnumToTexturePet30(1)
            set Pet31Texture = EnumToTexturePet31(1)
            set Pet32Texture = EnumToTexturePet32(1)
        if(Ownership_L8_Pet3=1):
            set Pet29Texture = EnumToTexturePet29(1)
            set Pet30Texture = EnumToTexturePet30(1)
            set Pet31Texture = EnumToTexturePet31(1)
        if(Ownership_L8_Pet2=1):
            set Pet29Texture = EnumToTexturePet29(1)
            set Pet30Texture = EnumToTexturePet30(1)
        if(Ownership_L8_Pet1=1):
            set Pet29Texture = EnumToTexturePet29(1)

        #L9
        if(Ownership_L9_Pet4=1):
            set Pet33Texture = EnumToTexturePet33(1)
            set Pet34Texture = EnumToTexturePet34(1)
            set Pet35Texture = EnumToTexturePet35(1)
            set Pet36Texture = EnumToTexturePet36(1)
        if(Ownership_L9_Pet3=1):
            set Pet33Texture = EnumToTexturePet33(1)
            set Pet34Texture = EnumToTexturePet34(1)
            set Pet35Texture = EnumToTexturePet35(1)
        if(Ownership_L9_Pet2=1):
            set Pet33Texture = EnumToTexturePet33(1)
            set Pet34Texture = EnumToTexturePet34(1)
        if(Ownership_L9_Pet1=1):
            set Pet33Texture = EnumToTexturePet33(1)

        #L10
        if(Ownership_L10_Pet4=1):
            set Pet37Texture = EnumToTexturePet37(1)
            set Pet38Texture = EnumToTexturePet38(1)
            set Pet39Texture = EnumToTexturePet39(1)
            set Pet40Texture = EnumToTexturePet40(1)
        if(Ownership_L10_Pet3=1):
            set Pet37Texture = EnumToTexturePet37(1)
            set Pet38Texture = EnumToTexturePet38(1)
            set Pet39Texture = EnumToTexturePet39(1)
        if(Ownership_L10_Pet2=1):
            set Pet37Texture = EnumToTexturePet37(1)
            set Pet38Texture = EnumToTexturePet38(1)
        if(Ownership_L10_Pet1=1):
            set Pet37Texture = EnumToTexturePet37(1)

        #L11
        if(Ownership_L11_Pet4=1):
            set Pet41Texture = EnumToTexturePet41(1)
            set Pet42Texture = EnumToTexturePet42(1)
            set Pet43Texture = EnumToTexturePet43(1)
            set Pet44Texture = EnumToTexturePet44(1)
        if(Ownership_L11_Pet3=1):
            set Pet41Texture = EnumToTexturePet41(1)
            set Pet42Texture = EnumToTexturePet42(1)
            set Pet43Texture = EnumToTexturePet43(1)
        if(Ownership_L11_Pet2=1):
            set Pet41Texture = EnumToTexturePet41(1)
            set Pet42Texture = EnumToTexturePet42(1)
        if(Ownership_L11_Pet1=1):
            set Pet41Texture = EnumToTexturePet41(1)

        #L12
        if(Ownership_L12_Pet4=1):
            set Pet45Texture = EnumToTexturePet45(1)
            set Pet46Texture = EnumToTexturePet46(1)
            set Pet47Texture = EnumToTexturePet47(1)
            set Pet48Texture = EnumToTexturePet48(1)
        if(Ownership_L12_Pet3=1):
            set Pet45Texture = EnumToTexturePet45(1)
            set Pet46Texture = EnumToTexturePet46(1)
            set Pet47Texture = EnumToTexturePet47(1)
        if(Ownership_L12_Pet2=1):
            set Pet45Texture = EnumToTexturePet45(1)
            set Pet46Texture = EnumToTexturePet46(1)
        if(Ownership_L12_Pet1=1):
            set Pet45Texture = EnumToTexturePet45(1)

        #L13
        if(Ownership_L13_Pet4=1):
            set Pet49Texture = EnumToTexturePet49(1)
            set Pet50Texture = EnumToTexturePet50(1)
            set Pet51Texture = EnumToTexturePet51(1)
            set Pet52Texture = EnumToTexturePet52(1)
        if(Ownership_L13_Pet3=1):
            set Pet49Texture = EnumToTexturePet49(1)
            set Pet50Texture = EnumToTexturePet50(1)
            set Pet51Texture = EnumToTexturePet51(1)
        if(Ownership_L13_Pet2=1):
            set Pet49Texture = EnumToTexturePet49(1)
            set Pet50Texture = EnumToTexturePet50(1)
        if(Ownership_L13_Pet1=1):
            set Pet49Texture = EnumToTexturePet49(1)

        #L14
        if(Ownership_L14_Pet4=1):
            set Pet53Texture = EnumToTexturePet53(1)
            set Pet54Texture = EnumToTexturePet54(1)
            set Pet55Texture = EnumToTexturePet55(1)
            set Pet56Texture = EnumToTexturePet56(1)
        if(Ownership_L14_Pet3=1):
            set Pet53Texture = EnumToTexturePet53(1)
            set Pet54Texture = EnumToTexturePet54(1)
            set Pet55Texture = EnumToTexturePet55(1)
        if(Ownership_L14_Pet2=1):
            set Pet53Texture = EnumToTexturePet53(1)
            set Pet54Texture = EnumToTexturePet54(1)
        if(Ownership_L14_Pet1=1):
            set Pet53Texture = EnumToTexturePet53(1)

        #L15
        if(Ownership_L15_Pet4=1):
            set Pet57Texture = EnumToTexturePet57(1)
            set Pet58Texture = EnumToTexturePet58(1)
            set Pet59Texture = EnumToTexturePet59(1)
            set Pet60Texture = EnumToTexturePet60(1)
        if(Ownership_L15_Pet3=1):
            set Pet57Texture = EnumToTexturePet57(1)
            set Pet58Texture = EnumToTexturePet58(1)
            set Pet59Texture = EnumToTexturePet59(1)
        if(Ownership_L15_Pet2=1):
            set Pet57Texture = EnumToTexturePet57(1)
            set Pet58Texture = EnumToTexturePet58(1)
        if(Ownership_L15_Pet1=1):
            set Pet57Texture = EnumToTexturePet57(1)

        #L16
        if(Ownership_L16_Pet4=1):
            set Pet61Texture = EnumToTexturePet61(1)
            set Pet62Texture = EnumToTexturePet62(1)
            set Pet63Texture = EnumToTexturePet63(1)
            set Pet64Texture = EnumToTexturePet64(1)
        if(Ownership_L16_Pet3=1):
            set Pet61Texture = EnumToTexturePet61(1)
            set Pet62Texture = EnumToTexturePet62(1)
            set Pet63Texture = EnumToTexturePet63(1)
        if(Ownership_L16_Pet2=1):
            set Pet61Texture = EnumToTexturePet61(1)
            set Pet62Texture = EnumToTexturePet62(1)
        if(Ownership_L16_Pet1=1):
            set Pet61Texture = EnumToTexturePet61(1)

        #L17
        if(Ownership_L17_Pet4=1):
            set Pet65Texture = EnumToTexturePet65(1)
            set Pet66Texture = EnumToTexturePet66(1)
            set Pet67Texture = EnumToTexturePet67(1)
            set Pet68Texture = EnumToTexturePet68(1)
        if(Ownership_L17_Pet3=1):
            set Pet65Texture = EnumToTexturePet65(1)
            set Pet66Texture = EnumToTexturePet66(1)
            set Pet67Texture = EnumToTexturePet67(1)
        if(Ownership_L17_Pet2=1):
            set Pet65Texture = EnumToTexturePet65(1)
            set Pet66Texture = EnumToTexturePet66(1)
        if(Ownership_L17_Pet1=1):
            set Pet65Texture = EnumToTexturePet65(1)

        #L18
        if(Ownership_L18_Pet4=1):
            set Pet69Texture = EnumToTexturePet69(1)
            set Pet70Texture = EnumToTexturePet70(1)
            set Pet71Texture = EnumToTexturePet71(1)
            set Pet72Texture = EnumToTexturePet72(1)
        if(Ownership_L18_Pet3=1):
            set Pet69Texture = EnumToTexturePet69(1)
            set Pet70Texture = EnumToTexturePet70(1)
            set Pet71Texture = EnumToTexturePet71(1)
        if(Ownership_L18_Pet2=1):
            set Pet69Texture = EnumToTexturePet69(1)
            set Pet70Texture = EnumToTexturePet70(1)
        if(Ownership_L18_Pet1=1):
            set Pet69Texture = EnumToTexturePet69(1)

        #L19
        if(Ownership_L19_Pet4=1):
            set Pet73Texture = EnumToTexturePet73(1)
            set Pet74Texture = EnumToTexturePet74(1)
            set Pet75Texture = EnumToTexturePet75(1)
            set Pet76Texture = EnumToTexturePet76(1)
        if(Ownership_L19_Pet3=1):
            set Pet73Texture = EnumToTexturePet73(1)
            set Pet74Texture = EnumToTexturePet74(1)
            set Pet75Texture = EnumToTexturePet75(1)
        if(Ownership_L19_Pet2=1):
            set Pet73Texture = EnumToTexturePet73(1)
            set Pet74Texture = EnumToTexturePet74(1)
        if(Ownership_L19_Pet1=1):
            set Pet73Texture = EnumToTexturePet73(1)

        #L20
        if(Ownership_L20_Pet4=1):
            set Pet77Texture = EnumToTexturePet77(1)
            set Pet78Texture = EnumToTexturePet78(1)
            set Pet79Texture = EnumToTexturePet79(1)
            set Pet80Texture = EnumToTexturePet80(1)
        if(Ownership_L20_Pet3=1):
            set Pet77Texture = EnumToTexturePet77(1)
            set Pet78Texture = EnumToTexturePet78(1)
            set Pet79Texture = EnumToTexturePet79(1)
        if(Ownership_L20_Pet2=1):
            set Pet77Texture = EnumToTexturePet77(1)
            set Pet78Texture = EnumToTexturePet78(1)
        if(Ownership_L20_Pet1=1):
            set Pet77Texture = EnumToTexturePet77(1)

        #L21
        if(Ownership_L21_Pet4=1):
            set Pet81Texture = EnumToTexturePet81(1)
            set Pet82Texture = EnumToTexturePet82(1)
            set Pet83Texture = EnumToTexturePet83(1)
            set Pet84Texture = EnumToTexturePet84(1)
        if(Ownership_L21_Pet3=1):
            set Pet81Texture = EnumToTexturePet81(1)
            set Pet82Texture = EnumToTexturePet82(1)
            set Pet83Texture = EnumToTexturePet83(1)
        if(Ownership_L21_Pet2=1):
            set Pet81Texture = EnumToTexturePet81(1)
            set Pet82Texture = EnumToTexturePet82(1)
        if(Ownership_L21_Pet1=1):
            set Pet81Texture = EnumToTexturePet81(1)

        #L22
        if(Ownership_L22_Pet4=1):
            set Pet85Texture = EnumToTexturePet85(1)
            set Pet86Texture = EnumToTexturePet86(1)
            set Pet87Texture = EnumToTexturePet87(1)
            set Pet88Texture = EnumToTexturePet88(1)
        if(Ownership_L22_Pet3=1):
            set Pet85Texture = EnumToTexturePet85(1)
            set Pet86Texture = EnumToTexturePet86(1)
            set Pet87Texture = EnumToTexturePet87(1)
        if(Ownership_L22_Pet2=1):
            set Pet85Texture = EnumToTexturePet85(1)
            set Pet86Texture = EnumToTexturePet86(1)
        if(Ownership_L22_Pet1=1):
            set Pet85Texture = EnumToTexturePet85(1)

        #L23
        if(Ownership_L23_Pet4=1):
            set Pet89Texture = EnumToTexturePet89(1)
            set Pet90Texture = EnumToTexturePet90(1)
            set Pet91Texture = EnumToTexturePet91(1)
            set Pet92Texture = EnumToTexturePet92(1)
        if(Ownership_L23_Pet3=1):
            set Pet89Texture = EnumToTexturePet89(1)
            set Pet90Texture = EnumToTexturePet90(1)
            set Pet91Texture = EnumToTexturePet91(1)
        if(Ownership_L23_Pet2=1):
            set Pet89Texture = EnumToTexturePet89(1)
            set Pet90Texture = EnumToTexturePet90(1)
        if(Ownership_L23_Pet1=1):
            set Pet89Texture = EnumToTexturePet89(1)

        #L24
        if(Ownership_L24_Pet4=1):
            set Pet93Texture = EnumToTexturePet93(1)
            set Pet94Texture = EnumToTexturePet94(1)
            set Pet95Texture = EnumToTexturePet95(1)
            set Pet96Texture = EnumToTexturePet96(1)
        if(Ownership_L24_Pet3=1):
            set Pet93Texture = EnumToTexturePet93(1)
            set Pet94Texture = EnumToTexturePet94(1)
            set Pet95Texture = EnumToTexturePet95(1)
        if(Ownership_L24_Pet2=1):
            set Pet93Texture = EnumToTexturePet93(1)
            set Pet94Texture = EnumToTexturePet94(1)
        if(Ownership_L24_Pet1=1):
            set Pet93Texture = EnumToTexturePet93(1)

        #L25
        if(Ownership_L25_Pet4=1):
            set Pet97Texture = EnumToTexturePet97(1)
            set Pet98Texture = EnumToTexturePet98(1)
            set Pet99Texture = EnumToTexturePet99(1)
            set Pet100Texture = EnumToTexturePet100(1)
        if(Ownership_L25_Pet3=1):
            set Pet97Texture = EnumToTexturePet97(1)
            set Pet98Texture = EnumToTexturePet98(1)
            set Pet99Texture = EnumToTexturePet99(1)
        if(Ownership_L25_Pet2=1):
            set Pet97Texture = EnumToTexturePet97(1)
            set Pet98Texture = EnumToTexturePet98(1)
        if(Ownership_L25_Pet1=1):
            set Pet97Texture = EnumToTexturePet97(1)

        #L26
        if(Ownership_L26_Pet4=1):
            set Pet101Texture = EnumToTexturePet101(1)
            set Pet102Texture = EnumToTexturePet102(1)
            set Pet103Texture = EnumToTexturePet103(1)
            set Pet104Texture = EnumToTexturePet104(1)
        if(Ownership_L26_Pet3=1):
            set Pet101Texture = EnumToTexturePet101(1)
            set Pet102Texture = EnumToTexturePet102(1)
            set Pet103Texture = EnumToTexturePet103(1)
        if(Ownership_L26_Pet2=1):
            set Pet101Texture = EnumToTexturePet101(1)
            set Pet102Texture = EnumToTexturePet102(1)
        if(Ownership_L26_Pet1=1):
            set Pet101Texture = EnumToTexturePet101(1)

        #L27
        if(Ownership_L27_Pet4=1):
            set Pet105Texture = EnumToTexturePet105(1)
            set Pet106Texture = EnumToTexturePet106(1)
            set Pet107Texture = EnumToTexturePet107(1)
            set Pet108Texture = EnumToTexturePet108(1)
        if(Ownership_L27_Pet3=1):
            set Pet105Texture = EnumToTexturePet105(1)
            set Pet106Texture = EnumToTexturePet106(1)
            set Pet107Texture = EnumToTexturePet107(1)
        if(Ownership_L27_Pet2=1):
            set Pet105Texture = EnumToTexturePet105(1)
            set Pet106Texture = EnumToTexturePet106(1)
        if(Ownership_L27_Pet1=1):
            set Pet105Texture = EnumToTexturePet105(1)

        #L28
        if(Ownership_L28_Pet4=1):
            set Pet109Texture = EnumToTexturePet109(1)
            set Pet110Texture = EnumToTexturePet110(1)
            set Pet111Texture = EnumToTexturePet111(1)
            set Pet112Texture = EnumToTexturePet112(1)
        if(Ownership_L28_Pet3=1):
            set Pet109Texture = EnumToTexturePet109(1)
            set Pet110Texture = EnumToTexturePet110(1)
            set Pet111Texture = EnumToTexturePet111(1)
        if(Ownership_L28_Pet2=1):
            set Pet109Texture = EnumToTexturePet109(1)
            set Pet110Texture = EnumToTexturePet110(1)
        if(Ownership_L28_Pet1=1):
            set Pet109Texture = EnumToTexturePet109(1)

        #L29
        if(Ownership_L29_Pet4=1):
            set Pet113Texture = EnumToTexturePet113(1)
            set Pet114Texture = EnumToTexturePet114(1)
            set Pet115Texture = EnumToTexturePet115(1)
            set Pet116Texture = EnumToTexturePet116(1)
        if(Ownership_L29_Pet3=1):
            set Pet113Texture = EnumToTexturePet113(1)
            set Pet114Texture = EnumToTexturePet114(1)
            set Pet115Texture = EnumToTexturePet115(1)
        if(Ownership_L29_Pet2=1):
            set Pet113Texture = EnumToTexturePet113(1)
            set Pet114Texture = EnumToTexturePet114(1)
        if(Ownership_L29_Pet1=1):
            set Pet113Texture = EnumToTexturePet113(1)

        #L30
        if(Ownership_L30_Pet4=1):
            set Pet117Texture = EnumToTexturePet117(1)
            set Pet118Texture = EnumToTexturePet118(1)
            set Pet119Texture = EnumToTexturePet119(1)
            set Pet120Texture = EnumToTexturePet120(1)
        if(Ownership_L30_Pet3=1):
            set Pet117Texture = EnumToTexturePet117(1)
            set Pet118Texture = EnumToTexturePet118(1)
            set Pet119Texture = EnumToTexturePet119(1)
        if(Ownership_L30_Pet2=1):
            set Pet117Texture = EnumToTexturePet117(1)
            set Pet118Texture = EnumToTexturePet118(1)
        if(Ownership_L30_Pet1=1):
            set Pet117Texture = EnumToTexturePet117(1)

        #LSecret
        set Pet121Texture = EnumToTexturePet121(Ownership_Secret_Pet1)
        set Pet122Texture = EnumToTexturePet122(Ownership_Secret_Pet2)
        set Pet123Texture = EnumToTexturePet123(Ownership_Secret_Pet3)
        set Pet124Texture = EnumToTexturePet124(Ownership_Secret_Pet4)
        set Pet125Texture = EnumToTexturePet125(Ownership_Secret_Pet5)

        Update_Total_Pets()

        set TextNumberOfPets = text_block{DefaultText := StringToMessage("You have {Ownership_Total_Pets}/120 pets!"),DefaultTextColor:=NamedColors.Black}
        TextBlockGold.SetShadowOpacity(1.0)

        ShopCanvas := CreateUserUI3()
        set MaybeShopCanvas = option{ ShopCanvas }

        if(PlayerObj:=player[MyAgentObj], PlayerUI := GetPlayerUI[PlayerObj]):
            PlayerUI.AddWidget(ShopCanvas, player_ui_slot{ InputMode := ui_input_mode.All })
                
    OnPageCloseClicked(WidgetMessage : widget_message):void=
        Print("Page Close Clicked")
        if (PlayerUI := GetPlayerUI[WidgetMessage.Player], ShopCanvas := MaybeShopCanvas?):
            set bShopOpen = false
            set MaybeShopCanvas = false
            PlayerUI.RemoveWidget(ShopCanvas)

    CreateUserUI<private>() : canvas=

        MyCanvas := canvas:
            Slots:=array:
                #Canvas
                canvas_slot:
                    #Canvas Slot Info:
                    Anchors:=anchors{Minimum:=vector2{X:=0.0,Y:=0.20},Maximum:=vector2{X:=0.0,Y:=0.20}}
                    Offsets:=margin{}
                    Alignment:=vector2{X:=0.0,Y:=0.0}
                    SizeToContent:=true
                    Widget:=stack_box:
                        #Big Stack Box
                        Orientation:=orientation.Vertical
                        Slots:=array:
                            #Pet
                            stack_box_slot:
                                HorizontalAlignment:=horizontal_alignment.Left
                                VerticalAlignment:=vertical_alignment.Fill
                                Widget:=overlay:
                                    Slots:=array:
                                        #Background Image
                                        overlay_slot:
                                            HorizontalAlignment:=horizontal_alignment.Fill
                                            VerticalAlignment:=vertical_alignment.Fill
                                            Widget:=texture_block:
                                                DefaultImage:=HUD_Level.back_pet
                                                DefaultDesiredSize:=vector2{X:=400.0,Y:=400.0}
                                        #Icon/Text
                                        overlay_slot:
                                            Widget:=stack_box:
                                                Orientation:=orientation.Horizontal
                                                Slots:=array:
                                                    #Icon
                                                    stack_box_slot:
                                                        Widget:=texture_block:
                                                            DefaultImage:=HUD_Level.icon_pet
                                                            DefaultDesiredSize:=vector2{X:=40.0,Y:=40.0}
                                                    #Desc
                                                    stack_box_slot:
                                                        Widget:=texture_block:
                                                            DefaultImage:=HUD_Level.desc_pets
                                                            DefaultDesiredSize:=vector2{X:=40.0,Y:=40.0}
                                                        Padding:=margin{Left:=5.0,Right:=5.0}
                                                    #Text
                                                    stack_box_slot:
                                                        Widget:=TextBlockPetNbr
                                                        HorizontalAlignment:=horizontal_alignment.Fill
                                                        VerticalAlignment:=vertical_alignment.Center
                                                    #Text
                                                    stack_box_slot:
                                                        Widget:=TextBlockPetMax
                                                        HorizontalAlignment:=horizontal_alignment.Fill
                                                        VerticalAlignment:=vertical_alignment.Center
                                            Padding:=margin{Left:=110.0}
                                            HorizontalAlignment:=horizontal_alignment.Left
                                            VerticalAlignment:=vertical_alignment.Center

                            #Power
                            stack_box_slot:
                                HorizontalAlignment:=horizontal_alignment.Left
                                VerticalAlignment:=vertical_alignment.Fill
                                Padding:=margin{Top:=-320.0}
                                Widget:=overlay:
                                    Slots:=array:
                                        #Background Image
                                        overlay_slot:
                                            HorizontalAlignment:=horizontal_alignment.Fill
                                            VerticalAlignment:=vertical_alignment.Fill
                                            Widget:=texture_block:
                                                DefaultImage:=HUD_Level.back_normal
                                                DefaultDesiredSize:=vector2{X:=400.0,Y:=400.0}
                                        #Icon/Text
                                        overlay_slot:
                                            Widget:=stack_box:
                                                Orientation:=orientation.Horizontal
                                                Slots:=array:
                                                    #Icon
                                                    stack_box_slot:
                                                        Widget:=texture_block:
                                                            DefaultImage:=HUD_Level.icon_power
                                                            DefaultDesiredSize:=vector2{X:=40.0,Y:=40.0}
                                                    #Desc
                                                    stack_box_slot:
                                                        Widget:=texture_block:
                                                            DefaultImage:=HUD_Level.desc_power
                                                            DefaultDesiredSize:=vector2{X:=40.0,Y:=40.0}
                                                        Padding:=margin{Left:=5.0,Right:=5.0}
                                                    #Text
                                                    stack_box_slot:
                                                        Widget:=TextBlockPower
                                                        HorizontalAlignment:=horizontal_alignment.Fill
                                                        VerticalAlignment:=vertical_alignment.Center
                                            Padding:=margin{Left:=110.0}
                                            HorizontalAlignment:=horizontal_alignment.Left
                                            VerticalAlignment:=vertical_alignment.Center
                            
                            #Crit
                            stack_box_slot:
                                HorizontalAlignment:=horizontal_alignment.Left
                                VerticalAlignment:=vertical_alignment.Fill
                                Padding:=margin{Top:=-320.0}
                                Widget:=overlay:
                                    Slots:=array:
                                        #Background Image
                                        overlay_slot:
                                            HorizontalAlignment:=horizontal_alignment.Fill
                                            VerticalAlignment:=vertical_alignment.Fill
                                            Widget:=texture_block:
                                                DefaultImage:=HUD_Level.back_normal
                                                DefaultDesiredSize:=vector2{X:=400.0,Y:=400.0}
                                        #Icon/Text
                                        overlay_slot:
                                            Widget:=stack_box:
                                                Orientation:=orientation.Horizontal
                                                Slots:=array:
                                                    #Icon
                                                    stack_box_slot:
                                                        Widget:=texture_block:
                                                            DefaultImage:=HUD_Level.icon_crit
                                                            DefaultDesiredSize:=vector2{X:=40.0,Y:=40.0}
                                                    #Desc
                                                    stack_box_slot:
                                                        Widget:=texture_block:
                                                            DefaultImage:=HUD_Level.desc_crit
                                                            DefaultDesiredSize:=vector2{X:=40.0,Y:=40.0}
                                                        Padding:=margin{Left:=5.0,Right:=5.0}
                                                    #Text
                                                    stack_box_slot:
                                                        Widget:=TextBlockCrit
                                                        HorizontalAlignment:=horizontal_alignment.Fill
                                                        VerticalAlignment:=vertical_alignment.Center
                                            Padding:=margin{Left:=110.0}
                                            HorizontalAlignment:=horizontal_alignment.Left
                                            VerticalAlignment:=vertical_alignment.Center
                            #Speed
                            stack_box_slot:                                
                                HorizontalAlignment:=horizontal_alignment.Left
                                VerticalAlignment:=vertical_alignment.Fill
                                Padding:=margin{Top:=-320.0}
                                Widget:=overlay:
                                    Slots:=array:
                                        #Background Image
                                        overlay_slot:
                                            HorizontalAlignment:=horizontal_alignment.Fill
                                            VerticalAlignment:=vertical_alignment.Fill
                                            Widget:=texture_block:
                                                DefaultImage:=HUD_Level.back_normal
                                                DefaultDesiredSize:=vector2{X:=400.0,Y:=400.0}
                                        #Icon/Text
                                        overlay_slot:
                                            Widget:=stack_box:
                                                Orientation:=orientation.Horizontal
                                                Slots:=array:
                                                    #Icon
                                                    stack_box_slot:
                                                        Widget:=texture_block:
                                                            DefaultImage:=HUD_Level.icon_speed
                                                            DefaultDesiredSize:=vector2{X:=40.0,Y:=40.0}
                                                    #Desc
                                                    stack_box_slot:
                                                        Widget:=texture_block:
                                                            DefaultImage:=HUD_Level.desc_speed
                                                            DefaultDesiredSize:=vector2{X:=40.0,Y:=40.0}
                                                        Padding:=margin{Left:=5.0,Right:=5.0}
                                                    #Text
                                                    stack_box_slot:
                                                        Widget:=TextBlockSpeed
                                                        HorizontalAlignment:=horizontal_alignment.Fill
                                                        VerticalAlignment:=vertical_alignment.Center
                                            Padding:=margin{Left:=110.0}
                                            HorizontalAlignment:=horizontal_alignment.Left
                                            VerticalAlignment:=vertical_alignment.Center
                            #PlayerSpeed
                            stack_box_slot:                                
                                HorizontalAlignment:=horizontal_alignment.Left
                                VerticalAlignment:=vertical_alignment.Fill
                                Padding:=margin{Top:=-320.0}
                                Widget:=overlay:
                                    Slots:=array:
                                        #Background Image
                                        overlay_slot:
                                            HorizontalAlignment:=horizontal_alignment.Fill
                                            VerticalAlignment:=vertical_alignment.Fill
                                            Widget:=texture_block:
                                                DefaultImage:=HUD_Level.back_normal
                                                DefaultDesiredSize:=vector2{X:=400.0,Y:=400.0}
                                        #Icon/Text
                                        overlay_slot:
                                            Widget:=stack_box:
                                                Orientation:=orientation.Horizontal
                                                Slots:=array:
                                                    #Icon
                                                    stack_box_slot:
                                                        Widget:=texture_block:
                                                            DefaultImage:=HUD_Level.icon_player_speed
                                                            DefaultDesiredSize:=vector2{X:=40.0,Y:=40.0}
                                                    #Desc
                                                    stack_box_slot:
                                                        Widget:=texture_block:
                                                            DefaultImage:=HUD_Level.desc_player_speed
                                                            DefaultDesiredSize:=vector2{X:=40.0,Y:=40.0}
                                                        Padding:=margin{Left:=5.0,Right:=5.0}
                                                    #Text
                                                    stack_box_slot:
                                                        Widget:=TextBlockPlayerSpeed
                                                        HorizontalAlignment:=horizontal_alignment.Fill
                                                        VerticalAlignment:=vertical_alignment.Center
                                            Padding:=margin{Left:=110.0}
                                            HorizontalAlignment:=horizontal_alignment.Left
                                            VerticalAlignment:=vertical_alignment.Center
                            #Luck
                            stack_box_slot:                                
                                HorizontalAlignment:=horizontal_alignment.Left
                                VerticalAlignment:=vertical_alignment.Fill
                                Padding:=margin{Top:=-320.0}
                                Widget:=overlay:
                                    Slots:=array:
                                        #Background Image
                                        overlay_slot:
                                            HorizontalAlignment:=horizontal_alignment.Fill
                                            VerticalAlignment:=vertical_alignment.Fill
                                            Widget:=texture_block:
                                                DefaultImage:=HUD_Level.back_normal
                                                DefaultDesiredSize:=vector2{X:=400.0,Y:=400.0}
                                        #Icon/Text
                                        overlay_slot:
                                            Widget:=stack_box:
                                                Orientation:=orientation.Horizontal
                                                Slots:=array:
                                                    #Icon
                                                    stack_box_slot:
                                                        Widget:=texture_block:
                                                            DefaultImage:=HUD_Level.icon_luck
                                                            DefaultDesiredSize:=vector2{X:=40.0,Y:=40.0}
                                                    #Desc
                                                    stack_box_slot:
                                                        Widget:=texture_block:
                                                            DefaultImage:=HUD_Level.desc_luck
                                                            DefaultDesiredSize:=vector2{X:=40.0,Y:=40.0}
                                                        Padding:=margin{Left:=5.0,Right:=5.0}
                                                    #Text
                                                    stack_box_slot:
                                                        Widget:=TextBlockLuck
                                                        HorizontalAlignment:=horizontal_alignment.Fill
                                                        VerticalAlignment:=vertical_alignment.Center
                                            Padding:=margin{Left:=110.0}
                                            HorizontalAlignment:=horizontal_alignment.Left
                                            VerticalAlignment:=vertical_alignment.Center

                            #Gold
                            stack_box_slot:                                
                                HorizontalAlignment:=horizontal_alignment.Left
                                VerticalAlignment:=vertical_alignment.Fill
                                Padding:=margin{Top:=-320.0}
                                Widget:=overlay:
                                    Slots:=array:
                                        #Background Image
                                        overlay_slot:
                                            HorizontalAlignment:=horizontal_alignment.Fill
                                            VerticalAlignment:=vertical_alignment.Fill
                                            Widget:=texture_block:
                                                DefaultImage:=HUD_Level.back_normal
                                                DefaultDesiredSize:=vector2{X:=400.0,Y:=400.0}
                                        #Icon/Text
                                        overlay_slot:
                                            Widget:=stack_box:
                                                Orientation:=orientation.Horizontal
                                                Slots:=array:
                                                    #Icon
                                                    stack_box_slot:
                                                        Widget:=texture_block:
                                                            DefaultImage:=HUD_Level.icon_gold
                                                            DefaultDesiredSize:=vector2{X:=40.0,Y:=40.0}
                                                    #Desc
                                                    stack_box_slot:
                                                        Widget:=texture_block:
                                                            DefaultImage:=HUD_Level.desc_gold
                                                            DefaultDesiredSize:=vector2{X:=40.0,Y:=40.0}
                                                        Padding:=margin{Left:=5.0,Right:=5.0}
                                                    #Text
                                                    stack_box_slot:
                                                        Widget:=TextBlockGold
                                                        HorizontalAlignment:=horizontal_alignment.Fill
                                                        VerticalAlignment:=vertical_alignment.Center
                                            Padding:=margin{Left:=110.0}
                                            HorizontalAlignment:=horizontal_alignment.Left
                                            VerticalAlignment:=vertical_alignment.Center

                            #Diamond
                            stack_box_slot:                                
                                HorizontalAlignment:=horizontal_alignment.Left
                                VerticalAlignment:=vertical_alignment.Fill
                                Padding:=margin{Top:=-320.0}
                                Widget:=overlay:
                                    Slots:=array:
                                        #Background Image
                                        overlay_slot:
                                            HorizontalAlignment:=horizontal_alignment.Fill
                                            VerticalAlignment:=vertical_alignment.Fill
                                            Widget:=texture_block:
                                                DefaultImage:=HUD_Level.back_normal
                                                DefaultDesiredSize:=vector2{X:=400.0,Y:=400.0}
                                        #Icon/Text
                                        overlay_slot:
                                            Widget:=stack_box:
                                                Orientation:=orientation.Horizontal
                                                Slots:=array:
                                                    #Icon
                                                    stack_box_slot:
                                                        Widget:=texture_block:
                                                            DefaultImage:=HUD_Level.icon_diamond
                                                            DefaultDesiredSize:=vector2{X:=40.0,Y:=40.0}
                                                    #Desc
                                                    stack_box_slot:
                                                        Widget:=texture_block:
                                                            DefaultImage:=HUD_Level.desc_diamond
                                                            DefaultDesiredSize:=vector2{X:=40.0,Y:=40.0}
                                                        Padding:=margin{Left:=5.0,Right:=5.0}
                                                    #Text
                                                    stack_box_slot:
                                                        Widget:=TextBlockDiamond
                                                        HorizontalAlignment:=horizontal_alignment.Fill
                                                        VerticalAlignment:=vertical_alignment.Center
                                            Padding:=margin{Left:=110.0}
                                            HorizontalAlignment:=horizontal_alignment.Left
                                            VerticalAlignment:=vertical_alignment.Center
    #Inventory UI                                    
    CreateUserUI2<private>() : canvas=

        MyCanvas := canvas:
            Slots:=array:
                #Canvas
                canvas_slot:
                    #-----------------------------TIMINGS
                    Anchors:=anchors{Minimum:=vector2{X:=1.0,Y:=0.5},Maximum:=vector2{X:=1.0,Y:=0.5}}
                    Alignment:=vector2{X:=0.0,Y:=0.0}
                    SizeToContent:=false
                    Offsets:=margin{Left:=-270.0,Top:=-230.0,Right:=100.0,Bottom:=30.0}
                    Widget:=stack_box:
                        Orientation:=orientation.Vertical
                            Slots:=array:
                                #Icons
                                stack_box_slot:
                                    Widget:=TextBlockPotion_Active_PetPower1_Sec
                                    HorizontalAlignment:=horizontal_alignment.Fill
                                    VerticalAlignment:=vertical_alignment.Fill 
                                stack_box_slot:
                                    Widget:=TextBlockPotion_Active_PetPower2_Sec
                                    HorizontalAlignment:=horizontal_alignment.Fill
                                    VerticalAlignment:=vertical_alignment.Fill
                                stack_box_slot:
                                    Widget:=TextBlockPotion_Active_PetPower3_Sec
                                    HorizontalAlignment:=horizontal_alignment.Fill
                                    VerticalAlignment:=vertical_alignment.Fill
                                stack_box_slot:
                                    Widget:=TextBlockPotion_Active_PlayerPower1_Sec
                                    HorizontalAlignment:=horizontal_alignment.Fill
                                    VerticalAlignment:=vertical_alignment.Fill
                                stack_box_slot:
                                    Widget:=TextBlockPotion_Active_PlayerPower2_Sec
                                    HorizontalAlignment:=horizontal_alignment.Fill
                                    VerticalAlignment:=vertical_alignment.Fill
                                stack_box_slot:
                                    Widget:=TextBlockPotion_Active_PlayerPower3_Sec
                                    HorizontalAlignment:=horizontal_alignment.Fill
                                    VerticalAlignment:=vertical_alignment.Fill
                                stack_box_slot:
                                    Widget:=TextBlockPotion_Active_Gold1_Sec
                                    HorizontalAlignment:=horizontal_alignment.Fill
                                    VerticalAlignment:=vertical_alignment.Fill
                                stack_box_slot:
                                    Widget:=TextBlockPotion_Active_Gold2_Sec
                                    HorizontalAlignment:=horizontal_alignment.Fill
                                    VerticalAlignment:=vertical_alignment.Fill
                                stack_box_slot:
                                    Widget:=TextBlockPotion_Active_Gold3_Sec
                                    HorizontalAlignment:=horizontal_alignment.Fill
                                    VerticalAlignment:=vertical_alignment.Fill
                                stack_box_slot:
                                    Widget:=TextBlockPotion_Active_Luck1_Sec
                                    HorizontalAlignment:=horizontal_alignment.Fill
                                    VerticalAlignment:=vertical_alignment.Fill
                                stack_box_slot:
                                    Widget:=TextBlockPotion_Active_Luck2_Sec
                                    HorizontalAlignment:=horizontal_alignment.Fill
                                    VerticalAlignment:=vertical_alignment.Fill
                                stack_box_slot:
                                    Widget:=TextBlockPotion_Active_Luck3_Sec
                                    HorizontalAlignment:=horizontal_alignment.Fill
                                    VerticalAlignment:=vertical_alignment.Fill

                canvas_slot:
                    #-----------------------------POTION ICON
                    Anchors:=anchors{Minimum:=vector2{X:=1.0,Y:=0.5},Maximum:=vector2{X:=1.0,Y:=0.5}}
                    Alignment:=vector2{X:=0.0,Y:=0.0}
                    SizeToContent:=false
                    Offsets:=margin{Left:=-150.0,Top:=-234.0,Right:=100.0,Bottom:=30.0}
                    Widget:=stack_box:
                        Orientation:=orientation.Vertical
                            Slots:=array:
                                #Icons
                                stack_box_slot:
                                    Widget:=texture_block:
                                        DefaultImage:=HUD_Inventory.pet_power_1
                                        DefaultDesiredSize:=vector2{X:=42.0,Y:=42.0}
                                stack_box_slot:
                                    Widget:=texture_block:
                                        DefaultImage:=HUD_Inventory.pet_power_2
                                        DefaultDesiredSize:=vector2{X:=42.0,Y:=42.0}
                                stack_box_slot:
                                    Widget:=texture_block:
                                        DefaultImage:=HUD_Inventory.pet_power_3
                                        DefaultDesiredSize:=vector2{X:=42.0,Y:=42.0}
                                stack_box_slot:
                                    Widget:=texture_block:
                                        DefaultImage:=HUD_Inventory.player_power_1
                                        DefaultDesiredSize:=vector2{X:=42.0,Y:=42.0}
                                stack_box_slot:
                                    Widget:=texture_block:
                                        DefaultImage:=HUD_Inventory.player_power_2
                                        DefaultDesiredSize:=vector2{X:=42.0,Y:=42.0}
                                stack_box_slot:
                                    Widget:=texture_block:
                                        DefaultImage:=HUD_Inventory.player_power_3
                                        DefaultDesiredSize:=vector2{X:=42.0,Y:=42.0}
                                stack_box_slot:
                                    Widget:=texture_block:
                                        DefaultImage:=HUD_Inventory.gold_1
                                        DefaultDesiredSize:=vector2{X:=42.0,Y:=42.0}
                                stack_box_slot:
                                    Widget:=texture_block:
                                        DefaultImage:=HUD_Inventory.gold_2
                                        DefaultDesiredSize:=vector2{X:=42.0,Y:=42.0}
                                stack_box_slot:
                                    Widget:=texture_block:
                                        DefaultImage:=HUD_Inventory.gold_3
                                        DefaultDesiredSize:=vector2{X:=42.0,Y:=42.0}
                                stack_box_slot:
                                    Widget:=texture_block:
                                        DefaultImage:=HUD_Inventory.luck_1
                                        DefaultDesiredSize:=vector2{X:=42.0,Y:=42.0}
                                stack_box_slot:
                                    Widget:=texture_block:
                                        DefaultImage:=HUD_Inventory.luck_2
                                        DefaultDesiredSize:=vector2{X:=42.0,Y:=42.0}
                                stack_box_slot:
                                    Widget:=texture_block:
                                        DefaultImage:=HUD_Inventory.luck_3
                                        DefaultDesiredSize:=vector2{X:=42.0,Y:=42.0}

                canvas_slot:
                    #-----------------------------POTION COUNT
                    Anchors:=anchors{Minimum:=vector2{X:=1.0,Y:=0.5},Maximum:=vector2{X:=1.0,Y:=0.5}}
                    Alignment:=vector2{X:=0.0,Y:=0.0}
                    SizeToContent:=false
                    Offsets:=margin{Left:=-75.0,Top:=-230.0,Right:=100.0,Bottom:=20.0}
                    Widget:=stack_box:
                        Orientation:=orientation.Vertical
                            Slots:=array:
                                #Icons
                                stack_box_slot:
                                    Widget:=TextBlockOwnedPetPower1
                                    HorizontalAlignment:=horizontal_alignment.Fill
                                    VerticalAlignment:=vertical_alignment.Fill 
                                stack_box_slot:
                                    Widget:=TextBlockOwnedPetPower2
                                    HorizontalAlignment:=horizontal_alignment.Fill
                                    VerticalAlignment:=vertical_alignment.Fill 
                                stack_box_slot:
                                    Widget:=TextBlockOwnedPetPower3
                                    HorizontalAlignment:=horizontal_alignment.Fill
                                    VerticalAlignment:=vertical_alignment.Fill 
                                stack_box_slot:
                                    Widget:=TextBlockOwnedPlayerPower1
                                    HorizontalAlignment:=horizontal_alignment.Fill
                                    VerticalAlignment:=vertical_alignment.Fill 
                                stack_box_slot:
                                    Widget:=TextBlockOwnedPlayerPower2
                                    HorizontalAlignment:=horizontal_alignment.Fill
                                    VerticalAlignment:=vertical_alignment.Fill 
                                stack_box_slot:
                                    Widget:=TextBlockOwnedPlayerPower3
                                    HorizontalAlignment:=horizontal_alignment.Fill
                                    VerticalAlignment:=vertical_alignment.Fill 
                                stack_box_slot:
                                    Widget:=TextBlockOwnedGold1
                                    HorizontalAlignment:=horizontal_alignment.Fill
                                    VerticalAlignment:=vertical_alignment.Fill 
                                stack_box_slot:
                                    Widget:=TextBlockOwnedGold2
                                    HorizontalAlignment:=horizontal_alignment.Fill
                                    VerticalAlignment:=vertical_alignment.Fill 
                                stack_box_slot:
                                    Widget:=TextBlockOwnedGold3
                                    HorizontalAlignment:=horizontal_alignment.Fill
                                    VerticalAlignment:=vertical_alignment.Fill 
                                stack_box_slot:
                                    Widget:=TextBlockOwnedLuck1
                                    HorizontalAlignment:=horizontal_alignment.Fill
                                    VerticalAlignment:=vertical_alignment.Fill 
                                stack_box_slot:
                                    Widget:=TextBlockOwnedLuck2
                                    HorizontalAlignment:=horizontal_alignment.Fill
                                    VerticalAlignment:=vertical_alignment.Fill 
                                stack_box_slot:
                                    Widget:=TextBlockOwnedLuck3
                                    HorizontalAlignment:=horizontal_alignment.Fill
                                    VerticalAlignment:=vertical_alignment.Fill 
                
    #PetDex                               
    CreateUserUI3<private>() : canvas=

        var CanvasSlots : []canvas_slot = array{}
        var CloseButtonString : string = ""

        set CanvasSlots += array
        {
        CanvaSlot1 := canvas_slot:
                    #Canvas Slot Info:
                    Anchors:=anchors{Minimum:=vector2{X:=0.5,Y:=0.5},Maximum:=vector2{X:=0.5,Y:=0.5}}
                    Offsets:=margin{Left:=-608.0,Top:=-528.0,Right:=100.0,Bottom:=30.0}
                    Alignment:=vector2{X:=0.0,Y:=0.0}
                    SizeToContent:=true
                    Widget:=overlay:
                        #Big Stack Box
                        Slots:=array:
                            #Background Image
                            overlay_slot:
                                HorizontalAlignment:=horizontal_alignment.Fill
                                VerticalAlignment:=vertical_alignment.Fill
                                Widget:=texture_block:
                                    DefaultImage:=HUD_PetDex.petdex_background
                                    DefaultDesiredSize:=vector2{X:=1222.0,Y:=1050.0}
                            #Big Stackbox 1
                            overlay_slot:
                                Widget:=stack_box:
                                    Orientation:=orientation.Vertical
                                    Slots:=array:
                                        stack_box_slot:
                                            #Sub Stackbox 1
                                            Widget:=stack_box:
                                                Orientation:=orientation.Horizontal
                                                Slots:=array:
                                                    #Pet1
                                                    stack_box_slot:
                                                        Widget:=texture_block:
                                                            DefaultImage:=Pet1Texture
                                                            DefaultDesiredSize:=vector2{X:=65.0,Y:=65.0}
                                                        Padding:=margin{Left:=5.0,Right:=5.0}
                                                    #Pet2
                                                    stack_box_slot:
                                                        Widget:=texture_block:
                                                            DefaultImage:=Pet2Texture
                                                            DefaultDesiredSize:=vector2{X:=65.0,Y:=65.0}
                                                        Padding:=margin{Left:=5.0,Right:=5.0}
                                                    #Pet3
                                                    stack_box_slot:
                                                        Widget:=texture_block:
                                                            DefaultImage:=Pet3Texture
                                                            DefaultDesiredSize:=vector2{X:=65.0,Y:=65.0}
                                                        Padding:=margin{Left:=5.0,Right:=5.0}
                                                    #Pet4
                                                    stack_box_slot:
                                                        Widget:=texture_block:
                                                            DefaultImage:=Pet4Texture
                                                            DefaultDesiredSize:=vector2{X:=65.0,Y:=65.0}
                                                        Padding:=margin{Left:=5.0,Right:=5.0}
                                            HorizontalAlignment:=horizontal_alignment.Left
                                            VerticalAlignment:=vertical_alignment.Top
                                            Padding:=margin{Left:=30.0,Right:=30.0,Bottom:=5.0}
                                        stack_box_slot:
                                            #Sub Stackbox 2
                                            Widget:=stack_box:
                                                Orientation:=orientation.Horizontal
                                                Slots:=array:
                                                    #Pet1
                                                    stack_box_slot:
                                                        Widget:=texture_block:
                                                            DefaultImage:=Pet5Texture
                                                            DefaultDesiredSize:=vector2{X:=65.0,Y:=65.0}
                                                        Padding:=margin{Left:=5.0,Right:=5.0}
                                                    #Pet2
                                                    stack_box_slot:
                                                        Widget:=texture_block:
                                                            DefaultImage:=Pet6Texture
                                                            DefaultDesiredSize:=vector2{X:=65.0,Y:=65.0}
                                                        Padding:=margin{Left:=5.0,Right:=5.0}
                                                    #Pet3
                                                    stack_box_slot:
                                                        Widget:=texture_block:
                                                            DefaultImage:=Pet7Texture
                                                            DefaultDesiredSize:=vector2{X:=65.0,Y:=65.0}
                                                        Padding:=margin{Left:=5.0,Right:=5.0}
                                                    #Pet4
                                                    stack_box_slot:
                                                        Widget:=texture_block:
                                                            DefaultImage:=Pet8Texture
                                                            DefaultDesiredSize:=vector2{X:=65.0,Y:=65.0}
                                                        Padding:=margin{Left:=5.0,Right:=5.0}
                                            HorizontalAlignment:=horizontal_alignment.Left
                                            VerticalAlignment:=vertical_alignment.Top
                                            Padding:=margin{Left:=30.0,Right:=30.0,Bottom:=5.0}

                                        stack_box_slot:
                                            #Sub Stackbox 3
                                            Widget:=stack_box:
                                                Orientation:=orientation.Horizontal
                                                Slots:=array:
                                                    #Pet1
                                                    stack_box_slot:
                                                        Widget:=texture_block:
                                                            DefaultImage:=Pet9Texture
                                                            DefaultDesiredSize:=vector2{X:=65.0,Y:=65.0}
                                                        Padding:=margin{Left:=5.0,Right:=5.0}
                                                    #Pet2
                                                    stack_box_slot:
                                                        Widget:=texture_block:
                                                            DefaultImage:=Pet10Texture
                                                            DefaultDesiredSize:=vector2{X:=65.0,Y:=65.0}
                                                        Padding:=margin{Left:=5.0,Right:=5.0}
                                                    #Pet3
                                                    stack_box_slot:
                                                        Widget:=texture_block:
                                                            DefaultImage:=Pet11Texture
                                                            DefaultDesiredSize:=vector2{X:=65.0,Y:=65.0}
                                                        Padding:=margin{Left:=5.0,Right:=5.0}
                                                    #Pet4
                                                    stack_box_slot:
                                                        Widget:=texture_block:
                                                            DefaultImage:=Pet12Texture
                                                            DefaultDesiredSize:=vector2{X:=65.0,Y:=65.0}
                                                        Padding:=margin{Left:=5.0,Right:=5.0}
                                            HorizontalAlignment:=horizontal_alignment.Left
                                            VerticalAlignment:=vertical_alignment.Top
                                            Padding:=margin{Left:=30.0,Right:=30.0,Bottom:=5.0}

                                        stack_box_slot:
                                            #Sub Stackbox 4
                                            Widget:=stack_box:
                                                Orientation:=orientation.Horizontal
                                                Slots:=array:
                                                    #Pet1
                                                    stack_box_slot:
                                                        Widget:=texture_block:
                                                            DefaultImage:=Pet13Texture
                                                            DefaultDesiredSize:=vector2{X:=65.0,Y:=65.0}
                                                        Padding:=margin{Left:=5.0,Right:=5.0}
                                                    #Pet2
                                                    stack_box_slot:
                                                        Widget:=texture_block:
                                                            DefaultImage:=Pet14Texture
                                                            DefaultDesiredSize:=vector2{X:=65.0,Y:=65.0}
                                                        Padding:=margin{Left:=5.0,Right:=5.0}
                                                    #Pet3
                                                    stack_box_slot:
                                                        Widget:=texture_block:
                                                            DefaultImage:=Pet15Texture
                                                            DefaultDesiredSize:=vector2{X:=65.0,Y:=65.0}
                                                        Padding:=margin{Left:=5.0,Right:=5.0}
                                                    #Pet4
                                                    stack_box_slot:
                                                        Widget:=texture_block:
                                                            DefaultImage:=Pet16Texture
                                                            DefaultDesiredSize:=vector2{X:=65.0,Y:=65.0}
                                                        Padding:=margin{Left:=5.0,Right:=5.0}
                                            HorizontalAlignment:=horizontal_alignment.Left
                                            VerticalAlignment:=vertical_alignment.Top
                                            Padding:=margin{Left:=30.0,Right:=30.0,Bottom:=5.0}

                                        stack_box_slot:
                                            #Sub Stackbox 5
                                            Widget:=stack_box:
                                                Orientation:=orientation.Horizontal
                                                Slots:=array:
                                                    #Pet1
                                                    stack_box_slot:
                                                        Widget:=texture_block:
                                                            DefaultImage:=Pet17Texture
                                                            DefaultDesiredSize:=vector2{X:=65.0,Y:=65.0}
                                                        Padding:=margin{Left:=5.0,Right:=5.0}
                                                    #Pet2
                                                    stack_box_slot:
                                                        Widget:=texture_block:
                                                            DefaultImage:=Pet18Texture
                                                            DefaultDesiredSize:=vector2{X:=65.0,Y:=65.0}
                                                        Padding:=margin{Left:=5.0,Right:=5.0}
                                                    #Pet3
                                                    stack_box_slot:
                                                        Widget:=texture_block:
                                                            DefaultImage:=Pet19Texture
                                                            DefaultDesiredSize:=vector2{X:=65.0,Y:=65.0}
                                                        Padding:=margin{Left:=5.0,Right:=5.0}
                                                    #Pet4
                                                    stack_box_slot:
                                                        Widget:=texture_block:
                                                            DefaultImage:=Pet20Texture
                                                            DefaultDesiredSize:=vector2{X:=65.0,Y:=65.0}
                                                        Padding:=margin{Left:=5.0,Right:=5.0}
                                            HorizontalAlignment:=horizontal_alignment.Left
                                            VerticalAlignment:=vertical_alignment.Top
                                            Padding:=margin{Left:=30.0,Right:=30.0,Bottom:=5.0}

                                        stack_box_slot:
                                            #Sub Stackbox 6 BLANK SPACE
                                            Widget:=stack_box:
                                                Orientation:=orientation.Horizontal
                                                Slots:=array:
                                                    #Pet1
                                                    stack_box_slot:
                                                        Widget:=texture_block:
                                                            DefaultImage:=HUD_PetDex.empty_image
                                                            DefaultDesiredSize:=vector2{X:=50.0,Y:=50.0}
                                                        Padding:=margin{Left:=5.0,Right:=5.0}
                                                    #Pet2
                                                    stack_box_slot:
                                                        Widget:=texture_block:
                                                            DefaultImage:=HUD_PetDex.empty_image
                                                            DefaultDesiredSize:=vector2{X:=50.0,Y:=50.0}
                                                        Padding:=margin{Left:=5.0,Right:=5.0}
                                                    #Pet3
                                                    stack_box_slot:
                                                        Widget:=texture_block:
                                                            DefaultImage:=HUD_PetDex.empty_image
                                                            DefaultDesiredSize:=vector2{X:=50.0,Y:=50.0}
                                                        Padding:=margin{Left:=5.0,Right:=5.0}
                                                    #Pet4
                                                    stack_box_slot:
                                                        Widget:=texture_block:
                                                            DefaultImage:=HUD_PetDex.empty_image
                                                            DefaultDesiredSize:=vector2{X:=50.0,Y:=50.0}
                                                        Padding:=margin{Left:=5.0,Right:=5.0}
                                            HorizontalAlignment:=horizontal_alignment.Left
                                            VerticalAlignment:=vertical_alignment.Top
                                            Padding:=margin{Left:=30.0,Right:=30.0,Bottom:=5.0}

                                        stack_box_slot:
                                            #Sub Stackbox 7
                                            Widget:=stack_box:
                                                Orientation:=orientation.Horizontal
                                                Slots:=array:
                                                    #Pet1
                                                    stack_box_slot:
                                                        Widget:=texture_block:
                                                            DefaultImage:=Pet21Texture
                                                            DefaultDesiredSize:=vector2{X:=65.0,Y:=65.0}
                                                        Padding:=margin{Left:=5.0,Right:=5.0}
                                                    #Pet2
                                                    stack_box_slot:
                                                        Widget:=texture_block:
                                                            DefaultImage:=Pet22Texture
                                                            DefaultDesiredSize:=vector2{X:=65.0,Y:=65.0}
                                                        Padding:=margin{Left:=5.0,Right:=5.0}
                                                    #Pet3
                                                    stack_box_slot:
                                                        Widget:=texture_block:
                                                            DefaultImage:=Pet23Texture
                                                            DefaultDesiredSize:=vector2{X:=65.0,Y:=65.0}
                                                        Padding:=margin{Left:=5.0,Right:=5.0}
                                                    #Pet4
                                                    stack_box_slot:
                                                        Widget:=texture_block:
                                                            DefaultImage:=Pet24Texture
                                                            DefaultDesiredSize:=vector2{X:=65.0,Y:=65.0}
                                                        Padding:=margin{Left:=5.0,Right:=5.0}
                                            HorizontalAlignment:=horizontal_alignment.Left
                                            VerticalAlignment:=vertical_alignment.Top
                                            Padding:=margin{Left:=30.0,Right:=30.0,Bottom:=5.0}

                                        stack_box_slot:
                                            #Sub Stackbox 8
                                            Widget:=stack_box:
                                                Orientation:=orientation.Horizontal
                                                Slots:=array:
                                                    #Pet1
                                                    stack_box_slot:
                                                        Widget:=texture_block:
                                                            DefaultImage:=Pet25Texture
                                                            DefaultDesiredSize:=vector2{X:=65.0,Y:=65.0}
                                                        Padding:=margin{Left:=5.0,Right:=5.0}
                                                    #Pet2
                                                    stack_box_slot:
                                                        Widget:=texture_block:
                                                            DefaultImage:=Pet26Texture
                                                            DefaultDesiredSize:=vector2{X:=65.0,Y:=65.0}
                                                        Padding:=margin{Left:=5.0,Right:=5.0}
                                                    #Pet3
                                                    stack_box_slot:
                                                        Widget:=texture_block:
                                                            DefaultImage:=Pet27Texture
                                                            DefaultDesiredSize:=vector2{X:=65.0,Y:=65.0}
                                                        Padding:=margin{Left:=5.0,Right:=5.0}
                                                    #Pet4
                                                    stack_box_slot:
                                                        Widget:=texture_block:
                                                            DefaultImage:=Pet28Texture
                                                            DefaultDesiredSize:=vector2{X:=65.0,Y:=65.0}
                                                        Padding:=margin{Left:=5.0,Right:=5.0}
                                            HorizontalAlignment:=horizontal_alignment.Left
                                            VerticalAlignment:=vertical_alignment.Top
                                            Padding:=margin{Left:=30.0,Right:=30.0,Bottom:=5.0}

                                        stack_box_slot:
                                            #Sub Stackbox 9
                                            Widget:=stack_box:
                                                Orientation:=orientation.Horizontal
                                                Slots:=array:
                                                    #Pet1
                                                    stack_box_slot:
                                                        Widget:=texture_block:
                                                            DefaultImage:=Pet29Texture
                                                            DefaultDesiredSize:=vector2{X:=65.0,Y:=65.0}
                                                        Padding:=margin{Left:=5.0,Right:=5.0}
                                                    #Pet2
                                                    stack_box_slot:
                                                        Widget:=texture_block:
                                                            DefaultImage:=Pet30Texture
                                                            DefaultDesiredSize:=vector2{X:=65.0,Y:=65.0}
                                                        Padding:=margin{Left:=5.0,Right:=5.0}
                                                    #Pet3
                                                    stack_box_slot:
                                                        Widget:=texture_block:
                                                            DefaultImage:=Pet31Texture
                                                            DefaultDesiredSize:=vector2{X:=65.0,Y:=65.0}
                                                        Padding:=margin{Left:=5.0,Right:=5.0}
                                                    #Pet4
                                                    stack_box_slot:
                                                        Widget:=texture_block:
                                                            DefaultImage:=Pet32Texture
                                                            DefaultDesiredSize:=vector2{X:=65.0,Y:=65.0}
                                                        Padding:=margin{Left:=5.0,Right:=5.0}
                                            HorizontalAlignment:=horizontal_alignment.Left
                                            VerticalAlignment:=vertical_alignment.Top
                                            Padding:=margin{Left:=30.0,Right:=30.0,Bottom:=5.0}

                                        stack_box_slot:
                                            #Sub Stackbox 10
                                            Widget:=stack_box:
                                                Orientation:=orientation.Horizontal
                                                Slots:=array:
                                                    #Pet1
                                                    stack_box_slot:
                                                        Widget:=texture_block:
                                                            DefaultImage:=Pet33Texture
                                                            DefaultDesiredSize:=vector2{X:=65.0,Y:=65.0}
                                                        Padding:=margin{Left:=5.0,Right:=5.0}
                                                    #Pet2
                                                    stack_box_slot:
                                                        Widget:=texture_block:
                                                            DefaultImage:=Pet34Texture
                                                            DefaultDesiredSize:=vector2{X:=65.0,Y:=65.0}
                                                        Padding:=margin{Left:=5.0,Right:=5.0}
                                                    #Pet3
                                                    stack_box_slot:
                                                        Widget:=texture_block:
                                                            DefaultImage:=Pet35Texture
                                                            DefaultDesiredSize:=vector2{X:=65.0,Y:=65.0}
                                                        Padding:=margin{Left:=5.0,Right:=5.0}
                                                    #Pet4
                                                    stack_box_slot:
                                                        Widget:=texture_block:
                                                            DefaultImage:=Pet36Texture
                                                            DefaultDesiredSize:=vector2{X:=65.0,Y:=65.0}
                                                        Padding:=margin{Left:=5.0,Right:=5.0}
                                            HorizontalAlignment:=horizontal_alignment.Left
                                            VerticalAlignment:=vertical_alignment.Top
                                            Padding:=margin{Left:=30.0,Right:=30.0,Bottom:=5.0}

                                        stack_box_slot:
                                            #Sub Stackbox 11
                                            Widget:=stack_box:
                                                Orientation:=orientation.Horizontal
                                                Slots:=array:
                                                    #Pet1
                                                    stack_box_slot:
                                                        Widget:=texture_block:
                                                            DefaultImage:=Pet37Texture
                                                            DefaultDesiredSize:=vector2{X:=65.0,Y:=65.0}
                                                        Padding:=margin{Left:=5.0,Right:=5.0}
                                                    #Pet2
                                                    stack_box_slot:
                                                        Widget:=texture_block:
                                                            DefaultImage:=Pet38Texture
                                                            DefaultDesiredSize:=vector2{X:=65.0,Y:=65.0}
                                                        Padding:=margin{Left:=5.0,Right:=5.0}
                                                    #Pet3
                                                    stack_box_slot:
                                                        Widget:=texture_block:
                                                            DefaultImage:=Pet39Texture
                                                            DefaultDesiredSize:=vector2{X:=65.0,Y:=65.0}
                                                        Padding:=margin{Left:=5.0,Right:=5.0}
                                                    #Pet4
                                                    stack_box_slot:
                                                        Widget:=texture_block:
                                                            DefaultImage:=Pet40Texture
                                                            DefaultDesiredSize:=vector2{X:=65.0,Y:=65.0}
                                                        Padding:=margin{Left:=5.0,Right:=5.0}
                                            HorizontalAlignment:=horizontal_alignment.Left
                                            VerticalAlignment:=vertical_alignment.Top
                                            Padding:=margin{Left:=30.0,Right:=30.0,Bottom:=5.0}

                                Padding:=margin{Left:=10.0,Top:=190.0}
                                HorizontalAlignment:=horizontal_alignment.Left
                                VerticalAlignment:=vertical_alignment.Top

                            #Big Stackbox 2
                            overlay_slot:
                                Widget:=stack_box:
                                    Orientation:=orientation.Vertical
                                    Slots:=array:
                                        stack_box_slot:
                                            #Sub Stackbox 1
                                            Widget:=stack_box:
                                                Orientation:=orientation.Horizontal
                                                Slots:=array:
                                                    #Pet1
                                                    stack_box_slot:
                                                        Widget:=texture_block:
                                                            DefaultImage:=Pet41Texture
                                                            DefaultDesiredSize:=vector2{X:=65.0,Y:=65.0}
                                                        Padding:=margin{Left:=5.0,Right:=5.0}
                                                    #Pet2
                                                    stack_box_slot:
                                                        Widget:=texture_block:
                                                            DefaultImage:=Pet42Texture
                                                            DefaultDesiredSize:=vector2{X:=65.0,Y:=65.0}
                                                        Padding:=margin{Left:=5.0,Right:=5.0}
                                                    #Pet3
                                                    stack_box_slot:
                                                        Widget:=texture_block:
                                                            DefaultImage:=Pet43Texture
                                                            DefaultDesiredSize:=vector2{X:=65.0,Y:=65.0}
                                                        Padding:=margin{Left:=5.0,Right:=5.0}
                                                    #Pet4
                                                    stack_box_slot:
                                                        Widget:=texture_block:
                                                            DefaultImage:=Pet44Texture
                                                            DefaultDesiredSize:=vector2{X:=65.0,Y:=65.0}
                                                        Padding:=margin{Left:=5.0,Right:=5.0}
                                            HorizontalAlignment:=horizontal_alignment.Left
                                            VerticalAlignment:=vertical_alignment.Top
                                            Padding:=margin{Left:=30.0,Right:=30.0,Bottom:=5.0}
                                        stack_box_slot:
                                            #Sub Stackbox 2
                                            Widget:=stack_box:
                                                Orientation:=orientation.Horizontal
                                                Slots:=array:
                                                    #Pet1
                                                    stack_box_slot:
                                                        Widget:=texture_block:
                                                            DefaultImage:=Pet45Texture
                                                            DefaultDesiredSize:=vector2{X:=65.0,Y:=65.0}
                                                        Padding:=margin{Left:=5.0,Right:=5.0}
                                                    #Pet2
                                                    stack_box_slot:
                                                        Widget:=texture_block:
                                                            DefaultImage:=Pet46Texture
                                                            DefaultDesiredSize:=vector2{X:=65.0,Y:=65.0}
                                                        Padding:=margin{Left:=5.0,Right:=5.0}
                                                    #Pet3
                                                    stack_box_slot:
                                                        Widget:=texture_block:
                                                            DefaultImage:=Pet47Texture
                                                            DefaultDesiredSize:=vector2{X:=65.0,Y:=65.0}
                                                        Padding:=margin{Left:=5.0,Right:=5.0}
                                                    #Pet4
                                                    stack_box_slot:
                                                        Widget:=texture_block:
                                                            DefaultImage:=Pet48Texture
                                                            DefaultDesiredSize:=vector2{X:=65.0,Y:=65.0}
                                                        Padding:=margin{Left:=5.0,Right:=5.0}
                                            HorizontalAlignment:=horizontal_alignment.Left
                                            VerticalAlignment:=vertical_alignment.Top
                                            Padding:=margin{Left:=30.0,Right:=30.0,Bottom:=5.0}

                                        stack_box_slot:
                                            #Sub Stackbox 3
                                            Widget:=stack_box:
                                                Orientation:=orientation.Horizontal
                                                Slots:=array:
                                                    #Pet1
                                                    stack_box_slot:
                                                        Widget:=texture_block:
                                                            DefaultImage:=Pet49Texture
                                                            DefaultDesiredSize:=vector2{X:=65.0,Y:=65.0}
                                                        Padding:=margin{Left:=5.0,Right:=5.0}
                                                    #Pet2
                                                    stack_box_slot:
                                                        Widget:=texture_block:
                                                            DefaultImage:=Pet50Texture
                                                            DefaultDesiredSize:=vector2{X:=65.0,Y:=65.0}
                                                        Padding:=margin{Left:=5.0,Right:=5.0}
                                                    #Pet3
                                                    stack_box_slot:
                                                        Widget:=texture_block:
                                                            DefaultImage:=Pet51Texture
                                                            DefaultDesiredSize:=vector2{X:=65.0,Y:=65.0}
                                                        Padding:=margin{Left:=5.0,Right:=5.0}
                                                    #Pet4
                                                    stack_box_slot:
                                                        Widget:=texture_block:
                                                            DefaultImage:=Pet52Texture
                                                            DefaultDesiredSize:=vector2{X:=65.0,Y:=65.0}
                                                        Padding:=margin{Left:=5.0,Right:=5.0}
                                            HorizontalAlignment:=horizontal_alignment.Left
                                            VerticalAlignment:=vertical_alignment.Top
                                            Padding:=margin{Left:=30.0,Right:=30.0,Bottom:=5.0}

                                        stack_box_slot:
                                            #Sub Stackbox 4
                                            Widget:=stack_box:
                                                Orientation:=orientation.Horizontal
                                                Slots:=array:
                                                    #Pet1
                                                    stack_box_slot:
                                                        Widget:=texture_block:
                                                            DefaultImage:=Pet53Texture
                                                            DefaultDesiredSize:=vector2{X:=65.0,Y:=65.0}
                                                        Padding:=margin{Left:=5.0,Right:=5.0}
                                                    #Pet2
                                                    stack_box_slot:
                                                        Widget:=texture_block:
                                                            DefaultImage:=Pet54Texture
                                                            DefaultDesiredSize:=vector2{X:=65.0,Y:=65.0}
                                                        Padding:=margin{Left:=5.0,Right:=5.0}
                                                    #Pet3
                                                    stack_box_slot:
                                                        Widget:=texture_block:
                                                            DefaultImage:=Pet55Texture
                                                            DefaultDesiredSize:=vector2{X:=65.0,Y:=65.0}
                                                        Padding:=margin{Left:=5.0,Right:=5.0}
                                                    #Pet4
                                                    stack_box_slot:
                                                        Widget:=texture_block:
                                                            DefaultImage:=Pet56Texture
                                                            DefaultDesiredSize:=vector2{X:=65.0,Y:=65.0}
                                                        Padding:=margin{Left:=5.0,Right:=5.0}
                                            HorizontalAlignment:=horizontal_alignment.Left
                                            VerticalAlignment:=vertical_alignment.Top
                                            Padding:=margin{Left:=30.0,Right:=30.0,Bottom:=5.0}

                                        stack_box_slot:
                                            #Sub Stackbox 5
                                            Widget:=stack_box:
                                                Orientation:=orientation.Horizontal
                                                Slots:=array:
                                                    #Pet1
                                                    stack_box_slot:
                                                        Widget:=texture_block:
                                                            DefaultImage:=Pet57Texture
                                                            DefaultDesiredSize:=vector2{X:=65.0,Y:=65.0}
                                                        Padding:=margin{Left:=5.0,Right:=5.0}
                                                    #Pet2
                                                    stack_box_slot:
                                                        Widget:=texture_block:
                                                            DefaultImage:=Pet58Texture
                                                            DefaultDesiredSize:=vector2{X:=65.0,Y:=65.0}
                                                        Padding:=margin{Left:=5.0,Right:=5.0}
                                                    #Pet3
                                                    stack_box_slot:
                                                        Widget:=texture_block:
                                                            DefaultImage:=Pet59Texture
                                                            DefaultDesiredSize:=vector2{X:=65.0,Y:=65.0}
                                                        Padding:=margin{Left:=5.0,Right:=5.0}
                                                    #Pet4
                                                    stack_box_slot:
                                                        Widget:=texture_block:
                                                            DefaultImage:=Pet60Texture
                                                            DefaultDesiredSize:=vector2{X:=65.0,Y:=65.0}
                                                        Padding:=margin{Left:=5.0,Right:=5.0}
                                            HorizontalAlignment:=horizontal_alignment.Left
                                            VerticalAlignment:=vertical_alignment.Top
                                            Padding:=margin{Left:=30.0,Right:=30.0,Bottom:=5.0}

                                        stack_box_slot:
                                            #Sub Stackbox 6 BLANK SPACE
                                            Widget:=stack_box:
                                                Orientation:=orientation.Horizontal
                                                Slots:=array:
                                                    #Pet1
                                                    stack_box_slot:
                                                        Widget:=texture_block:
                                                            DefaultImage:=HUD_PetDex.empty_image
                                                            DefaultDesiredSize:=vector2{X:=50.0,Y:=50.0}
                                                        Padding:=margin{Left:=5.0,Right:=5.0}
                                                    #Pet2
                                                    stack_box_slot:
                                                        Widget:=texture_block:
                                                            DefaultImage:=HUD_PetDex.empty_image
                                                            DefaultDesiredSize:=vector2{X:=50.0,Y:=50.0}
                                                        Padding:=margin{Left:=5.0,Right:=5.0}
                                                    #Pet3
                                                    stack_box_slot:
                                                        Widget:=texture_block:
                                                            DefaultImage:=HUD_PetDex.empty_image
                                                            DefaultDesiredSize:=vector2{X:=50.0,Y:=50.0}
                                                        Padding:=margin{Left:=5.0,Right:=5.0}
                                                    #Pet4
                                                    stack_box_slot:
                                                        Widget:=texture_block:
                                                            DefaultImage:=HUD_PetDex.empty_image
                                                            DefaultDesiredSize:=vector2{X:=50.0,Y:=50.0}
                                                        Padding:=margin{Left:=5.0,Right:=5.0}
                                            HorizontalAlignment:=horizontal_alignment.Left
                                            VerticalAlignment:=vertical_alignment.Top
                                            Padding:=margin{Left:=30.0,Right:=30.0,Bottom:=5.0}

                                        stack_box_slot:
                                            #Sub Stackbox 7
                                            Widget:=stack_box:
                                                Orientation:=orientation.Horizontal
                                                Slots:=array:
                                                    #Pet1
                                                    stack_box_slot:
                                                        Widget:=texture_block:
                                                            DefaultImage:=Pet61Texture
                                                            DefaultDesiredSize:=vector2{X:=65.0,Y:=65.0}
                                                        Padding:=margin{Left:=5.0,Right:=5.0}
                                                    #Pet2
                                                    stack_box_slot:
                                                        Widget:=texture_block:
                                                            DefaultImage:=Pet62Texture
                                                            DefaultDesiredSize:=vector2{X:=65.0,Y:=65.0}
                                                        Padding:=margin{Left:=5.0,Right:=5.0}
                                                    #Pet3
                                                    stack_box_slot:
                                                        Widget:=texture_block:
                                                            DefaultImage:=Pet63Texture
                                                            DefaultDesiredSize:=vector2{X:=65.0,Y:=65.0}
                                                        Padding:=margin{Left:=5.0,Right:=5.0}
                                                    #Pet4
                                                    stack_box_slot:
                                                        Widget:=texture_block:
                                                            DefaultImage:=Pet64Texture
                                                            DefaultDesiredSize:=vector2{X:=65.0,Y:=65.0}
                                                        Padding:=margin{Left:=5.0,Right:=5.0}
                                            HorizontalAlignment:=horizontal_alignment.Left
                                            VerticalAlignment:=vertical_alignment.Top
                                            Padding:=margin{Left:=30.0,Right:=30.0,Bottom:=5.0}

                                        stack_box_slot:
                                            #Sub Stackbox 8
                                            Widget:=stack_box:
                                                Orientation:=orientation.Horizontal
                                                Slots:=array:
                                                    #Pet1
                                                    stack_box_slot:
                                                        Widget:=texture_block:
                                                            DefaultImage:=Pet65Texture
                                                            DefaultDesiredSize:=vector2{X:=65.0,Y:=65.0}
                                                        Padding:=margin{Left:=5.0,Right:=5.0}
                                                    #Pet2
                                                    stack_box_slot:
                                                        Widget:=texture_block:
                                                            DefaultImage:=Pet66Texture
                                                            DefaultDesiredSize:=vector2{X:=65.0,Y:=65.0}
                                                        Padding:=margin{Left:=5.0,Right:=5.0}
                                                    #Pet3
                                                    stack_box_slot:
                                                        Widget:=texture_block:
                                                            DefaultImage:=Pet67Texture
                                                            DefaultDesiredSize:=vector2{X:=65.0,Y:=65.0}
                                                        Padding:=margin{Left:=5.0,Right:=5.0}
                                                    #Pet4
                                                    stack_box_slot:
                                                        Widget:=texture_block:
                                                            DefaultImage:=Pet68Texture
                                                            DefaultDesiredSize:=vector2{X:=65.0,Y:=65.0}
                                                        Padding:=margin{Left:=5.0,Right:=5.0}
                                            HorizontalAlignment:=horizontal_alignment.Left
                                            VerticalAlignment:=vertical_alignment.Top
                                            Padding:=margin{Left:=30.0,Right:=30.0,Bottom:=5.0}

                                        stack_box_slot:
                                            #Sub Stackbox 9
                                            Widget:=stack_box:
                                                Orientation:=orientation.Horizontal
                                                Slots:=array:
                                                    #Pet1
                                                    stack_box_slot:
                                                        Widget:=texture_block:
                                                            DefaultImage:=Pet69Texture
                                                            DefaultDesiredSize:=vector2{X:=65.0,Y:=65.0}
                                                        Padding:=margin{Left:=5.0,Right:=5.0}
                                                    #Pet2
                                                    stack_box_slot:
                                                        Widget:=texture_block:
                                                            DefaultImage:=Pet70Texture
                                                            DefaultDesiredSize:=vector2{X:=65.0,Y:=65.0}
                                                        Padding:=margin{Left:=5.0,Right:=5.0}
                                                    #Pet3
                                                    stack_box_slot:
                                                        Widget:=texture_block:
                                                            DefaultImage:=Pet71Texture
                                                            DefaultDesiredSize:=vector2{X:=65.0,Y:=65.0}
                                                        Padding:=margin{Left:=5.0,Right:=5.0}
                                                    #Pet4
                                                    stack_box_slot:
                                                        Widget:=texture_block:
                                                            DefaultImage:=Pet72Texture
                                                            DefaultDesiredSize:=vector2{X:=65.0,Y:=65.0}
                                                        Padding:=margin{Left:=5.0,Right:=5.0}
                                            HorizontalAlignment:=horizontal_alignment.Left
                                            VerticalAlignment:=vertical_alignment.Top
                                            Padding:=margin{Left:=30.0,Right:=30.0,Bottom:=5.0}

                                        stack_box_slot:
                                            #Sub Stackbox 10
                                            Widget:=stack_box:
                                                Orientation:=orientation.Horizontal
                                                Slots:=array:
                                                    #Pet1
                                                    stack_box_slot:
                                                        Widget:=texture_block:
                                                            DefaultImage:=Pet73Texture
                                                            DefaultDesiredSize:=vector2{X:=65.0,Y:=65.0}
                                                        Padding:=margin{Left:=5.0,Right:=5.0}
                                                    #Pet2
                                                    stack_box_slot:
                                                        Widget:=texture_block:
                                                            DefaultImage:=Pet74Texture
                                                            DefaultDesiredSize:=vector2{X:=65.0,Y:=65.0}
                                                        Padding:=margin{Left:=5.0,Right:=5.0}
                                                    #Pet3
                                                    stack_box_slot:
                                                        Widget:=texture_block:
                                                            DefaultImage:=Pet75Texture
                                                            DefaultDesiredSize:=vector2{X:=65.0,Y:=65.0}
                                                        Padding:=margin{Left:=5.0,Right:=5.0}
                                                    #Pet4
                                                    stack_box_slot:
                                                        Widget:=texture_block:
                                                            DefaultImage:=Pet76Texture
                                                            DefaultDesiredSize:=vector2{X:=65.0,Y:=65.0}
                                                        Padding:=margin{Left:=5.0,Right:=5.0}
                                            HorizontalAlignment:=horizontal_alignment.Left
                                            VerticalAlignment:=vertical_alignment.Top
                                            Padding:=margin{Left:=30.0,Right:=30.0,Bottom:=5.0}

                                        stack_box_slot:
                                            #Sub Stackbox 11
                                            Widget:=stack_box:
                                                Orientation:=orientation.Horizontal
                                                Slots:=array:
                                                    #Pet1
                                                    stack_box_slot:
                                                        Widget:=texture_block:
                                                            DefaultImage:=Pet77Texture
                                                            DefaultDesiredSize:=vector2{X:=65.0,Y:=65.0}
                                                        Padding:=margin{Left:=5.0,Right:=5.0}
                                                    #Pet2
                                                    stack_box_slot:
                                                        Widget:=texture_block:
                                                            DefaultImage:=Pet78Texture
                                                            DefaultDesiredSize:=vector2{X:=65.0,Y:=65.0}
                                                        Padding:=margin{Left:=5.0,Right:=5.0}
                                                    #Pet3
                                                    stack_box_slot:
                                                        Widget:=texture_block:
                                                            DefaultImage:=Pet79Texture
                                                            DefaultDesiredSize:=vector2{X:=65.0,Y:=65.0}
                                                        Padding:=margin{Left:=5.0,Right:=5.0}
                                                    #Pet4
                                                    stack_box_slot:
                                                        Widget:=texture_block:
                                                            DefaultImage:=Pet80Texture
                                                            DefaultDesiredSize:=vector2{X:=65.0,Y:=65.0}
                                                        Padding:=margin{Left:=5.0,Right:=5.0}
                                            HorizontalAlignment:=horizontal_alignment.Left
                                            VerticalAlignment:=vertical_alignment.Top
                                            Padding:=margin{Left:=30.0,Right:=30.0,Bottom:=5.0}

                                Padding:=margin{Left:=450.0,Top:=190.0}
                                HorizontalAlignment:=horizontal_alignment.Left
                                VerticalAlignment:=vertical_alignment.Top

                            #Big Stackbox 3
                            overlay_slot:
                                Widget:=stack_box:
                                    Orientation:=orientation.Vertical
                                    Slots:=array:
                                        stack_box_slot:
                                            #Sub Stackbox 1
                                            Widget:=stack_box:
                                                Orientation:=orientation.Horizontal
                                                Slots:=array:
                                                    #Pet1
                                                    stack_box_slot:
                                                        Widget:=texture_block:
                                                            DefaultImage:=Pet81Texture
                                                            DefaultDesiredSize:=vector2{X:=65.0,Y:=65.0}
                                                        Padding:=margin{Left:=5.0,Right:=5.0}
                                                    #Pet2
                                                    stack_box_slot:
                                                        Widget:=texture_block:
                                                            DefaultImage:=Pet82Texture
                                                            DefaultDesiredSize:=vector2{X:=65.0,Y:=65.0}
                                                        Padding:=margin{Left:=5.0,Right:=5.0}
                                                    #Pet3
                                                    stack_box_slot:
                                                        Widget:=texture_block:
                                                            DefaultImage:=Pet83Texture
                                                            DefaultDesiredSize:=vector2{X:=65.0,Y:=65.0}
                                                        Padding:=margin{Left:=5.0,Right:=5.0}
                                                    #Pet4
                                                    stack_box_slot:
                                                        Widget:=texture_block:
                                                            DefaultImage:=Pet84Texture
                                                            DefaultDesiredSize:=vector2{X:=65.0,Y:=65.0}
                                                        Padding:=margin{Left:=5.0,Right:=5.0}
                                            HorizontalAlignment:=horizontal_alignment.Left
                                            VerticalAlignment:=vertical_alignment.Top
                                            Padding:=margin{Left:=30.0,Right:=30.0,Bottom:=5.0}
                                        stack_box_slot:
                                            #Sub Stackbox 2
                                            Widget:=stack_box:
                                                Orientation:=orientation.Horizontal
                                                Slots:=array:
                                                    #Pet1
                                                    stack_box_slot:
                                                        Widget:=texture_block:
                                                            DefaultImage:=Pet85Texture
                                                            DefaultDesiredSize:=vector2{X:=65.0,Y:=65.0}
                                                        Padding:=margin{Left:=5.0,Right:=5.0}
                                                    #Pet2
                                                    stack_box_slot:
                                                        Widget:=texture_block:
                                                            DefaultImage:=Pet86Texture
                                                            DefaultDesiredSize:=vector2{X:=65.0,Y:=65.0}
                                                        Padding:=margin{Left:=5.0,Right:=5.0}
                                                    #Pet3
                                                    stack_box_slot:
                                                        Widget:=texture_block:
                                                            DefaultImage:=Pet87Texture
                                                            DefaultDesiredSize:=vector2{X:=65.0,Y:=65.0}
                                                        Padding:=margin{Left:=5.0,Right:=5.0}
                                                    #Pet4
                                                    stack_box_slot:
                                                        Widget:=texture_block:
                                                            DefaultImage:=Pet88Texture
                                                            DefaultDesiredSize:=vector2{X:=65.0,Y:=65.0}
                                                        Padding:=margin{Left:=5.0,Right:=5.0}
                                            HorizontalAlignment:=horizontal_alignment.Left
                                            VerticalAlignment:=vertical_alignment.Top
                                            Padding:=margin{Left:=30.0,Right:=30.0,Bottom:=5.0}

                                        stack_box_slot:
                                            #Sub Stackbox 3
                                            Widget:=stack_box:
                                                Orientation:=orientation.Horizontal
                                                Slots:=array:
                                                    #Pet1
                                                    stack_box_slot:
                                                        Widget:=texture_block:
                                                            DefaultImage:=Pet89Texture
                                                            DefaultDesiredSize:=vector2{X:=65.0,Y:=65.0}
                                                        Padding:=margin{Left:=5.0,Right:=5.0}
                                                    #Pet2
                                                    stack_box_slot:
                                                        Widget:=texture_block:
                                                            DefaultImage:=Pet90Texture
                                                            DefaultDesiredSize:=vector2{X:=65.0,Y:=65.0}
                                                        Padding:=margin{Left:=5.0,Right:=5.0}
                                                    #Pet3
                                                    stack_box_slot:
                                                        Widget:=texture_block:
                                                            DefaultImage:=Pet91Texture
                                                            DefaultDesiredSize:=vector2{X:=65.0,Y:=65.0}
                                                        Padding:=margin{Left:=5.0,Right:=5.0}
                                                    #Pet4
                                                    stack_box_slot:
                                                        Widget:=texture_block:
                                                            DefaultImage:=Pet92Texture
                                                            DefaultDesiredSize:=vector2{X:=65.0,Y:=65.0}
                                                        Padding:=margin{Left:=5.0,Right:=5.0}
                                            HorizontalAlignment:=horizontal_alignment.Left
                                            VerticalAlignment:=vertical_alignment.Top
                                            Padding:=margin{Left:=30.0,Right:=30.0,Bottom:=5.0}

                                        stack_box_slot:
                                            #Sub Stackbox 4
                                            Widget:=stack_box:
                                                Orientation:=orientation.Horizontal
                                                Slots:=array:
                                                    #Pet1
                                                    stack_box_slot:
                                                        Widget:=texture_block:
                                                            DefaultImage:=Pet93Texture
                                                            DefaultDesiredSize:=vector2{X:=65.0,Y:=65.0}
                                                        Padding:=margin{Left:=5.0,Right:=5.0}
                                                    #Pet2
                                                    stack_box_slot:
                                                        Widget:=texture_block:
                                                            DefaultImage:=Pet94Texture
                                                            DefaultDesiredSize:=vector2{X:=65.0,Y:=65.0}
                                                        Padding:=margin{Left:=5.0,Right:=5.0}
                                                    #Pet3
                                                    stack_box_slot:
                                                        Widget:=texture_block:
                                                            DefaultImage:=Pet95Texture
                                                            DefaultDesiredSize:=vector2{X:=65.0,Y:=65.0}
                                                        Padding:=margin{Left:=5.0,Right:=5.0}
                                                    #Pet4
                                                    stack_box_slot:
                                                        Widget:=texture_block:
                                                            DefaultImage:=Pet96Texture
                                                            DefaultDesiredSize:=vector2{X:=65.0,Y:=65.0}
                                                        Padding:=margin{Left:=5.0,Right:=5.0}
                                            HorizontalAlignment:=horizontal_alignment.Left
                                            VerticalAlignment:=vertical_alignment.Top
                                            Padding:=margin{Left:=30.0,Right:=30.0,Bottom:=5.0}

                                        stack_box_slot:
                                            #Sub Stackbox 5
                                            Widget:=stack_box:
                                                Orientation:=orientation.Horizontal
                                                Slots:=array:
                                                    #Pet1
                                                    stack_box_slot:
                                                        Widget:=texture_block:
                                                            DefaultImage:=Pet97Texture
                                                            DefaultDesiredSize:=vector2{X:=65.0,Y:=65.0}
                                                        Padding:=margin{Left:=5.0,Right:=5.0}
                                                    #Pet2
                                                    stack_box_slot:
                                                        Widget:=texture_block:
                                                            DefaultImage:=Pet98Texture
                                                            DefaultDesiredSize:=vector2{X:=65.0,Y:=65.0}
                                                        Padding:=margin{Left:=5.0,Right:=5.0}
                                                    #Pet3
                                                    stack_box_slot:
                                                        Widget:=texture_block:
                                                            DefaultImage:=Pet99Texture
                                                            DefaultDesiredSize:=vector2{X:=65.0,Y:=65.0}
                                                        Padding:=margin{Left:=5.0,Right:=5.0}
                                                    #Pet4
                                                    stack_box_slot:
                                                        Widget:=texture_block:
                                                            DefaultImage:=Pet100Texture
                                                            DefaultDesiredSize:=vector2{X:=65.0,Y:=65.0}
                                                        Padding:=margin{Left:=5.0,Right:=5.0}
                                            HorizontalAlignment:=horizontal_alignment.Left
                                            VerticalAlignment:=vertical_alignment.Top
                                            Padding:=margin{Left:=30.0,Right:=30.0,Bottom:=5.0}

                                        stack_box_slot:
                                            #Sub Stackbox 6 BLANK SPACE
                                            Widget:=stack_box:
                                                Orientation:=orientation.Horizontal
                                                Slots:=array:
                                                    #Pet1
                                                    stack_box_slot:
                                                        Widget:=texture_block:
                                                            DefaultImage:=HUD_PetDex.empty_image
                                                            DefaultDesiredSize:=vector2{X:=50.0,Y:=50.0}
                                                        Padding:=margin{Left:=5.0,Right:=5.0}
                                                    #Pet2
                                                    stack_box_slot:
                                                        Widget:=texture_block:
                                                            DefaultImage:=HUD_PetDex.empty_image
                                                            DefaultDesiredSize:=vector2{X:=50.0,Y:=50.0}
                                                        Padding:=margin{Left:=5.0,Right:=5.0}
                                                    #Pet3
                                                    stack_box_slot:
                                                        Widget:=texture_block:
                                                            DefaultImage:=HUD_PetDex.empty_image
                                                            DefaultDesiredSize:=vector2{X:=50.0,Y:=50.0}
                                                        Padding:=margin{Left:=5.0,Right:=5.0}
                                                    #Pet4
                                                    stack_box_slot:
                                                        Widget:=texture_block:
                                                            DefaultImage:=HUD_PetDex.empty_image
                                                            DefaultDesiredSize:=vector2{X:=50.0,Y:=50.0}
                                                        Padding:=margin{Left:=5.0,Right:=5.0}
                                            HorizontalAlignment:=horizontal_alignment.Left
                                            VerticalAlignment:=vertical_alignment.Top
                                            Padding:=margin{Left:=30.0,Right:=30.0,Bottom:=5.0}

                                        stack_box_slot:
                                            #Sub Stackbox 7
                                            Widget:=stack_box:
                                                Orientation:=orientation.Horizontal
                                                Slots:=array:
                                                    #Pet1
                                                    stack_box_slot:
                                                        Widget:=texture_block:
                                                            DefaultImage:=Pet101Texture
                                                            DefaultDesiredSize:=vector2{X:=65.0,Y:=65.0}
                                                        Padding:=margin{Left:=5.0,Right:=5.0}
                                                    #Pet2
                                                    stack_box_slot:
                                                        Widget:=texture_block:
                                                            DefaultImage:=Pet102Texture
                                                            DefaultDesiredSize:=vector2{X:=65.0,Y:=65.0}
                                                        Padding:=margin{Left:=5.0,Right:=5.0}
                                                    #Pet3
                                                    stack_box_slot:
                                                        Widget:=texture_block:
                                                            DefaultImage:=Pet103Texture
                                                            DefaultDesiredSize:=vector2{X:=65.0,Y:=65.0}
                                                        Padding:=margin{Left:=5.0,Right:=5.0}
                                                    #Pet4
                                                    stack_box_slot:
                                                        Widget:=texture_block:
                                                            DefaultImage:=Pet104Texture
                                                            DefaultDesiredSize:=vector2{X:=65.0,Y:=65.0}
                                                        Padding:=margin{Left:=5.0,Right:=5.0}
                                            HorizontalAlignment:=horizontal_alignment.Left
                                            VerticalAlignment:=vertical_alignment.Top
                                            Padding:=margin{Left:=30.0,Right:=30.0,Bottom:=5.0}

                                        stack_box_slot:
                                            #Sub Stackbox 8
                                            Widget:=stack_box:
                                                Orientation:=orientation.Horizontal
                                                Slots:=array:
                                                    #Pet1
                                                    stack_box_slot:
                                                        Widget:=texture_block:
                                                            DefaultImage:=Pet105Texture
                                                            DefaultDesiredSize:=vector2{X:=65.0,Y:=65.0}
                                                        Padding:=margin{Left:=5.0,Right:=5.0}
                                                    #Pet2
                                                    stack_box_slot:
                                                        Widget:=texture_block:
                                                            DefaultImage:=Pet106Texture
                                                            DefaultDesiredSize:=vector2{X:=65.0,Y:=65.0}
                                                        Padding:=margin{Left:=5.0,Right:=5.0}
                                                    #Pet3
                                                    stack_box_slot:
                                                        Widget:=texture_block:
                                                            DefaultImage:=Pet107Texture
                                                            DefaultDesiredSize:=vector2{X:=65.0,Y:=65.0}
                                                        Padding:=margin{Left:=5.0,Right:=5.0}
                                                    #Pet4
                                                    stack_box_slot:
                                                        Widget:=texture_block:
                                                            DefaultImage:=Pet108Texture
                                                            DefaultDesiredSize:=vector2{X:=65.0,Y:=65.0}
                                                        Padding:=margin{Left:=5.0,Right:=5.0}
                                            HorizontalAlignment:=horizontal_alignment.Left
                                            VerticalAlignment:=vertical_alignment.Top
                                            Padding:=margin{Left:=30.0,Right:=30.0,Bottom:=5.0}

                                        stack_box_slot:
                                            #Sub Stackbox 9
                                            Widget:=stack_box:
                                                Orientation:=orientation.Horizontal
                                                Slots:=array:
                                                    #Pet1
                                                    stack_box_slot:
                                                        Widget:=texture_block:
                                                            DefaultImage:=Pet109Texture
                                                            DefaultDesiredSize:=vector2{X:=65.0,Y:=65.0}
                                                        Padding:=margin{Left:=5.0,Right:=5.0}
                                                    #Pet2
                                                    stack_box_slot:
                                                        Widget:=texture_block:
                                                            DefaultImage:=Pet110Texture
                                                            DefaultDesiredSize:=vector2{X:=65.0,Y:=65.0}
                                                        Padding:=margin{Left:=5.0,Right:=5.0}
                                                    #Pet3
                                                    stack_box_slot:
                                                        Widget:=texture_block:
                                                            DefaultImage:=Pet111Texture
                                                            DefaultDesiredSize:=vector2{X:=65.0,Y:=65.0}
                                                        Padding:=margin{Left:=5.0,Right:=5.0}
                                                    #Pet4
                                                    stack_box_slot:
                                                        Widget:=texture_block:
                                                            DefaultImage:=Pet112Texture
                                                            DefaultDesiredSize:=vector2{X:=65.0,Y:=65.0}
                                                        Padding:=margin{Left:=5.0,Right:=5.0}
                                            HorizontalAlignment:=horizontal_alignment.Left
                                            VerticalAlignment:=vertical_alignment.Top
                                            Padding:=margin{Left:=30.0,Right:=30.0,Bottom:=5.0}

                                        stack_box_slot:
                                            #Sub Stackbox 10
                                            Widget:=stack_box:
                                                Orientation:=orientation.Horizontal
                                                Slots:=array:
                                                    #Pet1
                                                    stack_box_slot:
                                                        Widget:=texture_block:
                                                            DefaultImage:=Pet113Texture
                                                            DefaultDesiredSize:=vector2{X:=65.0,Y:=65.0}
                                                        Padding:=margin{Left:=5.0,Right:=5.0}
                                                    #Pet2
                                                    stack_box_slot:
                                                        Widget:=texture_block:
                                                            DefaultImage:=Pet114Texture
                                                            DefaultDesiredSize:=vector2{X:=65.0,Y:=65.0}
                                                        Padding:=margin{Left:=5.0,Right:=5.0}
                                                    #Pet3
                                                    stack_box_slot:
                                                        Widget:=texture_block:
                                                            DefaultImage:=Pet115Texture
                                                            DefaultDesiredSize:=vector2{X:=65.0,Y:=65.0}
                                                        Padding:=margin{Left:=5.0,Right:=5.0}
                                                    #Pet4
                                                    stack_box_slot:
                                                        Widget:=texture_block:
                                                            DefaultImage:=Pet116Texture
                                                            DefaultDesiredSize:=vector2{X:=65.0,Y:=65.0}
                                                        Padding:=margin{Left:=5.0,Right:=5.0}
                                            HorizontalAlignment:=horizontal_alignment.Left
                                            VerticalAlignment:=vertical_alignment.Top
                                            Padding:=margin{Left:=30.0,Right:=30.0,Bottom:=5.0}

                                        stack_box_slot:
                                            #Sub Stackbox 11
                                            Widget:=stack_box:
                                                Orientation:=orientation.Horizontal
                                                Slots:=array:
                                                    #Pet1
                                                    stack_box_slot:
                                                        Widget:=texture_block:
                                                            DefaultImage:=Pet117Texture
                                                            DefaultDesiredSize:=vector2{X:=65.0,Y:=65.0}
                                                        Padding:=margin{Left:=5.0,Right:=5.0}
                                                    #Pet2
                                                    stack_box_slot:
                                                        Widget:=texture_block:
                                                            DefaultImage:=Pet118Texture
                                                            DefaultDesiredSize:=vector2{X:=65.0,Y:=65.0}
                                                        Padding:=margin{Left:=5.0,Right:=5.0}
                                                    #Pet3
                                                    stack_box_slot:
                                                        Widget:=texture_block:
                                                            DefaultImage:=Pet119Texture
                                                            DefaultDesiredSize:=vector2{X:=65.0,Y:=65.0}
                                                        Padding:=margin{Left:=5.0,Right:=5.0}
                                                    #Pet4
                                                    stack_box_slot:
                                                        Widget:=texture_block:
                                                            DefaultImage:=Pet120Texture
                                                            DefaultDesiredSize:=vector2{X:=65.0,Y:=65.0}
                                                        Padding:=margin{Left:=5.0,Right:=5.0}
                                            HorizontalAlignment:=horizontal_alignment.Left
                                            VerticalAlignment:=vertical_alignment.Top
                                            Padding:=margin{Left:=30.0,Right:=30.0,Bottom:=5.0}

                                Padding:=margin{Left:=860.0,Top:=190.0}
                                HorizontalAlignment:=horizontal_alignment.Left
                                VerticalAlignment:=vertical_alignment.Top

                            #Secret Stackbox
                            overlay_slot:
                                Widget:=stack_box:
                                    Orientation:=orientation.Vertical
                                    Slots:=array:
                                        stack_box_slot:
                                            #Sub Stackbox 1
                                            Widget:=stack_box:
                                                Orientation:=orientation.Horizontal
                                                Slots:=array:
                                                    #Secret Pet1
                                                    stack_box_slot:
                                                        Widget:=texture_block:
                                                            DefaultImage:=Pet121Texture
                                                            DefaultDesiredSize:=vector2{X:=80.0,Y:=80.0}
                                                    #Secret Pet2
                                                    stack_box_slot:
                                                        Widget:=texture_block:
                                                            DefaultImage:=Pet122Texture
                                                            DefaultDesiredSize:=vector2{X:=80.0,Y:=80.0}
                                                    #Secret Pet3
                                                    stack_box_slot:
                                                        Widget:=texture_block:
                                                            DefaultImage:=Pet123Texture
                                                            DefaultDesiredSize:=vector2{X:=80.0,Y:=80.0}
                                                    #Secret Pet4
                                                    stack_box_slot:
                                                        Widget:=texture_block:
                                                            DefaultImage:=Pet124Texture
                                                            DefaultDesiredSize:=vector2{X:=80.0,Y:=80.0}
                                                    #Secret Pet5
                                                    stack_box_slot:
                                                        Widget:=texture_block:
                                                            DefaultImage:=Pet125Texture
                                                            DefaultDesiredSize:=vector2{X:=80.0,Y:=80.0}
                                            HorizontalAlignment:=horizontal_alignment.Left
                                            VerticalAlignment:=vertical_alignment.Top
                                            Padding:=margin{Bottom:=5.0}
                                        

                                Padding:=margin{Left:=820.0,Top:=965.0}
                                HorizontalAlignment:=horizontal_alignment.Left
                                VerticalAlignment:=vertical_alignment.Top

                            overlay_slot:
                                Widget:=TextNumberOfPets
                                HorizontalAlignment:=horizontal_alignment.Center
                                VerticalAlignment:=vertical_alignment.Bottom
                                Padding:=margin{Bottom:=15.0}
                                
        }
        
        set CanvasSlots += array
        {
            CanvasSlot2 := canvas_slot:
                Anchors := anchors{ Minimum := vector2{ X := 0.5, Y := 0.0 }, Maximum := vector2{ X := 0.5, Y := 0.0 } }
                Offsets := margin{ Top := 55.0, Left := 400.0, Right := 15.0, Bottom := 15.0 }
                Alignment := vector2{ X := 0.0, Y := 0.0 }
                SizeToContent := true
                Widget := button_regular{ DefaultText := StringToMessage("          ") }
        }
        
        if (ButtonRegular := button_regular[CanvasSlots[1].Widget]):
            ButtonRegular.OnClick().Subscribe(OnPageCloseClicked)
        
        Canvas : canvas = canvas:
                Slots := CanvasSlots

        return Canvas

    

    Update_Total_Pets():void=
        set Ownership_Total_Pets = Ownership_L1_Pet1 + Ownership_L1_Pet2*2 +Ownership_L1_Pet3*3 +Ownership_L1_Pet4*4+Ownership_L2_Pet1 +Ownership_L2_Pet2*2 +Ownership_L2_Pet3*3 +Ownership_L2_Pet4*4 +Ownership_L3_Pet1 +Ownership_L3_Pet2*2 +Ownership_L3_Pet3*3 +Ownership_L3_Pet4*4 +Ownership_L4_Pet1 +Ownership_L4_Pet2*2 +Ownership_L4_Pet3*3 +Ownership_L4_Pet4*4 +Ownership_L5_Pet1 +Ownership_L5_Pet2*2 +Ownership_L5_Pet3*3 +Ownership_L5_Pet4*4 +Ownership_L6_Pet1 +Ownership_L6_Pet2*2 +Ownership_L6_Pet3*3 +Ownership_L6_Pet4*4 +Ownership_L7_Pet1 +Ownership_L7_Pet2*2 +Ownership_L7_Pet3*3 +Ownership_L7_Pet4*4 +Ownership_L8_Pet1 +Ownership_L8_Pet2*2 +Ownership_L8_Pet3*3 +Ownership_L8_Pet4*4 +Ownership_L9_Pet1 +Ownership_L9_Pet2*2 +Ownership_L9_Pet3*3 +Ownership_L9_Pet4*4 +Ownership_L10_Pet1 +Ownership_L10_Pet2*2 +Ownership_L10_Pet3*3 +Ownership_L10_Pet4*4 +Ownership_L11_Pet1 +Ownership_L11_Pet2*2 +Ownership_L11_Pet3*3 +Ownership_L11_Pet4*4 +Ownership_L12_Pet1 +Ownership_L12_Pet2*2 +Ownership_L12_Pet3*3 +Ownership_L12_Pet4*4 +Ownership_L13_Pet1 +Ownership_L13_Pet2*2 +Ownership_L13_Pet3*3 +Ownership_L13_Pet4*4 +Ownership_L14_Pet1 +Ownership_L14_Pet2*2 +Ownership_L14_Pet3*3 +Ownership_L14_Pet4*4 +Ownership_L15_Pet1 +Ownership_L15_Pet2*2 +Ownership_L15_Pet3*3 +Ownership_L15_Pet4*4 +Ownership_L16_Pet1 +Ownership_L16_Pet2*2 +Ownership_L16_Pet3*3 +Ownership_L16_Pet4*4 +Ownership_L17_Pet1 +Ownership_L17_Pet2*2 +Ownership_L17_Pet3*3 +Ownership_L17_Pet4*4 +Ownership_L18_Pet1 +Ownership_L18_Pet2*2 +Ownership_L18_Pet3*3 +Ownership_L18_Pet4*4 +Ownership_L19_Pet1 +Ownership_L19_Pet2*2 +Ownership_L19_Pet3*3 +Ownership_L19_Pet4*4 +Ownership_L20_Pet1 +Ownership_L20_Pet2*2 +Ownership_L20_Pet3*3 +Ownership_L20_Pet4*4 +Ownership_L21_Pet1 +Ownership_L21_Pet2*2 +Ownership_L21_Pet3*3 +Ownership_L21_Pet4*4 +Ownership_L22_Pet1 +Ownership_L22_Pet2*2 +Ownership_L22_Pet3*3 +Ownership_L22_Pet4*4 +Ownership_L23_Pet1 +Ownership_L23_Pet2*2 +Ownership_L23_Pet3*3 +Ownership_L23_Pet4*4 +Ownership_L24_Pet1 +Ownership_L24_Pet2*2 +Ownership_L24_Pet3*3 +Ownership_L24_Pet4*4 +Ownership_L25_Pet1 +Ownership_L25_Pet2*2 +Ownership_L25_Pet3*3 +Ownership_L25_Pet4*4 +Ownership_L26_Pet1 +Ownership_L26_Pet2*2 +Ownership_L26_Pet3*3 +Ownership_L26_Pet4*4 +Ownership_L27_Pet1 +Ownership_L27_Pet2*2 +Ownership_L27_Pet3*3 +Ownership_L27_Pet4*4+Ownership_L28_Pet1 +Ownership_L28_Pet2*2 +Ownership_L28_Pet3*3 +Ownership_L28_Pet4*4 +Ownership_L29_Pet1 +Ownership_L29_Pet2*2 +Ownership_L29_Pet3*3 +Ownership_L29_Pet4*4 +Ownership_L30_Pet1 +Ownership_L30_Pet2*2 +Ownership_L30_Pet3*3 +Ownership_L30_Pet4*4 +Ownership_Secret_Pet1 +Ownership_Secret_Pet2 +Ownership_Secret_Pet3 +Ownership_Secret_Pet4 +Ownership_Secret_Pet5 
        Print("Total pets:{Ownership_Total_Pets}")
    
    Update_Pet_PowerL1(NewPower:int):void=
        set Pet_Power_ValueL1 = NewPower
        Update_Pet_Power()

    Update_Pet_PowerL2(NewPower:int):void=
        set Pet_Power_ValueL2 = NewPower
        Update_Pet_Power()

    Update_Pet_PowerL3(NewPower:int):void=
        set Pet_Power_ValueL3 = NewPower
        Update_Pet_Power()

    Update_Pet_PowerL4(NewPower:int):void=
        set Pet_Power_ValueL4 = NewPower
        Update_Pet_Power()

    Update_Pet_PowerL5(NewPower:int):void=
        set Pet_Power_ValueL5 = NewPower
        Update_Pet_Power()

    Update_Pet_PowerL6(NewPower:int):void=
        set Pet_Power_ValueL6 = NewPower
        Update_Pet_Power()

    Update_Pet_PowerL7(NewPower:int):void=
        set Pet_Power_ValueL7 = NewPower
        Update_Pet_Power()

    Update_Pet_PowerL8(NewPower:int):void=
        set Pet_Power_ValueL8 = NewPower
        Update_Pet_Power()

    Update_Pet_PowerL9(NewPower:int):void=
        set Pet_Power_ValueL9 = NewPower
        Update_Pet_Power()

    Update_Pet_PowerL10(NewPower:int):void=
        set Pet_Power_ValueL10 = NewPower
        Update_Pet_Power()

    Update_Pet_PowerL11(NewPower:int):void=
        set Pet_Power_ValueL11 = NewPower
        Update_Pet_Power()

    Update_Pet_PowerL12(NewPower:int):void=
        set Pet_Power_ValueL12 = NewPower
        Update_Pet_Power()

    Update_Pet_PowerL13(NewPower:int):void=
        set Pet_Power_ValueL13 = NewPower
        Update_Pet_Power()

    Update_Pet_PowerL14(NewPower:int):void=
        set Pet_Power_ValueL14 = NewPower
        Update_Pet_Power()

    Update_Pet_PowerL15(NewPower:int):void=
        set Pet_Power_ValueL15 = NewPower
        Update_Pet_Power()

    Update_Pet_PowerL16(NewPower:int):void=
        set Pet_Power_ValueL16 = NewPower
        Update_Pet_Power()

    Update_Pet_PowerL17(NewPower:int):void=
        set Pet_Power_ValueL17 = NewPower
        Update_Pet_Power()

    Update_Pet_PowerL18(NewPower:int):void=
        set Pet_Power_ValueL18 = NewPower
        Update_Pet_Power()

    Update_Pet_PowerL19(NewPower:int):void=
        set Pet_Power_ValueL19 = NewPower
        Update_Pet_Power()

    Update_Pet_PowerL20(NewPower:int):void=
        set Pet_Power_ValueL20 = NewPower
        Update_Pet_Power()
    
    Update_Pet_PowerL21(NewPower:int):void=
        set Pet_Power_ValueL21 = NewPower
        Update_Pet_Power()

    Update_Pet_PowerL22(NewPower:int):void=
        set Pet_Power_ValueL22 = NewPower
        Update_Pet_Power()

    Update_Pet_PowerL23(NewPower:int):void=
        set Pet_Power_ValueL23 = NewPower
        Update_Pet_Power()

    Update_Pet_PowerL24(NewPower:int):void=
        set Pet_Power_ValueL24 = NewPower
        Update_Pet_Power()

    Update_Pet_PowerL25(NewPower:int):void=
        set Pet_Power_ValueL25 = NewPower
        Update_Pet_Power()

    Update_Pet_PowerL26(NewPower:int):void=
        set Pet_Power_ValueL26 = NewPower
        Update_Pet_Power()

    Update_Pet_PowerL27(NewPower:int):void=
        set Pet_Power_ValueL27 = NewPower
        Update_Pet_Power()

    Update_Pet_PowerL28(NewPower:int):void=
        set Pet_Power_ValueL28 = NewPower
        Update_Pet_Power()

    Update_Pet_PowerL29(NewPower:int):void=
        set Pet_Power_ValueL29 = NewPower
        Update_Pet_Power()

    Update_Pet_PowerL30(NewPower:int):void=
        set Pet_Power_ValueL30 = NewPower
        Update_Pet_Power()
     
    #Number of Purchases on Eggs
    IncreaseButtonL1():void=
        set ButtonL1Pressed += 1 

    IncreaseButtonL2():void=
        set ButtonL2Pressed += 1 

    IncreaseButtonL3():void=
        set ButtonL3Pressed += 1 

    IncreaseButtonL4():void=
        set ButtonL4Pressed += 1 

    IncreaseButtonL5():void=
        set ButtonL5Pressed += 1 

    IncreaseButtonL6():void=
        set ButtonL6Pressed += 1 

    IncreaseButtonL7():void=
        set ButtonL7Pressed += 1 

    IncreaseButtonL8():void=
        set ButtonL8Pressed += 1 

    IncreaseButtonL9():void=
        set ButtonL9Pressed += 1 

    IncreaseButtonL10():void=
        set ButtonL10Pressed += 1 

    IncreaseButtonL11():void=
        set ButtonL11Pressed += 1 

    IncreaseButtonL12():void=
        set ButtonL12Pressed += 1 

    IncreaseButtonL13():void=
        set ButtonL13Pressed += 1 

    IncreaseButtonL14():void=
        set ButtonL14Pressed += 1 

    IncreaseButtonL15():void=
        set ButtonL15Pressed += 1 

    IncreaseButtonL16():void=
        set ButtonL16Pressed += 1 

    IncreaseButtonL17():void=
        set ButtonL17Pressed += 1 

    IncreaseButtonL18():void=
        set ButtonL18Pressed += 1 

    IncreaseButtonL19():void=
        set ButtonL19Pressed += 1 

    IncreaseButtonL20():void=
        set ButtonL20Pressed += 1 

    IncreaseButtonL21():void=
        set ButtonL21Pressed += 1 

    IncreaseButtonL22():void=
        set ButtonL22Pressed += 1 

    IncreaseButtonL23():void=
        set ButtonL23Pressed += 1 

    IncreaseButtonL24():void=
        set ButtonL24Pressed += 1 

    IncreaseButtonL25():void=
        set ButtonL25Pressed += 1 

    IncreaseButtonL26():void=
        set ButtonL26Pressed += 1 

    IncreaseButtonL27():void=
        set ButtonL27Pressed += 1 

    IncreaseButtonL28():void=
        set ButtonL28Pressed += 1 

    IncreaseButtonL29():void=
        set ButtonL29Pressed += 1 

    IncreaseButtonL30():void=
        set ButtonL30Pressed += 1 

    # Nbr pets active
    Update_Pets_Section():void=    
        set TotalPetsActive = FollowActiveL1+FollowActiveL2+FollowActiveL3+FollowActiveL4+FollowActiveL5+FollowActiveL6+FollowActiveL7+FollowActiveL8+FollowActiveL9+FollowActiveL10+FollowActiveL11+FollowActiveL12+FollowActiveL13+FollowActiveL14+FollowActiveL15+FollowActiveL16+FollowActiveL17+FollowActiveL18+FollowActiveL19+FollowActiveL20+FollowActiveL21+FollowActiveL22+FollowActiveL23+FollowActiveL24+FollowActiveL25+FollowActiveL26+FollowActiveL27+FollowActiveL28+FollowActiveL29+FollowActiveL30+FollowActiveL31+FollowActiveL32+FollowActiveL33+FollowActiveL34
        TextBlockPetNbr.SetText(StringToMessage("{TotalPetsActive}"))

        set TotalPetsBuyable = 1+B1_Unlocked+B2_Unlocked+B3_Unlocked+B4_Unlocked+B5_Unlocked+B6_Unlocked+B7_Unlocked+B8_Unlocked+B9_Unlocked+B10_Unlocked+B11_Unlocked+B12_Unlocked+B13_Unlocked+B14_Unlocked+B15_Unlocked+B16_Unlocked+B17_Unlocked+B18_Unlocked+B19_Unlocked+B20_Unlocked+B21_Unlocked+B22_Unlocked+B23_Unlocked+B24_Unlocked+B25_Unlocked+B26_Unlocked+B27_Unlocked+B28_Unlocked+B29_Unlocked+B31_Unlocked+B32_Unlocked+B33_Unlocked
        TextBlockPetMax.SetText(StringToMessage("/{TotalPetsBuyable}"))

    #Power
    Update_Pet_Power():void=
        set Pet_Power_Value = (Pet_Power_ValueL1+Pet_Power_ValueL2+Pet_Power_ValueL3+Pet_Power_ValueL4+Pet_Power_ValueL5+Pet_Power_ValueL6+Pet_Power_ValueL7+Pet_Power_ValueL8+Pet_Power_ValueL9+Pet_Power_ValueL10+Pet_Power_ValueL11+Pet_Power_ValueL12+Pet_Power_ValueL13+Pet_Power_ValueL14+Pet_Power_ValueL15+Pet_Power_ValueL16+Pet_Power_ValueL17+Pet_Power_ValueL18+Pet_Power_ValueL19+Pet_Power_ValueL20+Pet_Power_ValueL21+Pet_Power_ValueL22+Pet_Power_ValueL23+Pet_Power_ValueL24+Pet_Power_ValueL25+Pet_Power_ValueL26+Pet_Power_ValueL27+Pet_Power_ValueL28+Pet_Power_ValueL29+Pet_Power_ValueL30)*BonusPowerMultiplicator*Bonus_Pet_Power
        set Power_Player = (50.0 + Pet_Power_Value*0.2)*Bonus_Player_Power
        TextBlockPower.SetText(StringToMessage(formatter.Abbreviate(Pet_Power_Value)))

    #Gold
    IncreaseGoldBy(GoldIncrease:int):void=
        set PlayerGold += GoldIncrease
        TextBlockGold.SetText(StringToMessage(formatter.Abbreviate(PlayerGold*1.0)))

    DecreaseGoldBy(GoldDecrease:int):void=
        set PlayerGold -= GoldDecrease
        TextBlockGold.SetText(StringToMessage(formatter.Abbreviate(PlayerGold*1.0)))

    SetGoldBy(SetGold:int):void=
        set PlayerGold = SetGold
        TextBlockGold.SetText(StringToMessage(formatter.Abbreviate(PlayerGold*1.0)))

    #Diamond
    IncreaseDiamondBy(DiamondIncrease:int):void=
        set PlayerDiamond += DiamondIncrease
        TextBlockDiamond.SetText(StringToMessage(formatter.Abbreviate(PlayerDiamond*1.0)))

    DecreaseDiamondBy(DiamondDecrease:int):void=
        set PlayerDiamond -= DiamondDecrease
        TextBlockDiamond.SetText(StringToMessage(formatter.Abbreviate(PlayerDiamond*1.0)))

    SetDiamondBy(SetDiamond:int):void=
        set PlayerDiamond = SetDiamond
        TextBlockDiamond.SetText(StringToMessage(formatter.Abbreviate(PlayerDiamond*1.0)))

    #Crit
    SetCritBy(SetCrit:int):void=
        set Pet_Crit_Value = SetCrit
        TextBlockCrit.SetText(StringToMessage("+{Pet_Crit_Value}%"))

    #Speed
    SetSpeedBy(SetSpeed:float):void=
        set Pet_Speed_Value = SetSpeed
        TextBlockSpeed.SetText(StringToMessage("x{ToTruncatedString(Pet_Speed_Value,1)}"))

    #Player Speed
    SetPlayerSpeedBy(SetSpeed:float):void=
        set Player_Speed_Value = SetSpeed
        TextBlockPlayerSpeed.SetText(StringToMessage("x{ToTruncatedString(Player_Speed_Value,1)}"))

    #Luck
    SetLuckBy(SetLuck:int):void=
        set Pet_Luck_Value = SetLuck
        set Pet_Luck_Value_Final = Pet_Luck_Value + Bonus_Luck
        TextBlockLuck.SetText(StringToMessage("+{Pet_Luck_Value_Final}%"))

    Update_Luck():void=
        set Pet_Luck_Value_Final = Pet_Luck_Value + Bonus_Luck
        TextBlockLuck.SetText(StringToMessage("+{Pet_Luck_Value_Final}%"))

    #Bonus Power Multiplicator
    SetBonusPowerMultiplicator(SetMult:float):void=
        set BonusPowerMultiplicator = SetMult
        Print("Power Multiplicator set to:{ToTruncatedString(SetMult,3)}")

    #Inventory

    SetLuck1By(SetPotion:int):void=
        set Potion_Owned_Luck1 = SetPotion
        TextBlockOwnedLuck1.SetText(StringToMessage("{Potion_Owned_Luck1}"))

    SetLuck2By(SetPotion:int):void=
        set Potion_Owned_Luck2 = SetPotion
        TextBlockOwnedLuck2.SetText(StringToMessage("{Potion_Owned_Luck2}"))

    SetLuck3By(SetPotion:int):void=
        set Potion_Owned_Luck3 = SetPotion
        TextBlockOwnedLuck3.SetText(StringToMessage("{Potion_Owned_Luck3}"))

    SetPetPower1By(SetPotion:int):void=
        set Potion_Owned_Pet_Power1 = SetPotion
        TextBlockOwnedPetPower1.SetText(StringToMessage("{Potion_Owned_Pet_Power1}"))

    SetPetPower2By(SetPotion:int):void=
        set Potion_Owned_Pet_Power2 = SetPotion
        TextBlockOwnedPetPower2.SetText(StringToMessage("{Potion_Owned_Pet_Power2}"))

    SetPetPower3By(SetPotion:int):void=
        set Potion_Owned_Pet_Power3 = SetPotion
        TextBlockOwnedPetPower3.SetText(StringToMessage("{Potion_Owned_Pet_Power3}"))

    SetPlayerPower1By(SetPotion:int):void=
        set Potion_Owned_Player_Power1 = SetPotion
        TextBlockOwnedPlayerPower1.SetText(StringToMessage("{Potion_Owned_Player_Power1}"))

    SetPlayerPower2By(SetPotion:int):void=
        set Potion_Owned_Player_Power2 = SetPotion
        TextBlockOwnedPlayerPower2.SetText(StringToMessage("{Potion_Owned_Player_Power2}"))

    SetPlayerPower3By(SetPotion:int):void=
        set Potion_Owned_Player_Power3 = SetPotion
        TextBlockOwnedPlayerPower3.SetText(StringToMessage("{Potion_Owned_Player_Power3}"))

    SetGold1By(SetPotion:int):void=
        set Potion_Owned_Gold1 = SetPotion
        TextBlockOwnedGold1.SetText(StringToMessage("{Potion_Owned_Gold1}"))

    SetGold2By(SetPotion:int):void=
        set Potion_Owned_Gold2 = SetPotion
        TextBlockOwnedGold2.SetText(StringToMessage("{Potion_Owned_Gold2}"))

    SetGold3By(SetPotion:int):void=
        set Potion_Owned_Gold3 = SetPotion
        TextBlockOwnedGold3.SetText(StringToMessage("{Potion_Owned_Gold3}"))

    Update_Inventory_Count():void=
        #Pet power
        TextBlockOwnedPetPower1.SetText(StringToMessage("{Potion_Owned_Pet_Power1}"))
        TextBlockOwnedPetPower2.SetText(StringToMessage("{Potion_Owned_Pet_Power2}"))
        TextBlockOwnedPetPower3.SetText(StringToMessage("{Potion_Owned_Pet_Power3}"))
        #Player power
        TextBlockOwnedPlayerPower1.SetText(StringToMessage("{Potion_Owned_Player_Power1}"))
        TextBlockOwnedPlayerPower2.SetText(StringToMessage("{Potion_Owned_Player_Power2}"))
        TextBlockOwnedPlayerPower3.SetText(StringToMessage("{Potion_Owned_Player_Power3}"))
        #Gold
        TextBlockOwnedGold1.SetText(StringToMessage("{Potion_Owned_Gold1}"))
        TextBlockOwnedGold2.SetText(StringToMessage("{Potion_Owned_Gold2}"))
        TextBlockOwnedGold3.SetText(StringToMessage("{Potion_Owned_Gold3}"))
        #Luck
        TextBlockOwnedLuck1.SetText(StringToMessage("{Potion_Owned_Luck1}"))
        TextBlockOwnedLuck2.SetText(StringToMessage("{Potion_Owned_Luck2}"))
        TextBlockOwnedLuck3.SetText(StringToMessage("{Potion_Owned_Luck3}"))

    Update_Inventory_Timings():void=
        #PetPower
        if (Potion_Active_PetPower1_Sec>0):
            if (TimeString := ToStopWatchFormat[Potion_Active_PetPower1_Sec*1.0]):
                TextBlockPotion_Active_PetPower1_Sec.SetText(StringToMessage(TimeString))
        else:
            TextBlockPotion_Active_PetPower1_Sec.SetText(StringToMessage(""))

        if (Potion_Active_PetPower2_Sec>0):
            if (TimeString := ToStopWatchFormat[Potion_Active_PetPower2_Sec*1.0]):
                TextBlockPotion_Active_PetPower2_Sec.SetText(StringToMessage(TimeString))
        else:
            TextBlockPotion_Active_PetPower2_Sec.SetText(StringToMessage(""))

        if (Potion_Active_PetPower3_Sec>0):
            if (TimeString := ToStopWatchFormat[Potion_Active_PetPower3_Sec*1.0]):
                TextBlockPotion_Active_PetPower3_Sec.SetText(StringToMessage(TimeString))
        else:
            TextBlockPotion_Active_PetPower3_Sec.SetText(StringToMessage(""))
        #PlayerPower
        if (Potion_Active_PlayerPower1_Sec>0):
            if (TimeString := ToStopWatchFormat[Potion_Active_PlayerPower1_Sec*1.0]):
                TextBlockPotion_Active_PlayerPower1_Sec.SetText(StringToMessage(TimeString))
        else:
            TextBlockPotion_Active_PlayerPower1_Sec.SetText(StringToMessage(""))

        if (Potion_Active_PlayerPower2_Sec>0):
            if (TimeString := ToStopWatchFormat[Potion_Active_PlayerPower2_Sec*1.0]):
                TextBlockPotion_Active_PlayerPower2_Sec.SetText(StringToMessage(TimeString))
        else:
            TextBlockPotion_Active_PlayerPower2_Sec.SetText(StringToMessage(""))

        if (Potion_Active_PlayerPower3_Sec>0):
            if (TimeString := ToStopWatchFormat[Potion_Active_PlayerPower3_Sec*1.0]):
                TextBlockPotion_Active_PlayerPower3_Sec.SetText(StringToMessage(TimeString))
        else:
            TextBlockPotion_Active_PlayerPower3_Sec.SetText(StringToMessage(""))

        #Gold
        if (Potion_Active_Gold1_Sec>0):
            if (TimeString := ToStopWatchFormat[Potion_Active_Gold1_Sec*1.0]):
                TextBlockPotion_Active_Gold1_Sec.SetText(StringToMessage(TimeString))
        else:
            TextBlockPotion_Active_Gold1_Sec.SetText(StringToMessage(""))

        if (Potion_Active_Gold2_Sec>0):
            if (TimeString := ToStopWatchFormat[Potion_Active_Gold2_Sec*1.0]):
                TextBlockPotion_Active_Gold2_Sec.SetText(StringToMessage(TimeString))
        else:
            TextBlockPotion_Active_Gold2_Sec.SetText(StringToMessage(""))

        if (Potion_Active_Gold3_Sec>0):
            if (TimeString := ToStopWatchFormat[Potion_Active_Gold3_Sec*1.0]):
                TextBlockPotion_Active_Gold3_Sec.SetText(StringToMessage(TimeString))
        else:
            TextBlockPotion_Active_Gold3_Sec.SetText(StringToMessage(""))

        #Luck
        if (Potion_Active_Luck1_Sec>0):
            if (TimeString := ToStopWatchFormat[Potion_Active_Luck1_Sec*1.0]):
                TextBlockPotion_Active_Luck1_Sec.SetText(StringToMessage(TimeString))
        else:
            TextBlockPotion_Active_Luck1_Sec.SetText(StringToMessage(""))

        if (Potion_Active_Luck2_Sec>0):
            if (TimeString := ToStopWatchFormat[Potion_Active_Luck2_Sec*1.0]):
                TextBlockPotion_Active_Luck2_Sec.SetText(StringToMessage(TimeString))
        else:
            TextBlockPotion_Active_Luck2_Sec.SetText(StringToMessage(""))

        if (Potion_Active_Luck3_Sec>0):
            if (TimeString := ToStopWatchFormat[Potion_Active_Luck3_Sec*1.0]):
                TextBlockPotion_Active_Luck3_Sec.SetText(StringToMessage(TimeString))
        else:
            TextBlockPotion_Active_Luck3_Sec.SetText(StringToMessage(""))