formatter<public> := module:
    Abbreviations : []char = array { ' ', 'K', 'M', 'B', 'T', 'Q' }

    S2M2<localizes><public>(Message : string)<computes>:message = "{Message}"

    FloatToShortString(Input : float, DecimalAmount : int)<decides><transacts> : string=
        InitialString : string = ToString(Input)
        DecimalIndex := InitialString.Find['.']
        
        return InitialString.Slice[0, DecimalIndex + 1 + DecimalAmount]

    Abbreviate<public>(Number : float):string=
        for (Index : int = 0..Abbreviations.Length, Abbreviation := Abbreviations[Index]):
            if:
                Number < Pow(10.0, (Index+1)*3.0)
                AbbreviatedNum := Number/Pow(10.0, Index*3.0)
            then:
                var RoundedAbbreviation : string = ""

                if (Index = 0, NoDecimalString := ToString(Floor[Number])):
                    set RoundedAbbreviation = NoDecimalString
                else if (ShortFloat := FloatToShortString[AbbreviatedNum, 1]):
                    set RoundedAbbreviation = ShortFloat
                
                if(Index = 0):
                    return {RoundedAbbreviation}
                    
                return "{RoundedAbbreviation}{Abbreviation}"

        return ""

    CommaSeparator<public>(Number : int):string=
        NumberAsString : string = "{Number}"
        StringLength : int = NumberAsString.Length

        var FormattedString : string = ""

        for (Index : int = 0..StringLength-1, Character := NumberAsString[Index]):
            RemainingLength := StringLength-(Index+1)
            set FormattedString += "{Character}"

            if (RemainingLength > 0 and Mod[RemainingLength, 3] = 0):
                set FormattedString += ","             

        return FormattedString