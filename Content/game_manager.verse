using { /Fortnite.com/Devices }
using { /Verse.org/Simulation }
using { /UnrealEngine.com/Temporary/Diagnostics }
using { /Fortnite.com/Characters }
using { /Fortnite.com/Game }
using { /Verse.org/Simulation }
using { /Verse.org/Random }
using { /UnrealEngine.com/Temporary/UI }
using { /Fortnite.com/UI }
using { /UnrealEngine.com/Temporary/SpatialMath}
using { /Verse.org/Colors }
using { /Verse.org/Colors/NamedColors }
using { /Verse.org/Verse }
using { /Verse.org/Assets }

####################################### PLAYER STATS TABLE #######################################
####################################### PLAYER STATS TABLE #######################################
####################################### PLAYER STATS TABLE #######################################

# Tracks different persistable stats for each player.
player_stats_table := class<final><persistable>:
    # The version of the current stats table.
    Version:int = 0

    #Old stuff mendatory
    Score:player_stat = player_stat{}
    Wins:player_stat = player_stat{}
    Losses:player_stat = player_stat{}

    # Player gold
    GoldBalance:player_stat = player_stat{}

    # Player diamond
    DiamondBalance:player_stat = player_stat{}

    ## Player power
    #Power:player_stat = player_stat{}

    # Last level unlocked
    LevelUnlocked:player_stat = player_stat{}

    # Last rebirth unlocked
    RebirthUnlocked:player_stat = player_stat{}

    # Player Crit
    CritScore:player_stat = player_stat{}

    # Pet Speed
    SpeedScore:player_stat = player_stat{}

    # Player Speed
    PlayerSpeedScore:player_stat = player_stat{}

    # Player Luck
    LuckScore:player_stat = player_stat{}

    # Pet Ownership
    Ownership_L1_Pet1:player_stat = player_stat{}
    Ownership_L1_Pet2:player_stat = player_stat{}
    Ownership_L1_Pet3:player_stat = player_stat{}
    Ownership_L1_Pet4:player_stat = player_stat{}
    Ownership_L2_Pet1:player_stat = player_stat{}
    Ownership_L2_Pet2:player_stat = player_stat{}
    Ownership_L2_Pet3:player_stat = player_stat{}
    Ownership_L2_Pet4:player_stat = player_stat{}
    Ownership_L3_Pet1:player_stat = player_stat{}
    Ownership_L3_Pet2:player_stat = player_stat{}
    Ownership_L3_Pet3:player_stat = player_stat{}
    Ownership_L3_Pet4:player_stat = player_stat{}
    Ownership_L4_Pet1:player_stat = player_stat{}
    Ownership_L4_Pet2:player_stat = player_stat{}
    Ownership_L4_Pet3:player_stat = player_stat{}
    Ownership_L4_Pet4:player_stat = player_stat{}
    Ownership_L5_Pet1:player_stat = player_stat{}
    Ownership_L5_Pet2:player_stat = player_stat{}
    Ownership_L5_Pet3:player_stat = player_stat{}
    Ownership_L5_Pet4:player_stat = player_stat{}
    Ownership_L6_Pet1:player_stat = player_stat{}
    Ownership_L6_Pet2:player_stat = player_stat{}
    Ownership_L6_Pet3:player_stat = player_stat{}
    Ownership_L6_Pet4:player_stat = player_stat{}
    Ownership_L7_Pet1:player_stat = player_stat{}
    Ownership_L7_Pet2:player_stat = player_stat{}
    Ownership_L7_Pet3:player_stat = player_stat{}
    Ownership_L7_Pet4:player_stat = player_stat{}
    Ownership_L8_Pet1:player_stat = player_stat{}
    Ownership_L8_Pet2:player_stat = player_stat{}
    Ownership_L8_Pet3:player_stat = player_stat{}
    Ownership_L8_Pet4:player_stat = player_stat{}
    Ownership_L9_Pet1:player_stat = player_stat{}
    Ownership_L9_Pet2:player_stat = player_stat{}
    Ownership_L9_Pet3:player_stat = player_stat{}
    Ownership_L9_Pet4:player_stat = player_stat{}
    Ownership_L10_Pet1:player_stat = player_stat{}
    Ownership_L10_Pet2:player_stat = player_stat{}
    Ownership_L10_Pet3:player_stat = player_stat{}
    Ownership_L10_Pet4:player_stat = player_stat{}
    Ownership_L11_Pet1:player_stat = player_stat{}
    Ownership_L11_Pet2:player_stat = player_stat{}
    Ownership_L11_Pet3:player_stat = player_stat{}
    Ownership_L11_Pet4:player_stat = player_stat{}
    Ownership_L12_Pet1:player_stat = player_stat{}
    Ownership_L12_Pet2:player_stat = player_stat{}
    Ownership_L12_Pet3:player_stat = player_stat{}
    Ownership_L12_Pet4:player_stat = player_stat{}
    Ownership_L13_Pet1:player_stat = player_stat{}
    Ownership_L13_Pet2:player_stat = player_stat{}
    Ownership_L13_Pet3:player_stat = player_stat{}
    Ownership_L13_Pet4:player_stat = player_stat{}
    Ownership_L14_Pet1:player_stat = player_stat{}
    Ownership_L14_Pet2:player_stat = player_stat{}
    Ownership_L14_Pet3:player_stat = player_stat{}
    Ownership_L14_Pet4:player_stat = player_stat{}
    Ownership_L15_Pet1:player_stat = player_stat{}
    Ownership_L15_Pet2:player_stat = player_stat{}
    Ownership_L15_Pet3:player_stat = player_stat{}
    Ownership_L15_Pet4:player_stat = player_stat{}
    Ownership_L16_Pet1:player_stat = player_stat{}
    Ownership_L16_Pet2:player_stat = player_stat{}
    Ownership_L16_Pet3:player_stat = player_stat{}
    Ownership_L16_Pet4:player_stat = player_stat{}
    Ownership_L17_Pet1:player_stat = player_stat{}
    Ownership_L17_Pet2:player_stat = player_stat{}
    Ownership_L17_Pet3:player_stat = player_stat{}
    Ownership_L17_Pet4:player_stat = player_stat{}
    Ownership_L18_Pet1:player_stat = player_stat{}
    Ownership_L18_Pet2:player_stat = player_stat{}
    Ownership_L18_Pet3:player_stat = player_stat{}
    Ownership_L18_Pet4:player_stat = player_stat{}
    Ownership_L19_Pet1:player_stat = player_stat{}
    Ownership_L19_Pet2:player_stat = player_stat{}
    Ownership_L19_Pet3:player_stat = player_stat{}
    Ownership_L19_Pet4:player_stat = player_stat{}
    Ownership_L20_Pet1:player_stat = player_stat{}
    Ownership_L20_Pet2:player_stat = player_stat{}
    Ownership_L20_Pet3:player_stat = player_stat{}
    Ownership_L20_Pet4:player_stat = player_stat{}
    Ownership_L21_Pet1:player_stat = player_stat{}
    Ownership_L21_Pet2:player_stat = player_stat{}
    Ownership_L21_Pet3:player_stat = player_stat{}
    Ownership_L21_Pet4:player_stat = player_stat{}
    Ownership_L22_Pet1:player_stat = player_stat{}
    Ownership_L22_Pet2:player_stat = player_stat{}
    Ownership_L22_Pet3:player_stat = player_stat{}
    Ownership_L22_Pet4:player_stat = player_stat{}
    Ownership_L23_Pet1:player_stat = player_stat{}
    Ownership_L23_Pet2:player_stat = player_stat{}
    Ownership_L23_Pet3:player_stat = player_stat{}
    Ownership_L23_Pet4:player_stat = player_stat{}
    Ownership_L24_Pet1:player_stat = player_stat{}
    Ownership_L24_Pet2:player_stat = player_stat{}
    Ownership_L24_Pet3:player_stat = player_stat{}
    Ownership_L24_Pet4:player_stat = player_stat{}
    Ownership_L25_Pet1:player_stat = player_stat{}
    Ownership_L25_Pet2:player_stat = player_stat{}
    Ownership_L25_Pet3:player_stat = player_stat{}
    Ownership_L25_Pet4:player_stat = player_stat{}
    Ownership_L26_Pet1:player_stat = player_stat{}
    Ownership_L26_Pet2:player_stat = player_stat{}
    Ownership_L26_Pet3:player_stat = player_stat{}
    Ownership_L26_Pet4:player_stat = player_stat{}
    Ownership_L27_Pet1:player_stat = player_stat{}
    Ownership_L27_Pet2:player_stat = player_stat{}
    Ownership_L27_Pet3:player_stat = player_stat{}
    Ownership_L27_Pet4:player_stat = player_stat{}
    Ownership_L28_Pet1:player_stat = player_stat{}
    Ownership_L28_Pet2:player_stat = player_stat{}
    Ownership_L28_Pet3:player_stat = player_stat{}
    Ownership_L28_Pet4:player_stat = player_stat{}
    Ownership_L29_Pet1:player_stat = player_stat{}
    Ownership_L29_Pet2:player_stat = player_stat{}
    Ownership_L29_Pet3:player_stat = player_stat{}
    Ownership_L29_Pet4:player_stat = player_stat{}
    Ownership_L30_Pet1:player_stat = player_stat{}
    Ownership_L30_Pet2:player_stat = player_stat{}
    Ownership_L30_Pet3:player_stat = player_stat{}
    Ownership_L30_Pet4:player_stat = player_stat{}

    Ownership_L31_Pet1:player_stat = player_stat{}
    Ownership_L32_Pet1:player_stat = player_stat{}
    Ownership_L33_Pet1:player_stat = player_stat{}
    Ownership_L34_Pet1:player_stat = player_stat{}

    # Potions owned
    Ownership_Potion_Luck1:player_stat = player_stat{}
    Ownership_Potion_Luck2:player_stat = player_stat{}
    Ownership_Potion_Luck3:player_stat = player_stat{}
    Ownership_Potion_Gold1:player_stat = player_stat{}
    Ownership_Potion_Gold2:player_stat = player_stat{}
    Ownership_Potion_Gold3:player_stat = player_stat{}
    Ownership_Potion_PetPower1:player_stat = player_stat{}
    Ownership_Potion_PetPower2:player_stat = player_stat{}
    Ownership_Potion_PetPower3:player_stat = player_stat{}
    Ownership_Potion_PlayerPower1:player_stat = player_stat{}
    Ownership_Potion_PlayerPower2:player_stat = player_stat{}
    Ownership_Potion_PlayerPower3:player_stat = player_stat{}

    #Code entered
    Code1Entered:player_stat = player_stat{}
    Code2Entered:player_stat = player_stat{}
    Code3Entered:player_stat = player_stat{}
    Code4Entered:player_stat = player_stat{}

    #Tuto wheel
    Wheelshowed:player_stat = player_stat{}

# Maps players to a table of their player stats.
var PlayerStatsMap:weak_map(player, player_stats_table) = map{}

 # Creates a new player_stats_table with the same values as the previous player_stats_table.
 MakePlayerStatsTable<constructor>(OldTable:player_stats_table)<transacts> := player_stats_table:
    Version := OldTable.Version
    
    Score := OldTable.Score

    GoldBalance := OldTable.GoldBalance

    DiamondBalance := OldTable.DiamondBalance

    LevelUnlocked:= OldTable.LevelUnlocked

    RebirthUnlocked:= OldTable.RebirthUnlocked

    CritScore:= OldTable.CritScore

    SpeedScore:= OldTable.SpeedScore

    PlayerSpeedScore:= OldTable.PlayerSpeedScore

    LuckScore:= OldTable.LuckScore

    Ownership_L1_Pet1 := OldTable.Ownership_L1_Pet1
    Ownership_L1_Pet2 := OldTable.Ownership_L1_Pet2
    Ownership_L1_Pet3 := OldTable.Ownership_L1_Pet3
    Ownership_L1_Pet4 := OldTable.Ownership_L1_Pet4
    Ownership_L2_Pet1 := OldTable.Ownership_L2_Pet1
    Ownership_L2_Pet2 := OldTable.Ownership_L2_Pet2
    Ownership_L2_Pet3 := OldTable.Ownership_L2_Pet3
    Ownership_L2_Pet4 := OldTable.Ownership_L2_Pet4
    Ownership_L3_Pet1 := OldTable.Ownership_L3_Pet1
    Ownership_L3_Pet2 := OldTable.Ownership_L3_Pet2
    Ownership_L3_Pet3 := OldTable.Ownership_L3_Pet3
    Ownership_L3_Pet4 := OldTable.Ownership_L3_Pet4
    Ownership_L4_Pet1 := OldTable.Ownership_L4_Pet1
    Ownership_L4_Pet2 := OldTable.Ownership_L4_Pet2
    Ownership_L4_Pet3 := OldTable.Ownership_L4_Pet3
    Ownership_L4_Pet4 := OldTable.Ownership_L4_Pet4
    Ownership_L5_Pet1 := OldTable.Ownership_L5_Pet1
    Ownership_L5_Pet2 := OldTable.Ownership_L5_Pet2
    Ownership_L5_Pet3 := OldTable.Ownership_L5_Pet3
    Ownership_L5_Pet4 := OldTable.Ownership_L5_Pet4
    Ownership_L6_Pet1 := OldTable.Ownership_L6_Pet1
    Ownership_L6_Pet2 := OldTable.Ownership_L6_Pet2
    Ownership_L6_Pet3 := OldTable.Ownership_L6_Pet3
    Ownership_L6_Pet4 := OldTable.Ownership_L6_Pet4
    Ownership_L7_Pet1 := OldTable.Ownership_L7_Pet1
    Ownership_L7_Pet2 := OldTable.Ownership_L7_Pet2
    Ownership_L7_Pet3 := OldTable.Ownership_L7_Pet3
    Ownership_L7_Pet4 := OldTable.Ownership_L7_Pet4
    Ownership_L8_Pet1 := OldTable.Ownership_L8_Pet1
    Ownership_L8_Pet2 := OldTable.Ownership_L8_Pet2
    Ownership_L8_Pet3 := OldTable.Ownership_L8_Pet3
    Ownership_L8_Pet4 := OldTable.Ownership_L8_Pet4
    Ownership_L9_Pet1 := OldTable.Ownership_L9_Pet1
    Ownership_L9_Pet2 := OldTable.Ownership_L9_Pet2
    Ownership_L9_Pet3 := OldTable.Ownership_L9_Pet3
    Ownership_L9_Pet4 := OldTable.Ownership_L9_Pet4
    Ownership_L10_Pet1 := OldTable.Ownership_L10_Pet1
    Ownership_L10_Pet2 := OldTable.Ownership_L10_Pet2
    Ownership_L10_Pet3 := OldTable.Ownership_L10_Pet3
    Ownership_L10_Pet4 := OldTable.Ownership_L10_Pet4
    Ownership_L11_Pet1 := OldTable.Ownership_L11_Pet1
    Ownership_L11_Pet2 := OldTable.Ownership_L11_Pet2
    Ownership_L11_Pet3 := OldTable.Ownership_L11_Pet3
    Ownership_L11_Pet4 := OldTable.Ownership_L11_Pet4
    Ownership_L12_Pet1 := OldTable.Ownership_L12_Pet1
    Ownership_L12_Pet2 := OldTable.Ownership_L12_Pet2
    Ownership_L12_Pet3 := OldTable.Ownership_L12_Pet3
    Ownership_L12_Pet4 := OldTable.Ownership_L12_Pet4
    Ownership_L13_Pet1 := OldTable.Ownership_L13_Pet1
    Ownership_L13_Pet2 := OldTable.Ownership_L13_Pet2
    Ownership_L13_Pet3 := OldTable.Ownership_L13_Pet3
    Ownership_L13_Pet4 := OldTable.Ownership_L13_Pet4
    Ownership_L14_Pet1 := OldTable.Ownership_L14_Pet1
    Ownership_L14_Pet2 := OldTable.Ownership_L14_Pet2
    Ownership_L14_Pet3 := OldTable.Ownership_L14_Pet3
    Ownership_L14_Pet4 := OldTable.Ownership_L14_Pet4
    Ownership_L15_Pet1 := OldTable.Ownership_L15_Pet1
    Ownership_L15_Pet2 := OldTable.Ownership_L15_Pet2
    Ownership_L15_Pet3 := OldTable.Ownership_L15_Pet3
    Ownership_L15_Pet4 := OldTable.Ownership_L15_Pet4
    Ownership_L16_Pet1 := OldTable.Ownership_L16_Pet1
    Ownership_L16_Pet2 := OldTable.Ownership_L16_Pet2
    Ownership_L16_Pet3 := OldTable.Ownership_L16_Pet3
    Ownership_L16_Pet4 := OldTable.Ownership_L16_Pet4
    Ownership_L17_Pet1 := OldTable.Ownership_L17_Pet1
    Ownership_L17_Pet2 := OldTable.Ownership_L17_Pet2
    Ownership_L17_Pet3 := OldTable.Ownership_L17_Pet3
    Ownership_L17_Pet4 := OldTable.Ownership_L17_Pet4
    Ownership_L18_Pet1 := OldTable.Ownership_L18_Pet1
    Ownership_L18_Pet2 := OldTable.Ownership_L18_Pet2
    Ownership_L18_Pet3 := OldTable.Ownership_L18_Pet3
    Ownership_L18_Pet4 := OldTable.Ownership_L18_Pet4
    Ownership_L19_Pet1 := OldTable.Ownership_L19_Pet1
    Ownership_L19_Pet2 := OldTable.Ownership_L19_Pet2
    Ownership_L19_Pet3 := OldTable.Ownership_L19_Pet3
    Ownership_L19_Pet4 := OldTable.Ownership_L19_Pet4
    Ownership_L20_Pet1 := OldTable.Ownership_L20_Pet1
    Ownership_L20_Pet2 := OldTable.Ownership_L20_Pet2
    Ownership_L20_Pet3 := OldTable.Ownership_L20_Pet3
    Ownership_L20_Pet4 := OldTable.Ownership_L20_Pet4
    Ownership_L21_Pet1 := OldTable.Ownership_L21_Pet1
    Ownership_L21_Pet2 := OldTable.Ownership_L21_Pet2
    Ownership_L21_Pet3 := OldTable.Ownership_L21_Pet3
    Ownership_L21_Pet4 := OldTable.Ownership_L21_Pet4
    Ownership_L22_Pet1 := OldTable.Ownership_L22_Pet1
    Ownership_L22_Pet2 := OldTable.Ownership_L22_Pet2
    Ownership_L22_Pet3 := OldTable.Ownership_L22_Pet3
    Ownership_L22_Pet4 := OldTable.Ownership_L22_Pet4
    Ownership_L23_Pet1 := OldTable.Ownership_L23_Pet1
    Ownership_L23_Pet2 := OldTable.Ownership_L23_Pet2
    Ownership_L23_Pet3 := OldTable.Ownership_L23_Pet3
    Ownership_L23_Pet4 := OldTable.Ownership_L23_Pet4
    Ownership_L24_Pet1 := OldTable.Ownership_L24_Pet1
    Ownership_L24_Pet2 := OldTable.Ownership_L24_Pet2
    Ownership_L24_Pet3 := OldTable.Ownership_L24_Pet3
    Ownership_L24_Pet4 := OldTable.Ownership_L24_Pet4
    Ownership_L25_Pet1 := OldTable.Ownership_L25_Pet1
    Ownership_L25_Pet2 := OldTable.Ownership_L25_Pet2
    Ownership_L25_Pet3 := OldTable.Ownership_L25_Pet3
    Ownership_L25_Pet4 := OldTable.Ownership_L25_Pet4
    Ownership_L26_Pet1 := OldTable.Ownership_L26_Pet1
    Ownership_L26_Pet2 := OldTable.Ownership_L26_Pet2
    Ownership_L26_Pet3 := OldTable.Ownership_L26_Pet3
    Ownership_L26_Pet4 := OldTable.Ownership_L26_Pet4
    Ownership_L27_Pet1 := OldTable.Ownership_L27_Pet1
    Ownership_L27_Pet2 := OldTable.Ownership_L27_Pet2
    Ownership_L27_Pet3 := OldTable.Ownership_L27_Pet3
    Ownership_L27_Pet4 := OldTable.Ownership_L27_Pet4
    Ownership_L28_Pet1 := OldTable.Ownership_L28_Pet1
    Ownership_L28_Pet2 := OldTable.Ownership_L28_Pet2
    Ownership_L28_Pet3 := OldTable.Ownership_L28_Pet3
    Ownership_L28_Pet4 := OldTable.Ownership_L28_Pet4
    Ownership_L29_Pet1 := OldTable.Ownership_L29_Pet1
    Ownership_L29_Pet2 := OldTable.Ownership_L29_Pet2
    Ownership_L29_Pet3 := OldTable.Ownership_L29_Pet3
    Ownership_L29_Pet4 := OldTable.Ownership_L29_Pet4
    Ownership_L30_Pet1 := OldTable.Ownership_L30_Pet1
    Ownership_L30_Pet2 := OldTable.Ownership_L30_Pet2
    Ownership_L30_Pet3 := OldTable.Ownership_L30_Pet3
    Ownership_L30_Pet4 := OldTable.Ownership_L30_Pet4

    Ownership_L31_Pet1 := OldTable.Ownership_L31_Pet1
    Ownership_L32_Pet1 := OldTable.Ownership_L32_Pet1
    Ownership_L33_Pet1 := OldTable.Ownership_L33_Pet1
    Ownership_L34_Pet1 := OldTable.Ownership_L34_Pet1
    
    Ownership_Potion_Luck1 := OldTable.Ownership_Potion_Luck1
    Ownership_Potion_Luck2 := OldTable.Ownership_Potion_Luck2
    Ownership_Potion_Luck3 := OldTable.Ownership_Potion_Luck3
    Ownership_Potion_Gold1 := OldTable.Ownership_Potion_Gold1
    Ownership_Potion_Gold2 := OldTable.Ownership_Potion_Gold2
    Ownership_Potion_Gold3 := OldTable.Ownership_Potion_Gold3
    Ownership_Potion_PetPower1 := OldTable.Ownership_Potion_PetPower1
    Ownership_Potion_PetPower2 := OldTable.Ownership_Potion_PetPower2
    Ownership_Potion_PetPower3 := OldTable.Ownership_Potion_PetPower3
    Ownership_Potion_PlayerPower1 := OldTable.Ownership_Potion_PlayerPower1
    Ownership_Potion_PlayerPower2 := OldTable.Ownership_Potion_PlayerPower2
    Ownership_Potion_PlayerPower3 := OldTable.Ownership_Potion_PlayerPower3

    #Code entered
    Code1Entered:=OldTable.Code1Entered
    Code2Entered:=OldTable.Code2Entered
    Code3Entered:=OldTable.Code3Entered
    Code4Entered:=OldTable.Code4Entered

    #Tuto
    Wheelshowed:=OldTable.Wheelshowed

####################################### PLAYER STAT #######################################
####################################### PLAYER STAT #######################################
####################################### PLAYER STAT #######################################

player_stat := class<final><persistable>:
    # The current value of the player's stat.
    CurrentValue:int = 0

    # The highest value achieved so far.
    HighestValue:int = 0

    # Returns either the NewValue or HighestValue,
    # whichever is higher.
    SetHighestValue(NewValue:int)<transacts>:int=
        if (NewValue > HighestValue):
            NewValue
        else:
            HighestValue
    
# Creates a new player_stat class with the given NewValue and a source player_stat.
# The HighestValue is set to either the NewValue or the HighestValue
# from the source player_stat.
MakeUpdatedPlayerStat<constructor>(OldStats:player_stat, NewValue:int)<transacts> := player_stat:
    CurrentValue := NewValue
    HighestValue := OldStats.SetHighestValue(NewValue)

####################################### PLAYER STATS MANAGER #######################################
####################################### PLAYER STATS MANAGER #######################################
####################################### PLAYER STATS MANAGER #######################################

# Represents an abstract stat to update using RecordPlayerStat.
# The class has the unique specifier to make instances of the class comparable.
# The class has the computes specifier to be able to instantiate it at module-scope.
# The class has the abstract specifier so it cannot be instantiated directly, and
# requires subclasses to implement any non-initialized functions, like DebugString().
stat_type := class<computes><abstract>:
    DebugString()<transacts>:string

# This example uses instances of stat_type subclasses instead of the enum type
# so you can add more stat_types after the initial published version of the project.    
StatType := module:

    score_stat<public> := class<computes>(stat_type):
        DebugString<override>()<transacts>:string = "Score"

    #Ownership_L1_Pet1_stat<public> := class<computes><unique>(stat_type):
    #    DebugString<override>():string = "Ownership_L1_Pet1"

    # Instances of each stat_type you can use in your experience
    Score<public>:score_stat = score_stat{}

    LevelUnlocked<public>:score_stat = score_stat{}

    RebirthUnlocked<public>:score_stat = score_stat{}

    GoldBalance<public>:score_stat = score_stat{}

    DiamondBalance<public>:score_stat = score_stat{}
    
    CritScore<public>:score_stat = score_stat{}

    SpeedScore<public>:score_stat = score_stat{}

    PlayerSpeedScore<public>:score_stat = score_stat{}

    LuckScore<public>:score_stat = score_stat{}

    Ownership_L1_Pet1<public>:score_stat = score_stat{}
    Ownership_L1_Pet2<public>:score_stat = score_stat{}
    Ownership_L1_Pet3<public>:score_stat = score_stat{}
    Ownership_L1_Pet4<public>:score_stat = score_stat{}
    Ownership_L2_Pet1<public>:score_stat = score_stat{}
    Ownership_L2_Pet2<public>:score_stat = score_stat{}
    Ownership_L2_Pet3<public>:score_stat = score_stat{}
    Ownership_L2_Pet4<public>:score_stat = score_stat{}
    Ownership_L3_Pet1<public>:score_stat = score_stat{}
    Ownership_L3_Pet2<public>:score_stat = score_stat{}
    Ownership_L3_Pet3<public>:score_stat = score_stat{}
    Ownership_L3_Pet4<public>:score_stat = score_stat{}
    Ownership_L4_Pet1<public>:score_stat = score_stat{}
    Ownership_L4_Pet2<public>:score_stat = score_stat{}
    Ownership_L4_Pet3<public>:score_stat = score_stat{}
    Ownership_L4_Pet4<public>:score_stat = score_stat{}
    Ownership_L5_Pet1<public>:score_stat = score_stat{}
    Ownership_L5_Pet2<public>:score_stat = score_stat{}
    Ownership_L5_Pet3<public>:score_stat = score_stat{}
    Ownership_L5_Pet4<public>:score_stat = score_stat{}
    Ownership_L6_Pet1<public>:score_stat = score_stat{}
    Ownership_L6_Pet2<public>:score_stat = score_stat{}
    Ownership_L6_Pet3<public>:score_stat = score_stat{}
    Ownership_L6_Pet4<public>:score_stat = score_stat{}
    Ownership_L7_Pet1<public>:score_stat = score_stat{}
    Ownership_L7_Pet2<public>:score_stat = score_stat{}
    Ownership_L7_Pet3<public>:score_stat = score_stat{}
    Ownership_L7_Pet4<public>:score_stat = score_stat{}
    Ownership_L8_Pet1<public>:score_stat = score_stat{}
    Ownership_L8_Pet2<public>:score_stat = score_stat{}
    Ownership_L8_Pet3<public>:score_stat = score_stat{}
    Ownership_L8_Pet4<public>:score_stat = score_stat{}
    Ownership_L9_Pet1<public>:score_stat = score_stat{}
    Ownership_L9_Pet2<public>:score_stat = score_stat{}
    Ownership_L9_Pet3<public>:score_stat = score_stat{}
    Ownership_L9_Pet4<public>:score_stat = score_stat{}
    Ownership_L10_Pet1<public>:score_stat = score_stat{}
    Ownership_L10_Pet2<public>:score_stat = score_stat{}
    Ownership_L10_Pet3<public>:score_stat = score_stat{}
    Ownership_L10_Pet4<public>:score_stat = score_stat{}
    Ownership_L11_Pet1<public>:score_stat = score_stat{}
    Ownership_L11_Pet2<public>:score_stat = score_stat{}
    Ownership_L11_Pet3<public>:score_stat = score_stat{}
    Ownership_L11_Pet4<public>:score_stat = score_stat{}
    Ownership_L12_Pet1<public>:score_stat = score_stat{}
    Ownership_L12_Pet2<public>:score_stat = score_stat{}
    Ownership_L12_Pet3<public>:score_stat = score_stat{}
    Ownership_L12_Pet4<public>:score_stat = score_stat{}
    Ownership_L13_Pet1<public>:score_stat = score_stat{}
    Ownership_L13_Pet2<public>:score_stat = score_stat{}
    Ownership_L13_Pet3<public>:score_stat = score_stat{}
    Ownership_L13_Pet4<public>:score_stat = score_stat{}
    Ownership_L14_Pet1<public>:score_stat = score_stat{}
    Ownership_L14_Pet2<public>:score_stat = score_stat{}
    Ownership_L14_Pet3<public>:score_stat = score_stat{}
    Ownership_L14_Pet4<public>:score_stat = score_stat{}
    Ownership_L15_Pet1<public>:score_stat = score_stat{}
    Ownership_L15_Pet2<public>:score_stat = score_stat{}
    Ownership_L15_Pet3<public>:score_stat = score_stat{}
    Ownership_L15_Pet4<public>:score_stat = score_stat{}
    Ownership_L16_Pet1<public>:score_stat = score_stat{}
    Ownership_L16_Pet2<public>:score_stat = score_stat{}
    Ownership_L16_Pet3<public>:score_stat = score_stat{}
    Ownership_L16_Pet4<public>:score_stat = score_stat{}
    Ownership_L17_Pet1<public>:score_stat = score_stat{}
    Ownership_L17_Pet2<public>:score_stat = score_stat{}
    Ownership_L17_Pet3<public>:score_stat = score_stat{}
    Ownership_L17_Pet4<public>:score_stat = score_stat{}
    Ownership_L18_Pet1<public>:score_stat = score_stat{}
    Ownership_L18_Pet2<public>:score_stat = score_stat{}
    Ownership_L18_Pet3<public>:score_stat = score_stat{}
    Ownership_L18_Pet4<public>:score_stat = score_stat{}
    Ownership_L19_Pet1<public>:score_stat = score_stat{}
    Ownership_L19_Pet2<public>:score_stat = score_stat{}
    Ownership_L19_Pet3<public>:score_stat = score_stat{}
    Ownership_L19_Pet4<public>:score_stat = score_stat{}
    Ownership_L20_Pet1<public>:score_stat = score_stat{}
    Ownership_L20_Pet2<public>:score_stat = score_stat{}
    Ownership_L20_Pet3<public>:score_stat = score_stat{}
    Ownership_L20_Pet4<public>:score_stat = score_stat{}
    Ownership_L21_Pet1<public>:score_stat = score_stat{}
    Ownership_L21_Pet2<public>:score_stat = score_stat{}
    Ownership_L21_Pet3<public>:score_stat = score_stat{}
    Ownership_L21_Pet4<public>:score_stat = score_stat{}
    Ownership_L22_Pet1<public>:score_stat = score_stat{}
    Ownership_L22_Pet2<public>:score_stat = score_stat{}
    Ownership_L22_Pet3<public>:score_stat = score_stat{}
    Ownership_L22_Pet4<public>:score_stat = score_stat{}
    Ownership_L23_Pet1<public>:score_stat = score_stat{}
    Ownership_L23_Pet2<public>:score_stat = score_stat{}
    Ownership_L23_Pet3<public>:score_stat = score_stat{}
    Ownership_L23_Pet4<public>:score_stat = score_stat{}
    Ownership_L24_Pet1<public>:score_stat = score_stat{}
    Ownership_L24_Pet2<public>:score_stat = score_stat{}
    Ownership_L24_Pet3<public>:score_stat = score_stat{}
    Ownership_L24_Pet4<public>:score_stat = score_stat{}
    Ownership_L25_Pet1<public>:score_stat = score_stat{}
    Ownership_L25_Pet2<public>:score_stat = score_stat{}
    Ownership_L25_Pet3<public>:score_stat = score_stat{}
    Ownership_L25_Pet4<public>:score_stat = score_stat{}
    Ownership_L26_Pet1<public>:score_stat = score_stat{}
    Ownership_L26_Pet2<public>:score_stat = score_stat{}
    Ownership_L26_Pet3<public>:score_stat = score_stat{}
    Ownership_L26_Pet4<public>:score_stat = score_stat{}
    Ownership_L27_Pet1<public>:score_stat = score_stat{}
    Ownership_L27_Pet2<public>:score_stat = score_stat{}
    Ownership_L27_Pet3<public>:score_stat = score_stat{}
    Ownership_L27_Pet4<public>:score_stat = score_stat{}
    Ownership_L28_Pet1<public>:score_stat = score_stat{}
    Ownership_L28_Pet2<public>:score_stat = score_stat{}
    Ownership_L28_Pet3<public>:score_stat = score_stat{}
    Ownership_L28_Pet4<public>:score_stat = score_stat{}
    Ownership_L29_Pet1<public>:score_stat = score_stat{}
    Ownership_L29_Pet2<public>:score_stat = score_stat{}
    Ownership_L29_Pet3<public>:score_stat = score_stat{}
    Ownership_L29_Pet4<public>:score_stat = score_stat{}
    Ownership_L30_Pet1<public>:score_stat = score_stat{}
    Ownership_L30_Pet2<public>:score_stat = score_stat{}
    Ownership_L30_Pet3<public>:score_stat = score_stat{}
    Ownership_L30_Pet4<public>:score_stat = score_stat{}

    Ownership_L31_Pet1<public>:score_stat = score_stat{}
    Ownership_L32_Pet1<public>:score_stat = score_stat{}
    Ownership_L33_Pet1<public>:score_stat = score_stat{}
    Ownership_L34_Pet1<public>:score_stat = score_stat{}

    Ownership_Potion_Luck1<public>:score_stat = score_stat{}
    Ownership_Potion_Luck2<public>:score_stat = score_stat{}
    Ownership_Potion_Luck3<public>:score_stat = score_stat{}
    Ownership_Potion_Gold1<public>:score_stat = score_stat{}
    Ownership_Potion_Gold2<public>:score_stat = score_stat{}
    Ownership_Potion_Gold3<public>:score_stat = score_stat{}
    Ownership_Potion_PetPower1<public>:score_stat = score_stat{}
    Ownership_Potion_PetPower2<public>:score_stat = score_stat{}
    Ownership_Potion_PetPower3<public>:score_stat = score_stat{}
    Ownership_Potion_PlayerPower1<public>:score_stat = score_stat{}
    Ownership_Potion_PlayerPower2<public>:score_stat = score_stat{}
    Ownership_Potion_PlayerPower3<public>:score_stat = score_stat{}

    Code1Entered<public>:score_stat = score_stat{}
    Code2Entered<public>:score_stat = score_stat{}
    Code3Entered<public>:score_stat = score_stat{}
    Code4Entered<public>:score_stat = score_stat{}

    Wheelshowed<public>:score_stat = score_stat{}

# Manages and updates player_stat_tables for each player.
player_stats_manager := class():

    # Return the player_stats_table for the provided Agent.
    GetPlayerStats(Agent:agent)<decides><transacts>:player_stats_table=
        var PlayerStats:player_stats_table = player_stats_table{}
        if:
            Player := player[Agent]
            PlayerStatsTable := PlayerStatsMap[Player]
            set PlayerStats = MakePlayerStatsTable(PlayerStatsTable)
        PlayerStats

    # Initialize stats for all current players.
    InitializeAllPlayers(Players:[]player):void =
        for (Player : Players):
            InitializePlayer(Player)

    # Initialize stats for the given player.
    InitializePlayer(Player:player):void=
        if:
            not PlayerStatsMap[Player]
            set PlayerStatsMap[Player] = player_stats_table{}

    # Update the given stat for the given player by creating a new
    # player_stats_table and setting it in the PlayerStatsMap.
    RecordPlayerStat(Agent:agent, Stat:stat_type, ?Score:int = 0):void=
        if:
            Player := player[Agent]
            PlayerStatsTable := PlayerStatsMap[Player]
            if(Stat.DebugString() = StatType.Score.DebugString()):
                ScoreStat := PlayerStatsTable.Score

                # Create a new player_stats_table with the updated score,
                # then set that table in the PlayerStatsMap.
                set PlayerStatsMap[Player] = player_stats_table:
                    MakePlayerStatsTable<constructor>(PlayerStatsTable)
                    Score := MakeUpdatedPlayerStat(ScoreStat, Score)
            
    #Gold
    RecordGold(Agent:agent, Stat:stat_type, ?GoldBalance:int = 0):void=
        if:
            Player := player[Agent]
            PlayerStatsTable := PlayerStatsMap[Player]
            if(Stat.DebugString() = StatType.GoldBalance.DebugString()):
                GoldStat := PlayerStatsTable.GoldBalance
                set PlayerStatsMap[Player] = player_stats_table:
                    MakePlayerStatsTable<constructor>(PlayerStatsTable)
                    GoldBalance := MakeUpdatedPlayerStat(GoldStat, GoldBalance)

    #Diamond
    RecordDiamond(Agent:agent, Stat:stat_type, ?DiamondBalance:int = 0):void=
        if:
            Player := player[Agent]
            PlayerStatsTable := PlayerStatsMap[Player]
            if(Stat.DebugString() = StatType.DiamondBalance.DebugString()):
                DiamondStat := PlayerStatsTable.DiamondBalance
                set PlayerStatsMap[Player] = player_stats_table:
                    MakePlayerStatsTable<constructor>(PlayerStatsTable)
                    DiamondBalance := MakeUpdatedPlayerStat(DiamondStat, DiamondBalance)

    #Max Level Unlocked
    RecordMaxLevelUnlocked(Agent:agent, Stat:stat_type, ?LevelUnlocked:int = 0):void=
        if:
            Player := player[Agent]
            PlayerStatsTable := PlayerStatsMap[Player]
            if(Stat.DebugString() = StatType.LevelUnlocked.DebugString()):
                MaxLevelStat := PlayerStatsTable.LevelUnlocked
                set PlayerStatsMap[Player] = player_stats_table:
                    MakePlayerStatsTable<constructor>(PlayerStatsTable)
                    LevelUnlocked := MakeUpdatedPlayerStat(MaxLevelStat, LevelUnlocked)

    #Max Rebirth Unlocked
    RecordMaxRebirthUnlocked(Agent:agent, Stat:stat_type, ?RebirthUnlocked:int = 0):void=
        if:
            Player := player[Agent]
            PlayerStatsTable := PlayerStatsMap[Player]
            if(Stat.DebugString() = StatType.RebirthUnlocked.DebugString()):
                MaxRebirthStat := PlayerStatsTable.RebirthUnlocked
                set PlayerStatsMap[Player] = player_stats_table:
                    MakePlayerStatsTable<constructor>(PlayerStatsTable)
                    RebirthUnlocked := MakeUpdatedPlayerStat(MaxRebirthStat, RebirthUnlocked)

    #Crit
    RecordCrit(Agent:agent, Stat:stat_type, ?CritScore:int = 0):void=
        if:
            Player := player[Agent]
            PlayerStatsTable := PlayerStatsMap[Player]
            if(Stat.DebugString() = StatType.CritScore.DebugString()):
                CritStat := PlayerStatsTable.CritScore
                set PlayerStatsMap[Player] = player_stats_table:
                    MakePlayerStatsTable<constructor>(PlayerStatsTable)
                    CritScore := MakeUpdatedPlayerStat(CritStat, CritScore)

    
    #Speed
    RecordSpeed(Agent:agent, Stat:stat_type, ?SpeedScore:int = 0):void=
        if:
            Player := player[Agent]
            PlayerStatsTable := PlayerStatsMap[Player]
            if(Stat.DebugString() = StatType.SpeedScore.DebugString()):
                SpeedStat := PlayerStatsTable.SpeedScore
                set PlayerStatsMap[Player] = player_stats_table:
                    MakePlayerStatsTable<constructor>(PlayerStatsTable)
                    SpeedScore := MakeUpdatedPlayerStat(SpeedStat, SpeedScore)

    #PlayerSpeed
    RecordPlayerSpeed(Agent:agent, Stat:stat_type, ?PlayerSpeedScore:int = 0):void=
        if:
            Player := player[Agent]
            PlayerStatsTable := PlayerStatsMap[Player]
            if(Stat.DebugString() = StatType.PlayerSpeedScore.DebugString()):
                PlayerSpeedStat := PlayerStatsTable.PlayerSpeedScore
                set PlayerStatsMap[Player] = player_stats_table:
                    MakePlayerStatsTable<constructor>(PlayerStatsTable)
                    PlayerSpeedScore := MakeUpdatedPlayerStat(PlayerSpeedStat, PlayerSpeedScore)

    #Luck
    RecordLuck(Agent:agent, Stat:stat_type, ?LuckScore:int = 0):void=
        if:
            Player := player[Agent]
            PlayerStatsTable := PlayerStatsMap[Player]
            if(Stat.DebugString() = StatType.LuckScore.DebugString()):
                LuckStat := PlayerStatsTable.LuckScore
                set PlayerStatsMap[Player] = player_stats_table:
                    MakePlayerStatsTable<constructor>(PlayerStatsTable)
                    LuckScore := MakeUpdatedPlayerStat(LuckStat, LuckScore)

    #Code1Entered
    RecordCode1Entered(Agent:agent, Stat:stat_type, ?Code1Entered:int = 0):void=
        if:
            Player := player[Agent]
            PlayerStatsTable := PlayerStatsMap[Player]
            if(Stat.DebugString() = StatType.Code1Entered.DebugString()):
                Code1EnteredStat := PlayerStatsTable.Code1Entered
                set PlayerStatsMap[Player] = player_stats_table:
                    MakePlayerStatsTable<constructor>(PlayerStatsTable)
                    Code1Entered := MakeUpdatedPlayerStat(Code1EnteredStat, Code1Entered)

    #Code2Entered
    RecordCode2Entered(Agent:agent, Stat:stat_type, ?Code2Entered:int = 0):void=
        if:
            Player := player[Agent]
            PlayerStatsTable := PlayerStatsMap[Player]
            if(Stat.DebugString() = StatType.Code2Entered.DebugString()):
                Code2EnteredStat := PlayerStatsTable.Code2Entered
                set PlayerStatsMap[Player] = player_stats_table:
                    MakePlayerStatsTable<constructor>(PlayerStatsTable)
                    Code2Entered := MakeUpdatedPlayerStat(Code2EnteredStat, Code2Entered)

    #Code3Entered
    RecordCode3Entered(Agent:agent, Stat:stat_type, ?Code3Entered:int = 0):void=
        if:
            Player := player[Agent]
            PlayerStatsTable := PlayerStatsMap[Player]
            if(Stat.DebugString() = StatType.Code3Entered.DebugString()):
                Code3EnteredStat := PlayerStatsTable.Code3Entered
                set PlayerStatsMap[Player] = player_stats_table:
                    MakePlayerStatsTable<constructor>(PlayerStatsTable)
                    Code3Entered := MakeUpdatedPlayerStat(Code3EnteredStat, Code3Entered)

    #Code4Entered
    RecordCode4Entered(Agent:agent, Stat:stat_type, ?Code4Entered:int = 0):void=
        if:
            Player := player[Agent]
            PlayerStatsTable := PlayerStatsMap[Player]
            if(Stat.DebugString() = StatType.Code4Entered.DebugString()):
                Code4EnteredStat := PlayerStatsTable.Code4Entered
                set PlayerStatsMap[Player] = player_stats_table:
                    MakePlayerStatsTable<constructor>(PlayerStatsTable)
                    Code4Entered := MakeUpdatedPlayerStat(Code4EnteredStat, Code4Entered)

    #Pet Ownership
    #---------------------------------------L1
    #---------------------------------------
    RecordOwnershipL1P1(Agent:agent, Stat:stat_type, ?Ownership_L1_Pet1:int = 0):void=
        if:
            Player := player[Agent]
            PlayerStatsTable := PlayerStatsMap[Player]
            if(Stat.DebugString() = StatType.Ownership_L1_Pet1.DebugString()):
                OwnershipStat := PlayerStatsTable.Ownership_L1_Pet1
                set PlayerStatsMap[Player] = player_stats_table:
                    MakePlayerStatsTable<constructor>(PlayerStatsTable)
                    Ownership_L1_Pet1 := MakeUpdatedPlayerStat(OwnershipStat, 1)

    RemoveOwnershipL1P1(Agent:agent, Stat:stat_type, ?Ownership_L1_Pet1:int = 0):void=
        if:
            Player := player[Agent]
            PlayerStatsTable := PlayerStatsMap[Player]
            if(Stat.DebugString() = StatType.Ownership_L1_Pet1.DebugString()):
                OwnershipStat := PlayerStatsTable.Ownership_L1_Pet1
                set PlayerStatsMap[Player] = player_stats_table:
                    MakePlayerStatsTable<constructor>(PlayerStatsTable)
                    Ownership_L1_Pet1 := MakeUpdatedPlayerStat(OwnershipStat, 0)

    RecordOwnershipL1P2(Agent:agent, Stat:stat_type, ?Ownership_L1_Pet2:int = 0):void=
        if:
            Player := player[Agent]
            PlayerStatsTable := PlayerStatsMap[Player]
            if(Stat.DebugString() = StatType.Ownership_L1_Pet2.DebugString()):
                OwnershipStat := PlayerStatsTable.Ownership_L1_Pet2
                set PlayerStatsMap[Player] = player_stats_table:
                    MakePlayerStatsTable<constructor>(PlayerStatsTable)
                    Ownership_L1_Pet2 := MakeUpdatedPlayerStat(OwnershipStat, 1)

    RemoveOwnershipL1P2(Agent:agent, Stat:stat_type, ?Ownership_L1_Pet2:int = 0):void=
        if:
            Player := player[Agent]
            PlayerStatsTable := PlayerStatsMap[Player]
            if(Stat.DebugString() = StatType.Ownership_L1_Pet2.DebugString()):
                OwnershipStat := PlayerStatsTable.Ownership_L1_Pet2
                set PlayerStatsMap[Player] = player_stats_table:
                    MakePlayerStatsTable<constructor>(PlayerStatsTable)
                    Ownership_L1_Pet2 := MakeUpdatedPlayerStat(OwnershipStat, 0)

    RecordOwnershipL1P3(Agent:agent, Stat:stat_type, ?Ownership_L1_Pet3:int = 0):void=
        if:
            Player := player[Agent]
            PlayerStatsTable := PlayerStatsMap[Player]
            if(Stat.DebugString() = StatType.Ownership_L1_Pet3.DebugString()):
                OwnershipStat := PlayerStatsTable.Ownership_L1_Pet3
                set PlayerStatsMap[Player] = player_stats_table:
                    MakePlayerStatsTable<constructor>(PlayerStatsTable)
                    Ownership_L1_Pet3 := MakeUpdatedPlayerStat(OwnershipStat, 1)

    RemoveOwnershipL1P3(Agent:agent, Stat:stat_type, ?Ownership_L1_Pet3:int = 0):void=
        if:
            Player := player[Agent]
            PlayerStatsTable := PlayerStatsMap[Player]
            if(Stat.DebugString() = StatType.Ownership_L1_Pet3.DebugString()):
                OwnershipStat := PlayerStatsTable.Ownership_L1_Pet3
                set PlayerStatsMap[Player] = player_stats_table:
                    MakePlayerStatsTable<constructor>(PlayerStatsTable)
                    Ownership_L1_Pet3 := MakeUpdatedPlayerStat(OwnershipStat, 0)

    RecordOwnershipL1P4(Agent:agent, Stat:stat_type, ?Ownership_L1_Pet4:int = 0):void=
        if:
            Player := player[Agent]
            PlayerStatsTable := PlayerStatsMap[Player]
            if(Stat.DebugString() = StatType.Ownership_L1_Pet4.DebugString()):
                OwnershipStat := PlayerStatsTable.Ownership_L1_Pet4
                set PlayerStatsMap[Player] = player_stats_table:
                    MakePlayerStatsTable<constructor>(PlayerStatsTable)
                    Ownership_L1_Pet4 := MakeUpdatedPlayerStat(OwnershipStat, 1)

    RemoveOwnershipL1P4(Agent:agent, Stat:stat_type, ?Ownership_L1_Pet4:int = 0):void=
        if:
            Player := player[Agent]
            PlayerStatsTable := PlayerStatsMap[Player]
            if(Stat.DebugString() = StatType.Ownership_L1_Pet4.DebugString()):
                OwnershipStat := PlayerStatsTable.Ownership_L1_Pet4
                set PlayerStatsMap[Player] = player_stats_table:
                    MakePlayerStatsTable<constructor>(PlayerStatsTable)
                    Ownership_L1_Pet4 := MakeUpdatedPlayerStat(OwnershipStat, 0)
    
    #---------------------------------------L2
    #---------------------------------------
    RecordOwnershipL2P1(Agent:agent, Stat:stat_type, ?Ownership_L2_Pet1:int = 0):void=
        if:
            Player := player[Agent]
            PlayerStatsTable := PlayerStatsMap[Player]
            if(Stat.DebugString() = StatType.Ownership_L2_Pet1.DebugString()):
                OwnershipStat := PlayerStatsTable.Ownership_L2_Pet1
                set PlayerStatsMap[Player] = player_stats_table:
                    MakePlayerStatsTable<constructor>(PlayerStatsTable)
                    Ownership_L2_Pet1 := MakeUpdatedPlayerStat(OwnershipStat, 1)

    RemoveOwnershipL2P1(Agent:agent, Stat:stat_type, ?Ownership_L2_Pet1:int = 0):void=
        if:
            Player := player[Agent]
            PlayerStatsTable := PlayerStatsMap[Player]
            if(Stat.DebugString() = StatType.Ownership_L2_Pet1.DebugString()):
                OwnershipStat := PlayerStatsTable.Ownership_L2_Pet1
                set PlayerStatsMap[Player] = player_stats_table:
                    MakePlayerStatsTable<constructor>(PlayerStatsTable)
                    Ownership_L2_Pet1 := MakeUpdatedPlayerStat(OwnershipStat, 0)

    RecordOwnershipL2P2(Agent:agent, Stat:stat_type, ?Ownership_L2_Pet2:int = 0):void=
        if:
            Player := player[Agent]
            PlayerStatsTable := PlayerStatsMap[Player]
            if(Stat.DebugString() = StatType.Ownership_L2_Pet2.DebugString()):
                OwnershipStat := PlayerStatsTable.Ownership_L2_Pet2
                set PlayerStatsMap[Player] = player_stats_table:
                    MakePlayerStatsTable<constructor>(PlayerStatsTable)
                    Ownership_L2_Pet2 := MakeUpdatedPlayerStat(OwnershipStat, 1)

    RemoveOwnershipL2P2(Agent:agent, Stat:stat_type, ?Ownership_L2_Pet2:int = 0):void=
        if:
            Player := player[Agent]
            PlayerStatsTable := PlayerStatsMap[Player]
            if(Stat.DebugString() = StatType.Ownership_L2_Pet2.DebugString()):
                OwnershipStat := PlayerStatsTable.Ownership_L2_Pet2
                set PlayerStatsMap[Player] = player_stats_table:
                    MakePlayerStatsTable<constructor>(PlayerStatsTable)
                    Ownership_L2_Pet2 := MakeUpdatedPlayerStat(OwnershipStat, 0)

    RecordOwnershipL2P3(Agent:agent, Stat:stat_type, ?Ownership_L2_Pet3:int = 0):void=
        if:
            Player := player[Agent]
            PlayerStatsTable := PlayerStatsMap[Player]
            if(Stat.DebugString() = StatType.Ownership_L2_Pet3.DebugString()):
                OwnershipStat := PlayerStatsTable.Ownership_L2_Pet3
                set PlayerStatsMap[Player] = player_stats_table:
                    MakePlayerStatsTable<constructor>(PlayerStatsTable)
                    Ownership_L2_Pet3 := MakeUpdatedPlayerStat(OwnershipStat, 1)

    RemoveOwnershipL2P3(Agent:agent, Stat:stat_type, ?Ownership_L2_Pet3:int = 0):void=
        if:
            Player := player[Agent]
            PlayerStatsTable := PlayerStatsMap[Player]
            if(Stat.DebugString() = StatType.Ownership_L2_Pet3.DebugString()):
                OwnershipStat := PlayerStatsTable.Ownership_L2_Pet3
                set PlayerStatsMap[Player] = player_stats_table:
                    MakePlayerStatsTable<constructor>(PlayerStatsTable)
                    Ownership_L2_Pet3 := MakeUpdatedPlayerStat(OwnershipStat, 0)

    RecordOwnershipL2P4(Agent:agent, Stat:stat_type, ?Ownership_L2_Pet4:int = 0):void=
        if:
            Player := player[Agent]
            PlayerStatsTable := PlayerStatsMap[Player]
            if(Stat.DebugString() = StatType.Ownership_L2_Pet4.DebugString()):
                OwnershipStat := PlayerStatsTable.Ownership_L2_Pet4
                set PlayerStatsMap[Player] = player_stats_table:
                    MakePlayerStatsTable<constructor>(PlayerStatsTable)
                    Ownership_L2_Pet4 := MakeUpdatedPlayerStat(OwnershipStat, 1)

    RemoveOwnershipL2P4(Agent:agent, Stat:stat_type, ?Ownership_L2_Pet4:int = 0):void=
        if:
            Player := player[Agent]
            PlayerStatsTable := PlayerStatsMap[Player]
            if(Stat.DebugString() = StatType.Ownership_L2_Pet4.DebugString()):
                OwnershipStat := PlayerStatsTable.Ownership_L2_Pet4
                set PlayerStatsMap[Player] = player_stats_table:
                    MakePlayerStatsTable<constructor>(PlayerStatsTable)
                    Ownership_L2_Pet4 := MakeUpdatedPlayerStat(OwnershipStat, 0)

    #---------------------------------------L3
    #---------------------------------------
    RecordOwnershipL3P1(Agent:agent, Stat:stat_type, ?Ownership_L3_Pet1:int = 0):void=
        if:
            Player := player[Agent]
            PlayerStatsTable := PlayerStatsMap[Player]
            if(Stat.DebugString() = StatType.Ownership_L3_Pet1.DebugString()):
                OwnershipStat := PlayerStatsTable.Ownership_L3_Pet1
                set PlayerStatsMap[Player] = player_stats_table:
                    MakePlayerStatsTable<constructor>(PlayerStatsTable)
                    Ownership_L3_Pet1 := MakeUpdatedPlayerStat(OwnershipStat, 1)

    RemoveOwnershipL3P1(Agent:agent, Stat:stat_type, ?Ownership_L3_Pet1:int = 0):void=
        if:
            Player := player[Agent]
            PlayerStatsTable := PlayerStatsMap[Player]
            if(Stat.DebugString() = StatType.Ownership_L3_Pet1.DebugString()):
                OwnershipStat := PlayerStatsTable.Ownership_L3_Pet1
                set PlayerStatsMap[Player] = player_stats_table:
                    MakePlayerStatsTable<constructor>(PlayerStatsTable)
                    Ownership_L3_Pet1 := MakeUpdatedPlayerStat(OwnershipStat, 0)

    RecordOwnershipL3P2(Agent:agent, Stat:stat_type, ?Ownership_L3_Pet2:int = 0):void=
        if:
            Player := player[Agent]
            PlayerStatsTable := PlayerStatsMap[Player]
            if(Stat.DebugString() = StatType.Ownership_L3_Pet2.DebugString()):
                OwnershipStat := PlayerStatsTable.Ownership_L3_Pet2
                set PlayerStatsMap[Player] = player_stats_table:
                    MakePlayerStatsTable<constructor>(PlayerStatsTable)
                    Ownership_L3_Pet2 := MakeUpdatedPlayerStat(OwnershipStat, 1)

    RemoveOwnershipL3P2(Agent:agent, Stat:stat_type, ?Ownership_L3_Pet2:int = 0):void=
        if:
            Player := player[Agent]
            PlayerStatsTable := PlayerStatsMap[Player]
            if(Stat.DebugString() = StatType.Ownership_L3_Pet2.DebugString()):
                OwnershipStat := PlayerStatsTable.Ownership_L3_Pet2
                set PlayerStatsMap[Player] = player_stats_table:
                    MakePlayerStatsTable<constructor>(PlayerStatsTable)
                    Ownership_L3_Pet2 := MakeUpdatedPlayerStat(OwnershipStat, 0)

    RecordOwnershipL3P3(Agent:agent, Stat:stat_type, ?Ownership_L3_Pet3:int = 0):void=
        if:
            Player := player[Agent]
            PlayerStatsTable := PlayerStatsMap[Player]
            if(Stat.DebugString() = StatType.Ownership_L3_Pet3.DebugString()):
                OwnershipStat := PlayerStatsTable.Ownership_L3_Pet3
                set PlayerStatsMap[Player] = player_stats_table:
                    MakePlayerStatsTable<constructor>(PlayerStatsTable)
                    Ownership_L3_Pet3 := MakeUpdatedPlayerStat(OwnershipStat, 1)

    RemoveOwnershipL3P3(Agent:agent, Stat:stat_type, ?Ownership_L3_Pet3:int = 0):void=
        if:
            Player := player[Agent]
            PlayerStatsTable := PlayerStatsMap[Player]
            if(Stat.DebugString() = StatType.Ownership_L3_Pet3.DebugString()):
                OwnershipStat := PlayerStatsTable.Ownership_L3_Pet3
                set PlayerStatsMap[Player] = player_stats_table:
                    MakePlayerStatsTable<constructor>(PlayerStatsTable)
                    Ownership_L3_Pet3 := MakeUpdatedPlayerStat(OwnershipStat, 0)

    RecordOwnershipL3P4(Agent:agent, Stat:stat_type, ?Ownership_L3_Pet4:int = 0):void=
        if:
            Player := player[Agent]
            PlayerStatsTable := PlayerStatsMap[Player]
            if(Stat.DebugString() = StatType.Ownership_L3_Pet4.DebugString()):
                OwnershipStat := PlayerStatsTable.Ownership_L3_Pet4
                set PlayerStatsMap[Player] = player_stats_table:
                    MakePlayerStatsTable<constructor>(PlayerStatsTable)
                    Ownership_L3_Pet4 := MakeUpdatedPlayerStat(OwnershipStat, 1)

    RemoveOwnershipL3P4(Agent:agent, Stat:stat_type, ?Ownership_L3_Pet4:int = 0):void=
        if:
            Player := player[Agent]
            PlayerStatsTable := PlayerStatsMap[Player]
            if(Stat.DebugString() = StatType.Ownership_L3_Pet4.DebugString()):
                OwnershipStat := PlayerStatsTable.Ownership_L3_Pet4
                set PlayerStatsMap[Player] = player_stats_table:
                    MakePlayerStatsTable<constructor>(PlayerStatsTable)
                    Ownership_L3_Pet4 := MakeUpdatedPlayerStat(OwnershipStat, 0)

    #---------------------------------------L4
    #---------------------------------------
    RecordOwnershipL4P1(Agent:agent, Stat:stat_type, ?Ownership_L4_Pet1:int = 0):void=
        if:
            Player := player[Agent]
            PlayerStatsTable := PlayerStatsMap[Player]
            if(Stat.DebugString() = StatType.Ownership_L4_Pet1.DebugString()):
                OwnershipStat := PlayerStatsTable.Ownership_L4_Pet1
                set PlayerStatsMap[Player] = player_stats_table:
                    MakePlayerStatsTable<constructor>(PlayerStatsTable)
                    Ownership_L4_Pet1 := MakeUpdatedPlayerStat(OwnershipStat, 1)

    RemoveOwnershipL4P1(Agent:agent, Stat:stat_type, ?Ownership_L4_Pet1:int = 0):void=
        if:
            Player := player[Agent]
            PlayerStatsTable := PlayerStatsMap[Player]
            if(Stat.DebugString() = StatType.Ownership_L4_Pet1.DebugString()):
                OwnershipStat := PlayerStatsTable.Ownership_L4_Pet1
                set PlayerStatsMap[Player] = player_stats_table:
                    MakePlayerStatsTable<constructor>(PlayerStatsTable)
                    Ownership_L4_Pet1 := MakeUpdatedPlayerStat(OwnershipStat, 0)

    RecordOwnershipL4P2(Agent:agent, Stat:stat_type, ?Ownership_L4_Pet2:int = 0):void=
        if:
            Player := player[Agent]
            PlayerStatsTable := PlayerStatsMap[Player]
            if(Stat.DebugString() = StatType.Ownership_L4_Pet2.DebugString()):
                OwnershipStat := PlayerStatsTable.Ownership_L4_Pet2
                set PlayerStatsMap[Player] = player_stats_table:
                    MakePlayerStatsTable<constructor>(PlayerStatsTable)
                    Ownership_L4_Pet2 := MakeUpdatedPlayerStat(OwnershipStat, 1)

    RemoveOwnershipL4P2(Agent:agent, Stat:stat_type, ?Ownership_L4_Pet2:int = 0):void=
        if:
            Player := player[Agent]
            PlayerStatsTable := PlayerStatsMap[Player]
            if(Stat.DebugString() = StatType.Ownership_L4_Pet2.DebugString()):
                OwnershipStat := PlayerStatsTable.Ownership_L4_Pet2
                set PlayerStatsMap[Player] = player_stats_table:
                    MakePlayerStatsTable<constructor>(PlayerStatsTable)
                    Ownership_L4_Pet2 := MakeUpdatedPlayerStat(OwnershipStat, 0)

    RecordOwnershipL4P3(Agent:agent, Stat:stat_type, ?Ownership_L4_Pet3:int = 0):void=
        if:
            Player := player[Agent]
            PlayerStatsTable := PlayerStatsMap[Player]
            if(Stat.DebugString() = StatType.Ownership_L4_Pet3.DebugString()):
                OwnershipStat := PlayerStatsTable.Ownership_L4_Pet3
                set PlayerStatsMap[Player] = player_stats_table:
                    MakePlayerStatsTable<constructor>(PlayerStatsTable)
                    Ownership_L4_Pet3 := MakeUpdatedPlayerStat(OwnershipStat, 1)

    RemoveOwnershipL4P3(Agent:agent, Stat:stat_type, ?Ownership_L4_Pet3:int = 0):void=
        if:
            Player := player[Agent]
            PlayerStatsTable := PlayerStatsMap[Player]
            if(Stat.DebugString() = StatType.Ownership_L4_Pet3.DebugString()):
                OwnershipStat := PlayerStatsTable.Ownership_L4_Pet3
                set PlayerStatsMap[Player] = player_stats_table:
                    MakePlayerStatsTable<constructor>(PlayerStatsTable)
                    Ownership_L4_Pet3 := MakeUpdatedPlayerStat(OwnershipStat, 0)

    RecordOwnershipL4P4(Agent:agent, Stat:stat_type, ?Ownership_L4_Pet4:int = 0):void=
        if:
            Player := player[Agent]
            PlayerStatsTable := PlayerStatsMap[Player]
            if(Stat.DebugString() = StatType.Ownership_L4_Pet4.DebugString()):
                OwnershipStat := PlayerStatsTable.Ownership_L4_Pet4
                set PlayerStatsMap[Player] = player_stats_table:
                    MakePlayerStatsTable<constructor>(PlayerStatsTable)
                    Ownership_L4_Pet4 := MakeUpdatedPlayerStat(OwnershipStat, 1)

    RemoveOwnershipL4P4(Agent:agent, Stat:stat_type, ?Ownership_L4_Pet4:int = 0):void=
        if:
            Player := player[Agent]
            PlayerStatsTable := PlayerStatsMap[Player]
            if(Stat.DebugString() = StatType.Ownership_L4_Pet4.DebugString()):
                OwnershipStat := PlayerStatsTable.Ownership_L4_Pet4
                set PlayerStatsMap[Player] = player_stats_table:
                    MakePlayerStatsTable<constructor>(PlayerStatsTable)
                    Ownership_L4_Pet4 := MakeUpdatedPlayerStat(OwnershipStat, 0)

    #---------------------------------------L5
    #---------------------------------------
    RecordOwnershipL5P1(Agent:agent, Stat:stat_type, ?Ownership_L5_Pet1:int = 0):void=
        if:
            Player := player[Agent]
            PlayerStatsTable := PlayerStatsMap[Player]
            if(Stat.DebugString() = StatType.Ownership_L5_Pet1.DebugString()):
                OwnershipStat := PlayerStatsTable.Ownership_L5_Pet1
                set PlayerStatsMap[Player] = player_stats_table:
                    MakePlayerStatsTable<constructor>(PlayerStatsTable)
                    Ownership_L5_Pet1 := MakeUpdatedPlayerStat(OwnershipStat, 1)

    RemoveOwnershipL5P1(Agent:agent, Stat:stat_type, ?Ownership_L5_Pet1:int = 0):void=
        if:
            Player := player[Agent]
            PlayerStatsTable := PlayerStatsMap[Player]
            if(Stat.DebugString() = StatType.Ownership_L5_Pet1.DebugString()):
                OwnershipStat := PlayerStatsTable.Ownership_L5_Pet1
                set PlayerStatsMap[Player] = player_stats_table:
                    MakePlayerStatsTable<constructor>(PlayerStatsTable)
                    Ownership_L5_Pet1 := MakeUpdatedPlayerStat(OwnershipStat, 0)

    RecordOwnershipL5P2(Agent:agent, Stat:stat_type, ?Ownership_L5_Pet2:int = 0):void=
        if:
            Player := player[Agent]
            PlayerStatsTable := PlayerStatsMap[Player]
            if(Stat.DebugString() = StatType.Ownership_L5_Pet2.DebugString()):
                OwnershipStat := PlayerStatsTable.Ownership_L5_Pet2
                set PlayerStatsMap[Player] = player_stats_table:
                    MakePlayerStatsTable<constructor>(PlayerStatsTable)
                    Ownership_L5_Pet2 := MakeUpdatedPlayerStat(OwnershipStat, 1)

    RemoveOwnershipL5P2(Agent:agent, Stat:stat_type, ?Ownership_L5_Pet2:int = 0):void=
        if:
            Player := player[Agent]
            PlayerStatsTable := PlayerStatsMap[Player]
            if(Stat.DebugString() = StatType.Ownership_L5_Pet2.DebugString()):
                OwnershipStat := PlayerStatsTable.Ownership_L5_Pet2
                set PlayerStatsMap[Player] = player_stats_table:
                    MakePlayerStatsTable<constructor>(PlayerStatsTable)
                    Ownership_L5_Pet2 := MakeUpdatedPlayerStat(OwnershipStat, 0)

    RecordOwnershipL5P3(Agent:agent, Stat:stat_type, ?Ownership_L5_Pet3:int = 0):void=
        if:
            Player := player[Agent]
            PlayerStatsTable := PlayerStatsMap[Player]
            if(Stat.DebugString() = StatType.Ownership_L5_Pet3.DebugString()):
                OwnershipStat := PlayerStatsTable.Ownership_L5_Pet3
                set PlayerStatsMap[Player] = player_stats_table:
                    MakePlayerStatsTable<constructor>(PlayerStatsTable)
                    Ownership_L5_Pet3 := MakeUpdatedPlayerStat(OwnershipStat, 1)

    RemoveOwnershipL5P3(Agent:agent, Stat:stat_type, ?Ownership_L5_Pet3:int = 0):void=
        if:
            Player := player[Agent]
            PlayerStatsTable := PlayerStatsMap[Player]
            if(Stat.DebugString() = StatType.Ownership_L5_Pet3.DebugString()):
                OwnershipStat := PlayerStatsTable.Ownership_L5_Pet3
                set PlayerStatsMap[Player] = player_stats_table:
                    MakePlayerStatsTable<constructor>(PlayerStatsTable)
                    Ownership_L5_Pet3 := MakeUpdatedPlayerStat(OwnershipStat, 0)

    RecordOwnershipL5P4(Agent:agent, Stat:stat_type, ?Ownership_L5_Pet4:int = 0):void=
        if:
            Player := player[Agent]
            PlayerStatsTable := PlayerStatsMap[Player]
            if(Stat.DebugString() = StatType.Ownership_L5_Pet4.DebugString()):
                OwnershipStat := PlayerStatsTable.Ownership_L5_Pet4
                set PlayerStatsMap[Player] = player_stats_table:
                    MakePlayerStatsTable<constructor>(PlayerStatsTable)
                    Ownership_L5_Pet4 := MakeUpdatedPlayerStat(OwnershipStat, 1)

    RemoveOwnershipL5P4(Agent:agent, Stat:stat_type, ?Ownership_L5_Pet4:int = 0):void=
        if:
            Player := player[Agent]
            PlayerStatsTable := PlayerStatsMap[Player]
            if(Stat.DebugString() = StatType.Ownership_L5_Pet4.DebugString()):
                OwnershipStat := PlayerStatsTable.Ownership_L5_Pet4
                set PlayerStatsMap[Player] = player_stats_table:
                    MakePlayerStatsTable<constructor>(PlayerStatsTable)
                    Ownership_L5_Pet4 := MakeUpdatedPlayerStat(OwnershipStat, 0)

    #---------------------------------------L6
    #---------------------------------------
    RecordOwnershipL6P1(Agent:agent, Stat:stat_type, ?Ownership_L6_Pet1:int = 0):void=
        if:
            Player := player[Agent]
            PlayerStatsTable := PlayerStatsMap[Player]
            if(Stat.DebugString() = StatType.Ownership_L6_Pet1.DebugString()):
                OwnershipStat := PlayerStatsTable.Ownership_L6_Pet1
                set PlayerStatsMap[Player] = player_stats_table:
                    MakePlayerStatsTable<constructor>(PlayerStatsTable)
                    Ownership_L6_Pet1 := MakeUpdatedPlayerStat(OwnershipStat, 1)

    RemoveOwnershipL6P1(Agent:agent, Stat:stat_type, ?Ownership_L6_Pet1:int = 0):void=
        if:
            Player := player[Agent]
            PlayerStatsTable := PlayerStatsMap[Player]
            if(Stat.DebugString() = StatType.Ownership_L6_Pet1.DebugString()):
                OwnershipStat := PlayerStatsTable.Ownership_L6_Pet1
                set PlayerStatsMap[Player] = player_stats_table:
                    MakePlayerStatsTable<constructor>(PlayerStatsTable)
                    Ownership_L6_Pet1 := MakeUpdatedPlayerStat(OwnershipStat, 0)

    RecordOwnershipL6P2(Agent:agent, Stat:stat_type, ?Ownership_L6_Pet2:int = 0):void=
        if:
            Player := player[Agent]
            PlayerStatsTable := PlayerStatsMap[Player]
            if(Stat.DebugString() = StatType.Ownership_L6_Pet2.DebugString()):
                OwnershipStat := PlayerStatsTable.Ownership_L6_Pet2
                set PlayerStatsMap[Player] = player_stats_table:
                    MakePlayerStatsTable<constructor>(PlayerStatsTable)
                    Ownership_L6_Pet2 := MakeUpdatedPlayerStat(OwnershipStat, 1)

    RemoveOwnershipL6P2(Agent:agent, Stat:stat_type, ?Ownership_L6_Pet2:int = 0):void=
        if:
            Player := player[Agent]
            PlayerStatsTable := PlayerStatsMap[Player]
            if(Stat.DebugString() = StatType.Ownership_L6_Pet2.DebugString()):
                OwnershipStat := PlayerStatsTable.Ownership_L6_Pet2
                set PlayerStatsMap[Player] = player_stats_table:
                    MakePlayerStatsTable<constructor>(PlayerStatsTable)
                    Ownership_L6_Pet2 := MakeUpdatedPlayerStat(OwnershipStat, 0)

    RecordOwnershipL6P3(Agent:agent, Stat:stat_type, ?Ownership_L6_Pet3:int = 0):void=
        if:
            Player := player[Agent]
            PlayerStatsTable := PlayerStatsMap[Player]
            if(Stat.DebugString() = StatType.Ownership_L6_Pet3.DebugString()):
                OwnershipStat := PlayerStatsTable.Ownership_L6_Pet3
                set PlayerStatsMap[Player] = player_stats_table:
                    MakePlayerStatsTable<constructor>(PlayerStatsTable)
                    Ownership_L6_Pet3 := MakeUpdatedPlayerStat(OwnershipStat, 1)

    RemoveOwnershipL6P3(Agent:agent, Stat:stat_type, ?Ownership_L6_Pet3:int = 0):void=
        if:
            Player := player[Agent]
            PlayerStatsTable := PlayerStatsMap[Player]
            if(Stat.DebugString() = StatType.Ownership_L6_Pet3.DebugString()):
                OwnershipStat := PlayerStatsTable.Ownership_L6_Pet3
                set PlayerStatsMap[Player] = player_stats_table:
                    MakePlayerStatsTable<constructor>(PlayerStatsTable)
                    Ownership_L6_Pet3 := MakeUpdatedPlayerStat(OwnershipStat, 0)

    RecordOwnershipL6P4(Agent:agent, Stat:stat_type, ?Ownership_L6_Pet4:int = 0):void=
        if:
            Player := player[Agent]
            PlayerStatsTable := PlayerStatsMap[Player]
            if(Stat.DebugString() = StatType.Ownership_L6_Pet4.DebugString()):
                OwnershipStat := PlayerStatsTable.Ownership_L6_Pet4
                set PlayerStatsMap[Player] = player_stats_table:
                    MakePlayerStatsTable<constructor>(PlayerStatsTable)
                    Ownership_L6_Pet4 := MakeUpdatedPlayerStat(OwnershipStat, 1)

    RemoveOwnershipL6P4(Agent:agent, Stat:stat_type, ?Ownership_L6_Pet4:int = 0):void=
        if:
            Player := player[Agent]
            PlayerStatsTable := PlayerStatsMap[Player]
            if(Stat.DebugString() = StatType.Ownership_L6_Pet4.DebugString()):
                OwnershipStat := PlayerStatsTable.Ownership_L6_Pet4
                set PlayerStatsMap[Player] = player_stats_table:
                    MakePlayerStatsTable<constructor>(PlayerStatsTable)
                    Ownership_L6_Pet4 := MakeUpdatedPlayerStat(OwnershipStat, 0)

    #---------------------------------------L7
    #---------------------------------------
    RecordOwnershipL7P1(Agent:agent, Stat:stat_type, ?Ownership_L7_Pet1:int = 0):void=
        if:
            Player := player[Agent]
            PlayerStatsTable := PlayerStatsMap[Player]
            if(Stat.DebugString() = StatType.Ownership_L7_Pet1.DebugString()):
                OwnershipStat := PlayerStatsTable.Ownership_L7_Pet1
                set PlayerStatsMap[Player] = player_stats_table:
                    MakePlayerStatsTable<constructor>(PlayerStatsTable)
                    Ownership_L7_Pet1 := MakeUpdatedPlayerStat(OwnershipStat, 1)

    RemoveOwnershipL7P1(Agent:agent, Stat:stat_type, ?Ownership_L7_Pet1:int = 0):void=
        if:
            Player := player[Agent]
            PlayerStatsTable := PlayerStatsMap[Player]
            if(Stat.DebugString() = StatType.Ownership_L7_Pet1.DebugString()):
                OwnershipStat := PlayerStatsTable.Ownership_L7_Pet1
                set PlayerStatsMap[Player] = player_stats_table:
                    MakePlayerStatsTable<constructor>(PlayerStatsTable)
                    Ownership_L7_Pet1 := MakeUpdatedPlayerStat(OwnershipStat, 0)

    RecordOwnershipL7P2(Agent:agent, Stat:stat_type, ?Ownership_L7_Pet2:int = 0):void=
        if:
            Player := player[Agent]
            PlayerStatsTable := PlayerStatsMap[Player]
            if(Stat.DebugString() = StatType.Ownership_L7_Pet2.DebugString()):
                OwnershipStat := PlayerStatsTable.Ownership_L7_Pet2
                set PlayerStatsMap[Player] = player_stats_table:
                    MakePlayerStatsTable<constructor>(PlayerStatsTable)
                    Ownership_L7_Pet2 := MakeUpdatedPlayerStat(OwnershipStat, 1)

    RemoveOwnershipL7P2(Agent:agent, Stat:stat_type, ?Ownership_L7_Pet2:int = 0):void=
        if:
            Player := player[Agent]
            PlayerStatsTable := PlayerStatsMap[Player]
            if(Stat.DebugString() = StatType.Ownership_L7_Pet2.DebugString()):
                OwnershipStat := PlayerStatsTable.Ownership_L7_Pet2
                set PlayerStatsMap[Player] = player_stats_table:
                    MakePlayerStatsTable<constructor>(PlayerStatsTable)
                    Ownership_L7_Pet2 := MakeUpdatedPlayerStat(OwnershipStat, 0)

    RecordOwnershipL7P3(Agent:agent, Stat:stat_type, ?Ownership_L7_Pet3:int = 0):void=
        if:
            Player := player[Agent]
            PlayerStatsTable := PlayerStatsMap[Player]
            if(Stat.DebugString() = StatType.Ownership_L7_Pet3.DebugString()):
                OwnershipStat := PlayerStatsTable.Ownership_L7_Pet3
                set PlayerStatsMap[Player] = player_stats_table:
                    MakePlayerStatsTable<constructor>(PlayerStatsTable)
                    Ownership_L7_Pet3 := MakeUpdatedPlayerStat(OwnershipStat, 1)

    RemoveOwnershipL7P3(Agent:agent, Stat:stat_type, ?Ownership_L7_Pet3:int = 0):void=
        if:
            Player := player[Agent]
            PlayerStatsTable := PlayerStatsMap[Player]
            if(Stat.DebugString() = StatType.Ownership_L7_Pet3.DebugString()):
                OwnershipStat := PlayerStatsTable.Ownership_L7_Pet3
                set PlayerStatsMap[Player] = player_stats_table:
                    MakePlayerStatsTable<constructor>(PlayerStatsTable)
                    Ownership_L7_Pet3 := MakeUpdatedPlayerStat(OwnershipStat, 0)

    RecordOwnershipL7P4(Agent:agent, Stat:stat_type, ?Ownership_L7_Pet4:int = 0):void=
        if:
            Player := player[Agent]
            PlayerStatsTable := PlayerStatsMap[Player]
            if(Stat.DebugString() = StatType.Ownership_L7_Pet4.DebugString()):
                OwnershipStat := PlayerStatsTable.Ownership_L7_Pet4
                set PlayerStatsMap[Player] = player_stats_table:
                    MakePlayerStatsTable<constructor>(PlayerStatsTable)
                    Ownership_L7_Pet4 := MakeUpdatedPlayerStat(OwnershipStat, 1)

    RemoveOwnershipL7P4(Agent:agent, Stat:stat_type, ?Ownership_L7_Pet4:int = 0):void=
        if:
            Player := player[Agent]
            PlayerStatsTable := PlayerStatsMap[Player]
            if(Stat.DebugString() = StatType.Ownership_L7_Pet4.DebugString()):
                OwnershipStat := PlayerStatsTable.Ownership_L7_Pet4
                set PlayerStatsMap[Player] = player_stats_table:
                    MakePlayerStatsTable<constructor>(PlayerStatsTable)
                    Ownership_L7_Pet4 := MakeUpdatedPlayerStat(OwnershipStat, 0)

    #---------------------------------------L8
    #---------------------------------------
    RecordOwnershipL8P1(Agent:agent, Stat:stat_type, ?Ownership_L8_Pet1:int = 0):void=
        if:
            Player := player[Agent]
            PlayerStatsTable := PlayerStatsMap[Player]
            if(Stat.DebugString() = StatType.Ownership_L8_Pet1.DebugString()):
                OwnershipStat := PlayerStatsTable.Ownership_L8_Pet1
                set PlayerStatsMap[Player] = player_stats_table:
                    MakePlayerStatsTable<constructor>(PlayerStatsTable)
                    Ownership_L8_Pet1 := MakeUpdatedPlayerStat(OwnershipStat, 1)

    RemoveOwnershipL8P1(Agent:agent, Stat:stat_type, ?Ownership_L8_Pet1:int = 0):void=
        if:
            Player := player[Agent]
            PlayerStatsTable := PlayerStatsMap[Player]
            if(Stat.DebugString() = StatType.Ownership_L8_Pet1.DebugString()):
                OwnershipStat := PlayerStatsTable.Ownership_L8_Pet1
                set PlayerStatsMap[Player] = player_stats_table:
                    MakePlayerStatsTable<constructor>(PlayerStatsTable)
                    Ownership_L8_Pet1 := MakeUpdatedPlayerStat(OwnershipStat, 0)

    RecordOwnershipL8P2(Agent:agent, Stat:stat_type, ?Ownership_L8_Pet2:int = 0):void=
        if:
            Player := player[Agent]
            PlayerStatsTable := PlayerStatsMap[Player]
            if(Stat.DebugString() = StatType.Ownership_L8_Pet2.DebugString()):
                OwnershipStat := PlayerStatsTable.Ownership_L8_Pet2
                set PlayerStatsMap[Player] = player_stats_table:
                    MakePlayerStatsTable<constructor>(PlayerStatsTable)
                    Ownership_L8_Pet2 := MakeUpdatedPlayerStat(OwnershipStat, 1)

    RemoveOwnershipL8P2(Agent:agent, Stat:stat_type, ?Ownership_L8_Pet2:int = 0):void=
        if:
            Player := player[Agent]
            PlayerStatsTable := PlayerStatsMap[Player]
            if(Stat.DebugString() = StatType.Ownership_L8_Pet2.DebugString()):
                OwnershipStat := PlayerStatsTable.Ownership_L8_Pet2
                set PlayerStatsMap[Player] = player_stats_table:
                    MakePlayerStatsTable<constructor>(PlayerStatsTable)
                    Ownership_L8_Pet2 := MakeUpdatedPlayerStat(OwnershipStat, 0)

    RecordOwnershipL8P3(Agent:agent, Stat:stat_type, ?Ownership_L8_Pet3:int = 0):void=
        if:
            Player := player[Agent]
            PlayerStatsTable := PlayerStatsMap[Player]
            if(Stat.DebugString() = StatType.Ownership_L8_Pet3.DebugString()):
                OwnershipStat := PlayerStatsTable.Ownership_L8_Pet3
                set PlayerStatsMap[Player] = player_stats_table:
                    MakePlayerStatsTable<constructor>(PlayerStatsTable)
                    Ownership_L8_Pet3 := MakeUpdatedPlayerStat(OwnershipStat, 1)

    RemoveOwnershipL8P3(Agent:agent, Stat:stat_type, ?Ownership_L8_Pet3:int = 0):void=
        if:
            Player := player[Agent]
            PlayerStatsTable := PlayerStatsMap[Player]
            if(Stat.DebugString() = StatType.Ownership_L8_Pet3.DebugString()):
                OwnershipStat := PlayerStatsTable.Ownership_L8_Pet3
                set PlayerStatsMap[Player] = player_stats_table:
                    MakePlayerStatsTable<constructor>(PlayerStatsTable)
                    Ownership_L8_Pet3 := MakeUpdatedPlayerStat(OwnershipStat, 0)

    RecordOwnershipL8P4(Agent:agent, Stat:stat_type, ?Ownership_L8_Pet4:int = 0):void=
        if:
            Player := player[Agent]
            PlayerStatsTable := PlayerStatsMap[Player]
            if(Stat.DebugString() = StatType.Ownership_L8_Pet4.DebugString()):
                OwnershipStat := PlayerStatsTable.Ownership_L8_Pet4
                set PlayerStatsMap[Player] = player_stats_table:
                    MakePlayerStatsTable<constructor>(PlayerStatsTable)
                    Ownership_L8_Pet4 := MakeUpdatedPlayerStat(OwnershipStat, 1)

    RemoveOwnershipL8P4(Agent:agent, Stat:stat_type, ?Ownership_L8_Pet4:int = 0):void=
        if:
            Player := player[Agent]
            PlayerStatsTable := PlayerStatsMap[Player]
            if(Stat.DebugString() = StatType.Ownership_L8_Pet4.DebugString()):
                OwnershipStat := PlayerStatsTable.Ownership_L8_Pet4
                set PlayerStatsMap[Player] = player_stats_table:
                    MakePlayerStatsTable<constructor>(PlayerStatsTable)
                    Ownership_L8_Pet4 := MakeUpdatedPlayerStat(OwnershipStat, 0)

    #---------------------------------------L9
    #---------------------------------------
    RecordOwnershipL9P1(Agent:agent, Stat:stat_type, ?Ownership_L9_Pet1:int = 0):void=
        if:
            Player := player[Agent]
            PlayerStatsTable := PlayerStatsMap[Player]
            if(Stat.DebugString() = StatType.Ownership_L9_Pet1.DebugString()):
                OwnershipStat := PlayerStatsTable.Ownership_L9_Pet1
                set PlayerStatsMap[Player] = player_stats_table:
                    MakePlayerStatsTable<constructor>(PlayerStatsTable)
                    Ownership_L9_Pet1 := MakeUpdatedPlayerStat(OwnershipStat, 1)

    RemoveOwnershipL9P1(Agent:agent, Stat:stat_type, ?Ownership_L9_Pet1:int = 0):void=
        if:
            Player := player[Agent]
            PlayerStatsTable := PlayerStatsMap[Player]
            if(Stat.DebugString() = StatType.Ownership_L9_Pet1.DebugString()):
                OwnershipStat := PlayerStatsTable.Ownership_L9_Pet1
                set PlayerStatsMap[Player] = player_stats_table:
                    MakePlayerStatsTable<constructor>(PlayerStatsTable)
                    Ownership_L9_Pet1 := MakeUpdatedPlayerStat(OwnershipStat, 0)

    RecordOwnershipL9P2(Agent:agent, Stat:stat_type, ?Ownership_L9_Pet2:int = 0):void=
        if:
            Player := player[Agent]
            PlayerStatsTable := PlayerStatsMap[Player]
            if(Stat.DebugString() = StatType.Ownership_L9_Pet2.DebugString()):
                OwnershipStat := PlayerStatsTable.Ownership_L9_Pet2
                set PlayerStatsMap[Player] = player_stats_table:
                    MakePlayerStatsTable<constructor>(PlayerStatsTable)
                    Ownership_L9_Pet2 := MakeUpdatedPlayerStat(OwnershipStat, 1)

    RemoveOwnershipL9P2(Agent:agent, Stat:stat_type, ?Ownership_L9_Pet2:int = 0):void=
        if:
            Player := player[Agent]
            PlayerStatsTable := PlayerStatsMap[Player]
            if(Stat.DebugString() = StatType.Ownership_L9_Pet2.DebugString()):
                OwnershipStat := PlayerStatsTable.Ownership_L9_Pet2
                set PlayerStatsMap[Player] = player_stats_table:
                    MakePlayerStatsTable<constructor>(PlayerStatsTable)
                    Ownership_L9_Pet2 := MakeUpdatedPlayerStat(OwnershipStat, 0)

    RecordOwnershipL9P3(Agent:agent, Stat:stat_type, ?Ownership_L9_Pet3:int = 0):void=
        if:
            Player := player[Agent]
            PlayerStatsTable := PlayerStatsMap[Player]
            if(Stat.DebugString() = StatType.Ownership_L9_Pet3.DebugString()):
                OwnershipStat := PlayerStatsTable.Ownership_L9_Pet3
                set PlayerStatsMap[Player] = player_stats_table:
                    MakePlayerStatsTable<constructor>(PlayerStatsTable)
                    Ownership_L9_Pet3 := MakeUpdatedPlayerStat(OwnershipStat, 1)

    RemoveOwnershipL9P3(Agent:agent, Stat:stat_type, ?Ownership_L9_Pet3:int = 0):void=
        if:
            Player := player[Agent]
            PlayerStatsTable := PlayerStatsMap[Player]
            if(Stat.DebugString() = StatType.Ownership_L9_Pet3.DebugString()):
                OwnershipStat := PlayerStatsTable.Ownership_L9_Pet3
                set PlayerStatsMap[Player] = player_stats_table:
                    MakePlayerStatsTable<constructor>(PlayerStatsTable)
                    Ownership_L9_Pet3 := MakeUpdatedPlayerStat(OwnershipStat, 0)

    RecordOwnershipL9P4(Agent:agent, Stat:stat_type, ?Ownership_L9_Pet4:int = 0):void=
        if:
            Player := player[Agent]
            PlayerStatsTable := PlayerStatsMap[Player]
            if(Stat.DebugString() = StatType.Ownership_L9_Pet4.DebugString()):
                OwnershipStat := PlayerStatsTable.Ownership_L9_Pet4
                set PlayerStatsMap[Player] = player_stats_table:
                    MakePlayerStatsTable<constructor>(PlayerStatsTable)
                    Ownership_L9_Pet4 := MakeUpdatedPlayerStat(OwnershipStat, 1)

    RemoveOwnershipL9P4(Agent:agent, Stat:stat_type, ?Ownership_L9_Pet4:int = 0):void=
        if:
            Player := player[Agent]
            PlayerStatsTable := PlayerStatsMap[Player]
            if(Stat.DebugString() = StatType.Ownership_L9_Pet4.DebugString()):
                OwnershipStat := PlayerStatsTable.Ownership_L9_Pet4
                set PlayerStatsMap[Player] = player_stats_table:
                    MakePlayerStatsTable<constructor>(PlayerStatsTable)
                    Ownership_L9_Pet4 := MakeUpdatedPlayerStat(OwnershipStat, 0)

    #---------------------------------------L10
    #---------------------------------------
    RecordOwnershipL10P1(Agent:agent, Stat:stat_type, ?Ownership_L10_Pet1:int = 0):void=
        if:
            Player := player[Agent]
            PlayerStatsTable := PlayerStatsMap[Player]
            if(Stat.DebugString() = StatType.Ownership_L10_Pet1.DebugString()):
                OwnershipStat := PlayerStatsTable.Ownership_L10_Pet1
                set PlayerStatsMap[Player] = player_stats_table:
                    MakePlayerStatsTable<constructor>(PlayerStatsTable)
                    Ownership_L10_Pet1 := MakeUpdatedPlayerStat(OwnershipStat, 1)

    RemoveOwnershipL10P1(Agent:agent, Stat:stat_type, ?Ownership_L10_Pet1:int = 0):void=
        if:
            Player := player[Agent]
            PlayerStatsTable := PlayerStatsMap[Player]
            if(Stat.DebugString() = StatType.Ownership_L10_Pet1.DebugString()):
                OwnershipStat := PlayerStatsTable.Ownership_L10_Pet1
                set PlayerStatsMap[Player] = player_stats_table:
                    MakePlayerStatsTable<constructor>(PlayerStatsTable)
                    Ownership_L10_Pet1 := MakeUpdatedPlayerStat(OwnershipStat, 0)

    RecordOwnershipL10P2(Agent:agent, Stat:stat_type, ?Ownership_L10_Pet2:int = 0):void=
        if:
            Player := player[Agent]
            PlayerStatsTable := PlayerStatsMap[Player]
            if(Stat.DebugString() = StatType.Ownership_L10_Pet2.DebugString()):
                OwnershipStat := PlayerStatsTable.Ownership_L10_Pet2
                set PlayerStatsMap[Player] = player_stats_table:
                    MakePlayerStatsTable<constructor>(PlayerStatsTable)
                    Ownership_L10_Pet2 := MakeUpdatedPlayerStat(OwnershipStat, 1)

    RemoveOwnershipL10P2(Agent:agent, Stat:stat_type, ?Ownership_L10_Pet2:int = 0):void=
        if:
            Player := player[Agent]
            PlayerStatsTable := PlayerStatsMap[Player]
            if(Stat.DebugString() = StatType.Ownership_L10_Pet2.DebugString()):
                OwnershipStat := PlayerStatsTable.Ownership_L10_Pet2
                set PlayerStatsMap[Player] = player_stats_table:
                    MakePlayerStatsTable<constructor>(PlayerStatsTable)
                    Ownership_L10_Pet2 := MakeUpdatedPlayerStat(OwnershipStat, 0)

    RecordOwnershipL10P3(Agent:agent, Stat:stat_type, ?Ownership_L10_Pet3:int = 0):void=
        if:
            Player := player[Agent]
            PlayerStatsTable := PlayerStatsMap[Player]
            if(Stat.DebugString() = StatType.Ownership_L10_Pet3.DebugString()):
                OwnershipStat := PlayerStatsTable.Ownership_L10_Pet3
                set PlayerStatsMap[Player] = player_stats_table:
                    MakePlayerStatsTable<constructor>(PlayerStatsTable)
                    Ownership_L10_Pet3 := MakeUpdatedPlayerStat(OwnershipStat, 1)

    RemoveOwnershipL10P3(Agent:agent, Stat:stat_type, ?Ownership_L10_Pet3:int = 0):void=
        if:
            Player := player[Agent]
            PlayerStatsTable := PlayerStatsMap[Player]
            if(Stat.DebugString() = StatType.Ownership_L10_Pet3.DebugString()):
                OwnershipStat := PlayerStatsTable.Ownership_L10_Pet3
                set PlayerStatsMap[Player] = player_stats_table:
                    MakePlayerStatsTable<constructor>(PlayerStatsTable)
                    Ownership_L10_Pet3 := MakeUpdatedPlayerStat(OwnershipStat, 0)

    RecordOwnershipL10P4(Agent:agent, Stat:stat_type, ?Ownership_L10_Pet4:int = 0):void=
        if:
            Player := player[Agent]
            PlayerStatsTable := PlayerStatsMap[Player]
            if(Stat.DebugString() = StatType.Ownership_L10_Pet4.DebugString()):
                OwnershipStat := PlayerStatsTable.Ownership_L10_Pet4
                set PlayerStatsMap[Player] = player_stats_table:
                    MakePlayerStatsTable<constructor>(PlayerStatsTable)
                    Ownership_L10_Pet4 := MakeUpdatedPlayerStat(OwnershipStat, 1)

    RemoveOwnershipL10P4(Agent:agent, Stat:stat_type, ?Ownership_L10_Pet4:int = 0):void=
        if:
            Player := player[Agent]
            PlayerStatsTable := PlayerStatsMap[Player]
            if(Stat.DebugString() = StatType.Ownership_L10_Pet4.DebugString()):
                OwnershipStat := PlayerStatsTable.Ownership_L10_Pet4
                set PlayerStatsMap[Player] = player_stats_table:
                    MakePlayerStatsTable<constructor>(PlayerStatsTable)
                    Ownership_L10_Pet4 := MakeUpdatedPlayerStat(OwnershipStat, 0)

    #---------------------------------------L11
    #---------------------------------------
    RecordOwnershipL11P1(Agent:agent, Stat:stat_type, ?Ownership_L11_Pet1:int = 0):void=
        if:
            Player := player[Agent]
            PlayerStatsTable := PlayerStatsMap[Player]
            if(Stat.DebugString() = StatType.Ownership_L11_Pet1.DebugString()):
                OwnershipStat := PlayerStatsTable.Ownership_L11_Pet1
                set PlayerStatsMap[Player] = player_stats_table:
                    MakePlayerStatsTable<constructor>(PlayerStatsTable)
                    Ownership_L11_Pet1 := MakeUpdatedPlayerStat(OwnershipStat, 1)

    RemoveOwnershipL11P1(Agent:agent, Stat:stat_type, ?Ownership_L11_Pet1:int = 0):void=
        if:
            Player := player[Agent]
            PlayerStatsTable := PlayerStatsMap[Player]
            if(Stat.DebugString() = StatType.Ownership_L11_Pet1.DebugString()):
                OwnershipStat := PlayerStatsTable.Ownership_L11_Pet1
                set PlayerStatsMap[Player] = player_stats_table:
                    MakePlayerStatsTable<constructor>(PlayerStatsTable)
                    Ownership_L11_Pet1 := MakeUpdatedPlayerStat(OwnershipStat, 0)

    RecordOwnershipL11P2(Agent:agent, Stat:stat_type, ?Ownership_L11_Pet2:int = 0):void=
        if:
            Player := player[Agent]
            PlayerStatsTable := PlayerStatsMap[Player]
            if(Stat.DebugString() = StatType.Ownership_L11_Pet2.DebugString()):
                OwnershipStat := PlayerStatsTable.Ownership_L11_Pet2
                set PlayerStatsMap[Player] = player_stats_table:
                    MakePlayerStatsTable<constructor>(PlayerStatsTable)
                    Ownership_L11_Pet2 := MakeUpdatedPlayerStat(OwnershipStat, 1)

    RemoveOwnershipL11P2(Agent:agent, Stat:stat_type, ?Ownership_L11_Pet2:int = 0):void=
        if:
            Player := player[Agent]
            PlayerStatsTable := PlayerStatsMap[Player]
            if(Stat.DebugString() = StatType.Ownership_L11_Pet2.DebugString()):
                OwnershipStat := PlayerStatsTable.Ownership_L11_Pet2
                set PlayerStatsMap[Player] = player_stats_table:
                    MakePlayerStatsTable<constructor>(PlayerStatsTable)
                    Ownership_L11_Pet2 := MakeUpdatedPlayerStat(OwnershipStat, 0)

    RecordOwnershipL11P3(Agent:agent, Stat:stat_type, ?Ownership_L11_Pet3:int = 0):void=
        if:
            Player := player[Agent]
            PlayerStatsTable := PlayerStatsMap[Player]
            if(Stat.DebugString() = StatType.Ownership_L11_Pet3.DebugString()):
                OwnershipStat := PlayerStatsTable.Ownership_L11_Pet3
                set PlayerStatsMap[Player] = player_stats_table:
                    MakePlayerStatsTable<constructor>(PlayerStatsTable)
                    Ownership_L11_Pet3 := MakeUpdatedPlayerStat(OwnershipStat, 1)

    RemoveOwnershipL11P3(Agent:agent, Stat:stat_type, ?Ownership_L11_Pet3:int = 0):void=
        if:
            Player := player[Agent]
            PlayerStatsTable := PlayerStatsMap[Player]
            if(Stat.DebugString() = StatType.Ownership_L11_Pet3.DebugString()):
                OwnershipStat := PlayerStatsTable.Ownership_L11_Pet3
                set PlayerStatsMap[Player] = player_stats_table:
                    MakePlayerStatsTable<constructor>(PlayerStatsTable)
                    Ownership_L11_Pet3 := MakeUpdatedPlayerStat(OwnershipStat, 0)

    RecordOwnershipL11P4(Agent:agent, Stat:stat_type, ?Ownership_L11_Pet4:int = 0):void=
        if:
            Player := player[Agent]
            PlayerStatsTable := PlayerStatsMap[Player]
            if(Stat.DebugString() = StatType.Ownership_L11_Pet4.DebugString()):
                OwnershipStat := PlayerStatsTable.Ownership_L11_Pet4
                set PlayerStatsMap[Player] = player_stats_table:
                    MakePlayerStatsTable<constructor>(PlayerStatsTable)
                    Ownership_L11_Pet4 := MakeUpdatedPlayerStat(OwnershipStat, 1)

    RemoveOwnershipL11P4(Agent:agent, Stat:stat_type, ?Ownership_L11_Pet4:int = 0):void=
        if:
            Player := player[Agent]
            PlayerStatsTable := PlayerStatsMap[Player]
            if(Stat.DebugString() = StatType.Ownership_L11_Pet4.DebugString()):
                OwnershipStat := PlayerStatsTable.Ownership_L11_Pet4
                set PlayerStatsMap[Player] = player_stats_table:
                    MakePlayerStatsTable<constructor>(PlayerStatsTable)
                    Ownership_L11_Pet4 := MakeUpdatedPlayerStat(OwnershipStat, 0)

    #---------------------------------------L12
    #---------------------------------------
    RecordOwnershipL12P1(Agent:agent, Stat:stat_type, ?Ownership_L12_Pet1:int = 0):void=
        if:
            Player := player[Agent]
            PlayerStatsTable := PlayerStatsMap[Player]
            if(Stat.DebugString() = StatType.Ownership_L12_Pet1.DebugString()):
                OwnershipStat := PlayerStatsTable.Ownership_L12_Pet1
                set PlayerStatsMap[Player] = player_stats_table:
                    MakePlayerStatsTable<constructor>(PlayerStatsTable)
                    Ownership_L12_Pet1 := MakeUpdatedPlayerStat(OwnershipStat, 1)

    RemoveOwnershipL12P1(Agent:agent, Stat:stat_type, ?Ownership_L12_Pet1:int = 0):void=
        if:
            Player := player[Agent]
            PlayerStatsTable := PlayerStatsMap[Player]
            if(Stat.DebugString() = StatType.Ownership_L12_Pet1.DebugString()):
                OwnershipStat := PlayerStatsTable.Ownership_L12_Pet1
                set PlayerStatsMap[Player] = player_stats_table:
                    MakePlayerStatsTable<constructor>(PlayerStatsTable)
                    Ownership_L12_Pet1 := MakeUpdatedPlayerStat(OwnershipStat, 0)

    RecordOwnershipL12P2(Agent:agent, Stat:stat_type, ?Ownership_L12_Pet2:int = 0):void=
        if:
            Player := player[Agent]
            PlayerStatsTable := PlayerStatsMap[Player]
            if(Stat.DebugString() = StatType.Ownership_L12_Pet2.DebugString()):
                OwnershipStat := PlayerStatsTable.Ownership_L12_Pet2
                set PlayerStatsMap[Player] = player_stats_table:
                    MakePlayerStatsTable<constructor>(PlayerStatsTable)
                    Ownership_L12_Pet2 := MakeUpdatedPlayerStat(OwnershipStat, 1)

    RemoveOwnershipL12P2(Agent:agent, Stat:stat_type, ?Ownership_L12_Pet2:int = 0):void=
        if:
            Player := player[Agent]
            PlayerStatsTable := PlayerStatsMap[Player]
            if(Stat.DebugString() = StatType.Ownership_L12_Pet2.DebugString()):
                OwnershipStat := PlayerStatsTable.Ownership_L12_Pet2
                set PlayerStatsMap[Player] = player_stats_table:
                    MakePlayerStatsTable<constructor>(PlayerStatsTable)
                    Ownership_L12_Pet2 := MakeUpdatedPlayerStat(OwnershipStat, 0)

    RecordOwnershipL12P3(Agent:agent, Stat:stat_type, ?Ownership_L12_Pet3:int = 0):void=
        if:
            Player := player[Agent]
            PlayerStatsTable := PlayerStatsMap[Player]
            if(Stat.DebugString() = StatType.Ownership_L12_Pet3.DebugString()):
                OwnershipStat := PlayerStatsTable.Ownership_L12_Pet3
                set PlayerStatsMap[Player] = player_stats_table:
                    MakePlayerStatsTable<constructor>(PlayerStatsTable)
                    Ownership_L12_Pet3 := MakeUpdatedPlayerStat(OwnershipStat, 1)

    RemoveOwnershipL12P3(Agent:agent, Stat:stat_type, ?Ownership_L12_Pet3:int = 0):void=
        if:
            Player := player[Agent]
            PlayerStatsTable := PlayerStatsMap[Player]
            if(Stat.DebugString() = StatType.Ownership_L12_Pet3.DebugString()):
                OwnershipStat := PlayerStatsTable.Ownership_L12_Pet3
                set PlayerStatsMap[Player] = player_stats_table:
                    MakePlayerStatsTable<constructor>(PlayerStatsTable)
                    Ownership_L12_Pet3 := MakeUpdatedPlayerStat(OwnershipStat, 0)

    RecordOwnershipL12P4(Agent:agent, Stat:stat_type, ?Ownership_L12_Pet4:int = 0):void=
        if:
            Player := player[Agent]
            PlayerStatsTable := PlayerStatsMap[Player]
            if(Stat.DebugString() = StatType.Ownership_L12_Pet4.DebugString()):
                OwnershipStat := PlayerStatsTable.Ownership_L12_Pet4
                set PlayerStatsMap[Player] = player_stats_table:
                    MakePlayerStatsTable<constructor>(PlayerStatsTable)
                    Ownership_L12_Pet4 := MakeUpdatedPlayerStat(OwnershipStat, 1)

    RemoveOwnershipL12P4(Agent:agent, Stat:stat_type, ?Ownership_L12_Pet4:int = 0):void=
        if:
            Player := player[Agent]
            PlayerStatsTable := PlayerStatsMap[Player]
            if(Stat.DebugString() = StatType.Ownership_L12_Pet4.DebugString()):
                OwnershipStat := PlayerStatsTable.Ownership_L12_Pet4
                set PlayerStatsMap[Player] = player_stats_table:
                    MakePlayerStatsTable<constructor>(PlayerStatsTable)
                    Ownership_L12_Pet4 := MakeUpdatedPlayerStat(OwnershipStat, 0)

    #---------------------------------------L13
    #---------------------------------------
    RecordOwnershipL13P1(Agent:agent, Stat:stat_type, ?Ownership_L13_Pet1:int = 0):void=
        if:
            Player := player[Agent]
            PlayerStatsTable := PlayerStatsMap[Player]
            if(Stat.DebugString() = StatType.Ownership_L13_Pet1.DebugString()):
                OwnershipStat := PlayerStatsTable.Ownership_L13_Pet1
                set PlayerStatsMap[Player] = player_stats_table:
                    MakePlayerStatsTable<constructor>(PlayerStatsTable)
                    Ownership_L13_Pet1 := MakeUpdatedPlayerStat(OwnershipStat, 1)

    RemoveOwnershipL13P1(Agent:agent, Stat:stat_type, ?Ownership_L13_Pet1:int = 0):void=
        if:
            Player := player[Agent]
            PlayerStatsTable := PlayerStatsMap[Player]
            if(Stat.DebugString() = StatType.Ownership_L13_Pet1.DebugString()):
                OwnershipStat := PlayerStatsTable.Ownership_L13_Pet1
                set PlayerStatsMap[Player] = player_stats_table:
                    MakePlayerStatsTable<constructor>(PlayerStatsTable)
                    Ownership_L13_Pet1 := MakeUpdatedPlayerStat(OwnershipStat, 0)

    RecordOwnershipL13P2(Agent:agent, Stat:stat_type, ?Ownership_L13_Pet2:int = 0):void=
        if:
            Player := player[Agent]
            PlayerStatsTable := PlayerStatsMap[Player]
            if(Stat.DebugString() = StatType.Ownership_L13_Pet2.DebugString()):
                OwnershipStat := PlayerStatsTable.Ownership_L13_Pet2
                set PlayerStatsMap[Player] = player_stats_table:
                    MakePlayerStatsTable<constructor>(PlayerStatsTable)
                    Ownership_L13_Pet2 := MakeUpdatedPlayerStat(OwnershipStat, 1)

    RemoveOwnershipL13P2(Agent:agent, Stat:stat_type, ?Ownership_L13_Pet2:int = 0):void=
        if:
            Player := player[Agent]
            PlayerStatsTable := PlayerStatsMap[Player]
            if(Stat.DebugString() = StatType.Ownership_L13_Pet2.DebugString()):
                OwnershipStat := PlayerStatsTable.Ownership_L13_Pet2
                set PlayerStatsMap[Player] = player_stats_table:
                    MakePlayerStatsTable<constructor>(PlayerStatsTable)
                    Ownership_L13_Pet2 := MakeUpdatedPlayerStat(OwnershipStat, 0)

    RecordOwnershipL13P3(Agent:agent, Stat:stat_type, ?Ownership_L13_Pet3:int = 0):void=
        if:
            Player := player[Agent]
            PlayerStatsTable := PlayerStatsMap[Player]
            if(Stat.DebugString() = StatType.Ownership_L13_Pet3.DebugString()):
                OwnershipStat := PlayerStatsTable.Ownership_L13_Pet3
                set PlayerStatsMap[Player] = player_stats_table:
                    MakePlayerStatsTable<constructor>(PlayerStatsTable)
                    Ownership_L13_Pet3 := MakeUpdatedPlayerStat(OwnershipStat, 1)

    RemoveOwnershipL13P3(Agent:agent, Stat:stat_type, ?Ownership_L13_Pet3:int = 0):void=
        if:
            Player := player[Agent]
            PlayerStatsTable := PlayerStatsMap[Player]
            if(Stat.DebugString() = StatType.Ownership_L13_Pet3.DebugString()):
                OwnershipStat := PlayerStatsTable.Ownership_L13_Pet3
                set PlayerStatsMap[Player] = player_stats_table:
                    MakePlayerStatsTable<constructor>(PlayerStatsTable)
                    Ownership_L13_Pet3 := MakeUpdatedPlayerStat(OwnershipStat, 0)

    RecordOwnershipL13P4(Agent:agent, Stat:stat_type, ?Ownership_L13_Pet4:int = 0):void=
        if:
            Player := player[Agent]
            PlayerStatsTable := PlayerStatsMap[Player]
            if(Stat.DebugString() = StatType.Ownership_L13_Pet4.DebugString()):
                OwnershipStat := PlayerStatsTable.Ownership_L13_Pet4
                set PlayerStatsMap[Player] = player_stats_table:
                    MakePlayerStatsTable<constructor>(PlayerStatsTable)
                    Ownership_L13_Pet4 := MakeUpdatedPlayerStat(OwnershipStat, 1)

    RemoveOwnershipL13P4(Agent:agent, Stat:stat_type, ?Ownership_L13_Pet4:int = 0):void=
        if:
            Player := player[Agent]
            PlayerStatsTable := PlayerStatsMap[Player]
            if(Stat.DebugString() = StatType.Ownership_L13_Pet4.DebugString()):
                OwnershipStat := PlayerStatsTable.Ownership_L13_Pet4
                set PlayerStatsMap[Player] = player_stats_table:
                    MakePlayerStatsTable<constructor>(PlayerStatsTable)
                    Ownership_L13_Pet4 := MakeUpdatedPlayerStat(OwnershipStat, 0)

    #---------------------------------------L14
    #---------------------------------------
    RecordOwnershipL14P1(Agent:agent, Stat:stat_type, ?Ownership_L14_Pet1:int = 0):void=
        if:
            Player := player[Agent]
            PlayerStatsTable := PlayerStatsMap[Player]
            if(Stat.DebugString() = StatType.Ownership_L14_Pet1.DebugString()):
                OwnershipStat := PlayerStatsTable.Ownership_L14_Pet1
                set PlayerStatsMap[Player] = player_stats_table:
                    MakePlayerStatsTable<constructor>(PlayerStatsTable)
                    Ownership_L14_Pet1 := MakeUpdatedPlayerStat(OwnershipStat, 1)

    RemoveOwnershipL14P1(Agent:agent, Stat:stat_type, ?Ownership_L14_Pet1:int = 0):void=
        if:
            Player := player[Agent]
            PlayerStatsTable := PlayerStatsMap[Player]
            if(Stat.DebugString() = StatType.Ownership_L14_Pet1.DebugString()):
                OwnershipStat := PlayerStatsTable.Ownership_L14_Pet1
                set PlayerStatsMap[Player] = player_stats_table:
                    MakePlayerStatsTable<constructor>(PlayerStatsTable)
                    Ownership_L14_Pet1 := MakeUpdatedPlayerStat(OwnershipStat, 0)

    RecordOwnershipL14P2(Agent:agent, Stat:stat_type, ?Ownership_L14_Pet2:int = 0):void=
        if:
            Player := player[Agent]
            PlayerStatsTable := PlayerStatsMap[Player]
            if(Stat.DebugString() = StatType.Ownership_L14_Pet2.DebugString()):
                OwnershipStat := PlayerStatsTable.Ownership_L14_Pet2
                set PlayerStatsMap[Player] = player_stats_table:
                    MakePlayerStatsTable<constructor>(PlayerStatsTable)
                    Ownership_L14_Pet2 := MakeUpdatedPlayerStat(OwnershipStat, 1)

    RemoveOwnershipL14P2(Agent:agent, Stat:stat_type, ?Ownership_L14_Pet2:int = 0):void=
        if:
            Player := player[Agent]
            PlayerStatsTable := PlayerStatsMap[Player]
            if(Stat.DebugString() = StatType.Ownership_L14_Pet2.DebugString()):
                OwnershipStat := PlayerStatsTable.Ownership_L14_Pet2
                set PlayerStatsMap[Player] = player_stats_table:
                    MakePlayerStatsTable<constructor>(PlayerStatsTable)
                    Ownership_L14_Pet2 := MakeUpdatedPlayerStat(OwnershipStat, 0)

    RecordOwnershipL14P3(Agent:agent, Stat:stat_type, ?Ownership_L14_Pet3:int = 0):void=
        if:
            Player := player[Agent]
            PlayerStatsTable := PlayerStatsMap[Player]
            if(Stat.DebugString() = StatType.Ownership_L14_Pet3.DebugString()):
                OwnershipStat := PlayerStatsTable.Ownership_L14_Pet3
                set PlayerStatsMap[Player] = player_stats_table:
                    MakePlayerStatsTable<constructor>(PlayerStatsTable)
                    Ownership_L14_Pet3 := MakeUpdatedPlayerStat(OwnershipStat, 1)

    RemoveOwnershipL14P3(Agent:agent, Stat:stat_type, ?Ownership_L14_Pet3:int = 0):void=
        if:
            Player := player[Agent]
            PlayerStatsTable := PlayerStatsMap[Player]
            if(Stat.DebugString() = StatType.Ownership_L14_Pet3.DebugString()):
                OwnershipStat := PlayerStatsTable.Ownership_L14_Pet3
                set PlayerStatsMap[Player] = player_stats_table:
                    MakePlayerStatsTable<constructor>(PlayerStatsTable)
                    Ownership_L14_Pet3 := MakeUpdatedPlayerStat(OwnershipStat, 0)

    RecordOwnershipL14P4(Agent:agent, Stat:stat_type, ?Ownership_L14_Pet4:int = 0):void=
        if:
            Player := player[Agent]
            PlayerStatsTable := PlayerStatsMap[Player]
            if(Stat.DebugString() = StatType.Ownership_L14_Pet4.DebugString()):
                OwnershipStat := PlayerStatsTable.Ownership_L14_Pet4
                set PlayerStatsMap[Player] = player_stats_table:
                    MakePlayerStatsTable<constructor>(PlayerStatsTable)
                    Ownership_L14_Pet4 := MakeUpdatedPlayerStat(OwnershipStat, 1)

    RemoveOwnershipL14P4(Agent:agent, Stat:stat_type, ?Ownership_L14_Pet4:int = 0):void=
        if:
            Player := player[Agent]
            PlayerStatsTable := PlayerStatsMap[Player]
            if(Stat.DebugString() = StatType.Ownership_L14_Pet4.DebugString()):
                OwnershipStat := PlayerStatsTable.Ownership_L14_Pet4
                set PlayerStatsMap[Player] = player_stats_table:
                    MakePlayerStatsTable<constructor>(PlayerStatsTable)
                    Ownership_L14_Pet4 := MakeUpdatedPlayerStat(OwnershipStat, 0)

    #---------------------------------------L15
    #---------------------------------------
    RecordOwnershipL15P1(Agent:agent, Stat:stat_type, ?Ownership_L15_Pet1:int = 0):void=
        if:
            Player := player[Agent]
            PlayerStatsTable := PlayerStatsMap[Player]
            if(Stat.DebugString() = StatType.Ownership_L15_Pet1.DebugString()):
                OwnershipStat := PlayerStatsTable.Ownership_L15_Pet1
                set PlayerStatsMap[Player] = player_stats_table:
                    MakePlayerStatsTable<constructor>(PlayerStatsTable)
                    Ownership_L15_Pet1 := MakeUpdatedPlayerStat(OwnershipStat, 1)

    RemoveOwnershipL15P1(Agent:agent, Stat:stat_type, ?Ownership_L15_Pet1:int = 0):void=
        if:
            Player := player[Agent]
            PlayerStatsTable := PlayerStatsMap[Player]
            if(Stat.DebugString() = StatType.Ownership_L15_Pet1.DebugString()):
                OwnershipStat := PlayerStatsTable.Ownership_L15_Pet1
                set PlayerStatsMap[Player] = player_stats_table:
                    MakePlayerStatsTable<constructor>(PlayerStatsTable)
                    Ownership_L15_Pet1 := MakeUpdatedPlayerStat(OwnershipStat, 0)

    RecordOwnershipL15P2(Agent:agent, Stat:stat_type, ?Ownership_L15_Pet2:int = 0):void=
        if:
            Player := player[Agent]
            PlayerStatsTable := PlayerStatsMap[Player]
            if(Stat.DebugString() = StatType.Ownership_L15_Pet2.DebugString()):
                OwnershipStat := PlayerStatsTable.Ownership_L15_Pet2
                set PlayerStatsMap[Player] = player_stats_table:
                    MakePlayerStatsTable<constructor>(PlayerStatsTable)
                    Ownership_L15_Pet2 := MakeUpdatedPlayerStat(OwnershipStat, 1)

    RemoveOwnershipL15P2(Agent:agent, Stat:stat_type, ?Ownership_L15_Pet2:int = 0):void=
        if:
            Player := player[Agent]
            PlayerStatsTable := PlayerStatsMap[Player]
            if(Stat.DebugString() = StatType.Ownership_L15_Pet2.DebugString()):
                OwnershipStat := PlayerStatsTable.Ownership_L15_Pet2
                set PlayerStatsMap[Player] = player_stats_table:
                    MakePlayerStatsTable<constructor>(PlayerStatsTable)
                    Ownership_L15_Pet2 := MakeUpdatedPlayerStat(OwnershipStat, 0)

    RecordOwnershipL15P3(Agent:agent, Stat:stat_type, ?Ownership_L15_Pet3:int = 0):void=
        if:
            Player := player[Agent]
            PlayerStatsTable := PlayerStatsMap[Player]
            if(Stat.DebugString() = StatType.Ownership_L15_Pet3.DebugString()):
                OwnershipStat := PlayerStatsTable.Ownership_L15_Pet3
                set PlayerStatsMap[Player] = player_stats_table:
                    MakePlayerStatsTable<constructor>(PlayerStatsTable)
                    Ownership_L15_Pet3 := MakeUpdatedPlayerStat(OwnershipStat, 1)

    RemoveOwnershipL15P3(Agent:agent, Stat:stat_type, ?Ownership_L15_Pet3:int = 0):void=
        if:
            Player := player[Agent]
            PlayerStatsTable := PlayerStatsMap[Player]
            if(Stat.DebugString() = StatType.Ownership_L15_Pet3.DebugString()):
                OwnershipStat := PlayerStatsTable.Ownership_L15_Pet3
                set PlayerStatsMap[Player] = player_stats_table:
                    MakePlayerStatsTable<constructor>(PlayerStatsTable)
                    Ownership_L15_Pet3 := MakeUpdatedPlayerStat(OwnershipStat, 0)

    RecordOwnershipL15P4(Agent:agent, Stat:stat_type, ?Ownership_L15_Pet4:int = 0):void=
        if:
            Player := player[Agent]
            PlayerStatsTable := PlayerStatsMap[Player]
            if(Stat.DebugString() = StatType.Ownership_L15_Pet4.DebugString()):
                OwnershipStat := PlayerStatsTable.Ownership_L15_Pet4
                set PlayerStatsMap[Player] = player_stats_table:
                    MakePlayerStatsTable<constructor>(PlayerStatsTable)
                    Ownership_L15_Pet4 := MakeUpdatedPlayerStat(OwnershipStat, 1)

    RemoveOwnershipL15P4(Agent:agent, Stat:stat_type, ?Ownership_L15_Pet4:int = 0):void=
        if:
            Player := player[Agent]
            PlayerStatsTable := PlayerStatsMap[Player]
            if(Stat.DebugString() = StatType.Ownership_L15_Pet4.DebugString()):
                OwnershipStat := PlayerStatsTable.Ownership_L15_Pet4
                set PlayerStatsMap[Player] = player_stats_table:
                    MakePlayerStatsTable<constructor>(PlayerStatsTable)
                    Ownership_L15_Pet4 := MakeUpdatedPlayerStat(OwnershipStat, 0)

    #---------------------------------------L16
    #---------------------------------------
    RecordOwnershipL16P1(Agent:agent, Stat:stat_type, ?Ownership_L16_Pet1:int = 0):void=
        if:
            Player := player[Agent]
            PlayerStatsTable := PlayerStatsMap[Player]
            if(Stat.DebugString() = StatType.Ownership_L16_Pet1.DebugString()):
                OwnershipStat := PlayerStatsTable.Ownership_L16_Pet1
                set PlayerStatsMap[Player] = player_stats_table:
                    MakePlayerStatsTable<constructor>(PlayerStatsTable)
                    Ownership_L16_Pet1 := MakeUpdatedPlayerStat(OwnershipStat, 1)

    RemoveOwnershipL16P1(Agent:agent, Stat:stat_type, ?Ownership_L16_Pet1:int = 0):void=
        if:
            Player := player[Agent]
            PlayerStatsTable := PlayerStatsMap[Player]
            if(Stat.DebugString() = StatType.Ownership_L16_Pet1.DebugString()):
                OwnershipStat := PlayerStatsTable.Ownership_L16_Pet1
                set PlayerStatsMap[Player] = player_stats_table:
                    MakePlayerStatsTable<constructor>(PlayerStatsTable)
                    Ownership_L16_Pet1 := MakeUpdatedPlayerStat(OwnershipStat, 0)

    RecordOwnershipL16P2(Agent:agent, Stat:stat_type, ?Ownership_L16_Pet2:int = 0):void=
        if:
            Player := player[Agent]
            PlayerStatsTable := PlayerStatsMap[Player]
            if(Stat.DebugString() = StatType.Ownership_L16_Pet2.DebugString()):
                OwnershipStat := PlayerStatsTable.Ownership_L16_Pet2
                set PlayerStatsMap[Player] = player_stats_table:
                    MakePlayerStatsTable<constructor>(PlayerStatsTable)
                    Ownership_L16_Pet2 := MakeUpdatedPlayerStat(OwnershipStat, 1)

    RemoveOwnershipL16P2(Agent:agent, Stat:stat_type, ?Ownership_L16_Pet2:int = 0):void=
        if:
            Player := player[Agent]
            PlayerStatsTable := PlayerStatsMap[Player]
            if(Stat.DebugString() = StatType.Ownership_L16_Pet2.DebugString()):
                OwnershipStat := PlayerStatsTable.Ownership_L16_Pet2
                set PlayerStatsMap[Player] = player_stats_table:
                    MakePlayerStatsTable<constructor>(PlayerStatsTable)
                    Ownership_L16_Pet2 := MakeUpdatedPlayerStat(OwnershipStat, 0)

    RecordOwnershipL16P3(Agent:agent, Stat:stat_type, ?Ownership_L16_Pet3:int = 0):void=
        if:
            Player := player[Agent]
            PlayerStatsTable := PlayerStatsMap[Player]
            if(Stat.DebugString() = StatType.Ownership_L16_Pet3.DebugString()):
                OwnershipStat := PlayerStatsTable.Ownership_L16_Pet3
                set PlayerStatsMap[Player] = player_stats_table:
                    MakePlayerStatsTable<constructor>(PlayerStatsTable)
                    Ownership_L16_Pet3 := MakeUpdatedPlayerStat(OwnershipStat, 1)

    RemoveOwnershipL16P3(Agent:agent, Stat:stat_type, ?Ownership_L16_Pet3:int = 0):void=
        if:
            Player := player[Agent]
            PlayerStatsTable := PlayerStatsMap[Player]
            if(Stat.DebugString() = StatType.Ownership_L16_Pet3.DebugString()):
                OwnershipStat := PlayerStatsTable.Ownership_L16_Pet3
                set PlayerStatsMap[Player] = player_stats_table:
                    MakePlayerStatsTable<constructor>(PlayerStatsTable)
                    Ownership_L16_Pet3 := MakeUpdatedPlayerStat(OwnershipStat, 0)

    RecordOwnershipL16P4(Agent:agent, Stat:stat_type, ?Ownership_L16_Pet4:int = 0):void=
        if:
            Player := player[Agent]
            PlayerStatsTable := PlayerStatsMap[Player]
            if(Stat.DebugString() = StatType.Ownership_L16_Pet4.DebugString()):
                OwnershipStat := PlayerStatsTable.Ownership_L16_Pet4
                set PlayerStatsMap[Player] = player_stats_table:
                    MakePlayerStatsTable<constructor>(PlayerStatsTable)
                    Ownership_L16_Pet4 := MakeUpdatedPlayerStat(OwnershipStat, 1)

    RemoveOwnershipL16P4(Agent:agent, Stat:stat_type, ?Ownership_L16_Pet4:int = 0):void=
        if:
            Player := player[Agent]
            PlayerStatsTable := PlayerStatsMap[Player]
            if(Stat.DebugString() = StatType.Ownership_L16_Pet4.DebugString()):
                OwnershipStat := PlayerStatsTable.Ownership_L16_Pet4
                set PlayerStatsMap[Player] = player_stats_table:
                    MakePlayerStatsTable<constructor>(PlayerStatsTable)
                    Ownership_L16_Pet4 := MakeUpdatedPlayerStat(OwnershipStat, 0)

    #---------------------------------------L17
    #---------------------------------------
    RecordOwnershipL17P1(Agent:agent, Stat:stat_type, ?Ownership_L17_Pet1:int = 0):void=
        if:
            Player := player[Agent]
            PlayerStatsTable := PlayerStatsMap[Player]
            if(Stat.DebugString() = StatType.Ownership_L17_Pet1.DebugString()):
                OwnershipStat := PlayerStatsTable.Ownership_L17_Pet1
                set PlayerStatsMap[Player] = player_stats_table:
                    MakePlayerStatsTable<constructor>(PlayerStatsTable)
                    Ownership_L17_Pet1 := MakeUpdatedPlayerStat(OwnershipStat, 1)

    RemoveOwnershipL17P1(Agent:agent, Stat:stat_type, ?Ownership_L17_Pet1:int = 0):void=
        if:
            Player := player[Agent]
            PlayerStatsTable := PlayerStatsMap[Player]
            if(Stat.DebugString() = StatType.Ownership_L17_Pet1.DebugString()):
                OwnershipStat := PlayerStatsTable.Ownership_L17_Pet1
                set PlayerStatsMap[Player] = player_stats_table:
                    MakePlayerStatsTable<constructor>(PlayerStatsTable)
                    Ownership_L17_Pet1 := MakeUpdatedPlayerStat(OwnershipStat, 0)

    RecordOwnershipL17P2(Agent:agent, Stat:stat_type, ?Ownership_L17_Pet2:int = 0):void=
        if:
            Player := player[Agent]
            PlayerStatsTable := PlayerStatsMap[Player]
            if(Stat.DebugString() = StatType.Ownership_L17_Pet2.DebugString()):
                OwnershipStat := PlayerStatsTable.Ownership_L17_Pet2
                set PlayerStatsMap[Player] = player_stats_table:
                    MakePlayerStatsTable<constructor>(PlayerStatsTable)
                    Ownership_L17_Pet2 := MakeUpdatedPlayerStat(OwnershipStat, 1)

    RemoveOwnershipL17P2(Agent:agent, Stat:stat_type, ?Ownership_L17_Pet2:int = 0):void=
        if:
            Player := player[Agent]
            PlayerStatsTable := PlayerStatsMap[Player]
            if(Stat.DebugString() = StatType.Ownership_L17_Pet2.DebugString()):
                OwnershipStat := PlayerStatsTable.Ownership_L17_Pet2
                set PlayerStatsMap[Player] = player_stats_table:
                    MakePlayerStatsTable<constructor>(PlayerStatsTable)
                    Ownership_L17_Pet2 := MakeUpdatedPlayerStat(OwnershipStat, 0)

    RecordOwnershipL17P3(Agent:agent, Stat:stat_type, ?Ownership_L17_Pet3:int = 0):void=
        if:
            Player := player[Agent]
            PlayerStatsTable := PlayerStatsMap[Player]
            if(Stat.DebugString() = StatType.Ownership_L17_Pet3.DebugString()):
                OwnershipStat := PlayerStatsTable.Ownership_L17_Pet3
                set PlayerStatsMap[Player] = player_stats_table:
                    MakePlayerStatsTable<constructor>(PlayerStatsTable)
                    Ownership_L17_Pet3 := MakeUpdatedPlayerStat(OwnershipStat, 1)

    RemoveOwnershipL17P3(Agent:agent, Stat:stat_type, ?Ownership_L17_Pet3:int = 0):void=
        if:
            Player := player[Agent]
            PlayerStatsTable := PlayerStatsMap[Player]
            if(Stat.DebugString() = StatType.Ownership_L17_Pet3.DebugString()):
                OwnershipStat := PlayerStatsTable.Ownership_L17_Pet3
                set PlayerStatsMap[Player] = player_stats_table:
                    MakePlayerStatsTable<constructor>(PlayerStatsTable)
                    Ownership_L17_Pet3 := MakeUpdatedPlayerStat(OwnershipStat, 0)

    RecordOwnershipL17P4(Agent:agent, Stat:stat_type, ?Ownership_L17_Pet4:int = 0):void=
        if:
            Player := player[Agent]
            PlayerStatsTable := PlayerStatsMap[Player]
            if(Stat.DebugString() = StatType.Ownership_L17_Pet4.DebugString()):
                OwnershipStat := PlayerStatsTable.Ownership_L17_Pet4
                set PlayerStatsMap[Player] = player_stats_table:
                    MakePlayerStatsTable<constructor>(PlayerStatsTable)
                    Ownership_L17_Pet4 := MakeUpdatedPlayerStat(OwnershipStat, 1)

    RemoveOwnershipL17P4(Agent:agent, Stat:stat_type, ?Ownership_L17_Pet4:int = 0):void=
        if:
            Player := player[Agent]
            PlayerStatsTable := PlayerStatsMap[Player]
            if(Stat.DebugString() = StatType.Ownership_L17_Pet4.DebugString()):
                OwnershipStat := PlayerStatsTable.Ownership_L17_Pet4
                set PlayerStatsMap[Player] = player_stats_table:
                    MakePlayerStatsTable<constructor>(PlayerStatsTable)
                    Ownership_L17_Pet4 := MakeUpdatedPlayerStat(OwnershipStat, 0)

    #---------------------------------------L18
    #---------------------------------------
    RecordOwnershipL18P1(Agent:agent, Stat:stat_type, ?Ownership_L18_Pet1:int = 0):void=
        if:
            Player := player[Agent]
            PlayerStatsTable := PlayerStatsMap[Player]
            if(Stat.DebugString() = StatType.Ownership_L18_Pet1.DebugString()):
                OwnershipStat := PlayerStatsTable.Ownership_L18_Pet1
                set PlayerStatsMap[Player] = player_stats_table:
                    MakePlayerStatsTable<constructor>(PlayerStatsTable)
                    Ownership_L18_Pet1 := MakeUpdatedPlayerStat(OwnershipStat, 1)

    RemoveOwnershipL18P1(Agent:agent, Stat:stat_type, ?Ownership_L18_Pet1:int = 0):void=
        if:
            Player := player[Agent]
            PlayerStatsTable := PlayerStatsMap[Player]
            if(Stat.DebugString() = StatType.Ownership_L18_Pet1.DebugString()):
                OwnershipStat := PlayerStatsTable.Ownership_L18_Pet1
                set PlayerStatsMap[Player] = player_stats_table:
                    MakePlayerStatsTable<constructor>(PlayerStatsTable)
                    Ownership_L18_Pet1 := MakeUpdatedPlayerStat(OwnershipStat, 0)

    RecordOwnershipL18P2(Agent:agent, Stat:stat_type, ?Ownership_L18_Pet2:int = 0):void=
        if:
            Player := player[Agent]
            PlayerStatsTable := PlayerStatsMap[Player]
            if(Stat.DebugString() = StatType.Ownership_L18_Pet2.DebugString()):
                OwnershipStat := PlayerStatsTable.Ownership_L18_Pet2
                set PlayerStatsMap[Player] = player_stats_table:
                    MakePlayerStatsTable<constructor>(PlayerStatsTable)
                    Ownership_L18_Pet2 := MakeUpdatedPlayerStat(OwnershipStat, 1)

    RemoveOwnershipL18P2(Agent:agent, Stat:stat_type, ?Ownership_L18_Pet2:int = 0):void=
        if:
            Player := player[Agent]
            PlayerStatsTable := PlayerStatsMap[Player]
            if(Stat.DebugString() = StatType.Ownership_L18_Pet2.DebugString()):
                OwnershipStat := PlayerStatsTable.Ownership_L18_Pet2
                set PlayerStatsMap[Player] = player_stats_table:
                    MakePlayerStatsTable<constructor>(PlayerStatsTable)
                    Ownership_L18_Pet2 := MakeUpdatedPlayerStat(OwnershipStat, 0)

    RecordOwnershipL18P3(Agent:agent, Stat:stat_type, ?Ownership_L18_Pet3:int = 0):void=
        if:
            Player := player[Agent]
            PlayerStatsTable := PlayerStatsMap[Player]
            if(Stat.DebugString() = StatType.Ownership_L18_Pet3.DebugString()):
                OwnershipStat := PlayerStatsTable.Ownership_L18_Pet3
                set PlayerStatsMap[Player] = player_stats_table:
                    MakePlayerStatsTable<constructor>(PlayerStatsTable)
                    Ownership_L18_Pet3 := MakeUpdatedPlayerStat(OwnershipStat, 1)

    RemoveOwnershipL18P3(Agent:agent, Stat:stat_type, ?Ownership_L18_Pet3:int = 0):void=
        if:
            Player := player[Agent]
            PlayerStatsTable := PlayerStatsMap[Player]
            if(Stat.DebugString() = StatType.Ownership_L18_Pet3.DebugString()):
                OwnershipStat := PlayerStatsTable.Ownership_L18_Pet3
                set PlayerStatsMap[Player] = player_stats_table:
                    MakePlayerStatsTable<constructor>(PlayerStatsTable)
                    Ownership_L18_Pet3 := MakeUpdatedPlayerStat(OwnershipStat, 0)

    RecordOwnershipL18P4(Agent:agent, Stat:stat_type, ?Ownership_L18_Pet4:int = 0):void=
        if:
            Player := player[Agent]
            PlayerStatsTable := PlayerStatsMap[Player]
            if(Stat.DebugString() = StatType.Ownership_L18_Pet4.DebugString()):
                OwnershipStat := PlayerStatsTable.Ownership_L18_Pet4
                set PlayerStatsMap[Player] = player_stats_table:
                    MakePlayerStatsTable<constructor>(PlayerStatsTable)
                    Ownership_L18_Pet4 := MakeUpdatedPlayerStat(OwnershipStat, 1)

    RemoveOwnershipL18P4(Agent:agent, Stat:stat_type, ?Ownership_L18_Pet4:int = 0):void=
        if:
            Player := player[Agent]
            PlayerStatsTable := PlayerStatsMap[Player]
            if(Stat.DebugString() = StatType.Ownership_L18_Pet4.DebugString()):
                OwnershipStat := PlayerStatsTable.Ownership_L18_Pet4
                set PlayerStatsMap[Player] = player_stats_table:
                    MakePlayerStatsTable<constructor>(PlayerStatsTable)
                    Ownership_L18_Pet4 := MakeUpdatedPlayerStat(OwnershipStat, 0)

    #---------------------------------------L19
    #---------------------------------------
    RecordOwnershipL19P1(Agent:agent, Stat:stat_type, ?Ownership_L19_Pet1:int = 0):void=
        if:
            Player := player[Agent]
            PlayerStatsTable := PlayerStatsMap[Player]
            if(Stat.DebugString() = StatType.Ownership_L19_Pet1.DebugString()):
                OwnershipStat := PlayerStatsTable.Ownership_L19_Pet1
                set PlayerStatsMap[Player] = player_stats_table:
                    MakePlayerStatsTable<constructor>(PlayerStatsTable)
                    Ownership_L19_Pet1 := MakeUpdatedPlayerStat(OwnershipStat, 1)

    RemoveOwnershipL19P1(Agent:agent, Stat:stat_type, ?Ownership_L19_Pet1:int = 0):void=
        if:
            Player := player[Agent]
            PlayerStatsTable := PlayerStatsMap[Player]
            if(Stat.DebugString() = StatType.Ownership_L19_Pet1.DebugString()):
                OwnershipStat := PlayerStatsTable.Ownership_L19_Pet1
                set PlayerStatsMap[Player] = player_stats_table:
                    MakePlayerStatsTable<constructor>(PlayerStatsTable)
                    Ownership_L19_Pet1 := MakeUpdatedPlayerStat(OwnershipStat, 0)

    RecordOwnershipL19P2(Agent:agent, Stat:stat_type, ?Ownership_L19_Pet2:int = 0):void=
        if:
            Player := player[Agent]
            PlayerStatsTable := PlayerStatsMap[Player]
            if(Stat.DebugString() = StatType.Ownership_L19_Pet2.DebugString()):
                OwnershipStat := PlayerStatsTable.Ownership_L19_Pet2
                set PlayerStatsMap[Player] = player_stats_table:
                    MakePlayerStatsTable<constructor>(PlayerStatsTable)
                    Ownership_L19_Pet2 := MakeUpdatedPlayerStat(OwnershipStat, 1)

    RemoveOwnershipL19P2(Agent:agent, Stat:stat_type, ?Ownership_L19_Pet2:int = 0):void=
        if:
            Player := player[Agent]
            PlayerStatsTable := PlayerStatsMap[Player]
            if(Stat.DebugString() = StatType.Ownership_L19_Pet2.DebugString()):
                OwnershipStat := PlayerStatsTable.Ownership_L19_Pet2
                set PlayerStatsMap[Player] = player_stats_table:
                    MakePlayerStatsTable<constructor>(PlayerStatsTable)
                    Ownership_L19_Pet2 := MakeUpdatedPlayerStat(OwnershipStat, 0)

    RecordOwnershipL19P3(Agent:agent, Stat:stat_type, ?Ownership_L19_Pet3:int = 0):void=
        if:
            Player := player[Agent]
            PlayerStatsTable := PlayerStatsMap[Player]
            if(Stat.DebugString() = StatType.Ownership_L19_Pet3.DebugString()):
                OwnershipStat := PlayerStatsTable.Ownership_L19_Pet3
                set PlayerStatsMap[Player] = player_stats_table:
                    MakePlayerStatsTable<constructor>(PlayerStatsTable)
                    Ownership_L19_Pet3 := MakeUpdatedPlayerStat(OwnershipStat, 1)

    RemoveOwnershipL19P3(Agent:agent, Stat:stat_type, ?Ownership_L19_Pet3:int = 0):void=
        if:
            Player := player[Agent]
            PlayerStatsTable := PlayerStatsMap[Player]
            if(Stat.DebugString() = StatType.Ownership_L19_Pet3.DebugString()):
                OwnershipStat := PlayerStatsTable.Ownership_L19_Pet3
                set PlayerStatsMap[Player] = player_stats_table:
                    MakePlayerStatsTable<constructor>(PlayerStatsTable)
                    Ownership_L19_Pet3 := MakeUpdatedPlayerStat(OwnershipStat, 0)

    RecordOwnershipL19P4(Agent:agent, Stat:stat_type, ?Ownership_L19_Pet4:int = 0):void=
        if:
            Player := player[Agent]
            PlayerStatsTable := PlayerStatsMap[Player]
            if(Stat.DebugString() = StatType.Ownership_L19_Pet4.DebugString()):
                OwnershipStat := PlayerStatsTable.Ownership_L19_Pet4
                set PlayerStatsMap[Player] = player_stats_table:
                    MakePlayerStatsTable<constructor>(PlayerStatsTable)
                    Ownership_L19_Pet4 := MakeUpdatedPlayerStat(OwnershipStat, 1)

    RemoveOwnershipL19P4(Agent:agent, Stat:stat_type, ?Ownership_L19_Pet4:int = 0):void=
        if:
            Player := player[Agent]
            PlayerStatsTable := PlayerStatsMap[Player]
            if(Stat.DebugString() = StatType.Ownership_L19_Pet4.DebugString()):
                OwnershipStat := PlayerStatsTable.Ownership_L19_Pet4
                set PlayerStatsMap[Player] = player_stats_table:
                    MakePlayerStatsTable<constructor>(PlayerStatsTable)
                    Ownership_L19_Pet4 := MakeUpdatedPlayerStat(OwnershipStat, 0)

    #---------------------------------------L20
    #---------------------------------------
    RecordOwnershipL20P1(Agent:agent, Stat:stat_type, ?Ownership_L20_Pet1:int = 0):void=
        if:
            Player := player[Agent]
            PlayerStatsTable := PlayerStatsMap[Player]
            if(Stat.DebugString() = StatType.Ownership_L20_Pet1.DebugString()):
                OwnershipStat := PlayerStatsTable.Ownership_L20_Pet1
                set PlayerStatsMap[Player] = player_stats_table:
                    MakePlayerStatsTable<constructor>(PlayerStatsTable)
                    Ownership_L20_Pet1 := MakeUpdatedPlayerStat(OwnershipStat, 1)

    RemoveOwnershipL20P1(Agent:agent, Stat:stat_type, ?Ownership_L20_Pet1:int = 0):void=
        if:
            Player := player[Agent]
            PlayerStatsTable := PlayerStatsMap[Player]
            if(Stat.DebugString() = StatType.Ownership_L20_Pet1.DebugString()):
                OwnershipStat := PlayerStatsTable.Ownership_L20_Pet1
                set PlayerStatsMap[Player] = player_stats_table:
                    MakePlayerStatsTable<constructor>(PlayerStatsTable)
                    Ownership_L20_Pet1 := MakeUpdatedPlayerStat(OwnershipStat, 0)

    RecordOwnershipL20P2(Agent:agent, Stat:stat_type, ?Ownership_L20_Pet2:int = 0):void=
        if:
            Player := player[Agent]
            PlayerStatsTable := PlayerStatsMap[Player]
            if(Stat.DebugString() = StatType.Ownership_L20_Pet2.DebugString()):
                OwnershipStat := PlayerStatsTable.Ownership_L20_Pet2
                set PlayerStatsMap[Player] = player_stats_table:
                    MakePlayerStatsTable<constructor>(PlayerStatsTable)
                    Ownership_L20_Pet2 := MakeUpdatedPlayerStat(OwnershipStat, 1)

    RemoveOwnershipL20P2(Agent:agent, Stat:stat_type, ?Ownership_L20_Pet2:int = 0):void=
        if:
            Player := player[Agent]
            PlayerStatsTable := PlayerStatsMap[Player]
            if(Stat.DebugString() = StatType.Ownership_L20_Pet2.DebugString()):
                OwnershipStat := PlayerStatsTable.Ownership_L20_Pet2
                set PlayerStatsMap[Player] = player_stats_table:
                    MakePlayerStatsTable<constructor>(PlayerStatsTable)
                    Ownership_L20_Pet2 := MakeUpdatedPlayerStat(OwnershipStat, 0)

    RecordOwnershipL20P3(Agent:agent, Stat:stat_type, ?Ownership_L20_Pet3:int = 0):void=
        if:
            Player := player[Agent]
            PlayerStatsTable := PlayerStatsMap[Player]
            if(Stat.DebugString() = StatType.Ownership_L20_Pet3.DebugString()):
                OwnershipStat := PlayerStatsTable.Ownership_L20_Pet3
                set PlayerStatsMap[Player] = player_stats_table:
                    MakePlayerStatsTable<constructor>(PlayerStatsTable)
                    Ownership_L20_Pet3 := MakeUpdatedPlayerStat(OwnershipStat, 1)

    RemoveOwnershipL20P3(Agent:agent, Stat:stat_type, ?Ownership_L20_Pet3:int = 0):void=
        if:
            Player := player[Agent]
            PlayerStatsTable := PlayerStatsMap[Player]
            if(Stat.DebugString() = StatType.Ownership_L20_Pet3.DebugString()):
                OwnershipStat := PlayerStatsTable.Ownership_L20_Pet3
                set PlayerStatsMap[Player] = player_stats_table:
                    MakePlayerStatsTable<constructor>(PlayerStatsTable)
                    Ownership_L20_Pet3 := MakeUpdatedPlayerStat(OwnershipStat, 0)

    RecordOwnershipL20P4(Agent:agent, Stat:stat_type, ?Ownership_L20_Pet4:int = 0):void=
        if:
            Player := player[Agent]
            PlayerStatsTable := PlayerStatsMap[Player]
            if(Stat.DebugString() = StatType.Ownership_L20_Pet4.DebugString()):
                OwnershipStat := PlayerStatsTable.Ownership_L20_Pet4
                set PlayerStatsMap[Player] = player_stats_table:
                    MakePlayerStatsTable<constructor>(PlayerStatsTable)
                    Ownership_L20_Pet4 := MakeUpdatedPlayerStat(OwnershipStat, 1)

    RemoveOwnershipL20P4(Agent:agent, Stat:stat_type, ?Ownership_L20_Pet4:int = 0):void=
        if:
            Player := player[Agent]
            PlayerStatsTable := PlayerStatsMap[Player]
            if(Stat.DebugString() = StatType.Ownership_L20_Pet4.DebugString()):
                OwnershipStat := PlayerStatsTable.Ownership_L20_Pet4
                set PlayerStatsMap[Player] = player_stats_table:
                    MakePlayerStatsTable<constructor>(PlayerStatsTable)
                    Ownership_L20_Pet4 := MakeUpdatedPlayerStat(OwnershipStat, 0)

    #---------------------------------------L21
    #---------------------------------------
    RecordOwnershipL21P1(Agent:agent, Stat:stat_type, ?Ownership_L21_Pet1:int = 0):void=
        if:
            Player := player[Agent]
            PlayerStatsTable := PlayerStatsMap[Player]
            if(Stat.DebugString() = StatType.Ownership_L21_Pet1.DebugString()):
                OwnershipStat := PlayerStatsTable.Ownership_L21_Pet1
                set PlayerStatsMap[Player] = player_stats_table:
                    MakePlayerStatsTable<constructor>(PlayerStatsTable)
                    Ownership_L21_Pet1 := MakeUpdatedPlayerStat(OwnershipStat, 1)

    RemoveOwnershipL21P1(Agent:agent, Stat:stat_type, ?Ownership_L21_Pet1:int = 0):void=
        if:
            Player := player[Agent]
            PlayerStatsTable := PlayerStatsMap[Player]
            if(Stat.DebugString() = StatType.Ownership_L21_Pet1.DebugString()):
                OwnershipStat := PlayerStatsTable.Ownership_L21_Pet1
                set PlayerStatsMap[Player] = player_stats_table:
                    MakePlayerStatsTable<constructor>(PlayerStatsTable)
                    Ownership_L21_Pet1 := MakeUpdatedPlayerStat(OwnershipStat, 0)

    RecordOwnershipL21P2(Agent:agent, Stat:stat_type, ?Ownership_L21_Pet2:int = 0):void=
        if:
            Player := player[Agent]
            PlayerStatsTable := PlayerStatsMap[Player]
            if(Stat.DebugString() = StatType.Ownership_L21_Pet2.DebugString()):
                OwnershipStat := PlayerStatsTable.Ownership_L21_Pet2
                set PlayerStatsMap[Player] = player_stats_table:
                    MakePlayerStatsTable<constructor>(PlayerStatsTable)
                    Ownership_L21_Pet2 := MakeUpdatedPlayerStat(OwnershipStat, 1)

    RemoveOwnershipL21P2(Agent:agent, Stat:stat_type, ?Ownership_L21_Pet2:int = 0):void=
        if:
            Player := player[Agent]
            PlayerStatsTable := PlayerStatsMap[Player]
            if(Stat.DebugString() = StatType.Ownership_L21_Pet2.DebugString()):
                OwnershipStat := PlayerStatsTable.Ownership_L21_Pet2
                set PlayerStatsMap[Player] = player_stats_table:
                    MakePlayerStatsTable<constructor>(PlayerStatsTable)
                    Ownership_L21_Pet2 := MakeUpdatedPlayerStat(OwnershipStat, 0)

    RecordOwnershipL21P3(Agent:agent, Stat:stat_type, ?Ownership_L21_Pet3:int = 0):void=
        if:
            Player := player[Agent]
            PlayerStatsTable := PlayerStatsMap[Player]
            if(Stat.DebugString() = StatType.Ownership_L21_Pet3.DebugString()):
                OwnershipStat := PlayerStatsTable.Ownership_L21_Pet3
                set PlayerStatsMap[Player] = player_stats_table:
                    MakePlayerStatsTable<constructor>(PlayerStatsTable)
                    Ownership_L21_Pet3 := MakeUpdatedPlayerStat(OwnershipStat, 1)

    RemoveOwnershipL21P3(Agent:agent, Stat:stat_type, ?Ownership_L21_Pet3:int = 0):void=
        if:
            Player := player[Agent]
            PlayerStatsTable := PlayerStatsMap[Player]
            if(Stat.DebugString() = StatType.Ownership_L21_Pet3.DebugString()):
                OwnershipStat := PlayerStatsTable.Ownership_L21_Pet3
                set PlayerStatsMap[Player] = player_stats_table:
                    MakePlayerStatsTable<constructor>(PlayerStatsTable)
                    Ownership_L21_Pet3 := MakeUpdatedPlayerStat(OwnershipStat, 0)

    RecordOwnershipL21P4(Agent:agent, Stat:stat_type, ?Ownership_L21_Pet4:int = 0):void=
        if:
            Player := player[Agent]
            PlayerStatsTable := PlayerStatsMap[Player]
            if(Stat.DebugString() = StatType.Ownership_L21_Pet4.DebugString()):
                OwnershipStat := PlayerStatsTable.Ownership_L21_Pet4
                set PlayerStatsMap[Player] = player_stats_table:
                    MakePlayerStatsTable<constructor>(PlayerStatsTable)
                    Ownership_L21_Pet4 := MakeUpdatedPlayerStat(OwnershipStat, 1)

    RemoveOwnershipL21P4(Agent:agent, Stat:stat_type, ?Ownership_L21_Pet4:int = 0):void=
        if:
            Player := player[Agent]
            PlayerStatsTable := PlayerStatsMap[Player]
            if(Stat.DebugString() = StatType.Ownership_L21_Pet4.DebugString()):
                OwnershipStat := PlayerStatsTable.Ownership_L21_Pet4
                set PlayerStatsMap[Player] = player_stats_table:
                    MakePlayerStatsTable<constructor>(PlayerStatsTable)
                    Ownership_L21_Pet4 := MakeUpdatedPlayerStat(OwnershipStat, 0)

    #---------------------------------------L22
    #---------------------------------------
    RecordOwnershipL22P1(Agent:agent, Stat:stat_type, ?Ownership_L22_Pet1:int = 0):void=
        if:
            Player := player[Agent]
            PlayerStatsTable := PlayerStatsMap[Player]
            if(Stat.DebugString() = StatType.Ownership_L22_Pet1.DebugString()):
                OwnershipStat := PlayerStatsTable.Ownership_L22_Pet1
                set PlayerStatsMap[Player] = player_stats_table:
                    MakePlayerStatsTable<constructor>(PlayerStatsTable)
                    Ownership_L22_Pet1 := MakeUpdatedPlayerStat(OwnershipStat, 1)

    RemoveOwnershipL22P1(Agent:agent, Stat:stat_type, ?Ownership_L22_Pet1:int = 0):void=
        if:
            Player := player[Agent]
            PlayerStatsTable := PlayerStatsMap[Player]
            if(Stat.DebugString() = StatType.Ownership_L22_Pet1.DebugString()):
                OwnershipStat := PlayerStatsTable.Ownership_L22_Pet1
                set PlayerStatsMap[Player] = player_stats_table:
                    MakePlayerStatsTable<constructor>(PlayerStatsTable)
                    Ownership_L22_Pet1 := MakeUpdatedPlayerStat(OwnershipStat, 0)

    RecordOwnershipL22P2(Agent:agent, Stat:stat_type, ?Ownership_L22_Pet2:int = 0):void=
        if:
            Player := player[Agent]
            PlayerStatsTable := PlayerStatsMap[Player]
            if(Stat.DebugString() = StatType.Ownership_L22_Pet2.DebugString()):
                OwnershipStat := PlayerStatsTable.Ownership_L22_Pet2
                set PlayerStatsMap[Player] = player_stats_table:
                    MakePlayerStatsTable<constructor>(PlayerStatsTable)
                    Ownership_L22_Pet2 := MakeUpdatedPlayerStat(OwnershipStat, 1)

    RemoveOwnershipL22P2(Agent:agent, Stat:stat_type, ?Ownership_L22_Pet2:int = 0):void=
        if:
            Player := player[Agent]
            PlayerStatsTable := PlayerStatsMap[Player]
            if(Stat.DebugString() = StatType.Ownership_L22_Pet2.DebugString()):
                OwnershipStat := PlayerStatsTable.Ownership_L22_Pet2
                set PlayerStatsMap[Player] = player_stats_table:
                    MakePlayerStatsTable<constructor>(PlayerStatsTable)
                    Ownership_L22_Pet2 := MakeUpdatedPlayerStat(OwnershipStat, 0)

    RecordOwnershipL22P3(Agent:agent, Stat:stat_type, ?Ownership_L22_Pet3:int = 0):void=
        if:
            Player := player[Agent]
            PlayerStatsTable := PlayerStatsMap[Player]
            if(Stat.DebugString() = StatType.Ownership_L22_Pet3.DebugString()):
                OwnershipStat := PlayerStatsTable.Ownership_L22_Pet3
                set PlayerStatsMap[Player] = player_stats_table:
                    MakePlayerStatsTable<constructor>(PlayerStatsTable)
                    Ownership_L22_Pet3 := MakeUpdatedPlayerStat(OwnershipStat, 1)

    RemoveOwnershipL22P3(Agent:agent, Stat:stat_type, ?Ownership_L22_Pet3:int = 0):void=
        if:
            Player := player[Agent]
            PlayerStatsTable := PlayerStatsMap[Player]
            if(Stat.DebugString() = StatType.Ownership_L22_Pet3.DebugString()):
                OwnershipStat := PlayerStatsTable.Ownership_L22_Pet3
                set PlayerStatsMap[Player] = player_stats_table:
                    MakePlayerStatsTable<constructor>(PlayerStatsTable)
                    Ownership_L22_Pet3 := MakeUpdatedPlayerStat(OwnershipStat, 0)

    RecordOwnershipL22P4(Agent:agent, Stat:stat_type, ?Ownership_L22_Pet4:int = 0):void=
        if:
            Player := player[Agent]
            PlayerStatsTable := PlayerStatsMap[Player]
            if(Stat.DebugString() = StatType.Ownership_L22_Pet4.DebugString()):
                OwnershipStat := PlayerStatsTable.Ownership_L22_Pet4
                set PlayerStatsMap[Player] = player_stats_table:
                    MakePlayerStatsTable<constructor>(PlayerStatsTable)
                    Ownership_L22_Pet4 := MakeUpdatedPlayerStat(OwnershipStat, 1)

    RemoveOwnershipL22P4(Agent:agent, Stat:stat_type, ?Ownership_L22_Pet4:int = 0):void=
        if:
            Player := player[Agent]
            PlayerStatsTable := PlayerStatsMap[Player]
            if(Stat.DebugString() = StatType.Ownership_L22_Pet4.DebugString()):
                OwnershipStat := PlayerStatsTable.Ownership_L22_Pet4
                set PlayerStatsMap[Player] = player_stats_table:
                    MakePlayerStatsTable<constructor>(PlayerStatsTable)
                    Ownership_L22_Pet4 := MakeUpdatedPlayerStat(OwnershipStat, 0)

    #---------------------------------------L23
    #---------------------------------------
    RecordOwnershipL23P1(Agent:agent, Stat:stat_type, ?Ownership_L23_Pet1:int = 0):void=
        if:
            Player := player[Agent]
            PlayerStatsTable := PlayerStatsMap[Player]
            if(Stat.DebugString() = StatType.Ownership_L23_Pet1.DebugString()):
                OwnershipStat := PlayerStatsTable.Ownership_L23_Pet1
                set PlayerStatsMap[Player] = player_stats_table:
                    MakePlayerStatsTable<constructor>(PlayerStatsTable)
                    Ownership_L23_Pet1 := MakeUpdatedPlayerStat(OwnershipStat, 1)

    RemoveOwnershipL23P1(Agent:agent, Stat:stat_type, ?Ownership_L23_Pet1:int = 0):void=
        if:
            Player := player[Agent]
            PlayerStatsTable := PlayerStatsMap[Player]
            if(Stat.DebugString() = StatType.Ownership_L23_Pet1.DebugString()):
                OwnershipStat := PlayerStatsTable.Ownership_L23_Pet1
                set PlayerStatsMap[Player] = player_stats_table:
                    MakePlayerStatsTable<constructor>(PlayerStatsTable)
                    Ownership_L23_Pet1 := MakeUpdatedPlayerStat(OwnershipStat, 0)

    RecordOwnershipL23P2(Agent:agent, Stat:stat_type, ?Ownership_L23_Pet2:int = 0):void=
        if:
            Player := player[Agent]
            PlayerStatsTable := PlayerStatsMap[Player]
            if(Stat.DebugString() = StatType.Ownership_L23_Pet2.DebugString()):
                OwnershipStat := PlayerStatsTable.Ownership_L23_Pet2
                set PlayerStatsMap[Player] = player_stats_table:
                    MakePlayerStatsTable<constructor>(PlayerStatsTable)
                    Ownership_L23_Pet2 := MakeUpdatedPlayerStat(OwnershipStat, 1)

    RemoveOwnershipL23P2(Agent:agent, Stat:stat_type, ?Ownership_L23_Pet2:int = 0):void=
        if:
            Player := player[Agent]
            PlayerStatsTable := PlayerStatsMap[Player]
            if(Stat.DebugString() = StatType.Ownership_L23_Pet2.DebugString()):
                OwnershipStat := PlayerStatsTable.Ownership_L23_Pet2
                set PlayerStatsMap[Player] = player_stats_table:
                    MakePlayerStatsTable<constructor>(PlayerStatsTable)
                    Ownership_L23_Pet2 := MakeUpdatedPlayerStat(OwnershipStat, 0)

    RecordOwnershipL23P3(Agent:agent, Stat:stat_type, ?Ownership_L23_Pet3:int = 0):void=
        if:
            Player := player[Agent]
            PlayerStatsTable := PlayerStatsMap[Player]
            if(Stat.DebugString() = StatType.Ownership_L23_Pet3.DebugString()):
                OwnershipStat := PlayerStatsTable.Ownership_L23_Pet3
                set PlayerStatsMap[Player] = player_stats_table:
                    MakePlayerStatsTable<constructor>(PlayerStatsTable)
                    Ownership_L23_Pet3 := MakeUpdatedPlayerStat(OwnershipStat, 1)

    RemoveOwnershipL23P3(Agent:agent, Stat:stat_type, ?Ownership_L23_Pet3:int = 0):void=
        if:
            Player := player[Agent]
            PlayerStatsTable := PlayerStatsMap[Player]
            if(Stat.DebugString() = StatType.Ownership_L23_Pet3.DebugString()):
                OwnershipStat := PlayerStatsTable.Ownership_L23_Pet3
                set PlayerStatsMap[Player] = player_stats_table:
                    MakePlayerStatsTable<constructor>(PlayerStatsTable)
                    Ownership_L23_Pet3 := MakeUpdatedPlayerStat(OwnershipStat, 0)

    RecordOwnershipL23P4(Agent:agent, Stat:stat_type, ?Ownership_L23_Pet4:int = 0):void=
        if:
            Player := player[Agent]
            PlayerStatsTable := PlayerStatsMap[Player]
            if(Stat.DebugString() = StatType.Ownership_L23_Pet4.DebugString()):
                OwnershipStat := PlayerStatsTable.Ownership_L23_Pet4
                set PlayerStatsMap[Player] = player_stats_table:
                    MakePlayerStatsTable<constructor>(PlayerStatsTable)
                    Ownership_L23_Pet4 := MakeUpdatedPlayerStat(OwnershipStat, 1)

    RemoveOwnershipL23P4(Agent:agent, Stat:stat_type, ?Ownership_L23_Pet4:int = 0):void=
        if:
            Player := player[Agent]
            PlayerStatsTable := PlayerStatsMap[Player]
            if(Stat.DebugString() = StatType.Ownership_L23_Pet4.DebugString()):
                OwnershipStat := PlayerStatsTable.Ownership_L23_Pet4
                set PlayerStatsMap[Player] = player_stats_table:
                    MakePlayerStatsTable<constructor>(PlayerStatsTable)
                    Ownership_L23_Pet4 := MakeUpdatedPlayerStat(OwnershipStat, 0)

    #---------------------------------------L24
    #---------------------------------------
    RecordOwnershipL24P1(Agent:agent, Stat:stat_type, ?Ownership_L24_Pet1:int = 0):void=
        if:
            Player := player[Agent]
            PlayerStatsTable := PlayerStatsMap[Player]
            if(Stat.DebugString() = StatType.Ownership_L24_Pet1.DebugString()):
                OwnershipStat := PlayerStatsTable.Ownership_L24_Pet1
                set PlayerStatsMap[Player] = player_stats_table:
                    MakePlayerStatsTable<constructor>(PlayerStatsTable)
                    Ownership_L24_Pet1 := MakeUpdatedPlayerStat(OwnershipStat, 1)

    RemoveOwnershipL24P1(Agent:agent, Stat:stat_type, ?Ownership_L24_Pet1:int = 0):void=
        if:
            Player := player[Agent]
            PlayerStatsTable := PlayerStatsMap[Player]
            if(Stat.DebugString() = StatType.Ownership_L24_Pet1.DebugString()):
                OwnershipStat := PlayerStatsTable.Ownership_L24_Pet1
                set PlayerStatsMap[Player] = player_stats_table:
                    MakePlayerStatsTable<constructor>(PlayerStatsTable)
                    Ownership_L24_Pet1 := MakeUpdatedPlayerStat(OwnershipStat, 0)

    RecordOwnershipL24P2(Agent:agent, Stat:stat_type, ?Ownership_L24_Pet2:int = 0):void=
        if:
            Player := player[Agent]
            PlayerStatsTable := PlayerStatsMap[Player]
            if(Stat.DebugString() = StatType.Ownership_L24_Pet2.DebugString()):
                OwnershipStat := PlayerStatsTable.Ownership_L24_Pet2
                set PlayerStatsMap[Player] = player_stats_table:
                    MakePlayerStatsTable<constructor>(PlayerStatsTable)
                    Ownership_L24_Pet2 := MakeUpdatedPlayerStat(OwnershipStat, 1)

    RemoveOwnershipL24P2(Agent:agent, Stat:stat_type, ?Ownership_L24_Pet2:int = 0):void=
        if:
            Player := player[Agent]
            PlayerStatsTable := PlayerStatsMap[Player]
            if(Stat.DebugString() = StatType.Ownership_L24_Pet2.DebugString()):
                OwnershipStat := PlayerStatsTable.Ownership_L24_Pet2
                set PlayerStatsMap[Player] = player_stats_table:
                    MakePlayerStatsTable<constructor>(PlayerStatsTable)
                    Ownership_L24_Pet2 := MakeUpdatedPlayerStat(OwnershipStat, 0)

    RecordOwnershipL24P3(Agent:agent, Stat:stat_type, ?Ownership_L24_Pet3:int = 0):void=
        if:
            Player := player[Agent]
            PlayerStatsTable := PlayerStatsMap[Player]
            if(Stat.DebugString() = StatType.Ownership_L24_Pet3.DebugString()):
                OwnershipStat := PlayerStatsTable.Ownership_L24_Pet3
                set PlayerStatsMap[Player] = player_stats_table:
                    MakePlayerStatsTable<constructor>(PlayerStatsTable)
                    Ownership_L24_Pet3 := MakeUpdatedPlayerStat(OwnershipStat, 1)

    RemoveOwnershipL24P3(Agent:agent, Stat:stat_type, ?Ownership_L24_Pet3:int = 0):void=
        if:
            Player := player[Agent]
            PlayerStatsTable := PlayerStatsMap[Player]
            if(Stat.DebugString() = StatType.Ownership_L24_Pet3.DebugString()):
                OwnershipStat := PlayerStatsTable.Ownership_L24_Pet3
                set PlayerStatsMap[Player] = player_stats_table:
                    MakePlayerStatsTable<constructor>(PlayerStatsTable)
                    Ownership_L24_Pet3 := MakeUpdatedPlayerStat(OwnershipStat, 0)

    RecordOwnershipL24P4(Agent:agent, Stat:stat_type, ?Ownership_L24_Pet4:int = 0):void=
        if:
            Player := player[Agent]
            PlayerStatsTable := PlayerStatsMap[Player]
            if(Stat.DebugString() = StatType.Ownership_L24_Pet4.DebugString()):
                OwnershipStat := PlayerStatsTable.Ownership_L24_Pet4
                set PlayerStatsMap[Player] = player_stats_table:
                    MakePlayerStatsTable<constructor>(PlayerStatsTable)
                    Ownership_L24_Pet4 := MakeUpdatedPlayerStat(OwnershipStat, 1)

    RemoveOwnershipL24P4(Agent:agent, Stat:stat_type, ?Ownership_L24_Pet4:int = 0):void=
        if:
            Player := player[Agent]
            PlayerStatsTable := PlayerStatsMap[Player]
            if(Stat.DebugString() = StatType.Ownership_L24_Pet4.DebugString()):
                OwnershipStat := PlayerStatsTable.Ownership_L24_Pet4
                set PlayerStatsMap[Player] = player_stats_table:
                    MakePlayerStatsTable<constructor>(PlayerStatsTable)
                    Ownership_L24_Pet4 := MakeUpdatedPlayerStat(OwnershipStat, 0)

    #---------------------------------------L25
    #---------------------------------------
    RecordOwnershipL25P1(Agent:agent, Stat:stat_type, ?Ownership_L25_Pet1:int = 0):void=
        if:
            Player := player[Agent]
            PlayerStatsTable := PlayerStatsMap[Player]
            if(Stat.DebugString() = StatType.Ownership_L25_Pet1.DebugString()):
                OwnershipStat := PlayerStatsTable.Ownership_L25_Pet1
                set PlayerStatsMap[Player] = player_stats_table:
                    MakePlayerStatsTable<constructor>(PlayerStatsTable)
                    Ownership_L25_Pet1 := MakeUpdatedPlayerStat(OwnershipStat, 1)

    RemoveOwnershipL25P1(Agent:agent, Stat:stat_type, ?Ownership_L25_Pet1:int = 0):void=
        if:
            Player := player[Agent]
            PlayerStatsTable := PlayerStatsMap[Player]
            if(Stat.DebugString() = StatType.Ownership_L25_Pet1.DebugString()):
                OwnershipStat := PlayerStatsTable.Ownership_L25_Pet1
                set PlayerStatsMap[Player] = player_stats_table:
                    MakePlayerStatsTable<constructor>(PlayerStatsTable)
                    Ownership_L25_Pet1 := MakeUpdatedPlayerStat(OwnershipStat, 0)

    RecordOwnershipL25P2(Agent:agent, Stat:stat_type, ?Ownership_L25_Pet2:int = 0):void=
        if:
            Player := player[Agent]
            PlayerStatsTable := PlayerStatsMap[Player]
            if(Stat.DebugString() = StatType.Ownership_L25_Pet2.DebugString()):
                OwnershipStat := PlayerStatsTable.Ownership_L25_Pet2
                set PlayerStatsMap[Player] = player_stats_table:
                    MakePlayerStatsTable<constructor>(PlayerStatsTable)
                    Ownership_L25_Pet2 := MakeUpdatedPlayerStat(OwnershipStat, 1)

    RemoveOwnershipL25P2(Agent:agent, Stat:stat_type, ?Ownership_L25_Pet2:int = 0):void=
        if:
            Player := player[Agent]
            PlayerStatsTable := PlayerStatsMap[Player]
            if(Stat.DebugString() = StatType.Ownership_L25_Pet2.DebugString()):
                OwnershipStat := PlayerStatsTable.Ownership_L25_Pet2
                set PlayerStatsMap[Player] = player_stats_table:
                    MakePlayerStatsTable<constructor>(PlayerStatsTable)
                    Ownership_L25_Pet2 := MakeUpdatedPlayerStat(OwnershipStat, 0)

    RecordOwnershipL25P3(Agent:agent, Stat:stat_type, ?Ownership_L25_Pet3:int = 0):void=
        if:
            Player := player[Agent]
            PlayerStatsTable := PlayerStatsMap[Player]
            if(Stat.DebugString() = StatType.Ownership_L25_Pet3.DebugString()):
                OwnershipStat := PlayerStatsTable.Ownership_L25_Pet3
                set PlayerStatsMap[Player] = player_stats_table:
                    MakePlayerStatsTable<constructor>(PlayerStatsTable)
                    Ownership_L25_Pet3 := MakeUpdatedPlayerStat(OwnershipStat, 1)

    RemoveOwnershipL25P3(Agent:agent, Stat:stat_type, ?Ownership_L25_Pet3:int = 0):void=
        if:
            Player := player[Agent]
            PlayerStatsTable := PlayerStatsMap[Player]
            if(Stat.DebugString() = StatType.Ownership_L25_Pet3.DebugString()):
                OwnershipStat := PlayerStatsTable.Ownership_L25_Pet3
                set PlayerStatsMap[Player] = player_stats_table:
                    MakePlayerStatsTable<constructor>(PlayerStatsTable)
                    Ownership_L25_Pet3 := MakeUpdatedPlayerStat(OwnershipStat, 0)

    RecordOwnershipL25P4(Agent:agent, Stat:stat_type, ?Ownership_L25_Pet4:int = 0):void=
        if:
            Player := player[Agent]
            PlayerStatsTable := PlayerStatsMap[Player]
            if(Stat.DebugString() = StatType.Ownership_L25_Pet4.DebugString()):
                OwnershipStat := PlayerStatsTable.Ownership_L25_Pet4
                set PlayerStatsMap[Player] = player_stats_table:
                    MakePlayerStatsTable<constructor>(PlayerStatsTable)
                    Ownership_L25_Pet4 := MakeUpdatedPlayerStat(OwnershipStat, 1)

    RemoveOwnershipL25P4(Agent:agent, Stat:stat_type, ?Ownership_L25_Pet4:int = 0):void=
        if:
            Player := player[Agent]
            PlayerStatsTable := PlayerStatsMap[Player]
            if(Stat.DebugString() = StatType.Ownership_L25_Pet4.DebugString()):
                OwnershipStat := PlayerStatsTable.Ownership_L25_Pet4
                set PlayerStatsMap[Player] = player_stats_table:
                    MakePlayerStatsTable<constructor>(PlayerStatsTable)
                    Ownership_L25_Pet4 := MakeUpdatedPlayerStat(OwnershipStat, 0)

    #---------------------------------------L26
    #---------------------------------------
    RecordOwnershipL26P1(Agent:agent, Stat:stat_type, ?Ownership_L26_Pet1:int = 0):void=
        if:
            Player := player[Agent]
            PlayerStatsTable := PlayerStatsMap[Player]
            if(Stat.DebugString() = StatType.Ownership_L26_Pet1.DebugString()):
                OwnershipStat := PlayerStatsTable.Ownership_L26_Pet1
                set PlayerStatsMap[Player] = player_stats_table:
                    MakePlayerStatsTable<constructor>(PlayerStatsTable)
                    Ownership_L26_Pet1 := MakeUpdatedPlayerStat(OwnershipStat, 1)

    RemoveOwnershipL26P1(Agent:agent, Stat:stat_type, ?Ownership_L26_Pet1:int = 0):void=
        if:
            Player := player[Agent]
            PlayerStatsTable := PlayerStatsMap[Player]
            if(Stat.DebugString() = StatType.Ownership_L26_Pet1.DebugString()):
                OwnershipStat := PlayerStatsTable.Ownership_L26_Pet1
                set PlayerStatsMap[Player] = player_stats_table:
                    MakePlayerStatsTable<constructor>(PlayerStatsTable)
                    Ownership_L26_Pet1 := MakeUpdatedPlayerStat(OwnershipStat, 0)

    RecordOwnershipL26P2(Agent:agent, Stat:stat_type, ?Ownership_L26_Pet2:int = 0):void=
        if:
            Player := player[Agent]
            PlayerStatsTable := PlayerStatsMap[Player]
            if(Stat.DebugString() = StatType.Ownership_L26_Pet2.DebugString()):
                OwnershipStat := PlayerStatsTable.Ownership_L26_Pet2
                set PlayerStatsMap[Player] = player_stats_table:
                    MakePlayerStatsTable<constructor>(PlayerStatsTable)
                    Ownership_L26_Pet2 := MakeUpdatedPlayerStat(OwnershipStat, 1)

    RemoveOwnershipL26P2(Agent:agent, Stat:stat_type, ?Ownership_L26_Pet2:int = 0):void=
        if:
            Player := player[Agent]
            PlayerStatsTable := PlayerStatsMap[Player]
            if(Stat.DebugString() = StatType.Ownership_L26_Pet2.DebugString()):
                OwnershipStat := PlayerStatsTable.Ownership_L26_Pet2
                set PlayerStatsMap[Player] = player_stats_table:
                    MakePlayerStatsTable<constructor>(PlayerStatsTable)
                    Ownership_L26_Pet2 := MakeUpdatedPlayerStat(OwnershipStat, 0)

    RecordOwnershipL26P3(Agent:agent, Stat:stat_type, ?Ownership_L26_Pet3:int = 0):void=
        if:
            Player := player[Agent]
            PlayerStatsTable := PlayerStatsMap[Player]
            if(Stat.DebugString() = StatType.Ownership_L26_Pet3.DebugString()):
                OwnershipStat := PlayerStatsTable.Ownership_L26_Pet3
                set PlayerStatsMap[Player] = player_stats_table:
                    MakePlayerStatsTable<constructor>(PlayerStatsTable)
                    Ownership_L26_Pet3 := MakeUpdatedPlayerStat(OwnershipStat, 1)

    RemoveOwnershipL26P3(Agent:agent, Stat:stat_type, ?Ownership_L26_Pet3:int = 0):void=
        if:
            Player := player[Agent]
            PlayerStatsTable := PlayerStatsMap[Player]
            if(Stat.DebugString() = StatType.Ownership_L26_Pet3.DebugString()):
                OwnershipStat := PlayerStatsTable.Ownership_L26_Pet3
                set PlayerStatsMap[Player] = player_stats_table:
                    MakePlayerStatsTable<constructor>(PlayerStatsTable)
                    Ownership_L26_Pet3 := MakeUpdatedPlayerStat(OwnershipStat, 0)

    RecordOwnershipL26P4(Agent:agent, Stat:stat_type, ?Ownership_L26_Pet4:int = 0):void=
        if:
            Player := player[Agent]
            PlayerStatsTable := PlayerStatsMap[Player]
            if(Stat.DebugString() = StatType.Ownership_L26_Pet4.DebugString()):
                OwnershipStat := PlayerStatsTable.Ownership_L26_Pet4
                set PlayerStatsMap[Player] = player_stats_table:
                    MakePlayerStatsTable<constructor>(PlayerStatsTable)
                    Ownership_L26_Pet4 := MakeUpdatedPlayerStat(OwnershipStat, 1)

    RemoveOwnershipL26P4(Agent:agent, Stat:stat_type, ?Ownership_L26_Pet4:int = 0):void=
        if:
            Player := player[Agent]
            PlayerStatsTable := PlayerStatsMap[Player]
            if(Stat.DebugString() = StatType.Ownership_L26_Pet4.DebugString()):
                OwnershipStat := PlayerStatsTable.Ownership_L26_Pet4
                set PlayerStatsMap[Player] = player_stats_table:
                    MakePlayerStatsTable<constructor>(PlayerStatsTable)
                    Ownership_L26_Pet4 := MakeUpdatedPlayerStat(OwnershipStat, 0)

    #---------------------------------------L27
    #---------------------------------------
    RecordOwnershipL27P1(Agent:agent, Stat:stat_type, ?Ownership_L27_Pet1:int = 0):void=
        if:
            Player := player[Agent]
            PlayerStatsTable := PlayerStatsMap[Player]
            if(Stat.DebugString() = StatType.Ownership_L27_Pet1.DebugString()):
                OwnershipStat := PlayerStatsTable.Ownership_L27_Pet1
                set PlayerStatsMap[Player] = player_stats_table:
                    MakePlayerStatsTable<constructor>(PlayerStatsTable)
                    Ownership_L27_Pet1 := MakeUpdatedPlayerStat(OwnershipStat, 1)

    RemoveOwnershipL27P1(Agent:agent, Stat:stat_type, ?Ownership_L27_Pet1:int = 0):void=
        if:
            Player := player[Agent]
            PlayerStatsTable := PlayerStatsMap[Player]
            if(Stat.DebugString() = StatType.Ownership_L27_Pet1.DebugString()):
                OwnershipStat := PlayerStatsTable.Ownership_L27_Pet1
                set PlayerStatsMap[Player] = player_stats_table:
                    MakePlayerStatsTable<constructor>(PlayerStatsTable)
                    Ownership_L27_Pet1 := MakeUpdatedPlayerStat(OwnershipStat, 0)

    RecordOwnershipL27P2(Agent:agent, Stat:stat_type, ?Ownership_L27_Pet2:int = 0):void=
        if:
            Player := player[Agent]
            PlayerStatsTable := PlayerStatsMap[Player]
            if(Stat.DebugString() = StatType.Ownership_L27_Pet2.DebugString()):
                OwnershipStat := PlayerStatsTable.Ownership_L27_Pet2
                set PlayerStatsMap[Player] = player_stats_table:
                    MakePlayerStatsTable<constructor>(PlayerStatsTable)
                    Ownership_L27_Pet2 := MakeUpdatedPlayerStat(OwnershipStat, 1)

    RemoveOwnershipL27P2(Agent:agent, Stat:stat_type, ?Ownership_L27_Pet2:int = 0):void=
        if:
            Player := player[Agent]
            PlayerStatsTable := PlayerStatsMap[Player]
            if(Stat.DebugString() = StatType.Ownership_L27_Pet2.DebugString()):
                OwnershipStat := PlayerStatsTable.Ownership_L27_Pet2
                set PlayerStatsMap[Player] = player_stats_table:
                    MakePlayerStatsTable<constructor>(PlayerStatsTable)
                    Ownership_L27_Pet2 := MakeUpdatedPlayerStat(OwnershipStat, 0)

    RecordOwnershipL27P3(Agent:agent, Stat:stat_type, ?Ownership_L27_Pet3:int = 0):void=
        if:
            Player := player[Agent]
            PlayerStatsTable := PlayerStatsMap[Player]
            if(Stat.DebugString() = StatType.Ownership_L27_Pet3.DebugString()):
                OwnershipStat := PlayerStatsTable.Ownership_L27_Pet3
                set PlayerStatsMap[Player] = player_stats_table:
                    MakePlayerStatsTable<constructor>(PlayerStatsTable)
                    Ownership_L27_Pet3 := MakeUpdatedPlayerStat(OwnershipStat, 1)

    RemoveOwnershipL27P3(Agent:agent, Stat:stat_type, ?Ownership_L27_Pet3:int = 0):void=
        if:
            Player := player[Agent]
            PlayerStatsTable := PlayerStatsMap[Player]
            if(Stat.DebugString() = StatType.Ownership_L27_Pet3.DebugString()):
                OwnershipStat := PlayerStatsTable.Ownership_L27_Pet3
                set PlayerStatsMap[Player] = player_stats_table:
                    MakePlayerStatsTable<constructor>(PlayerStatsTable)
                    Ownership_L27_Pet3 := MakeUpdatedPlayerStat(OwnershipStat, 0)

    RecordOwnershipL27P4(Agent:agent, Stat:stat_type, ?Ownership_L27_Pet4:int = 0):void=
        if:
            Player := player[Agent]
            PlayerStatsTable := PlayerStatsMap[Player]
            if(Stat.DebugString() = StatType.Ownership_L27_Pet4.DebugString()):
                OwnershipStat := PlayerStatsTable.Ownership_L27_Pet4
                set PlayerStatsMap[Player] = player_stats_table:
                    MakePlayerStatsTable<constructor>(PlayerStatsTable)
                    Ownership_L27_Pet4 := MakeUpdatedPlayerStat(OwnershipStat, 1)

    RemoveOwnershipL27P4(Agent:agent, Stat:stat_type, ?Ownership_L27_Pet4:int = 0):void=
        if:
            Player := player[Agent]
            PlayerStatsTable := PlayerStatsMap[Player]
            if(Stat.DebugString() = StatType.Ownership_L27_Pet4.DebugString()):
                OwnershipStat := PlayerStatsTable.Ownership_L27_Pet4
                set PlayerStatsMap[Player] = player_stats_table:
                    MakePlayerStatsTable<constructor>(PlayerStatsTable)
                    Ownership_L27_Pet4 := MakeUpdatedPlayerStat(OwnershipStat, 0)

    #---------------------------------------L28
    #---------------------------------------
    RecordOwnershipL28P1(Agent:agent, Stat:stat_type, ?Ownership_L28_Pet1:int = 0):void=
        if:
            Player := player[Agent]
            PlayerStatsTable := PlayerStatsMap[Player]
            if(Stat.DebugString() = StatType.Ownership_L28_Pet1.DebugString()):
                OwnershipStat := PlayerStatsTable.Ownership_L28_Pet1
                set PlayerStatsMap[Player] = player_stats_table:
                    MakePlayerStatsTable<constructor>(PlayerStatsTable)
                    Ownership_L28_Pet1 := MakeUpdatedPlayerStat(OwnershipStat, 1)

    RemoveOwnershipL28P1(Agent:agent, Stat:stat_type, ?Ownership_L28_Pet1:int = 0):void=
        if:
            Player := player[Agent]
            PlayerStatsTable := PlayerStatsMap[Player]
            if(Stat.DebugString() = StatType.Ownership_L28_Pet1.DebugString()):
                OwnershipStat := PlayerStatsTable.Ownership_L28_Pet1
                set PlayerStatsMap[Player] = player_stats_table:
                    MakePlayerStatsTable<constructor>(PlayerStatsTable)
                    Ownership_L28_Pet1 := MakeUpdatedPlayerStat(OwnershipStat, 0)

    RecordOwnershipL28P2(Agent:agent, Stat:stat_type, ?Ownership_L28_Pet2:int = 0):void=
        if:
            Player := player[Agent]
            PlayerStatsTable := PlayerStatsMap[Player]
            if(Stat.DebugString() = StatType.Ownership_L28_Pet2.DebugString()):
                OwnershipStat := PlayerStatsTable.Ownership_L28_Pet2
                set PlayerStatsMap[Player] = player_stats_table:
                    MakePlayerStatsTable<constructor>(PlayerStatsTable)
                    Ownership_L28_Pet2 := MakeUpdatedPlayerStat(OwnershipStat, 1)

    RemoveOwnershipL28P2(Agent:agent, Stat:stat_type, ?Ownership_L28_Pet2:int = 0):void=
        if:
            Player := player[Agent]
            PlayerStatsTable := PlayerStatsMap[Player]
            if(Stat.DebugString() = StatType.Ownership_L28_Pet2.DebugString()):
                OwnershipStat := PlayerStatsTable.Ownership_L28_Pet2
                set PlayerStatsMap[Player] = player_stats_table:
                    MakePlayerStatsTable<constructor>(PlayerStatsTable)
                    Ownership_L28_Pet2 := MakeUpdatedPlayerStat(OwnershipStat, 0)

    RecordOwnershipL28P3(Agent:agent, Stat:stat_type, ?Ownership_L28_Pet3:int = 0):void=
        if:
            Player := player[Agent]
            PlayerStatsTable := PlayerStatsMap[Player]
            if(Stat.DebugString() = StatType.Ownership_L28_Pet3.DebugString()):
                OwnershipStat := PlayerStatsTable.Ownership_L28_Pet3
                set PlayerStatsMap[Player] = player_stats_table:
                    MakePlayerStatsTable<constructor>(PlayerStatsTable)
                    Ownership_L28_Pet3 := MakeUpdatedPlayerStat(OwnershipStat, 1)

    RemoveOwnershipL28P3(Agent:agent, Stat:stat_type, ?Ownership_L28_Pet3:int = 0):void=
        if:
            Player := player[Agent]
            PlayerStatsTable := PlayerStatsMap[Player]
            if(Stat.DebugString() = StatType.Ownership_L28_Pet3.DebugString()):
                OwnershipStat := PlayerStatsTable.Ownership_L28_Pet3
                set PlayerStatsMap[Player] = player_stats_table:
                    MakePlayerStatsTable<constructor>(PlayerStatsTable)
                    Ownership_L28_Pet3 := MakeUpdatedPlayerStat(OwnershipStat, 0)

    RecordOwnershipL28P4(Agent:agent, Stat:stat_type, ?Ownership_L28_Pet4:int = 0):void=
        if:
            Player := player[Agent]
            PlayerStatsTable := PlayerStatsMap[Player]
            if(Stat.DebugString() = StatType.Ownership_L28_Pet4.DebugString()):
                OwnershipStat := PlayerStatsTable.Ownership_L28_Pet4
                set PlayerStatsMap[Player] = player_stats_table:
                    MakePlayerStatsTable<constructor>(PlayerStatsTable)
                    Ownership_L28_Pet4 := MakeUpdatedPlayerStat(OwnershipStat, 1)

    RemoveOwnershipL28P4(Agent:agent, Stat:stat_type, ?Ownership_L28_Pet4:int = 0):void=
        if:
            Player := player[Agent]
            PlayerStatsTable := PlayerStatsMap[Player]
            if(Stat.DebugString() = StatType.Ownership_L28_Pet4.DebugString()):
                OwnershipStat := PlayerStatsTable.Ownership_L28_Pet4
                set PlayerStatsMap[Player] = player_stats_table:
                    MakePlayerStatsTable<constructor>(PlayerStatsTable)
                    Ownership_L28_Pet4 := MakeUpdatedPlayerStat(OwnershipStat, 0)

    #---------------------------------------L29
    #---------------------------------------
    RecordOwnershipL29P1(Agent:agent, Stat:stat_type, ?Ownership_L29_Pet1:int = 0):void=
        if:
            Player := player[Agent]
            PlayerStatsTable := PlayerStatsMap[Player]
            if(Stat.DebugString() = StatType.Ownership_L29_Pet1.DebugString()):
                OwnershipStat := PlayerStatsTable.Ownership_L29_Pet1
                set PlayerStatsMap[Player] = player_stats_table:
                    MakePlayerStatsTable<constructor>(PlayerStatsTable)
                    Ownership_L29_Pet1 := MakeUpdatedPlayerStat(OwnershipStat, 1)

    RemoveOwnershipL29P1(Agent:agent, Stat:stat_type, ?Ownership_L29_Pet1:int = 0):void=
        if:
            Player := player[Agent]
            PlayerStatsTable := PlayerStatsMap[Player]
            if(Stat.DebugString() = StatType.Ownership_L29_Pet1.DebugString()):
                OwnershipStat := PlayerStatsTable.Ownership_L29_Pet1
                set PlayerStatsMap[Player] = player_stats_table:
                    MakePlayerStatsTable<constructor>(PlayerStatsTable)
                    Ownership_L29_Pet1 := MakeUpdatedPlayerStat(OwnershipStat, 0)

    RecordOwnershipL29P2(Agent:agent, Stat:stat_type, ?Ownership_L29_Pet2:int = 0):void=
        if:
            Player := player[Agent]
            PlayerStatsTable := PlayerStatsMap[Player]
            if(Stat.DebugString() = StatType.Ownership_L29_Pet2.DebugString()):
                OwnershipStat := PlayerStatsTable.Ownership_L29_Pet2
                set PlayerStatsMap[Player] = player_stats_table:
                    MakePlayerStatsTable<constructor>(PlayerStatsTable)
                    Ownership_L29_Pet2 := MakeUpdatedPlayerStat(OwnershipStat, 1)

    RemoveOwnershipL29P2(Agent:agent, Stat:stat_type, ?Ownership_L29_Pet2:int = 0):void=
        if:
            Player := player[Agent]
            PlayerStatsTable := PlayerStatsMap[Player]
            if(Stat.DebugString() = StatType.Ownership_L29_Pet2.DebugString()):
                OwnershipStat := PlayerStatsTable.Ownership_L29_Pet2
                set PlayerStatsMap[Player] = player_stats_table:
                    MakePlayerStatsTable<constructor>(PlayerStatsTable)
                    Ownership_L29_Pet2 := MakeUpdatedPlayerStat(OwnershipStat, 0)

    RecordOwnershipL29P3(Agent:agent, Stat:stat_type, ?Ownership_L29_Pet3:int = 0):void=
        if:
            Player := player[Agent]
            PlayerStatsTable := PlayerStatsMap[Player]
            if(Stat.DebugString() = StatType.Ownership_L29_Pet3.DebugString()):
                OwnershipStat := PlayerStatsTable.Ownership_L29_Pet3
                set PlayerStatsMap[Player] = player_stats_table:
                    MakePlayerStatsTable<constructor>(PlayerStatsTable)
                    Ownership_L29_Pet3 := MakeUpdatedPlayerStat(OwnershipStat, 1)

    RemoveOwnershipL29P3(Agent:agent, Stat:stat_type, ?Ownership_L29_Pet3:int = 0):void=
        if:
            Player := player[Agent]
            PlayerStatsTable := PlayerStatsMap[Player]
            if(Stat.DebugString() = StatType.Ownership_L29_Pet3.DebugString()):
                OwnershipStat := PlayerStatsTable.Ownership_L29_Pet3
                set PlayerStatsMap[Player] = player_stats_table:
                    MakePlayerStatsTable<constructor>(PlayerStatsTable)
                    Ownership_L29_Pet3 := MakeUpdatedPlayerStat(OwnershipStat, 0)

    RecordOwnershipL29P4(Agent:agent, Stat:stat_type, ?Ownership_L29_Pet4:int = 0):void=
        if:
            Player := player[Agent]
            PlayerStatsTable := PlayerStatsMap[Player]
            if(Stat.DebugString() = StatType.Ownership_L29_Pet4.DebugString()):
                OwnershipStat := PlayerStatsTable.Ownership_L29_Pet4
                set PlayerStatsMap[Player] = player_stats_table:
                    MakePlayerStatsTable<constructor>(PlayerStatsTable)
                    Ownership_L29_Pet4 := MakeUpdatedPlayerStat(OwnershipStat, 1)

    RemoveOwnershipL29P4(Agent:agent, Stat:stat_type, ?Ownership_L29_Pet4:int = 0):void=
        if:
            Player := player[Agent]
            PlayerStatsTable := PlayerStatsMap[Player]
            if(Stat.DebugString() = StatType.Ownership_L29_Pet4.DebugString()):
                OwnershipStat := PlayerStatsTable.Ownership_L29_Pet4
                set PlayerStatsMap[Player] = player_stats_table:
                    MakePlayerStatsTable<constructor>(PlayerStatsTable)
                    Ownership_L29_Pet4 := MakeUpdatedPlayerStat(OwnershipStat, 0)

    #---------------------------------------L30
    #---------------------------------------
    RecordOwnershipL30P1(Agent:agent, Stat:stat_type, ?Ownership_L30_Pet1:int = 0):void=
        if:
            Player := player[Agent]
            PlayerStatsTable := PlayerStatsMap[Player]
            if(Stat.DebugString() = StatType.Ownership_L30_Pet1.DebugString()):
                OwnershipStat := PlayerStatsTable.Ownership_L30_Pet1
                set PlayerStatsMap[Player] = player_stats_table:
                    MakePlayerStatsTable<constructor>(PlayerStatsTable)
                    Ownership_L30_Pet1 := MakeUpdatedPlayerStat(OwnershipStat, 1)

    RemoveOwnershipL30P1(Agent:agent, Stat:stat_type, ?Ownership_L30_Pet1:int = 0):void=
        if:
            Player := player[Agent]
            PlayerStatsTable := PlayerStatsMap[Player]
            if(Stat.DebugString() = StatType.Ownership_L30_Pet1.DebugString()):
                OwnershipStat := PlayerStatsTable.Ownership_L30_Pet1
                set PlayerStatsMap[Player] = player_stats_table:
                    MakePlayerStatsTable<constructor>(PlayerStatsTable)
                    Ownership_L30_Pet1 := MakeUpdatedPlayerStat(OwnershipStat, 0)

    RecordOwnershipL30P2(Agent:agent, Stat:stat_type, ?Ownership_L30_Pet2:int = 0):void=
        if:
            Player := player[Agent]
            PlayerStatsTable := PlayerStatsMap[Player]
            if(Stat.DebugString() = StatType.Ownership_L30_Pet2.DebugString()):
                OwnershipStat := PlayerStatsTable.Ownership_L30_Pet2
                set PlayerStatsMap[Player] = player_stats_table:
                    MakePlayerStatsTable<constructor>(PlayerStatsTable)
                    Ownership_L30_Pet2 := MakeUpdatedPlayerStat(OwnershipStat, 1)

    RemoveOwnershipL30P2(Agent:agent, Stat:stat_type, ?Ownership_L30_Pet2:int = 0):void=
        if:
            Player := player[Agent]
            PlayerStatsTable := PlayerStatsMap[Player]
            if(Stat.DebugString() = StatType.Ownership_L30_Pet2.DebugString()):
                OwnershipStat := PlayerStatsTable.Ownership_L30_Pet2
                set PlayerStatsMap[Player] = player_stats_table:
                    MakePlayerStatsTable<constructor>(PlayerStatsTable)
                    Ownership_L30_Pet2 := MakeUpdatedPlayerStat(OwnershipStat, 0)

    RecordOwnershipL30P3(Agent:agent, Stat:stat_type, ?Ownership_L30_Pet3:int = 0):void=
        if:
            Player := player[Agent]
            PlayerStatsTable := PlayerStatsMap[Player]
            if(Stat.DebugString() = StatType.Ownership_L30_Pet3.DebugString()):
                OwnershipStat := PlayerStatsTable.Ownership_L30_Pet3
                set PlayerStatsMap[Player] = player_stats_table:
                    MakePlayerStatsTable<constructor>(PlayerStatsTable)
                    Ownership_L30_Pet3 := MakeUpdatedPlayerStat(OwnershipStat, 1)

    RemoveOwnershipL30P3(Agent:agent, Stat:stat_type, ?Ownership_L30_Pet3:int = 0):void=
        if:
            Player := player[Agent]
            PlayerStatsTable := PlayerStatsMap[Player]
            if(Stat.DebugString() = StatType.Ownership_L30_Pet3.DebugString()):
                OwnershipStat := PlayerStatsTable.Ownership_L30_Pet3
                set PlayerStatsMap[Player] = player_stats_table:
                    MakePlayerStatsTable<constructor>(PlayerStatsTable)
                    Ownership_L30_Pet3 := MakeUpdatedPlayerStat(OwnershipStat, 0)

    RecordOwnershipL30P4(Agent:agent, Stat:stat_type, ?Ownership_L30_Pet4:int = 0):void=
        if:
            Player := player[Agent]
            PlayerStatsTable := PlayerStatsMap[Player]
            if(Stat.DebugString() = StatType.Ownership_L30_Pet4.DebugString()):
                OwnershipStat := PlayerStatsTable.Ownership_L30_Pet4
                set PlayerStatsMap[Player] = player_stats_table:
                    MakePlayerStatsTable<constructor>(PlayerStatsTable)
                    Ownership_L30_Pet4 := MakeUpdatedPlayerStat(OwnershipStat, 1)

    RemoveOwnershipL30P4(Agent:agent, Stat:stat_type, ?Ownership_L30_Pet4:int = 0):void=
        if:
            Player := player[Agent]
            PlayerStatsTable := PlayerStatsMap[Player]
            if(Stat.DebugString() = StatType.Ownership_L30_Pet4.DebugString()):
                OwnershipStat := PlayerStatsTable.Ownership_L30_Pet4
                set PlayerStatsMap[Player] = player_stats_table:
                    MakePlayerStatsTable<constructor>(PlayerStatsTable)
                    Ownership_L30_Pet4 := MakeUpdatedPlayerStat(OwnershipStat, 0)

    #---------------------------------------L31
    #---------------------------------------
    RecordOwnershipL31P1(Agent:agent, Stat:stat_type, ?Ownership_L31_Pet1:int = 0):void=
        if:
            Player := player[Agent]
            PlayerStatsTable := PlayerStatsMap[Player]
            if(Stat.DebugString() = StatType.Ownership_L31_Pet1.DebugString()):
                OwnershipStat := PlayerStatsTable.Ownership_L31_Pet1
                set PlayerStatsMap[Player] = player_stats_table:
                    MakePlayerStatsTable<constructor>(PlayerStatsTable)
                    Ownership_L31_Pet1 := MakeUpdatedPlayerStat(OwnershipStat, 1)

    #---------------------------------------L32
    #---------------------------------------
    RecordOwnershipL32P1(Agent:agent, Stat:stat_type, ?Ownership_L32_Pet1:int = 0):void=
        if:
            Player := player[Agent]
            PlayerStatsTable := PlayerStatsMap[Player]
            if(Stat.DebugString() = StatType.Ownership_L32_Pet1.DebugString()):
                OwnershipStat := PlayerStatsTable.Ownership_L32_Pet1
                set PlayerStatsMap[Player] = player_stats_table:
                    MakePlayerStatsTable<constructor>(PlayerStatsTable)
                    Ownership_L32_Pet1 := MakeUpdatedPlayerStat(OwnershipStat, 1)

    #---------------------------------------L33
    #---------------------------------------
    RecordOwnershipL33P1(Agent:agent, Stat:stat_type, ?Ownership_L33_Pet1:int = 0):void=
        if:
            Player := player[Agent]
            PlayerStatsTable := PlayerStatsMap[Player]
            if(Stat.DebugString() = StatType.Ownership_L33_Pet1.DebugString()):
                OwnershipStat := PlayerStatsTable.Ownership_L33_Pet1
                set PlayerStatsMap[Player] = player_stats_table:
                    MakePlayerStatsTable<constructor>(PlayerStatsTable)
                    Ownership_L33_Pet1 := MakeUpdatedPlayerStat(OwnershipStat, 1)

    #---------------------------------------L34
    #---------------------------------------
    RecordOwnershipL34P1(Agent:agent, Stat:stat_type, ?Ownership_L34_Pet1:int = 0):void=
        if:
            Player := player[Agent]
            PlayerStatsTable := PlayerStatsMap[Player]
            if(Stat.DebugString() = StatType.Ownership_L34_Pet1.DebugString()):
                OwnershipStat := PlayerStatsTable.Ownership_L34_Pet1
                set PlayerStatsMap[Player] = player_stats_table:
                    MakePlayerStatsTable<constructor>(PlayerStatsTable)
                    Ownership_L34_Pet1 := MakeUpdatedPlayerStat(OwnershipStat, 1)
    # Potions:

    #Luck
    RecordOwnershipPotionLuck1(Agent:agent, Stat:stat_type, ?Ownership_Potion_Luck1:int = 0):void=
        if:
            Player := player[Agent]
            PlayerStatsTable := PlayerStatsMap[Player]
            if(Stat.DebugString() = StatType.Ownership_Potion_Luck1.DebugString()):
                OwnershipStat := PlayerStatsTable.Ownership_Potion_Luck1
                set PlayerStatsMap[Player] = player_stats_table:
                    MakePlayerStatsTable<constructor>(PlayerStatsTable)
                    Ownership_Potion_Luck1 := MakeUpdatedPlayerStat(OwnershipStat, Ownership_Potion_Luck1)

    RecordOwnershipPotionLuck2(Agent:agent, Stat:stat_type, ?Ownership_Potion_Luck2:int = 0):void=
        if:
            Player := player[Agent]
            PlayerStatsTable := PlayerStatsMap[Player]
            if(Stat.DebugString() = StatType.Ownership_Potion_Luck2.DebugString()):
                OwnershipStat := PlayerStatsTable.Ownership_Potion_Luck2
                set PlayerStatsMap[Player] = player_stats_table:
                    MakePlayerStatsTable<constructor>(PlayerStatsTable)
                    Ownership_Potion_Luck2 := MakeUpdatedPlayerStat(OwnershipStat, Ownership_Potion_Luck2)

    RecordOwnershipPotionLuck3(Agent:agent, Stat:stat_type, ?Ownership_Potion_Luck3:int = 0):void=
        if:
            Player := player[Agent]
            PlayerStatsTable := PlayerStatsMap[Player]
            if(Stat.DebugString() = StatType.Ownership_Potion_Luck3.DebugString()):
                OwnershipStat := PlayerStatsTable.Ownership_Potion_Luck3
                set PlayerStatsMap[Player] = player_stats_table:
                    MakePlayerStatsTable<constructor>(PlayerStatsTable)
                    Ownership_Potion_Luck3 := MakeUpdatedPlayerStat(OwnershipStat, Ownership_Potion_Luck3)

    # Gold
    RecordOwnershipPotionGold1(Agent:agent, Stat:stat_type, ?Ownership_Potion_Gold1:int = 0):void=
        if:
            Player := player[Agent]
            PlayerStatsTable := PlayerStatsMap[Player]
            if(Stat.DebugString() = StatType.Ownership_Potion_Gold1.DebugString()):
                OwnershipStat := PlayerStatsTable.Ownership_Potion_Gold1
                set PlayerStatsMap[Player] = player_stats_table:
                    MakePlayerStatsTable<constructor>(PlayerStatsTable)
                    Ownership_Potion_Gold1 := MakeUpdatedPlayerStat(OwnershipStat, Ownership_Potion_Gold1)

    RecordOwnershipPotionGold2(Agent:agent, Stat:stat_type, ?Ownership_Potion_Gold2:int = 0):void=
        if:
            Player := player[Agent]
            PlayerStatsTable := PlayerStatsMap[Player]
            if(Stat.DebugString() = StatType.Ownership_Potion_Gold2.DebugString()):
                OwnershipStat := PlayerStatsTable.Ownership_Potion_Gold2
                set PlayerStatsMap[Player] = player_stats_table:
                    MakePlayerStatsTable<constructor>(PlayerStatsTable)
                    Ownership_Potion_Gold2 := MakeUpdatedPlayerStat(OwnershipStat, Ownership_Potion_Gold2)

    RecordOwnershipPotionGold3(Agent:agent, Stat:stat_type, ?Ownership_Potion_Gold3:int = 0):void=
        if:
            Player := player[Agent]
            PlayerStatsTable := PlayerStatsMap[Player]
            if(Stat.DebugString() = StatType.Ownership_Potion_Gold3.DebugString()):
                OwnershipStat := PlayerStatsTable.Ownership_Potion_Gold3
                set PlayerStatsMap[Player] = player_stats_table:
                    MakePlayerStatsTable<constructor>(PlayerStatsTable)
                    Ownership_Potion_Gold3 := MakeUpdatedPlayerStat(OwnershipStat, Ownership_Potion_Gold3)

    # Pet Power
    RecordOwnershipPotionPetPower1(Agent:agent, Stat:stat_type, ?Ownership_Potion_PetPower1:int = 0):void=
        if:
            Player := player[Agent]
            PlayerStatsTable := PlayerStatsMap[Player]
            if(Stat.DebugString() = StatType.Ownership_Potion_PetPower1.DebugString()):
                OwnershipStat := PlayerStatsTable.Ownership_Potion_PetPower1
                set PlayerStatsMap[Player] = player_stats_table:
                    MakePlayerStatsTable<constructor>(PlayerStatsTable)
                    Ownership_Potion_PetPower1 := MakeUpdatedPlayerStat(OwnershipStat, Ownership_Potion_PetPower1)

    RecordOwnershipPotionPetPower2(Agent:agent, Stat:stat_type, ?Ownership_Potion_PetPower2:int = 0):void=
        if:
            Player := player[Agent]
            PlayerStatsTable := PlayerStatsMap[Player]
            if(Stat.DebugString() = StatType.Ownership_Potion_PetPower2.DebugString()):
                OwnershipStat := PlayerStatsTable.Ownership_Potion_PetPower2
                set PlayerStatsMap[Player] = player_stats_table:
                    MakePlayerStatsTable<constructor>(PlayerStatsTable)
                    Ownership_Potion_PetPower2 := MakeUpdatedPlayerStat(OwnershipStat, Ownership_Potion_PetPower2)

    RecordOwnershipPotionPetPower3(Agent:agent, Stat:stat_type, ?Ownership_Potion_PetPower3:int = 0):void=
        if:
            Player := player[Agent]
            PlayerStatsTable := PlayerStatsMap[Player]
            if(Stat.DebugString() = StatType.Ownership_Potion_PetPower3.DebugString()):
                OwnershipStat := PlayerStatsTable.Ownership_Potion_PetPower3
                set PlayerStatsMap[Player] = player_stats_table:
                    MakePlayerStatsTable<constructor>(PlayerStatsTable)
                    Ownership_Potion_PetPower3 := MakeUpdatedPlayerStat(OwnershipStat, Ownership_Potion_PetPower3)

    # Player Power
    RecordOwnershipPotionPlayerPower1(Agent:agent, Stat:stat_type, ?Ownership_Potion_PlayerPower1:int = 0):void=
        if:
            Player := player[Agent]
            PlayerStatsTable := PlayerStatsMap[Player]
            if(Stat.DebugString() = StatType.Ownership_Potion_PlayerPower1.DebugString()):
                OwnershipStat := PlayerStatsTable.Ownership_Potion_PlayerPower1
                set PlayerStatsMap[Player] = player_stats_table:
                    MakePlayerStatsTable<constructor>(PlayerStatsTable)
                    Ownership_Potion_PlayerPower1 := MakeUpdatedPlayerStat(OwnershipStat, Ownership_Potion_PlayerPower1)

    RecordOwnershipPotionPlayerPower2(Agent:agent, Stat:stat_type, ?Ownership_Potion_PlayerPower2:int = 0):void=
        if:
            Player := player[Agent]
            PlayerStatsTable := PlayerStatsMap[Player]
            if(Stat.DebugString() = StatType.Ownership_Potion_PlayerPower2.DebugString()):
                OwnershipStat := PlayerStatsTable.Ownership_Potion_PlayerPower2
                set PlayerStatsMap[Player] = player_stats_table:
                    MakePlayerStatsTable<constructor>(PlayerStatsTable)
                    Ownership_Potion_PlayerPower2 := MakeUpdatedPlayerStat(OwnershipStat, Ownership_Potion_PlayerPower2)

    RecordOwnershipPotionPlayerPower3(Agent:agent, Stat:stat_type, ?Ownership_Potion_PlayerPower3:int = 0):void=
        if:
            Player := player[Agent]
            PlayerStatsTable := PlayerStatsMap[Player]
            if(Stat.DebugString() = StatType.Ownership_Potion_PlayerPower3.DebugString()):
                OwnershipStat := PlayerStatsTable.Ownership_Potion_PlayerPower3
                set PlayerStatsMap[Player] = player_stats_table:
                    MakePlayerStatsTable<constructor>(PlayerStatsTable)
                    Ownership_Potion_PlayerPower3 := MakeUpdatedPlayerStat(OwnershipStat, Ownership_Potion_PlayerPower3)

    #Tuto
    RecordWheelshowed(Agent:agent, Stat:stat_type, ?Wheelshowed:int = 0):void=
        if:
            Player := player[Agent]
            PlayerStatsTable := PlayerStatsMap[Player]
            if(Stat.DebugString() = StatType.Wheelshowed.DebugString()):
                WheelshowedStat := PlayerStatsTable.Wheelshowed
                set PlayerStatsMap[Player] = player_stats_table:
                    MakePlayerStatsTable<constructor>(PlayerStatsTable)
                    Wheelshowed := MakeUpdatedPlayerStat(WheelshowedStat, Wheelshowed)

####################################### GAME #######################################
####################################### GAME #######################################
####################################### GAME #######################################

Chest_SuperClass := class<concrete>():
    var TotalHP<public> : float = 300.0
    var CurrentHP<public> : float = 300.0
    var Respawn_time<private> : float = 5.0
    var CanGrantGold<public> : int = 1
    var CanVFX<public> : int = 1

    @editable ChestProp : creative_prop = creative_prop{}
    @editable Chest_VFX_Destroyed : vfx_spawner_device = vfx_spawner_device{}
    @editable Chest_VFX_Destroyed2 : vfx_spawner_device = vfx_spawner_device{}
    @editable ProgressB:?ProgressBar = false

    TakeDamage<public>(DamagePoints: float)<suspends>:void=
        set CurrentHP -= DamagePoints
        #spawn { WobbleAnimation() }
        if(ValidProgressBar := ProgressB?):
            var HealthPercent : float = CurrentHP / TotalHP
            Print("animate bar {HealthPercent}%")
             ValidProgressBar.AnimateToPercent(HealthPercent, 0.3)
        
        Print("Chest HP Left: {CurrentHP}")
        if (CurrentHP <= 0.0):
            Print("Chest Destroyed")
            #SpawnParticleEffectFromType(CompletionEffect ,ChestProp.GetTransform(), 0.1, 5.0)
            if(CanVFX = 1):
                Chest_VFX_Destroyed.Restart()
                Chest_VFX_Destroyed2.Restart()
            ChestProp.Hide()
            if(ValidProgressBar := ProgressB?):
                ValidProgressBar.HideBar()

            Sleep(Respawn_time)
            ResetProp()

    IsAlive<public>()<decides>:void=
        if (CurrentHP > 0.0) { true }
        else { false }

    ResetProp<public>():void=
        set CurrentHP = TotalHP

        Print("Chest Repop")

        if(ValidProgressBar := ProgressB?):
            ValidProgressBar.ShowBar()
            spawn { ValidProgressBar.Reset() }
        ChestProp.Show()

        set CanGrantGold = 1
        set CanVFX = 1

    WobbleAnimation()<suspends>:void=
        DefaultTransform := ChestProp.GetTransform()

        SmallTransform := transform:
            Translation := ChestProp.GetTransform().Translation
            Rotation := ChestProp.GetTransform().Rotation
            Scale := ChestProp.GetTransform().Scale * 0.8

        ChestProp.MoveTo(SmallTransform, 0.2)
        
        ChestProp.MoveTo(DefaultTransform, 0.2) # Reset to original position after wobble

ProgressBar<public> := class<concrete>:
    @editable
    FinalScale : vector3 = vector3{X:= 0.2, Y:= 1.0, Z:= 1.0}
    @editable
    FillBarProp : creative_prop = creative_prop{}

    @editable
    ProgressBarRoot<public> : creative_prop = creative_prop{}

    CancelAnimationEvent<public> : event() = event(){}

    ResetBar<public>(Title : string):void=
        spawn { AnimateToPercent(0.0, 0.01) }

    Reset<public>()<suspends>:void=
        AnimateToPercent(1.0, 0.02)

    HideBar<public>():void=
        ProgressBarRoot.Hide()
        FillBarProp.Hide()

    ShowBar<public>():void=
        ProgressBarRoot.Show()
        FillBarProp.Show()

    AnimateToPercent<public>(percent : float, Time : float)<suspends>:void=
        CurrentTransform := FillBarProp.GetTransform()
        var TargetFill : float = percent
        if (percent >= 1.0):
            set TargetFill = 1.0
        if (percent <= 0.1):
            set TargetFill = 0.01
        TargetScale := vector3{X:= CurrentTransform.Scale.X, Y:= FinalScale.Y * TargetFill, Z:= FinalScale.Z}
            
        FinalTransform := transform:
            Translation := CurrentTransform.Translation
            Rotation := CurrentTransform.Rotation
            Scale := TargetScale
    
        FillBarProp.MoveTo(FinalTransform, Time)
        
    
    AnimateTaskPartial<public>(TaskTime : float, StartPercent : float, TargetPercent : float)<suspends>:void= 
        AnimateToPercent(TargetPercent, TaskTime)
        Sleep(0.1)
    
    AnimateFromTo<public>(StartPercent : float, EndPercent : float, Time : float)<suspends>:void=
        AnimateToPercent(StartPercent, 0.0)
        Sleep(0.1)
        AnimateTaskPartial(Time, StartPercent, EndPercent)

Chest_L1 := class<concrete>(Chest_SuperClass):
    var TotalHP<override>:float  = 100.0
    var CurrentHP<override>:float  = 100.0

Chest_L2 := class<concrete>(Chest_SuperClass):
    var TotalHP<override>:float  = 200.0
    var CurrentHP<override>:float  = 200.0

Chest_L3 := class<concrete>(Chest_SuperClass):
    var TotalHP<override>:float  = 450.0
    var CurrentHP<override>:float  = 450.0

Chest_L4 := class<concrete>(Chest_SuperClass):
    var TotalHP<override>:float  = 750.0
    var CurrentHP<override>:float  = 750.0

Chest_L5 := class<concrete>(Chest_SuperClass):
    var TotalHP<override>:float  = 1200.0
    var CurrentHP<override>:float  = 1200.0

Chest_L6 := class<concrete>(Chest_SuperClass):
    var TotalHP<override>:float  = 2000.0
    var CurrentHP<override>:float  = 2000.0

Chest_L7 := class<concrete>(Chest_SuperClass):
    var TotalHP<override>:float  = 3300.0
    var CurrentHP<override>:float  = 3300.0

Chest_L8 := class<concrete>(Chest_SuperClass):
    var TotalHP<override>:float  = 5500.0
    var CurrentHP<override>:float  = 5500.0

Chest_L9 := class<concrete>(Chest_SuperClass):
    var TotalHP<override>:float  = 9000.0
    var CurrentHP<override>:float  = 9000.0

Chest_L10 := class<concrete>(Chest_SuperClass):
    var TotalHP<override>:float  = 14800.0
    var CurrentHP<override>:float  = 14800.0

Chest_L11 := class<concrete>(Chest_SuperClass):
    var TotalHP<override>:float  = 24500.0*2
    var CurrentHP<override>:float  = 24500.0*2

Chest_L12 := class<concrete>(Chest_SuperClass):
    var TotalHP<override>:float  = 40300.0*2
    var CurrentHP<override>:float  = 40300.0*2

Chest_L13 := class<concrete>(Chest_SuperClass):
    var TotalHP<override>:float  = 66500.0*2
    var CurrentHP<override>:float  = 66500.0*2

Chest_L14 := class<concrete>(Chest_SuperClass):
    var TotalHP<override>:float  = 110000.0*2
    var CurrentHP<override>:float  = 110000.0*2

Chest_L15 := class<concrete>(Chest_SuperClass):
    var TotalHP<override>:float  = 180800.0*2
    var CurrentHP<override>:float  = 180800.0*2

Chest_L16 := class<concrete>(Chest_SuperClass):
    var TotalHP<override>:float  = 298000.0*2
    var CurrentHP<override>:float  = 298000.0*2

Chest_L17 := class<concrete>(Chest_SuperClass):
    var TotalHP<override>:float  = 491500.0*2
    var CurrentHP<override>:float  = 491500.0*2

Chest_L18 := class<concrete>(Chest_SuperClass):
    var TotalHP<override>:float  = 810300.0*2
    var CurrentHP<override>:float  = 810300.0*2

Chest_L19 := class<concrete>(Chest_SuperClass):
    var TotalHP<override>:float  = 1336000.0*2
    var CurrentHP<override>:float  = 1336000.0*2

Chest_L20 := class<concrete>(Chest_SuperClass):
    var TotalHP<override>:float  = 2200700.0*2
    var CurrentHP<override>:float  = 2200700.0*2

Chest_L21 := class<concrete>(Chest_SuperClass):
    var TotalHP<override>:float  = 3631000.0*3
    var CurrentHP<override>:float  = 3631000.0*3

Chest_L22 := class<concrete>(Chest_SuperClass):
    var TotalHP<override>:float  = 5987000.0*3
    var CurrentHP<override>:float  = 5987000.0*3

Chest_L23 := class<concrete>(Chest_SuperClass):
    var TotalHP<override>:float  = 9871000.0*3
    var CurrentHP<override>:float  = 9871000.0*3

Chest_L24 := class<concrete>(Chest_SuperClass):
    var TotalHP<override>:float  = 16275000.0*3
    var CurrentHP<override>:float  = 16275000.0*3

Chest_L25 := class<concrete>(Chest_SuperClass):
    var TotalHP<override>:float  = 26832000.0*3
    var CurrentHP<override>:float  = 26832000.0*3

Chest_L26 := class<concrete>(Chest_SuperClass):
    var TotalHP<override>:float  = 44241000.0*3
    var CurrentHP<override>:float  = 44241000.0*3

Chest_L27 := class<concrete>(Chest_SuperClass):
    var TotalHP<override>:float  = 72940000.0*3
    var CurrentHP<override>:float  = 72940000.0*3

Chest_L28 := class<concrete>(Chest_SuperClass):
    var TotalHP<override>:float  = 120260000.0*3
    var CurrentHP<override>:float  = 120260000.0*3

Chest_L29 := class<concrete>(Chest_SuperClass):
    var TotalHP<override>:float  = 198275000.0*3
    var CurrentHP<override>:float  = 198275000.0*3

Chest_L30 := class<concrete>(Chest_SuperClass):
    var TotalHP<override>:float  = 326901000.0*3
    var CurrentHP<override>:float  = 326901000.0*3

PetClass := class<concrete>():
    @editable PetProp : creative_prop = creative_prop{}
    @editable AttackAnim : cinematic_sequence_device = cinematic_sequence_device{}

    Attack():void=
        AttackAnim.Play()

(Listenable : listenable(agent)).SubscribeAgent(OutputFunc : tuple(agent, t)->void, ExtraData : t where t:type) : cancelable =
    Wrapper := wrapper_agent(t){ExtraData := ExtraData, OutputFunc := OutputFunc}
    Listenable.Subscribe(Wrapper.InputFunc)
    
wrapper_agent(t : type) := class():
    ExtraData : t;
    OutputFunc : tuple(agent, t) -> void
    InputFunc(Agent : agent):void = OutputFunc(Agent, ExtraData)
    
(Listenable : listenable(tuple())).SubscribeEmpty(OutputFunc : t -> void, ExtraData : t where t:type) : cancelable =
    Wrapper := wrapper_empty(t) {ExtraData := ExtraData, OutputFunc := OutputFunc}
    Listenable.Subscribe(Wrapper.InputFunc)
    
wrapper_empty(t : type) := class():
    ExtraData : t;
    OutputFunc : t -> void
    InputFunc():void = OutputFunc(ExtraData)

#--------------------------------------------------------Inventory stuff-------------------------------------------------------#
#--------------------------------------------------------Inventory stuff-------------------------------------------------------#
#--------------------------------------------------------Inventory stuff-------------------------------------------------------#
custom_game_action_causer := class(game_action_causer){}

purchase_type := enum:
    Luck1,
    Luck2,
    Luck3,
    Gold1,
    Gold2,
    Gold3,
    Pet_Power1,
    Pet_Power2,
    Pet_Power3,
    Player_Power1,
    Player_Power2,
    Player_Power3,
    Empty

purchase_data := class<concrete>:
    @editable
    PurchaseType : purchase_type = purchase_type.Empty

    #@editable:
    #    ToolTip := MsgPowerupValue
    PowerupValue : float = 0.0

    @editable
    PurchasePrice : type { X:int where X >= 0 } = 0

    @editable
    bOneTimePurchase : logic = true

    #@editable:
    #    ToolTip := MsgPriceScaler
    PriceScalerAfterEachPurchase : type { X:float where X >= 0.0 } = 0.0

agent_data := class:
    var bShopOpen : logic = false
    var MaybeShopCanvas : ?canvas = false
    var CurrentPageIndex : int = 0

    var Purchases : []int = array{}

#--------------------------------------------------------GAME MANAGER DEVICE-------------------------------------------------------#
#--------------------------------------------------------GAME MANAGER DEVICE-------------------------------------------------------#
#--------------------------------------------------------GAME MANAGER DEVICE-------------------------------------------------------#
game_manager := class(creative_device):

    #Player
    var PlayersMap:[player]Custom_Player = map{}

    @editable P1Spawner:player_spawner_device = player_spawner_device{}
    @editable P2Spawner:player_spawner_device = player_spawner_device{}
    @editable P3Spawner:player_spawner_device = player_spawner_device{}

    #Load game
    @editable var LOADSaveMutator : mutator_zone_device = mutator_zone_device{}

    #Rebirth

    @editable var Rebirth1 : button_device = button_device{}
    @editable var Rebirth2 : button_device = button_device{}
    @editable var Rebirth3 : button_device = button_device{}
    @editable var Rebirth4 : button_device = button_device{}
    @editable var Rebirth5 : button_device = button_device{}
    @editable var Rebirth5Code : button_device = button_device{}
    @editable var ButtonTest : button_device = button_device{}

    @editable var TeleporterRebirth : teleporter_device = teleporter_device {}

    #Lobby

    @editable var PurchaseButtonTUTO : button_device = button_device{}
    @editable var PurchaseButtonL1 : button_device = button_device{}
    @editable var PurchaseButtonL2 : button_device = button_device{}
    @editable var PurchaseButtonL3 : button_device = button_device{}
    @editable var PurchaseButtonL4 : button_device = button_device{}
    @editable var PurchaseButtonL5 : button_device = button_device{}
    @editable var PurchaseButtonL6 : button_device = button_device{}
    @editable var PurchaseButtonL7 : button_device = button_device{}
    @editable var PurchaseButtonL8 : button_device = button_device{}
    @editable var PurchaseButtonL9 : button_device = button_device{}
    @editable var PurchaseButtonL10 : button_device = button_device{}
    @editable var PurchaseButtonL11 : button_device = button_device{}
    @editable var PurchaseButtonL12 : button_device = button_device{}
    @editable var PurchaseButtonL13 : button_device = button_device{}
    @editable var PurchaseButtonL14 : button_device = button_device{}
    @editable var PurchaseButtonL15 : button_device = button_device{}
    @editable var PurchaseButtonL16 : button_device = button_device{}
    @editable var PurchaseButtonL17 : button_device = button_device{}
    @editable var PurchaseButtonL18 : button_device = button_device{}
    @editable var PurchaseButtonL19 : button_device = button_device{}
    @editable var PurchaseButtonL20 : button_device = button_device{}
    @editable var PurchaseButtonL21 : button_device = button_device{}
    @editable var PurchaseButtonL22 : button_device = button_device{}
    @editable var PurchaseButtonL23 : button_device = button_device{}
    @editable var PurchaseButtonL24 : button_device = button_device{}
    @editable var PurchaseButtonL25 : button_device = button_device{}
    @editable var PurchaseButtonL26 : button_device = button_device{}
    @editable var PurchaseButtonL27 : button_device = button_device{}
    @editable var PurchaseButtonL28 : button_device = button_device{}
    @editable var PurchaseButtonL29 : button_device = button_device{}
    @editable var PurchaseButtonL30 : button_device = button_device{}

    #Secret Pet
    @editable var PurchaseButtonL31 : button_device = button_device{}
    @editable var PurchaseButtonL32 : button_device = button_device{}
    @editable var PurchaseButtonL33 : button_device = button_device{}
    @editable var PurchaseButtonL34 : button_device = button_device{} #for testing only

    @editable var IncreaseGoldButton : button_device = button_device{}
    @editable var DecreaseGoldButton : button_device = button_device{}
    @editable var IncreaseGoldBonus : button_device = button_device{}
    @editable var IncreaseDiamondButton : button_device = button_device{}
    @editable var DecreaseDiamondButton : button_device = button_device{}

    @editable var Obby1DiamondNormalP1 : button_device = button_device{}
    @editable var Obby2DiamondNormalP1 : button_device = button_device{}
    @editable var Obby1DiamondNormalP2 : button_device = button_device{}
    @editable var Obby2DiamondNormalP2 : button_device = button_device{}
    @editable var Obby1DiamondNormalP3 : button_device = button_device{}
    @editable var Obby2DiamondNormalP3 : button_device = button_device{}
    @editable var Obby1DiamondOnTimeP1 : button_device = button_device{}
    @editable var Obby2DiamondOnTimeP1 : button_device = button_device{}
    @editable var Obby1DiamondOnTimeP2 : button_device = button_device{}
    @editable var Obby2DiamondOnTimeP2 : button_device = button_device{}
    @editable var Obby1DiamondOnTimeP3 : button_device = button_device{}
    @editable var Obby2DiamondOnTimeP3 : button_device = button_device{}
    
    #@editable var PurchaseButtonTestTeam : button_device = button_device{}
    #@editable var GetTeamButton : button_device = button_device{}
    @editable var CheckFunctionButton : button_device = button_device{}
    #@editable var ReleasePetButton1 : button_device = button_device{}

    #@editable 
    #PetBuyButtons : []button_device = array{}

    #PetDex Button Test
    @editable var OpenPetDexButton : button_device = button_device{}

    #Purchase Village

    @editable var UpgradeCrit1 : button_device = button_device{}
    @editable var UpgradeCrit2 : button_device = button_device{}
    @editable var UpgradeCrit3 : button_device = button_device{}
    @editable var UpgradeCrit4 : button_device = button_device{}
    @editable var UpgradeCrit5 : button_device = button_device{}

    @editable var UpgradeSpeed1 : button_device = button_device{}
    @editable var UpgradeSpeed2 : button_device = button_device{}
    @editable var UpgradeSpeed3 : button_device = button_device{}
    @editable var UpgradeSpeed4 : button_device = button_device{}
    @editable var UpgradeSpeed5 : button_device = button_device{}

    @editable var UpgradeLuck1 : button_device = button_device{}
    @editable var UpgradeLuck2 : button_device = button_device{}
    @editable var UpgradeLuck3 : button_device = button_device{}
    @editable var UpgradeLuck4 : button_device = button_device{}
    @editable var UpgradeLuck5 : button_device = button_device{}

    @editable var UpgradePlayerSpeed1 : button_device = button_device{}
    @editable var UpgradePlayerSpeed2 : button_device = button_device{}
    @editable var UpgradePlayerSpeed3 : button_device = button_device{}
    @editable var UpgradePlayerSpeed4 : button_device = button_device{}
    @editable var UpgradePlayerSpeed5 : button_device = button_device{}

    @editable var Luck100 : button_device = button_device{}

    #Level Area
    @editable var TreasureAreas : []mutator_zone_device = array{}

    #@editable var BarrierTUTO : barrier_device = barrier_device {}
    @editable var BarrierVILLAGE : barrier_device = barrier_device {}
    @editable var BarrierL1 : barrier_device = barrier_device {}
    @editable var BarrierL2 : barrier_device = barrier_device {}
    @editable var BarrierL3 : barrier_device = barrier_device {}
    @editable var BarrierL4 : barrier_device = barrier_device {}
    @editable var BarrierL5 : barrier_device = barrier_device {}
    @editable var BarrierL6 : barrier_device = barrier_device {}
    @editable var BarrierL7 : barrier_device = barrier_device {}
    @editable var BarrierL8 : barrier_device = barrier_device {}
    @editable var BarrierL9 : barrier_device = barrier_device {}
    @editable var BarrierL10 : barrier_device = barrier_device {}
    @editable var BarrierL11 : barrier_device = barrier_device {}
    @editable var BarrierL12 : barrier_device = barrier_device {}
    @editable var BarrierL13 : barrier_device = barrier_device {}
    @editable var BarrierL14 : barrier_device = barrier_device {}
    @editable var BarrierL15 : barrier_device = barrier_device {}
    @editable var BarrierL16 : barrier_device = barrier_device {}
    @editable var BarrierL17 : barrier_device = barrier_device {}
    @editable var BarrierL18 : barrier_device = barrier_device {}
    @editable var BarrierL19 : barrier_device = barrier_device {}
    @editable var BarrierL20 : barrier_device = barrier_device {}
    @editable var BarrierL21 : barrier_device = barrier_device {}
    @editable var BarrierL22 : barrier_device = barrier_device {}
    @editable var BarrierL23 : barrier_device = barrier_device {}
    @editable var BarrierL24 : barrier_device = barrier_device {}
    @editable var BarrierL25 : barrier_device = barrier_device {}
    @editable var BarrierL26 : barrier_device = barrier_device {}
    @editable var BarrierL27 : barrier_device = barrier_device {}
    @editable var BarrierL28 : barrier_device = barrier_device {}
    @editable var BarrierL29 : barrier_device = barrier_device {}
    @editable var BarrierL30 : barrier_device = barrier_device {}

    #Rebirth
    @editable var BarrierL31 : barrier_device = barrier_device {}
    @editable var BarrierL32 : barrier_device = barrier_device {}
    @editable var BarrierL33 : barrier_device = barrier_device {}

    @editable var VFX_B1 : vfx_spawner_device = vfx_spawner_device{}
    @editable var VFX_B2 : vfx_spawner_device = vfx_spawner_device{}
    @editable var VFX_B3 : vfx_spawner_device = vfx_spawner_device{}
    @editable var VFX_B4 : vfx_spawner_device = vfx_spawner_device{}
    @editable var VFX_B5 : vfx_spawner_device = vfx_spawner_device{}
    @editable var VFX_B6 : vfx_spawner_device = vfx_spawner_device{}
    @editable var VFX_B7 : vfx_spawner_device = vfx_spawner_device{}
    @editable var VFX_B8 : vfx_spawner_device = vfx_spawner_device{}
    @editable var VFX_B9 : vfx_spawner_device = vfx_spawner_device{}
    @editable var VFX_B10 : vfx_spawner_device = vfx_spawner_device{}
    @editable var VFX_B11 : vfx_spawner_device = vfx_spawner_device{}
    @editable var VFX_B12 : vfx_spawner_device = vfx_spawner_device{}
    @editable var VFX_B13 : vfx_spawner_device = vfx_spawner_device{}
    @editable var VFX_B14 : vfx_spawner_device = vfx_spawner_device{}
    @editable var VFX_B15 : vfx_spawner_device = vfx_spawner_device{}
    @editable var VFX_B16 : vfx_spawner_device = vfx_spawner_device{}
    @editable var VFX_B17 : vfx_spawner_device = vfx_spawner_device{}
    @editable var VFX_B18 : vfx_spawner_device = vfx_spawner_device{}
    @editable var VFX_B19 : vfx_spawner_device = vfx_spawner_device{}
    @editable var VFX_B20 : vfx_spawner_device = vfx_spawner_device{}
    @editable var VFX_B21 : vfx_spawner_device = vfx_spawner_device{}
    @editable var VFX_B22 : vfx_spawner_device = vfx_spawner_device{}
    @editable var VFX_B23 : vfx_spawner_device = vfx_spawner_device{}
    @editable var VFX_B24 : vfx_spawner_device = vfx_spawner_device{}
    @editable var VFX_B25 : vfx_spawner_device = vfx_spawner_device{}
    @editable var VFX_B26 : vfx_spawner_device = vfx_spawner_device{}
    @editable var VFX_B27 : vfx_spawner_device = vfx_spawner_device{}
    @editable var VFX_B28 : vfx_spawner_device = vfx_spawner_device{}
    @editable var VFX_B29 : vfx_spawner_device = vfx_spawner_device{}
    @editable var VFX_B30 : vfx_spawner_device = vfx_spawner_device{}

        #Purchase Barrier

        @editable var BuyBarrier1 : button_device = button_device{}
        @editable var BuyBarrier2 : button_device = button_device{}
        @editable var BuyBarrier3 : button_device = button_device{}
        @editable var BuyBarrier4 : button_device = button_device{}
        @editable var BuyBarrier5 : button_device = button_device{}
        @editable var BuyBarrier6 : button_device = button_device{}
        @editable var BuyBarrier7 : button_device = button_device{}
        @editable var BuyBarrier8 : button_device = button_device{}
        @editable var BuyBarrier9 : button_device = button_device{}
        @editable var BuyBarrier10 : button_device = button_device{}
        @editable var BuyBarrier11 : button_device = button_device{}
        @editable var BuyBarrier12 : button_device = button_device{}
        @editable var BuyBarrier13 : button_device = button_device{}
        @editable var BuyBarrier14 : button_device = button_device{}
        @editable var BuyBarrier15 : button_device = button_device{}
        @editable var BuyBarrier16 : button_device = button_device{}
        @editable var BuyBarrier17 : button_device = button_device{}
        @editable var BuyBarrier18 : button_device = button_device{}
        @editable var BuyBarrier19 : button_device = button_device{}
        @editable var BuyBarrier20 : button_device = button_device{}
        @editable var BuyBarrier21 : button_device = button_device{}
        @editable var BuyBarrier22 : button_device = button_device{}
        @editable var BuyBarrier23 : button_device = button_device{}
        @editable var BuyBarrier24 : button_device = button_device{}
        @editable var BuyBarrier25 : button_device = button_device{}
        @editable var BuyBarrier26 : button_device = button_device{}
        @editable var BuyBarrier27 : button_device = button_device{}
        @editable var BuyBarrier28 : button_device = button_device{}
        @editable var BuyBarrier29 : button_device = button_device{}
        @editable var BuyBarrier30 : button_device = button_device{}

        CostBarrier1:int = 1000
        CostBarrier2:int = 1500
        CostBarrier3:int = 2000
        CostBarrier4:int = 3500
        CostBarrier5:int = 6000
        CostBarrier6:int = 10000
        CostBarrier7:int = 15000
        CostBarrier8:int = 30000
        CostBarrier9:int = 45000
        CostBarrier10:int = 70000
        CostBarrier11:int = 120000
        CostBarrier12:int = 200000
        CostBarrier13:int = 320000
        CostBarrier14:int = 540000
        CostBarrier15:int = 900000
        CostBarrier16:int = 1500000
        CostBarrier17:int = 2500000
        CostBarrier18:int = 4000000
        CostBarrier19:int = 7000000
        CostBarrier20:int = 10000000
        CostBarrier21:int = 18000000
        CostBarrier22:int = 29000000
        CostBarrier23:int = 50000000
        CostBarrier24:int = 81000000
        CostBarrier25:int = 134000000
        CostBarrier26:int = 221000000
        CostBarrier27:int = 365000000
        CostBarrier28:int = 600000000
        CostBarrier29:int = 1000000000
        CostBarrier30:int = 17000000000

    #Player related
    var PlayerInTreasureArea : logic = false

    #Pet related
    @editable 
    Pets : []PetClass = array{}

    Power_L1P1:int = 50
    Power_L1P2:int = 62
    Power_L1P3:int = 75
    Power_L1P4:int = 100

    Power_L2P1:int = 90
    Power_L2P2:int = 113
    Power_L2P3:int = 135
    Power_L2P4:int = 180

    Power_L3P1:int = 150
    Power_L3P2:int = 190
    Power_L3P3:int = 225
    Power_L3P4:int = 300

    Power_L4P1:int = 250
    Power_L4P2:int = 310
    Power_L4P3:int = 375
    Power_L4P4:int = 500

    Power_L5P1:int = 400
    Power_L5P2:int = 500
    Power_L5P3:int = 600
    Power_L5P4:int = 800

    Power_L6P1:int = 680
    Power_L6P2:int = 850
    Power_L6P3:int = 1020
    Power_L6P4:int = 1360

    Power_L7P1:int = 1100
    Power_L7P2:int = 1375
    Power_L7P3:int = 1650
    Power_L7P4:int = 2200

    Power_L8P1:int = 1800
    Power_L8P2:int = 2250
    Power_L8P3:int = 2700
    Power_L8P4:int = 3600

    Power_L9P1:int = 3000
    Power_L9P2:int = 3750
    Power_L9P3:int = 4500
    Power_L9P4:int = 6000

    Power_L10P1:int = 5000
    Power_L10P2:int = 6250
    Power_L10P3:int = 7500
    Power_L10P4:int = 10000

    Power_L11P1:int = 8200
    Power_L11P2:int = 10250
    Power_L11P3:int = 12300
    Power_L11P4:int = 16400

    Power_L12P1:int = 13500
    Power_L12P2:int = 16875
    Power_L12P3:int = 20250
    Power_L12P4:int = 27000

    Power_L13P1:int = 22200
    Power_L13P2:int = 27750
    Power_L13P3:int = 33300
    Power_L13P4:int = 44400

    Power_L14P1:int = 36500
    Power_L14P2:int = 45625
    Power_L14P3:int = 54750
    Power_L14P4:int = 73000

    Power_L15P1:int = 60000
    Power_L15P2:int = 75000
    Power_L15P3:int = 90000
    Power_L15P4:int = 120000

    Power_L16P1:int = 100000
    Power_L16P2:int = 125000
    Power_L16P3:int = 150000
    Power_L16P4:int = 200000

    Power_L17P1:int = 163000
    Power_L17P2:int = 203750
    Power_L17P3:int = 244500
    Power_L17P4:int = 326000

    Power_L18P1:int = 270000
    Power_L18P2:int = 337500
    Power_L18P3:int = 405000
    Power_L18P4:int = 540000

    Power_L19P1:int = 445000
    Power_L19P2:int = 556250
    Power_L19P3:int = 667500
    Power_L19P4:int = 890000

    Power_L20P1:int = 730000
    Power_L20P2:int = 912500
    Power_L20P3:int = 1095000
    Power_L20P4:int = 5000000

    Power_L21P1:int = 1210000
    Power_L21P2:int = 1512500
    Power_L21P3:int = 1815000
    Power_L21P4:int = 2420000

    Power_L22P1:int = 2000000
    Power_L22P2:int = 2500000
    Power_L22P3:int = 3000000
    Power_L22P4:int = 4000000

    Power_L23P1:int = 3290000
    Power_L23P2:int = 4112500
    Power_L23P3:int = 4935000
    Power_L23P4:int = 6580000

    Power_L24P1:int = 5425000
    Power_L24P2:int = 6781250
    Power_L24P3:int = 8137500
    Power_L24P4:int = 10850000

    Power_L25P1:int = 8950000
    Power_L25P2:int = 11187500
    Power_L25P3:int = 13425000
    Power_L25P4:int = 17900000

    Power_L26P1:int = 14747000
    Power_L26P2:int = 18433750
    Power_L26P3:int = 22120500
    Power_L26P4:int = 29494000

    Power_L27P1:int = 24312000
    Power_L27P2:int = 30390000
    Power_L27P3:int = 36468000
    Power_L27P4:int = 48624000

    Power_L28P1:int = 40000000
    Power_L28P2:int = 50000000
    Power_L28P3:int = 60000000
    Power_L28P4:int = 80000000

    Power_L29P1:int = 66000000
    Power_L29P2:int = 82500000
    Power_L29P3:int = 99000000
    Power_L29P4:int = 132000000

    Power_L30P1:int = 108000000
    Power_L30P2:int = 135000000
    Power_L30P3:int = 162000000
    Power_L30P4:int = 216000000

    #Village opening

    @editable Cine_Unlock_Village : []cinematic_sequence_device = array{}
    
    #Purchase related
    @editable Cine_Unlock_L2 : []cinematic_sequence_device = array{}
    @editable Cine_Unlock_L2_Egg : []cinematic_sequence_device = array{}
    @editable Cine_Unlock_L3 : []cinematic_sequence_device = array{}
    @editable Cine_Unlock_L4 : []cinematic_sequence_device = array{}
    @editable Cine_Unlock_L5 : []cinematic_sequence_device = array{}
    @editable Cine_Unlock_L6 : []cinematic_sequence_device = array{}
    @editable Cine_Unlock_L7 : []cinematic_sequence_device = array{}
    @editable Cine_Unlock_L8 : []cinematic_sequence_device = array{}
    @editable Cine_Unlock_L9 : []cinematic_sequence_device = array{}
    @editable Cine_Unlock_L10 : []cinematic_sequence_device = array{}
    @editable Cine_Unlock_L11 : []cinematic_sequence_device = array{}
    @editable Cine_Unlock_L12 : []cinematic_sequence_device = array{}
    @editable Cine_Unlock_L13 : []cinematic_sequence_device = array{}
    @editable Cine_Unlock_L14 : []cinematic_sequence_device = array{}
    @editable Cine_Unlock_L15 : []cinematic_sequence_device = array{}
    @editable Cine_Unlock_L16 : []cinematic_sequence_device = array{}
    @editable Cine_Unlock_L17 : []cinematic_sequence_device = array{}
    @editable Cine_Unlock_L18 : []cinematic_sequence_device = array{}
    @editable Cine_Unlock_L19 : []cinematic_sequence_device = array{}
    @editable Cine_Unlock_L20 : []cinematic_sequence_device = array{}
    @editable Cine_Unlock_L21 : []cinematic_sequence_device = array{}
    @editable Cine_Unlock_L22 : []cinematic_sequence_device = array{}
    @editable Cine_Unlock_L23 : []cinematic_sequence_device = array{}
    @editable Cine_Unlock_L24 : []cinematic_sequence_device = array{}
    @editable Cine_Unlock_L25 : []cinematic_sequence_device = array{}
    @editable Cine_Unlock_L26 : []cinematic_sequence_device = array{}
    @editable Cine_Unlock_L27 : []cinematic_sequence_device = array{}
    @editable Cine_Unlock_L28 : []cinematic_sequence_device = array{}
    @editable Cine_Unlock_L29 : []cinematic_sequence_device = array{}
    @editable Cine_Unlock_L30 : []cinematic_sequence_device = array{}
    
    #Rebirth
    @editable Cine_Unlock_L31 : []cinematic_sequence_device = array{}
    @editable Cine_Unlock_L32 : []cinematic_sequence_device = array{}
    @editable Cine_Unlock_L33 : []cinematic_sequence_device = array{}


    @editable Cine_L1_Normal : []cinematic_sequence_device = array{}
    @editable Cine_L1_Rare : []cinematic_sequence_device = array{}
    @editable Cine_L1_Epic : []cinematic_sequence_device = array{}
    @editable Cine_L1_Legendary : []cinematic_sequence_device = array{}

    @editable Cine_L2_Normal : []cinematic_sequence_device = array{}
    @editable Cine_L2_Rare : []cinematic_sequence_device = array{}
    @editable Cine_L2_Epic : []cinematic_sequence_device = array{}
    @editable Cine_L2_Legendary : []cinematic_sequence_device = array{}

    @editable Cine_L3_Normal : []cinematic_sequence_device = array{}
    @editable Cine_L3_Rare : []cinematic_sequence_device = array{}
    @editable Cine_L3_Epic : []cinematic_sequence_device = array{}
    @editable Cine_L3_Legendary : []cinematic_sequence_device = array{}

    @editable Cine_L4_Normal : []cinematic_sequence_device = array{}
    @editable Cine_L4_Rare : []cinematic_sequence_device = array{}
    @editable Cine_L4_Epic : []cinematic_sequence_device = array{}
    @editable Cine_L4_Legendary : []cinematic_sequence_device = array{}

    @editable Cine_L5_Normal : []cinematic_sequence_device = array{}
    @editable Cine_L5_Rare : []cinematic_sequence_device = array{}
    @editable Cine_L5_Epic : []cinematic_sequence_device = array{}
    @editable Cine_L5_Legendary : []cinematic_sequence_device = array{}

    @editable Cine_L6_Normal : []cinematic_sequence_device = array{}
    @editable Cine_L6_Rare : []cinematic_sequence_device = array{}
    @editable Cine_L6_Epic : []cinematic_sequence_device = array{}
    @editable Cine_L6_Legendary : []cinematic_sequence_device = array{}

    @editable Cine_L7_Normal : []cinematic_sequence_device = array{}
    @editable Cine_L7_Rare : []cinematic_sequence_device = array{}
    @editable Cine_L7_Epic : []cinematic_sequence_device = array{}
    @editable Cine_L7_Legendary : []cinematic_sequence_device = array{}

    @editable Cine_L8_Normal : []cinematic_sequence_device = array{}
    @editable Cine_L8_Rare : []cinematic_sequence_device = array{}
    @editable Cine_L8_Epic : []cinematic_sequence_device = array{}
    @editable Cine_L8_Legendary : []cinematic_sequence_device = array{}

    @editable Cine_L9_Normal : []cinematic_sequence_device = array{}
    @editable Cine_L9_Rare : []cinematic_sequence_device = array{}
    @editable Cine_L9_Epic : []cinematic_sequence_device = array{}
    @editable Cine_L9_Legendary : []cinematic_sequence_device = array{}

    @editable Cine_L10_Normal : []cinematic_sequence_device = array{}
    @editable Cine_L10_Rare : []cinematic_sequence_device = array{}
    @editable Cine_L10_Epic : []cinematic_sequence_device = array{}
    @editable Cine_L10_Legendary : []cinematic_sequence_device = array{}

    @editable Cine_L11_Normal : []cinematic_sequence_device = array{}
    @editable Cine_L11_Rare : []cinematic_sequence_device = array{}
    @editable Cine_L11_Epic : []cinematic_sequence_device = array{}
    @editable Cine_L11_Legendary : []cinematic_sequence_device = array{}

    @editable Cine_L12_Normal : []cinematic_sequence_device = array{}
    @editable Cine_L12_Rare : []cinematic_sequence_device = array{}
    @editable Cine_L12_Epic : []cinematic_sequence_device = array{}
    @editable Cine_L12_Legendary : []cinematic_sequence_device = array{}

    @editable Cine_L13_Normal : []cinematic_sequence_device = array{}
    @editable Cine_L13_Rare : []cinematic_sequence_device = array{}
    @editable Cine_L13_Epic : []cinematic_sequence_device = array{}
    @editable Cine_L13_Legendary : []cinematic_sequence_device = array{}

    @editable Cine_L14_Normal : []cinematic_sequence_device = array{}
    @editable Cine_L14_Rare : []cinematic_sequence_device = array{}
    @editable Cine_L14_Epic : []cinematic_sequence_device = array{}
    @editable Cine_L14_Legendary : []cinematic_sequence_device = array{}

    @editable Cine_L15_Normal : []cinematic_sequence_device = array{}
    @editable Cine_L15_Rare : []cinematic_sequence_device = array{}
    @editable Cine_L15_Epic : []cinematic_sequence_device = array{}
    @editable Cine_L15_Legendary : []cinematic_sequence_device = array{}

    @editable Cine_L16_Normal : []cinematic_sequence_device = array{}
    @editable Cine_L16_Rare : []cinematic_sequence_device = array{}
    @editable Cine_L16_Epic : []cinematic_sequence_device = array{}
    @editable Cine_L16_Legendary : []cinematic_sequence_device = array{}

    @editable Cine_L17_Normal : []cinematic_sequence_device = array{}
    @editable Cine_L17_Rare : []cinematic_sequence_device = array{}
    @editable Cine_L17_Epic : []cinematic_sequence_device = array{}
    @editable Cine_L17_Legendary : []cinematic_sequence_device = array{}

    @editable Cine_L18_Normal : []cinematic_sequence_device = array{}
    @editable Cine_L18_Rare : []cinematic_sequence_device = array{}
    @editable Cine_L18_Epic : []cinematic_sequence_device = array{}
    @editable Cine_L18_Legendary : []cinematic_sequence_device = array{}

    @editable Cine_L19_Normal : []cinematic_sequence_device = array{}
    @editable Cine_L19_Rare : []cinematic_sequence_device = array{}
    @editable Cine_L19_Epic : []cinematic_sequence_device = array{}
    @editable Cine_L19_Legendary : []cinematic_sequence_device = array{}

    @editable Cine_L20_Normal : []cinematic_sequence_device = array{}
    @editable Cine_L20_Rare : []cinematic_sequence_device = array{}
    @editable Cine_L20_Epic : []cinematic_sequence_device = array{}
    @editable Cine_L20_Legendary : []cinematic_sequence_device = array{}

    @editable Cine_L21_Normal : []cinematic_sequence_device = array{}
    @editable Cine_L21_Rare : []cinematic_sequence_device = array{}
    @editable Cine_L21_Epic : []cinematic_sequence_device = array{}
    @editable Cine_L21_Legendary : []cinematic_sequence_device = array{}

    @editable Cine_L22_Normal : []cinematic_sequence_device = array{}
    @editable Cine_L22_Rare : []cinematic_sequence_device = array{}
    @editable Cine_L22_Epic : []cinematic_sequence_device = array{}
    @editable Cine_L22_Legendary : []cinematic_sequence_device = array{}

    @editable Cine_L23_Normal : []cinematic_sequence_device = array{}
    @editable Cine_L23_Rare : []cinematic_sequence_device = array{}
    @editable Cine_L23_Epic : []cinematic_sequence_device = array{}
    @editable Cine_L23_Legendary : []cinematic_sequence_device = array{}

    @editable Cine_L24_Normal : []cinematic_sequence_device = array{}
    @editable Cine_L24_Rare : []cinematic_sequence_device = array{}
    @editable Cine_L24_Epic : []cinematic_sequence_device = array{}
    @editable Cine_L24_Legendary : []cinematic_sequence_device = array{}

    @editable Cine_L25_Normal : []cinematic_sequence_device = array{}
    @editable Cine_L25_Rare : []cinematic_sequence_device = array{}
    @editable Cine_L25_Epic : []cinematic_sequence_device = array{}
    @editable Cine_L25_Legendary : []cinematic_sequence_device = array{}

    @editable Cine_L26_Normal : []cinematic_sequence_device = array{}
    @editable Cine_L26_Rare : []cinematic_sequence_device = array{}
    @editable Cine_L26_Epic : []cinematic_sequence_device = array{}
    @editable Cine_L26_Legendary : []cinematic_sequence_device = array{}

    @editable Cine_L27_Normal : []cinematic_sequence_device = array{}
    @editable Cine_L27_Rare : []cinematic_sequence_device = array{}
    @editable Cine_L27_Epic : []cinematic_sequence_device = array{}
    @editable Cine_L27_Legendary : []cinematic_sequence_device = array{}

    @editable Cine_L28_Normal : []cinematic_sequence_device = array{}
    @editable Cine_L28_Rare : []cinematic_sequence_device = array{}
    @editable Cine_L28_Epic : []cinematic_sequence_device = array{}
    @editable Cine_L28_Legendary : []cinematic_sequence_device = array{}

    @editable Cine_L29_Normal : []cinematic_sequence_device = array{}
    @editable Cine_L29_Rare : []cinematic_sequence_device = array{}
    @editable Cine_L29_Epic : []cinematic_sequence_device = array{}
    @editable Cine_L29_Legendary : []cinematic_sequence_device = array{}

    @editable Cine_L30_Normal : []cinematic_sequence_device = array{}
    @editable Cine_L30_Rare : []cinematic_sequence_device = array{}
    @editable Cine_L30_Epic : []cinematic_sequence_device = array{}
    @editable Cine_L30_Legendary : []cinematic_sequence_device = array{}

    @editable Cine_L31_Legendary : []cinematic_sequence_device = array{}
    @editable Cine_L32_Legendary : []cinematic_sequence_device = array{}
    @editable Cine_L33_Legendary : []cinematic_sequence_device = array{}
    @editable Cine_L34_Legendary : []cinematic_sequence_device = array{}

    @editable Cine_RebirthCamera1 : []cinematic_sequence_device = array{}
    @editable Cine_RebirthCamera2 : []cinematic_sequence_device = array{}
    @editable Cine_RebirthCamera3 : []cinematic_sequence_device = array{}

    @editable Cine_RebirthBox1 : []cinematic_sequence_device = array{}
    @editable Cine_RebirthBox2 : []cinematic_sequence_device = array{}
    @editable Cine_RebirthBox3 : []cinematic_sequence_device = array{}

    CostL1:int = 10
    CostL2:int = 75
    CostL3:int = 100
    CostL4:int = 175
    CostL5:int = 300
    CostL6:int = 500
    CostL7:int = 750
    CostL8:int = 1500
    CostL9:int = 2250
    CostL10:int = 3500
    CostL11:int = 6000
    CostL12:int = 10000
    CostL13:int = 16000
    CostL14:int = 27000
    CostL15:int = 45000
    CostL16:int = 75000
    CostL17:int = 125000
    CostL18:int = 200000
    CostL19:int = 350000
    CostL20:int = 500000
    CostL21:int = 900000
    CostL22:int = 1450000
    CostL23:int = 2500000
    CostL24:int = 4050000
    CostL25:int = 6700000
    CostL26:int = 11050000
    CostL27:int = 18250000
    CostL28:int = 30000000
    CostL29:int = 50000000
    CostL30:int = 850000000

    #Upgrade Related

    CostCrit1 : int = 1000
    CostCrit2 : int = 15000
    CostCrit3 : int = 160000
    CostCrit4 : int = 5500000
    CostCrit5 : int = 110500000

    CostSpeed1 : int = 3000
    CostSpeed2 : int = 35000
    CostSpeed3 : int = 450000
    CostSpeed4 : int = 14500000
    CostSpeed5 : int = 300000000

    CostPlayerSpeed1 : int = 5000
    CostPlayerSpeed2 : int = 60000
    CostPlayerSpeed3 : int = 750000
    CostPlayerSpeed4 : int = 25000000
    CostPlayerSpeed5 : int = 500000000

    CostLuck1 : int = 7500
    CostLuck2 : int = 100000
    CostLuck3 : int = 1250000
    CostLuck4 : int = 40500000
    CostLuck5 : int = 8500000000

    #Chest related

    @editable ChestsL1 : []Chest_L1 = array{}
    @editable ChestsL2 : []Chest_L2 = array{}
    @editable ChestsL3 : []Chest_L3 = array{}
    @editable ChestsL4 : []Chest_L4 = array{}
    @editable ChestsL5 : []Chest_L5 = array{}
    @editable ChestsL6 : []Chest_L6 = array{}
    @editable ChestsL7 : []Chest_L7 = array{}
    @editable ChestsL8 : []Chest_L8 = array{}
    @editable ChestsL9 : []Chest_L9 = array{}
    @editable ChestsL10 : []Chest_L10 = array{}
    @editable ChestsL11 : []Chest_L11 = array{}
    @editable ChestsL12 : []Chest_L12 = array{}
    @editable ChestsL13 : []Chest_L13 = array{}
    @editable ChestsL14 : []Chest_L14 = array{}
    @editable ChestsL15 : []Chest_L15 = array{}
    @editable ChestsL16 : []Chest_L16 = array{}
    @editable ChestsL17 : []Chest_L17 = array{}
    @editable ChestsL18 : []Chest_L18 = array{}
    @editable ChestsL19 : []Chest_L19 = array{}
    @editable ChestsL20 : []Chest_L20 = array{}
    @editable ChestsL21 : []Chest_L21 = array{}
    @editable ChestsL22 : []Chest_L22 = array{}
    @editable ChestsL23 : []Chest_L23 = array{}
    @editable ChestsL24 : []Chest_L24 = array{}
    @editable ChestsL25 : []Chest_L25 = array{}
    @editable ChestsL26 : []Chest_L26 = array{}
    @editable ChestsL27 : []Chest_L27 = array{}
    @editable ChestsL28 : []Chest_L28 = array{}
    @editable ChestsL29 : []Chest_L29 = array{}
    @editable ChestsL30 : []Chest_L30 = array{}
    
    @editable PropManipulatorTUTO : prop_manipulator_device = prop_manipulator_device{}
    @editable PropManipulatorL1 : prop_manipulator_device = prop_manipulator_device{}
    @editable PropManipulatorL2 : prop_manipulator_device = prop_manipulator_device{}
    @editable PropManipulatorL3 : prop_manipulator_device = prop_manipulator_device{}
    @editable PropManipulatorL4 : prop_manipulator_device = prop_manipulator_device{}
    @editable PropManipulatorL5 : prop_manipulator_device = prop_manipulator_device{}
    @editable PropManipulatorL6 : prop_manipulator_device = prop_manipulator_device{}
    @editable PropManipulatorL7 : prop_manipulator_device = prop_manipulator_device{}
    @editable PropManipulatorL8 : prop_manipulator_device = prop_manipulator_device{}
    @editable PropManipulatorL9 : prop_manipulator_device = prop_manipulator_device{}
    @editable PropManipulatorL10 : prop_manipulator_device = prop_manipulator_device{}
    @editable PropManipulatorL11 : prop_manipulator_device = prop_manipulator_device{}
    @editable PropManipulatorL12 : prop_manipulator_device = prop_manipulator_device{}
    @editable PropManipulatorL13 : prop_manipulator_device = prop_manipulator_device{}
    @editable PropManipulatorL14 : prop_manipulator_device = prop_manipulator_device{}
    @editable PropManipulatorL15 : prop_manipulator_device = prop_manipulator_device{}
    @editable PropManipulatorL16 : prop_manipulator_device = prop_manipulator_device{}
    @editable PropManipulatorL17 : prop_manipulator_device = prop_manipulator_device{}
    @editable PropManipulatorL18 : prop_manipulator_device = prop_manipulator_device{}
    @editable PropManipulatorL19 : prop_manipulator_device = prop_manipulator_device{}
    @editable PropManipulatorL20 : prop_manipulator_device = prop_manipulator_device{}
    @editable PropManipulatorL21 : prop_manipulator_device = prop_manipulator_device{}
    @editable PropManipulatorL22 : prop_manipulator_device = prop_manipulator_device{}
    @editable PropManipulatorL23 : prop_manipulator_device = prop_manipulator_device{}
    @editable PropManipulatorL24 : prop_manipulator_device = prop_manipulator_device{}
    @editable PropManipulatorL25 : prop_manipulator_device = prop_manipulator_device{}
    @editable PropManipulatorL26 : prop_manipulator_device = prop_manipulator_device{}
    @editable PropManipulatorL27 : prop_manipulator_device = prop_manipulator_device{}
    @editable PropManipulatorL28 : prop_manipulator_device = prop_manipulator_device{}
    @editable PropManipulatorL29 : prop_manipulator_device = prop_manipulator_device{}
    @editable PropManipulatorL30 : prop_manipulator_device = prop_manipulator_device{}

    #Gold Events Triggers
    @editable GoldChestL1:item_granter_device = item_granter_device{}
    @editable GoldHitL1:item_granter_device = item_granter_device{}
    @editable GoldChestL2:item_granter_device = item_granter_device{}
    @editable GoldHitL2:item_granter_device = item_granter_device{}
    @editable GoldChestL3:item_granter_device = item_granter_device{}
    @editable GoldHitL3:item_granter_device = item_granter_device{}
    @editable GoldChestL4:item_granter_device = item_granter_device{}
    @editable GoldHitL4:item_granter_device = item_granter_device{}
    @editable GoldChestL5:item_granter_device = item_granter_device{}
    @editable GoldHitL5:item_granter_device = item_granter_device{}
    @editable GoldChestL6:item_granter_device = item_granter_device{}
    @editable GoldHitL6:item_granter_device = item_granter_device{}
    @editable GoldChestL7:item_granter_device = item_granter_device{}
    @editable GoldHitL7:item_granter_device = item_granter_device{}
    @editable GoldChestL8:item_granter_device = item_granter_device{}
    @editable GoldHitL8:item_granter_device = item_granter_device{}
    @editable GoldChestL9:item_granter_device = item_granter_device{}
    @editable GoldHitL9:item_granter_device = item_granter_device{}
    @editable GoldChestL10:item_granter_device = item_granter_device{}
    @editable GoldHitL10:item_granter_device = item_granter_device{}
    @editable GoldChestL11:item_granter_device = item_granter_device{}
    @editable GoldHitL11:item_granter_device = item_granter_device{}
    @editable GoldChestL12:item_granter_device = item_granter_device{}
    @editable GoldHitL12:item_granter_device = item_granter_device{}
    @editable GoldChestL13:item_granter_device = item_granter_device{}
    @editable GoldHitL13:item_granter_device = item_granter_device{}
    @editable GoldChestL14:item_granter_device = item_granter_device{}
    @editable GoldHitL14:item_granter_device = item_granter_device{}
    @editable GoldChestL15:item_granter_device = item_granter_device{}
    @editable GoldHitL15:item_granter_device = item_granter_device{}
    @editable GoldChestL16:item_granter_device = item_granter_device{}
    @editable GoldHitL16:item_granter_device = item_granter_device{}
    @editable GoldChestL17:item_granter_device = item_granter_device{}
    @editable GoldHitL17:item_granter_device = item_granter_device{}
    @editable GoldChestL18:item_granter_device = item_granter_device{}
    @editable GoldHitL18:item_granter_device = item_granter_device{}
    @editable GoldChestL19:item_granter_device = item_granter_device{}
    @editable GoldHitL19:item_granter_device = item_granter_device{}
    @editable GoldChestL20:item_granter_device = item_granter_device{}
    @editable GoldHitL20:item_granter_device = item_granter_device{}
    @editable GoldChestL21:item_granter_device = item_granter_device{}
    @editable GoldHitL21:item_granter_device = item_granter_device{}
    @editable GoldChestL22:item_granter_device = item_granter_device{}
    @editable GoldHitL22:item_granter_device = item_granter_device{}
    @editable GoldChestL23:item_granter_device = item_granter_device{}
    @editable GoldHitL23:item_granter_device = item_granter_device{}
    @editable GoldChestL24:item_granter_device = item_granter_device{}
    @editable GoldHitL24:item_granter_device = item_granter_device{}
    @editable GoldChestL25:item_granter_device = item_granter_device{}
    @editable GoldHitL25:item_granter_device = item_granter_device{}
    @editable GoldChestL26:item_granter_device = item_granter_device{}
    @editable GoldHitL26:item_granter_device = item_granter_device{}
    @editable GoldChestL27:item_granter_device = item_granter_device{}
    @editable GoldHitL27:item_granter_device = item_granter_device{}
    @editable GoldChestL28:item_granter_device = item_granter_device{}
    @editable GoldHitL28:item_granter_device = item_granter_device{}
    @editable GoldChestL29:item_granter_device = item_granter_device{}
    @editable GoldHitL29:item_granter_device = item_granter_device{}
    @editable GoldChestL30:item_granter_device = item_granter_device{}
    @editable GoldHitL30:item_granter_device = item_granter_device{}

    #Side chests (BIG ones)
    @editable GoldChestBIG1:item_granter_device = item_granter_device{}
    @editable GoldChestBIG2:item_granter_device = item_granter_device{}
    @editable GoldChestBIG3:item_granter_device = item_granter_device{}
    @editable GoldChestBIG4:item_granter_device = item_granter_device{}
    @editable GoldChestBIG5:item_granter_device = item_granter_device{}
    @editable GoldChestBIG6:item_granter_device = item_granter_device{}
    @editable GoldChestBIG7:item_granter_device = item_granter_device{}
    @editable GoldChestBIG8:item_granter_device = item_granter_device{}

    #Gift Reward play time
    @editable Gift1:item_granter_device = item_granter_device{}
    @editable Gift2:item_granter_device = item_granter_device{}
    @editable Gift3:item_granter_device = item_granter_device{}
    @editable Gift4:item_granter_device = item_granter_device{}
    @editable Gift5:item_granter_device = item_granter_device{}
    @editable Gift6:item_granter_device = item_granter_device{}
    @editable Gift7:item_granter_device = item_granter_device{}
    @editable Gift8:item_granter_device = item_granter_device{}
    @editable Gift9:item_granter_device = item_granter_device{}

    #Mutator side chests
    @editable Mutator_BIG1:mutator_zone_device = mutator_zone_device{}
    @editable Mutator_BIG2:mutator_zone_device = mutator_zone_device{}
    @editable Mutator_BIG3:mutator_zone_device = mutator_zone_device{}
    @editable Mutator_BIG4:mutator_zone_device = mutator_zone_device{}
    @editable Mutator_BIG5:mutator_zone_device = mutator_zone_device{}
    @editable Mutator_BIG6:mutator_zone_device = mutator_zone_device{}
    @editable Mutator_BIG7:mutator_zone_device = mutator_zone_device{}
    @editable Mutator_BIG8:mutator_zone_device = mutator_zone_device{}

    @editable NotEnoughGoldHUD: hud_message_device = hud_message_device{}
    @editable NotEnoughDiamondHUD: hud_message_device = hud_message_device{}
    @editable AlreadyPurchasedHUD: hud_message_device = hud_message_device{}
    @editable AlreadyUnlockedHUD: hud_message_device = hud_message_device{}
    @editable RequirePreviousLevelHUD: hud_message_device = hud_message_device{}
    @editable RequirePreviousRebirthHUD: hud_message_device = hud_message_device{}
    @editable RequireHigherLevelHUD: hud_message_device = hud_message_device{}
    @editable NewEggHUD: hud_message_device = hud_message_device{}
    @editable VillageUnlockedHUD: hud_message_device = hud_message_device{}
    @editable Barrel_Found_SomethingHUD: hud_message_device = hud_message_device{}
    @editable Barrel_EmptyHUD: hud_message_device = hud_message_device{}
    @editable Found_PotionHUD: hud_message_device = hud_message_device{}
    @editable Code_ValidHUD: hud_message_device = hud_message_device{}
    @editable Try_Wheel_HUD: hud_message_device = hud_message_device{}

    @editable Upgrade_Crit1_HUD: hud_message_device = hud_message_device{}
    @editable Upgrade_Crit2_HUD: hud_message_device = hud_message_device{}
    @editable Upgrade_Crit3_HUD: hud_message_device = hud_message_device{}
    @editable Upgrade_Crit4_HUD: hud_message_device = hud_message_device{}
    @editable Upgrade_Crit5_HUD: hud_message_device = hud_message_device{}

    @editable Upgrade_Speed1_HUD: hud_message_device = hud_message_device{}
    @editable Upgrade_Speed2_HUD: hud_message_device = hud_message_device{}
    @editable Upgrade_Speed3_HUD: hud_message_device = hud_message_device{}
    @editable Upgrade_Speed4_HUD: hud_message_device = hud_message_device{}
    @editable Upgrade_Speed5_HUD: hud_message_device = hud_message_device{}

    @editable Upgrade_PlayerSpeed1_HUD: hud_message_device = hud_message_device{}
    @editable Upgrade_PlayerSpeed2_HUD: hud_message_device = hud_message_device{}
    @editable Upgrade_PlayerSpeed3_HUD: hud_message_device = hud_message_device{}
    @editable Upgrade_PlayerSpeed4_HUD: hud_message_device = hud_message_device{}
    @editable Upgrade_PlayerSpeed5_HUD: hud_message_device = hud_message_device{}

    @editable Upgrade_Luck1_HUD: hud_message_device = hud_message_device{}
    @editable Upgrade_Luck2_HUD: hud_message_device = hud_message_device{}
    @editable Upgrade_Luck3_HUD: hud_message_device = hud_message_device{}
    @editable Upgrade_Luck4_HUD: hud_message_device = hud_message_device{}
    @editable Upgrade_Luck5_HUD: hud_message_device = hud_message_device{}

    @editable Code_Rebirth5: hud_message_device = hud_message_device{}

    @editable Rebirth0_HUD : []hud_message_device = array{}
    @editable Rebirth1_HUD : []hud_message_device = array{}
    @editable Rebirth2_HUD : []hud_message_device = array{}
    @editable Rebirth3_HUD : []hud_message_device = array{}
    @editable Rebirth4_HUD : []hud_message_device = array{}
    @editable Rebirth5_HUD : []hud_message_device = array{}


    #To sync with the values entered in each item granters
    GoldChestL1Value:int = 50
    GoldChestL2Value:int = 60
    GoldChestL3Value:int = 70
    GoldChestL4Value:int = 100
    GoldChestL5Value:int = 160
    GoldChestL6Value:int = 220
    GoldChestL7Value:int = 280
    GoldChestL8Value:int = 480
    GoldChestL9Value:int = 610
    GoldChestL10Value:int = 810*2
    GoldChestL11Value:int = 1200*2
    GoldChestL12Value:int = 1700*2
    GoldChestL13Value:int = 2400*2
    GoldChestL14Value:int = 3400*2
    GoldChestL15Value:int = 4900*2
    GoldChestL16Value:int = 7100*2
    GoldChestL17Value:int = 10200*2
    GoldChestL18Value:int = 14000*2
    GoldChestL19Value:int = 21300*2
    GoldChestL20Value:int = 26300*2
    GoldChestL21Value:int = 40000*2
    GoldChestL22Value:int = 57000*2
    GoldChestL23Value:int = 85000*2
    GoldChestL24Value:int = 119000*2
    GoldChestL25Value:int = 170000*2
    GoldChestL26Value:int = 240000*2
    GoldChestL27Value:int = 349000*2
    GoldChestL28Value:int = 500000*2
    GoldChestL29Value:int = 700000*2
    GoldChestL30Value:int = 1000000*2
    
    #Side chests
    GoldChestBIG1Value:int = 250
    GoldChestBIG2Value:int = 550
    GoldChestBIG3Value:int = 2300
    GoldChestBIG4Value:int = 11000
    GoldChestBIG5Value:int = 30000
    GoldChestBIG6Value:int = 375000
    GoldChestBIG7Value:int = 1050000
    GoldChestBIG8Value:int = 100000

    # Gift Playtime
    GoldGift1Value:int = 250
    GoldGift2Value:int = 550
    GoldGift3Value:int = 2500
    GoldGift4Value:int = 5000
    GoldGift5Value:int = 20000
    GoldGift6Value:int = 40000
    GoldGift7Value:int = 200000
    GoldGift8Value:int = 200000
    GoldGift9Value:int = 200000

    #To sync with the values entered in each item granters
    GoldHitL1Value:int = 5
    GoldHitL2Value:int = 5
    GoldHitL3Value:int = 5
    GoldHitL4Value:int = 10
    GoldHitL5Value:int = 15
    GoldHitL6Value:int = 20
    GoldHitL7Value:int = 30
    GoldHitL8Value:int = 50
    GoldHitL9Value:int = 60
    GoldHitL10Value:int = 80
    GoldHitL11Value:int = 120
    GoldHitL12Value:int = 170
    GoldHitL13Value:int = 240
    GoldHitL14Value:int = 340
    GoldHitL15Value:int = 490
    GoldHitL16Value:int = 710
    GoldHitL17Value:int = 1000
    GoldHitL18Value:int = 1400
    GoldHitL19Value:int = 2100
    GoldHitL20Value:int = 2600
    GoldHitL21Value:int = 2600
    GoldHitL22Value:int = 2600
    GoldHitL23Value:int = 2600
    GoldHitL24Value:int = 2600
    GoldHitL25Value:int = 2600
    GoldHitL26Value:int = 2600
    GoldHitL27Value:int = 2600
    GoldHitL28Value:int = 2600
    GoldHitL29Value:int = 2600
    GoldHitL30Value:int = 2600

    #Accolade

    @editable Accolade_Egg1_2 : accolades_device = accolades_device{}
    @editable Accolade_Egg3_4 : accolades_device = accolades_device{}
    @editable Accolade_Egg5_6 : accolades_device = accolades_device{}
    @editable Accolade_Egg7_8 : accolades_device = accolades_device{}
    @editable Accolade_Egg9_10 : accolades_device = accolades_device{}
    @editable Accolade_Egg11_12 : accolades_device = accolades_device{}
    @editable Accolade_Egg13_14 : accolades_device = accolades_device{}
    @editable Accolade_Egg15_16 : accolades_device = accolades_device{}
    @editable Accolade_Egg17_18 : accolades_device = accolades_device{}
    @editable Accolade_Egg19_20 : accolades_device = accolades_device{}
    @editable Accolade_Egg21_22 : accolades_device = accolades_device{}
    @editable Accolade_Egg23_24 : accolades_device = accolades_device{}
    @editable Accolade_Egg25_26 : accolades_device = accolades_device{}
    @editable Accolade_Egg27_28 : accolades_device = accolades_device{}
    @editable Accolade_Egg29_30 : accolades_device = accolades_device{}

    @editable Accolade_B1B5 : accolades_device = accolades_device{}
    @editable Accolade_B6B10 : accolades_device = accolades_device{}
    @editable Accolade_B11B15 : accolades_device = accolades_device{}
    @editable Accolade_B16B20 : accolades_device = accolades_device{}
    @editable Accolade_B21B25 : accolades_device = accolades_device{}
    @editable Accolade_B26B30 : accolades_device = accolades_device{}

    @editable Accolade_ChestL1L5 : accolades_device = accolades_device{}
    @editable Accolade_ChestL6L10 : accolades_device = accolades_device{}
    @editable Accolade_ChestL11L15 : accolades_device = accolades_device{}
    @editable Accolade_ChestL16L20 : accolades_device = accolades_device{}
    @editable Accolade_ChestL21L25 : accolades_device = accolades_device{}
    @editable Accolade_ChestL26L30 : accolades_device = accolades_device{}


    @editable Accolade_Upgrades : accolades_device = accolades_device{}

    @editable Accolade_Final_Level : accolades_device = accolades_device{}

    @editable RadioMusicBackground : radio_device = radio_device{}

    @editable RebirthSFX : radio_device = radio_device{}

    @editable RebirthSFX_End : radio_device = radio_device{}

    #Gift Codes
    @editable var CodeInputButton0 : button_device = button_device{}
    @editable var CodeInputButton1 : button_device = button_device{}
    @editable var CodeInputButton2 : button_device = button_device{}
    @editable var CodeInputButton3 : button_device = button_device{}
    @editable var CodeInputButton4 : button_device = button_device{}
    @editable var CodeInputButton5 : button_device = button_device{}
    @editable var CodeInputButton6 : button_device = button_device{}
    @editable var CodeInputButton7 : button_device = button_device{}
    @editable var CodeInputButton8 : button_device = button_device{}
    @editable var CodeInputButton9 : button_device = button_device{}
    @editable var CodeEnterButton : button_device = button_device{}
    @editable var CodeResetButton : button_device = button_device{}

    #SAVE RELATED:

        # Manages and updates stats for each player.
        PlayerStatsManager:player_stats_manager = player_stats_manager{}

    #INVENTORY RELATED:
    @editable   HUD_Trigger_Inventory:trigger_device = trigger_device{}

    @editable
    ShopOpenButtons : []button_device = array{}

    @editable
    ShopPurchases : []purchase_data = array{}

    TextForUI<localizes>(Text : string) : message = "{Text}"
    PurchasePerPage : int = 3

    var AgentMap : [agent]agent_data = map{}

    #WHEEL RELATED:

    @editable WheelID : type{_X : int where _X >= 0, _X <= 3} = 0
    @editable LuckyWheelProp : creative_prop = creative_prop{}
    @editable WheelButton : button_device = button_device{}
    @editable SlotsNumber : int = 16
    @editable ScreenBackgroundColor : color = color{}
    @editable RadioWheel : radio_device = radio_device{}

    #Player speed:

    @editable var ClassSelector1 : class_and_team_selector_device = class_and_team_selector_device{}
    @editable var ClassSelector2 : class_and_team_selector_device = class_and_team_selector_device{}
    @editable var ClassSelector3 : class_and_team_selector_device = class_and_team_selector_device{}
    @editable var ClassSelector4 : class_and_team_selector_device = class_and_team_selector_device{}
    @editable var ClassSelector5 : class_and_team_selector_device = class_and_team_selector_device{}

    #Diamonds
    #@editable GranterDiamondLoot:item_granter_device = item_granter_device{}
    @editable
    CollectibleDiamonds : []collectible_object_device = array{}
    
    @editable
    BarrelButtons : []button_device = array{}

    #Popupdialogue:
    @editable
    PopupdialogueDevice:popup_dialog_device=popup_dialog_device{}

    #Tuto
    @editable
    Objective_Wheel:objective_device=objective_device{}

    @editable
    Wheel_Mutator:mutator_zone_device = mutator_zone_device{}

#--------------------------------------------------------ON BEGIN-------------------------------------------------------#
#--------------------------------------------------------ON BEGIN-------------------------------------------------------#
#--------------------------------------------------------ON BEGIN-------------------------------------------------------#
    OnBegin<override>()<suspends>: void =
        
        LOADSaveMutator.AgentExitsEvent.Subscribe(LoadPlayerSave)
        Rebirth1.InteractedWithEvent.Subscribe(Rebirth1_Func)
        Rebirth2.InteractedWithEvent.Subscribe(Rebirth2_Func)
        Rebirth3.InteractedWithEvent.Subscribe(Rebirth3_Func)
        Rebirth4.InteractedWithEvent.Subscribe(Rebirth4_Func)
        Rebirth5.InteractedWithEvent.Subscribe(Rebirth5_Func)
        Rebirth5Code.InteractedWithEvent.Subscribe(Rebirth5_Func_Code)
        ButtonTest.InteractedWithEvent.Subscribe(Test_Func)

        PopupdialogueDevice.RespondingButtonEvent.Subscribe(PopupButton1)

        PurchaseButtonTUTO.InteractedWithEvent.Subscribe(PurchaseL1)
        PurchaseButtonL1.InteractedWithEvent.Subscribe(PurchaseL1)
        PurchaseButtonL2.InteractedWithEvent.Subscribe(PurchaseL2)
        PurchaseButtonL3.InteractedWithEvent.Subscribe(PurchaseL3)
        PurchaseButtonL4.InteractedWithEvent.Subscribe(PurchaseL4)
        PurchaseButtonL5.InteractedWithEvent.Subscribe(PurchaseL5)
        PurchaseButtonL6.InteractedWithEvent.Subscribe(PurchaseL6)
        PurchaseButtonL7.InteractedWithEvent.Subscribe(PurchaseL7)
        PurchaseButtonL8.InteractedWithEvent.Subscribe(PurchaseL8)
        PurchaseButtonL9.InteractedWithEvent.Subscribe(PurchaseL9)
        PurchaseButtonL10.InteractedWithEvent.Subscribe(PurchaseL10)
        PurchaseButtonL11.InteractedWithEvent.Subscribe(PurchaseL11)
        PurchaseButtonL12.InteractedWithEvent.Subscribe(PurchaseL12)
        PurchaseButtonL13.InteractedWithEvent.Subscribe(PurchaseL13)
        PurchaseButtonL14.InteractedWithEvent.Subscribe(PurchaseL14)
        PurchaseButtonL15.InteractedWithEvent.Subscribe(PurchaseL15)
        PurchaseButtonL16.InteractedWithEvent.Subscribe(PurchaseL16)
        PurchaseButtonL17.InteractedWithEvent.Subscribe(PurchaseL17)
        PurchaseButtonL18.InteractedWithEvent.Subscribe(PurchaseL18)
        PurchaseButtonL19.InteractedWithEvent.Subscribe(PurchaseL19)
        PurchaseButtonL20.InteractedWithEvent.Subscribe(PurchaseL20)
        PurchaseButtonL21.InteractedWithEvent.Subscribe(PurchaseL21)
        PurchaseButtonL22.InteractedWithEvent.Subscribe(PurchaseL22)
        PurchaseButtonL23.InteractedWithEvent.Subscribe(PurchaseL23)
        PurchaseButtonL24.InteractedWithEvent.Subscribe(PurchaseL24)
        PurchaseButtonL25.InteractedWithEvent.Subscribe(PurchaseL25)
        PurchaseButtonL26.InteractedWithEvent.Subscribe(PurchaseL26)
        PurchaseButtonL27.InteractedWithEvent.Subscribe(PurchaseL27)
        PurchaseButtonL28.InteractedWithEvent.Subscribe(PurchaseL28)
        PurchaseButtonL29.InteractedWithEvent.Subscribe(PurchaseL29)
        PurchaseButtonL30.InteractedWithEvent.Subscribe(PurchaseL30)

        #Secret Pets
        PurchaseButtonL31.InteractedWithEvent.Subscribe(PurchaseL31)
        PurchaseButtonL32.InteractedWithEvent.Subscribe(PurchaseL32)
        PurchaseButtonL33.InteractedWithEvent.Subscribe(PurchaseL33)
        PurchaseButtonL34.InteractedWithEvent.Subscribe(PurchaseL34)

        #Gold test
        IncreaseGoldButton.InteractedWithEvent.Subscribe(Increasegoldevent)
        DecreaseGoldButton.InteractedWithEvent.Subscribe(Decreasegoldevent)
        IncreaseGoldBonus.InteractedWithEvent.Subscribe(Increasegoldbonusevent)

        #Diamond test
        IncreaseDiamondButton.InteractedWithEvent.Subscribe(Increasediamondevent)
        DecreaseDiamondButton.InteractedWithEvent.Subscribe(Decreasediamondevent)

        #Diamond Loot Obby Normal
        Obby1DiamondNormalP1.InteractedWithEvent.Subscribe(DiamondObbyNormal)
        Obby2DiamondNormalP1.InteractedWithEvent.Subscribe(DiamondObbyNormal)
        Obby1DiamondNormalP2.InteractedWithEvent.Subscribe(DiamondObbyNormal)
        Obby2DiamondNormalP2.InteractedWithEvent.Subscribe(DiamondObbyNormal)
        Obby1DiamondNormalP3.InteractedWithEvent.Subscribe(DiamondObbyNormal)
        Obby2DiamondNormalP3.InteractedWithEvent.Subscribe(DiamondObbyNormal)

        #Diamond Loot Obby On Time
        Obby1DiamondOnTimeP1.InteractedWithEvent.Subscribe(DiamondObbyOnTime)
        Obby2DiamondOnTimeP1.InteractedWithEvent.Subscribe(DiamondObbyOnTime)
        Obby1DiamondOnTimeP2.InteractedWithEvent.Subscribe(DiamondObbyOnTime)
        Obby2DiamondOnTimeP2.InteractedWithEvent.Subscribe(DiamondObbyOnTime)
        Obby1DiamondOnTimeP3.InteractedWithEvent.Subscribe(DiamondObbyOnTime)
        Obby2DiamondOnTimeP3.InteractedWithEvent.Subscribe(DiamondObbyOnTime)

        #PetDex
        OpenPetDexButton.InteractedWithEvent.Subscribe(OpenPetDexEvent)

        #Upgrades
        UpgradeCrit1.InteractedWithEvent.Subscribe(UpgradeCrit_1)
        UpgradeCrit2.InteractedWithEvent.Subscribe(UpgradeCrit_2)
        UpgradeCrit3.InteractedWithEvent.Subscribe(UpgradeCrit_3)
        UpgradeCrit4.InteractedWithEvent.Subscribe(UpgradeCrit_4)
        UpgradeCrit5.InteractedWithEvent.Subscribe(UpgradeCrit_5)

        UpgradeSpeed1.InteractedWithEvent.Subscribe(UpgradeSpeed_1)
        UpgradeSpeed2.InteractedWithEvent.Subscribe(UpgradeSpeed_2)
        UpgradeSpeed3.InteractedWithEvent.Subscribe(UpgradeSpeed_3)
        UpgradeSpeed4.InteractedWithEvent.Subscribe(UpgradeSpeed_4)
        UpgradeSpeed5.InteractedWithEvent.Subscribe(UpgradeSpeed_5)

        UpgradePlayerSpeed1.InteractedWithEvent.Subscribe(UpgradePlayerSpeed_1)
        UpgradePlayerSpeed2.InteractedWithEvent.Subscribe(UpgradePlayerSpeed_2)
        UpgradePlayerSpeed3.InteractedWithEvent.Subscribe(UpgradePlayerSpeed_3)
        UpgradePlayerSpeed4.InteractedWithEvent.Subscribe(UpgradePlayerSpeed_4)
        UpgradePlayerSpeed5.InteractedWithEvent.Subscribe(UpgradePlayerSpeed_5)

        UpgradeLuck1.InteractedWithEvent.Subscribe(UpgradeLuck_1)
        UpgradeLuck2.InteractedWithEvent.Subscribe(UpgradeLuck_2)
        UpgradeLuck3.InteractedWithEvent.Subscribe(UpgradeLuck_3)
        UpgradeLuck4.InteractedWithEvent.Subscribe(UpgradeLuck_4)
        UpgradeLuck5.InteractedWithEvent.Subscribe(UpgradeLuck_5)

        Luck100.InteractedWithEvent.Subscribe(UpgradeLuck_100)

        #Levels
        BuyBarrier1.InteractedWithEvent.Subscribe(Buy_B1)
        BuyBarrier2.InteractedWithEvent.Subscribe(Buy_B2)
        BuyBarrier3.InteractedWithEvent.Subscribe(Buy_B3)
        BuyBarrier4.InteractedWithEvent.Subscribe(Buy_B4)
        BuyBarrier5.InteractedWithEvent.Subscribe(Buy_B5)
        BuyBarrier6.InteractedWithEvent.Subscribe(Buy_B6)
        BuyBarrier7.InteractedWithEvent.Subscribe(Buy_B7)
        BuyBarrier8.InteractedWithEvent.Subscribe(Buy_B8)
        BuyBarrier9.InteractedWithEvent.Subscribe(Buy_B9)
        BuyBarrier10.InteractedWithEvent.Subscribe(Buy_B10)
        BuyBarrier11.InteractedWithEvent.Subscribe(Buy_B11)
        BuyBarrier12.InteractedWithEvent.Subscribe(Buy_B12)
        BuyBarrier13.InteractedWithEvent.Subscribe(Buy_B13)
        BuyBarrier14.InteractedWithEvent.Subscribe(Buy_B14)
        BuyBarrier15.InteractedWithEvent.Subscribe(Buy_B15)
        BuyBarrier16.InteractedWithEvent.Subscribe(Buy_B16)
        BuyBarrier17.InteractedWithEvent.Subscribe(Buy_B17)
        BuyBarrier18.InteractedWithEvent.Subscribe(Buy_B18)
        BuyBarrier19.InteractedWithEvent.Subscribe(Buy_B19)
        BuyBarrier20.InteractedWithEvent.Subscribe(Buy_B20)
        BuyBarrier21.InteractedWithEvent.Subscribe(Buy_B21)
        BuyBarrier22.InteractedWithEvent.Subscribe(Buy_B22)
        BuyBarrier23.InteractedWithEvent.Subscribe(Buy_B23)
        BuyBarrier24.InteractedWithEvent.Subscribe(Buy_B24)
        BuyBarrier25.InteractedWithEvent.Subscribe(Buy_B25)
        BuyBarrier26.InteractedWithEvent.Subscribe(Buy_B26)
        BuyBarrier27.InteractedWithEvent.Subscribe(Buy_B27)
        BuyBarrier28.InteractedWithEvent.Subscribe(Buy_B28)
        BuyBarrier29.InteractedWithEvent.Subscribe(Buy_B29)
        BuyBarrier30.InteractedWithEvent.Subscribe(Buy_B30)

        #Gold Events

        GoldHitL1.ItemGrantedEvent.Subscribe(GrantGoldHitL1)
        GoldHitL2.ItemGrantedEvent.Subscribe(GrantGoldHitL2)
        GoldHitL3.ItemGrantedEvent.Subscribe(GrantGoldHitL3)
        GoldHitL4.ItemGrantedEvent.Subscribe(GrantGoldHitL4)
        GoldHitL5.ItemGrantedEvent.Subscribe(GrantGoldHitL5)
        GoldHitL6.ItemGrantedEvent.Subscribe(GrantGoldHitL6)
        GoldHitL7.ItemGrantedEvent.Subscribe(GrantGoldHitL7)
        GoldHitL8.ItemGrantedEvent.Subscribe(GrantGoldHitL8)
        GoldHitL9.ItemGrantedEvent.Subscribe(GrantGoldHitL9)
        GoldHitL10.ItemGrantedEvent.Subscribe(GrantGoldHitL10)
        GoldHitL11.ItemGrantedEvent.Subscribe(GrantGoldHitL11)
        GoldHitL12.ItemGrantedEvent.Subscribe(GrantGoldHitL12)
        GoldHitL13.ItemGrantedEvent.Subscribe(GrantGoldHitL13)
        GoldHitL14.ItemGrantedEvent.Subscribe(GrantGoldHitL14)
        GoldHitL15.ItemGrantedEvent.Subscribe(GrantGoldHitL15)
        GoldHitL16.ItemGrantedEvent.Subscribe(GrantGoldHitL16)
        GoldHitL17.ItemGrantedEvent.Subscribe(GrantGoldHitL17)
        GoldHitL18.ItemGrantedEvent.Subscribe(GrantGoldHitL18)
        GoldHitL19.ItemGrantedEvent.Subscribe(GrantGoldHitL19)
        GoldHitL20.ItemGrantedEvent.Subscribe(GrantGoldHitL20)
        GoldHitL21.ItemGrantedEvent.Subscribe(GrantGoldHitL21)
        GoldHitL22.ItemGrantedEvent.Subscribe(GrantGoldHitL22)
        GoldHitL23.ItemGrantedEvent.Subscribe(GrantGoldHitL23)
        GoldHitL24.ItemGrantedEvent.Subscribe(GrantGoldHitL24)
        GoldHitL25.ItemGrantedEvent.Subscribe(GrantGoldHitL25)
        GoldHitL26.ItemGrantedEvent.Subscribe(GrantGoldHitL26)
        GoldHitL27.ItemGrantedEvent.Subscribe(GrantGoldHitL27)
        GoldHitL28.ItemGrantedEvent.Subscribe(GrantGoldHitL28)
        GoldHitL29.ItemGrantedEvent.Subscribe(GrantGoldHitL29)
        GoldHitL30.ItemGrantedEvent.Subscribe(GrantGoldHitL30)

        PropManipulatorTUTO.DamagedEvent.Subscribe(OnPlayerHitDamage)
        PropManipulatorL1.DamagedEvent.Subscribe(OnPlayerHitDamage)
        PropManipulatorL2.DamagedEvent.Subscribe(OnPlayerHitDamage)
        PropManipulatorL3.DamagedEvent.Subscribe(OnPlayerHitDamage)
        PropManipulatorL4.DamagedEvent.Subscribe(OnPlayerHitDamage)
        PropManipulatorL5.DamagedEvent.Subscribe(OnPlayerHitDamage)
        PropManipulatorL6.DamagedEvent.Subscribe(OnPlayerHitDamage)
        PropManipulatorL7.DamagedEvent.Subscribe(OnPlayerHitDamage)
        PropManipulatorL8.DamagedEvent.Subscribe(OnPlayerHitDamage)
        PropManipulatorL9.DamagedEvent.Subscribe(OnPlayerHitDamage)
        PropManipulatorL10.DamagedEvent.Subscribe(OnPlayerHitDamage)
        PropManipulatorL11.DamagedEvent.Subscribe(OnPlayerHitDamage)
        PropManipulatorL12.DamagedEvent.Subscribe(OnPlayerHitDamage)
        PropManipulatorL13.DamagedEvent.Subscribe(OnPlayerHitDamage)
        PropManipulatorL14.DamagedEvent.Subscribe(OnPlayerHitDamage)
        PropManipulatorL15.DamagedEvent.Subscribe(OnPlayerHitDamage)
        PropManipulatorL16.DamagedEvent.Subscribe(OnPlayerHitDamage)
        PropManipulatorL17.DamagedEvent.Subscribe(OnPlayerHitDamage)
        PropManipulatorL18.DamagedEvent.Subscribe(OnPlayerHitDamage)
        PropManipulatorL19.DamagedEvent.Subscribe(OnPlayerHitDamage)
        PropManipulatorL20.DamagedEvent.Subscribe(OnPlayerHitDamage)
        PropManipulatorL21.DamagedEvent.Subscribe(OnPlayerHitDamage)
        PropManipulatorL22.DamagedEvent.Subscribe(OnPlayerHitDamage)
        PropManipulatorL23.DamagedEvent.Subscribe(OnPlayerHitDamage)
        PropManipulatorL24.DamagedEvent.Subscribe(OnPlayerHitDamage)
        PropManipulatorL25.DamagedEvent.Subscribe(OnPlayerHitDamage)
        PropManipulatorL26.DamagedEvent.Subscribe(OnPlayerHitDamage)
        PropManipulatorL27.DamagedEvent.Subscribe(OnPlayerHitDamage)
        PropManipulatorL28.DamagedEvent.Subscribe(OnPlayerHitDamage)
        PropManipulatorL29.DamagedEvent.Subscribe(OnPlayerHitDamage)
        PropManipulatorL30.DamagedEvent.Subscribe(OnPlayerHitDamage)

        #BIG CHESTS
        Mutator_BIG1.AgentEntersEvent.Subscribe(GrantGoldChestBIG1)
        Mutator_BIG2.AgentEntersEvent.Subscribe(GrantGoldChestBIG2)
        Mutator_BIG3.AgentEntersEvent.Subscribe(GrantGoldChestBIG3)
        Mutator_BIG4.AgentEntersEvent.Subscribe(GrantGoldChestBIG4)
        Mutator_BIG5.AgentEntersEvent.Subscribe(GrantGoldChestBIG5)
        Mutator_BIG6.AgentEntersEvent.Subscribe(GrantGoldChestBIG6)
        Mutator_BIG7.AgentEntersEvent.Subscribe(GrantGoldChestBIG7)
        Mutator_BIG8.AgentEntersEvent.Subscribe(GrantGoldChestBIG8)

        #Gift Paytime
        Gift1.ItemGrantedEvent.Subscribe(GrantGift1)
        Gift2.ItemGrantedEvent.Subscribe(GrantGift2)
        Gift3.ItemGrantedEvent.Subscribe(GrantGift3)
        Gift4.ItemGrantedEvent.Subscribe(GrantGift4)
        Gift5.ItemGrantedEvent.Subscribe(GrantGift5)
        Gift6.ItemGrantedEvent.Subscribe(GrantGift6)
        Gift7.ItemGrantedEvent.Subscribe(GrantGift7)
        Gift8.ItemGrantedEvent.Subscribe(GrantGift8)
        Gift9.ItemGrantedEvent.Subscribe(GrantGift9)

        #Codes map 1
        CodeInputButton0.InteractedWithEvent.Subscribe(Input0)
        CodeInputButton1.InteractedWithEvent.Subscribe(Input1)
        CodeInputButton2.InteractedWithEvent.Subscribe(Input2)
        CodeInputButton3.InteractedWithEvent.Subscribe(Input3)
        CodeInputButton4.InteractedWithEvent.Subscribe(Input4)
        CodeInputButton5.InteractedWithEvent.Subscribe(Input5)
        CodeInputButton6.InteractedWithEvent.Subscribe(Input6)
        CodeInputButton7.InteractedWithEvent.Subscribe(Input7)
        CodeInputButton8.InteractedWithEvent.Subscribe(Input8)
        CodeInputButton9.InteractedWithEvent.Subscribe(Input9)
        CodeEnterButton.InteractedWithEvent.Subscribe(InputEnter)
        CodeResetButton.InteractedWithEvent.Subscribe(InputReset)

        #Other
        CheckFunctionButton.InteractedWithEvent.Subscribe(CheckFunction)
        #ReleasePetButton1.InteractedWithEvent.Subscribe(ReleasePet1)

        P1Spawner.SpawnedEvent.Subscribe(OnPlayerAdded)
        P2Spawner.SpawnedEvent.Subscribe(OnPlayerAdded)
        P3Spawner.SpawnedEvent.Subscribe(OnPlayerAdded)
        GetPlayspace().PlayerRemovedEvent().Subscribe(OnPlayerRemoved)

        
        #for(index->PBB:PetBuyButtons):  
        #    PBB.InteractedWithEvent.SubscribeAgent(PBB_Event, index)

        for(index->TA:TreasureAreas):  
            TA.AgentEntersEvent.Subscribe(OnPlayerEnterTreasureArea)

        for(index->TA:TreasureAreas):  
            TA.AgentExitsEvent.Subscribe(OnPlayerExitTreasureArea)

        Players := GetPlayspace().GetPlayers()
        # Initialize player stats (game save)
        PlayerStatsManager.InitializeAllPlayers(Players)

        #Iventory
        for (ShopOpenButton : ShopOpenButtons):
            ShopOpenButton.InteractedWithEvent.Subscribe(OnShopOpen)
        
        #Collectible diamonds
        #GranterDiamondLoot.ItemGrantedEvent.Subscribe(CollectDiamondLoot)
        for (CollectibleDiamond : CollectibleDiamonds):
            CollectibleDiamond.CollectedEvent.Subscribe(CollectDiamondLoot)

        #Barrels
        for (BarrelButton : BarrelButtons):
            BarrelButton.InteractedWithEvent.Subscribe(BarrelLoot)

        #Tuto
        Wheel_Mutator.AgentEntersEvent.Subscribe(Entered_Wheel)

        #Wheel related
        LuckyMaterials1 : Materials.MI_LuckyWheelScreen_1_material = Materials.MI_LuckyWheelScreen_1_material{}
        if (WheelID = 0):            
            LuckyWheelProp.SetMaterial(LuckyMaterials1)
            set LuckyMaterials1.BackgroundColor = ScreenBackgroundColor
            AwaitButtonInteraction(LuckyMaterials1)

    LoadPlayerSave(Agent:agent):void=
        if:
            PlayerObj := player[Agent]
            CustomPlayer:Custom_Player = PlayersMap[PlayerObj]
        then:
            Print("CustomPlayer.GameLoaded: {CustomPlayer.GameLoaded}")
            if(CustomPlayer.GameLoaded=0): 
                Print("Loading Game")
                set CustomPlayer.GameLoaded=1
                #Initialize Rebirth Effect
                InitializeRebirthBonus(Agent)
                #Initialize Pets
                spawn:
                    InitializeL1Pets(Agent)
                spawn:
                    InitializeL2Pets(Agent)
                spawn:
                    InitializeL3Pets(Agent)
                spawn:
                    InitializeL4Pets(Agent)
                spawn:
                    InitializeL5Pets(Agent)
                spawn:
                    InitializeL6Pets(Agent)
                spawn:
                    InitializeL7Pets(Agent)
                spawn:
                    InitializeL8Pets(Agent)
                spawn:
                    InitializeL9Pets(Agent)
                spawn:
                    InitializeL10Pets(Agent)
                spawn:
                    InitializeL11Pets(Agent)
                spawn:
                    InitializeL12Pets(Agent)
                spawn:
                    InitializeL13Pets(Agent)
                spawn:
                    InitializeL14Pets(Agent)
                spawn:
                    InitializeL15Pets(Agent)
                spawn:
                    InitializeL16Pets(Agent)
                spawn:
                    InitializeL17Pets(Agent)
                spawn:
                    InitializeL18Pets(Agent)
                spawn:
                    InitializeL19Pets(Agent)
                spawn:
                    InitializeL20Pets(Agent)
                spawn:
                    InitializeL21Pets(Agent)
                spawn:
                    InitializeL22Pets(Agent)
                spawn:
                    InitializeL23Pets(Agent)
                spawn:
                    InitializeL24Pets(Agent)
                spawn:
                    InitializeL25Pets(Agent)
                spawn:
                    InitializeL26Pets(Agent)
                spawn:
                    InitializeL27Pets(Agent)
                spawn:
                    InitializeL28Pets(Agent)
                spawn:
                    InitializeL29Pets(Agent)
                spawn:
                    InitializeL30Pets(Agent)
                spawn:
                    InitializeL31Pets(Agent)
                spawn:
                    InitializeL32Pets(Agent)
                spawn:
                    InitializeL33Pets(Agent)
                spawn:
                    InitializeL34Pets(Agent)

                #Initialize Barriers slowly
                spawn:
                    InitializeB1(Agent)
                spawn:
                    InitializeB2(Agent)
                spawn:
                    InitializeB3(Agent)
                spawn:
                    InitializeB4(Agent)
                spawn:
                    InitializeB5(Agent)
                spawn:
                    InitializeB6(Agent)
                spawn:
                    InitializeB7(Agent)
                spawn:
                    InitializeB8(Agent)
                spawn:
                    InitializeB9(Agent)
                spawn:
                    InitializeB10(Agent)
                spawn:
                    InitializeB11(Agent)
                spawn:
                    InitializeB12(Agent)
                spawn:
                    InitializeB13(Agent)
                spawn:
                    InitializeB14(Agent)
                spawn:
                    InitializeB15(Agent)
                spawn:
                    InitializeB16(Agent)
                spawn:
                    InitializeB17(Agent)
                spawn:
                    InitializeB18(Agent)
                spawn:
                    InitializeB19(Agent)
                spawn:
                    InitializeB20(Agent)
                spawn:
                    InitializeB21(Agent)
                spawn:
                    InitializeB22(Agent)
                spawn:
                    InitializeB23(Agent)
                spawn:
                    InitializeB24(Agent)
                spawn:
                    InitializeB25(Agent)
                spawn:
                    InitializeB26(Agent)
                spawn:
                    InitializeB27(Agent)
                spawn:
                    InitializeB28(Agent)
                spawn:
                    InitializeB29(Agent)
                spawn:
                    InitializeB31(Agent)
                spawn:
                    InitializeB32(Agent)
                spawn:
                    InitializeB33(Agent)

    OnPlayerAdded(Agent:agent):void=
        #Initialize SAVE
        Players := GetPlayspace().GetPlayers()
        PlayerStatsManager.InitializeAllPlayers(Players)
        #Initialize UI
        if:
            PlayerObj := player[Agent]
        then:
            if(PlayerExists := PlayersMap[PlayerObj]):
                #do nothing, player already exists
            else:
                if:
                    FC := Agent.GetFortCharacter[]
                then:
                    CustomPlayer:Custom_Player = Custom_Player{MyAgentObj := Agent}
                    option:
                        set PlayersMap[PlayerObj] = CustomPlayer
                    CustomPlayer.InitiateUI()
        #Initialize Agent Map for inventory
        if (not AgentMap[Agent]):
            AgentData := agent_data{}

            for (I := 1 .. ShopPurchases.Length):
                set AgentData.Purchases += array{ 0 }

            if (set AgentMap[Agent] = AgentData){}
        #Initialize Ownership UI
        InitializePetOwnerShipUI(Agent)
        #Initialize Crit Speed Luck Gold + Inventory
        InitializeStatUI(Agent)
        #Initialize Map code
        if(HUDDevice:=Rebirth0_HUD[GetTeamIndexAgent[Agent]-1]):
            HUDDevice.Show(Agent)
        #Manage Inventory
        spawn:
            ManagePotions(Agent)

    OpenPetDexEvent(Agent : agent) : void =
        if:
            PlayerObj := player[Agent]
            CustomPlayer:Custom_Player = PlayersMap[PlayerObj]
        then:
            set CustomPlayer.bShopOpen = true
            CustomPlayer.InitiateUI_PetDex()

    OnPlayerRemoved(Player:player):void=
        if:
            Agent := agent[Player]
            PlayerObj := player[Agent]
            CustomPlayer:Custom_Player = PlayersMap[PlayerObj]
        then:
            CustomPlayer.IncreaseButtonL1() #close all the loops of all pets active
            CustomPlayer.IncreaseButtonL2()
            CustomPlayer.IncreaseButtonL3()
            CustomPlayer.IncreaseButtonL4()
            CustomPlayer.IncreaseButtonL5()
            CustomPlayer.IncreaseButtonL6()
            CustomPlayer.IncreaseButtonL7()
            CustomPlayer.IncreaseButtonL8()
            CustomPlayer.IncreaseButtonL9()
            CustomPlayer.IncreaseButtonL10()
            CustomPlayer.IncreaseButtonL11()
            CustomPlayer.IncreaseButtonL12()
            CustomPlayer.IncreaseButtonL13()
            CustomPlayer.IncreaseButtonL14()
            CustomPlayer.IncreaseButtonL15()
            CustomPlayer.IncreaseButtonL16()
            CustomPlayer.IncreaseButtonL17()
            CustomPlayer.IncreaseButtonL18()
            CustomPlayer.IncreaseButtonL19()
            CustomPlayer.IncreaseButtonL20()
            CustomPlayer.IncreaseButtonL21()
            CustomPlayer.IncreaseButtonL22()
            CustomPlayer.IncreaseButtonL23()
            CustomPlayer.IncreaseButtonL24()
            CustomPlayer.IncreaseButtonL25()
            CustomPlayer.IncreaseButtonL26()
            CustomPlayer.IncreaseButtonL27()
            CustomPlayer.IncreaseButtonL28()
            CustomPlayer.IncreaseButtonL29()
            CustomPlayer.IncreaseButtonL30()

        if(RemoveCustomPlayer:Custom_Player = PlayersMap[Player]):
            #remove the player from the players
            var TempAllPlayersMap:[player]Custom_Player = map{}
            for (Key -> Value : PlayersMap, Key <> Player):
                set TempAllPlayersMap = ConcatenateMaps(TempAllPlayersMap, map{Key => Value})
            set PlayersMap = TempAllPlayersMap   

    GetTeamIndexAgent(Agent:agent)<decides><transacts>:int=
        TeamCollection := GetPlayspace().GetTeamCollection()
        AgentTeam := TeamCollection.GetTeam[Agent]
        TeamCollection.GetTeams().Find[AgentTeam] + 1

    CheckFunction(Agent:agent):void=
        if:
            PlayerObj := player[Agent]
            CurrentPlayerStats := PlayerStatsManager.GetPlayerStats[Agent]
            CustomPlayer:Custom_Player = PlayersMap[PlayerObj]
        then:
            #Print("Ownership_L1_Pet1: {CustomPlayer.Ownership_L1_Pet1}")
            #Print("Pet_Unlocked_L1: {CustomPlayer.Pet_Unlocked_L1}")
            #Print("L1_Pet2: {CustomPlayer.Ownership_L1_Pet2}")
            #Print("L1_Pet3: {CustomPlayer.Ownership_L1_Pet3}")
            #Print("L1_Pet4: {CustomPlayer.Ownership_L1_Pet4}")
            #Print("L1_Pet1: {CustomPlayer.Ownership_L1_Pet1}")
            #Print("L1_Pet2: {CustomPlayer.Ownership_L1_Pet2}")
            #Print("L1_Pet3: {CustomPlayer.Ownership_L1_Pet3}")
            #Print("L1_Pet4: {CustomPlayer.Ownership_L1_Pet4}")
            #Print("Max Level Unlocked:{CurrentPlayerStats.LevelUnlocked.CurrentValue}")
            #Print("Max Level Unlocked:{CurrentPlayerStats.RebirthUnlocked.CurrentValue}")
            #Print("Gold saved:{CurrentPlayerStats.GoldBalance.CurrentValue}")
            #Print("Crit saved:{CurrentPlayerStats.CritScore.CurrentValue}")
            #Print("Speed saved:{CurrentPlayerStats.SpeedScore.CurrentValue}")
            #Print("Luck saved:{CurrentPlayerStats.LuckScore.CurrentValue}")PlayerGold
            Print("Luck 1 owned:{CustomPlayer.Potion_Owned_Luck1}")
            Print("Luck 2 owned:{CustomPlayer.Potion_Owned_Luck2}")
            Print("Luck 3 owned:{CustomPlayer.Potion_Owned_Luck3}")
            Print("Luck 1 remaining seconds:{CustomPlayer.Potion_Active_Luck1_Sec}")
            Print("Luck 2 remaining seconds:{CustomPlayer.Potion_Active_Luck2_Sec}")
            Print("Luck 3 remaining seconds:{CustomPlayer.Potion_Active_Luck3_Sec}")
            Print("Gold balance:{CustomPlayer.PlayerGold}")
    
    OnPlayerHitDamage(Agent : agent) : void =
        Allthechests:= ChestsL1+ChestsL2+ChestsL3+ChestsL4+ChestsL5+ChestsL6+ChestsL7+ChestsL8+ChestsL9+ChestsL10+ChestsL11+ChestsL12+ChestsL13+ChestsL14+ChestsL15+ChestsL16+ChestsL17+ChestsL18+ChestsL19+ChestsL20+ChestsL21+ChestsL22+ChestsL23+ChestsL24+ChestsL25+ChestsL26+ChestsL27+ChestsL28+ChestsL29+ChestsL30 #peut etre a ameliorer
        Print("Player Hit a Chest")
        if (FortniteCharacter := Agent.GetFortCharacter[]):
            var ClosestChest:creative_prop = creative_prop{}
            var DistanceToMe:float = 100000000.0
            var FoundChest:logic = false
            var ChestIndexTemp:int = -1
            var ChestIndexFinal:int = 0
            var ClosestChestIndex:int=0
            for(Chestt:Allthechests):
                set ChestIndexTemp+=1
                MyPos:vector3 = FortniteCharacter.GetTransform().Translation
                ChestPos:vector3 = Chestt.ChestProp.GetTransform().Translation
                TotalDistance:float = Distance(MyPos, ChestPos) / 100.0
                if(TotalDistance < DistanceToMe and Chestt.CurrentHP>0.0):
                    set ChestIndexFinal = ChestIndexTemp
                    #Print("Chest index{ChestIndexFinal}")
                    set FoundChest = true
                    set DistanceToMe = TotalDistance
                    set ClosestChest = Chestt.ChestProp
            if:
                PlayerObj := player[Agent]
            then:
                if(CustomPlayer:Custom_Player = PlayersMap[PlayerObj]):
                    Attack_Chest(ChestIndexFinal,CustomPlayer.Power_Player)
            Check_Chest_Gold(ChestIndexFinal,FortniteCharacter)

    Input0( Agent : agent ): void =
        if:
            PlayerObj := player[Agent]
            CustomPlayer:Custom_Player = PlayersMap[PlayerObj]
        then:
            set CustomPlayer.InputCode0+=1
            Print("InputCode0:{CustomPlayer.InputCode0}")

    Input1( Agent : agent ): void =
        if:
            PlayerObj := player[Agent]
            CustomPlayer:Custom_Player = PlayersMap[PlayerObj]
        then:
            set CustomPlayer.InputCode1+=1
            Print("InputCode1:{CustomPlayer.InputCode1}")

    Input2( Agent : agent ): void =
        if:
            PlayerObj := player[Agent]
            CustomPlayer:Custom_Player = PlayersMap[PlayerObj]
        then:
            set CustomPlayer.InputCode2+=1
            Print("InputCode2:{CustomPlayer.InputCode2}")

    Input3( Agent : agent ): void =
        if:
            PlayerObj := player[Agent]
            CustomPlayer:Custom_Player = PlayersMap[PlayerObj]
        then:
            set CustomPlayer.InputCode3+=1
            Print("InputCode3:{CustomPlayer.InputCode3}")

    Input4( Agent : agent ): void =
        if:
            PlayerObj := player[Agent]
            CustomPlayer:Custom_Player = PlayersMap[PlayerObj]
        then:
            set CustomPlayer.InputCode4+=1
            Print("InputCode4:{CustomPlayer.InputCode4}")

    Input5( Agent : agent ): void =
        if:
            PlayerObj := player[Agent]
            CustomPlayer:Custom_Player = PlayersMap[PlayerObj]
        then:
            set CustomPlayer.InputCode5+=1
            Print("InputCode5:{CustomPlayer.InputCode5}")

    Input6( Agent : agent ): void =
        if:
            PlayerObj := player[Agent]
            CustomPlayer:Custom_Player = PlayersMap[PlayerObj]
        then:
            set CustomPlayer.InputCode6+=1
            Print("InputCode6:{CustomPlayer.InputCode6}")

    Input7( Agent : agent ): void =
        if:
            PlayerObj := player[Agent]
            CustomPlayer:Custom_Player = PlayersMap[PlayerObj]
        then:
            set CustomPlayer.InputCode7+=1
            Print("InputCode7:{CustomPlayer.InputCode7}")

    Input8( Agent : agent ): void =
        if:
            PlayerObj := player[Agent]
            CustomPlayer:Custom_Player = PlayersMap[PlayerObj]
        then:
            set CustomPlayer.InputCode8+=1
            Print("InputCode8:{CustomPlayer.InputCode8}")

    Input9( Agent : agent ): void =
        if:
            PlayerObj := player[Agent]
            CustomPlayer:Custom_Player = PlayersMap[PlayerObj]
        then:
            set CustomPlayer.InputCode9+=1
            Print("InputCode9:{CustomPlayer.InputCode9}")

    InputEnter( Agent : agent ): void =
        if:
            PlayerObj := player[Agent]
            CustomPlayer:Custom_Player = PlayersMap[PlayerObj]
        then:
            Print("InputCode0:{CustomPlayer.InputCode0}")
            Print("InputCode1:{CustomPlayer.InputCode1}")
            Print("InputCode2:{CustomPlayer.InputCode2}")
            Print("InputCode3:{CustomPlayer.InputCode3}")
            Print("InputCode4:{CustomPlayer.InputCode4}")
            Print("InputCode5:{CustomPlayer.InputCode5}")
            Print("InputCode6:{CustomPlayer.InputCode6}")
            Print("InputCode7:{CustomPlayer.InputCode7}")
            Print("InputCode8:{CustomPlayer.InputCode8}")
            Print("InputCode9:{CustomPlayer.InputCode9}")
            #Code1 2000 gold + 2 potions level 1 of each
            if:
                CustomPlayer.InputCode0=1
                CustomPlayer.InputCode1=2
                CustomPlayer.InputCode2=0
                CustomPlayer.InputCode3=1
                CustomPlayer.InputCode4=0
                CustomPlayer.InputCode5=0
                CustomPlayer.InputCode6=1
                CustomPlayer.InputCode7=3
                CustomPlayer.InputCode8=0
                CustomPlayer.InputCode9=2
                CustomPlayer.Code1Opened=0
            then:
                Print("Bonus 1 Activated")
                AddPotionLuck1(Agent,2)
                set CustomPlayer.Potion_Owned_Luck1 +=2
                AddPotionGold1(Agent,2)
                set CustomPlayer.Potion_Owned_Gold1 += 2
                AddPotionPetPower1(Agent,2)
                set CustomPlayer.Potion_Owned_Pet_Power1 += 2
                AddPotionPlayerPower1(Agent,2)
                set CustomPlayer.Potion_Owned_Player_Power1 += 2
                CustomPlayer.IncreaseGoldBy(2000)
                AddGoldBalance(Agent,2000)
                SetRecordCode1(Agent,1)
                set CustomPlayer.Code1Opened=1
                Code_ValidHUD.Show(Agent)
            #Code2 6000 gold + 3 potions level 2 of each
            if:
                CustomPlayer.InputCode0=1
                CustomPlayer.InputCode1=2
                CustomPlayer.InputCode2=0
                CustomPlayer.InputCode3=3
                CustomPlayer.InputCode4=0
                CustomPlayer.InputCode5=1
                CustomPlayer.InputCode6=0
                CustomPlayer.InputCode7=3
                CustomPlayer.InputCode8=0
                CustomPlayer.InputCode9=0
                CustomPlayer.Code2Opened=0
            then:
                Print("Bonus 2 Activated")
                AddPotionLuck2(Agent,3)
                set CustomPlayer.Potion_Owned_Luck2 +=3
                AddPotionGold2(Agent,3)
                set CustomPlayer.Potion_Owned_Gold2 += 3
                AddPotionPetPower2(Agent,3)
                set CustomPlayer.Potion_Owned_Pet_Power2 += 3
                AddPotionPlayerPower2(Agent,3)
                set CustomPlayer.Potion_Owned_Player_Power2 += 3
                CustomPlayer.IncreaseGoldBy(6000)
                AddGoldBalance(Agent,6000)
                SetRecordCode2(Agent,1)
                set CustomPlayer.Code2Opened=1
                Code_ValidHUD.Show(Agent)
            #Code3 6000 gold + 3 potions level 2 of each + 5% gold buff
            if:
                CustomPlayer.InputCode0=2
                CustomPlayer.InputCode1=1
                CustomPlayer.InputCode2=0
                CustomPlayer.InputCode3=0
                CustomPlayer.InputCode4=0
                CustomPlayer.InputCode5=2
                CustomPlayer.InputCode6=0
                CustomPlayer.InputCode7=0
                CustomPlayer.InputCode8=4
                CustomPlayer.InputCode9=1
                CustomPlayer.Code3Opened=0
            then:
                Print("Bonus 3 Activated")
                AddPotionLuck2(Agent,3)
                set CustomPlayer.Potion_Owned_Luck2 +=3
                AddPotionGold2(Agent,3)
                set CustomPlayer.Potion_Owned_Gold2 += 3
                AddPotionPetPower2(Agent,3)
                set CustomPlayer.Potion_Owned_Pet_Power2 += 3
                AddPotionPlayerPower2(Agent,3)
                set CustomPlayer.Potion_Owned_Player_Power2 += 3
                CustomPlayer.IncreaseGoldBy(6000)
                AddGoldBalance(Agent,6000)
                set CustomPlayer.Bonus_Gold_Perma=1.05
                SetRecordCode3(Agent,1)
                set CustomPlayer.Code3Opened=1
                Code_ValidHUD.Show(Agent)
            #Code4 Rebirth 5, golden cat + 1 potions lvl 3 of each
            if:
                CustomPlayer.InputCode0=0
                CustomPlayer.InputCode1=2
                CustomPlayer.InputCode2=1
                CustomPlayer.InputCode3=1
                CustomPlayer.InputCode4=1
                CustomPlayer.InputCode5=2
                CustomPlayer.InputCode6=0
                CustomPlayer.InputCode7=3
                CustomPlayer.InputCode8=0
                CustomPlayer.InputCode9=0
                CustomPlayer.Code4Opened=0
            then:
                Print("Bonus 4 Activated")
                AddPotionLuck3(Agent,1)
                set CustomPlayer.Potion_Owned_Luck3 +=1
                AddPotionGold3(Agent,1)
                set CustomPlayer.Potion_Owned_Gold3 += 1
                AddPotionPetPower3(Agent,1)
                set CustomPlayer.Potion_Owned_Pet_Power3 += 1
                AddPotionPlayerPower3(Agent,1)
                set CustomPlayer.Potion_Owned_Player_Power3 += 1
                PurchaseL34(Agent)
                SetRecordCode4(Agent,1)
                set CustomPlayer.Code4Opened=1
                Code_ValidHUD.Show(Agent)

        InputReset(Agent)

    InputReset( Agent : agent ): void =
        if:
            PlayerObj := player[Agent]
            CustomPlayer:Custom_Player = PlayersMap[PlayerObj]
        then:
            set CustomPlayer.InputCode0=0
            set CustomPlayer.InputCode1=0
            set CustomPlayer.InputCode2=0
            set CustomPlayer.InputCode3=0
            set CustomPlayer.InputCode4=0
            set CustomPlayer.InputCode5=0
            set CustomPlayer.InputCode6=0
            set CustomPlayer.InputCode7=0
            set CustomPlayer.InputCode8=0
            set CustomPlayer.InputCode9=0
    
    PurchaseL1( Agent : agent ): void =
        spawn:
            PurchasePetL1(Agent)

    PurchaseL2( Agent : agent ): void =
        spawn:
            PurchasePetL2(Agent)

    PurchaseL3( Agent : agent ): void =
        spawn:
            PurchasePetL3(Agent)

    PurchaseL4( Agent : agent ): void =
        spawn:
            PurchasePetL4(Agent)

    PurchaseL5( Agent : agent ): void =
        spawn:
            PurchasePetL5(Agent)

    PurchaseL6( Agent : agent ): void =
        spawn:
            PurchasePetL6(Agent)

    PurchaseL7( Agent : agent ): void =
        spawn:
            PurchasePetL7(Agent)

    PurchaseL8( Agent : agent ): void =
        spawn:
            PurchasePetL8(Agent)

    PurchaseL9( Agent : agent ): void =
        spawn:
            PurchasePetL9(Agent)

    PurchaseL10( Agent : agent ): void =
        spawn:
            PurchasePetL10(Agent)

    PurchaseL11( Agent : agent ): void =
        spawn:
            PurchasePetL11(Agent)

    PurchaseL12( Agent : agent ): void =
        spawn:
            PurchasePetL12(Agent)

    PurchaseL13( Agent : agent ): void =
        spawn:
            PurchasePetL13(Agent)

    PurchaseL14( Agent : agent ): void =
        spawn:
            PurchasePetL14(Agent)

    PurchaseL15( Agent : agent ): void =
        spawn:
            PurchasePetL15(Agent)

    PurchaseL16( Agent : agent ): void =
        spawn:
            PurchasePetL16(Agent)

    PurchaseL17( Agent : agent ): void =
        spawn:
            PurchasePetL17(Agent)

    PurchaseL18( Agent : agent ): void =
        spawn:
            PurchasePetL18(Agent)

    PurchaseL19( Agent : agent ): void =
        spawn:
            PurchasePetL19(Agent)

    PurchaseL20( Agent : agent ): void =
        spawn:
            PurchasePetL20(Agent)

    PurchaseL21( Agent : agent ): void =
        spawn:
            PurchasePetL21(Agent)
    
    PurchaseL22( Agent : agent ): void =
        spawn:
            PurchasePetL22(Agent)

    PurchaseL23( Agent : agent ): void =
        spawn:
            PurchasePetL23(Agent)

    PurchaseL24( Agent : agent ): void =
        spawn:
            PurchasePetL24(Agent)

    PurchaseL25( Agent : agent ): void =
        spawn:
            PurchasePetL25(Agent)

    PurchaseL26( Agent : agent ): void =
        spawn:
            PurchasePetL26(Agent)

    PurchaseL27( Agent : agent ): void =
        spawn:
            PurchasePetL27(Agent)

    PurchaseL28( Agent : agent ): void =
        spawn:
            PurchasePetL28(Agent)

    PurchaseL29( Agent : agent ): void =
        spawn:
            PurchasePetL29(Agent)

    PurchaseL30( Agent : agent ): void =
        spawn:
            PurchasePetL30(Agent)

    #Rebirth Dog 1
    PurchaseL31( Agent : agent ): void =
        spawn:
            PurchasePetL31(Agent)

    #Rebirth Dog 2
    PurchaseL32( Agent : agent ): void =
        spawn:
            PurchasePetL32(Agent)

    #Rebirth Dog 3
    PurchaseL33( Agent : agent ): void =
        spawn:
            PurchasePetL33(Agent)

    #Golden Cat
    PurchaseL34( Agent : agent ): void =
        spawn:
            PurchasePetL34(Agent)

    PurchasePetL1( Agent : agent )<suspends>: void =
        #Requires Unlock L1 (L1 is already unlocked)
        if:
            PlayerObj := player[Agent]
            CurrentPlayerStats := PlayerStatsManager.GetPlayerStats[Agent]
            CurrentPlayerStats.LevelUnlocked.CurrentValue<0
        then:
            RequireHigherLevelHUD.Show(Agent) 
        else:
            if:
                PlayerObj := player[Agent]
                CustomPlayer:Custom_Player = PlayersMap[PlayerObj]
                CustomPlayer.PlayerGold<CostL1
            then:
                NotEnoughGoldHUD.Show(Agent)
            else:
                DecreaseGoldFunc(Agent,CostL1)
                var FirstPet:creative_prop = creative_prop{}
                var SecondPet:creative_prop = creative_prop{}
                var ThirdPet:creative_prop = creative_prop{}
                var ForthPet:creative_prop = creative_prop{}

                var FirstPetIndex:int =  0
                var SecondPetIndex:int = 0
                var ThirdPetIndex:int =  0
                var ForthPetIndex:int =  0

                var IndexBase:int = 0

                #Select right pets for each player
                if(GetTeamIndexAgent[Agent]=1):
                    set FirstPetIndex =  IndexBase
                    set SecondPetIndex = IndexBase+1
                    set ThirdPetIndex =  IndexBase+2
                    set ForthPetIndex =  IndexBase+3
                if(GetTeamIndexAgent[Agent]=2):
                    set FirstPetIndex =  IndexBase+4
                    set SecondPetIndex = IndexBase+5
                    set ThirdPetIndex =  IndexBase+6
                    set ForthPetIndex =  IndexBase+7
                if(GetTeamIndexAgent[Agent]=3):
                    set FirstPetIndex =  IndexBase+8
                    set SecondPetIndex = IndexBase+9
                    set ThirdPetIndex =  IndexBase+10
                    set ForthPetIndex =  IndexBase+11
                if (Mypet:=Pets[FirstPetIndex]):
                    set FirstPet = Mypet.PetProp
                if (Mypet:=Pets[SecondPetIndex]):
                    set SecondPet = Mypet.PetProp
                if (Mypet:=Pets[ThirdPetIndex]):
                    set ThirdPet = Mypet.PetProp
                if (Mypet:=Pets[ForthPetIndex]):
                    set ForthPet = Mypet.PetProp

                var PetWon:creative_prop = creative_prop{}
                var RandResult:int = GetRandomInt(0,1000)
                Print("Rand: {RandResult}")
                if:
                    PlayerObj := player[Agent]
                    Fort := Agent.GetFortCharacter[]
                then:
                    if(CustomPlayer:Custom_Player = PlayersMap[PlayerObj]):
                        #--------------Luck
                        set RandResult+=CustomPlayer.Pet_Luck_Value*5+CustomPlayer.Bonus_Luck*5
                        #--------------
                        if (RandResult>995):
                            Print("Legendary")
                            if(Cine:=Cine_L1_Legendary[GetTeamIndexAgent[Agent]-1]):
                                Cine.Play(Agent)
                            Sleep(4.0)
                            if(Cine:=Cine_L1_Legendary[GetTeamIndexAgent[Agent]-1]):
                                Cine.Stop(Agent)
                                Accolade_Egg1_2.Award(Agent)


                            if (CustomPlayer.Ownership_L1_Pet4=0):
                                #Release potential previous pets
                                if (CustomPlayer.Ownership_L1_Pet1=1):
                                    #--------------Button pressed:
                                    CustomPlayer.IncreaseButtonL1()
                                    #Print("Button L1 Pressed: {CustomPlayer.ButtonL1Pressed}")
                                    #--------------Spawn
                                    Sleep(0.5)
                                    ##spawn:
                                    ##    ReleasePet(FirstPet)
                                    set CustomPlayer.Ownership_L1_Pet1=0
                                    L1_Pet1_DesOwn(Agent)
                                if (CustomPlayer.Ownership_L1_Pet2=1):
                                    #--------------Button pressed:
                                    CustomPlayer.IncreaseButtonL1()
                                    #Print("Button L1 Pressed: {CustomPlayer.ButtonL1Pressed}")
                                    #--------------Spawn
                                    Sleep(0.5)
                                    ##spawn:
                                    ##    ReleasePet(SecondPet)
                                    set CustomPlayer.Ownership_L1_Pet2=0
                                    L1_Pet2_DesOwn(Agent)
                                if (CustomPlayer.Ownership_L1_Pet3=1):
                                    #--------------Button pressed:
                                    CustomPlayer.IncreaseButtonL1()
                                    #Print("Button L1 Pressed: {CustomPlayer.ButtonL1Pressed}")
                                    #--------------Spawn
                                    Sleep(0.5)
                                    ##spawn:
                                    ##    ReleasePet(ThirdPet)
                                    set CustomPlayer.Ownership_L1_Pet3=0
                                    L1_Pet3_DesOwn(Agent)
                                #Get New pet
                                L1_Pet4_Own(Agent)
                                set CustomPlayer.Ownership_L1_Pet4=1
                                set CustomPlayer.Pet_Unlocked_L1=4
                                CustomPlayer.Update_Pet_PowerL1(Power_L1P4)
                                var PetPower:float=1.0*Power_L1P4
                                set PetWon = ForthPet
                                spawn:
                                    FollowPlayerL1(Fort, PetWon, PetPower, ForthPetIndex)
                        else if (RandResult>855):
                            Print("Epic")
                            if(Cine:=Cine_L1_Epic[GetTeamIndexAgent[Agent]-1]):
                                Cine.Play(Agent)
                            Sleep(4.0)
                            if(Cine:=Cine_L1_Epic[GetTeamIndexAgent[Agent]-1]):
                                Cine.Stop(Agent)
                                Accolade_Egg1_2.Award(Agent)

                            if (CustomPlayer.Ownership_L1_Pet3=0 and CustomPlayer.Ownership_L1_Pet4=0):
                                #Release potential previous pets
                                if (CustomPlayer.Ownership_L1_Pet1=1):
                                    #--------------Button pressed:
                                    CustomPlayer.IncreaseButtonL1()
                                    #Print("Button L1 Pressed: {CustomPlayer.ButtonL1Pressed}")
                                    #--------------Spawn
                                    Sleep(0.5)
                                    ##spawn:
                                    ##    ReleasePet(FirstPet)
                                    set CustomPlayer.Ownership_L1_Pet1=0
                                    L1_Pet1_DesOwn(Agent)
                                if (CustomPlayer.Ownership_L1_Pet2=1):
                                    #--------------Button pressed:
                                    CustomPlayer.IncreaseButtonL1()
                                    #Print("Button L1 Pressed: {CustomPlayer.ButtonL1Pressed}")
                                    #--------------Spawn
                                    Sleep(0.5)
                                    ##spawn:
                                    ##    ReleasePet(SecondPet)
                                    set CustomPlayer.Ownership_L1_Pet2=0
                                    L1_Pet2_DesOwn(Agent)
                                #Get New pet
                                L1_Pet3_Own(Agent)
                                set CustomPlayer.Ownership_L1_Pet3=1
                                set CustomPlayer.Pet_Unlocked_L1=3
                                CustomPlayer.Update_Pet_PowerL1(Power_L1P3)
                                var PetPower:float=1.0*Power_L1P3
                                set PetWon = ThirdPet
                                spawn:
                                    FollowPlayerL1(Fort, PetWon, PetPower, ThirdPetIndex)
                        else if (RandResult>750):
                            Print("Rare")
                            if(Cine:=Cine_L1_Rare[GetTeamIndexAgent[Agent]-1]):
                                Cine.Play(Agent)
                            Sleep(4.0)
                            if(Cine:=Cine_L1_Rare[GetTeamIndexAgent[Agent]-1]):
                                Cine.Stop(Agent)
                                Accolade_Egg1_2.Award(Agent)

                            if (CustomPlayer.Ownership_L1_Pet2=0 and CustomPlayer.Ownership_L1_Pet3=0 and CustomPlayer.Ownership_L1_Pet4=0):
                                #Release potential previous pets
                                if (CustomPlayer.Ownership_L1_Pet1=1):
                                    #--------------Button pressed:
                                    CustomPlayer.IncreaseButtonL1()
                                    #Print("Button L1 Pressed: {CustomPlayer.ButtonL1Pressed}")
                                    #--------------Spawn
                                    Sleep(0.5)
                                    ##spawn:
                                    ##    ReleasePet(FirstPet)
                                    set CustomPlayer.Ownership_L1_Pet1=0
                                    L1_Pet1_DesOwn(Agent)
                                #Get New pet
                                L1_Pet2_Own(Agent)
                                set CustomPlayer.Ownership_L1_Pet2=1
                                set CustomPlayer.Pet_Unlocked_L1=2
                                CustomPlayer.Update_Pet_PowerL1(Power_L1P2)
                                var PetPower:float=1.0*Power_L1P2
                                set PetWon = SecondPet
                                spawn:
                                    FollowPlayerL1(Fort, PetWon, PetPower, SecondPetIndex)
                        else:
                            Print("Normal")
                            if(Cine:=Cine_L1_Normal[GetTeamIndexAgent[Agent]-1]):
                                Cine.Play(Agent)
                            Sleep(4.0)
                            if(Cine:=Cine_L1_Normal[GetTeamIndexAgent[Agent]-1]):
                                Cine.Stop(Agent)
                                Accolade_Egg1_2.Award(Agent)

                            if (CustomPlayer.Ownership_L1_Pet1=0 and CustomPlayer.Ownership_L1_Pet2=0 and CustomPlayer.Ownership_L1_Pet3=0 and CustomPlayer.Ownership_L1_Pet4=0):
                                #Get New pet
                                L1_Pet1_Own(Agent)
                                set CustomPlayer.Ownership_L1_Pet1=1
                                set CustomPlayer.Pet_Unlocked_L1=1
                                CustomPlayer.Update_Pet_PowerL1(Power_L1P1)
                                var PetPower:float=1.0*Power_L1P1
                                set PetWon = FirstPet
                                spawn:
                                    FollowPlayerL1(Fort, PetWon, PetPower, FirstPetIndex)

    PurchasePetL2( Agent : agent )<suspends>: void =
        #Requires Unlock L2
        if:
            PlayerObj := player[Agent]
            CurrentPlayerStats := PlayerStatsManager.GetPlayerStats[Agent]
            CurrentPlayerStats.LevelUnlocked.CurrentValue<1
        then:
            RequireHigherLevelHUD.Show(Agent) 
        else:
            if:
                PlayerObj := player[Agent]
                CustomPlayer:Custom_Player = PlayersMap[PlayerObj]
                CustomPlayer.PlayerGold<CostL2
            then:
                NotEnoughGoldHUD.Show(Agent)
            else:
                DecreaseGoldFunc(Agent,CostL2)
                var FirstPet:creative_prop = creative_prop{}
                var SecondPet:creative_prop = creative_prop{}
                var ThirdPet:creative_prop = creative_prop{}
                var ForthPet:creative_prop = creative_prop{}

                var FirstPetIndex:int = 0
                var SecondPetIndex:int = 0
                var ThirdPetIndex:int = 0
                var ForthPetIndex:int = 0

                var IndexBase:int = 12

                #Select right pets for each player
                if(GetTeamIndexAgent[Agent]=1):
                    set FirstPetIndex =  IndexBase
                    set SecondPetIndex = IndexBase+1
                    set ThirdPetIndex =  IndexBase+2
                    set ForthPetIndex =  IndexBase+3
                if(GetTeamIndexAgent[Agent]=2):
                    set FirstPetIndex =  IndexBase+4
                    set SecondPetIndex = IndexBase+5
                    set ThirdPetIndex =  IndexBase+6
                    set ForthPetIndex =  IndexBase+7
                if(GetTeamIndexAgent[Agent]=3):
                    set FirstPetIndex =  IndexBase+8
                    set SecondPetIndex = IndexBase+9
                    set ThirdPetIndex =  IndexBase+10
                    set ForthPetIndex =  IndexBase+11
                if (Mypet:=Pets[FirstPetIndex]):
                    set FirstPet = Mypet.PetProp
                if (Mypet:=Pets[SecondPetIndex]):
                    set SecondPet = Mypet.PetProp
                if (Mypet:=Pets[ThirdPetIndex]):
                    set ThirdPet = Mypet.PetProp
                if (Mypet:=Pets[ForthPetIndex]):
                    set ForthPet = Mypet.PetProp

                var PetWon:creative_prop = creative_prop{}
                var RandResult:int = GetRandomInt(0,1000)
                Print("Rand: {RandResult}")
                if:
                    PlayerObj := player[Agent]
                    Fort := Agent.GetFortCharacter[]
                then:
                    if(CustomPlayer:Custom_Player = PlayersMap[PlayerObj]):
                        #--------------Luck
                        set RandResult+=CustomPlayer.Pet_Luck_Value*5+CustomPlayer.Bonus_Luck*5
                        #--------------
                        if (RandResult>995):
                            Print("Legendary")
                            if(Cine:=Cine_L2_Legendary[GetTeamIndexAgent[Agent]-1]):
                                Cine.Play(Agent)
                            Sleep(4.0)
                            if(Cine:=Cine_L2_Legendary[GetTeamIndexAgent[Agent]-1]):
                                Cine.Stop(Agent)
                                Accolade_Egg1_2.Award(Agent)

                            if (CustomPlayer.Ownership_L2_Pet4=0):
                                #Release potential previous pets
                                if (CustomPlayer.Ownership_L2_Pet1=1):
                                    #--------------Button pressed:
                                    CustomPlayer.IncreaseButtonL2()
                                    #Print("Button L2 Pressed: {CustomPlayer.ButtonL2Pressed}")
                                    #--------------Spawn
                                    Sleep(0.5)
                                    ##spawn:
                                    ##    ReleasePet(FirstPet)
                                    set CustomPlayer.Ownership_L2_Pet1=0
                                    L2_Pet1_DesOwn(Agent)
                                if (CustomPlayer.Ownership_L2_Pet2=1):
                                    #--------------Button pressed:
                                    CustomPlayer.IncreaseButtonL2()
                                    #Print("Button L2 Pressed: {CustomPlayer.ButtonL2Pressed}")
                                    #--------------Spawn
                                    Sleep(0.5)
                                    ##spawn:
                                    ##    ReleasePet(SecondPet)
                                    set CustomPlayer.Ownership_L2_Pet2=0
                                    L2_Pet2_DesOwn(Agent)
                                if (CustomPlayer.Ownership_L2_Pet3=1):
                                    #--------------Button pressed:
                                    CustomPlayer.IncreaseButtonL2()
                                    #Print("Button L2 Pressed: {CustomPlayer.ButtonL2Pressed}")
                                    #--------------Spawn
                                    Sleep(0.5)
                                    ##spawn:
                                    ##    ReleasePet(ThirdPet)
                                    set CustomPlayer.Ownership_L2_Pet3=0
                                    L2_Pet3_DesOwn(Agent)
                                #Get New pet
                                L2_Pet4_Own(Agent)
                                set CustomPlayer.Ownership_L2_Pet4=1
                                set CustomPlayer.Pet_Unlocked_L2=4
                                CustomPlayer.Update_Pet_PowerL2(Power_L2P4)
                                var PetPower:float=1.0*Power_L2P4
                                set PetWon = ForthPet
                                spawn:
                                    FollowPlayerL2(Fort, PetWon, PetPower, ForthPetIndex)
                        else if (RandResult>855):
                            Print("Epic")
                            if(Cine:=Cine_L2_Epic[GetTeamIndexAgent[Agent]-1]):
                                Cine.Play(Agent)
                            Sleep(4.0)
                            if(Cine:=Cine_L2_Epic[GetTeamIndexAgent[Agent]-1]):
                                Cine.Stop(Agent)
                                Accolade_Egg1_2.Award(Agent)

                            if (CustomPlayer.Ownership_L2_Pet3=0 and CustomPlayer.Ownership_L2_Pet4=0):
                                #Release potential previous pets
                                if (CustomPlayer.Ownership_L2_Pet1=1):
                                    #--------------Button pressed:
                                    CustomPlayer.IncreaseButtonL2()
                                    #Print("Button L2 Pressed: {CustomPlayer.ButtonL2Pressed}")
                                    #--------------Spawn
                                    Sleep(0.5)
                                    ##spawn:
                                    ##    ReleasePet(FirstPet)
                                    set CustomPlayer.Ownership_L2_Pet1=0
                                    L2_Pet1_DesOwn(Agent)
                                if (CustomPlayer.Ownership_L2_Pet2=1):
                                    #--------------Button pressed:
                                    CustomPlayer.IncreaseButtonL2()
                                    #Print("Button L2 Pressed: {CustomPlayer.ButtonL2Pressed}")
                                    #--------------Spawn
                                    Sleep(0.5)
                                    ##spawn:
                                    ##    ReleasePet(SecondPet)
                                    set CustomPlayer.Ownership_L2_Pet2=0
                                    L2_Pet2_DesOwn(Agent)
                                #Get New pet
                                L2_Pet3_Own(Agent)
                                set CustomPlayer.Ownership_L2_Pet3=1
                                set CustomPlayer.Pet_Unlocked_L2=3
                                CustomPlayer.Update_Pet_PowerL2(Power_L2P3)
                                var PetPower:float=1.0*Power_L2P3
                                set PetWon = ThirdPet
                                spawn:
                                    FollowPlayerL2(Fort, PetWon, PetPower, ThirdPetIndex)
                        else if (RandResult>750):
                            Print("Rare")
                            if(Cine:=Cine_L2_Rare[GetTeamIndexAgent[Agent]-1]):
                                Cine.Play(Agent)
                            Sleep(4.0)
                            if(Cine:=Cine_L2_Rare[GetTeamIndexAgent[Agent]-1]):
                                Cine.Stop(Agent)
                                Accolade_Egg1_2.Award(Agent)

                            if (CustomPlayer.Ownership_L2_Pet2=0 and CustomPlayer.Ownership_L2_Pet3=0 and CustomPlayer.Ownership_L2_Pet4=0):
                                #Release potential previous pets
                                if (CustomPlayer.Ownership_L2_Pet1=1):
                                    #--------------Button pressed:
                                    CustomPlayer.IncreaseButtonL2()
                                    #Print("Button L2 Pressed: {CustomPlayer.ButtonL2Pressed}")
                                    #--------------Spawn
                                    Sleep(0.5)
                                    ##spawn:
                                    ##    ReleasePet(FirstPet)
                                    set CustomPlayer.Ownership_L2_Pet1=0
                                    L2_Pet1_DesOwn(Agent)
                                #Get New pet
                                L2_Pet2_Own(Agent)
                                set CustomPlayer.Ownership_L2_Pet2=1
                                set CustomPlayer.Pet_Unlocked_L2=2
                                CustomPlayer.Update_Pet_PowerL2(Power_L2P2)
                                var PetPower:float=1.0*Power_L2P2
                                set PetWon = SecondPet
                                spawn:
                                    FollowPlayerL2(Fort, PetWon, PetPower, SecondPetIndex)
                        else:
                            Print("Normal")
                            if(Cine:=Cine_L2_Normal[GetTeamIndexAgent[Agent]-1]):
                                Cine.Play(Agent)
                            Sleep(4.0)
                            if(Cine:=Cine_L2_Normal[GetTeamIndexAgent[Agent]-1]):
                                Cine.Stop(Agent)
                                Accolade_Egg1_2.Award(Agent)

                            if (CustomPlayer.Ownership_L2_Pet1=0 and CustomPlayer.Ownership_L2_Pet2=0 and CustomPlayer.Ownership_L2_Pet3=0 and CustomPlayer.Ownership_L2_Pet4=0):
                                #Get New pet
                                L2_Pet1_Own(Agent)
                                set CustomPlayer.Ownership_L2_Pet1=1
                                set CustomPlayer.Pet_Unlocked_L2=1
                                CustomPlayer.Update_Pet_PowerL2(Power_L2P1)
                                var PetPower:float=1.0*Power_L2P1
                                set PetWon = FirstPet
                                spawn:
                                    FollowPlayerL2(Fort, PetWon, PetPower, FirstPetIndex)
                            
    PurchasePetL3( Agent : agent )<suspends>: void =
        #Requires Unlock L3
        if:
            PlayerObj := player[Agent]
            CurrentPlayerStats := PlayerStatsManager.GetPlayerStats[Agent]
            CurrentPlayerStats.LevelUnlocked.CurrentValue<2
        then:
            RequireHigherLevelHUD.Show(Agent) 
        else:
            if:
                PlayerObj := player[Agent]
                CustomPlayer:Custom_Player = PlayersMap[PlayerObj]
                CustomPlayer.PlayerGold<CostL3
            then:
                NotEnoughGoldHUD.Show(Agent)
            else:
                DecreaseGoldFunc(Agent,CostL3)
                var FirstPet:creative_prop = creative_prop{}
                var SecondPet:creative_prop = creative_prop{}
                var ThirdPet:creative_prop = creative_prop{}
                var ForthPet:creative_prop = creative_prop{}

                var FirstPetIndex:int = 0
                var SecondPetIndex:int = 0
                var ThirdPetIndex:int = 0
                var ForthPetIndex:int = 0

                var IndexBase:int = 12*2

                #Select right pets for each player
                if(GetTeamIndexAgent[Agent]=1):
                    set FirstPetIndex =  IndexBase
                    set SecondPetIndex = IndexBase+1
                    set ThirdPetIndex =  IndexBase+2
                    set ForthPetIndex =  IndexBase+3
                if(GetTeamIndexAgent[Agent]=2):
                    set FirstPetIndex =  IndexBase+4
                    set SecondPetIndex = IndexBase+5
                    set ThirdPetIndex =  IndexBase+6
                    set ForthPetIndex =  IndexBase+7
                if(GetTeamIndexAgent[Agent]=3):
                    set FirstPetIndex =  IndexBase+8
                    set SecondPetIndex = IndexBase+9
                    set ThirdPetIndex =  IndexBase+10
                    set ForthPetIndex =  IndexBase+11
                if (Mypet:=Pets[FirstPetIndex]):
                    set FirstPet = Mypet.PetProp
                if (Mypet:=Pets[SecondPetIndex]):
                    set SecondPet = Mypet.PetProp
                if (Mypet:=Pets[ThirdPetIndex]):
                    set ThirdPet = Mypet.PetProp
                if (Mypet:=Pets[ForthPetIndex]):
                    set ForthPet = Mypet.PetProp

                var PetWon:creative_prop = creative_prop{}
                var RandResult:int = GetRandomInt(0,1000)
                Print("Rand: {RandResult}")
                if:
                    PlayerObj := player[Agent]
                    Fort := Agent.GetFortCharacter[]
                then:
                    if(CustomPlayer:Custom_Player = PlayersMap[PlayerObj]):
                        #--------------Luck
                        set RandResult+=CustomPlayer.Pet_Luck_Value*5+CustomPlayer.Bonus_Luck*5
                        #--------------
                        if (RandResult>995):
                            Print("Legendary")
                            if(Cine:=Cine_L3_Legendary[GetTeamIndexAgent[Agent]-1]):
                                Cine.Play(Agent)
                            Sleep(4.0)
                            if(Cine:=Cine_L3_Legendary[GetTeamIndexAgent[Agent]-1]):
                                Cine.Stop(Agent)
                                Accolade_Egg3_4.Award(Agent)

                            if (CustomPlayer.Ownership_L3_Pet4=0):
                                #Release potential previous pets
                                if (CustomPlayer.Ownership_L3_Pet1=1):
                                    #--------------Button pressed:
                                    CustomPlayer.IncreaseButtonL3()
                                    #Print("Button L3 Pressed: {CustomPlayer.ButtonL3Pressed}")
                                    #--------------Spawn
                                    Sleep(0.5)
                                    #spawn:
                                    #    ReleasePet(FirstPet)
                                    set CustomPlayer.Ownership_L3_Pet1=0
                                    L3_Pet1_DesOwn(Agent)
                                if (CustomPlayer.Ownership_L3_Pet2=1):
                                    #--------------Button pressed:
                                    CustomPlayer.IncreaseButtonL3()
                                    #Print("Button L3 Pressed: {CustomPlayer.ButtonL3Pressed}")
                                    #--------------Spawn
                                    Sleep(0.5)
                                    #spawn:
                                    #    ReleasePet(SecondPet)
                                    set CustomPlayer.Ownership_L3_Pet2=0
                                    L3_Pet2_DesOwn(Agent)
                                if (CustomPlayer.Ownership_L3_Pet3=1):
                                    #--------------Button pressed:
                                    CustomPlayer.IncreaseButtonL3()
                                    #Print("Button L3 Pressed: {CustomPlayer.ButtonL3Pressed}")
                                    #--------------Spawn
                                    Sleep(0.5)
                                    #spawn:
                                    #    ReleasePet(ThirdPet)
                                    set CustomPlayer.Ownership_L3_Pet3=0
                                    L3_Pet3_DesOwn(Agent)
                                #Get New pet
                                L3_Pet4_Own(Agent)
                                set CustomPlayer.Ownership_L3_Pet4=1
                                set CustomPlayer.Pet_Unlocked_L3=4
                                CustomPlayer.Update_Pet_PowerL3(Power_L3P4)
                                var PetPower:float=1.0*Power_L3P4
                                set PetWon = ForthPet
                                spawn:
                                    FollowPlayerL3(Fort, PetWon, PetPower, ForthPetIndex)
                        else if (RandResult>855):
                            Print("Epic")
                            if(Cine:=Cine_L3_Epic[GetTeamIndexAgent[Agent]-1]):
                                Cine.Play(Agent)
                            Sleep(4.0)
                            if(Cine:=Cine_L3_Epic[GetTeamIndexAgent[Agent]-1]):
                                Cine.Stop(Agent)
                                Accolade_Egg3_4.Award(Agent)

                            if (CustomPlayer.Ownership_L3_Pet3=0 and CustomPlayer.Ownership_L3_Pet4=0):
                                #Release potential previous pets
                                if (CustomPlayer.Ownership_L3_Pet1=1):
                                    #--------------Button pressed:
                                    CustomPlayer.IncreaseButtonL3()
                                    #Print("Button L3 Pressed: {CustomPlayer.ButtonL3Pressed}")
                                    #--------------Spawn
                                    Sleep(0.5)
                                    ##spawn:
                                    ##    ReleasePet(FirstPet)
                                    set CustomPlayer.Ownership_L3_Pet1=0
                                    L3_Pet1_DesOwn(Agent)
                                if (CustomPlayer.Ownership_L3_Pet2=1):
                                    #--------------Button pressed:
                                    CustomPlayer.IncreaseButtonL3()
                                    #Print("Button L3 Pressed: {CustomPlayer.ButtonL3Pressed}")
                                    #--------------Spawn
                                    Sleep(0.5)
                                    ##spawn:
                                    ##    ReleasePet(SecondPet)
                                    set CustomPlayer.Ownership_L3_Pet2=0
                                    L3_Pet2_DesOwn(Agent)
                                #Get New pet
                                L3_Pet3_Own(Agent)
                                set CustomPlayer.Ownership_L3_Pet3=1
                                set CustomPlayer.Pet_Unlocked_L3=3
                                CustomPlayer.Update_Pet_PowerL3(Power_L3P3)
                                var PetPower:float=1.0*Power_L3P3
                                set PetWon = ThirdPet
                                spawn:
                                    FollowPlayerL3(Fort, PetWon, PetPower, ThirdPetIndex)
                        else if (RandResult>750):
                            Print("Rare")
                            if(Cine:=Cine_L3_Rare[GetTeamIndexAgent[Agent]-1]):
                                Cine.Play(Agent)
                            Sleep(4.0)
                            if(Cine:=Cine_L3_Rare[GetTeamIndexAgent[Agent]-1]):
                                Cine.Stop(Agent)
                                Accolade_Egg3_4.Award(Agent)

                            if (CustomPlayer.Ownership_L3_Pet2=0 and CustomPlayer.Ownership_L3_Pet3=0 and CustomPlayer.Ownership_L3_Pet4=0):
                                #Release potential previous pets
                                if (CustomPlayer.Ownership_L3_Pet1=1):
                                    #--------------Button pressed:
                                    CustomPlayer.IncreaseButtonL3()
                                    #Print("Button L3 Pressed: {CustomPlayer.ButtonL3Pressed}")
                                    #--------------Spawn
                                    Sleep(0.5)
                                    ##spawn:
                                    ##    ReleasePet(FirstPet)
                                    set CustomPlayer.Ownership_L3_Pet1=0
                                    L3_Pet1_DesOwn(Agent)
                                #Get New pet
                                L3_Pet2_Own(Agent)
                                set CustomPlayer.Ownership_L3_Pet2=1
                                set CustomPlayer.Pet_Unlocked_L3=2
                                CustomPlayer.Update_Pet_PowerL3(Power_L3P2)
                                var PetPower:float=1.0*Power_L3P2
                                set PetWon = SecondPet
                                spawn:
                                    FollowPlayerL3(Fort, PetWon, PetPower, SecondPetIndex)
                        else:
                            Print("Normal")
                            if(Cine:=Cine_L3_Normal[GetTeamIndexAgent[Agent]-1]):
                                Cine.Play(Agent)
                            Sleep(4.0)
                            if(Cine:=Cine_L3_Normal[GetTeamIndexAgent[Agent]-1]):
                                Cine.Stop(Agent)
                                Accolade_Egg3_4.Award(Agent)

                            if (CustomPlayer.Ownership_L3_Pet1=0 and CustomPlayer.Ownership_L3_Pet2=0 and CustomPlayer.Ownership_L3_Pet3=0 and CustomPlayer.Ownership_L3_Pet4=0):
                                #Get New pet
                                L3_Pet1_Own(Agent)
                                set CustomPlayer.Ownership_L3_Pet1=1
                                set CustomPlayer.Pet_Unlocked_L3=1
                                CustomPlayer.Update_Pet_PowerL3(Power_L3P1)
                                var PetPower:float=1.0*Power_L3P1
                                set PetWon = FirstPet
                                spawn:
                                    FollowPlayerL3(Fort, PetWon, PetPower, FirstPetIndex)

    PurchasePetL4( Agent : agent )<suspends>: void =
        #Requires Unlock L4
        if:
            PlayerObj := player[Agent]
            CurrentPlayerStats := PlayerStatsManager.GetPlayerStats[Agent]
            CurrentPlayerStats.LevelUnlocked.CurrentValue<3
        then:
            RequireHigherLevelHUD.Show(Agent) 
        else:
            if:
                PlayerObj := player[Agent]
                CustomPlayer:Custom_Player = PlayersMap[PlayerObj]
                CustomPlayer.PlayerGold<CostL4
            then:
                NotEnoughGoldHUD.Show(Agent)
            else:
                DecreaseGoldFunc(Agent,CostL4)
                var FirstPet:creative_prop = creative_prop{}
                var SecondPet:creative_prop = creative_prop{}
                var ThirdPet:creative_prop = creative_prop{}
                var ForthPet:creative_prop = creative_prop{}

                var FirstPetIndex:int = 0
                var SecondPetIndex:int = 0
                var ThirdPetIndex:int = 0
                var ForthPetIndex:int = 0

                var IndexBase:int = 12*3

                #Select right pets for each player
                if(GetTeamIndexAgent[Agent]=1):
                    set FirstPetIndex =  IndexBase
                    set SecondPetIndex = IndexBase+1
                    set ThirdPetIndex =  IndexBase+2
                    set ForthPetIndex =  IndexBase+3
                if(GetTeamIndexAgent[Agent]=2):
                    set FirstPetIndex =  IndexBase+4
                    set SecondPetIndex = IndexBase+5
                    set ThirdPetIndex =  IndexBase+6
                    set ForthPetIndex =  IndexBase+7
                if(GetTeamIndexAgent[Agent]=3):
                    set FirstPetIndex =  IndexBase+8
                    set SecondPetIndex = IndexBase+9
                    set ThirdPetIndex =  IndexBase+10
                    set ForthPetIndex =  IndexBase+11
                if (Mypet:=Pets[FirstPetIndex]):
                    set FirstPet = Mypet.PetProp
                if (Mypet:=Pets[SecondPetIndex]):
                    set SecondPet = Mypet.PetProp
                if (Mypet:=Pets[ThirdPetIndex]):
                    set ThirdPet = Mypet.PetProp
                if (Mypet:=Pets[ForthPetIndex]):
                    set ForthPet = Mypet.PetProp

                var PetWon:creative_prop = creative_prop{}
                var RandResult:int = GetRandomInt(0,1000)
                Print("Rand: {RandResult}")
                if:
                    PlayerObj := player[Agent]
                    Fort := Agent.GetFortCharacter[]
                then:
                    if(CustomPlayer:Custom_Player = PlayersMap[PlayerObj]):
                        #--------------Luck
                        set RandResult+=CustomPlayer.Pet_Luck_Value*5+CustomPlayer.Bonus_Luck*5
                        #--------------
                        if (RandResult>995):
                            Print("Legendary")
                            if(Cine:=Cine_L4_Legendary[GetTeamIndexAgent[Agent]-1]):
                                Cine.Play(Agent)
                            Sleep(4.0)
                            if(Cine:=Cine_L4_Legendary[GetTeamIndexAgent[Agent]-1]):
                                Cine.Stop(Agent)
                                Accolade_Egg3_4.Award(Agent)

                            if (CustomPlayer.Ownership_L4_Pet4=0):
                                #Release potential previous pets
                                if (CustomPlayer.Ownership_L4_Pet1=1):
                                    #--------------Button pressed:
                                    CustomPlayer.IncreaseButtonL4()
                                    #Print("Button L4 Pressed: {CustomPlayer.ButtonL4Pressed}")
                                    #--------------Spawn
                                    Sleep(0.5)
                                    ##spawn:
                                    ##    ReleasePet(FirstPet)
                                    set CustomPlayer.Ownership_L4_Pet1=0
                                    L4_Pet1_DesOwn(Agent)
                                if (CustomPlayer.Ownership_L4_Pet2=1):
                                    #--------------Button pressed:
                                    CustomPlayer.IncreaseButtonL4()
                                    #Print("Button L4 Pressed: {CustomPlayer.ButtonL4Pressed}")
                                    #--------------Spawn
                                    Sleep(0.5)
                                    ##spawn:
                                    ##    ReleasePet(SecondPet)
                                    set CustomPlayer.Ownership_L4_Pet2=0
                                    L4_Pet2_DesOwn(Agent)
                                if (CustomPlayer.Ownership_L4_Pet3=1):
                                    #--------------Button pressed:
                                    CustomPlayer.IncreaseButtonL4()
                                    #Print("Button L4 Pressed: {CustomPlayer.ButtonL4Pressed}")
                                    #--------------Spawn
                                    Sleep(0.5)
                                    ##spawn:
                                    ##    ReleasePet(ThirdPet)
                                    set CustomPlayer.Ownership_L4_Pet3=0
                                    L4_Pet3_DesOwn(Agent)
                                #Get New pet
                                L4_Pet4_Own(Agent)
                                set CustomPlayer.Ownership_L4_Pet4=1
                                set CustomPlayer.Pet_Unlocked_L4=4
                                CustomPlayer.Update_Pet_PowerL4(Power_L4P4)
                                var PetPower:float=1.0*Power_L4P4
                                set PetWon = ForthPet
                                spawn:
                                    FollowPlayerL4(Fort, PetWon, PetPower, ForthPetIndex)
                        else if (RandResult>855):
                            Print("Epic")
                            if(Cine:=Cine_L4_Epic[GetTeamIndexAgent[Agent]-1]):
                                Cine.Play(Agent)
                            Sleep(4.0)
                            if(Cine:=Cine_L4_Epic[GetTeamIndexAgent[Agent]-1]):
                                Cine.Stop(Agent)
                                Accolade_Egg3_4.Award(Agent)

                            if (CustomPlayer.Ownership_L4_Pet3=0 and CustomPlayer.Ownership_L4_Pet4=0):
                                #Release potential previous pets
                                if (CustomPlayer.Ownership_L4_Pet1=1):
                                    #--------------Button pressed:
                                    CustomPlayer.IncreaseButtonL4()
                                    #Print("Button L4 Pressed: {CustomPlayer.ButtonL4Pressed}")
                                    #--------------Spawn
                                    Sleep(0.5)
                                    ##spawn:
                                    ##    ReleasePet(FirstPet)
                                    set CustomPlayer.Ownership_L4_Pet1=0
                                    L4_Pet1_DesOwn(Agent)
                                if (CustomPlayer.Ownership_L4_Pet2=1):
                                    #--------------Button pressed:
                                    CustomPlayer.IncreaseButtonL4()
                                    #Print("Button L4 Pressed: {CustomPlayer.ButtonL4Pressed}")
                                    #--------------Spawn
                                    Sleep(0.5)
                                    ##spawn:
                                    ##    ReleasePet(SecondPet)
                                    set CustomPlayer.Ownership_L4_Pet2=0
                                    L4_Pet2_DesOwn(Agent)
                                #Get New pet
                                L4_Pet3_Own(Agent)
                                set CustomPlayer.Ownership_L4_Pet3=1
                                set CustomPlayer.Pet_Unlocked_L4=3
                                CustomPlayer.Update_Pet_PowerL4(Power_L4P3)
                                var PetPower:float=1.0*Power_L4P3
                                set PetWon = ThirdPet
                                spawn:
                                    FollowPlayerL4(Fort, PetWon, PetPower, ThirdPetIndex)
                        else if (RandResult>750):
                            Print("Rare")
                            if(Cine:=Cine_L4_Rare[GetTeamIndexAgent[Agent]-1]):
                                Cine.Play(Agent)
                            Sleep(4.0)
                            if(Cine:=Cine_L4_Rare[GetTeamIndexAgent[Agent]-1]):
                                Cine.Stop(Agent)
                                Accolade_Egg3_4.Award(Agent)

                            if (CustomPlayer.Ownership_L4_Pet2=0 and CustomPlayer.Ownership_L4_Pet3=0 and CustomPlayer.Ownership_L4_Pet4=0):
                                #Release potential previous pets
                                if (CustomPlayer.Ownership_L4_Pet1=1):
                                    #--------------Button pressed:
                                    CustomPlayer.IncreaseButtonL4()
                                    #Print("Button L4 Pressed: {CustomPlayer.ButtonL4Pressed}")
                                    #--------------Spawn
                                    Sleep(0.5)
                                    ##spawn:
                                    ##    ReleasePet(FirstPet)
                                    set CustomPlayer.Ownership_L4_Pet1=0
                                    L4_Pet1_DesOwn(Agent)
                                #Get New pet
                                L4_Pet2_Own(Agent)
                                set CustomPlayer.Ownership_L4_Pet2=1
                                set CustomPlayer.Pet_Unlocked_L4=2
                                CustomPlayer.Update_Pet_PowerL4(Power_L4P2)
                                var PetPower:float=1.0*Power_L4P2
                                set PetWon = SecondPet
                                spawn:
                                    FollowPlayerL4(Fort, PetWon, PetPower, SecondPetIndex)
                        else:
                            Print("Normal")
                            if(Cine:=Cine_L4_Normal[GetTeamIndexAgent[Agent]-1]):
                                Cine.Play(Agent)
                            Sleep(4.0)
                            if(Cine:=Cine_L4_Normal[GetTeamIndexAgent[Agent]-1]):
                                Cine.Stop(Agent)
                                Accolade_Egg3_4.Award(Agent)

                            if (CustomPlayer.Ownership_L4_Pet1=0 and CustomPlayer.Ownership_L4_Pet2=0 and CustomPlayer.Ownership_L4_Pet3=0 and CustomPlayer.Ownership_L4_Pet4=0):
                                #Get New pet
                                L4_Pet1_Own(Agent)
                                set CustomPlayer.Ownership_L4_Pet1=1
                                set CustomPlayer.Pet_Unlocked_L4=1
                                CustomPlayer.Update_Pet_PowerL4(Power_L4P1)
                                var PetPower:float=1.0*Power_L4P1
                                set PetWon = FirstPet
                                spawn:
                                    FollowPlayerL4(Fort, PetWon, PetPower, FirstPetIndex)

    PurchasePetL5( Agent : agent )<suspends>: void =
        #Requires Unlock L5
        if:
            PlayerObj := player[Agent]
            CurrentPlayerStats := PlayerStatsManager.GetPlayerStats[Agent]
            CurrentPlayerStats.LevelUnlocked.CurrentValue<4
        then:
            RequireHigherLevelHUD.Show(Agent) 
        else:
            if:
                PlayerObj := player[Agent]
                CustomPlayer:Custom_Player = PlayersMap[PlayerObj]
                CustomPlayer.PlayerGold<CostL5
            then:
                NotEnoughGoldHUD.Show(Agent)
            else:
                DecreaseGoldFunc(Agent,CostL5)
                var FirstPet:creative_prop = creative_prop{}
                var SecondPet:creative_prop = creative_prop{}
                var ThirdPet:creative_prop = creative_prop{}
                var ForthPet:creative_prop = creative_prop{}

                var FirstPetIndex:int = 0
                var SecondPetIndex:int = 0
                var ThirdPetIndex:int = 0
                var ForthPetIndex:int = 0

                var IndexBase:int = 12*4

                #Select right pets for each player
                if(GetTeamIndexAgent[Agent]=1):
                    set FirstPetIndex =  IndexBase
                    set SecondPetIndex = IndexBase+1
                    set ThirdPetIndex =  IndexBase+2
                    set ForthPetIndex =  IndexBase+3
                if(GetTeamIndexAgent[Agent]=2):
                    set FirstPetIndex =  IndexBase+4
                    set SecondPetIndex = IndexBase+5
                    set ThirdPetIndex =  IndexBase+6
                    set ForthPetIndex =  IndexBase+7
                if(GetTeamIndexAgent[Agent]=3):
                    set FirstPetIndex =  IndexBase+8
                    set SecondPetIndex = IndexBase+9
                    set ThirdPetIndex =  IndexBase+10
                    set ForthPetIndex =  IndexBase+11
                if (Mypet:=Pets[FirstPetIndex]):
                    set FirstPet = Mypet.PetProp
                if (Mypet:=Pets[SecondPetIndex]):
                    set SecondPet = Mypet.PetProp
                if (Mypet:=Pets[ThirdPetIndex]):
                    set ThirdPet = Mypet.PetProp
                if (Mypet:=Pets[ForthPetIndex]):
                    set ForthPet = Mypet.PetProp

                var PetWon:creative_prop = creative_prop{}
                var RandResult:int = GetRandomInt(0,1000)
                Print("Rand: {RandResult}")
                if:
                    PlayerObj := player[Agent]
                    Fort := Agent.GetFortCharacter[]
                then:
                    if(CustomPlayer:Custom_Player = PlayersMap[PlayerObj]):
                        #--------------Luck
                        set RandResult+=CustomPlayer.Pet_Luck_Value*5+CustomPlayer.Bonus_Luck*5
                        #--------------
                        if (RandResult>995):
                            Print("Legendary")
                            if(Cine:=Cine_L5_Legendary[GetTeamIndexAgent[Agent]-1]):
                                Cine.Play(Agent)
                            Sleep(4.0)
                            if(Cine:=Cine_L5_Legendary[GetTeamIndexAgent[Agent]-1]):
                                Cine.Stop(Agent)
                                Accolade_Egg5_6.Award(Agent)

                            if (CustomPlayer.Ownership_L5_Pet4=0):
                                #Release potential previous pets
                                if (CustomPlayer.Ownership_L5_Pet1=1):
                                    #--------------Button pressed:
                                    CustomPlayer.IncreaseButtonL5()
                                    #Print("Button L5 Pressed: {CustomPlayer.ButtonL5Pressed}")
                                    #--------------Spawn
                                    Sleep(0.5)
                                    ##spawn:
                                    ##    ReleasePet(FirstPet)
                                    set CustomPlayer.Ownership_L5_Pet1=0
                                    L5_Pet1_DesOwn(Agent)
                                if (CustomPlayer.Ownership_L5_Pet2=1):
                                    #--------------Button pressed:
                                    CustomPlayer.IncreaseButtonL5()
                                    #Print("Button L5 Pressed: {CustomPlayer.ButtonL5Pressed}")
                                    #--------------Spawn
                                    Sleep(0.5)
                                    ##spawn:
                                    ##    ReleasePet(SecondPet)
                                    set CustomPlayer.Ownership_L5_Pet2=0
                                    L5_Pet2_DesOwn(Agent)
                                if (CustomPlayer.Ownership_L5_Pet3=1):
                                    #--------------Button pressed:
                                    CustomPlayer.IncreaseButtonL5()
                                    #Print("Button L5 Pressed: {CustomPlayer.ButtonL5Pressed}")
                                    #--------------Spawn
                                    Sleep(0.5)
                                    ##spawn:
                                    ##    ReleasePet(ThirdPet)
                                    set CustomPlayer.Ownership_L5_Pet3=0
                                    L5_Pet3_DesOwn(Agent)
                                #Get New pet
                                L5_Pet4_Own(Agent)
                                set CustomPlayer.Ownership_L5_Pet4=1
                                set CustomPlayer.Pet_Unlocked_L5=4
                                CustomPlayer.Update_Pet_PowerL5(Power_L5P4)
                                var PetPower:float=1.0*Power_L5P4
                                set PetWon = ForthPet
                                spawn:
                                    FollowPlayerL5(Fort, PetWon, PetPower, ForthPetIndex)
                        else if (RandResult>855):
                            Print("Epic")
                            if(Cine:=Cine_L5_Epic[GetTeamIndexAgent[Agent]-1]):
                                Cine.Play(Agent)
                            Sleep(4.0)
                            if(Cine:=Cine_L5_Epic[GetTeamIndexAgent[Agent]-1]):
                                Cine.Stop(Agent)
                                Accolade_Egg5_6.Award(Agent)

                            if (CustomPlayer.Ownership_L5_Pet3=0 and CustomPlayer.Ownership_L5_Pet4=0):
                                #Release potential previous pets
                                if (CustomPlayer.Ownership_L5_Pet1=1):
                                    #--------------Button pressed:
                                    CustomPlayer.IncreaseButtonL5()
                                    #Print("Button L5 Pressed: {CustomPlayer.ButtonL5Pressed}")
                                    #--------------Spawn
                                    Sleep(0.5)
                                    ##spawn:
                                    ##    ReleasePet(FirstPet)
                                    set CustomPlayer.Ownership_L5_Pet1=0
                                    L5_Pet1_DesOwn(Agent)
                                if (CustomPlayer.Ownership_L5_Pet2=1):
                                    #--------------Button pressed:
                                    CustomPlayer.IncreaseButtonL5()
                                    #Print("Button L5 Pressed: {CustomPlayer.ButtonL5Pressed}")
                                    #--------------Spawn
                                    Sleep(0.5)
                                    ##spawn:
                                    ##    ReleasePet(SecondPet)
                                    set CustomPlayer.Ownership_L5_Pet2=0
                                    L5_Pet2_DesOwn(Agent)
                                #Get New pet
                                L5_Pet3_Own(Agent)
                                set CustomPlayer.Ownership_L5_Pet3=1
                                set CustomPlayer.Pet_Unlocked_L5=3
                                CustomPlayer.Update_Pet_PowerL5(Power_L5P3)
                                var PetPower:float=1.0*Power_L5P3
                                set PetWon = ThirdPet
                                spawn:
                                    FollowPlayerL5(Fort, PetWon, PetPower, ThirdPetIndex)
                        else if (RandResult>750):
                            Print("Rare")
                            if(Cine:=Cine_L5_Rare[GetTeamIndexAgent[Agent]-1]):
                                Cine.Play(Agent)
                            Sleep(4.0)
                            if(Cine:=Cine_L5_Rare[GetTeamIndexAgent[Agent]-1]):
                                Cine.Stop(Agent)
                                Accolade_Egg5_6.Award(Agent)

                            if (CustomPlayer.Ownership_L5_Pet2=0 and CustomPlayer.Ownership_L5_Pet3=0 and CustomPlayer.Ownership_L5_Pet4=0):
                                #Release potential previous pets
                                if (CustomPlayer.Ownership_L5_Pet1=1):
                                    #--------------Button pressed:
                                    CustomPlayer.IncreaseButtonL5()
                                    #Print("Button L5 Pressed: {CustomPlayer.ButtonL5Pressed}")
                                    #--------------Spawn
                                    Sleep(0.5)
                                    ##spawn:
                                    ##    ReleasePet(FirstPet)
                                    set CustomPlayer.Ownership_L5_Pet1=0
                                    L5_Pet1_DesOwn(Agent)
                                #Get New pet
                                L5_Pet2_Own(Agent)
                                set CustomPlayer.Ownership_L5_Pet2=1
                                set CustomPlayer.Pet_Unlocked_L5=2
                                CustomPlayer.Update_Pet_PowerL5(Power_L5P2)
                                var PetPower:float=1.0*Power_L5P2
                                set PetWon = SecondPet
                                spawn:
                                    FollowPlayerL5(Fort, PetWon, PetPower, SecondPetIndex)
                        else:
                            Print("Normal")
                            if(Cine:=Cine_L5_Normal[GetTeamIndexAgent[Agent]-1]):
                                Cine.Play(Agent)
                            Sleep(4.0)
                            if(Cine:=Cine_L5_Normal[GetTeamIndexAgent[Agent]-1]):
                                Cine.Stop(Agent)
                                Accolade_Egg5_6.Award(Agent)

                            if (CustomPlayer.Ownership_L5_Pet1=0 and CustomPlayer.Ownership_L5_Pet2=0 and CustomPlayer.Ownership_L5_Pet3=0 and CustomPlayer.Ownership_L5_Pet4=0):
                                #Get New pet
                                L5_Pet1_Own(Agent)
                                set CustomPlayer.Ownership_L5_Pet1=1
                                set CustomPlayer.Pet_Unlocked_L5=1
                                CustomPlayer.Update_Pet_PowerL5(Power_L5P1)
                                var PetPower:float=1.0*Power_L5P1
                                set PetWon = FirstPet
                                spawn:
                                    FollowPlayerL5(Fort, PetWon, PetPower, FirstPetIndex)

    PurchasePetL6( Agent : agent )<suspends>: void =
        #Requires Unlock L6
        if:
            PlayerObj := player[Agent]
            CurrentPlayerStats := PlayerStatsManager.GetPlayerStats[Agent]
            CurrentPlayerStats.LevelUnlocked.CurrentValue<5
        then:
            RequireHigherLevelHUD.Show(Agent) 
        else:
            if:
                PlayerObj := player[Agent]
                CustomPlayer:Custom_Player = PlayersMap[PlayerObj]
                CustomPlayer.PlayerGold<CostL6
            then:
                NotEnoughGoldHUD.Show(Agent)
            else:
                DecreaseGoldFunc(Agent,CostL6)
                var FirstPet:creative_prop = creative_prop{}
                var SecondPet:creative_prop = creative_prop{}
                var ThirdPet:creative_prop = creative_prop{}
                var ForthPet:creative_prop = creative_prop{}

                var FirstPetIndex:int = 0
                var SecondPetIndex:int = 0
                var ThirdPetIndex:int = 0
                var ForthPetIndex:int = 0

                var IndexBase:int = 12*5

                #Select right pets for each player
                if(GetTeamIndexAgent[Agent]=1):
                    set FirstPetIndex =  IndexBase
                    set SecondPetIndex = IndexBase+1
                    set ThirdPetIndex =  IndexBase+2
                    set ForthPetIndex =  IndexBase+3
                if(GetTeamIndexAgent[Agent]=2):
                    set FirstPetIndex =  IndexBase+4
                    set SecondPetIndex = IndexBase+5
                    set ThirdPetIndex =  IndexBase+6
                    set ForthPetIndex =  IndexBase+7
                if(GetTeamIndexAgent[Agent]=3):
                    set FirstPetIndex =  IndexBase+8
                    set SecondPetIndex = IndexBase+9
                    set ThirdPetIndex =  IndexBase+10
                    set ForthPetIndex =  IndexBase+11
                if (Mypet:=Pets[FirstPetIndex]):
                    set FirstPet = Mypet.PetProp
                if (Mypet:=Pets[SecondPetIndex]):
                    set SecondPet = Mypet.PetProp
                if (Mypet:=Pets[ThirdPetIndex]):
                    set ThirdPet = Mypet.PetProp
                if (Mypet:=Pets[ForthPetIndex]):
                    set ForthPet = Mypet.PetProp

                var PetWon:creative_prop = creative_prop{}
                var RandResult:int = GetRandomInt(0,1000)
                Print("Rand: {RandResult}")
                if:
                    PlayerObj := player[Agent]
                    Fort := Agent.GetFortCharacter[]
                then:
                    if(CustomPlayer:Custom_Player = PlayersMap[PlayerObj]):
                        #--------------Luck
                        set RandResult+=CustomPlayer.Pet_Luck_Value*5+CustomPlayer.Bonus_Luck*5
                        #--------------
                        if (RandResult>995):
                            Print("Legendary")
                            if(Cine:=Cine_L6_Legendary[GetTeamIndexAgent[Agent]-1]):
                                Cine.Play(Agent)
                            Sleep(4.0)
                            if(Cine:=Cine_L6_Legendary[GetTeamIndexAgent[Agent]-1]):
                                Cine.Stop(Agent)
                                Accolade_Egg5_6.Award(Agent)

                            if (CustomPlayer.Ownership_L6_Pet4=0):
                                #Release potential previous pets
                                if (CustomPlayer.Ownership_L6_Pet1=1):
                                    #--------------Button pressed:
                                    CustomPlayer.IncreaseButtonL6()
                                    #Print("Button L6 Pressed: {CustomPlayer.ButtonL6Pressed}")
                                    #--------------Spawn
                                    Sleep(0.5)
                                    ##spawn:
                                    ##    ReleasePet(FirstPet)
                                    set CustomPlayer.Ownership_L6_Pet1=0
                                    L6_Pet1_DesOwn(Agent)
                                if (CustomPlayer.Ownership_L6_Pet2=1):
                                    #--------------Button pressed:
                                    CustomPlayer.IncreaseButtonL6()
                                    #Print("Button L6 Pressed: {CustomPlayer.ButtonL6Pressed}")
                                    #--------------Spawn
                                    Sleep(0.5)
                                    ##spawn:
                                    ##    ReleasePet(SecondPet)
                                    set CustomPlayer.Ownership_L6_Pet2=0
                                    L6_Pet2_DesOwn(Agent)
                                if (CustomPlayer.Ownership_L6_Pet3=1):
                                    #--------------Button pressed:
                                    CustomPlayer.IncreaseButtonL6()
                                    #Print("Button L6 Pressed: {CustomPlayer.ButtonL6Pressed}")
                                    #--------------Spawn
                                    Sleep(0.5)
                                    ##spawn:
                                    ##    ReleasePet(ThirdPet)
                                    set CustomPlayer.Ownership_L6_Pet3=0
                                    L6_Pet3_DesOwn(Agent)
                                #Get New pet
                                L6_Pet4_Own(Agent)
                                set CustomPlayer.Ownership_L6_Pet4=1
                                set CustomPlayer.Pet_Unlocked_L6=4
                                CustomPlayer.Update_Pet_PowerL6(Power_L6P4)
                                var PetPower:float=1.0*Power_L6P4
                                set PetWon = ForthPet
                                spawn:
                                    FollowPlayerL6(Fort, PetWon, PetPower, ForthPetIndex)
                        else if (RandResult>855):
                            Print("Epic")
                            if(Cine:=Cine_L6_Epic[GetTeamIndexAgent[Agent]-1]):
                                Cine.Play(Agent)
                            Sleep(4.0)
                            if(Cine:=Cine_L6_Epic[GetTeamIndexAgent[Agent]-1]):
                                Cine.Stop(Agent)
                                Accolade_Egg5_6.Award(Agent)

                            if (CustomPlayer.Ownership_L6_Pet3=0 and CustomPlayer.Ownership_L6_Pet4=0):
                                #Release potential previous pets
                                if (CustomPlayer.Ownership_L6_Pet1=1):
                                    #--------------Button pressed:
                                    CustomPlayer.IncreaseButtonL6()
                                    #Print("Button L6 Pressed: {CustomPlayer.ButtonL6Pressed}")
                                    #--------------Spawn
                                    Sleep(0.5)
                                    ##spawn:
                                    ##    ReleasePet(FirstPet)
                                    set CustomPlayer.Ownership_L6_Pet1=0
                                    L6_Pet1_DesOwn(Agent)
                                if (CustomPlayer.Ownership_L6_Pet2=1):
                                    #--------------Button pressed:
                                    CustomPlayer.IncreaseButtonL6()
                                    #Print("Button L6 Pressed: {CustomPlayer.ButtonL6Pressed}")
                                    #--------------Spawn
                                    Sleep(0.5)
                                    ##spawn:
                                    ##    ReleasePet(SecondPet)
                                    set CustomPlayer.Ownership_L6_Pet2=0
                                    L6_Pet2_DesOwn(Agent)
                                #Get New pet
                                L6_Pet3_Own(Agent)
                                set CustomPlayer.Ownership_L6_Pet3=1
                                set CustomPlayer.Pet_Unlocked_L6=3
                                CustomPlayer.Update_Pet_PowerL6(Power_L6P3)
                                var PetPower:float=1.0*Power_L6P3
                                set PetWon = ThirdPet
                                spawn:
                                    FollowPlayerL6(Fort, PetWon, PetPower, ThirdPetIndex)
                        else if (RandResult>750):
                            Print("Rare")
                            if(Cine:=Cine_L6_Rare[GetTeamIndexAgent[Agent]-1]):
                                Cine.Play(Agent)
                            Sleep(4.0)
                            if(Cine:=Cine_L6_Rare[GetTeamIndexAgent[Agent]-1]):
                                Cine.Stop(Agent)
                                Accolade_Egg5_6.Award(Agent)

                            if (CustomPlayer.Ownership_L6_Pet2=0 and CustomPlayer.Ownership_L6_Pet3=0 and CustomPlayer.Ownership_L6_Pet4=0):
                                #Release potential previous pets
                                if (CustomPlayer.Ownership_L6_Pet1=1):
                                    #--------------Button pressed:
                                    CustomPlayer.IncreaseButtonL6()
                                    #Print("Button L6 Pressed: {CustomPlayer.ButtonL6Pressed}")
                                    #--------------Spawn
                                    Sleep(0.5)
                                    ##spawn:
                                    ##    ReleasePet(FirstPet)
                                    set CustomPlayer.Ownership_L6_Pet1=0
                                    L6_Pet1_DesOwn(Agent)
                                #Get New pet
                                L6_Pet2_Own(Agent)
                                set CustomPlayer.Ownership_L6_Pet2=1
                                set CustomPlayer.Pet_Unlocked_L6=2
                                CustomPlayer.Update_Pet_PowerL6(Power_L6P2)
                                var PetPower:float=1.0*Power_L6P2
                                set PetWon = SecondPet
                                spawn:
                                    FollowPlayerL6(Fort, PetWon, PetPower, SecondPetIndex)
                        else:
                            Print("Normal")
                            if(Cine:=Cine_L6_Normal[GetTeamIndexAgent[Agent]-1]):
                                Cine.Play(Agent)
                            Sleep(4.0)
                            if(Cine:=Cine_L6_Normal[GetTeamIndexAgent[Agent]-1]):
                                Cine.Stop(Agent)
                                Accolade_Egg5_6.Award(Agent)

                            if (CustomPlayer.Ownership_L6_Pet1=0 and CustomPlayer.Ownership_L6_Pet2=0 and CustomPlayer.Ownership_L6_Pet3=0 and CustomPlayer.Ownership_L6_Pet4=0):
                                #Get New pet
                                L6_Pet1_Own(Agent)
                                set CustomPlayer.Ownership_L6_Pet1=1
                                set CustomPlayer.Pet_Unlocked_L6=1
                                CustomPlayer.Update_Pet_PowerL6(Power_L6P1)
                                var PetPower:float=1.0*Power_L6P1
                                set PetWon = FirstPet
                                spawn:
                                    FollowPlayerL6(Fort, PetWon, PetPower, FirstPetIndex)

    PurchasePetL7( Agent : agent )<suspends>: void =
        #Requires Unlock L7
        if:
            PlayerObj := player[Agent]
            CurrentPlayerStats := PlayerStatsManager.GetPlayerStats[Agent]
            CurrentPlayerStats.LevelUnlocked.CurrentValue<6
        then:
            RequireHigherLevelHUD.Show(Agent) 
        else:
            if:
                PlayerObj := player[Agent]
                CustomPlayer:Custom_Player = PlayersMap[PlayerObj]
                CustomPlayer.PlayerGold<CostL7
            then:
                NotEnoughGoldHUD.Show(Agent)
            else:
                DecreaseGoldFunc(Agent,CostL7)
                var FirstPet:creative_prop = creative_prop{}
                var SecondPet:creative_prop = creative_prop{}
                var ThirdPet:creative_prop = creative_prop{}
                var ForthPet:creative_prop = creative_prop{}

                var FirstPetIndex:int = 0
                var SecondPetIndex:int = 0
                var ThirdPetIndex:int = 0
                var ForthPetIndex:int = 0

                var IndexBase:int = 12*6

                #Select right pets for each player
                if(GetTeamIndexAgent[Agent]=1):
                    set FirstPetIndex =  IndexBase
                    set SecondPetIndex = IndexBase+1
                    set ThirdPetIndex =  IndexBase+2
                    set ForthPetIndex =  IndexBase+3
                if(GetTeamIndexAgent[Agent]=2):
                    set FirstPetIndex =  IndexBase+4
                    set SecondPetIndex = IndexBase+5
                    set ThirdPetIndex =  IndexBase+6
                    set ForthPetIndex =  IndexBase+7
                if(GetTeamIndexAgent[Agent]=3):
                    set FirstPetIndex =  IndexBase+8
                    set SecondPetIndex = IndexBase+9
                    set ThirdPetIndex =  IndexBase+10
                    set ForthPetIndex =  IndexBase+11
                if (Mypet:=Pets[FirstPetIndex]):
                    set FirstPet = Mypet.PetProp
                if (Mypet:=Pets[SecondPetIndex]):
                    set SecondPet = Mypet.PetProp
                if (Mypet:=Pets[ThirdPetIndex]):
                    set ThirdPet = Mypet.PetProp
                if (Mypet:=Pets[ForthPetIndex]):
                    set ForthPet = Mypet.PetProp

                var PetWon:creative_prop = creative_prop{}
                var RandResult:int = GetRandomInt(0,1000)
                Print("Rand: {RandResult}")
                if:
                    PlayerObj := player[Agent]
                    Fort := Agent.GetFortCharacter[]
                then:
                    if(CustomPlayer:Custom_Player = PlayersMap[PlayerObj]):
                        #--------------Luck
                        set RandResult+=CustomPlayer.Pet_Luck_Value*5+CustomPlayer.Bonus_Luck*5
                        #--------------
                        if (RandResult>995):
                            Print("Legendary")
                            if(Cine:=Cine_L7_Legendary[GetTeamIndexAgent[Agent]-1]):
                                Cine.Play(Agent)
                            Sleep(4.0)
                            if(Cine:=Cine_L7_Legendary[GetTeamIndexAgent[Agent]-1]):
                                Cine.Stop(Agent)
                                Accolade_Egg7_8.Award(Agent)

                            if (CustomPlayer.Ownership_L7_Pet4=0):
                                #Release potential previous pets
                                if (CustomPlayer.Ownership_L7_Pet1=1):
                                    #--------------Button pressed:
                                    CustomPlayer.IncreaseButtonL7()
                                    #Print("Button L7 Pressed: {CustomPlayer.ButtonL7Pressed}")
                                    #--------------Spawn
                                    Sleep(0.5)
                                    ##spawn:
                                    ##    ReleasePet(FirstPet)
                                    set CustomPlayer.Ownership_L7_Pet1=0
                                    L7_Pet1_DesOwn(Agent)
                                if (CustomPlayer.Ownership_L7_Pet2=1):
                                    #--------------Button pressed:
                                    CustomPlayer.IncreaseButtonL7()
                                    #Print("Button L7 Pressed: {CustomPlayer.ButtonL7Pressed}")
                                    #--------------Spawn
                                    Sleep(0.5)
                                    ##spawn:
                                    ##    ReleasePet(SecondPet)
                                    set CustomPlayer.Ownership_L7_Pet2=0
                                    L7_Pet2_DesOwn(Agent)
                                if (CustomPlayer.Ownership_L7_Pet3=1):
                                    #--------------Button pressed:
                                    CustomPlayer.IncreaseButtonL7()
                                    #Print("Button L7 Pressed: {CustomPlayer.ButtonL7Pressed}")
                                    #--------------Spawn
                                    Sleep(0.5)
                                    ##spawn:
                                    ##    ReleasePet(ThirdPet)
                                    set CustomPlayer.Ownership_L7_Pet3=0
                                    L7_Pet3_DesOwn(Agent)
                                #Get New pet
                                L7_Pet4_Own(Agent)
                                set CustomPlayer.Ownership_L7_Pet4=1
                                set CustomPlayer.Pet_Unlocked_L7=4
                                CustomPlayer.Update_Pet_PowerL7(Power_L7P4)
                                var PetPower:float=1.0*Power_L7P4
                                set PetWon = ForthPet
                                spawn:
                                    FollowPlayerL7(Fort, PetWon, PetPower, ForthPetIndex)
                        else if (RandResult>855):
                            Print("Epic")
                            if(Cine:=Cine_L7_Epic[GetTeamIndexAgent[Agent]-1]):
                                Cine.Play(Agent)
                            Sleep(4.0)
                            if(Cine:=Cine_L7_Epic[GetTeamIndexAgent[Agent]-1]):
                                Cine.Stop(Agent)
                                Accolade_Egg7_8.Award(Agent)

                            if (CustomPlayer.Ownership_L7_Pet3=0 and CustomPlayer.Ownership_L7_Pet4=0):
                                #Release potential previous pets
                                if (CustomPlayer.Ownership_L7_Pet1=1):
                                    #--------------Button pressed:
                                    CustomPlayer.IncreaseButtonL7()
                                    #Print("Button L7 Pressed: {CustomPlayer.ButtonL7Pressed}")
                                    #--------------Spawn
                                    Sleep(0.5)
                                    ##spawn:
                                    ##    ReleasePet(FirstPet)
                                    set CustomPlayer.Ownership_L7_Pet1=0
                                    L7_Pet1_DesOwn(Agent)
                                if (CustomPlayer.Ownership_L7_Pet2=1):
                                    #--------------Button pressed:
                                    CustomPlayer.IncreaseButtonL7()
                                    #Print("Button L7 Pressed: {CustomPlayer.ButtonL7Pressed}")
                                    #--------------Spawn
                                    Sleep(0.5)
                                    ##spawn:
                                    ##    ReleasePet(SecondPet)
                                    set CustomPlayer.Ownership_L7_Pet2=0
                                    L7_Pet2_DesOwn(Agent)
                                #Get New pet
                                L7_Pet3_Own(Agent)
                                set CustomPlayer.Ownership_L7_Pet3=1
                                set CustomPlayer.Pet_Unlocked_L7=3
                                CustomPlayer.Update_Pet_PowerL7(Power_L7P3)
                                var PetPower:float=1.0*Power_L7P3
                                set PetWon = ThirdPet
                                spawn:
                                    FollowPlayerL7(Fort, PetWon, PetPower, ThirdPetIndex)
                        else if (RandResult>750):
                            Print("Rare")
                            if(Cine:=Cine_L7_Rare[GetTeamIndexAgent[Agent]-1]):
                                Cine.Play(Agent)
                            Sleep(4.0)
                            if(Cine:=Cine_L7_Rare[GetTeamIndexAgent[Agent]-1]):
                                Cine.Stop(Agent)
                                Accolade_Egg7_8.Award(Agent)

                            if (CustomPlayer.Ownership_L7_Pet2=0 and CustomPlayer.Ownership_L7_Pet3=0 and CustomPlayer.Ownership_L7_Pet4=0):
                                #Release potential previous pets
                                if (CustomPlayer.Ownership_L7_Pet1=1):
                                    #--------------Button pressed:
                                    CustomPlayer.IncreaseButtonL7()
                                    #Print("Button L7 Pressed: {CustomPlayer.ButtonL7Pressed}")
                                    #--------------Spawn
                                    Sleep(0.5)
                                    ##spawn:
                                    ##    ReleasePet(FirstPet)
                                    set CustomPlayer.Ownership_L7_Pet1=0
                                    L7_Pet1_DesOwn(Agent)
                                #Get New pet
                                L7_Pet2_Own(Agent)
                                set CustomPlayer.Ownership_L7_Pet2=1
                                set CustomPlayer.Pet_Unlocked_L7=2
                                CustomPlayer.Update_Pet_PowerL7(Power_L7P2)
                                var PetPower:float=1.0*Power_L7P2
                                set PetWon = SecondPet
                                spawn:
                                    FollowPlayerL7(Fort, PetWon, PetPower, SecondPetIndex)
                        else:
                            Print("Normal")
                            if(Cine:=Cine_L7_Normal[GetTeamIndexAgent[Agent]-1]):
                                Cine.Play(Agent)
                            Sleep(4.0)
                            if(Cine:=Cine_L7_Normal[GetTeamIndexAgent[Agent]-1]):
                                Cine.Stop(Agent)
                                Accolade_Egg7_8.Award(Agent)

                            if (CustomPlayer.Ownership_L7_Pet1=0 and CustomPlayer.Ownership_L7_Pet2=0 and CustomPlayer.Ownership_L7_Pet3=0 and CustomPlayer.Ownership_L7_Pet4=0):
                                #Get New pet
                                L7_Pet1_Own(Agent)
                                set CustomPlayer.Ownership_L7_Pet1=1
                                set CustomPlayer.Pet_Unlocked_L7=1
                                CustomPlayer.Update_Pet_PowerL7(Power_L7P1)
                                var PetPower:float=1.0*Power_L7P1
                                set PetWon = FirstPet
                                spawn:
                                    FollowPlayerL7(Fort, PetWon, PetPower, FirstPetIndex)

    PurchasePetL8( Agent : agent )<suspends>: void =
        #Requires Unlock L8
        if:
            PlayerObj := player[Agent]
            CurrentPlayerStats := PlayerStatsManager.GetPlayerStats[Agent]
            CurrentPlayerStats.LevelUnlocked.CurrentValue<7
        then:
            RequireHigherLevelHUD.Show(Agent) 
        else:
            if:
                PlayerObj := player[Agent]
                CustomPlayer:Custom_Player = PlayersMap[PlayerObj]
                CustomPlayer.PlayerGold<CostL8
            then:
                NotEnoughGoldHUD.Show(Agent)
            else:
                DecreaseGoldFunc(Agent,CostL8)
                var FirstPet:creative_prop = creative_prop{}
                var SecondPet:creative_prop = creative_prop{}
                var ThirdPet:creative_prop = creative_prop{}
                var ForthPet:creative_prop = creative_prop{}

                var FirstPetIndex:int = 0
                var SecondPetIndex:int = 0
                var ThirdPetIndex:int = 0
                var ForthPetIndex:int = 0

                var IndexBase:int = 12*7

                #Select right pets for each player
                if(GetTeamIndexAgent[Agent]=1):
                    set FirstPetIndex =  IndexBase
                    set SecondPetIndex = IndexBase+1
                    set ThirdPetIndex =  IndexBase+2
                    set ForthPetIndex =  IndexBase+3
                if(GetTeamIndexAgent[Agent]=2):
                    set FirstPetIndex =  IndexBase+4
                    set SecondPetIndex = IndexBase+5
                    set ThirdPetIndex =  IndexBase+6
                    set ForthPetIndex =  IndexBase+7
                if(GetTeamIndexAgent[Agent]=3):
                    set FirstPetIndex =  IndexBase+8
                    set SecondPetIndex = IndexBase+9
                    set ThirdPetIndex =  IndexBase+10
                    set ForthPetIndex =  IndexBase+11
                if (Mypet:=Pets[FirstPetIndex]):
                    set FirstPet = Mypet.PetProp
                if (Mypet:=Pets[SecondPetIndex]):
                    set SecondPet = Mypet.PetProp
                if (Mypet:=Pets[ThirdPetIndex]):
                    set ThirdPet = Mypet.PetProp
                if (Mypet:=Pets[ForthPetIndex]):
                    set ForthPet = Mypet.PetProp

                var PetWon:creative_prop = creative_prop{}
                var RandResult:int = GetRandomInt(0,1000)
                Print("Rand: {RandResult}")
                if:
                    PlayerObj := player[Agent]
                    Fort := Agent.GetFortCharacter[]
                then:
                    if(CustomPlayer:Custom_Player = PlayersMap[PlayerObj]):
                        #--------------Luck
                        set RandResult+=CustomPlayer.Pet_Luck_Value*5+CustomPlayer.Bonus_Luck*5
                        #--------------
                        if (RandResult>995):
                            Print("Legendary")
                            if(Cine:=Cine_L8_Legendary[GetTeamIndexAgent[Agent]-1]):
                                Cine.Play(Agent)
                            Sleep(4.0)
                            if(Cine:=Cine_L8_Legendary[GetTeamIndexAgent[Agent]-1]):
                                Cine.Stop(Agent)
                                Accolade_Egg7_8.Award(Agent)

                            if (CustomPlayer.Ownership_L8_Pet4=0):
                                #Release potential previous pets
                                if (CustomPlayer.Ownership_L8_Pet1=1):
                                    #--------------Button pressed:
                                    CustomPlayer.IncreaseButtonL8()
                                    #Print("Button L8 Pressed: {CustomPlayer.ButtonL8Pressed}")
                                    #--------------Spawn
                                    Sleep(0.5)
                                    ##spawn:
                                    ##    ReleasePet(FirstPet)
                                    set CustomPlayer.Ownership_L8_Pet1=0
                                    L8_Pet1_DesOwn(Agent)
                                if (CustomPlayer.Ownership_L8_Pet2=1):
                                    #--------------Button pressed:
                                    CustomPlayer.IncreaseButtonL8()
                                    #Print("Button L8 Pressed: {CustomPlayer.ButtonL8Pressed}")
                                    #--------------Spawn
                                    Sleep(0.5)
                                    ##spawn:
                                    ##    ReleasePet(SecondPet)
                                    set CustomPlayer.Ownership_L8_Pet2=0
                                    L8_Pet2_DesOwn(Agent)
                                if (CustomPlayer.Ownership_L8_Pet3=1):
                                    #--------------Button pressed:
                                    CustomPlayer.IncreaseButtonL8()
                                    #Print("Button L8 Pressed: {CustomPlayer.ButtonL8Pressed}")
                                    #--------------Spawn
                                    Sleep(0.5)
                                    ##spawn:
                                    ##    ReleasePet(ThirdPet)
                                    set CustomPlayer.Ownership_L8_Pet3=0
                                    L8_Pet3_DesOwn(Agent)
                                #Get New pet
                                L8_Pet4_Own(Agent)
                                set CustomPlayer.Ownership_L8_Pet4=1
                                set CustomPlayer.Pet_Unlocked_L8=4
                                CustomPlayer.Update_Pet_PowerL8(Power_L8P4)
                                var PetPower:float=1.0*Power_L8P4
                                set PetWon = ForthPet
                                spawn:
                                    FollowPlayerL8(Fort, PetWon, PetPower, ForthPetIndex)
                        else if (RandResult>855):
                            Print("Epic")
                            if(Cine:=Cine_L8_Epic[GetTeamIndexAgent[Agent]-1]):
                                Cine.Play(Agent)
                            Sleep(4.0)
                            if(Cine:=Cine_L8_Epic[GetTeamIndexAgent[Agent]-1]):
                                Cine.Stop(Agent)
                                Accolade_Egg7_8.Award(Agent)

                            if (CustomPlayer.Ownership_L8_Pet3=0 and CustomPlayer.Ownership_L8_Pet4=0):
                                #Release potential previous pets
                                if (CustomPlayer.Ownership_L8_Pet1=1):
                                    #--------------Button pressed:
                                    CustomPlayer.IncreaseButtonL8()
                                    #Print("Button L8 Pressed: {CustomPlayer.ButtonL8Pressed}")
                                    #--------------Spawn
                                    Sleep(0.5)
                                    ##spawn:
                                    ##    ReleasePet(FirstPet)
                                    set CustomPlayer.Ownership_L8_Pet1=0
                                    L8_Pet1_DesOwn(Agent)
                                if (CustomPlayer.Ownership_L8_Pet2=1):
                                    #--------------Button pressed:
                                    CustomPlayer.IncreaseButtonL8()
                                    #Print("Button L8 Pressed: {CustomPlayer.ButtonL8Pressed}")
                                    #--------------Spawn
                                    Sleep(0.5)
                                    ##spawn:
                                    ##    ReleasePet(SecondPet)
                                    set CustomPlayer.Ownership_L8_Pet2=0
                                    L8_Pet2_DesOwn(Agent)
                                #Get New pet
                                L8_Pet3_Own(Agent)
                                set CustomPlayer.Ownership_L8_Pet3=1
                                set CustomPlayer.Pet_Unlocked_L8=3
                                CustomPlayer.Update_Pet_PowerL8(Power_L8P3)
                                var PetPower:float=1.0*Power_L8P3
                                set PetWon = ThirdPet
                                spawn:
                                    FollowPlayerL8(Fort, PetWon, PetPower, ThirdPetIndex)
                        else if (RandResult>750):
                            Print("Rare")
                            if(Cine:=Cine_L8_Rare[GetTeamIndexAgent[Agent]-1]):
                                Cine.Play(Agent)
                            Sleep(4.0)
                            if(Cine:=Cine_L8_Rare[GetTeamIndexAgent[Agent]-1]):
                                Cine.Stop(Agent)
                                Accolade_Egg7_8.Award(Agent)

                            if (CustomPlayer.Ownership_L8_Pet2=0 and CustomPlayer.Ownership_L8_Pet3=0 and CustomPlayer.Ownership_L8_Pet4=0):
                                #Release potential previous pets
                                if (CustomPlayer.Ownership_L8_Pet1=1):
                                    #--------------Button pressed:
                                    CustomPlayer.IncreaseButtonL8()
                                    #Print("Button L8 Pressed: {CustomPlayer.ButtonL8Pressed}")
                                    #--------------Spawn
                                    Sleep(0.5)
                                    ##spawn:
                                    ##    ReleasePet(FirstPet)
                                    set CustomPlayer.Ownership_L8_Pet1=0
                                    L8_Pet1_DesOwn(Agent)
                                #Get New pet
                                L8_Pet2_Own(Agent)
                                set CustomPlayer.Ownership_L8_Pet2=1
                                set CustomPlayer.Pet_Unlocked_L8=2
                                CustomPlayer.Update_Pet_PowerL8(Power_L8P2)
                                var PetPower:float=1.0*Power_L8P2
                                set PetWon = SecondPet
                                spawn:
                                    FollowPlayerL8(Fort, PetWon, PetPower, SecondPetIndex)
                        else:
                            Print("Normal")
                            if(Cine:=Cine_L8_Normal[GetTeamIndexAgent[Agent]-1]):
                                Cine.Play(Agent)
                            Sleep(4.0)
                            if(Cine:=Cine_L8_Normal[GetTeamIndexAgent[Agent]-1]):
                                Cine.Stop(Agent)
                                Accolade_Egg7_8.Award(Agent)

                            if (CustomPlayer.Ownership_L8_Pet1=0 and CustomPlayer.Ownership_L8_Pet2=0 and CustomPlayer.Ownership_L8_Pet3=0 and CustomPlayer.Ownership_L8_Pet4=0):
                                #Get New pet
                                L8_Pet1_Own(Agent)
                                set CustomPlayer.Ownership_L8_Pet1=1
                                set CustomPlayer.Pet_Unlocked_L8=1
                                CustomPlayer.Update_Pet_PowerL8(Power_L8P1)
                                var PetPower:float=1.0*Power_L8P1
                                set PetWon = FirstPet
                                spawn:
                                    FollowPlayerL8(Fort, PetWon, PetPower, FirstPetIndex)

    PurchasePetL9( Agent : agent )<suspends>: void =
        #Requires Unlock L9
        if:
            PlayerObj := player[Agent]
            CurrentPlayerStats := PlayerStatsManager.GetPlayerStats[Agent]
            CurrentPlayerStats.LevelUnlocked.CurrentValue<8
        then:
            RequireHigherLevelHUD.Show(Agent) 
        else:
            if:
                PlayerObj := player[Agent]
                CustomPlayer:Custom_Player = PlayersMap[PlayerObj]
                CustomPlayer.PlayerGold<CostL9
            then:
                NotEnoughGoldHUD.Show(Agent)
            else:
                DecreaseGoldFunc(Agent,CostL9)
                var FirstPet:creative_prop = creative_prop{}
                var SecondPet:creative_prop = creative_prop{}
                var ThirdPet:creative_prop = creative_prop{}
                var ForthPet:creative_prop = creative_prop{}

                var FirstPetIndex:int = 0
                var SecondPetIndex:int = 0
                var ThirdPetIndex:int = 0
                var ForthPetIndex:int = 0

                var IndexBase:int = 12*8

                #Select right pets for each player
                if(GetTeamIndexAgent[Agent]=1):
                    set FirstPetIndex =  IndexBase
                    set SecondPetIndex = IndexBase+1
                    set ThirdPetIndex =  IndexBase+2
                    set ForthPetIndex =  IndexBase+3
                if(GetTeamIndexAgent[Agent]=2):
                    set FirstPetIndex =  IndexBase+4
                    set SecondPetIndex = IndexBase+5
                    set ThirdPetIndex =  IndexBase+6
                    set ForthPetIndex =  IndexBase+7
                if(GetTeamIndexAgent[Agent]=3):
                    set FirstPetIndex =  IndexBase+8
                    set SecondPetIndex = IndexBase+9
                    set ThirdPetIndex =  IndexBase+10
                    set ForthPetIndex =  IndexBase+11
                if (Mypet:=Pets[FirstPetIndex]):
                    set FirstPet = Mypet.PetProp
                if (Mypet:=Pets[SecondPetIndex]):
                    set SecondPet = Mypet.PetProp
                if (Mypet:=Pets[ThirdPetIndex]):
                    set ThirdPet = Mypet.PetProp
                if (Mypet:=Pets[ForthPetIndex]):
                    set ForthPet = Mypet.PetProp

                var PetWon:creative_prop = creative_prop{}
                var RandResult:int = GetRandomInt(0,1000)
                Print("Rand: {RandResult}")
                if:
                    PlayerObj := player[Agent]
                    Fort := Agent.GetFortCharacter[]
                then:
                    if(CustomPlayer:Custom_Player = PlayersMap[PlayerObj]):
                        #--------------Luck
                        set RandResult+=CustomPlayer.Pet_Luck_Value*5+CustomPlayer.Bonus_Luck*5
                        #--------------
                        if (RandResult>995):
                            Print("Legendary")
                            if(Cine:=Cine_L9_Legendary[GetTeamIndexAgent[Agent]-1]):
                                Cine.Play(Agent)
                            Sleep(4.0)
                            if(Cine:=Cine_L9_Legendary[GetTeamIndexAgent[Agent]-1]):
                                Cine.Stop(Agent)
                                Accolade_Egg9_10.Award(Agent)

                            if (CustomPlayer.Ownership_L9_Pet4=0):
                                #Release potential previous pets
                                if (CustomPlayer.Ownership_L9_Pet1=1):
                                    #--------------Button pressed:
                                    CustomPlayer.IncreaseButtonL9()
                                    #Print("Button L9 Pressed: {CustomPlayer.ButtonL9Pressed}")
                                    #--------------Spawn
                                    Sleep(0.5)
                                    ##spawn:
                                    ##    ReleasePet(FirstPet)
                                    set CustomPlayer.Ownership_L9_Pet1=0
                                    L9_Pet1_DesOwn(Agent)
                                if (CustomPlayer.Ownership_L9_Pet2=1):
                                    #--------------Button pressed:
                                    CustomPlayer.IncreaseButtonL9()
                                    #Print("Button L9 Pressed: {CustomPlayer.ButtonL9Pressed}")
                                    #--------------Spawn
                                    Sleep(0.5)
                                    ##spawn:
                                    ##    ReleasePet(SecondPet)
                                    set CustomPlayer.Ownership_L9_Pet2=0
                                    L9_Pet2_DesOwn(Agent)
                                if (CustomPlayer.Ownership_L9_Pet3=1):
                                    #--------------Button pressed:
                                    CustomPlayer.IncreaseButtonL9()
                                    #Print("Button L9 Pressed: {CustomPlayer.ButtonL9Pressed}")
                                    #--------------Spawn
                                    Sleep(0.5)
                                    ##spawn:
                                    ##    ReleasePet(ThirdPet)
                                    set CustomPlayer.Ownership_L9_Pet3=0
                                    L9_Pet3_DesOwn(Agent)
                                #Get New pet
                                L9_Pet4_Own(Agent)
                                set CustomPlayer.Ownership_L9_Pet4=1
                                set CustomPlayer.Pet_Unlocked_L9=4
                                CustomPlayer.Update_Pet_PowerL9(Power_L9P4)
                                var PetPower:float=1.0*Power_L9P4
                                set PetWon = ForthPet
                                spawn:
                                    FollowPlayerL9(Fort, PetWon, PetPower, ForthPetIndex)
                        else if (RandResult>855):
                            Print("Epic")
                            if(Cine:=Cine_L9_Epic[GetTeamIndexAgent[Agent]-1]):
                                Cine.Play(Agent)
                            Sleep(4.0)
                            if(Cine:=Cine_L9_Epic[GetTeamIndexAgent[Agent]-1]):
                                Cine.Stop(Agent)
                                Accolade_Egg9_10.Award(Agent)

                            if (CustomPlayer.Ownership_L9_Pet3=0 and CustomPlayer.Ownership_L9_Pet4=0):
                                #Release potential previous pets
                                if (CustomPlayer.Ownership_L9_Pet1=1):
                                    #--------------Button pressed:
                                    CustomPlayer.IncreaseButtonL9()
                                    #Print("Button L9 Pressed: {CustomPlayer.ButtonL9Pressed}")
                                    #--------------Spawn
                                    Sleep(0.5)
                                    ##spawn:
                                    ##    ReleasePet(FirstPet)
                                    set CustomPlayer.Ownership_L9_Pet1=0
                                    L9_Pet1_DesOwn(Agent)
                                if (CustomPlayer.Ownership_L9_Pet2=1):
                                    #--------------Button pressed:
                                    CustomPlayer.IncreaseButtonL9()
                                    #Print("Button L9 Pressed: {CustomPlayer.ButtonL9Pressed}")
                                    #--------------Spawn
                                    Sleep(0.5)
                                    ##spawn:
                                    ##    ReleasePet(SecondPet)
                                    set CustomPlayer.Ownership_L9_Pet2=0
                                    L9_Pet2_DesOwn(Agent)
                                #Get New pet
                                L9_Pet3_Own(Agent)
                                set CustomPlayer.Ownership_L9_Pet3=1
                                set CustomPlayer.Pet_Unlocked_L9=3
                                CustomPlayer.Update_Pet_PowerL9(Power_L9P3)
                                var PetPower:float=1.0*Power_L9P3
                                set PetWon = ThirdPet
                                spawn:
                                    FollowPlayerL9(Fort, PetWon, PetPower, ThirdPetIndex)
                        else if (RandResult>750):
                            Print("Rare")
                            if(Cine:=Cine_L9_Rare[GetTeamIndexAgent[Agent]-1]):
                                Cine.Play(Agent)
                            Sleep(4.0)
                            if(Cine:=Cine_L9_Rare[GetTeamIndexAgent[Agent]-1]):
                                Cine.Stop(Agent)
                                Accolade_Egg9_10.Award(Agent)

                            if (CustomPlayer.Ownership_L9_Pet2=0 and CustomPlayer.Ownership_L9_Pet3=0 and CustomPlayer.Ownership_L9_Pet4=0):
                                #Release potential previous pets
                                if (CustomPlayer.Ownership_L9_Pet1=1):
                                    #--------------Button pressed:
                                    CustomPlayer.IncreaseButtonL9()
                                    #Print("Button L9 Pressed: {CustomPlayer.ButtonL9Pressed}")
                                    #--------------Spawn
                                    Sleep(0.5)
                                    ##spawn:
                                    ##    ReleasePet(FirstPet)
                                    set CustomPlayer.Ownership_L9_Pet1=0
                                    L9_Pet1_DesOwn(Agent)
                                #Get New pet
                                L9_Pet2_Own(Agent)
                                set CustomPlayer.Ownership_L9_Pet2=1
                                set CustomPlayer.Pet_Unlocked_L9=2
                                CustomPlayer.Update_Pet_PowerL9(Power_L9P2)
                                var PetPower:float=1.0*Power_L9P2
                                set PetWon = SecondPet
                                spawn:
                                    FollowPlayerL9(Fort, PetWon, PetPower, SecondPetIndex)
                        else:
                            Print("Normal")
                            if(Cine:=Cine_L9_Normal[GetTeamIndexAgent[Agent]-1]):
                                Cine.Play(Agent)
                            Sleep(4.0)
                            if(Cine:=Cine_L9_Normal[GetTeamIndexAgent[Agent]-1]):
                                Cine.Stop(Agent)
                                Accolade_Egg9_10.Award(Agent)

                            if (CustomPlayer.Ownership_L9_Pet1=0 and CustomPlayer.Ownership_L9_Pet2=0 and CustomPlayer.Ownership_L9_Pet3=0 and CustomPlayer.Ownership_L9_Pet4=0):
                                #Get New pet
                                L9_Pet1_Own(Agent)
                                set CustomPlayer.Ownership_L9_Pet1=1
                                set CustomPlayer.Pet_Unlocked_L9=1
                                CustomPlayer.Update_Pet_PowerL9(Power_L9P1)
                                var PetPower:float=1.0*Power_L9P1
                                set PetWon = FirstPet
                                spawn:
                                    FollowPlayerL9(Fort, PetWon, PetPower, FirstPetIndex)

    PurchasePetL10( Agent : agent )<suspends>: void =
        #Requires Unlock L10
        if:
            PlayerObj := player[Agent]
            CurrentPlayerStats := PlayerStatsManager.GetPlayerStats[Agent]
            CurrentPlayerStats.LevelUnlocked.CurrentValue<9
        then:
            RequireHigherLevelHUD.Show(Agent) 
        else:
            if:
                PlayerObj := player[Agent]
                CustomPlayer:Custom_Player = PlayersMap[PlayerObj]
                CustomPlayer.PlayerGold<CostL10
            then:
                NotEnoughGoldHUD.Show(Agent)
            else:
                DecreaseGoldFunc(Agent,CostL10)
                var FirstPet:creative_prop = creative_prop{}
                var SecondPet:creative_prop = creative_prop{}
                var ThirdPet:creative_prop = creative_prop{}
                var ForthPet:creative_prop = creative_prop{}

                var FirstPetIndex:int = 0
                var SecondPetIndex:int = 0
                var ThirdPetIndex:int = 0
                var ForthPetIndex:int = 0

                var IndexBase:int = 12*9

                #Select right pets for each player
                if(GetTeamIndexAgent[Agent]=1):
                    set FirstPetIndex =  IndexBase
                    set SecondPetIndex = IndexBase+1
                    set ThirdPetIndex =  IndexBase+2
                    set ForthPetIndex =  IndexBase+3
                if(GetTeamIndexAgent[Agent]=2):
                    set FirstPetIndex =  IndexBase+4
                    set SecondPetIndex = IndexBase+5
                    set ThirdPetIndex =  IndexBase+6
                    set ForthPetIndex =  IndexBase+7
                if(GetTeamIndexAgent[Agent]=3):
                    set FirstPetIndex =  IndexBase+8
                    set SecondPetIndex = IndexBase+9
                    set ThirdPetIndex =  IndexBase+10
                    set ForthPetIndex =  IndexBase+11

                if (Mypet:=Pets[FirstPetIndex]):
                    set FirstPet = Mypet.PetProp
                if (Mypet:=Pets[SecondPetIndex]):
                    set SecondPet = Mypet.PetProp
                if (Mypet:=Pets[ThirdPetIndex]):
                    set ThirdPet = Mypet.PetProp
                if (Mypet:=Pets[ForthPetIndex]):
                    set ForthPet = Mypet.PetProp

                var PetWon:creative_prop = creative_prop{}
                var RandResult:int = GetRandomInt(0,1000)
                Print("Rand: {RandResult}")
                if:
                    PlayerObj := player[Agent]
                    Fort := Agent.GetFortCharacter[]
                then:
                    if(CustomPlayer:Custom_Player = PlayersMap[PlayerObj]):
                        #--------------Luck
                        set RandResult+=CustomPlayer.Pet_Luck_Value*5+CustomPlayer.Bonus_Luck*5
                        #--------------
                        if (RandResult>995):
                            Print("Legendary")
                            if(Cine:=Cine_L10_Legendary[GetTeamIndexAgent[Agent]-1]):
                                Cine.Play(Agent)
                            Sleep(4.0)
                            if(Cine:=Cine_L10_Legendary[GetTeamIndexAgent[Agent]-1]):
                                Cine.Stop(Agent)
                                Accolade_Egg9_10.Award(Agent)

                            if (CustomPlayer.Ownership_L10_Pet4=0):
                                #Release potential previous pets
                                if (CustomPlayer.Ownership_L10_Pet1=1):
                                    #--------------Button pressed:
                                    CustomPlayer.IncreaseButtonL10()
                                    #Print("Button L10 Pressed: {CustomPlayer.ButtonL10Pressed}")
                                    #--------------Spawn
                                    Sleep(0.5)
                                    ##spawn:
                                    ##    ReleasePet(FirstPet)
                                    set CustomPlayer.Ownership_L10_Pet1=0
                                    L10_Pet1_DesOwn(Agent)
                                if (CustomPlayer.Ownership_L10_Pet2=1):
                                    #--------------Button pressed:
                                    CustomPlayer.IncreaseButtonL10()
                                    #Print("Button L10 Pressed: {CustomPlayer.ButtonL10Pressed}")
                                    #--------------Spawn
                                    Sleep(0.5)
                                    ##spawn:
                                    ##    ReleasePet(SecondPet)
                                    set CustomPlayer.Ownership_L10_Pet2=0
                                    L10_Pet2_DesOwn(Agent)
                                if (CustomPlayer.Ownership_L10_Pet3=1):
                                    #--------------Button pressed:
                                    CustomPlayer.IncreaseButtonL10()
                                    #Print("Button L10 Pressed: {CustomPlayer.ButtonL10Pressed}")
                                    #--------------Spawn
                                    Sleep(0.5)
                                    ##spawn:
                                    ##    ReleasePet(ThirdPet)
                                    set CustomPlayer.Ownership_L10_Pet3=0
                                    L10_Pet3_DesOwn(Agent)
                                #Get New pet
                                L10_Pet4_Own(Agent)
                                set CustomPlayer.Ownership_L10_Pet4=1
                                set CustomPlayer.Pet_Unlocked_L10=4
                                CustomPlayer.Update_Pet_PowerL10(Power_L10P4)
                                var PetPower:float=1.0*Power_L10P4
                                set PetWon = ForthPet
                                spawn:
                                    FollowPlayerL10(Fort, PetWon, PetPower, ForthPetIndex)
                        else if (RandResult>855):
                            Print("Epic")
                            if(Cine:=Cine_L10_Epic[GetTeamIndexAgent[Agent]-1]):
                                Cine.Play(Agent)
                            Sleep(4.0)
                            if(Cine:=Cine_L10_Epic[GetTeamIndexAgent[Agent]-1]):
                                Cine.Stop(Agent)
                                Accolade_Egg9_10.Award(Agent)

                            if (CustomPlayer.Ownership_L10_Pet3=0 and CustomPlayer.Ownership_L10_Pet4=0):
                                #Release potential previous pets
                                if (CustomPlayer.Ownership_L10_Pet1=1):
                                    #--------------Button pressed:
                                    CustomPlayer.IncreaseButtonL10()
                                    #Print("Button L10 Pressed: {CustomPlayer.ButtonL10Pressed}")
                                    #--------------Spawn
                                    Sleep(0.5)
                                    ##spawn:
                                    ##    ReleasePet(FirstPet)
                                    set CustomPlayer.Ownership_L10_Pet1=0
                                    L10_Pet1_DesOwn(Agent)
                                if (CustomPlayer.Ownership_L10_Pet2=1):
                                    #--------------Button pressed:
                                    CustomPlayer.IncreaseButtonL10()
                                    #Print("Button L10 Pressed: {CustomPlayer.ButtonL10Pressed}")
                                    #--------------Spawn
                                    Sleep(0.5)
                                    ##spawn:
                                    ##    ReleasePet(SecondPet)
                                    set CustomPlayer.Ownership_L10_Pet2=0
                                    L10_Pet2_DesOwn(Agent)
                                #Get New pet
                                L10_Pet3_Own(Agent)
                                set CustomPlayer.Ownership_L10_Pet3=1
                                set CustomPlayer.Pet_Unlocked_L10=3
                                CustomPlayer.Update_Pet_PowerL10(Power_L10P3)
                                var PetPower:float=1.0*Power_L10P3
                                set PetWon = ThirdPet
                                spawn:
                                    FollowPlayerL10(Fort, PetWon, PetPower, ThirdPetIndex)
                        else if (RandResult>750):
                            Print("Rare")
                            if(Cine:=Cine_L10_Rare[GetTeamIndexAgent[Agent]-1]):
                                Cine.Play(Agent)
                            Sleep(4.0)
                            if(Cine:=Cine_L10_Rare[GetTeamIndexAgent[Agent]-1]):
                                Cine.Stop(Agent)
                                Accolade_Egg9_10.Award(Agent)

                            if (CustomPlayer.Ownership_L10_Pet2=0 and CustomPlayer.Ownership_L10_Pet3=0 and CustomPlayer.Ownership_L10_Pet4=0):
                                #Release potential previous pets
                                if (CustomPlayer.Ownership_L10_Pet1=1):
                                    #--------------Button pressed:
                                    CustomPlayer.IncreaseButtonL10()
                                    #Print("Button L10 Pressed: {CustomPlayer.ButtonL10Pressed}")
                                    #--------------Spawn
                                    Sleep(0.5)
                                    ##spawn:
                                    ##    ReleasePet(FirstPet)
                                    set CustomPlayer.Ownership_L10_Pet1=0
                                    L10_Pet1_DesOwn(Agent)
                                #Get New pet
                                L10_Pet2_Own(Agent)
                                set CustomPlayer.Ownership_L10_Pet2=1
                                set CustomPlayer.Pet_Unlocked_L10=2
                                CustomPlayer.Update_Pet_PowerL10(Power_L10P2)
                                var PetPower:float=1.0*Power_L10P2
                                set PetWon = SecondPet
                                spawn:
                                    FollowPlayerL10(Fort, PetWon, PetPower, SecondPetIndex)
                        else:
                            Print("Normal")
                            if(Cine:=Cine_L10_Normal[GetTeamIndexAgent[Agent]-1]):
                                Cine.Play(Agent)
                            Sleep(4.0)
                            if(Cine:=Cine_L10_Normal[GetTeamIndexAgent[Agent]-1]):
                                Cine.Stop(Agent)
                                Accolade_Egg9_10.Award(Agent)

                            if (CustomPlayer.Ownership_L10_Pet1=0 and CustomPlayer.Ownership_L10_Pet2=0 and CustomPlayer.Ownership_L10_Pet3=0 and CustomPlayer.Ownership_L10_Pet4=0):
                                #Get New pet
                                L10_Pet1_Own(Agent)
                                set CustomPlayer.Ownership_L10_Pet1=1
                                set CustomPlayer.Pet_Unlocked_L10=1
                                CustomPlayer.Update_Pet_PowerL10(Power_L10P1)
                                var PetPower:float=1.0*Power_L10P1
                                set PetWon = FirstPet
                                spawn:
                                    FollowPlayerL10(Fort, PetWon, PetPower, FirstPetIndex)

    PurchasePetL11( Agent : agent )<suspends>: void =
        #Requires Unlock L11
        if:
            PlayerObj := player[Agent]
            CurrentPlayerStats := PlayerStatsManager.GetPlayerStats[Agent]
            CurrentPlayerStats.LevelUnlocked.CurrentValue<10
        then:
            RequireHigherLevelHUD.Show(Agent) 
        else:
            if:
                PlayerObj := player[Agent]
                CustomPlayer:Custom_Player = PlayersMap[PlayerObj]
                CustomPlayer.PlayerGold<CostL11
            then:
                NotEnoughGoldHUD.Show(Agent)
            else:
                DecreaseGoldFunc(Agent,CostL11)
                var FirstPet:creative_prop = creative_prop{}
                var SecondPet:creative_prop = creative_prop{}
                var ThirdPet:creative_prop = creative_prop{}
                var ForthPet:creative_prop = creative_prop{}

                var FirstPetIndex:int = 0
                var SecondPetIndex:int = 0
                var ThirdPetIndex:int = 0
                var ForthPetIndex:int = 0

                var IndexBase:int = 12*10

                #Select right pets for each player
                if(GetTeamIndexAgent[Agent]=1):
                    set FirstPetIndex =  IndexBase
                    set SecondPetIndex = IndexBase+1
                    set ThirdPetIndex =  IndexBase+2
                    set ForthPetIndex =  IndexBase+3
                if(GetTeamIndexAgent[Agent]=2):
                    set FirstPetIndex =  IndexBase+4
                    set SecondPetIndex = IndexBase+5
                    set ThirdPetIndex =  IndexBase+6
                    set ForthPetIndex =  IndexBase+7
                if(GetTeamIndexAgent[Agent]=3):
                    set FirstPetIndex =  IndexBase+8
                    set SecondPetIndex = IndexBase+9
                    set ThirdPetIndex =  IndexBase+10
                    set ForthPetIndex =  IndexBase+11

                if (Mypet:=Pets[FirstPetIndex]):
                    set FirstPet = Mypet.PetProp
                if (Mypet:=Pets[SecondPetIndex]):
                    set SecondPet = Mypet.PetProp
                if (Mypet:=Pets[ThirdPetIndex]):
                    set ThirdPet = Mypet.PetProp
                if (Mypet:=Pets[ForthPetIndex]):
                    set ForthPet = Mypet.PetProp

                var PetWon:creative_prop = creative_prop{}
                var RandResult:int = GetRandomInt(0,1000)
                Print("Rand: {RandResult}")
                if:
                    PlayerObj := player[Agent]
                    Fort := Agent.GetFortCharacter[]
                then:
                    if(CustomPlayer:Custom_Player = PlayersMap[PlayerObj]):
                        #--------------Luck
                        set RandResult+=CustomPlayer.Pet_Luck_Value*5+CustomPlayer.Bonus_Luck*5
                        #--------------
                        if (RandResult>995):
                            Print("Legendary")
                            if(Cine:=Cine_L11_Legendary[GetTeamIndexAgent[Agent]-1]):
                                Cine.Play(Agent)
                            Sleep(4.0)
                            if(Cine:=Cine_L11_Legendary[GetTeamIndexAgent[Agent]-1]):
                                Cine.Stop(Agent)
                                Accolade_Egg11_12.Award(Agent)

                            if (CustomPlayer.Ownership_L11_Pet4=0):
                                #Release potential previous pets
                                if (CustomPlayer.Ownership_L11_Pet1=1):
                                    #--------------Button pressed:
                                    CustomPlayer.IncreaseButtonL11()
                                    #Print("Button L11 Pressed: {CustomPlayer.ButtonL11Pressed}")
                                    #--------------Spawn
                                    Sleep(0.5)
                                    ##spawn:
                                    ##    ReleasePet(FirstPet)
                                    set CustomPlayer.Ownership_L11_Pet1=0
                                    L11_Pet1_DesOwn(Agent)
                                if (CustomPlayer.Ownership_L11_Pet2=1):
                                    #--------------Button pressed:
                                    CustomPlayer.IncreaseButtonL11()
                                    #Print("Button L11 Pressed: {CustomPlayer.ButtonL11Pressed}")
                                    #--------------Spawn
                                    Sleep(0.5)
                                    ##spawn:
                                    ##    ReleasePet(SecondPet)
                                    set CustomPlayer.Ownership_L11_Pet2=0
                                    L11_Pet2_DesOwn(Agent)
                                if (CustomPlayer.Ownership_L11_Pet3=1):
                                    #--------------Button pressed:
                                    CustomPlayer.IncreaseButtonL11()
                                    #Print("Button L11 Pressed: {CustomPlayer.ButtonL11Pressed}")
                                    #--------------Spawn
                                    Sleep(0.5)
                                    ##spawn:
                                    ##    ReleasePet(ThirdPet)
                                    set CustomPlayer.Ownership_L11_Pet3=0
                                    L11_Pet3_DesOwn(Agent)
                                #Get New pet
                                L11_Pet4_Own(Agent)
                                set CustomPlayer.Ownership_L11_Pet4=1
                                set CustomPlayer.Pet_Unlocked_L11=4
                                CustomPlayer.Update_Pet_PowerL11(Power_L11P4)
                                var PetPower:float=1.0*Power_L11P4
                                set PetWon = ForthPet
                                spawn:
                                    FollowPlayerL11(Fort, PetWon, PetPower, ForthPetIndex)
                        else if (RandResult>855):
                            Print("Epic")
                            if(Cine:=Cine_L11_Epic[GetTeamIndexAgent[Agent]-1]):
                                Cine.Play(Agent)
                            Sleep(4.0)
                            if(Cine:=Cine_L11_Epic[GetTeamIndexAgent[Agent]-1]):
                                Cine.Stop(Agent)
                                Accolade_Egg11_12.Award(Agent)

                            if (CustomPlayer.Ownership_L11_Pet3=0 and CustomPlayer.Ownership_L11_Pet4=0):
                                #Release potential previous pets
                                if (CustomPlayer.Ownership_L11_Pet1=1):
                                    #--------------Button pressed:
                                    CustomPlayer.IncreaseButtonL11()
                                    #Print("Button L11 Pressed: {CustomPlayer.ButtonL11Pressed}")
                                    #--------------Spawn
                                    Sleep(0.5)
                                    ##spawn:
                                    ##    ReleasePet(FirstPet)
                                    set CustomPlayer.Ownership_L11_Pet1=0
                                    L11_Pet1_DesOwn(Agent)
                                if (CustomPlayer.Ownership_L11_Pet2=1):
                                    #--------------Button pressed:
                                    CustomPlayer.IncreaseButtonL11()
                                    #Print("Button L11 Pressed: {CustomPlayer.ButtonL11Pressed}")
                                    #--------------Spawn
                                    Sleep(0.5)
                                    ##spawn:
                                    ##    ReleasePet(SecondPet)
                                    set CustomPlayer.Ownership_L11_Pet2=0
                                    L11_Pet2_DesOwn(Agent)
                                #Get New pet
                                L11_Pet3_Own(Agent)
                                set CustomPlayer.Ownership_L11_Pet3=1
                                set CustomPlayer.Pet_Unlocked_L11=3
                                CustomPlayer.Update_Pet_PowerL11(Power_L11P3)
                                var PetPower:float=1.0*Power_L11P3
                                set PetWon = ThirdPet
                                spawn:
                                    FollowPlayerL11(Fort, PetWon, PetPower, ThirdPetIndex)
                        else if (RandResult>750):
                            Print("Rare")
                            if(Cine:=Cine_L11_Rare[GetTeamIndexAgent[Agent]-1]):
                                Cine.Play(Agent)
                            Sleep(4.0)
                            if(Cine:=Cine_L11_Rare[GetTeamIndexAgent[Agent]-1]):
                                Cine.Stop(Agent)
                                Accolade_Egg11_12.Award(Agent)

                            if (CustomPlayer.Ownership_L11_Pet2=0 and CustomPlayer.Ownership_L11_Pet3=0 and CustomPlayer.Ownership_L11_Pet4=0):
                                #Release potential previous pets
                                if (CustomPlayer.Ownership_L11_Pet1=1):
                                    #--------------Button pressed:
                                    CustomPlayer.IncreaseButtonL11()
                                    #Print("Button L11 Pressed: {CustomPlayer.ButtonL11Pressed}")
                                    #--------------Spawn
                                    Sleep(0.5)
                                    ##spawn:
                                    ##    ReleasePet(FirstPet)
                                    set CustomPlayer.Ownership_L11_Pet1=0
                                    L11_Pet1_DesOwn(Agent)
                                #Get New pet
                                L11_Pet2_Own(Agent)
                                set CustomPlayer.Ownership_L11_Pet2=1
                                set CustomPlayer.Pet_Unlocked_L11=2
                                CustomPlayer.Update_Pet_PowerL11(Power_L11P2)
                                var PetPower:float=1.0*Power_L11P2
                                set PetWon = SecondPet
                                spawn:
                                    FollowPlayerL11(Fort, PetWon, PetPower, SecondPetIndex)
                        else:
                            Print("Normal")
                            if(Cine:=Cine_L11_Normal[GetTeamIndexAgent[Agent]-1]):
                                Cine.Play(Agent)
                            Sleep(4.0)
                            if(Cine:=Cine_L11_Normal[GetTeamIndexAgent[Agent]-1]):
                                Cine.Stop(Agent)
                                Accolade_Egg11_12.Award(Agent)

                            if (CustomPlayer.Ownership_L11_Pet1=0 and CustomPlayer.Ownership_L11_Pet2=0 and CustomPlayer.Ownership_L11_Pet3=0 and CustomPlayer.Ownership_L11_Pet4=0):
                                #Get New pet
                                L11_Pet1_Own(Agent)
                                set CustomPlayer.Ownership_L11_Pet1=1
                                set CustomPlayer.Pet_Unlocked_L11=1
                                CustomPlayer.Update_Pet_PowerL11(Power_L11P1)
                                var PetPower:float=1.0*Power_L11P1
                                set PetWon = FirstPet
                                spawn:
                                    FollowPlayerL11(Fort, PetWon, PetPower, FirstPetIndex)

    PurchasePetL12( Agent : agent )<suspends>: void =
        #Requires Unlock L12
        if:
            PlayerObj := player[Agent]
            CurrentPlayerStats := PlayerStatsManager.GetPlayerStats[Agent]
            CurrentPlayerStats.LevelUnlocked.CurrentValue<11
        then:
            RequireHigherLevelHUD.Show(Agent) 
        else:
            if:
                PlayerObj := player[Agent]
                CustomPlayer:Custom_Player = PlayersMap[PlayerObj]
                CustomPlayer.PlayerGold<CostL12
            then:
                NotEnoughGoldHUD.Show(Agent)
            else:
                DecreaseGoldFunc(Agent,CostL12)
                var FirstPet:creative_prop = creative_prop{}
                var SecondPet:creative_prop = creative_prop{}
                var ThirdPet:creative_prop = creative_prop{}
                var ForthPet:creative_prop = creative_prop{}

                var FirstPetIndex:int = 0
                var SecondPetIndex:int = 0
                var ThirdPetIndex:int = 0
                var ForthPetIndex:int = 0

                var IndexBase:int = 12*11

                #Select right pets for each player
                if(GetTeamIndexAgent[Agent]=1):
                    set FirstPetIndex =  IndexBase
                    set SecondPetIndex = IndexBase+1
                    set ThirdPetIndex =  IndexBase+2
                    set ForthPetIndex =  IndexBase+3
                if(GetTeamIndexAgent[Agent]=2):
                    set FirstPetIndex =  IndexBase+4
                    set SecondPetIndex = IndexBase+5
                    set ThirdPetIndex =  IndexBase+6
                    set ForthPetIndex =  IndexBase+7
                if(GetTeamIndexAgent[Agent]=3):
                    set FirstPetIndex =  IndexBase+8
                    set SecondPetIndex = IndexBase+9
                    set ThirdPetIndex =  IndexBase+10
                    set ForthPetIndex =  IndexBase+11

                if (Mypet:=Pets[FirstPetIndex]):
                    set FirstPet = Mypet.PetProp
                if (Mypet:=Pets[SecondPetIndex]):
                    set SecondPet = Mypet.PetProp
                if (Mypet:=Pets[ThirdPetIndex]):
                    set ThirdPet = Mypet.PetProp
                if (Mypet:=Pets[ForthPetIndex]):
                    set ForthPet = Mypet.PetProp

                var PetWon:creative_prop = creative_prop{}
                var RandResult:int = GetRandomInt(0,1000)
                Print("Rand: {RandResult}")
                if:
                    PlayerObj := player[Agent]
                    Fort := Agent.GetFortCharacter[]
                then:
                    if(CustomPlayer:Custom_Player = PlayersMap[PlayerObj]):
                        #--------------Luck
                        set RandResult+=CustomPlayer.Pet_Luck_Value*5+CustomPlayer.Bonus_Luck*5
                        #--------------
                        if (RandResult>995):
                            Print("Legendary")
                            if(Cine:=Cine_L12_Legendary[GetTeamIndexAgent[Agent]-1]):
                                Cine.Play(Agent)
                            Sleep(4.0)
                            if(Cine:=Cine_L12_Legendary[GetTeamIndexAgent[Agent]-1]):
                                Cine.Stop(Agent)
                                Accolade_Egg11_12.Award(Agent)

                            if (CustomPlayer.Ownership_L12_Pet4=0):
                                #Release potential previous pets
                                if (CustomPlayer.Ownership_L12_Pet1=1):
                                    #--------------Button pressed:
                                    CustomPlayer.IncreaseButtonL12()
                                    #Print("Button L12 Pressed: {CustomPlayer.ButtonL12Pressed}")
                                    #--------------Spawn
                                    Sleep(0.5)
                                    ##spawn:
                                    ##    ReleasePet(FirstPet)
                                    set CustomPlayer.Ownership_L12_Pet1=0
                                    L12_Pet1_DesOwn(Agent)
                                if (CustomPlayer.Ownership_L12_Pet2=1):
                                    #--------------Button pressed:
                                    CustomPlayer.IncreaseButtonL12()
                                    #Print("Button L12 Pressed: {CustomPlayer.ButtonL12Pressed}")
                                    #--------------Spawn
                                    Sleep(0.5)
                                    ##spawn:
                                    ##    ReleasePet(SecondPet)
                                    set CustomPlayer.Ownership_L12_Pet2=0
                                    L12_Pet2_DesOwn(Agent)
                                if (CustomPlayer.Ownership_L12_Pet3=1):
                                    #--------------Button pressed:
                                    CustomPlayer.IncreaseButtonL12()
                                    #Print("Button L12 Pressed: {CustomPlayer.ButtonL12Pressed}")
                                    #--------------Spawn
                                    Sleep(0.5)
                                    ##spawn:
                                    ##    ReleasePet(ThirdPet)
                                    set CustomPlayer.Ownership_L12_Pet3=0
                                    L12_Pet3_DesOwn(Agent)
                                #Get New pet
                                L12_Pet4_Own(Agent)
                                set CustomPlayer.Ownership_L12_Pet4=1
                                set CustomPlayer.Pet_Unlocked_L12=4
                                CustomPlayer.Update_Pet_PowerL12(Power_L12P4)
                                var PetPower:float=1.0*Power_L12P4
                                set PetWon = ForthPet
                                spawn:
                                    FollowPlayerL12(Fort, PetWon, PetPower, ForthPetIndex)
                        else if (RandResult>855):
                            Print("Epic")
                            if(Cine:=Cine_L12_Epic[GetTeamIndexAgent[Agent]-1]):
                                Cine.Play(Agent)
                            Sleep(4.0)
                            if(Cine:=Cine_L12_Epic[GetTeamIndexAgent[Agent]-1]):
                                Cine.Stop(Agent)
                                Accolade_Egg11_12.Award(Agent)

                            if (CustomPlayer.Ownership_L12_Pet3=0 and CustomPlayer.Ownership_L12_Pet4=0):
                                #Release potential previous pets
                                if (CustomPlayer.Ownership_L12_Pet1=1):
                                    #--------------Button pressed:
                                    CustomPlayer.IncreaseButtonL12()
                                    #Print("Button L12 Pressed: {CustomPlayer.ButtonL12Pressed}")
                                    #--------------Spawn
                                    Sleep(0.5)
                                    ##spawn:
                                    ##    ReleasePet(FirstPet)
                                    set CustomPlayer.Ownership_L12_Pet1=0
                                    L12_Pet1_DesOwn(Agent)
                                if (CustomPlayer.Ownership_L12_Pet2=1):
                                    #--------------Button pressed:
                                    CustomPlayer.IncreaseButtonL12()
                                    #Print("Button L12 Pressed: {CustomPlayer.ButtonL12Pressed}")
                                    #--------------Spawn
                                    Sleep(0.5)
                                    ##spawn:
                                    ##    ReleasePet(SecondPet)
                                    set CustomPlayer.Ownership_L12_Pet2=0
                                    L12_Pet2_DesOwn(Agent)
                                #Get New pet
                                L12_Pet3_Own(Agent)
                                set CustomPlayer.Ownership_L12_Pet3=1
                                set CustomPlayer.Pet_Unlocked_L12=3
                                CustomPlayer.Update_Pet_PowerL12(Power_L12P3)
                                var PetPower:float=1.0*Power_L12P3
                                set PetWon = ThirdPet
                                spawn:
                                    FollowPlayerL12(Fort, PetWon, PetPower, ThirdPetIndex)
                        else if (RandResult>750):
                            Print("Rare")
                            if(Cine:=Cine_L12_Rare[GetTeamIndexAgent[Agent]-1]):
                                Cine.Play(Agent)
                            Sleep(4.0)
                            if(Cine:=Cine_L12_Rare[GetTeamIndexAgent[Agent]-1]):
                                Cine.Stop(Agent)
                                Accolade_Egg11_12.Award(Agent)

                            if (CustomPlayer.Ownership_L12_Pet2=0 and CustomPlayer.Ownership_L12_Pet3=0 and CustomPlayer.Ownership_L12_Pet4=0):
                                #Release potential previous pets
                                if (CustomPlayer.Ownership_L12_Pet1=1):
                                    #--------------Button pressed:
                                    CustomPlayer.IncreaseButtonL12()
                                    #Print("Button L12 Pressed: {CustomPlayer.ButtonL12Pressed}")
                                    #--------------Spawn
                                    Sleep(0.5)
                                    ##spawn:
                                    ##    ReleasePet(FirstPet)
                                    set CustomPlayer.Ownership_L12_Pet1=0
                                    L12_Pet1_DesOwn(Agent)
                                #Get New pet
                                L12_Pet2_Own(Agent)
                                set CustomPlayer.Ownership_L12_Pet2=1
                                set CustomPlayer.Pet_Unlocked_L12=2
                                CustomPlayer.Update_Pet_PowerL12(Power_L12P2)
                                var PetPower:float=1.0*Power_L12P2
                                set PetWon = SecondPet
                                spawn:
                                    FollowPlayerL12(Fort, PetWon, PetPower, SecondPetIndex)
                        else:
                            Print("Normal")
                            if(Cine:=Cine_L12_Normal[GetTeamIndexAgent[Agent]-1]):
                                Cine.Play(Agent)
                            Sleep(4.0)
                            if(Cine:=Cine_L12_Normal[GetTeamIndexAgent[Agent]-1]):
                                Cine.Stop(Agent)
                                Accolade_Egg11_12.Award(Agent)

                            if (CustomPlayer.Ownership_L12_Pet1=0 and CustomPlayer.Ownership_L12_Pet2=0 and CustomPlayer.Ownership_L12_Pet3=0 and CustomPlayer.Ownership_L12_Pet4=0):
                                #Get New pet
                                L12_Pet1_Own(Agent)
                                set CustomPlayer.Ownership_L12_Pet1=1
                                set CustomPlayer.Pet_Unlocked_L12=1
                                CustomPlayer.Update_Pet_PowerL12(Power_L12P1)
                                var PetPower:float=1.0*Power_L12P1
                                set PetWon = FirstPet
                                spawn:
                                    FollowPlayerL12(Fort, PetWon, PetPower, FirstPetIndex)

    PurchasePetL13( Agent : agent )<suspends>: void =
        #Requires Unlock L13
        if:
            PlayerObj := player[Agent]
            CurrentPlayerStats := PlayerStatsManager.GetPlayerStats[Agent]
            CurrentPlayerStats.LevelUnlocked.CurrentValue<12
        then:
            RequireHigherLevelHUD.Show(Agent) 
        else:
            if:
                PlayerObj := player[Agent]
                CustomPlayer:Custom_Player = PlayersMap[PlayerObj]
                CustomPlayer.PlayerGold<CostL13
            then:
                NotEnoughGoldHUD.Show(Agent)
            else:
                DecreaseGoldFunc(Agent,CostL13)
                var FirstPet:creative_prop = creative_prop{}
                var SecondPet:creative_prop = creative_prop{}
                var ThirdPet:creative_prop = creative_prop{}
                var ForthPet:creative_prop = creative_prop{}

                var FirstPetIndex:int = 0
                var SecondPetIndex:int = 0
                var ThirdPetIndex:int = 0
                var ForthPetIndex:int = 0

                var IndexBase:int = 12*12

                #Select right pets for each player
                if(GetTeamIndexAgent[Agent]=1):
                    set FirstPetIndex =  IndexBase
                    set SecondPetIndex = IndexBase+1
                    set ThirdPetIndex =  IndexBase+2
                    set ForthPetIndex =  IndexBase+3
                if(GetTeamIndexAgent[Agent]=2):
                    set FirstPetIndex =  IndexBase+4
                    set SecondPetIndex = IndexBase+5
                    set ThirdPetIndex =  IndexBase+6
                    set ForthPetIndex =  IndexBase+7
                if(GetTeamIndexAgent[Agent]=3):
                    set FirstPetIndex =  IndexBase+8
                    set SecondPetIndex = IndexBase+9
                    set ThirdPetIndex =  IndexBase+10
                    set ForthPetIndex =  IndexBase+11

                if (Mypet:=Pets[FirstPetIndex]):
                    set FirstPet = Mypet.PetProp
                if (Mypet:=Pets[SecondPetIndex]):
                    set SecondPet = Mypet.PetProp
                if (Mypet:=Pets[ThirdPetIndex]):
                    set ThirdPet = Mypet.PetProp
                if (Mypet:=Pets[ForthPetIndex]):
                    set ForthPet = Mypet.PetProp

                var PetWon:creative_prop = creative_prop{}
                var RandResult:int = GetRandomInt(0,1000)
                Print("Rand: {RandResult}")
                if:
                    PlayerObj := player[Agent]
                    Fort := Agent.GetFortCharacter[]
                then:
                    if(CustomPlayer:Custom_Player = PlayersMap[PlayerObj]):
                        #--------------Luck
                        set RandResult+=CustomPlayer.Pet_Luck_Value*5+CustomPlayer.Bonus_Luck*5
                        #--------------
                        if (RandResult>995):
                            Print("Legendary")
                            if(Cine:=Cine_L13_Legendary[GetTeamIndexAgent[Agent]-1]):
                                Cine.Play(Agent)
                            Sleep(4.0)
                            if(Cine:=Cine_L13_Legendary[GetTeamIndexAgent[Agent]-1]):
                                Cine.Stop(Agent)
                                Accolade_Egg13_14.Award(Agent)

                            if (CustomPlayer.Ownership_L13_Pet4=0):
                                #Release potential previous pets
                                if (CustomPlayer.Ownership_L13_Pet1=1):
                                    #--------------Button pressed:
                                    CustomPlayer.IncreaseButtonL13()
                                    #Print("Button L13 Pressed: {CustomPlayer.ButtonL13Pressed}")
                                    #--------------Spawn
                                    Sleep(0.5)
                                    ##spawn:
                                    ##    ReleasePet(FirstPet)
                                    set CustomPlayer.Ownership_L13_Pet1=0
                                    L13_Pet1_DesOwn(Agent)
                                if (CustomPlayer.Ownership_L13_Pet2=1):
                                    #--------------Button pressed:
                                    CustomPlayer.IncreaseButtonL13()
                                    #Print("Button L13 Pressed: {CustomPlayer.ButtonL13Pressed}")
                                    #--------------Spawn
                                    Sleep(0.5)
                                    ##spawn:
                                    ##    ReleasePet(SecondPet)
                                    set CustomPlayer.Ownership_L13_Pet2=0
                                    L13_Pet2_DesOwn(Agent)
                                if (CustomPlayer.Ownership_L13_Pet3=1):
                                    #--------------Button pressed:
                                    CustomPlayer.IncreaseButtonL13()
                                    #Print("Button L13 Pressed: {CustomPlayer.ButtonL13Pressed}")
                                    #--------------Spawn
                                    Sleep(0.5)
                                    ##spawn:
                                    ##    ReleasePet(ThirdPet)
                                    set CustomPlayer.Ownership_L13_Pet3=0
                                    L13_Pet3_DesOwn(Agent)
                                #Get New pet
                                L13_Pet4_Own(Agent)
                                set CustomPlayer.Ownership_L13_Pet4=1
                                set CustomPlayer.Pet_Unlocked_L13=4
                                CustomPlayer.Update_Pet_PowerL13(Power_L13P4)
                                var PetPower:float=1.0*Power_L13P4
                                set PetWon = ForthPet
                                spawn:
                                    FollowPlayerL13(Fort, PetWon, PetPower, ForthPetIndex)
                        else if (RandResult>855):
                            Print("Epic")
                            if(Cine:=Cine_L13_Epic[GetTeamIndexAgent[Agent]-1]):
                                Cine.Play(Agent)
                            Sleep(4.0)
                            if(Cine:=Cine_L13_Epic[GetTeamIndexAgent[Agent]-1]):
                                Cine.Stop(Agent)
                                Accolade_Egg13_14.Award(Agent)

                            if (CustomPlayer.Ownership_L13_Pet3=0 and CustomPlayer.Ownership_L13_Pet4=0):
                                #Release potential previous pets
                                if (CustomPlayer.Ownership_L13_Pet1=1):
                                    #--------------Button pressed:
                                    CustomPlayer.IncreaseButtonL13()
                                    #Print("Button L13 Pressed: {CustomPlayer.ButtonL13Pressed}")
                                    #--------------Spawn
                                    Sleep(0.5)
                                    ##spawn:
                                    ##    ReleasePet(FirstPet)
                                    set CustomPlayer.Ownership_L13_Pet1=0
                                    L13_Pet1_DesOwn(Agent)
                                if (CustomPlayer.Ownership_L13_Pet2=1):
                                    #--------------Button pressed:
                                    CustomPlayer.IncreaseButtonL13()
                                    #Print("Button L13 Pressed: {CustomPlayer.ButtonL13Pressed}")
                                    #--------------Spawn
                                    Sleep(0.5)
                                    ##spawn:
                                    ##    ReleasePet(SecondPet)
                                    set CustomPlayer.Ownership_L13_Pet2=0
                                    L13_Pet2_DesOwn(Agent)
                                #Get New pet
                                L13_Pet3_Own(Agent)
                                set CustomPlayer.Ownership_L13_Pet3=1
                                set CustomPlayer.Pet_Unlocked_L13=3
                                CustomPlayer.Update_Pet_PowerL13(Power_L13P3)
                                var PetPower:float=1.0*Power_L13P3
                                set PetWon = ThirdPet
                                spawn:
                                    FollowPlayerL13(Fort, PetWon, PetPower, ThirdPetIndex)
                        else if (RandResult>750):
                            Print("Rare")
                            if(Cine:=Cine_L13_Rare[GetTeamIndexAgent[Agent]-1]):
                                Cine.Play(Agent)
                            Sleep(4.0)
                            if(Cine:=Cine_L13_Rare[GetTeamIndexAgent[Agent]-1]):
                                Cine.Stop(Agent)
                                Accolade_Egg13_14.Award(Agent)

                            if (CustomPlayer.Ownership_L13_Pet2=0 and CustomPlayer.Ownership_L13_Pet3=0 and CustomPlayer.Ownership_L13_Pet4=0):
                                #Release potential previous pets
                                if (CustomPlayer.Ownership_L13_Pet1=1):
                                    #--------------Button pressed:
                                    CustomPlayer.IncreaseButtonL13()
                                    #Print("Button L13 Pressed: {CustomPlayer.ButtonL13Pressed}")
                                    #--------------Spawn
                                    Sleep(0.5)
                                    ##spawn:
                                    ##    ReleasePet(FirstPet)
                                    set CustomPlayer.Ownership_L13_Pet1=0
                                    L13_Pet1_DesOwn(Agent)
                                #Get New pet
                                L13_Pet2_Own(Agent)
                                set CustomPlayer.Ownership_L13_Pet2=1
                                set CustomPlayer.Pet_Unlocked_L13=2
                                CustomPlayer.Update_Pet_PowerL13(Power_L13P2)
                                var PetPower:float=1.0*Power_L13P2
                                set PetWon = SecondPet
                                spawn:
                                    FollowPlayerL13(Fort, PetWon, PetPower, SecondPetIndex)
                        else:
                            Print("Normal")
                            if(Cine:=Cine_L13_Normal[GetTeamIndexAgent[Agent]-1]):
                                Cine.Play(Agent)
                            Sleep(4.0)
                            if(Cine:=Cine_L13_Normal[GetTeamIndexAgent[Agent]-1]):
                                Cine.Stop(Agent)
                                Accolade_Egg13_14.Award(Agent)

                            if (CustomPlayer.Ownership_L13_Pet1=0 and CustomPlayer.Ownership_L13_Pet2=0 and CustomPlayer.Ownership_L13_Pet3=0 and CustomPlayer.Ownership_L13_Pet4=0):
                                #Get New pet
                                L13_Pet1_Own(Agent)
                                set CustomPlayer.Ownership_L13_Pet1=1
                                set CustomPlayer.Pet_Unlocked_L13=1
                                CustomPlayer.Update_Pet_PowerL13(Power_L13P1)
                                var PetPower:float=1.0*Power_L13P1
                                set PetWon = FirstPet
                                spawn:
                                    FollowPlayerL13(Fort, PetWon, PetPower, FirstPetIndex)

    PurchasePetL14( Agent : agent )<suspends>: void =
        #Requires Unlock L14
        if:
            PlayerObj := player[Agent]
            CurrentPlayerStats := PlayerStatsManager.GetPlayerStats[Agent]
            CurrentPlayerStats.LevelUnlocked.CurrentValue<13
        then:
            RequireHigherLevelHUD.Show(Agent) 
        else:
            if:
                PlayerObj := player[Agent]
                CustomPlayer:Custom_Player = PlayersMap[PlayerObj]
                CustomPlayer.PlayerGold<CostL14
            then:
                NotEnoughGoldHUD.Show(Agent)
            else:
                DecreaseGoldFunc(Agent,CostL14)
                var FirstPet:creative_prop = creative_prop{}
                var SecondPet:creative_prop = creative_prop{}
                var ThirdPet:creative_prop = creative_prop{}
                var ForthPet:creative_prop = creative_prop{}

                var FirstPetIndex:int = 0
                var SecondPetIndex:int = 0
                var ThirdPetIndex:int = 0
                var ForthPetIndex:int = 0

                var IndexBase:int = 12*13

                #Select right pets for each player
                if(GetTeamIndexAgent[Agent]=1):
                    set FirstPetIndex =  IndexBase
                    set SecondPetIndex = IndexBase+1
                    set ThirdPetIndex =  IndexBase+2
                    set ForthPetIndex =  IndexBase+3
                if(GetTeamIndexAgent[Agent]=2):
                    set FirstPetIndex =  IndexBase+4
                    set SecondPetIndex = IndexBase+5
                    set ThirdPetIndex =  IndexBase+6
                    set ForthPetIndex =  IndexBase+7
                if(GetTeamIndexAgent[Agent]=3):
                    set FirstPetIndex =  IndexBase+8
                    set SecondPetIndex = IndexBase+9
                    set ThirdPetIndex =  IndexBase+10
                    set ForthPetIndex =  IndexBase+11

                if (Mypet:=Pets[FirstPetIndex]):
                    set FirstPet = Mypet.PetProp
                if (Mypet:=Pets[SecondPetIndex]):
                    set SecondPet = Mypet.PetProp
                if (Mypet:=Pets[ThirdPetIndex]):
                    set ThirdPet = Mypet.PetProp
                if (Mypet:=Pets[ForthPetIndex]):
                    set ForthPet = Mypet.PetProp

                var PetWon:creative_prop = creative_prop{}
                var RandResult:int = GetRandomInt(0,1000)
                Print("Rand: {RandResult}")
                if:
                    PlayerObj := player[Agent]
                    Fort := Agent.GetFortCharacter[]
                then:
                    if(CustomPlayer:Custom_Player = PlayersMap[PlayerObj]):
                        #--------------Luck
                        set RandResult+=CustomPlayer.Pet_Luck_Value*5+CustomPlayer.Bonus_Luck*5
                        #--------------
                        if (RandResult>995):
                            Print("Legendary")
                            if(Cine:=Cine_L14_Legendary[GetTeamIndexAgent[Agent]-1]):
                                Cine.Play(Agent)
                            Sleep(4.0)
                            if(Cine:=Cine_L14_Legendary[GetTeamIndexAgent[Agent]-1]):
                                Cine.Stop(Agent)
                                Accolade_Egg13_14.Award(Agent)

                            if (CustomPlayer.Ownership_L14_Pet4=0):
                                #Release potential previous pets
                                if (CustomPlayer.Ownership_L14_Pet1=1):
                                    #--------------Button pressed:
                                    CustomPlayer.IncreaseButtonL14()
                                    #Print("Button L14 Pressed: {CustomPlayer.ButtonL14Pressed}")
                                    #--------------Spawn
                                    Sleep(0.5)
                                    ##spawn:
                                    ##    ReleasePet(FirstPet)
                                    set CustomPlayer.Ownership_L14_Pet1=0
                                    L14_Pet1_DesOwn(Agent)
                                if (CustomPlayer.Ownership_L14_Pet2=1):
                                    #--------------Button pressed:
                                    CustomPlayer.IncreaseButtonL14()
                                    #Print("Button L14 Pressed: {CustomPlayer.ButtonL14Pressed}")
                                    #--------------Spawn
                                    Sleep(0.5)
                                    ##spawn:
                                    ##    ReleasePet(SecondPet)
                                    set CustomPlayer.Ownership_L14_Pet2=0
                                    L14_Pet2_DesOwn(Agent)
                                if (CustomPlayer.Ownership_L14_Pet3=1):
                                    #--------------Button pressed:
                                    CustomPlayer.IncreaseButtonL14()
                                    #Print("Button L14 Pressed: {CustomPlayer.ButtonL14Pressed}")
                                    #--------------Spawn
                                    Sleep(0.5)
                                    ##spawn:
                                    ##    ReleasePet(ThirdPet)
                                    set CustomPlayer.Ownership_L14_Pet3=0
                                    L14_Pet3_DesOwn(Agent)
                                #Get New pet
                                L14_Pet4_Own(Agent)
                                set CustomPlayer.Ownership_L14_Pet4=1
                                set CustomPlayer.Pet_Unlocked_L14=4
                                CustomPlayer.Update_Pet_PowerL14(Power_L14P4)
                                var PetPower:float=1.0*Power_L14P4
                                set PetWon = ForthPet
                                spawn:
                                    FollowPlayerL14(Fort, PetWon, PetPower, ForthPetIndex)
                        else if (RandResult>855):
                            Print("Epic")
                            if(Cine:=Cine_L14_Epic[GetTeamIndexAgent[Agent]-1]):
                                Cine.Play(Agent)
                            Sleep(4.0)
                            if(Cine:=Cine_L14_Epic[GetTeamIndexAgent[Agent]-1]):
                                Cine.Stop(Agent)
                                Accolade_Egg13_14.Award(Agent)

                            if (CustomPlayer.Ownership_L14_Pet3=0 and CustomPlayer.Ownership_L14_Pet4=0):
                                #Release potential previous pets
                                if (CustomPlayer.Ownership_L14_Pet1=1):
                                    #--------------Button pressed:
                                    CustomPlayer.IncreaseButtonL14()
                                    #Print("Button L14 Pressed: {CustomPlayer.ButtonL14Pressed}")
                                    #--------------Spawn
                                    Sleep(0.5)
                                    ##spawn:
                                    ##    ReleasePet(FirstPet)
                                    set CustomPlayer.Ownership_L14_Pet1=0
                                    L14_Pet1_DesOwn(Agent)
                                if (CustomPlayer.Ownership_L14_Pet2=1):
                                    #--------------Button pressed:
                                    CustomPlayer.IncreaseButtonL14()
                                    #Print("Button L14 Pressed: {CustomPlayer.ButtonL14Pressed}")
                                    #--------------Spawn
                                    Sleep(0.5)
                                    ##spawn:
                                    ##    ReleasePet(SecondPet)
                                    set CustomPlayer.Ownership_L14_Pet2=0
                                    L14_Pet2_DesOwn(Agent)
                                #Get New pet
                                L14_Pet3_Own(Agent)
                                set CustomPlayer.Ownership_L14_Pet3=1
                                set CustomPlayer.Pet_Unlocked_L14=3
                                CustomPlayer.Update_Pet_PowerL14(Power_L14P3)
                                var PetPower:float=1.0*Power_L14P3
                                set PetWon = ThirdPet
                                spawn:
                                    FollowPlayerL14(Fort, PetWon, PetPower, ThirdPetIndex)
                        else if (RandResult>750):
                            Print("Rare")
                            if(Cine:=Cine_L14_Rare[GetTeamIndexAgent[Agent]-1]):
                                Cine.Play(Agent)
                            Sleep(4.0)
                            if(Cine:=Cine_L14_Rare[GetTeamIndexAgent[Agent]-1]):
                                Cine.Stop(Agent)
                                Accolade_Egg13_14.Award(Agent)

                            if (CustomPlayer.Ownership_L14_Pet2=0 and CustomPlayer.Ownership_L14_Pet3=0 and CustomPlayer.Ownership_L14_Pet4=0):
                                #Release potential previous pets
                                if (CustomPlayer.Ownership_L14_Pet1=1):
                                    #--------------Button pressed:
                                    CustomPlayer.IncreaseButtonL14()
                                    #Print("Button L14 Pressed: {CustomPlayer.ButtonL14Pressed}")
                                    #--------------Spawn
                                    Sleep(0.5)
                                    ##spawn:
                                    ##    ReleasePet(FirstPet)
                                    set CustomPlayer.Ownership_L14_Pet1=0
                                    L14_Pet1_DesOwn(Agent)
                                #Get New pet
                                L14_Pet2_Own(Agent)
                                set CustomPlayer.Ownership_L14_Pet2=1
                                set CustomPlayer.Pet_Unlocked_L14=2
                                CustomPlayer.Update_Pet_PowerL14(Power_L14P2)
                                var PetPower:float=1.0*Power_L14P2
                                set PetWon = SecondPet
                                spawn:
                                    FollowPlayerL14(Fort, PetWon, PetPower, SecondPetIndex)
                        else:
                            Print("Normal")
                            if(Cine:=Cine_L14_Normal[GetTeamIndexAgent[Agent]-1]):
                                Cine.Play(Agent)
                            Sleep(4.0)
                            if(Cine:=Cine_L14_Normal[GetTeamIndexAgent[Agent]-1]):
                                Cine.Stop(Agent)
                                Accolade_Egg13_14.Award(Agent)

                            if (CustomPlayer.Ownership_L14_Pet1=0 and CustomPlayer.Ownership_L14_Pet2=0 and CustomPlayer.Ownership_L14_Pet3=0 and CustomPlayer.Ownership_L14_Pet4=0):
                                #Get New pet
                                L14_Pet1_Own(Agent)
                                set CustomPlayer.Ownership_L14_Pet1=1
                                set CustomPlayer.Pet_Unlocked_L14=1
                                CustomPlayer.Update_Pet_PowerL14(Power_L14P1)
                                var PetPower:float=1.0*Power_L14P1
                                set PetWon = FirstPet
                                spawn:
                                    FollowPlayerL14(Fort, PetWon, PetPower, FirstPetIndex)

    PurchasePetL15( Agent : agent )<suspends>: void =
        #Requires Unlock L15
        if:
            PlayerObj := player[Agent]
            CurrentPlayerStats := PlayerStatsManager.GetPlayerStats[Agent]
            CurrentPlayerStats.LevelUnlocked.CurrentValue<14
        then:
            RequireHigherLevelHUD.Show(Agent) 
        else:
            if:
                PlayerObj := player[Agent]
                CustomPlayer:Custom_Player = PlayersMap[PlayerObj]
                CustomPlayer.PlayerGold<CostL15
            then:
                NotEnoughGoldHUD.Show(Agent)
            else:
                DecreaseGoldFunc(Agent,CostL15)
                var FirstPet:creative_prop = creative_prop{}
                var SecondPet:creative_prop = creative_prop{}
                var ThirdPet:creative_prop = creative_prop{}
                var ForthPet:creative_prop = creative_prop{}

                var FirstPetIndex:int = 0
                var SecondPetIndex:int = 0
                var ThirdPetIndex:int = 0
                var ForthPetIndex:int = 0

                var IndexBase:int = 12*14

                #Select right pets for each player
                if(GetTeamIndexAgent[Agent]=1):
                    set FirstPetIndex =  IndexBase
                    set SecondPetIndex = IndexBase+1
                    set ThirdPetIndex =  IndexBase+2
                    set ForthPetIndex =  IndexBase+3
                if(GetTeamIndexAgent[Agent]=2):
                    set FirstPetIndex =  IndexBase+4
                    set SecondPetIndex = IndexBase+5
                    set ThirdPetIndex =  IndexBase+6
                    set ForthPetIndex =  IndexBase+7
                if(GetTeamIndexAgent[Agent]=3):
                    set FirstPetIndex =  IndexBase+8
                    set SecondPetIndex = IndexBase+9
                    set ThirdPetIndex =  IndexBase+10
                    set ForthPetIndex =  IndexBase+11

                if (Mypet:=Pets[FirstPetIndex]):
                    set FirstPet = Mypet.PetProp
                if (Mypet:=Pets[SecondPetIndex]):
                    set SecondPet = Mypet.PetProp
                if (Mypet:=Pets[ThirdPetIndex]):
                    set ThirdPet = Mypet.PetProp
                if (Mypet:=Pets[ForthPetIndex]):
                    set ForthPet = Mypet.PetProp

                var PetWon:creative_prop = creative_prop{}
                var RandResult:int = GetRandomInt(0,1000)
                Print("Rand: {RandResult}")
                if:
                    PlayerObj := player[Agent]
                    Fort := Agent.GetFortCharacter[]
                then:
                    if(CustomPlayer:Custom_Player = PlayersMap[PlayerObj]):
                        #--------------Luck
                        set RandResult+=CustomPlayer.Pet_Luck_Value*5+CustomPlayer.Bonus_Luck*5
                        #--------------
                        if (RandResult>995):
                            Print("Legendary")
                            if(Cine:=Cine_L15_Legendary[GetTeamIndexAgent[Agent]-1]):
                                Cine.Play(Agent)
                            Sleep(4.0)
                            if(Cine:=Cine_L15_Legendary[GetTeamIndexAgent[Agent]-1]):
                                Cine.Stop(Agent)
                                Accolade_Egg15_16.Award(Agent)

                            if (CustomPlayer.Ownership_L15_Pet4=0):
                                #Release potential previous pets
                                if (CustomPlayer.Ownership_L15_Pet1=1):
                                    #--------------Button pressed:
                                    CustomPlayer.IncreaseButtonL15()
                                    #Print("Button L15 Pressed: {CustomPlayer.ButtonL15Pressed}")
                                    #--------------Spawn
                                    Sleep(0.5)
                                    ##spawn:
                                    ##    ReleasePet(FirstPet)
                                    set CustomPlayer.Ownership_L15_Pet1=0
                                    L15_Pet1_DesOwn(Agent)
                                if (CustomPlayer.Ownership_L15_Pet2=1):
                                    #--------------Button pressed:
                                    CustomPlayer.IncreaseButtonL15()
                                    #Print("Button L15 Pressed: {CustomPlayer.ButtonL15Pressed}")
                                    #--------------Spawn
                                    Sleep(0.5)
                                    ##spawn:
                                    ##    ReleasePet(SecondPet)
                                    set CustomPlayer.Ownership_L15_Pet2=0
                                    L15_Pet2_DesOwn(Agent)
                                if (CustomPlayer.Ownership_L15_Pet3=1):
                                    #--------------Button pressed:
                                    CustomPlayer.IncreaseButtonL15()
                                    #Print("Button L15 Pressed: {CustomPlayer.ButtonL15Pressed}")
                                    #--------------Spawn
                                    Sleep(0.5)
                                    ##spawn:
                                    ##    ReleasePet(ThirdPet)
                                    set CustomPlayer.Ownership_L15_Pet3=0
                                    L15_Pet3_DesOwn(Agent)
                                #Get New pet
                                L15_Pet4_Own(Agent)
                                set CustomPlayer.Ownership_L15_Pet4=1
                                set CustomPlayer.Pet_Unlocked_L15=4
                                CustomPlayer.Update_Pet_PowerL15(Power_L15P4)
                                var PetPower:float=1.0*Power_L15P4
                                set PetWon = ForthPet
                                spawn:
                                    FollowPlayerL15(Fort, PetWon, PetPower, ForthPetIndex)
                        else if (RandResult>855):
                            Print("Epic")
                            if(Cine:=Cine_L15_Epic[GetTeamIndexAgent[Agent]-1]):
                                Cine.Play(Agent)
                            Sleep(4.0)
                            if(Cine:=Cine_L15_Epic[GetTeamIndexAgent[Agent]-1]):
                                Cine.Stop(Agent)
                                Accolade_Egg15_16.Award(Agent)

                            if (CustomPlayer.Ownership_L15_Pet3=0 and CustomPlayer.Ownership_L15_Pet4=0):
                                #Release potential previous pets
                                if (CustomPlayer.Ownership_L15_Pet1=1):
                                    #--------------Button pressed:
                                    CustomPlayer.IncreaseButtonL15()
                                    #Print("Button L15 Pressed: {CustomPlayer.ButtonL15Pressed}")
                                    #--------------Spawn
                                    Sleep(0.5)
                                    ##spawn:
                                    ##    ReleasePet(FirstPet)
                                    set CustomPlayer.Ownership_L15_Pet1=0
                                    L15_Pet1_DesOwn(Agent)
                                if (CustomPlayer.Ownership_L15_Pet2=1):
                                    #--------------Button pressed:
                                    CustomPlayer.IncreaseButtonL15()
                                    #Print("Button L15 Pressed: {CustomPlayer.ButtonL15Pressed}")
                                    #--------------Spawn
                                    Sleep(0.5)
                                    ##spawn:
                                    ##    ReleasePet(SecondPet)
                                    set CustomPlayer.Ownership_L15_Pet2=0
                                    L15_Pet2_DesOwn(Agent)
                                #Get New pet
                                L15_Pet3_Own(Agent)
                                set CustomPlayer.Ownership_L15_Pet3=1
                                set CustomPlayer.Pet_Unlocked_L15=3
                                CustomPlayer.Update_Pet_PowerL15(Power_L15P3)
                                var PetPower:float=1.0*Power_L15P3
                                set PetWon = ThirdPet
                                spawn:
                                    FollowPlayerL15(Fort, PetWon, PetPower, ThirdPetIndex)
                        else if (RandResult>750):
                            Print("Rare")
                            if(Cine:=Cine_L15_Rare[GetTeamIndexAgent[Agent]-1]):
                                Cine.Play(Agent)
                            Sleep(4.0)
                            if(Cine:=Cine_L15_Rare[GetTeamIndexAgent[Agent]-1]):
                                Cine.Stop(Agent)
                                Accolade_Egg15_16.Award(Agent)

                            if (CustomPlayer.Ownership_L15_Pet2=0 and CustomPlayer.Ownership_L15_Pet3=0 and CustomPlayer.Ownership_L15_Pet4=0):
                                #Release potential previous pets
                                if (CustomPlayer.Ownership_L15_Pet1=1):
                                    #--------------Button pressed:
                                    CustomPlayer.IncreaseButtonL15()
                                    #Print("Button L15 Pressed: {CustomPlayer.ButtonL15Pressed}")
                                    #--------------Spawn
                                    Sleep(0.5)
                                    ##spawn:
                                    ##    ReleasePet(FirstPet)
                                    set CustomPlayer.Ownership_L15_Pet1=0
                                    L15_Pet1_DesOwn(Agent)
                                #Get New pet
                                L15_Pet2_Own(Agent)
                                set CustomPlayer.Ownership_L15_Pet2=1
                                set CustomPlayer.Pet_Unlocked_L15=2
                                CustomPlayer.Update_Pet_PowerL15(Power_L15P2)
                                var PetPower:float=1.0*Power_L15P2
                                set PetWon = SecondPet
                                spawn:
                                    FollowPlayerL15(Fort, PetWon, PetPower, SecondPetIndex)
                        else:
                            Print("Normal")
                            if(Cine:=Cine_L15_Normal[GetTeamIndexAgent[Agent]-1]):
                                Cine.Play(Agent)
                            Sleep(4.0)
                            if(Cine:=Cine_L15_Normal[GetTeamIndexAgent[Agent]-1]):
                                Cine.Stop(Agent)
                                Accolade_Egg15_16.Award(Agent)

                            if (CustomPlayer.Ownership_L15_Pet1=0 and CustomPlayer.Ownership_L15_Pet2=0 and CustomPlayer.Ownership_L15_Pet3=0 and CustomPlayer.Ownership_L15_Pet4=0):
                                #Get New pet
                                L15_Pet1_Own(Agent)
                                set CustomPlayer.Ownership_L15_Pet1=1
                                set CustomPlayer.Pet_Unlocked_L15=1
                                CustomPlayer.Update_Pet_PowerL15(Power_L15P1)
                                var PetPower:float=1.0*Power_L15P1
                                set PetWon = FirstPet
                                spawn:
                                    FollowPlayerL15(Fort, PetWon, PetPower, FirstPetIndex)

    PurchasePetL16( Agent : agent )<suspends>: void =
        #Requires Unlock L16
        if:
            PlayerObj := player[Agent]
            CurrentPlayerStats := PlayerStatsManager.GetPlayerStats[Agent]
            CurrentPlayerStats.LevelUnlocked.CurrentValue<15
        then:
            RequireHigherLevelHUD.Show(Agent) 
        else:
            if:
                PlayerObj := player[Agent]
                CustomPlayer:Custom_Player = PlayersMap[PlayerObj]
                CustomPlayer.PlayerGold<CostL16
            then:
                NotEnoughGoldHUD.Show(Agent)
            else:
                DecreaseGoldFunc(Agent,CostL16)
                var FirstPet:creative_prop = creative_prop{}
                var SecondPet:creative_prop = creative_prop{}
                var ThirdPet:creative_prop = creative_prop{}
                var ForthPet:creative_prop = creative_prop{}

                var FirstPetIndex:int = 0
                var SecondPetIndex:int = 0
                var ThirdPetIndex:int = 0
                var ForthPetIndex:int = 0

                var IndexBase:int = 12*15

                #Select right pets for each player
                if(GetTeamIndexAgent[Agent]=1):
                    set FirstPetIndex =  IndexBase
                    set SecondPetIndex = IndexBase+1
                    set ThirdPetIndex =  IndexBase+2
                    set ForthPetIndex =  IndexBase+3
                if(GetTeamIndexAgent[Agent]=2):
                    set FirstPetIndex =  IndexBase+4
                    set SecondPetIndex = IndexBase+5
                    set ThirdPetIndex =  IndexBase+6
                    set ForthPetIndex =  IndexBase+7
                if(GetTeamIndexAgent[Agent]=3):
                    set FirstPetIndex =  IndexBase+8
                    set SecondPetIndex = IndexBase+9
                    set ThirdPetIndex =  IndexBase+10
                    set ForthPetIndex =  IndexBase+11

                if (Mypet:=Pets[FirstPetIndex]):
                    set FirstPet = Mypet.PetProp
                if (Mypet:=Pets[SecondPetIndex]):
                    set SecondPet = Mypet.PetProp
                if (Mypet:=Pets[ThirdPetIndex]):
                    set ThirdPet = Mypet.PetProp
                if (Mypet:=Pets[ForthPetIndex]):
                    set ForthPet = Mypet.PetProp

                var PetWon:creative_prop = creative_prop{}
                var RandResult:int = GetRandomInt(0,1000)
                Print("Rand: {RandResult}")
                if:
                    PlayerObj := player[Agent]
                    Fort := Agent.GetFortCharacter[]
                then:
                    if(CustomPlayer:Custom_Player = PlayersMap[PlayerObj]):
                        #--------------Luck
                        set RandResult+=CustomPlayer.Pet_Luck_Value*5+CustomPlayer.Bonus_Luck*5
                        #--------------
                        if (RandResult>995):
                            Print("Legendary")
                            if(Cine:=Cine_L16_Legendary[GetTeamIndexAgent[Agent]-1]):
                                Cine.Play(Agent)
                            Sleep(4.0)
                            if(Cine:=Cine_L16_Legendary[GetTeamIndexAgent[Agent]-1]):
                                Cine.Stop(Agent)
                                Accolade_Egg15_16.Award(Agent)

                            if (CustomPlayer.Ownership_L16_Pet4=0):
                                #Release potential previous pets
                                if (CustomPlayer.Ownership_L16_Pet1=1):
                                    #--------------Button pressed:
                                    CustomPlayer.IncreaseButtonL16()
                                    #Print("Button L16 Pressed: {CustomPlayer.ButtonL16Pressed}")
                                    #--------------Spawn
                                    Sleep(0.5)
                                    ##spawn:
                                    ##    ReleasePet(FirstPet)
                                    set CustomPlayer.Ownership_L16_Pet1=0
                                    L16_Pet1_DesOwn(Agent)
                                if (CustomPlayer.Ownership_L16_Pet2=1):
                                    #--------------Button pressed:
                                    CustomPlayer.IncreaseButtonL16()
                                    #Print("Button L16 Pressed: {CustomPlayer.ButtonL16Pressed}")
                                    #--------------Spawn
                                    Sleep(0.5)
                                    ##spawn:
                                    ##    ReleasePet(SecondPet)
                                    set CustomPlayer.Ownership_L16_Pet2=0
                                    L16_Pet2_DesOwn(Agent)
                                if (CustomPlayer.Ownership_L16_Pet3=1):
                                    #--------------Button pressed:
                                    CustomPlayer.IncreaseButtonL16()
                                    #Print("Button L16 Pressed: {CustomPlayer.ButtonL16Pressed}")
                                    #--------------Spawn
                                    Sleep(0.5)
                                    ##spawn:
                                    ##    ReleasePet(ThirdPet)
                                    set CustomPlayer.Ownership_L16_Pet3=0
                                    L16_Pet3_DesOwn(Agent)
                                #Get New pet
                                L16_Pet4_Own(Agent)
                                set CustomPlayer.Ownership_L16_Pet4=1
                                set CustomPlayer.Pet_Unlocked_L16=4
                                CustomPlayer.Update_Pet_PowerL16(Power_L16P4)
                                var PetPower:float=1.0*Power_L16P4
                                set PetWon = ForthPet
                                spawn:
                                    FollowPlayerL16(Fort, PetWon, PetPower, ForthPetIndex)
                        else if (RandResult>855):
                            Print("Epic")
                            if(Cine:=Cine_L16_Epic[GetTeamIndexAgent[Agent]-1]):
                                Cine.Play(Agent)
                            Sleep(4.0)
                            if(Cine:=Cine_L16_Epic[GetTeamIndexAgent[Agent]-1]):
                                Cine.Stop(Agent)
                                Accolade_Egg15_16.Award(Agent)

                            if (CustomPlayer.Ownership_L16_Pet3=0 and CustomPlayer.Ownership_L16_Pet4=0):
                                #Release potential previous pets
                                if (CustomPlayer.Ownership_L16_Pet1=1):
                                    #--------------Button pressed:
                                    CustomPlayer.IncreaseButtonL16()
                                    #Print("Button L16 Pressed: {CustomPlayer.ButtonL16Pressed}")
                                    #--------------Spawn
                                    Sleep(0.5)
                                    ##spawn:
                                    ##    ReleasePet(FirstPet)
                                    set CustomPlayer.Ownership_L16_Pet1=0
                                    L16_Pet1_DesOwn(Agent)
                                if (CustomPlayer.Ownership_L16_Pet2=1):
                                    #--------------Button pressed:
                                    CustomPlayer.IncreaseButtonL16()
                                    #Print("Button L16 Pressed: {CustomPlayer.ButtonL16Pressed}")
                                    #--------------Spawn
                                    Sleep(0.5)
                                    ##spawn:
                                    ##    ReleasePet(SecondPet)
                                    set CustomPlayer.Ownership_L16_Pet2=0
                                    L16_Pet2_DesOwn(Agent)
                                #Get New pet
                                L16_Pet3_Own(Agent)
                                set CustomPlayer.Ownership_L16_Pet3=1
                                set CustomPlayer.Pet_Unlocked_L16=3
                                CustomPlayer.Update_Pet_PowerL16(Power_L16P3)
                                var PetPower:float=1.0*Power_L16P3
                                set PetWon = ThirdPet
                                spawn:
                                    FollowPlayerL16(Fort, PetWon, PetPower, ThirdPetIndex)
                        else if (RandResult>750):
                            Print("Rare")
                            if(Cine:=Cine_L16_Rare[GetTeamIndexAgent[Agent]-1]):
                                Cine.Play(Agent)
                            Sleep(4.0)
                            if(Cine:=Cine_L16_Rare[GetTeamIndexAgent[Agent]-1]):
                                Cine.Stop(Agent)
                                Accolade_Egg15_16.Award(Agent)

                            if (CustomPlayer.Ownership_L16_Pet2=0 and CustomPlayer.Ownership_L16_Pet3=0 and CustomPlayer.Ownership_L16_Pet4=0):
                                #Release potential previous pets
                                if (CustomPlayer.Ownership_L16_Pet1=1):
                                    #--------------Button pressed:
                                    CustomPlayer.IncreaseButtonL16()
                                    #Print("Button L16 Pressed: {CustomPlayer.ButtonL16Pressed}")
                                    #--------------Spawn
                                    Sleep(0.5)
                                    ##spawn:
                                    ##    ReleasePet(FirstPet)
                                    set CustomPlayer.Ownership_L16_Pet1=0
                                    L16_Pet1_DesOwn(Agent)
                                #Get New pet
                                L16_Pet2_Own(Agent)
                                set CustomPlayer.Ownership_L16_Pet2=1
                                set CustomPlayer.Pet_Unlocked_L16=2
                                CustomPlayer.Update_Pet_PowerL16(Power_L16P2)
                                var PetPower:float=1.0*Power_L16P2
                                set PetWon = SecondPet
                                spawn:
                                    FollowPlayerL16(Fort, PetWon, PetPower, SecondPetIndex)
                        else:
                            Print("Normal")
                            if(Cine:=Cine_L16_Normal[GetTeamIndexAgent[Agent]-1]):
                                Cine.Play(Agent)
                            Sleep(4.0)
                            if(Cine:=Cine_L16_Normal[GetTeamIndexAgent[Agent]-1]):
                                Cine.Stop(Agent)
                                Accolade_Egg15_16.Award(Agent)

                            if (CustomPlayer.Ownership_L16_Pet1=0 and CustomPlayer.Ownership_L16_Pet2=0 and CustomPlayer.Ownership_L16_Pet3=0 and CustomPlayer.Ownership_L16_Pet4=0):
                                #Get New pet
                                L16_Pet1_Own(Agent)
                                set CustomPlayer.Ownership_L16_Pet1=1
                                set CustomPlayer.Pet_Unlocked_L16=1
                                CustomPlayer.Update_Pet_PowerL16(Power_L16P1)
                                var PetPower:float=1.0*Power_L16P1
                                set PetWon = FirstPet
                                spawn:
                                    FollowPlayerL16(Fort, PetWon, PetPower, FirstPetIndex)

    PurchasePetL17( Agent : agent )<suspends>: void =
        #Requires Unlock L17
        if:
            PlayerObj := player[Agent]
            CurrentPlayerStats := PlayerStatsManager.GetPlayerStats[Agent]
            CurrentPlayerStats.LevelUnlocked.CurrentValue<16
        then:
            RequireHigherLevelHUD.Show(Agent) 
        else:
            if:
                PlayerObj := player[Agent]
                CustomPlayer:Custom_Player = PlayersMap[PlayerObj]
                CustomPlayer.PlayerGold<CostL17
            then:
                NotEnoughGoldHUD.Show(Agent)
            else:
                DecreaseGoldFunc(Agent,CostL17)
                var FirstPet:creative_prop = creative_prop{}
                var SecondPet:creative_prop = creative_prop{}
                var ThirdPet:creative_prop = creative_prop{}
                var ForthPet:creative_prop = creative_prop{}

                var FirstPetIndex:int = 0
                var SecondPetIndex:int = 0
                var ThirdPetIndex:int = 0
                var ForthPetIndex:int = 0

                var IndexBase:int = 12*16

                #Select right pets for each player
                if(GetTeamIndexAgent[Agent]=1):
                    set FirstPetIndex =  IndexBase
                    set SecondPetIndex = IndexBase+1
                    set ThirdPetIndex =  IndexBase+2
                    set ForthPetIndex =  IndexBase+3
                if(GetTeamIndexAgent[Agent]=2):
                    set FirstPetIndex =  IndexBase+4
                    set SecondPetIndex = IndexBase+5
                    set ThirdPetIndex =  IndexBase+6
                    set ForthPetIndex =  IndexBase+7
                if(GetTeamIndexAgent[Agent]=3):
                    set FirstPetIndex =  IndexBase+8
                    set SecondPetIndex = IndexBase+9
                    set ThirdPetIndex =  IndexBase+10
                    set ForthPetIndex =  IndexBase+11

                if (Mypet:=Pets[FirstPetIndex]):
                    set FirstPet = Mypet.PetProp
                if (Mypet:=Pets[SecondPetIndex]):
                    set SecondPet = Mypet.PetProp
                if (Mypet:=Pets[ThirdPetIndex]):
                    set ThirdPet = Mypet.PetProp
                if (Mypet:=Pets[ForthPetIndex]):
                    set ForthPet = Mypet.PetProp

                var PetWon:creative_prop = creative_prop{}
                var RandResult:int = GetRandomInt(0,1000)
                Print("Rand: {RandResult}")
                if:
                    PlayerObj := player[Agent]
                    Fort := Agent.GetFortCharacter[]
                then:
                    if(CustomPlayer:Custom_Player = PlayersMap[PlayerObj]):
                        #--------------Luck
                        set RandResult+=CustomPlayer.Pet_Luck_Value*5+CustomPlayer.Bonus_Luck*5
                        #--------------
                        if (RandResult>995):
                            Print("Legendary")
                            if(Cine:=Cine_L17_Legendary[GetTeamIndexAgent[Agent]-1]):
                                Cine.Play(Agent)
                            Sleep(4.0)
                            if(Cine:=Cine_L17_Legendary[GetTeamIndexAgent[Agent]-1]):
                                Cine.Stop(Agent)
                                Accolade_Egg17_18.Award(Agent)

                            if (CustomPlayer.Ownership_L17_Pet4=0):
                                #Release potential previous pets
                                if (CustomPlayer.Ownership_L17_Pet1=1):
                                    #--------------Button pressed:
                                    CustomPlayer.IncreaseButtonL17()
                                    #Print("Button L17 Pressed: {CustomPlayer.ButtonL17Pressed}")
                                    #--------------Spawn
                                    Sleep(0.5)
                                    ##spawn:
                                    ##    ReleasePet(FirstPet)
                                    set CustomPlayer.Ownership_L17_Pet1=0
                                    L17_Pet1_DesOwn(Agent)
                                if (CustomPlayer.Ownership_L17_Pet2=1):
                                    #--------------Button pressed:
                                    CustomPlayer.IncreaseButtonL17()
                                    #Print("Button L17 Pressed: {CustomPlayer.ButtonL17Pressed}")
                                    #--------------Spawn
                                    Sleep(0.5)
                                    ##spawn:
                                    ##    ReleasePet(SecondPet)
                                    set CustomPlayer.Ownership_L17_Pet2=0
                                    L17_Pet2_DesOwn(Agent)
                                if (CustomPlayer.Ownership_L17_Pet3=1):
                                    #--------------Button pressed:
                                    CustomPlayer.IncreaseButtonL17()
                                    #Print("Button L17 Pressed: {CustomPlayer.ButtonL17Pressed}")
                                    #--------------Spawn
                                    Sleep(0.5)
                                    ##spawn:
                                    ##    ReleasePet(ThirdPet)
                                    set CustomPlayer.Ownership_L17_Pet3=0
                                    L17_Pet3_DesOwn(Agent)
                                #Get New pet
                                L17_Pet4_Own(Agent)
                                set CustomPlayer.Ownership_L17_Pet4=1
                                set CustomPlayer.Pet_Unlocked_L17=4
                                CustomPlayer.Update_Pet_PowerL17(Power_L17P4)
                                var PetPower:float=1.0*Power_L17P4
                                set PetWon = ForthPet
                                spawn:
                                    FollowPlayerL17(Fort, PetWon, PetPower, ForthPetIndex)
                        else if (RandResult>855):
                            Print("Epic")
                            if(Cine:=Cine_L17_Epic[GetTeamIndexAgent[Agent]-1]):
                                Cine.Play(Agent)
                            Sleep(4.0)
                            if(Cine:=Cine_L17_Epic[GetTeamIndexAgent[Agent]-1]):
                                Cine.Stop(Agent)
                                Accolade_Egg17_18.Award(Agent)

                            if (CustomPlayer.Ownership_L17_Pet3=0 and CustomPlayer.Ownership_L17_Pet4=0):
                                #Release potential previous pets
                                if (CustomPlayer.Ownership_L17_Pet1=1):
                                    #--------------Button pressed:
                                    CustomPlayer.IncreaseButtonL17()
                                    #Print("Button L17 Pressed: {CustomPlayer.ButtonL17Pressed}")
                                    #--------------Spawn
                                    Sleep(0.5)
                                    ##spawn:
                                    ##    ReleasePet(FirstPet)
                                    set CustomPlayer.Ownership_L17_Pet1=0
                                    L17_Pet1_DesOwn(Agent)
                                if (CustomPlayer.Ownership_L17_Pet2=1):
                                    #--------------Button pressed:
                                    CustomPlayer.IncreaseButtonL17()
                                    #Print("Button L17 Pressed: {CustomPlayer.ButtonL17Pressed}")
                                    #--------------Spawn
                                    Sleep(0.5)
                                    ##spawn:
                                    ##    ReleasePet(SecondPet)
                                    set CustomPlayer.Ownership_L17_Pet2=0
                                    L17_Pet2_DesOwn(Agent)
                                #Get New pet
                                L17_Pet3_Own(Agent)
                                set CustomPlayer.Ownership_L17_Pet3=1
                                set CustomPlayer.Pet_Unlocked_L17=3
                                CustomPlayer.Update_Pet_PowerL17(Power_L17P3)
                                var PetPower:float=1.0*Power_L17P3
                                set PetWon = ThirdPet
                                spawn:
                                    FollowPlayerL17(Fort, PetWon, PetPower, ThirdPetIndex)
                        else if (RandResult>750):
                            Print("Rare")
                            if(Cine:=Cine_L17_Rare[GetTeamIndexAgent[Agent]-1]):
                                Cine.Play(Agent)
                            Sleep(4.0)
                            if(Cine:=Cine_L17_Rare[GetTeamIndexAgent[Agent]-1]):
                                Cine.Stop(Agent)
                                Accolade_Egg17_18.Award(Agent)

                            if (CustomPlayer.Ownership_L17_Pet2=0 and CustomPlayer.Ownership_L17_Pet3=0 and CustomPlayer.Ownership_L17_Pet4=0):
                                #Release potential previous pets
                                if (CustomPlayer.Ownership_L17_Pet1=1):
                                    #--------------Button pressed:
                                    CustomPlayer.IncreaseButtonL17()
                                    #Print("Button L17 Pressed: {CustomPlayer.ButtonL17Pressed}")
                                    #--------------Spawn
                                    Sleep(0.5)
                                    ##spawn:
                                    ##    ReleasePet(FirstPet)
                                    set CustomPlayer.Ownership_L17_Pet1=0
                                    L17_Pet1_DesOwn(Agent)
                                #Get New pet
                                L17_Pet2_Own(Agent)
                                set CustomPlayer.Ownership_L17_Pet2=1
                                set CustomPlayer.Pet_Unlocked_L17=2
                                CustomPlayer.Update_Pet_PowerL17(Power_L17P2)
                                var PetPower:float=1.0*Power_L17P2
                                set PetWon = SecondPet
                                spawn:
                                    FollowPlayerL17(Fort, PetWon, PetPower, SecondPetIndex)
                        else:
                            Print("Normal")
                            if(Cine:=Cine_L17_Normal[GetTeamIndexAgent[Agent]-1]):
                                Cine.Play(Agent)
                            Sleep(4.0)
                            if(Cine:=Cine_L17_Normal[GetTeamIndexAgent[Agent]-1]):
                                Cine.Stop(Agent)
                                Accolade_Egg17_18.Award(Agent)

                            if (CustomPlayer.Ownership_L17_Pet1=0 and CustomPlayer.Ownership_L17_Pet2=0 and CustomPlayer.Ownership_L17_Pet3=0 and CustomPlayer.Ownership_L17_Pet4=0):
                                #Get New pet
                                L17_Pet1_Own(Agent)
                                set CustomPlayer.Ownership_L17_Pet1=1
                                set CustomPlayer.Pet_Unlocked_L17=1
                                CustomPlayer.Update_Pet_PowerL17(Power_L17P1)
                                var PetPower:float=1.0*Power_L17P1
                                set PetWon = FirstPet
                                spawn:
                                    FollowPlayerL17(Fort, PetWon, PetPower, FirstPetIndex)

    PurchasePetL18( Agent : agent )<suspends>: void =
        #Requires Unlock L18
        if:
            PlayerObj := player[Agent]
            CurrentPlayerStats := PlayerStatsManager.GetPlayerStats[Agent]
            CurrentPlayerStats.LevelUnlocked.CurrentValue<17
        then:
            RequireHigherLevelHUD.Show(Agent) 
        else:
            if:
                PlayerObj := player[Agent]
                CustomPlayer:Custom_Player = PlayersMap[PlayerObj]
                CustomPlayer.PlayerGold<CostL18
            then:
                NotEnoughGoldHUD.Show(Agent)
            else:
                DecreaseGoldFunc(Agent,CostL18)
                var FirstPet:creative_prop = creative_prop{}
                var SecondPet:creative_prop = creative_prop{}
                var ThirdPet:creative_prop = creative_prop{}
                var ForthPet:creative_prop = creative_prop{}

                var FirstPetIndex:int = 0
                var SecondPetIndex:int = 0
                var ThirdPetIndex:int = 0
                var ForthPetIndex:int = 0

                var IndexBase:int = 12*17

                #Select right pets for each player
                if(GetTeamIndexAgent[Agent]=1):
                    set FirstPetIndex =  IndexBase
                    set SecondPetIndex = IndexBase+1
                    set ThirdPetIndex =  IndexBase+2
                    set ForthPetIndex =  IndexBase+3
                if(GetTeamIndexAgent[Agent]=2):
                    set FirstPetIndex =  IndexBase+4
                    set SecondPetIndex = IndexBase+5
                    set ThirdPetIndex =  IndexBase+6
                    set ForthPetIndex =  IndexBase+7
                if(GetTeamIndexAgent[Agent]=3):
                    set FirstPetIndex =  IndexBase+8
                    set SecondPetIndex = IndexBase+9
                    set ThirdPetIndex =  IndexBase+10
                    set ForthPetIndex =  IndexBase+11

                if (Mypet:=Pets[FirstPetIndex]):
                    set FirstPet = Mypet.PetProp
                if (Mypet:=Pets[SecondPetIndex]):
                    set SecondPet = Mypet.PetProp
                if (Mypet:=Pets[ThirdPetIndex]):
                    set ThirdPet = Mypet.PetProp
                if (Mypet:=Pets[ForthPetIndex]):
                    set ForthPet = Mypet.PetProp

                var PetWon:creative_prop = creative_prop{}
                var RandResult:int = GetRandomInt(0,1000)
                Print("Rand: {RandResult}")
                if:
                    PlayerObj := player[Agent]
                    Fort := Agent.GetFortCharacter[]
                then:
                    if(CustomPlayer:Custom_Player = PlayersMap[PlayerObj]):
                        #--------------Luck
                        set RandResult+=CustomPlayer.Pet_Luck_Value*5+CustomPlayer.Bonus_Luck*5
                        #--------------
                        if (RandResult>995):
                            Print("Legendary")
                            if(Cine:=Cine_L18_Legendary[GetTeamIndexAgent[Agent]-1]):
                                Cine.Play(Agent)
                            Sleep(4.0)
                            if(Cine:=Cine_L18_Legendary[GetTeamIndexAgent[Agent]-1]):
                                Cine.Stop(Agent)
                                Accolade_Egg17_18.Award(Agent)

                            if (CustomPlayer.Ownership_L18_Pet4=0):
                                #Release potential previous pets
                                if (CustomPlayer.Ownership_L18_Pet1=1):
                                    #--------------Button pressed:
                                    CustomPlayer.IncreaseButtonL18()
                                    #Print("Button L18 Pressed: {CustomPlayer.ButtonL18Pressed}")
                                    #--------------Spawn
                                    Sleep(0.5)
                                    ##spawn:
                                    ##    ReleasePet(FirstPet)
                                    set CustomPlayer.Ownership_L18_Pet1=0
                                    L18_Pet1_DesOwn(Agent)
                                if (CustomPlayer.Ownership_L18_Pet2=1):
                                    #--------------Button pressed:
                                    CustomPlayer.IncreaseButtonL18()
                                    #Print("Button L18 Pressed: {CustomPlayer.ButtonL18Pressed}")
                                    #--------------Spawn
                                    Sleep(0.5)
                                    ##spawn:
                                    ##    ReleasePet(SecondPet)
                                    set CustomPlayer.Ownership_L18_Pet2=0
                                    L18_Pet2_DesOwn(Agent)
                                if (CustomPlayer.Ownership_L18_Pet3=1):
                                    #--------------Button pressed:
                                    CustomPlayer.IncreaseButtonL18()
                                    #Print("Button L18 Pressed: {CustomPlayer.ButtonL18Pressed}")
                                    #--------------Spawn
                                    Sleep(0.5)
                                    ##spawn:
                                    ##    ReleasePet(ThirdPet)
                                    set CustomPlayer.Ownership_L18_Pet3=0
                                    L18_Pet3_DesOwn(Agent)
                                #Get New pet
                                L18_Pet4_Own(Agent)
                                set CustomPlayer.Ownership_L18_Pet4=1
                                set CustomPlayer.Pet_Unlocked_L18=4
                                CustomPlayer.Update_Pet_PowerL18(Power_L18P4)
                                var PetPower:float=1.0*Power_L18P4
                                set PetWon = ForthPet
                                spawn:
                                    FollowPlayerL18(Fort, PetWon, PetPower, ForthPetIndex)
                        else if (RandResult>855):
                            Print("Epic")
                            if(Cine:=Cine_L18_Epic[GetTeamIndexAgent[Agent]-1]):
                                Cine.Play(Agent)
                            Sleep(4.0)
                            if(Cine:=Cine_L18_Epic[GetTeamIndexAgent[Agent]-1]):
                                Cine.Stop(Agent)
                                Accolade_Egg17_18.Award(Agent)

                            if (CustomPlayer.Ownership_L18_Pet3=0 and CustomPlayer.Ownership_L18_Pet4=0):
                                #Release potential previous pets
                                if (CustomPlayer.Ownership_L18_Pet1=1):
                                    #--------------Button pressed:
                                    CustomPlayer.IncreaseButtonL18()
                                    #Print("Button L18 Pressed: {CustomPlayer.ButtonL18Pressed}")
                                    #--------------Spawn
                                    Sleep(0.5)
                                    ##spawn:
                                    ##    ReleasePet(FirstPet)
                                    set CustomPlayer.Ownership_L18_Pet1=0
                                    L18_Pet1_DesOwn(Agent)
                                if (CustomPlayer.Ownership_L18_Pet2=1):
                                    #--------------Button pressed:
                                    CustomPlayer.IncreaseButtonL18()
                                    #Print("Button L18 Pressed: {CustomPlayer.ButtonL18Pressed}")
                                    #--------------Spawn
                                    Sleep(0.5)
                                    ##spawn:
                                    ##    ReleasePet(SecondPet)
                                    set CustomPlayer.Ownership_L18_Pet2=0
                                    L18_Pet2_DesOwn(Agent)
                                #Get New pet
                                L18_Pet3_Own(Agent)
                                set CustomPlayer.Ownership_L18_Pet3=1
                                set CustomPlayer.Pet_Unlocked_L18=3
                                CustomPlayer.Update_Pet_PowerL18(Power_L18P3)
                                var PetPower:float=1.0*Power_L18P3
                                set PetWon = ThirdPet
                                spawn:
                                    FollowPlayerL18(Fort, PetWon, PetPower, ThirdPetIndex)
                        else if (RandResult>750):
                            Print("Rare")
                            if(Cine:=Cine_L18_Rare[GetTeamIndexAgent[Agent]-1]):
                                Cine.Play(Agent)
                            Sleep(4.0)
                            if(Cine:=Cine_L18_Rare[GetTeamIndexAgent[Agent]-1]):
                                Cine.Stop(Agent)
                                Accolade_Egg17_18.Award(Agent)

                            if (CustomPlayer.Ownership_L18_Pet2=0 and CustomPlayer.Ownership_L18_Pet3=0 and CustomPlayer.Ownership_L18_Pet4=0):
                                #Release potential previous pets
                                if (CustomPlayer.Ownership_L18_Pet1=1):
                                    #--------------Button pressed:
                                    CustomPlayer.IncreaseButtonL18()
                                    #Print("Button L18 Pressed: {CustomPlayer.ButtonL18Pressed}")
                                    #--------------Spawn
                                    Sleep(0.5)
                                    ##spawn:
                                    ##    ReleasePet(FirstPet)
                                    set CustomPlayer.Ownership_L18_Pet1=0
                                    L18_Pet1_DesOwn(Agent)
                                #Get New pet
                                L18_Pet2_Own(Agent)
                                set CustomPlayer.Ownership_L18_Pet2=1
                                set CustomPlayer.Pet_Unlocked_L18=2
                                CustomPlayer.Update_Pet_PowerL18(Power_L18P2)
                                var PetPower:float=1.0*Power_L18P2
                                set PetWon = SecondPet
                                spawn:
                                    FollowPlayerL18(Fort, PetWon, PetPower, SecondPetIndex)
                        else:
                            Print("Normal")
                            if(Cine:=Cine_L18_Normal[GetTeamIndexAgent[Agent]-1]):
                                Cine.Play(Agent)
                            Sleep(4.0)
                            if(Cine:=Cine_L18_Normal[GetTeamIndexAgent[Agent]-1]):
                                Cine.Stop(Agent)
                                Accolade_Egg17_18.Award(Agent)

                            if (CustomPlayer.Ownership_L18_Pet1=0 and CustomPlayer.Ownership_L18_Pet2=0 and CustomPlayer.Ownership_L18_Pet3=0 and CustomPlayer.Ownership_L18_Pet4=0):
                                #Get New pet
                                L18_Pet1_Own(Agent)
                                set CustomPlayer.Ownership_L18_Pet1=1
                                set CustomPlayer.Pet_Unlocked_L18=1
                                CustomPlayer.Update_Pet_PowerL18(Power_L18P1)
                                var PetPower:float=1.0*Power_L18P1
                                set PetWon = FirstPet
                                spawn:
                                    FollowPlayerL18(Fort, PetWon, PetPower, FirstPetIndex)

    PurchasePetL19( Agent : agent )<suspends>: void =
        #Requires Unlock L19
        if:
            PlayerObj := player[Agent]
            CurrentPlayerStats := PlayerStatsManager.GetPlayerStats[Agent]
            CurrentPlayerStats.LevelUnlocked.CurrentValue<18
        then:
            RequireHigherLevelHUD.Show(Agent) 
        else:
            if:
                PlayerObj := player[Agent]
                CustomPlayer:Custom_Player = PlayersMap[PlayerObj]
                CustomPlayer.PlayerGold<CostL19
            then:
                NotEnoughGoldHUD.Show(Agent)
            else:
                DecreaseGoldFunc(Agent,CostL19)
                var FirstPet:creative_prop = creative_prop{}
                var SecondPet:creative_prop = creative_prop{}
                var ThirdPet:creative_prop = creative_prop{}
                var ForthPet:creative_prop = creative_prop{}

                var FirstPetIndex:int = 0
                var SecondPetIndex:int = 0
                var ThirdPetIndex:int = 0
                var ForthPetIndex:int = 0

                var IndexBase:int = 12*18

                #Select right pets for each player
                if(GetTeamIndexAgent[Agent]=1):
                    set FirstPetIndex =  IndexBase
                    set SecondPetIndex = IndexBase+1
                    set ThirdPetIndex =  IndexBase+2
                    set ForthPetIndex =  IndexBase+3
                if(GetTeamIndexAgent[Agent]=2):
                    set FirstPetIndex =  IndexBase+4
                    set SecondPetIndex = IndexBase+5
                    set ThirdPetIndex =  IndexBase+6
                    set ForthPetIndex =  IndexBase+7
                if(GetTeamIndexAgent[Agent]=3):
                    set FirstPetIndex =  IndexBase+8
                    set SecondPetIndex = IndexBase+9
                    set ThirdPetIndex =  IndexBase+10
                    set ForthPetIndex =  IndexBase+11

                if (Mypet:=Pets[FirstPetIndex]):
                    set FirstPet = Mypet.PetProp
                if (Mypet:=Pets[SecondPetIndex]):
                    set SecondPet = Mypet.PetProp
                if (Mypet:=Pets[ThirdPetIndex]):
                    set ThirdPet = Mypet.PetProp
                if (Mypet:=Pets[ForthPetIndex]):
                    set ForthPet = Mypet.PetProp

                var PetWon:creative_prop = creative_prop{}
                var RandResult:int = GetRandomInt(0,1000)
                Print("Rand: {RandResult}")
                if:
                    PlayerObj := player[Agent]
                    Fort := Agent.GetFortCharacter[]
                then:
                    if(CustomPlayer:Custom_Player = PlayersMap[PlayerObj]):
                        #--------------Luck
                        set RandResult+=CustomPlayer.Pet_Luck_Value*5+CustomPlayer.Bonus_Luck*5
                        #--------------
                        if (RandResult>995):
                            Print("Legendary")
                            if(Cine:=Cine_L19_Legendary[GetTeamIndexAgent[Agent]-1]):
                                Cine.Play(Agent)
                            Sleep(4.0)
                            if(Cine:=Cine_L19_Legendary[GetTeamIndexAgent[Agent]-1]):
                                Cine.Stop(Agent)
                                Accolade_Egg19_20.Award(Agent)

                            if (CustomPlayer.Ownership_L19_Pet4=0):
                                #Release potential previous pets
                                if (CustomPlayer.Ownership_L19_Pet1=1):
                                    #--------------Button pressed:
                                    CustomPlayer.IncreaseButtonL19()
                                    #Print("Button L19 Pressed: {CustomPlayer.ButtonL19Pressed}")
                                    #--------------Spawn
                                    Sleep(0.5)
                                    ##spawn:
                                    ##    ReleasePet(FirstPet)
                                    set CustomPlayer.Ownership_L19_Pet1=0
                                    L19_Pet1_DesOwn(Agent)
                                if (CustomPlayer.Ownership_L19_Pet2=1):
                                    #--------------Button pressed:
                                    CustomPlayer.IncreaseButtonL19()
                                    #Print("Button L19 Pressed: {CustomPlayer.ButtonL19Pressed}")
                                    #--------------Spawn
                                    Sleep(0.5)
                                    ##spawn:
                                    ##    ReleasePet(SecondPet)
                                    set CustomPlayer.Ownership_L19_Pet2=0
                                    L19_Pet2_DesOwn(Agent)
                                if (CustomPlayer.Ownership_L19_Pet3=1):
                                    #--------------Button pressed:
                                    CustomPlayer.IncreaseButtonL19()
                                    #Print("Button L19 Pressed: {CustomPlayer.ButtonL19Pressed}")
                                    #--------------Spawn
                                    Sleep(0.5)
                                    ##spawn:
                                    ##    ReleasePet(ThirdPet)
                                    set CustomPlayer.Ownership_L19_Pet3=0
                                    L19_Pet3_DesOwn(Agent)
                                #Get New pet
                                L19_Pet4_Own(Agent)
                                set CustomPlayer.Ownership_L19_Pet4=1
                                set CustomPlayer.Pet_Unlocked_L19=4
                                CustomPlayer.Update_Pet_PowerL19(Power_L19P4)
                                var PetPower:float=1.0*Power_L19P4
                                set PetWon = ForthPet
                                spawn:
                                    FollowPlayerL19(Fort, PetWon, PetPower, ForthPetIndex)
                        else if (RandResult>855):
                            Print("Epic")
                            if(Cine:=Cine_L19_Epic[GetTeamIndexAgent[Agent]-1]):
                                Cine.Play(Agent)
                            Sleep(4.0)
                            if(Cine:=Cine_L19_Epic[GetTeamIndexAgent[Agent]-1]):
                                Cine.Stop(Agent)
                                Accolade_Egg19_20.Award(Agent)

                            if (CustomPlayer.Ownership_L19_Pet3=0 and CustomPlayer.Ownership_L19_Pet4=0):
                                #Release potential previous pets
                                if (CustomPlayer.Ownership_L19_Pet1=1):
                                    #--------------Button pressed:
                                    CustomPlayer.IncreaseButtonL19()
                                    #Print("Button L19 Pressed: {CustomPlayer.ButtonL19Pressed}")
                                    #--------------Spawn
                                    Sleep(0.5)
                                    ##spawn:
                                    ##    ReleasePet(FirstPet)
                                    set CustomPlayer.Ownership_L19_Pet1=0
                                    L19_Pet1_DesOwn(Agent)
                                if (CustomPlayer.Ownership_L19_Pet2=1):
                                    #--------------Button pressed:
                                    CustomPlayer.IncreaseButtonL19()
                                    #Print("Button L19 Pressed: {CustomPlayer.ButtonL19Pressed}")
                                    #--------------Spawn
                                    Sleep(0.5)
                                    ##spawn:
                                    ##    ReleasePet(SecondPet)
                                    set CustomPlayer.Ownership_L19_Pet2=0
                                    L19_Pet2_DesOwn(Agent)
                                #Get New pet
                                L19_Pet3_Own(Agent)
                                set CustomPlayer.Ownership_L19_Pet3=1
                                set CustomPlayer.Pet_Unlocked_L19=3
                                CustomPlayer.Update_Pet_PowerL19(Power_L19P3)
                                var PetPower:float=1.0*Power_L19P3
                                set PetWon = ThirdPet
                                spawn:
                                    FollowPlayerL19(Fort, PetWon, PetPower, ThirdPetIndex)
                        else if (RandResult>750):
                            Print("Rare")
                            if(Cine:=Cine_L19_Rare[GetTeamIndexAgent[Agent]-1]):
                                Cine.Play(Agent)
                            Sleep(4.0)
                            if(Cine:=Cine_L19_Rare[GetTeamIndexAgent[Agent]-1]):
                                Cine.Stop(Agent)
                                Accolade_Egg19_20.Award(Agent)

                            if (CustomPlayer.Ownership_L19_Pet2=0 and CustomPlayer.Ownership_L19_Pet3=0 and CustomPlayer.Ownership_L19_Pet4=0):
                                #Release potential previous pets
                                if (CustomPlayer.Ownership_L19_Pet1=1):
                                    #--------------Button pressed:
                                    CustomPlayer.IncreaseButtonL19()
                                    #Print("Button L19 Pressed: {CustomPlayer.ButtonL19Pressed}")
                                    #--------------Spawn
                                    Sleep(0.5)
                                    ##spawn:
                                    ##    ReleasePet(FirstPet)
                                    set CustomPlayer.Ownership_L19_Pet1=0
                                    L19_Pet1_DesOwn(Agent)
                                #Get New pet
                                L19_Pet2_Own(Agent)
                                set CustomPlayer.Ownership_L19_Pet2=1
                                set CustomPlayer.Pet_Unlocked_L19=2
                                CustomPlayer.Update_Pet_PowerL19(Power_L19P2)
                                var PetPower:float=1.0*Power_L19P2
                                set PetWon = SecondPet
                                spawn:
                                    FollowPlayerL19(Fort, PetWon, PetPower, SecondPetIndex)
                        else:
                            Print("Normal")
                            if(Cine:=Cine_L19_Normal[GetTeamIndexAgent[Agent]-1]):
                                Cine.Play(Agent)
                            Sleep(4.0)
                            if(Cine:=Cine_L19_Normal[GetTeamIndexAgent[Agent]-1]):
                                Cine.Stop(Agent)
                                Accolade_Egg19_20.Award(Agent)

                            if (CustomPlayer.Ownership_L19_Pet1=0 and CustomPlayer.Ownership_L19_Pet2=0 and CustomPlayer.Ownership_L19_Pet3=0 and CustomPlayer.Ownership_L19_Pet4=0):
                                #Get New pet
                                L19_Pet1_Own(Agent)
                                set CustomPlayer.Ownership_L19_Pet1=1
                                set CustomPlayer.Pet_Unlocked_L19=1
                                CustomPlayer.Update_Pet_PowerL19(Power_L19P1)
                                var PetPower:float=1.0*Power_L19P1
                                set PetWon = FirstPet
                                spawn:
                                    FollowPlayerL19(Fort, PetWon, PetPower, FirstPetIndex)

    PurchasePetL20( Agent : agent )<suspends>: void =
        #Requires Unlock L20
        if:
            PlayerObj := player[Agent]
            CurrentPlayerStats := PlayerStatsManager.GetPlayerStats[Agent]
            CurrentPlayerStats.LevelUnlocked.CurrentValue<19
        then:
            RequireHigherLevelHUD.Show(Agent) 
        else:
            if:
                PlayerObj := player[Agent]
                CustomPlayer:Custom_Player = PlayersMap[PlayerObj]
                CustomPlayer.PlayerGold<CostL20
            then:
                NotEnoughGoldHUD.Show(Agent)
            else:
                DecreaseGoldFunc(Agent,CostL20)
                var FirstPet:creative_prop = creative_prop{}
                var SecondPet:creative_prop = creative_prop{}
                var ThirdPet:creative_prop = creative_prop{}
                var ForthPet:creative_prop = creative_prop{}

                var FirstPetIndex:int = 0
                var SecondPetIndex:int = 0
                var ThirdPetIndex:int = 0
                var ForthPetIndex:int = 0

                var IndexBase:int = 12*19

                #Select right pets for each player
                if(GetTeamIndexAgent[Agent]=1):
                    set FirstPetIndex =  IndexBase
                    set SecondPetIndex = IndexBase+1
                    set ThirdPetIndex =  IndexBase+2
                    set ForthPetIndex =  IndexBase+3
                if(GetTeamIndexAgent[Agent]=2):
                    set FirstPetIndex =  IndexBase+4
                    set SecondPetIndex = IndexBase+5
                    set ThirdPetIndex =  IndexBase+6
                    set ForthPetIndex =  IndexBase+7
                if(GetTeamIndexAgent[Agent]=3):
                    set FirstPetIndex =  IndexBase+8
                    set SecondPetIndex = IndexBase+9
                    set ThirdPetIndex =  IndexBase+10
                    set ForthPetIndex =  IndexBase+11

                if (Mypet:=Pets[FirstPetIndex]):
                    set FirstPet = Mypet.PetProp
                if (Mypet:=Pets[SecondPetIndex]):
                    set SecondPet = Mypet.PetProp
                if (Mypet:=Pets[ThirdPetIndex]):
                    set ThirdPet = Mypet.PetProp
                if (Mypet:=Pets[ForthPetIndex]):
                    set ForthPet = Mypet.PetProp

                var PetWon:creative_prop = creative_prop{}
                var RandResult:int = GetRandomInt(0,1000)
                Print("Rand: {RandResult}")
                if:
                    PlayerObj := player[Agent]
                    Fort := Agent.GetFortCharacter[]
                then:
                    if(CustomPlayer:Custom_Player = PlayersMap[PlayerObj]):
                        #--------------Luck
                        set RandResult+=CustomPlayer.Pet_Luck_Value*5+CustomPlayer.Bonus_Luck*5
                        #--------------
                        if (RandResult>995):
                            Print("Legendary")
                            if(Cine:=Cine_L20_Legendary[GetTeamIndexAgent[Agent]-1]):
                                Cine.Play(Agent)
                            Sleep(4.0)
                            if(Cine:=Cine_L20_Legendary[GetTeamIndexAgent[Agent]-1]):
                                Cine.Stop(Agent)
                                Accolade_Egg19_20.Award(Agent)

                            if (CustomPlayer.Ownership_L20_Pet4=0):
                                #Release potential previous pets
                                if (CustomPlayer.Ownership_L20_Pet1=1):
                                    #--------------Button pressed:
                                    CustomPlayer.IncreaseButtonL20()
                                    #Print("Button L20 Pressed: {CustomPlayer.ButtonL20Pressed}")
                                    #--------------Spawn
                                    Sleep(0.5)
                                    ##spawn:
                                    ##    ReleasePet(FirstPet)
                                    set CustomPlayer.Ownership_L20_Pet1=0
                                    L20_Pet1_DesOwn(Agent)
                                if (CustomPlayer.Ownership_L20_Pet2=1):
                                    #--------------Button pressed:
                                    CustomPlayer.IncreaseButtonL20()
                                    #Print("Button L20 Pressed: {CustomPlayer.ButtonL20Pressed}")
                                    #--------------Spawn
                                    Sleep(0.5)
                                    ##spawn:
                                    ##    ReleasePet(SecondPet)
                                    set CustomPlayer.Ownership_L20_Pet2=0
                                    L20_Pet2_DesOwn(Agent)
                                if (CustomPlayer.Ownership_L20_Pet3=1):
                                    #--------------Button pressed:
                                    CustomPlayer.IncreaseButtonL20()
                                    #Print("Button L20 Pressed: {CustomPlayer.ButtonL20Pressed}")
                                    #--------------Spawn
                                    Sleep(0.5)
                                    ##spawn:
                                    ##    ReleasePet(ThirdPet)
                                    set CustomPlayer.Ownership_L20_Pet3=0
                                    L20_Pet3_DesOwn(Agent)
                                #Get New pet
                                L20_Pet4_Own(Agent)
                                set CustomPlayer.Ownership_L20_Pet4=1
                                set CustomPlayer.Pet_Unlocked_L20=4
                                CustomPlayer.Update_Pet_PowerL20(Power_L20P4)
                                var PetPower:float=1.0*Power_L20P4
                                set PetWon = ForthPet
                                spawn:
                                    FollowPlayerL20(Fort, PetWon, PetPower, ForthPetIndex)
                        else if (RandResult>855):
                            Print("Epic")
                            if(Cine:=Cine_L20_Epic[GetTeamIndexAgent[Agent]-1]):
                                Cine.Play(Agent)
                            Sleep(4.0)
                            if(Cine:=Cine_L20_Epic[GetTeamIndexAgent[Agent]-1]):
                                Cine.Stop(Agent)
                                Accolade_Egg19_20.Award(Agent)

                            if (CustomPlayer.Ownership_L20_Pet3=0 and CustomPlayer.Ownership_L20_Pet4=0):
                                #Release potential previous pets
                                if (CustomPlayer.Ownership_L20_Pet1=1):
                                    #--------------Button pressed:
                                    CustomPlayer.IncreaseButtonL20()
                                    #Print("Button L20 Pressed: {CustomPlayer.ButtonL20Pressed}")
                                    #--------------Spawn
                                    Sleep(0.5)
                                    ##spawn:
                                    ##    ReleasePet(FirstPet)
                                    set CustomPlayer.Ownership_L20_Pet1=0
                                    L20_Pet1_DesOwn(Agent)
                                if (CustomPlayer.Ownership_L20_Pet2=1):
                                    #--------------Button pressed:
                                    CustomPlayer.IncreaseButtonL20()
                                    #Print("Button L20 Pressed: {CustomPlayer.ButtonL20Pressed}")
                                    #--------------Spawn
                                    Sleep(0.5)
                                    ##spawn:
                                    ##    ReleasePet(SecondPet)
                                    set CustomPlayer.Ownership_L20_Pet2=0
                                    L20_Pet2_DesOwn(Agent)
                                #Get New pet
                                L20_Pet3_Own(Agent)
                                set CustomPlayer.Ownership_L20_Pet3=1
                                set CustomPlayer.Pet_Unlocked_L20=3
                                CustomPlayer.Update_Pet_PowerL20(Power_L20P3)
                                var PetPower:float=1.0*Power_L20P3
                                set PetWon = ThirdPet
                                spawn:
                                    FollowPlayerL20(Fort, PetWon, PetPower, ThirdPetIndex)
                        else if (RandResult>750):
                            Print("Rare")
                            if(Cine:=Cine_L20_Rare[GetTeamIndexAgent[Agent]-1]):
                                Cine.Play(Agent)
                            Sleep(4.0)
                            if(Cine:=Cine_L20_Rare[GetTeamIndexAgent[Agent]-1]):
                                Cine.Stop(Agent)
                                Accolade_Egg19_20.Award(Agent)

                            if (CustomPlayer.Ownership_L20_Pet2=0 and CustomPlayer.Ownership_L20_Pet3=0 and CustomPlayer.Ownership_L20_Pet4=0):
                                #Release potential previous pets
                                if (CustomPlayer.Ownership_L20_Pet1=1):
                                    #--------------Button pressed:
                                    CustomPlayer.IncreaseButtonL20()
                                    #Print("Button L20 Pressed: {CustomPlayer.ButtonL20Pressed}")
                                    #--------------Spawn
                                    Sleep(0.5)
                                    ##spawn:
                                    ##    ReleasePet(FirstPet)
                                    set CustomPlayer.Ownership_L20_Pet1=0
                                    L20_Pet1_DesOwn(Agent)
                                #Get New pet
                                L20_Pet2_Own(Agent)
                                set CustomPlayer.Ownership_L20_Pet2=1
                                set CustomPlayer.Pet_Unlocked_L20=2
                                CustomPlayer.Update_Pet_PowerL20(Power_L20P2)
                                var PetPower:float=1.0*Power_L20P2
                                set PetWon = SecondPet
                                spawn:
                                    FollowPlayerL20(Fort, PetWon, PetPower, SecondPetIndex)
                        else:
                            Print("Normal")
                            if(Cine:=Cine_L20_Normal[GetTeamIndexAgent[Agent]-1]):
                                Cine.Play(Agent)
                            Sleep(4.0)
                            if(Cine:=Cine_L20_Normal[GetTeamIndexAgent[Agent]-1]):
                                Cine.Stop(Agent)
                                Accolade_Egg19_20.Award(Agent)

                            if (CustomPlayer.Ownership_L20_Pet1=0 and CustomPlayer.Ownership_L20_Pet2=0 and CustomPlayer.Ownership_L20_Pet3=0 and CustomPlayer.Ownership_L20_Pet4=0):
                                #Get New pet
                                L20_Pet1_Own(Agent)
                                set CustomPlayer.Ownership_L20_Pet1=1
                                set CustomPlayer.Pet_Unlocked_L20=1
                                CustomPlayer.Update_Pet_PowerL20(Power_L20P1)
                                var PetPower:float=1.0*Power_L20P1
                                set PetWon = FirstPet
                                spawn:
                                    FollowPlayerL20(Fort, PetWon, PetPower, FirstPetIndex)

    PurchasePetL21( Agent : agent )<suspends>: void =
        #Requires Unlock L21
        if:
            PlayerObj := player[Agent]
            CurrentPlayerStats := PlayerStatsManager.GetPlayerStats[Agent]
            CurrentPlayerStats.LevelUnlocked.CurrentValue<20
        then:
            RequireHigherLevelHUD.Show(Agent) 
        else:
            if:
                PlayerObj := player[Agent]
                CustomPlayer:Custom_Player = PlayersMap[PlayerObj]
                CustomPlayer.PlayerGold<CostL21
            then:
                NotEnoughGoldHUD.Show(Agent)
            else:
                DecreaseGoldFunc(Agent,CostL21)
                var FirstPet:creative_prop = creative_prop{}
                var SecondPet:creative_prop = creative_prop{}
                var ThirdPet:creative_prop = creative_prop{}
                var ForthPet:creative_prop = creative_prop{}

                var FirstPetIndex:int = 0
                var SecondPetIndex:int = 0
                var ThirdPetIndex:int = 0
                var ForthPetIndex:int = 0

                var IndexBase:int = 12*20

                #Select right pets for each player
                if(GetTeamIndexAgent[Agent]=1):
                    set FirstPetIndex =  IndexBase
                    set SecondPetIndex = IndexBase+1
                    set ThirdPetIndex =  IndexBase+2
                    set ForthPetIndex =  IndexBase+3
                if(GetTeamIndexAgent[Agent]=2):
                    set FirstPetIndex =  IndexBase+4
                    set SecondPetIndex = IndexBase+5
                    set ThirdPetIndex =  IndexBase+6
                    set ForthPetIndex =  IndexBase+7
                if(GetTeamIndexAgent[Agent]=3):
                    set FirstPetIndex =  IndexBase+8
                    set SecondPetIndex = IndexBase+9
                    set ThirdPetIndex =  IndexBase+10
                    set ForthPetIndex =  IndexBase+11

                if (Mypet:=Pets[FirstPetIndex]):
                    set FirstPet = Mypet.PetProp
                if (Mypet:=Pets[SecondPetIndex]):
                    set SecondPet = Mypet.PetProp
                if (Mypet:=Pets[ThirdPetIndex]):
                    set ThirdPet = Mypet.PetProp
                if (Mypet:=Pets[ForthPetIndex]):
                    set ForthPet = Mypet.PetProp

                var PetWon:creative_prop = creative_prop{}
                var RandResult:int = GetRandomInt(0,1000)
                Print("Rand: {RandResult}")
                if:
                    PlayerObj := player[Agent]
                    Fort := Agent.GetFortCharacter[]
                then:
                    if(CustomPlayer:Custom_Player = PlayersMap[PlayerObj]):
                        #--------------Luck
                        set RandResult+=CustomPlayer.Pet_Luck_Value*5+CustomPlayer.Bonus_Luck*5
                        #--------------
                        if (RandResult>995):
                            Print("Legendary")
                            if(Cine:=Cine_L21_Legendary[GetTeamIndexAgent[Agent]-1]):
                                Cine.Play(Agent)
                            Sleep(4.0)
                            if(Cine:=Cine_L21_Legendary[GetTeamIndexAgent[Agent]-1]):
                                Cine.Stop(Agent)
                                Accolade_Egg21_22.Award(Agent)

                            if (CustomPlayer.Ownership_L21_Pet4=0):
                                #Release potential previous pets
                                if (CustomPlayer.Ownership_L21_Pet1=1):
                                    #--------------Button pressed:
                                    CustomPlayer.IncreaseButtonL21()
                                    #Print("Button L21 Pressed: {CustomPlayer.ButtonL21Pressed}")
                                    #--------------Spawn
                                    Sleep(0.5)
                                    ##spawn:
                                    ##    ReleasePet(FirstPet)
                                    set CustomPlayer.Ownership_L21_Pet1=0
                                    L21_Pet1_DesOwn(Agent)
                                if (CustomPlayer.Ownership_L21_Pet2=1):
                                    #--------------Button pressed:
                                    CustomPlayer.IncreaseButtonL21()
                                    #Print("Button L21 Pressed: {CustomPlayer.ButtonL21Pressed}")
                                    #--------------Spawn
                                    Sleep(0.5)
                                    ##spawn:
                                    ##    ReleasePet(SecondPet)
                                    set CustomPlayer.Ownership_L21_Pet2=0
                                    L21_Pet2_DesOwn(Agent)
                                if (CustomPlayer.Ownership_L21_Pet3=1):
                                    #--------------Button pressed:
                                    CustomPlayer.IncreaseButtonL21()
                                    #Print("Button L21 Pressed: {CustomPlayer.ButtonL21Pressed}")
                                    #--------------Spawn
                                    Sleep(0.5)
                                    ##spawn:
                                    ##    ReleasePet(ThirdPet)
                                    set CustomPlayer.Ownership_L21_Pet3=0
                                    L21_Pet3_DesOwn(Agent)
                                #Get New pet
                                L21_Pet4_Own(Agent)
                                set CustomPlayer.Ownership_L21_Pet4=1
                                set CustomPlayer.Pet_Unlocked_L21=4
                                CustomPlayer.Update_Pet_PowerL21(Power_L21P4)
                                var PetPower:float=1.0*Power_L21P4
                                set PetWon = ForthPet
                                spawn:
                                    FollowPlayerL21(Fort, PetWon, PetPower, ForthPetIndex)
                        else if (RandResult>855):
                            Print("Epic")
                            if(Cine:=Cine_L21_Epic[GetTeamIndexAgent[Agent]-1]):
                                Cine.Play(Agent)
                            Sleep(4.0)
                            if(Cine:=Cine_L21_Epic[GetTeamIndexAgent[Agent]-1]):
                                Cine.Stop(Agent)
                                Accolade_Egg21_22.Award(Agent)

                            if (CustomPlayer.Ownership_L21_Pet3=0 and CustomPlayer.Ownership_L21_Pet4=0):
                                #Release potential previous pets
                                if (CustomPlayer.Ownership_L21_Pet1=1):
                                    #--------------Button pressed:
                                    CustomPlayer.IncreaseButtonL21()
                                    #Print("Button L21 Pressed: {CustomPlayer.ButtonL21Pressed}")
                                    #--------------Spawn
                                    Sleep(0.5)
                                    ##spawn:
                                    ##    ReleasePet(FirstPet)
                                    set CustomPlayer.Ownership_L21_Pet1=0
                                    L21_Pet1_DesOwn(Agent)
                                if (CustomPlayer.Ownership_L21_Pet2=1):
                                    #--------------Button pressed:
                                    CustomPlayer.IncreaseButtonL21()
                                    #Print("Button L21 Pressed: {CustomPlayer.ButtonL21Pressed}")
                                    #--------------Spawn
                                    Sleep(0.5)
                                    ##spawn:
                                    ##    ReleasePet(SecondPet)
                                    set CustomPlayer.Ownership_L21_Pet2=0
                                    L21_Pet2_DesOwn(Agent)
                                #Get New pet
                                L21_Pet3_Own(Agent)
                                set CustomPlayer.Ownership_L21_Pet3=1
                                set CustomPlayer.Pet_Unlocked_L21=3
                                CustomPlayer.Update_Pet_PowerL21(Power_L21P3)
                                var PetPower:float=1.0*Power_L21P3
                                set PetWon = ThirdPet
                                spawn:
                                    FollowPlayerL21(Fort, PetWon, PetPower, ThirdPetIndex)
                        else if (RandResult>750):
                            Print("Rare")
                            if(Cine:=Cine_L21_Rare[GetTeamIndexAgent[Agent]-1]):
                                Cine.Play(Agent)
                            Sleep(4.0)
                            if(Cine:=Cine_L21_Rare[GetTeamIndexAgent[Agent]-1]):
                                Cine.Stop(Agent)
                                Accolade_Egg21_22.Award(Agent)

                            if (CustomPlayer.Ownership_L21_Pet2=0 and CustomPlayer.Ownership_L21_Pet3=0 and CustomPlayer.Ownership_L21_Pet4=0):
                                #Release potential previous pets
                                if (CustomPlayer.Ownership_L21_Pet1=1):
                                    #--------------Button pressed:
                                    CustomPlayer.IncreaseButtonL21()
                                    #Print("Button L21 Pressed: {CustomPlayer.ButtonL21Pressed}")
                                    #--------------Spawn
                                    Sleep(0.5)
                                    ##spawn:
                                    ##    ReleasePet(FirstPet)
                                    set CustomPlayer.Ownership_L21_Pet1=0
                                    L21_Pet1_DesOwn(Agent)
                                #Get New pet
                                L21_Pet2_Own(Agent)
                                set CustomPlayer.Ownership_L21_Pet2=1
                                set CustomPlayer.Pet_Unlocked_L21=2
                                CustomPlayer.Update_Pet_PowerL21(Power_L21P2)
                                var PetPower:float=1.0*Power_L21P2
                                set PetWon = SecondPet
                                spawn:
                                    FollowPlayerL21(Fort, PetWon, PetPower, SecondPetIndex)
                        else:
                            Print("Normal")
                            if(Cine:=Cine_L21_Normal[GetTeamIndexAgent[Agent]-1]):
                                Cine.Play(Agent)
                            Sleep(4.0)
                            if(Cine:=Cine_L21_Normal[GetTeamIndexAgent[Agent]-1]):
                                Cine.Stop(Agent)
                                Accolade_Egg21_22.Award(Agent)

                            if (CustomPlayer.Ownership_L21_Pet1=0 and CustomPlayer.Ownership_L21_Pet2=0 and CustomPlayer.Ownership_L21_Pet3=0 and CustomPlayer.Ownership_L21_Pet4=0):
                                #Get New pet
                                L21_Pet1_Own(Agent)
                                set CustomPlayer.Ownership_L21_Pet1=1
                                set CustomPlayer.Pet_Unlocked_L21=1
                                CustomPlayer.Update_Pet_PowerL21(Power_L21P1)
                                var PetPower:float=1.0*Power_L21P1
                                set PetWon = FirstPet
                                spawn:
                                    FollowPlayerL21(Fort, PetWon, PetPower, FirstPetIndex)


    PurchasePetL22( Agent : agent )<suspends>: void =
        #Requires Unlock L22
        if:
            PlayerObj := player[Agent]
            CurrentPlayerStats := PlayerStatsManager.GetPlayerStats[Agent]
            CurrentPlayerStats.LevelUnlocked.CurrentValue<21
        then:
            RequireHigherLevelHUD.Show(Agent) 
        else:
            if:
                PlayerObj := player[Agent]
                CustomPlayer:Custom_Player = PlayersMap[PlayerObj]
                CustomPlayer.PlayerGold<CostL22
            then:
                NotEnoughGoldHUD.Show(Agent)
            else:
                DecreaseGoldFunc(Agent,CostL22)
                var FirstPet:creative_prop = creative_prop{}
                var SecondPet:creative_prop = creative_prop{}
                var ThirdPet:creative_prop = creative_prop{}
                var ForthPet:creative_prop = creative_prop{}

                var FirstPetIndex:int = 0
                var SecondPetIndex:int = 0
                var ThirdPetIndex:int = 0
                var ForthPetIndex:int = 0

                var IndexBase:int = 12*21

                #Select right pets for each player
                if(GetTeamIndexAgent[Agent]=1):
                    set FirstPetIndex =  IndexBase
                    set SecondPetIndex = IndexBase+1
                    set ThirdPetIndex =  IndexBase+2
                    set ForthPetIndex =  IndexBase+3
                if(GetTeamIndexAgent[Agent]=2):
                    set FirstPetIndex =  IndexBase+4
                    set SecondPetIndex = IndexBase+5
                    set ThirdPetIndex =  IndexBase+6
                    set ForthPetIndex =  IndexBase+7
                if(GetTeamIndexAgent[Agent]=3):
                    set FirstPetIndex =  IndexBase+8
                    set SecondPetIndex = IndexBase+9
                    set ThirdPetIndex =  IndexBase+10
                    set ForthPetIndex =  IndexBase+11

                if (Mypet:=Pets[FirstPetIndex]):
                    set FirstPet = Mypet.PetProp
                if (Mypet:=Pets[SecondPetIndex]):
                    set SecondPet = Mypet.PetProp
                if (Mypet:=Pets[ThirdPetIndex]):
                    set ThirdPet = Mypet.PetProp
                if (Mypet:=Pets[ForthPetIndex]):
                    set ForthPet = Mypet.PetProp

                var PetWon:creative_prop = creative_prop{}
                var RandResult:int = GetRandomInt(0,1000)
                Print("Rand: {RandResult}")
                if:
                    PlayerObj := player[Agent]
                    Fort := Agent.GetFortCharacter[]
                then:
                    if(CustomPlayer:Custom_Player = PlayersMap[PlayerObj]):
                        #--------------Luck
                        set RandResult+=CustomPlayer.Pet_Luck_Value*5+CustomPlayer.Bonus_Luck*5
                        #--------------
                        if (RandResult>995):
                            Print("Legendary")
                            if(Cine:=Cine_L22_Legendary[GetTeamIndexAgent[Agent]-1]):
                                Cine.Play(Agent)
                            Sleep(4.0)
                            if(Cine:=Cine_L22_Legendary[GetTeamIndexAgent[Agent]-1]):
                                Cine.Stop(Agent)
                                Accolade_Egg21_22.Award(Agent)

                            if (CustomPlayer.Ownership_L22_Pet4=0):
                                #Release potential previous pets
                                if (CustomPlayer.Ownership_L22_Pet1=1):
                                    #--------------Button pressed:
                                    CustomPlayer.IncreaseButtonL22()
                                    #Print("Button L22 Pressed: {CustomPlayer.ButtonL22Pressed}")
                                    #--------------Spawn
                                    Sleep(0.5)
                                    ##spawn:
                                    ##    ReleasePet(FirstPet)
                                    set CustomPlayer.Ownership_L22_Pet1=0
                                    L22_Pet1_DesOwn(Agent)
                                if (CustomPlayer.Ownership_L22_Pet2=1):
                                    #--------------Button pressed:
                                    CustomPlayer.IncreaseButtonL22()
                                    #Print("Button L22 Pressed: {CustomPlayer.ButtonL22Pressed}")
                                    #--------------Spawn
                                    Sleep(0.5)
                                    ##spawn:
                                    ##    ReleasePet(SecondPet)
                                    set CustomPlayer.Ownership_L22_Pet2=0
                                    L22_Pet2_DesOwn(Agent)
                                if (CustomPlayer.Ownership_L22_Pet3=1):
                                    #--------------Button pressed:
                                    CustomPlayer.IncreaseButtonL22()
                                    #Print("Button L22 Pressed: {CustomPlayer.ButtonL22Pressed}")
                                    #--------------Spawn
                                    Sleep(0.5)
                                    ##spawn:
                                    ##    ReleasePet(ThirdPet)
                                    set CustomPlayer.Ownership_L22_Pet3=0
                                    L22_Pet3_DesOwn(Agent)
                                #Get New pet
                                L22_Pet4_Own(Agent)
                                set CustomPlayer.Ownership_L22_Pet4=1
                                set CustomPlayer.Pet_Unlocked_L22=4
                                CustomPlayer.Update_Pet_PowerL22(Power_L22P4)
                                var PetPower:float=1.0*Power_L22P4
                                set PetWon = ForthPet
                                spawn:
                                    FollowPlayerL22(Fort, PetWon, PetPower, ForthPetIndex)
                        else if (RandResult>855):
                            Print("Epic")
                            if(Cine:=Cine_L22_Epic[GetTeamIndexAgent[Agent]-1]):
                                Cine.Play(Agent)
                            Sleep(4.0)
                            if(Cine:=Cine_L22_Epic[GetTeamIndexAgent[Agent]-1]):
                                Cine.Stop(Agent)
                                Accolade_Egg21_22.Award(Agent)

                            if (CustomPlayer.Ownership_L22_Pet3=0 and CustomPlayer.Ownership_L22_Pet4=0):
                                #Release potential previous pets
                                if (CustomPlayer.Ownership_L22_Pet1=1):
                                    #--------------Button pressed:
                                    CustomPlayer.IncreaseButtonL22()
                                    #Print("Button L22 Pressed: {CustomPlayer.ButtonL22Pressed}")
                                    #--------------Spawn
                                    Sleep(0.5)
                                    ##spawn:
                                    ##    ReleasePet(FirstPet)
                                    set CustomPlayer.Ownership_L22_Pet1=0
                                    L22_Pet1_DesOwn(Agent)
                                if (CustomPlayer.Ownership_L22_Pet2=1):
                                    #--------------Button pressed:
                                    CustomPlayer.IncreaseButtonL22()
                                    #Print("Button L22 Pressed: {CustomPlayer.ButtonL22Pressed}")
                                    #--------------Spawn
                                    Sleep(0.5)
                                    ##spawn:
                                    ##    ReleasePet(SecondPet)
                                    set CustomPlayer.Ownership_L22_Pet2=0
                                    L22_Pet2_DesOwn(Agent)
                                #Get New pet
                                L22_Pet3_Own(Agent)
                                set CustomPlayer.Ownership_L22_Pet3=1
                                set CustomPlayer.Pet_Unlocked_L22=3
                                CustomPlayer.Update_Pet_PowerL22(Power_L22P3)
                                var PetPower:float=1.0*Power_L22P3
                                set PetWon = ThirdPet
                                spawn:
                                    FollowPlayerL22(Fort, PetWon, PetPower, ThirdPetIndex)
                        else if (RandResult>750):
                            Print("Rare")
                            if(Cine:=Cine_L22_Rare[GetTeamIndexAgent[Agent]-1]):
                                Cine.Play(Agent)
                            Sleep(4.0)
                            if(Cine:=Cine_L22_Rare[GetTeamIndexAgent[Agent]-1]):
                                Cine.Stop(Agent)
                                Accolade_Egg21_22.Award(Agent)

                            if (CustomPlayer.Ownership_L22_Pet2=0 and CustomPlayer.Ownership_L22_Pet3=0 and CustomPlayer.Ownership_L22_Pet4=0):
                                #Release potential previous pets
                                if (CustomPlayer.Ownership_L22_Pet1=1):
                                    #--------------Button pressed:
                                    CustomPlayer.IncreaseButtonL22()
                                    #Print("Button L22 Pressed: {CustomPlayer.ButtonL22Pressed}")
                                    #--------------Spawn
                                    Sleep(0.5)
                                    ##spawn:
                                    ##    ReleasePet(FirstPet)
                                    set CustomPlayer.Ownership_L22_Pet1=0
                                    L22_Pet1_DesOwn(Agent)
                                #Get New pet
                                L22_Pet2_Own(Agent)
                                set CustomPlayer.Ownership_L22_Pet2=1
                                set CustomPlayer.Pet_Unlocked_L22=2
                                CustomPlayer.Update_Pet_PowerL22(Power_L22P2)
                                var PetPower:float=1.0*Power_L22P2
                                set PetWon = SecondPet
                                spawn:
                                    FollowPlayerL22(Fort, PetWon, PetPower, SecondPetIndex)
                        else:
                            Print("Normal")
                            if(Cine:=Cine_L22_Normal[GetTeamIndexAgent[Agent]-1]):
                                Cine.Play(Agent)
                            Sleep(4.0)
                            if(Cine:=Cine_L22_Normal[GetTeamIndexAgent[Agent]-1]):
                                Cine.Stop(Agent)
                                Accolade_Egg21_22.Award(Agent)

                            if (CustomPlayer.Ownership_L22_Pet1=0 and CustomPlayer.Ownership_L22_Pet2=0 and CustomPlayer.Ownership_L22_Pet3=0 and CustomPlayer.Ownership_L22_Pet4=0):
                                #Get New pet
                                L22_Pet1_Own(Agent)
                                set CustomPlayer.Ownership_L22_Pet1=1
                                set CustomPlayer.Pet_Unlocked_L22=1
                                CustomPlayer.Update_Pet_PowerL22(Power_L22P1)
                                var PetPower:float=1.0*Power_L22P1
                                set PetWon = FirstPet
                                spawn:
                                    FollowPlayerL22(Fort, PetWon, PetPower, FirstPetIndex)

    PurchasePetL23( Agent : agent )<suspends>: void =
        #Requires Unlock L23
        if:
            PlayerObj := player[Agent]
            CurrentPlayerStats := PlayerStatsManager.GetPlayerStats[Agent]
            CurrentPlayerStats.LevelUnlocked.CurrentValue<22
        then:
            RequireHigherLevelHUD.Show(Agent) 
        else:
            if:
                PlayerObj := player[Agent]
                CustomPlayer:Custom_Player = PlayersMap[PlayerObj]
                CustomPlayer.PlayerGold<CostL23
            then:
                NotEnoughGoldHUD.Show(Agent)
            else:
                DecreaseGoldFunc(Agent,CostL23)
                var FirstPet:creative_prop = creative_prop{}
                var SecondPet:creative_prop = creative_prop{}
                var ThirdPet:creative_prop = creative_prop{}
                var ForthPet:creative_prop = creative_prop{}

                var FirstPetIndex:int = 0
                var SecondPetIndex:int = 0
                var ThirdPetIndex:int = 0
                var ForthPetIndex:int = 0

                var IndexBase:int = 12*22

                #Select right pets for each player
                if(GetTeamIndexAgent[Agent]=1):
                    set FirstPetIndex =  IndexBase
                    set SecondPetIndex = IndexBase+1
                    set ThirdPetIndex =  IndexBase+2
                    set ForthPetIndex =  IndexBase+3
                if(GetTeamIndexAgent[Agent]=2):
                    set FirstPetIndex =  IndexBase+4
                    set SecondPetIndex = IndexBase+5
                    set ThirdPetIndex =  IndexBase+6
                    set ForthPetIndex =  IndexBase+7
                if(GetTeamIndexAgent[Agent]=3):
                    set FirstPetIndex =  IndexBase+8
                    set SecondPetIndex = IndexBase+9
                    set ThirdPetIndex =  IndexBase+10
                    set ForthPetIndex =  IndexBase+11

                if (Mypet:=Pets[FirstPetIndex]):
                    set FirstPet = Mypet.PetProp
                if (Mypet:=Pets[SecondPetIndex]):
                    set SecondPet = Mypet.PetProp
                if (Mypet:=Pets[ThirdPetIndex]):
                    set ThirdPet = Mypet.PetProp
                if (Mypet:=Pets[ForthPetIndex]):
                    set ForthPet = Mypet.PetProp

                var PetWon:creative_prop = creative_prop{}
                var RandResult:int = GetRandomInt(0,1000)
                Print("Rand: {RandResult}")
                if:
                    PlayerObj := player[Agent]
                    Fort := Agent.GetFortCharacter[]
                then:
                    if(CustomPlayer:Custom_Player = PlayersMap[PlayerObj]):
                        #--------------Luck
                        set RandResult+=CustomPlayer.Pet_Luck_Value*5+CustomPlayer.Bonus_Luck*5
                        #--------------
                        if (RandResult>995):
                            Print("Legendary")
                            if(Cine:=Cine_L23_Legendary[GetTeamIndexAgent[Agent]-1]):
                                Cine.Play(Agent)
                            Sleep(4.0)
                            if(Cine:=Cine_L23_Legendary[GetTeamIndexAgent[Agent]-1]):
                                Cine.Stop(Agent)
                                Accolade_Egg23_24.Award(Agent)

                            if (CustomPlayer.Ownership_L23_Pet4=0):
                                #Release potential previous pets
                                if (CustomPlayer.Ownership_L23_Pet1=1):
                                    #--------------Button pressed:
                                    CustomPlayer.IncreaseButtonL23()
                                    #Print("Button L23 Pressed: {CustomPlayer.ButtonL23Pressed}")
                                    #--------------Spawn
                                    Sleep(0.5)
                                    ##spawn:
                                    ##    ReleasePet(FirstPet)
                                    set CustomPlayer.Ownership_L23_Pet1=0
                                    L23_Pet1_DesOwn(Agent)
                                if (CustomPlayer.Ownership_L23_Pet2=1):
                                    #--------------Button pressed:
                                    CustomPlayer.IncreaseButtonL23()
                                    #Print("Button L23 Pressed: {CustomPlayer.ButtonL23Pressed}")
                                    #--------------Spawn
                                    Sleep(0.5)
                                    ##spawn:
                                    ##    ReleasePet(SecondPet)
                                    set CustomPlayer.Ownership_L23_Pet2=0
                                    L23_Pet2_DesOwn(Agent)
                                if (CustomPlayer.Ownership_L23_Pet3=1):
                                    #--------------Button pressed:
                                    CustomPlayer.IncreaseButtonL23()
                                    #Print("Button L23 Pressed: {CustomPlayer.ButtonL23Pressed}")
                                    #--------------Spawn
                                    Sleep(0.5)
                                    ##spawn:
                                    ##    ReleasePet(ThirdPet)
                                    set CustomPlayer.Ownership_L23_Pet3=0
                                    L23_Pet3_DesOwn(Agent)
                                #Get New pet
                                L23_Pet4_Own(Agent)
                                set CustomPlayer.Ownership_L23_Pet4=1
                                set CustomPlayer.Pet_Unlocked_L23=4
                                CustomPlayer.Update_Pet_PowerL23(Power_L23P4)
                                var PetPower:float=1.0*Power_L23P4
                                set PetWon = ForthPet
                                spawn:
                                    FollowPlayerL23(Fort, PetWon, PetPower, ForthPetIndex)
                        else if (RandResult>855):
                            Print("Epic")
                            if(Cine:=Cine_L23_Epic[GetTeamIndexAgent[Agent]-1]):
                                Cine.Play(Agent)
                            Sleep(4.0)
                            if(Cine:=Cine_L23_Epic[GetTeamIndexAgent[Agent]-1]):
                                Cine.Stop(Agent)
                                Accolade_Egg23_24.Award(Agent)

                            if (CustomPlayer.Ownership_L23_Pet3=0 and CustomPlayer.Ownership_L23_Pet4=0):
                                #Release potential previous pets
                                if (CustomPlayer.Ownership_L23_Pet1=1):
                                    #--------------Button pressed:
                                    CustomPlayer.IncreaseButtonL23()
                                    #Print("Button L23 Pressed: {CustomPlayer.ButtonL23Pressed}")
                                    #--------------Spawn
                                    Sleep(0.5)
                                    ##spawn:
                                    ##    ReleasePet(FirstPet)
                                    set CustomPlayer.Ownership_L23_Pet1=0
                                    L23_Pet1_DesOwn(Agent)
                                if (CustomPlayer.Ownership_L23_Pet2=1):
                                    #--------------Button pressed:
                                    CustomPlayer.IncreaseButtonL23()
                                    #Print("Button L23 Pressed: {CustomPlayer.ButtonL23Pressed}")
                                    #--------------Spawn
                                    Sleep(0.5)
                                    ##spawn:
                                    ##    ReleasePet(SecondPet)
                                    set CustomPlayer.Ownership_L23_Pet2=0
                                    L23_Pet2_DesOwn(Agent)
                                #Get New pet
                                L23_Pet3_Own(Agent)
                                set CustomPlayer.Ownership_L23_Pet3=1
                                set CustomPlayer.Pet_Unlocked_L23=3
                                CustomPlayer.Update_Pet_PowerL23(Power_L23P3)
                                var PetPower:float=1.0*Power_L23P3
                                set PetWon = ThirdPet
                                spawn:
                                    FollowPlayerL23(Fort, PetWon, PetPower, ThirdPetIndex)
                        else if (RandResult>750):
                            Print("Rare")
                            if(Cine:=Cine_L23_Rare[GetTeamIndexAgent[Agent]-1]):
                                Cine.Play(Agent)
                            Sleep(4.0)
                            if(Cine:=Cine_L23_Rare[GetTeamIndexAgent[Agent]-1]):
                                Cine.Stop(Agent)
                                Accolade_Egg23_24.Award(Agent)

                            if (CustomPlayer.Ownership_L23_Pet2=0 and CustomPlayer.Ownership_L23_Pet3=0 and CustomPlayer.Ownership_L23_Pet4=0):
                                #Release potential previous pets
                                if (CustomPlayer.Ownership_L23_Pet1=1):
                                    #--------------Button pressed:
                                    CustomPlayer.IncreaseButtonL23()
                                    #Print("Button L23 Pressed: {CustomPlayer.ButtonL23Pressed}")
                                    #--------------Spawn
                                    Sleep(0.5)
                                    ##spawn:
                                    ##    ReleasePet(FirstPet)
                                    set CustomPlayer.Ownership_L23_Pet1=0
                                    L23_Pet1_DesOwn(Agent)
                                #Get New pet
                                L23_Pet2_Own(Agent)
                                set CustomPlayer.Ownership_L23_Pet2=1
                                set CustomPlayer.Pet_Unlocked_L23=2
                                CustomPlayer.Update_Pet_PowerL23(Power_L23P2)
                                var PetPower:float=1.0*Power_L23P2
                                set PetWon = SecondPet
                                spawn:
                                    FollowPlayerL23(Fort, PetWon, PetPower, SecondPetIndex)
                        else:
                            Print("Normal")
                            if(Cine:=Cine_L23_Normal[GetTeamIndexAgent[Agent]-1]):
                                Cine.Play(Agent)
                            Sleep(4.0)
                            if(Cine:=Cine_L23_Normal[GetTeamIndexAgent[Agent]-1]):
                                Cine.Stop(Agent)
                                Accolade_Egg23_24.Award(Agent)

                            if (CustomPlayer.Ownership_L23_Pet1=0 and CustomPlayer.Ownership_L23_Pet2=0 and CustomPlayer.Ownership_L23_Pet3=0 and CustomPlayer.Ownership_L23_Pet4=0):
                                #Get New pet
                                L23_Pet1_Own(Agent)
                                set CustomPlayer.Ownership_L23_Pet1=1
                                set CustomPlayer.Pet_Unlocked_L23=1
                                CustomPlayer.Update_Pet_PowerL23(Power_L23P1)
                                var PetPower:float=1.0*Power_L23P1
                                set PetWon = FirstPet
                                spawn:
                                    FollowPlayerL23(Fort, PetWon, PetPower, FirstPetIndex)

    PurchasePetL24( Agent : agent )<suspends>: void =
        #Requires Unlock L24
        if:
            PlayerObj := player[Agent]
            CurrentPlayerStats := PlayerStatsManager.GetPlayerStats[Agent]
            CurrentPlayerStats.LevelUnlocked.CurrentValue<23
        then:
            RequireHigherLevelHUD.Show(Agent) 
        else:
            if:
                PlayerObj := player[Agent]
                CustomPlayer:Custom_Player = PlayersMap[PlayerObj]
                CustomPlayer.PlayerGold<CostL24
            then:
                NotEnoughGoldHUD.Show(Agent)
            else:
                DecreaseGoldFunc(Agent,CostL24)
                var FirstPet:creative_prop = creative_prop{}
                var SecondPet:creative_prop = creative_prop{}
                var ThirdPet:creative_prop = creative_prop{}
                var ForthPet:creative_prop = creative_prop{}

                var FirstPetIndex:int = 0
                var SecondPetIndex:int = 0
                var ThirdPetIndex:int = 0
                var ForthPetIndex:int = 0

                var IndexBase:int = 12*23

                #Select right pets for each player
                if(GetTeamIndexAgent[Agent]=1):
                    set FirstPetIndex =  IndexBase
                    set SecondPetIndex = IndexBase+1
                    set ThirdPetIndex =  IndexBase+2
                    set ForthPetIndex =  IndexBase+3
                if(GetTeamIndexAgent[Agent]=2):
                    set FirstPetIndex =  IndexBase+4
                    set SecondPetIndex = IndexBase+5
                    set ThirdPetIndex =  IndexBase+6
                    set ForthPetIndex =  IndexBase+7
                if(GetTeamIndexAgent[Agent]=3):
                    set FirstPetIndex =  IndexBase+8
                    set SecondPetIndex = IndexBase+9
                    set ThirdPetIndex =  IndexBase+10
                    set ForthPetIndex =  IndexBase+11

                if (Mypet:=Pets[FirstPetIndex]):
                    set FirstPet = Mypet.PetProp
                if (Mypet:=Pets[SecondPetIndex]):
                    set SecondPet = Mypet.PetProp
                if (Mypet:=Pets[ThirdPetIndex]):
                    set ThirdPet = Mypet.PetProp
                if (Mypet:=Pets[ForthPetIndex]):
                    set ForthPet = Mypet.PetProp

                var PetWon:creative_prop = creative_prop{}
                var RandResult:int = GetRandomInt(0,1000)
                Print("Rand: {RandResult}")
                if:
                    PlayerObj := player[Agent]
                    Fort := Agent.GetFortCharacter[]
                then:
                    if(CustomPlayer:Custom_Player = PlayersMap[PlayerObj]):
                        #--------------Luck
                        set RandResult+=CustomPlayer.Pet_Luck_Value*5+CustomPlayer.Bonus_Luck*5
                        #--------------
                        if (RandResult>995):
                            Print("Legendary")
                            if(Cine:=Cine_L24_Legendary[GetTeamIndexAgent[Agent]-1]):
                                Cine.Play(Agent)
                            Sleep(4.0)
                            if(Cine:=Cine_L24_Legendary[GetTeamIndexAgent[Agent]-1]):
                                Cine.Stop(Agent)
                                Accolade_Egg23_24.Award(Agent)

                            if (CustomPlayer.Ownership_L24_Pet4=0):
                                #Release potential previous pets
                                if (CustomPlayer.Ownership_L24_Pet1=1):
                                    #--------------Button pressed:
                                    CustomPlayer.IncreaseButtonL24()
                                    #Print("Button L24 Pressed: {CustomPlayer.ButtonL24Pressed}")
                                    #--------------Spawn
                                    Sleep(0.5)
                                    ##spawn:
                                    ##    ReleasePet(FirstPet)
                                    set CustomPlayer.Ownership_L24_Pet1=0
                                    L24_Pet1_DesOwn(Agent)
                                if (CustomPlayer.Ownership_L24_Pet2=1):
                                    #--------------Button pressed:
                                    CustomPlayer.IncreaseButtonL24()
                                    #Print("Button L24 Pressed: {CustomPlayer.ButtonL24Pressed}")
                                    #--------------Spawn
                                    Sleep(0.5)
                                    ##spawn:
                                    ##    ReleasePet(SecondPet)
                                    set CustomPlayer.Ownership_L24_Pet2=0
                                    L24_Pet2_DesOwn(Agent)
                                if (CustomPlayer.Ownership_L24_Pet3=1):
                                    #--------------Button pressed:
                                    CustomPlayer.IncreaseButtonL24()
                                    #Print("Button L24 Pressed: {CustomPlayer.ButtonL24Pressed}")
                                    #--------------Spawn
                                    Sleep(0.5)
                                    ##spawn:
                                    ##    ReleasePet(ThirdPet)
                                    set CustomPlayer.Ownership_L24_Pet3=0
                                    L24_Pet3_DesOwn(Agent)
                                #Get New pet
                                L24_Pet4_Own(Agent)
                                set CustomPlayer.Ownership_L24_Pet4=1
                                set CustomPlayer.Pet_Unlocked_L24=4
                                CustomPlayer.Update_Pet_PowerL24(Power_L24P4)
                                var PetPower:float=1.0*Power_L24P4
                                set PetWon = ForthPet
                                spawn:
                                    FollowPlayerL24(Fort, PetWon, PetPower, ForthPetIndex)
                        else if (RandResult>855):
                            Print("Epic")
                            if(Cine:=Cine_L24_Epic[GetTeamIndexAgent[Agent]-1]):
                                Cine.Play(Agent)
                            Sleep(4.0)
                            if(Cine:=Cine_L24_Epic[GetTeamIndexAgent[Agent]-1]):
                                Cine.Stop(Agent)
                                Accolade_Egg23_24.Award(Agent)

                            if (CustomPlayer.Ownership_L24_Pet3=0 and CustomPlayer.Ownership_L24_Pet4=0):
                                #Release potential previous pets
                                if (CustomPlayer.Ownership_L24_Pet1=1):
                                    #--------------Button pressed:
                                    CustomPlayer.IncreaseButtonL24()
                                    #Print("Button L24 Pressed: {CustomPlayer.ButtonL24Pressed}")
                                    #--------------Spawn
                                    Sleep(0.5)
                                    ##spawn:
                                    ##    ReleasePet(FirstPet)
                                    set CustomPlayer.Ownership_L24_Pet1=0
                                    L24_Pet1_DesOwn(Agent)
                                if (CustomPlayer.Ownership_L24_Pet2=1):
                                    #--------------Button pressed:
                                    CustomPlayer.IncreaseButtonL24()
                                    #Print("Button L24 Pressed: {CustomPlayer.ButtonL24Pressed}")
                                    #--------------Spawn
                                    Sleep(0.5)
                                    ##spawn:
                                    ##    ReleasePet(SecondPet)
                                    set CustomPlayer.Ownership_L24_Pet2=0
                                    L24_Pet2_DesOwn(Agent)
                                #Get New pet
                                L24_Pet3_Own(Agent)
                                set CustomPlayer.Ownership_L24_Pet3=1
                                set CustomPlayer.Pet_Unlocked_L24=3
                                CustomPlayer.Update_Pet_PowerL24(Power_L24P3)
                                var PetPower:float=1.0*Power_L24P3
                                set PetWon = ThirdPet
                                spawn:
                                    FollowPlayerL24(Fort, PetWon, PetPower, ThirdPetIndex)
                        else if (RandResult>750):
                            Print("Rare")
                            if(Cine:=Cine_L24_Rare[GetTeamIndexAgent[Agent]-1]):
                                Cine.Play(Agent)
                            Sleep(4.0)
                            if(Cine:=Cine_L24_Rare[GetTeamIndexAgent[Agent]-1]):
                                Cine.Stop(Agent)
                                Accolade_Egg23_24.Award(Agent)

                            if (CustomPlayer.Ownership_L24_Pet2=0 and CustomPlayer.Ownership_L24_Pet3=0 and CustomPlayer.Ownership_L24_Pet4=0):
                                #Release potential previous pets
                                if (CustomPlayer.Ownership_L24_Pet1=1):
                                    #--------------Button pressed:
                                    CustomPlayer.IncreaseButtonL24()
                                    #Print("Button L24 Pressed: {CustomPlayer.ButtonL24Pressed}")
                                    #--------------Spawn
                                    Sleep(0.5)
                                    ##spawn:
                                    ##    ReleasePet(FirstPet)
                                    set CustomPlayer.Ownership_L24_Pet1=0
                                    L24_Pet1_DesOwn(Agent)
                                #Get New pet
                                L24_Pet2_Own(Agent)
                                set CustomPlayer.Ownership_L24_Pet2=1
                                set CustomPlayer.Pet_Unlocked_L24=2
                                CustomPlayer.Update_Pet_PowerL24(Power_L24P2)
                                var PetPower:float=1.0*Power_L24P2
                                set PetWon = SecondPet
                                spawn:
                                    FollowPlayerL24(Fort, PetWon, PetPower, SecondPetIndex)
                        else:
                            Print("Normal")
                            if(Cine:=Cine_L24_Normal[GetTeamIndexAgent[Agent]-1]):
                                Cine.Play(Agent)
                            Sleep(4.0)
                            if(Cine:=Cine_L24_Normal[GetTeamIndexAgent[Agent]-1]):
                                Cine.Stop(Agent)
                                Accolade_Egg23_24.Award(Agent)

                            if (CustomPlayer.Ownership_L24_Pet1=0 and CustomPlayer.Ownership_L24_Pet2=0 and CustomPlayer.Ownership_L24_Pet3=0 and CustomPlayer.Ownership_L24_Pet4=0):
                                #Get New pet
                                L24_Pet1_Own(Agent)
                                set CustomPlayer.Ownership_L24_Pet1=1
                                set CustomPlayer.Pet_Unlocked_L24=1
                                CustomPlayer.Update_Pet_PowerL24(Power_L24P1)
                                var PetPower:float=1.0*Power_L24P1
                                set PetWon = FirstPet
                                spawn:
                                    FollowPlayerL24(Fort, PetWon, PetPower, FirstPetIndex)

    PurchasePetL25( Agent : agent )<suspends>: void =
        #Requires Unlock L25
        if:
            PlayerObj := player[Agent]
            CurrentPlayerStats := PlayerStatsManager.GetPlayerStats[Agent]
            CurrentPlayerStats.LevelUnlocked.CurrentValue<24
        then:
            RequireHigherLevelHUD.Show(Agent) 
        else:
            if:
                PlayerObj := player[Agent]
                CustomPlayer:Custom_Player = PlayersMap[PlayerObj]
                CustomPlayer.PlayerGold<CostL25
            then:
                NotEnoughGoldHUD.Show(Agent)
            else:
                DecreaseGoldFunc(Agent,CostL25)
                var FirstPet:creative_prop = creative_prop{}
                var SecondPet:creative_prop = creative_prop{}
                var ThirdPet:creative_prop = creative_prop{}
                var ForthPet:creative_prop = creative_prop{}

                var FirstPetIndex:int = 0
                var SecondPetIndex:int = 0
                var ThirdPetIndex:int = 0
                var ForthPetIndex:int = 0

                var IndexBase:int = 12*24

                #Select right pets for each player
                if(GetTeamIndexAgent[Agent]=1):
                    set FirstPetIndex =  IndexBase
                    set SecondPetIndex = IndexBase+1
                    set ThirdPetIndex =  IndexBase+2
                    set ForthPetIndex =  IndexBase+3
                if(GetTeamIndexAgent[Agent]=2):
                    set FirstPetIndex =  IndexBase+4
                    set SecondPetIndex = IndexBase+5
                    set ThirdPetIndex =  IndexBase+6
                    set ForthPetIndex =  IndexBase+7
                if(GetTeamIndexAgent[Agent]=3):
                    set FirstPetIndex =  IndexBase+8
                    set SecondPetIndex = IndexBase+9
                    set ThirdPetIndex =  IndexBase+10
                    set ForthPetIndex =  IndexBase+11

                if (Mypet:=Pets[FirstPetIndex]):
                    set FirstPet = Mypet.PetProp
                if (Mypet:=Pets[SecondPetIndex]):
                    set SecondPet = Mypet.PetProp
                if (Mypet:=Pets[ThirdPetIndex]):
                    set ThirdPet = Mypet.PetProp
                if (Mypet:=Pets[ForthPetIndex]):
                    set ForthPet = Mypet.PetProp

                var PetWon:creative_prop = creative_prop{}
                var RandResult:int = GetRandomInt(0,1000)
                Print("Rand: {RandResult}")
                if:
                    PlayerObj := player[Agent]
                    Fort := Agent.GetFortCharacter[]
                then:
                    if(CustomPlayer:Custom_Player = PlayersMap[PlayerObj]):
                        #--------------Luck
                        set RandResult+=CustomPlayer.Pet_Luck_Value*5+CustomPlayer.Bonus_Luck*5
                        #--------------
                        if (RandResult>995):
                            Print("Legendary")
                            if(Cine:=Cine_L25_Legendary[GetTeamIndexAgent[Agent]-1]):
                                Cine.Play(Agent)
                            Sleep(4.0)
                            if(Cine:=Cine_L25_Legendary[GetTeamIndexAgent[Agent]-1]):
                                Cine.Stop(Agent)
                                Accolade_Egg25_26.Award(Agent)

                            if (CustomPlayer.Ownership_L25_Pet4=0):
                                #Release potential previous pets
                                if (CustomPlayer.Ownership_L25_Pet1=1):
                                    #--------------Button pressed:
                                    CustomPlayer.IncreaseButtonL25()
                                    #Print("Button L25 Pressed: {CustomPlayer.ButtonL25Pressed}")
                                    #--------------Spawn
                                    Sleep(0.5)
                                    ##spawn:
                                    ##    ReleasePet(FirstPet)
                                    set CustomPlayer.Ownership_L25_Pet1=0
                                    L25_Pet1_DesOwn(Agent)
                                if (CustomPlayer.Ownership_L25_Pet2=1):
                                    #--------------Button pressed:
                                    CustomPlayer.IncreaseButtonL25()
                                    #Print("Button L25 Pressed: {CustomPlayer.ButtonL25Pressed}")
                                    #--------------Spawn
                                    Sleep(0.5)
                                    ##spawn:
                                    ##    ReleasePet(SecondPet)
                                    set CustomPlayer.Ownership_L25_Pet2=0
                                    L25_Pet2_DesOwn(Agent)
                                if (CustomPlayer.Ownership_L25_Pet3=1):
                                    #--------------Button pressed:
                                    CustomPlayer.IncreaseButtonL25()
                                    #Print("Button L25 Pressed: {CustomPlayer.ButtonL25Pressed}")
                                    #--------------Spawn
                                    Sleep(0.5)
                                    ##spawn:
                                    ##    ReleasePet(ThirdPet)
                                    set CustomPlayer.Ownership_L25_Pet3=0
                                    L25_Pet3_DesOwn(Agent)
                                #Get New pet
                                L25_Pet4_Own(Agent)
                                set CustomPlayer.Ownership_L25_Pet4=1
                                set CustomPlayer.Pet_Unlocked_L25=4
                                CustomPlayer.Update_Pet_PowerL25(Power_L25P4)
                                var PetPower:float=1.0*Power_L25P4
                                set PetWon = ForthPet
                                spawn:
                                    FollowPlayerL25(Fort, PetWon, PetPower, ForthPetIndex)
                        else if (RandResult>855):
                            Print("Epic")
                            if(Cine:=Cine_L25_Epic[GetTeamIndexAgent[Agent]-1]):
                                Cine.Play(Agent)
                            Sleep(4.0)
                            if(Cine:=Cine_L25_Epic[GetTeamIndexAgent[Agent]-1]):
                                Cine.Stop(Agent)
                                Accolade_Egg25_26.Award(Agent)

                            if (CustomPlayer.Ownership_L25_Pet3=0 and CustomPlayer.Ownership_L25_Pet4=0):
                                #Release potential previous pets
                                if (CustomPlayer.Ownership_L25_Pet1=1):
                                    #--------------Button pressed:
                                    CustomPlayer.IncreaseButtonL25()
                                    #Print("Button L25 Pressed: {CustomPlayer.ButtonL25Pressed}")
                                    #--------------Spawn
                                    Sleep(0.5)
                                    ##spawn:
                                    ##    ReleasePet(FirstPet)
                                    set CustomPlayer.Ownership_L25_Pet1=0
                                    L25_Pet1_DesOwn(Agent)
                                if (CustomPlayer.Ownership_L25_Pet2=1):
                                    #--------------Button pressed:
                                    CustomPlayer.IncreaseButtonL25()
                                    #Print("Button L25 Pressed: {CustomPlayer.ButtonL25Pressed}")
                                    #--------------Spawn
                                    Sleep(0.5)
                                    ##spawn:
                                    ##    ReleasePet(SecondPet)
                                    set CustomPlayer.Ownership_L25_Pet2=0
                                    L25_Pet2_DesOwn(Agent)
                                #Get New pet
                                L25_Pet3_Own(Agent)
                                set CustomPlayer.Ownership_L25_Pet3=1
                                set CustomPlayer.Pet_Unlocked_L25=3
                                CustomPlayer.Update_Pet_PowerL25(Power_L25P3)
                                var PetPower:float=1.0*Power_L25P3
                                set PetWon = ThirdPet
                                spawn:
                                    FollowPlayerL25(Fort, PetWon, PetPower, ThirdPetIndex)
                        else if (RandResult>750):
                            Print("Rare")
                            if(Cine:=Cine_L25_Rare[GetTeamIndexAgent[Agent]-1]):
                                Cine.Play(Agent)
                            Sleep(4.0)
                            if(Cine:=Cine_L25_Rare[GetTeamIndexAgent[Agent]-1]):
                                Cine.Stop(Agent)
                                Accolade_Egg25_26.Award(Agent)

                            if (CustomPlayer.Ownership_L25_Pet2=0 and CustomPlayer.Ownership_L25_Pet3=0 and CustomPlayer.Ownership_L25_Pet4=0):
                                #Release potential previous pets
                                if (CustomPlayer.Ownership_L25_Pet1=1):
                                    #--------------Button pressed:
                                    CustomPlayer.IncreaseButtonL25()
                                    #Print("Button L25 Pressed: {CustomPlayer.ButtonL25Pressed}")
                                    #--------------Spawn
                                    Sleep(0.5)
                                    ##spawn:
                                    ##    ReleasePet(FirstPet)
                                    set CustomPlayer.Ownership_L25_Pet1=0
                                    L25_Pet1_DesOwn(Agent)
                                #Get New pet
                                L25_Pet2_Own(Agent)
                                set CustomPlayer.Ownership_L25_Pet2=1
                                set CustomPlayer.Pet_Unlocked_L25=2
                                CustomPlayer.Update_Pet_PowerL25(Power_L25P2)
                                var PetPower:float=1.0*Power_L25P2
                                set PetWon = SecondPet
                                spawn:
                                    FollowPlayerL25(Fort, PetWon, PetPower, SecondPetIndex)
                        else:
                            Print("Normal")
                            if(Cine:=Cine_L25_Normal[GetTeamIndexAgent[Agent]-1]):
                                Cine.Play(Agent)
                            Sleep(4.0)
                            if(Cine:=Cine_L25_Normal[GetTeamIndexAgent[Agent]-1]):
                                Cine.Stop(Agent)
                                Accolade_Egg25_26.Award(Agent)

                            if (CustomPlayer.Ownership_L25_Pet1=0 and CustomPlayer.Ownership_L25_Pet2=0 and CustomPlayer.Ownership_L25_Pet3=0 and CustomPlayer.Ownership_L25_Pet4=0):
                                #Get New pet
                                L25_Pet1_Own(Agent)
                                set CustomPlayer.Ownership_L25_Pet1=1
                                set CustomPlayer.Pet_Unlocked_L25=1
                                CustomPlayer.Update_Pet_PowerL25(Power_L25P1)
                                var PetPower:float=1.0*Power_L25P1
                                set PetWon = FirstPet
                                spawn:
                                    FollowPlayerL25(Fort, PetWon, PetPower, FirstPetIndex)

    PurchasePetL26( Agent : agent )<suspends>: void =
        #Requires Unlock L26
        if:
            PlayerObj := player[Agent]
            CurrentPlayerStats := PlayerStatsManager.GetPlayerStats[Agent]
            CurrentPlayerStats.LevelUnlocked.CurrentValue<25
        then:
            RequireHigherLevelHUD.Show(Agent) 
        else:
            if:
                PlayerObj := player[Agent]
                CustomPlayer:Custom_Player = PlayersMap[PlayerObj]
                CustomPlayer.PlayerGold<CostL26
            then:
                NotEnoughGoldHUD.Show(Agent)
            else:
                DecreaseGoldFunc(Agent,CostL26)
                var FirstPet:creative_prop = creative_prop{}
                var SecondPet:creative_prop = creative_prop{}
                var ThirdPet:creative_prop = creative_prop{}
                var ForthPet:creative_prop = creative_prop{}

                var FirstPetIndex:int = 0
                var SecondPetIndex:int = 0
                var ThirdPetIndex:int = 0
                var ForthPetIndex:int = 0

                var IndexBase:int = 12*25

                #Select right pets for each player
                if(GetTeamIndexAgent[Agent]=1):
                    set FirstPetIndex =  IndexBase
                    set SecondPetIndex = IndexBase+1
                    set ThirdPetIndex =  IndexBase+2
                    set ForthPetIndex =  IndexBase+3
                if(GetTeamIndexAgent[Agent]=2):
                    set FirstPetIndex =  IndexBase+4
                    set SecondPetIndex = IndexBase+5
                    set ThirdPetIndex =  IndexBase+6
                    set ForthPetIndex =  IndexBase+7
                if(GetTeamIndexAgent[Agent]=3):
                    set FirstPetIndex =  IndexBase+8
                    set SecondPetIndex = IndexBase+9
                    set ThirdPetIndex =  IndexBase+10
                    set ForthPetIndex =  IndexBase+11

                if (Mypet:=Pets[FirstPetIndex]):
                    set FirstPet = Mypet.PetProp
                if (Mypet:=Pets[SecondPetIndex]):
                    set SecondPet = Mypet.PetProp
                if (Mypet:=Pets[ThirdPetIndex]):
                    set ThirdPet = Mypet.PetProp
                if (Mypet:=Pets[ForthPetIndex]):
                    set ForthPet = Mypet.PetProp

                var PetWon:creative_prop = creative_prop{}
                var RandResult:int = GetRandomInt(0,1000)
                Print("Rand: {RandResult}")
                if:
                    PlayerObj := player[Agent]
                    Fort := Agent.GetFortCharacter[]
                then:
                    if(CustomPlayer:Custom_Player = PlayersMap[PlayerObj]):
                        #--------------Luck
                        set RandResult+=CustomPlayer.Pet_Luck_Value*5+CustomPlayer.Bonus_Luck*5
                        #--------------
                        if (RandResult>995):
                            Print("Legendary")
                            if(Cine:=Cine_L26_Legendary[GetTeamIndexAgent[Agent]-1]):
                                Cine.Play(Agent)
                            Sleep(4.0)
                            if(Cine:=Cine_L26_Legendary[GetTeamIndexAgent[Agent]-1]):
                                Cine.Stop(Agent)
                                Accolade_Egg25_26.Award(Agent)

                            if (CustomPlayer.Ownership_L26_Pet4=0):
                                #Release potential previous pets
                                if (CustomPlayer.Ownership_L26_Pet1=1):
                                    #--------------Button pressed:
                                    CustomPlayer.IncreaseButtonL26()
                                    #Print("Button L26 Pressed: {CustomPlayer.ButtonL26Pressed}")
                                    #--------------Spawn
                                    Sleep(0.5)
                                    ##spawn:
                                    ##    ReleasePet(FirstPet)
                                    set CustomPlayer.Ownership_L26_Pet1=0
                                    L26_Pet1_DesOwn(Agent)
                                if (CustomPlayer.Ownership_L26_Pet2=1):
                                    #--------------Button pressed:
                                    CustomPlayer.IncreaseButtonL26()
                                    #Print("Button L26 Pressed: {CustomPlayer.ButtonL26Pressed}")
                                    #--------------Spawn
                                    Sleep(0.5)
                                    ##spawn:
                                    ##    ReleasePet(SecondPet)
                                    set CustomPlayer.Ownership_L26_Pet2=0
                                    L26_Pet2_DesOwn(Agent)
                                if (CustomPlayer.Ownership_L26_Pet3=1):
                                    #--------------Button pressed:
                                    CustomPlayer.IncreaseButtonL26()
                                    #Print("Button L26 Pressed: {CustomPlayer.ButtonL26Pressed}")
                                    #--------------Spawn
                                    Sleep(0.5)
                                    ##spawn:
                                    ##    ReleasePet(ThirdPet)
                                    set CustomPlayer.Ownership_L26_Pet3=0
                                    L26_Pet3_DesOwn(Agent)
                                #Get New pet
                                L26_Pet4_Own(Agent)
                                set CustomPlayer.Ownership_L26_Pet4=1
                                set CustomPlayer.Pet_Unlocked_L26=4
                                CustomPlayer.Update_Pet_PowerL26(Power_L26P4)
                                var PetPower:float=1.0*Power_L26P4
                                set PetWon = ForthPet
                                spawn:
                                    FollowPlayerL26(Fort, PetWon, PetPower, ForthPetIndex)
                        else if (RandResult>855):
                            Print("Epic")
                            if(Cine:=Cine_L26_Epic[GetTeamIndexAgent[Agent]-1]):
                                Cine.Play(Agent)
                            Sleep(4.0)
                            if(Cine:=Cine_L26_Epic[GetTeamIndexAgent[Agent]-1]):
                                Cine.Stop(Agent)
                                Accolade_Egg25_26.Award(Agent)

                            if (CustomPlayer.Ownership_L26_Pet3=0 and CustomPlayer.Ownership_L26_Pet4=0):
                                #Release potential previous pets
                                if (CustomPlayer.Ownership_L26_Pet1=1):
                                    #--------------Button pressed:
                                    CustomPlayer.IncreaseButtonL26()
                                    #Print("Button L26 Pressed: {CustomPlayer.ButtonL26Pressed}")
                                    #--------------Spawn
                                    Sleep(0.5)
                                    ##spawn:
                                    ##    ReleasePet(FirstPet)
                                    set CustomPlayer.Ownership_L26_Pet1=0
                                    L26_Pet1_DesOwn(Agent)
                                if (CustomPlayer.Ownership_L26_Pet2=1):
                                    #--------------Button pressed:
                                    CustomPlayer.IncreaseButtonL26()
                                    #Print("Button L26 Pressed: {CustomPlayer.ButtonL26Pressed}")
                                    #--------------Spawn
                                    Sleep(0.5)
                                    ##spawn:
                                    ##    ReleasePet(SecondPet)
                                    set CustomPlayer.Ownership_L26_Pet2=0
                                    L26_Pet2_DesOwn(Agent)
                                #Get New pet
                                L26_Pet3_Own(Agent)
                                set CustomPlayer.Ownership_L26_Pet3=1
                                set CustomPlayer.Pet_Unlocked_L26=3
                                CustomPlayer.Update_Pet_PowerL26(Power_L26P3)
                                var PetPower:float=1.0*Power_L26P3
                                set PetWon = ThirdPet
                                spawn:
                                    FollowPlayerL26(Fort, PetWon, PetPower, ThirdPetIndex)
                        else if (RandResult>750):
                            Print("Rare")
                            if(Cine:=Cine_L26_Rare[GetTeamIndexAgent[Agent]-1]):
                                Cine.Play(Agent)
                            Sleep(4.0)
                            if(Cine:=Cine_L26_Rare[GetTeamIndexAgent[Agent]-1]):
                                Cine.Stop(Agent)
                                Accolade_Egg25_26.Award(Agent)

                            if (CustomPlayer.Ownership_L26_Pet2=0 and CustomPlayer.Ownership_L26_Pet3=0 and CustomPlayer.Ownership_L26_Pet4=0):
                                #Release potential previous pets
                                if (CustomPlayer.Ownership_L26_Pet1=1):
                                    #--------------Button pressed:
                                    CustomPlayer.IncreaseButtonL26()
                                    #Print("Button L26 Pressed: {CustomPlayer.ButtonL26Pressed}")
                                    #--------------Spawn
                                    Sleep(0.5)
                                    ##spawn:
                                    ##    ReleasePet(FirstPet)
                                    set CustomPlayer.Ownership_L26_Pet1=0
                                    L26_Pet1_DesOwn(Agent)
                                #Get New pet
                                L26_Pet2_Own(Agent)
                                set CustomPlayer.Ownership_L26_Pet2=1
                                set CustomPlayer.Pet_Unlocked_L26=2
                                CustomPlayer.Update_Pet_PowerL26(Power_L26P2)
                                var PetPower:float=1.0*Power_L26P2
                                set PetWon = SecondPet
                                spawn:
                                    FollowPlayerL26(Fort, PetWon, PetPower, SecondPetIndex)
                        else:
                            Print("Normal")
                            if(Cine:=Cine_L26_Normal[GetTeamIndexAgent[Agent]-1]):
                                Cine.Play(Agent)
                            Sleep(4.0)
                            if(Cine:=Cine_L26_Normal[GetTeamIndexAgent[Agent]-1]):
                                Cine.Stop(Agent)
                                Accolade_Egg25_26.Award(Agent)

                            if (CustomPlayer.Ownership_L26_Pet1=0 and CustomPlayer.Ownership_L26_Pet2=0 and CustomPlayer.Ownership_L26_Pet3=0 and CustomPlayer.Ownership_L26_Pet4=0):
                                #Get New pet
                                L26_Pet1_Own(Agent)
                                set CustomPlayer.Ownership_L26_Pet1=1
                                set CustomPlayer.Pet_Unlocked_L26=1
                                CustomPlayer.Update_Pet_PowerL26(Power_L26P1)
                                var PetPower:float=1.0*Power_L26P1
                                set PetWon = FirstPet
                                spawn:
                                    FollowPlayerL26(Fort, PetWon, PetPower, FirstPetIndex)

    PurchasePetL27( Agent : agent )<suspends>: void =
        #Requires Unlock L27
        if:
            PlayerObj := player[Agent]
            CurrentPlayerStats := PlayerStatsManager.GetPlayerStats[Agent]
            CurrentPlayerStats.LevelUnlocked.CurrentValue<26
        then:
            RequireHigherLevelHUD.Show(Agent) 
        else:
            if:
                PlayerObj := player[Agent]
                CustomPlayer:Custom_Player = PlayersMap[PlayerObj]
                CustomPlayer.PlayerGold<CostL27
            then:
                NotEnoughGoldHUD.Show(Agent)
            else:
                DecreaseGoldFunc(Agent,CostL27)
                var FirstPet:creative_prop = creative_prop{}
                var SecondPet:creative_prop = creative_prop{}
                var ThirdPet:creative_prop = creative_prop{}
                var ForthPet:creative_prop = creative_prop{}

                var FirstPetIndex:int = 0
                var SecondPetIndex:int = 0
                var ThirdPetIndex:int = 0
                var ForthPetIndex:int = 0

                var IndexBase:int = 12*26

                #Select right pets for each player
                if(GetTeamIndexAgent[Agent]=1):
                    set FirstPetIndex =  IndexBase
                    set SecondPetIndex = IndexBase+1
                    set ThirdPetIndex =  IndexBase+2
                    set ForthPetIndex =  IndexBase+3
                if(GetTeamIndexAgent[Agent]=2):
                    set FirstPetIndex =  IndexBase+4
                    set SecondPetIndex = IndexBase+5
                    set ThirdPetIndex =  IndexBase+6
                    set ForthPetIndex =  IndexBase+7
                if(GetTeamIndexAgent[Agent]=3):
                    set FirstPetIndex =  IndexBase+8
                    set SecondPetIndex = IndexBase+9
                    set ThirdPetIndex =  IndexBase+10
                    set ForthPetIndex =  IndexBase+11

                if (Mypet:=Pets[FirstPetIndex]):
                    set FirstPet = Mypet.PetProp
                if (Mypet:=Pets[SecondPetIndex]):
                    set SecondPet = Mypet.PetProp
                if (Mypet:=Pets[ThirdPetIndex]):
                    set ThirdPet = Mypet.PetProp
                if (Mypet:=Pets[ForthPetIndex]):
                    set ForthPet = Mypet.PetProp

                var PetWon:creative_prop = creative_prop{}
                var RandResult:int = GetRandomInt(0,1000)
                Print("Rand: {RandResult}")
                if:
                    PlayerObj := player[Agent]
                    Fort := Agent.GetFortCharacter[]
                then:
                    if(CustomPlayer:Custom_Player = PlayersMap[PlayerObj]):
                        #--------------Luck
                        set RandResult+=CustomPlayer.Pet_Luck_Value*5+CustomPlayer.Bonus_Luck*5
                        #--------------
                        if (RandResult>995):
                            Print("Legendary")
                            if(Cine:=Cine_L27_Legendary[GetTeamIndexAgent[Agent]-1]):
                                Cine.Play(Agent)
                            Sleep(4.0)
                            if(Cine:=Cine_L27_Legendary[GetTeamIndexAgent[Agent]-1]):
                                Cine.Stop(Agent)
                                Accolade_Egg27_28.Award(Agent)

                            if (CustomPlayer.Ownership_L27_Pet4=0):
                                #Release potential previous pets
                                if (CustomPlayer.Ownership_L27_Pet1=1):
                                    #--------------Button pressed:
                                    CustomPlayer.IncreaseButtonL27()
                                    #Print("Button L27 Pressed: {CustomPlayer.ButtonL27Pressed}")
                                    #--------------Spawn
                                    Sleep(0.5)
                                    ##spawn:
                                    ##    ReleasePet(FirstPet)
                                    set CustomPlayer.Ownership_L27_Pet1=0
                                    L27_Pet1_DesOwn(Agent)
                                if (CustomPlayer.Ownership_L27_Pet2=1):
                                    #--------------Button pressed:
                                    CustomPlayer.IncreaseButtonL27()
                                    #Print("Button L27 Pressed: {CustomPlayer.ButtonL27Pressed}")
                                    #--------------Spawn
                                    Sleep(0.5)
                                    ##spawn:
                                    ##    ReleasePet(SecondPet)
                                    set CustomPlayer.Ownership_L27_Pet2=0
                                    L27_Pet2_DesOwn(Agent)
                                if (CustomPlayer.Ownership_L27_Pet3=1):
                                    #--------------Button pressed:
                                    CustomPlayer.IncreaseButtonL27()
                                    #Print("Button L27 Pressed: {CustomPlayer.ButtonL27Pressed}")
                                    #--------------Spawn
                                    Sleep(0.5)
                                    ##spawn:
                                    ##    ReleasePet(ThirdPet)
                                    set CustomPlayer.Ownership_L27_Pet3=0
                                    L27_Pet3_DesOwn(Agent)
                                #Get New pet
                                L27_Pet4_Own(Agent)
                                set CustomPlayer.Ownership_L27_Pet4=1
                                set CustomPlayer.Pet_Unlocked_L27=4
                                CustomPlayer.Update_Pet_PowerL27(Power_L27P4)
                                var PetPower:float=1.0*Power_L27P4
                                set PetWon = ForthPet
                                spawn:
                                    FollowPlayerL27(Fort, PetWon, PetPower, ForthPetIndex)
                        else if (RandResult>855):
                            Print("Epic")
                            if(Cine:=Cine_L27_Epic[GetTeamIndexAgent[Agent]-1]):
                                Cine.Play(Agent)
                            Sleep(4.0)
                            if(Cine:=Cine_L27_Epic[GetTeamIndexAgent[Agent]-1]):
                                Cine.Stop(Agent)
                                Accolade_Egg27_28.Award(Agent)

                            if (CustomPlayer.Ownership_L27_Pet3=0 and CustomPlayer.Ownership_L27_Pet4=0):
                                #Release potential previous pets
                                if (CustomPlayer.Ownership_L27_Pet1=1):
                                    #--------------Button pressed:
                                    CustomPlayer.IncreaseButtonL27()
                                    #Print("Button L27 Pressed: {CustomPlayer.ButtonL27Pressed}")
                                    #--------------Spawn
                                    Sleep(0.5)
                                    ##spawn:
                                    ##    ReleasePet(FirstPet)
                                    set CustomPlayer.Ownership_L27_Pet1=0
                                    L27_Pet1_DesOwn(Agent)
                                if (CustomPlayer.Ownership_L27_Pet2=1):
                                    #--------------Button pressed:
                                    CustomPlayer.IncreaseButtonL27()
                                    #Print("Button L27 Pressed: {CustomPlayer.ButtonL27Pressed}")
                                    #--------------Spawn
                                    Sleep(0.5)
                                    ##spawn:
                                    ##    ReleasePet(SecondPet)
                                    set CustomPlayer.Ownership_L27_Pet2=0
                                    L27_Pet2_DesOwn(Agent)
                                #Get New pet
                                L27_Pet3_Own(Agent)
                                set CustomPlayer.Ownership_L27_Pet3=1
                                set CustomPlayer.Pet_Unlocked_L27=3
                                CustomPlayer.Update_Pet_PowerL27(Power_L27P3)
                                var PetPower:float=1.0*Power_L27P3
                                set PetWon = ThirdPet
                                spawn:
                                    FollowPlayerL27(Fort, PetWon, PetPower, ThirdPetIndex)
                        else if (RandResult>750):
                            Print("Rare")
                            if(Cine:=Cine_L27_Rare[GetTeamIndexAgent[Agent]-1]):
                                Cine.Play(Agent)
                            Sleep(4.0)
                            if(Cine:=Cine_L27_Rare[GetTeamIndexAgent[Agent]-1]):
                                Cine.Stop(Agent)
                                Accolade_Egg27_28.Award(Agent)

                            if (CustomPlayer.Ownership_L27_Pet2=0 and CustomPlayer.Ownership_L27_Pet3=0 and CustomPlayer.Ownership_L27_Pet4=0):
                                #Release potential previous pets
                                if (CustomPlayer.Ownership_L27_Pet1=1):
                                    #--------------Button pressed:
                                    CustomPlayer.IncreaseButtonL27()
                                    #Print("Button L27 Pressed: {CustomPlayer.ButtonL27Pressed}")
                                    #--------------Spawn
                                    Sleep(0.5)
                                    ##spawn:
                                    ##    ReleasePet(FirstPet)
                                    set CustomPlayer.Ownership_L27_Pet1=0
                                    L27_Pet1_DesOwn(Agent)
                                #Get New pet
                                L27_Pet2_Own(Agent)
                                set CustomPlayer.Ownership_L27_Pet2=1
                                set CustomPlayer.Pet_Unlocked_L27=2
                                CustomPlayer.Update_Pet_PowerL27(Power_L27P2)
                                var PetPower:float=1.0*Power_L27P2
                                set PetWon = SecondPet
                                spawn:
                                    FollowPlayerL27(Fort, PetWon, PetPower, SecondPetIndex)
                        else:
                            Print("Normal")
                            if(Cine:=Cine_L27_Normal[GetTeamIndexAgent[Agent]-1]):
                                Cine.Play(Agent)
                            Sleep(4.0)
                            if(Cine:=Cine_L27_Normal[GetTeamIndexAgent[Agent]-1]):
                                Cine.Stop(Agent)
                                Accolade_Egg27_28.Award(Agent)

                            if (CustomPlayer.Ownership_L27_Pet1=0 and CustomPlayer.Ownership_L27_Pet2=0 and CustomPlayer.Ownership_L27_Pet3=0 and CustomPlayer.Ownership_L27_Pet4=0):
                                #Get New pet
                                L27_Pet1_Own(Agent)
                                set CustomPlayer.Ownership_L27_Pet1=1
                                set CustomPlayer.Pet_Unlocked_L27=1
                                CustomPlayer.Update_Pet_PowerL27(Power_L27P1)
                                var PetPower:float=1.0*Power_L27P1
                                set PetWon = FirstPet
                                spawn:
                                    FollowPlayerL27(Fort, PetWon, PetPower, FirstPetIndex)

    PurchasePetL28( Agent : agent )<suspends>: void =
        #Requires Unlock L28
        if:
            PlayerObj := player[Agent]
            CurrentPlayerStats := PlayerStatsManager.GetPlayerStats[Agent]
            CurrentPlayerStats.LevelUnlocked.CurrentValue<27
        then:
            RequireHigherLevelHUD.Show(Agent) 
        else:
            if:
                PlayerObj := player[Agent]
                CustomPlayer:Custom_Player = PlayersMap[PlayerObj]
                CustomPlayer.PlayerGold<CostL28
            then:
                NotEnoughGoldHUD.Show(Agent)
            else:
                DecreaseGoldFunc(Agent,CostL28)
                var FirstPet:creative_prop = creative_prop{}
                var SecondPet:creative_prop = creative_prop{}
                var ThirdPet:creative_prop = creative_prop{}
                var ForthPet:creative_prop = creative_prop{}

                var FirstPetIndex:int = 0
                var SecondPetIndex:int = 0
                var ThirdPetIndex:int = 0
                var ForthPetIndex:int = 0

                var IndexBase:int = 12*27

                #Select right pets for each player
                if(GetTeamIndexAgent[Agent]=1):
                    set FirstPetIndex =  IndexBase
                    set SecondPetIndex = IndexBase+1
                    set ThirdPetIndex =  IndexBase+2
                    set ForthPetIndex =  IndexBase+3
                if(GetTeamIndexAgent[Agent]=2):
                    set FirstPetIndex =  IndexBase+4
                    set SecondPetIndex = IndexBase+5
                    set ThirdPetIndex =  IndexBase+6
                    set ForthPetIndex =  IndexBase+7
                if(GetTeamIndexAgent[Agent]=3):
                    set FirstPetIndex =  IndexBase+8
                    set SecondPetIndex = IndexBase+9
                    set ThirdPetIndex =  IndexBase+10
                    set ForthPetIndex =  IndexBase+11

                if (Mypet:=Pets[FirstPetIndex]):
                    set FirstPet = Mypet.PetProp
                if (Mypet:=Pets[SecondPetIndex]):
                    set SecondPet = Mypet.PetProp
                if (Mypet:=Pets[ThirdPetIndex]):
                    set ThirdPet = Mypet.PetProp
                if (Mypet:=Pets[ForthPetIndex]):
                    set ForthPet = Mypet.PetProp

                var PetWon:creative_prop = creative_prop{}
                var RandResult:int = GetRandomInt(0,1000)
                Print("Rand: {RandResult}")
                if:
                    PlayerObj := player[Agent]
                    Fort := Agent.GetFortCharacter[]
                then:
                    if(CustomPlayer:Custom_Player = PlayersMap[PlayerObj]):
                        #--------------Luck
                        set RandResult+=CustomPlayer.Pet_Luck_Value*5+CustomPlayer.Bonus_Luck*5
                        #--------------
                        if (RandResult>995):
                            Print("Legendary")
                            if(Cine:=Cine_L28_Legendary[GetTeamIndexAgent[Agent]-1]):
                                Cine.Play(Agent)
                            Sleep(4.0)
                            if(Cine:=Cine_L28_Legendary[GetTeamIndexAgent[Agent]-1]):
                                Cine.Stop(Agent)
                                Accolade_Egg27_28.Award(Agent)

                            if (CustomPlayer.Ownership_L28_Pet4=0):
                                #Release potential previous pets
                                if (CustomPlayer.Ownership_L28_Pet1=1):
                                    #--------------Button pressed:
                                    CustomPlayer.IncreaseButtonL28()
                                    #Print("Button L28 Pressed: {CustomPlayer.ButtonL28Pressed}")
                                    #--------------Spawn
                                    Sleep(0.5)
                                    ##spawn:
                                    ##    ReleasePet(FirstPet)
                                    set CustomPlayer.Ownership_L28_Pet1=0
                                    L28_Pet1_DesOwn(Agent)
                                if (CustomPlayer.Ownership_L28_Pet2=1):
                                    #--------------Button pressed:
                                    CustomPlayer.IncreaseButtonL28()
                                    #Print("Button L28 Pressed: {CustomPlayer.ButtonL28Pressed}")
                                    #--------------Spawn
                                    Sleep(0.5)
                                    ##spawn:
                                    ##    ReleasePet(SecondPet)
                                    set CustomPlayer.Ownership_L28_Pet2=0
                                    L28_Pet2_DesOwn(Agent)
                                if (CustomPlayer.Ownership_L28_Pet3=1):
                                    #--------------Button pressed:
                                    CustomPlayer.IncreaseButtonL28()
                                    #Print("Button L28 Pressed: {CustomPlayer.ButtonL28Pressed}")
                                    #--------------Spawn
                                    Sleep(0.5)
                                    ##spawn:
                                    ##    ReleasePet(ThirdPet)
                                    set CustomPlayer.Ownership_L28_Pet3=0
                                    L28_Pet3_DesOwn(Agent)
                                #Get New pet
                                L28_Pet4_Own(Agent)
                                set CustomPlayer.Ownership_L28_Pet4=1
                                set CustomPlayer.Pet_Unlocked_L28=4
                                CustomPlayer.Update_Pet_PowerL28(Power_L28P4)
                                var PetPower:float=1.0*Power_L28P4
                                set PetWon = ForthPet
                                spawn:
                                    FollowPlayerL28(Fort, PetWon, PetPower, ForthPetIndex)
                        else if (RandResult>855):
                            Print("Epic")
                            if(Cine:=Cine_L28_Epic[GetTeamIndexAgent[Agent]-1]):
                                Cine.Play(Agent)
                            Sleep(4.0)
                            if(Cine:=Cine_L28_Epic[GetTeamIndexAgent[Agent]-1]):
                                Cine.Stop(Agent)
                                Accolade_Egg27_28.Award(Agent)

                            if (CustomPlayer.Ownership_L28_Pet3=0 and CustomPlayer.Ownership_L28_Pet4=0):
                                #Release potential previous pets
                                if (CustomPlayer.Ownership_L28_Pet1=1):
                                    #--------------Button pressed:
                                    CustomPlayer.IncreaseButtonL28()
                                    #Print("Button L28 Pressed: {CustomPlayer.ButtonL28Pressed}")
                                    #--------------Spawn
                                    Sleep(0.5)
                                    ##spawn:
                                    ##    ReleasePet(FirstPet)
                                    set CustomPlayer.Ownership_L28_Pet1=0
                                    L28_Pet1_DesOwn(Agent)
                                if (CustomPlayer.Ownership_L28_Pet2=1):
                                    #--------------Button pressed:
                                    CustomPlayer.IncreaseButtonL28()
                                    #Print("Button L28 Pressed: {CustomPlayer.ButtonL28Pressed}")
                                    #--------------Spawn
                                    Sleep(0.5)
                                    ##spawn:
                                    ##    ReleasePet(SecondPet)
                                    set CustomPlayer.Ownership_L28_Pet2=0
                                    L28_Pet2_DesOwn(Agent)
                                #Get New pet
                                L28_Pet3_Own(Agent)
                                set CustomPlayer.Ownership_L28_Pet3=1
                                set CustomPlayer.Pet_Unlocked_L28=3
                                CustomPlayer.Update_Pet_PowerL28(Power_L28P3)
                                var PetPower:float=1.0*Power_L28P3
                                set PetWon = ThirdPet
                                spawn:
                                    FollowPlayerL28(Fort, PetWon, PetPower, ThirdPetIndex)
                        else if (RandResult>750):
                            Print("Rare")
                            if(Cine:=Cine_L28_Rare[GetTeamIndexAgent[Agent]-1]):
                                Cine.Play(Agent)
                            Sleep(4.0)
                            if(Cine:=Cine_L28_Rare[GetTeamIndexAgent[Agent]-1]):
                                Cine.Stop(Agent)
                                Accolade_Egg27_28.Award(Agent)

                            if (CustomPlayer.Ownership_L28_Pet2=0 and CustomPlayer.Ownership_L28_Pet3=0 and CustomPlayer.Ownership_L28_Pet4=0):
                                #Release potential previous pets
                                if (CustomPlayer.Ownership_L28_Pet1=1):
                                    #--------------Button pressed:
                                    CustomPlayer.IncreaseButtonL28()
                                    #Print("Button L28 Pressed: {CustomPlayer.ButtonL28Pressed}")
                                    #--------------Spawn
                                    Sleep(0.5)
                                    ##spawn:
                                    ##    ReleasePet(FirstPet)
                                    set CustomPlayer.Ownership_L28_Pet1=0
                                    L28_Pet1_DesOwn(Agent)
                                #Get New pet
                                L28_Pet2_Own(Agent)
                                set CustomPlayer.Ownership_L28_Pet2=1
                                set CustomPlayer.Pet_Unlocked_L28=2
                                CustomPlayer.Update_Pet_PowerL28(Power_L28P2)
                                var PetPower:float=1.0*Power_L28P2
                                set PetWon = SecondPet
                                spawn:
                                    FollowPlayerL28(Fort, PetWon, PetPower, SecondPetIndex)
                        else:
                            Print("Normal")
                            if(Cine:=Cine_L28_Normal[GetTeamIndexAgent[Agent]-1]):
                                Cine.Play(Agent)
                            Sleep(4.0)
                            if(Cine:=Cine_L28_Normal[GetTeamIndexAgent[Agent]-1]):
                                Cine.Stop(Agent)
                                Accolade_Egg27_28.Award(Agent)

                            if (CustomPlayer.Ownership_L28_Pet1=0 and CustomPlayer.Ownership_L28_Pet2=0 and CustomPlayer.Ownership_L28_Pet3=0 and CustomPlayer.Ownership_L28_Pet4=0):
                                #Get New pet
                                L28_Pet1_Own(Agent)
                                set CustomPlayer.Ownership_L28_Pet1=1
                                set CustomPlayer.Pet_Unlocked_L28=1
                                CustomPlayer.Update_Pet_PowerL28(Power_L28P1)
                                var PetPower:float=1.0*Power_L28P1
                                set PetWon = FirstPet
                                spawn:
                                    FollowPlayerL28(Fort, PetWon, PetPower, FirstPetIndex)

    PurchasePetL29( Agent : agent )<suspends>: void =
        #Requires Unlock L29
        if:
            PlayerObj := player[Agent]
            CurrentPlayerStats := PlayerStatsManager.GetPlayerStats[Agent]
            CurrentPlayerStats.LevelUnlocked.CurrentValue<28
        then:
            RequireHigherLevelHUD.Show(Agent) 
        else:
            if:
                PlayerObj := player[Agent]
                CustomPlayer:Custom_Player = PlayersMap[PlayerObj]
                CustomPlayer.PlayerGold<CostL29
            then:
                NotEnoughGoldHUD.Show(Agent)
            else:
                DecreaseGoldFunc(Agent,CostL29)
                var FirstPet:creative_prop = creative_prop{}
                var SecondPet:creative_prop = creative_prop{}
                var ThirdPet:creative_prop = creative_prop{}
                var ForthPet:creative_prop = creative_prop{}

                var FirstPetIndex:int = 0
                var SecondPetIndex:int = 0
                var ThirdPetIndex:int = 0
                var ForthPetIndex:int = 0

                var IndexBase:int = 12*28

                #Select right pets for each player
                if(GetTeamIndexAgent[Agent]=1):
                    set FirstPetIndex =  IndexBase
                    set SecondPetIndex = IndexBase+1
                    set ThirdPetIndex =  IndexBase+2
                    set ForthPetIndex =  IndexBase+3
                if(GetTeamIndexAgent[Agent]=2):
                    set FirstPetIndex =  IndexBase+4
                    set SecondPetIndex = IndexBase+5
                    set ThirdPetIndex =  IndexBase+6
                    set ForthPetIndex =  IndexBase+7
                if(GetTeamIndexAgent[Agent]=3):
                    set FirstPetIndex =  IndexBase+8
                    set SecondPetIndex = IndexBase+9
                    set ThirdPetIndex =  IndexBase+10
                    set ForthPetIndex =  IndexBase+11

                if (Mypet:=Pets[FirstPetIndex]):
                    set FirstPet = Mypet.PetProp
                if (Mypet:=Pets[SecondPetIndex]):
                    set SecondPet = Mypet.PetProp
                if (Mypet:=Pets[ThirdPetIndex]):
                    set ThirdPet = Mypet.PetProp
                if (Mypet:=Pets[ForthPetIndex]):
                    set ForthPet = Mypet.PetProp

                var PetWon:creative_prop = creative_prop{}
                var RandResult:int = GetRandomInt(0,1000)
                Print("Rand: {RandResult}")
                if:
                    PlayerObj := player[Agent]
                    Fort := Agent.GetFortCharacter[]
                then:
                    if(CustomPlayer:Custom_Player = PlayersMap[PlayerObj]):
                        #--------------Luck
                        set RandResult+=CustomPlayer.Pet_Luck_Value*5+CustomPlayer.Bonus_Luck*5
                        #--------------
                        if (RandResult>995):
                            Print("Legendary")
                            if(Cine:=Cine_L29_Legendary[GetTeamIndexAgent[Agent]-1]):
                                Cine.Play(Agent)
                            Sleep(4.0)
                            if(Cine:=Cine_L29_Legendary[GetTeamIndexAgent[Agent]-1]):
                                Cine.Stop(Agent)
                                Accolade_Egg29_30.Award(Agent)

                            if (CustomPlayer.Ownership_L29_Pet4=0):
                                #Release potential previous pets
                                if (CustomPlayer.Ownership_L29_Pet1=1):
                                    #--------------Button pressed:
                                    CustomPlayer.IncreaseButtonL29()
                                    #Print("Button L29 Pressed: {CustomPlayer.ButtonL29Pressed}")
                                    #--------------Spawn
                                    Sleep(0.5)
                                    ##spawn:
                                    ##    ReleasePet(FirstPet)
                                    set CustomPlayer.Ownership_L29_Pet1=0
                                    L29_Pet1_DesOwn(Agent)
                                if (CustomPlayer.Ownership_L29_Pet2=1):
                                    #--------------Button pressed:
                                    CustomPlayer.IncreaseButtonL29()
                                    #Print("Button L29 Pressed: {CustomPlayer.ButtonL29Pressed}")
                                    #--------------Spawn
                                    Sleep(0.5)
                                    ##spawn:
                                    ##    ReleasePet(SecondPet)
                                    set CustomPlayer.Ownership_L29_Pet2=0
                                    L29_Pet2_DesOwn(Agent)
                                if (CustomPlayer.Ownership_L29_Pet3=1):
                                    #--------------Button pressed:
                                    CustomPlayer.IncreaseButtonL29()
                                    #Print("Button L29 Pressed: {CustomPlayer.ButtonL29Pressed}")
                                    #--------------Spawn
                                    Sleep(0.5)
                                    ##spawn:
                                    ##    ReleasePet(ThirdPet)
                                    set CustomPlayer.Ownership_L29_Pet3=0
                                    L29_Pet3_DesOwn(Agent)
                                #Get New pet
                                L29_Pet4_Own(Agent)
                                set CustomPlayer.Ownership_L29_Pet4=1
                                set CustomPlayer.Pet_Unlocked_L29=4
                                CustomPlayer.Update_Pet_PowerL29(Power_L29P4)
                                var PetPower:float=1.0*Power_L29P4
                                set PetWon = ForthPet
                                spawn:
                                    FollowPlayerL29(Fort, PetWon, PetPower, ForthPetIndex)
                        else if (RandResult>855):
                            Print("Epic")
                            if(Cine:=Cine_L29_Epic[GetTeamIndexAgent[Agent]-1]):
                                Cine.Play(Agent)
                            Sleep(4.0)
                            if(Cine:=Cine_L29_Epic[GetTeamIndexAgent[Agent]-1]):
                                Cine.Stop(Agent)
                                Accolade_Egg29_30.Award(Agent)

                            if (CustomPlayer.Ownership_L29_Pet3=0 and CustomPlayer.Ownership_L29_Pet4=0):
                                #Release potential previous pets
                                if (CustomPlayer.Ownership_L29_Pet1=1):
                                    #--------------Button pressed:
                                    CustomPlayer.IncreaseButtonL29()
                                    #Print("Button L29 Pressed: {CustomPlayer.ButtonL29Pressed}")
                                    #--------------Spawn
                                    Sleep(0.5)
                                    ##spawn:
                                    ##    ReleasePet(FirstPet)
                                    set CustomPlayer.Ownership_L29_Pet1=0
                                    L29_Pet1_DesOwn(Agent)
                                if (CustomPlayer.Ownership_L29_Pet2=1):
                                    #--------------Button pressed:
                                    CustomPlayer.IncreaseButtonL29()
                                    #Print("Button L29 Pressed: {CustomPlayer.ButtonL29Pressed}")
                                    #--------------Spawn
                                    Sleep(0.5)
                                    ##spawn:
                                    ##    ReleasePet(SecondPet)
                                    set CustomPlayer.Ownership_L29_Pet2=0
                                    L29_Pet2_DesOwn(Agent)
                                #Get New pet
                                L29_Pet3_Own(Agent)
                                set CustomPlayer.Ownership_L29_Pet3=1
                                set CustomPlayer.Pet_Unlocked_L29=3
                                CustomPlayer.Update_Pet_PowerL29(Power_L29P3)
                                var PetPower:float=1.0*Power_L29P3
                                set PetWon = ThirdPet
                                spawn:
                                    FollowPlayerL29(Fort, PetWon, PetPower, ThirdPetIndex)
                        else if (RandResult>750):
                            Print("Rare")
                            if(Cine:=Cine_L29_Rare[GetTeamIndexAgent[Agent]-1]):
                                Cine.Play(Agent)
                            Sleep(4.0)
                            if(Cine:=Cine_L29_Rare[GetTeamIndexAgent[Agent]-1]):
                                Cine.Stop(Agent)
                                Accolade_Egg29_30.Award(Agent)

                            if (CustomPlayer.Ownership_L29_Pet2=0 and CustomPlayer.Ownership_L29_Pet3=0 and CustomPlayer.Ownership_L29_Pet4=0):
                                #Release potential previous pets
                                if (CustomPlayer.Ownership_L29_Pet1=1):
                                    #--------------Button pressed:
                                    CustomPlayer.IncreaseButtonL29()
                                    #Print("Button L29 Pressed: {CustomPlayer.ButtonL29Pressed}")
                                    #--------------Spawn
                                    Sleep(0.5)
                                    ##spawn:
                                    ##    ReleasePet(FirstPet)
                                    set CustomPlayer.Ownership_L29_Pet1=0
                                    L29_Pet1_DesOwn(Agent)
                                #Get New pet
                                L29_Pet2_Own(Agent)
                                set CustomPlayer.Ownership_L29_Pet2=1
                                set CustomPlayer.Pet_Unlocked_L29=2
                                CustomPlayer.Update_Pet_PowerL29(Power_L29P2)
                                var PetPower:float=1.0*Power_L29P2
                                set PetWon = SecondPet
                                spawn:
                                    FollowPlayerL29(Fort, PetWon, PetPower, SecondPetIndex)
                        else:
                            Print("Normal")
                            if(Cine:=Cine_L29_Normal[GetTeamIndexAgent[Agent]-1]):
                                Cine.Play(Agent)
                            Sleep(4.0)
                            if(Cine:=Cine_L29_Normal[GetTeamIndexAgent[Agent]-1]):
                                Cine.Stop(Agent)
                                Accolade_Egg29_30.Award(Agent)

                            if (CustomPlayer.Ownership_L29_Pet1=0 and CustomPlayer.Ownership_L29_Pet2=0 and CustomPlayer.Ownership_L29_Pet3=0 and CustomPlayer.Ownership_L29_Pet4=0):
                                #Get New pet
                                L29_Pet1_Own(Agent)
                                set CustomPlayer.Ownership_L29_Pet1=1
                                set CustomPlayer.Pet_Unlocked_L29=1
                                CustomPlayer.Update_Pet_PowerL29(Power_L29P1)
                                var PetPower:float=1.0*Power_L29P1
                                set PetWon = FirstPet
                                spawn:
                                    FollowPlayerL29(Fort, PetWon, PetPower, FirstPetIndex)

    PurchasePetL30( Agent : agent )<suspends>: void =
        #Requires Unlock L30
        if:
            PlayerObj := player[Agent]
            CurrentPlayerStats := PlayerStatsManager.GetPlayerStats[Agent]
            CurrentPlayerStats.LevelUnlocked.CurrentValue<29
        then:
            RequireHigherLevelHUD.Show(Agent) 
        else:
            if:
                PlayerObj := player[Agent]
                CustomPlayer:Custom_Player = PlayersMap[PlayerObj]
                CustomPlayer.PlayerGold<CostL30
            then:
                NotEnoughGoldHUD.Show(Agent)
            else:
                DecreaseGoldFunc(Agent,CostL30)
                var FirstPet:creative_prop = creative_prop{}
                var SecondPet:creative_prop = creative_prop{}
                var ThirdPet:creative_prop = creative_prop{}
                var ForthPet:creative_prop = creative_prop{}

                var FirstPetIndex:int = 0
                var SecondPetIndex:int = 0
                var ThirdPetIndex:int = 0
                var ForthPetIndex:int = 0

                var IndexBase:int = 12*29

                #Select right pets for each player
                if(GetTeamIndexAgent[Agent]=1):
                    set FirstPetIndex =  IndexBase
                    set SecondPetIndex = IndexBase+1
                    set ThirdPetIndex =  IndexBase+2
                    set ForthPetIndex =  IndexBase+3
                if(GetTeamIndexAgent[Agent]=2):
                    set FirstPetIndex =  IndexBase+4
                    set SecondPetIndex = IndexBase+5
                    set ThirdPetIndex =  IndexBase+6
                    set ForthPetIndex =  IndexBase+7
                if(GetTeamIndexAgent[Agent]=3):
                    set FirstPetIndex =  IndexBase+8
                    set SecondPetIndex = IndexBase+9
                    set ThirdPetIndex =  IndexBase+10
                    set ForthPetIndex =  IndexBase+11

                if (Mypet:=Pets[FirstPetIndex]):
                    set FirstPet = Mypet.PetProp
                if (Mypet:=Pets[SecondPetIndex]):
                    set SecondPet = Mypet.PetProp
                if (Mypet:=Pets[ThirdPetIndex]):
                    set ThirdPet = Mypet.PetProp
                if (Mypet:=Pets[ForthPetIndex]):
                    set ForthPet = Mypet.PetProp

                var PetWon:creative_prop = creative_prop{}
                var RandResult:int = GetRandomInt(0,1000)
                Print("Rand: {RandResult}")
                if:
                    PlayerObj := player[Agent]
                    Fort := Agent.GetFortCharacter[]
                then:
                    if(CustomPlayer:Custom_Player = PlayersMap[PlayerObj]):
                        #--------------Luck
                        set RandResult+=CustomPlayer.Pet_Luck_Value*5+CustomPlayer.Bonus_Luck*5
                        #--------------
                        if (RandResult>995):
                            Print("Legendary")
                            if(Cine:=Cine_L30_Legendary[GetTeamIndexAgent[Agent]-1]):
                                Cine.Play(Agent)
                            Sleep(4.0)
                            if(Cine:=Cine_L30_Legendary[GetTeamIndexAgent[Agent]-1]):
                                Cine.Stop(Agent)
                                Accolade_Egg29_30.Award(Agent)

                            if (CustomPlayer.Ownership_L30_Pet4=0):
                                #Release potential previous pets
                                if (CustomPlayer.Ownership_L30_Pet1=1):
                                    #--------------Button pressed:
                                    CustomPlayer.IncreaseButtonL30()
                                    #Print("Button L30 Pressed: {CustomPlayer.ButtonL30Pressed}")
                                    #--------------Spawn
                                    Sleep(0.5)
                                    ##spawn:
                                    ##    ReleasePet(FirstPet)
                                    set CustomPlayer.Ownership_L30_Pet1=0
                                    L30_Pet1_DesOwn(Agent)
                                if (CustomPlayer.Ownership_L30_Pet2=1):
                                    #--------------Button pressed:
                                    CustomPlayer.IncreaseButtonL30()
                                    #Print("Button L30 Pressed: {CustomPlayer.ButtonL30Pressed}")
                                    #--------------Spawn
                                    Sleep(0.5)
                                    ##spawn:
                                    ##    ReleasePet(SecondPet)
                                    set CustomPlayer.Ownership_L30_Pet2=0
                                    L30_Pet2_DesOwn(Agent)
                                if (CustomPlayer.Ownership_L30_Pet3=1):
                                    #--------------Button pressed:
                                    CustomPlayer.IncreaseButtonL30()
                                    #Print("Button L30 Pressed: {CustomPlayer.ButtonL30Pressed}")
                                    #--------------Spawn
                                    Sleep(0.5)
                                    ##spawn:
                                    ##    ReleasePet(ThirdPet)
                                    set CustomPlayer.Ownership_L30_Pet3=0
                                    L30_Pet3_DesOwn(Agent)
                                #Get New pet
                                L30_Pet4_Own(Agent)
                                set CustomPlayer.Ownership_L30_Pet4=1
                                set CustomPlayer.Pet_Unlocked_L30=4
                                CustomPlayer.Update_Pet_PowerL30(Power_L30P4)
                                var PetPower:float=1.0*Power_L30P4
                                set PetWon = ForthPet
                                spawn:
                                    FollowPlayerL30(Fort, PetWon, PetPower, ForthPetIndex)
                        else if (RandResult>855):
                            Print("Epic")
                            if(Cine:=Cine_L30_Epic[GetTeamIndexAgent[Agent]-1]):
                                Cine.Play(Agent)
                            Sleep(4.0)
                            if(Cine:=Cine_L30_Epic[GetTeamIndexAgent[Agent]-1]):
                                Cine.Stop(Agent)
                                Accolade_Egg29_30.Award(Agent)

                            if (CustomPlayer.Ownership_L30_Pet3=0 and CustomPlayer.Ownership_L30_Pet4=0):
                                #Release potential previous pets
                                if (CustomPlayer.Ownership_L30_Pet1=1):
                                    #--------------Button pressed:
                                    CustomPlayer.IncreaseButtonL30()
                                    #Print("Button L30 Pressed: {CustomPlayer.ButtonL30Pressed}")
                                    #--------------Spawn
                                    Sleep(0.5)
                                    ##spawn:
                                    ##    ReleasePet(FirstPet)
                                    set CustomPlayer.Ownership_L30_Pet1=0
                                    L30_Pet1_DesOwn(Agent)
                                if (CustomPlayer.Ownership_L30_Pet2=1):
                                    #--------------Button pressed:
                                    CustomPlayer.IncreaseButtonL30()
                                    #Print("Button L30 Pressed: {CustomPlayer.ButtonL30Pressed}")
                                    #--------------Spawn
                                    Sleep(0.5)
                                    ##spawn:
                                    ##    ReleasePet(SecondPet)
                                    set CustomPlayer.Ownership_L30_Pet2=0
                                    L30_Pet2_DesOwn(Agent)
                                #Get New pet
                                L30_Pet3_Own(Agent)
                                set CustomPlayer.Ownership_L30_Pet3=1
                                set CustomPlayer.Pet_Unlocked_L30=3
                                CustomPlayer.Update_Pet_PowerL30(Power_L30P3)
                                var PetPower:float=1.0*Power_L30P3
                                set PetWon = ThirdPet
                                spawn:
                                    FollowPlayerL30(Fort, PetWon, PetPower, ThirdPetIndex)
                        else if (RandResult>750):
                            Print("Rare")
                            if(Cine:=Cine_L30_Rare[GetTeamIndexAgent[Agent]-1]):
                                Cine.Play(Agent)
                            Sleep(4.0)
                            if(Cine:=Cine_L30_Rare[GetTeamIndexAgent[Agent]-1]):
                                Cine.Stop(Agent)
                                Accolade_Egg29_30.Award(Agent)

                            if (CustomPlayer.Ownership_L30_Pet2=0 and CustomPlayer.Ownership_L30_Pet3=0 and CustomPlayer.Ownership_L30_Pet4=0):
                                #Release potential previous pets
                                if (CustomPlayer.Ownership_L30_Pet1=1):
                                    #--------------Button pressed:
                                    CustomPlayer.IncreaseButtonL30()
                                    #Print("Button L30 Pressed: {CustomPlayer.ButtonL30Pressed}")
                                    #--------------Spawn
                                    Sleep(0.5)
                                    ##spawn:
                                    ##    ReleasePet(FirstPet)
                                    set CustomPlayer.Ownership_L30_Pet1=0
                                    L30_Pet1_DesOwn(Agent)
                                #Get New pet
                                L30_Pet2_Own(Agent)
                                set CustomPlayer.Ownership_L30_Pet2=1
                                set CustomPlayer.Pet_Unlocked_L30=2
                                CustomPlayer.Update_Pet_PowerL30(Power_L30P2)
                                var PetPower:float=1.0*Power_L30P2
                                set PetWon = SecondPet
                                spawn:
                                    FollowPlayerL30(Fort, PetWon, PetPower, SecondPetIndex)
                        else:
                            Print("Normal")
                            if(Cine:=Cine_L30_Normal[GetTeamIndexAgent[Agent]-1]):
                                Cine.Play(Agent)
                            Sleep(4.0)
                            if(Cine:=Cine_L30_Normal[GetTeamIndexAgent[Agent]-1]):
                                Cine.Stop(Agent)
                                Accolade_Egg29_30.Award(Agent)

                            if (CustomPlayer.Ownership_L30_Pet1=0 and CustomPlayer.Ownership_L30_Pet2=0 and CustomPlayer.Ownership_L30_Pet3=0 and CustomPlayer.Ownership_L30_Pet4=0):
                                #Get New pet
                                L30_Pet1_Own(Agent)
                                set CustomPlayer.Ownership_L30_Pet1=1
                                set CustomPlayer.Pet_Unlocked_L30=1
                                CustomPlayer.Update_Pet_PowerL30(Power_L30P1)
                                var PetPower:float=1.0*Power_L30P1
                                set PetWon = FirstPet
                                spawn:
                                    FollowPlayerL30(Fort, PetWon, PetPower, FirstPetIndex)

    PurchasePetL31( Agent : agent )<suspends>: void =
        #Requires Rebirth 1 + Not already looted
        if:
            PlayerObj := player[Agent]
            CustomPlayer:Custom_Player = PlayersMap[PlayerObj]
            CustomPlayer.Ownership_Secret_Pet1=1
        then:
            AlreadyUnlockedHUD.Show(Agent)
        else:
            if:
                PlayerObj := player[Agent]
                CustomPlayer:Custom_Player = PlayersMap[PlayerObj]
                Fort := Agent.GetFortCharacter[]
            then:
                set CustomPlayer.Ownership_Secret_Pet1=1
                L31_Pet1_Own(Agent)

                var FirstPet:creative_prop = creative_prop{}
                var FirstPetIndex:int = 0

                var IndexBase:int = 359

                #Select right pets for each player
                if(GetTeamIndexAgent[Agent]=1):
                    set FirstPetIndex =  IndexBase+1
                if(GetTeamIndexAgent[Agent]=2):
                    set FirstPetIndex =  IndexBase+2
                if(GetTeamIndexAgent[Agent]=3):
                    set FirstPetIndex =  IndexBase+3

                if (Mypet:=Pets[FirstPetIndex]):
                    set FirstPet = Mypet.PetProp

                var PetWon:creative_prop = creative_prop{}

                set PetWon = FirstPet

                if(Cine:=Cine_L31_Legendary[GetTeamIndexAgent[Agent]-1]):
                    Cine.Play(Agent)
                Sleep(4.0)
                if(Cine:=Cine_L31_Legendary[GetTeamIndexAgent[Agent]-1]):
                    Cine.Stop(Agent)
                    Accolade_Egg29_30.Award(Agent)
                    var PetPower:float=0.0
                    spawn:
                        FollowPlayerL31(Fort, PetWon, PetPower, FirstPetIndex)
                        
    PurchasePetL32( Agent : agent )<suspends>: void =
        #Requires Rebirth 1 + Not already looted
        if:
            PlayerObj := player[Agent]
            CustomPlayer:Custom_Player = PlayersMap[PlayerObj]
            CustomPlayer.Ownership_Secret_Pet2=1
        then:
            AlreadyUnlockedHUD.Show(Agent)
        else:
            if:
                PlayerObj := player[Agent]
                CustomPlayer:Custom_Player = PlayersMap[PlayerObj]
                Fort := Agent.GetFortCharacter[]
            then:
                set CustomPlayer.Ownership_Secret_Pet2=1
                L32_Pet1_Own(Agent)

                var FirstPet:creative_prop = creative_prop{}
                var FirstPetIndex:int = 0

                var IndexBase:int = 359+3

                #Select right pets for each player
                if(GetTeamIndexAgent[Agent]=1):
                    set FirstPetIndex =  IndexBase+1
                if(GetTeamIndexAgent[Agent]=2):
                    set FirstPetIndex =  IndexBase+2
                if(GetTeamIndexAgent[Agent]=3):
                    set FirstPetIndex =  IndexBase+3

                if (Mypet:=Pets[FirstPetIndex]):
                    set FirstPet = Mypet.PetProp

                var PetWon:creative_prop = creative_prop{}

                set PetWon = FirstPet

                if(Cine:=Cine_L32_Legendary[GetTeamIndexAgent[Agent]-1]):
                    Cine.Play(Agent)
                Sleep(4.0)
                if(Cine:=Cine_L32_Legendary[GetTeamIndexAgent[Agent]-1]):
                    Cine.Stop(Agent)
                    Accolade_Egg29_30.Award(Agent)
                    var PetPower:float=0.0
                    spawn:
                        FollowPlayerL32(Fort, PetWon, PetPower, FirstPetIndex)

    PurchasePetL33( Agent : agent )<suspends>: void =
        #Requires Rebirth 1 + Not already looted
        if:
            PlayerObj := player[Agent]
            CustomPlayer:Custom_Player = PlayersMap[PlayerObj]
            CustomPlayer.Ownership_Secret_Pet3=1
        then:
            AlreadyUnlockedHUD.Show(Agent)
        else:
            if:
                PlayerObj := player[Agent]
                CustomPlayer:Custom_Player = PlayersMap[PlayerObj]
                Fort := Agent.GetFortCharacter[]
            then:
                set CustomPlayer.Ownership_Secret_Pet3=1
                L33_Pet1_Own(Agent)

                var FirstPet:creative_prop = creative_prop{}
                var FirstPetIndex:int = 0

                var IndexBase:int = 359+6

                #Select right pets for each player
                if(GetTeamIndexAgent[Agent]=1):
                    set FirstPetIndex =  IndexBase+1
                if(GetTeamIndexAgent[Agent]=2):
                    set FirstPetIndex =  IndexBase+2
                if(GetTeamIndexAgent[Agent]=3):
                    set FirstPetIndex =  IndexBase+3

                if (Mypet:=Pets[FirstPetIndex]):
                    set FirstPet = Mypet.PetProp

                var PetWon:creative_prop = creative_prop{}

                set PetWon = FirstPet

                if(Cine:=Cine_L33_Legendary[GetTeamIndexAgent[Agent]-1]):
                    Cine.Play(Agent)
                Sleep(4.0)
                if(Cine:=Cine_L33_Legendary[GetTeamIndexAgent[Agent]-1]):
                    Cine.Stop(Agent)
                    Accolade_Egg29_30.Award(Agent)
                    var PetPower:float=0.0
                    spawn:
                        FollowPlayerL33(Fort, PetWon, PetPower, FirstPetIndex)

    PurchasePetL34( Agent : agent )<suspends>: void =
        #Requires Rebirth 1 + Not already looted
        if:
            PlayerObj := player[Agent]
            CustomPlayer:Custom_Player = PlayersMap[PlayerObj]
            CustomPlayer.Ownership_Secret_Pet4=1
        then:
            AlreadyUnlockedHUD.Show(Agent)
        else:
            if:
                PlayerObj := player[Agent]
                CustomPlayer:Custom_Player = PlayersMap[PlayerObj]
                Fort := Agent.GetFortCharacter[]
            then:
                set CustomPlayer.Ownership_Secret_Pet4=1
                L34_Pet1_Own(Agent)

                var FirstPet:creative_prop = creative_prop{}
                var FirstPetIndex:int = 0

                var IndexBase:int = 359+9

                #Select right pets for each player
                if(GetTeamIndexAgent[Agent]=1):
                    set FirstPetIndex =  IndexBase+1
                if(GetTeamIndexAgent[Agent]=2):
                    set FirstPetIndex =  IndexBase+2
                if(GetTeamIndexAgent[Agent]=3):
                    set FirstPetIndex =  IndexBase+3

                if (Mypet:=Pets[FirstPetIndex]):
                    set FirstPet = Mypet.PetProp

                var PetWon:creative_prop = creative_prop{}

                set PetWon = FirstPet

                if(Cine:=Cine_L34_Legendary[GetTeamIndexAgent[Agent]-1]):
                    Cine.Play(Agent)
                Sleep(4.0)
                if(Cine:=Cine_L34_Legendary[GetTeamIndexAgent[Agent]-1]):
                    Cine.Stop(Agent)
                    Accolade_Egg29_30.Award(Agent)
                    var PetPower:float=0.0
                    spawn:
                        FollowPlayerL34(Fort, PetWon, 500.0, FirstPetIndex)
    
    FollowPlayerL1(FortniteCharacter : fort_character, CreativeProp : creative_prop, Power:float, PetIndex:int)<suspends>: void =
        Print("Follow Player with pet index:{PetIndex}")
        #Future position of the pet
        Xplayer:float=-100.0
        Yplayer:float=-100.0
        if:
            Agent := FortniteCharacter.GetAgent[]
            PlayerObj := player[Agent]
            CustomPlayer:Custom_Player = PlayersMap[PlayerObj]
        then:
            var CheckValue:int = CustomPlayer.ButtonL1Pressed  
            set CustomPlayer.FollowActiveL1=1
            CustomPlayer.Update_Pets_Section()
            loop:
                #Print("CheckValue={CheckValue}")
                #Print("ButtonL1Pressed={CustomPlayer.ButtonL1Pressed}")
                if(CheckValue <> CustomPlayer.ButtonL1Pressed) {break}
                else if(CustomPlayer.PlayerInTreasureAreaFlag=0):
                    #AttackCinema.Stop()
                    PetRotateToPlayer(FortniteCharacter,CreativeProp)
                    Move_To_Player(FortniteCharacter,CreativeProp,Xplayer,Yplayer)
                else if(CustomPlayer.PlayerInTreasureAreaFlag=1):
                    #-----------------Find closest chest
                    var ClosestChest:creative_prop = creative_prop{}
                    var DistanceToMe:float = 100000000.0
                    var FoundChest:logic = false
                    var ChestIndexTemp:int = -1
                    var ChestIndexFinal:int = 0
                    var ClosestChestIndex:int=0
                    Allthechests:= ChestsL1+ChestsL2+ChestsL3+ChestsL4+ChestsL5+ChestsL6+ChestsL7+ChestsL8+ChestsL9+ChestsL10+ChestsL11+ChestsL12+ChestsL13+ChestsL14+ChestsL15+ChestsL16+ChestsL17+ChestsL18+ChestsL19+ChestsL20+ChestsL21+ChestsL22+ChestsL23+ChestsL24+ChestsL25+ChestsL26+ChestsL27+ChestsL28+ChestsL29+ChestsL30
                    for(Chestt:Allthechests):
                        set ChestIndexTemp+=1
                        MyPos:vector3 = FortniteCharacter.GetTransform().Translation
                        ChestPos:vector3 = Chestt.ChestProp.GetTransform().Translation
                        TotalDistance:float = Distance(MyPos, ChestPos) / 100.0
                        if(TotalDistance < DistanceToMe and Chestt.CurrentHP>0.0):
                            set ChestIndexFinal = ChestIndexTemp
                            #Print("Chest index{ChestIndexFinal}")
                            set FoundChest = true
                            set DistanceToMe = TotalDistance
                            set ClosestChest = Chestt.ChestProp
            
                    if(FoundChest = true):
                    #-----------------Rotate and move to chest + rerotate because position is random
                    PetRotateToChest(ClosestChest,CreativeProp)
                    Move_To_Chest(FortniteCharacter,ClosestChest,CreativeProp)
                    PetRotateToChest(ClosestChest,CreativeProp)
                    #-----------------Attack Chest
                    if(Cine:=Pets[PetIndex].AttackAnim):
                        Cine.Play(Agent)
                        Sleep(1.5) #Attack duration
                        Cine.Stop(Agent)
                    #----------------Crit
                    var PowerCrit:float=0.0
                    var CritRand:int = GetRandomInt(0,100)
                    if (CritRand>CustomPlayer.Pet_Crit_Value):
                        set PowerCrit=Power*CustomPlayer.BonusPowerMultiplicator*CustomPlayer.Bonus_Pet_Power
                    else:
                        set PowerCrit=Power*1.5*CustomPlayer.BonusPowerMultiplicator*CustomPlayer.Bonus_Pet_Power
                    #-------------------
                    Attack_Chest(ChestIndexFinal,PowerCrit)
                    Check_Chest_Gold(ChestIndexFinal,FortniteCharacter)

            ReleasePet(CreativeProp)
                    
    FollowPlayerL2(FortniteCharacter : fort_character, CreativeProp : creative_prop, Power:float, PetIndex:int)<suspends>: void =
        #Print("Follow Player")
        #Future position of the pet
        Xplayer:float=-100.0
        Yplayer:float= 100.0
        if:
            Agent := FortniteCharacter.GetAgent[]
            PlayerObj := player[Agent]
            CustomPlayer:Custom_Player = PlayersMap[PlayerObj]
        then:
            var CheckValue:int = CustomPlayer.ButtonL2Pressed  
            set CustomPlayer.FollowActiveL2=1
            CustomPlayer.Update_Pets_Section()
            loop:
                if(CheckValue <> CustomPlayer.ButtonL2Pressed) {break}
                else if(CustomPlayer.PlayerInTreasureAreaFlag=0):
                    #AttackCinema.Stop()
                    PetRotateToPlayer(FortniteCharacter,CreativeProp)
                    Move_To_Player(FortniteCharacter,CreativeProp,Xplayer,Yplayer)
                else if(CustomPlayer.PlayerInTreasureAreaFlag=1):
                    #-----------------Find closest chest
                    var ClosestChest:creative_prop = creative_prop{}
                    var DistanceToMe:float = 100000000.0
                    var FoundChest:logic = false
                    var ChestIndexTemp:int = -1
                    var ChestIndexFinal:int = 0
                    var ClosestChestIndex:int=0
                    Allthechests:= ChestsL1+ChestsL2+ChestsL3+ChestsL4+ChestsL5+ChestsL6+ChestsL7+ChestsL8+ChestsL9+ChestsL10+ChestsL11+ChestsL12+ChestsL13+ChestsL14+ChestsL15+ChestsL16+ChestsL17+ChestsL18+ChestsL19+ChestsL20+ChestsL21+ChestsL22+ChestsL23+ChestsL24+ChestsL25+ChestsL26+ChestsL27+ChestsL28+ChestsL29+ChestsL30
                    for(Chestt:Allthechests):
                        set ChestIndexTemp+=1
                        MyPos:vector3 = FortniteCharacter.GetTransform().Translation
                        ChestPos:vector3 = Chestt.ChestProp.GetTransform().Translation
                        TotalDistance:float = Distance(MyPos, ChestPos) / 100.0
                        if(TotalDistance < DistanceToMe and Chestt.CurrentHP>0.0):
                            set ChestIndexFinal = ChestIndexTemp
                            #Print("Chest index{ChestIndexFinal}")
                            set FoundChest = true
                            set DistanceToMe = TotalDistance
                            set ClosestChest = Chestt.ChestProp
            
                    if(FoundChest = true):
                    #-----------------Rotate and move to chest+ rerotate because position is random
                    PetRotateToChest(ClosestChest,CreativeProp)
                    Move_To_Chest(FortniteCharacter,ClosestChest,CreativeProp)
                    PetRotateToChest(ClosestChest,CreativeProp)
                    #-----------------Attack Chest
                    if(Cine:=Pets[PetIndex].AttackAnim):
                        Cine.Play(Agent)
                        Sleep(1.5) #Attack duration
                        Cine.Stop(Agent)
                    #----------------Crit
                    var PowerCrit:float=0.0
                    var CritRand:int = GetRandomInt(0,100)
                    if (CritRand>CustomPlayer.Pet_Crit_Value):
                        set PowerCrit=Power*CustomPlayer.BonusPowerMultiplicator*CustomPlayer.Bonus_Pet_Power
                    else:
                        set PowerCrit=Power*1.5*CustomPlayer.BonusPowerMultiplicator*CustomPlayer.Bonus_Pet_Power
                    #-------------------
                    Attack_Chest(ChestIndexFinal,PowerCrit)
                    Check_Chest_Gold(ChestIndexFinal,FortniteCharacter)

            ReleasePet(CreativeProp)

    FollowPlayerL3(FortniteCharacter : fort_character, CreativeProp : creative_prop, Power:float, PetIndex:int)<suspends>: void =
        #Print("Follow Player")
        #Future position of the pet
        Xplayer:float=-100.0
        Yplayer:float= 0.0
        if:
            Agent := FortniteCharacter.GetAgent[]
            PlayerObj := player[Agent]
            CustomPlayer:Custom_Player = PlayersMap[PlayerObj]
        then:
            var CheckValue:int = CustomPlayer.ButtonL3Pressed  
            set CustomPlayer.FollowActiveL3=1
            CustomPlayer.Update_Pets_Section()
            loop:
                if(CheckValue <> CustomPlayer.ButtonL3Pressed) {break}
                else if(CustomPlayer.PlayerInTreasureAreaFlag=0):
                    #AttackCinema.Stop()
                    PetRotateToPlayer(FortniteCharacter,CreativeProp)
                    Move_To_Player(FortniteCharacter,CreativeProp,Xplayer,Yplayer)
                else if(CustomPlayer.PlayerInTreasureAreaFlag=1):
                    #-----------------Find closest chest
                    var ClosestChest:creative_prop = creative_prop{}
                    var DistanceToMe:float = 100000000.0
                    var FoundChest:logic = false
                    var ChestIndexTemp:int = -1
                    var ChestIndexFinal:int = 0
                    var ClosestChestIndex:int=0
                    Allthechests:= ChestsL1+ChestsL2+ChestsL3+ChestsL4+ChestsL5+ChestsL6+ChestsL7+ChestsL8+ChestsL9+ChestsL10+ChestsL11+ChestsL12+ChestsL13+ChestsL14+ChestsL15+ChestsL16+ChestsL17+ChestsL18+ChestsL19+ChestsL20+ChestsL21+ChestsL22+ChestsL23+ChestsL24+ChestsL25+ChestsL26+ChestsL27+ChestsL28+ChestsL29+ChestsL30
                    for(Chestt:Allthechests):
                        set ChestIndexTemp+=1
                        MyPos:vector3 = FortniteCharacter.GetTransform().Translation
                        ChestPos:vector3 = Chestt.ChestProp.GetTransform().Translation
                        TotalDistance:float = Distance(MyPos, ChestPos) / 100.0
                        if(TotalDistance < DistanceToMe and Chestt.CurrentHP>0.0):
                            set ChestIndexFinal = ChestIndexTemp
                            #Print("Chest index{ChestIndexFinal}")
                            set FoundChest = true
                            set DistanceToMe = TotalDistance
                            set ClosestChest = Chestt.ChestProp
            
                    if(FoundChest = true):
                    #-----------------Rotate and move to chest+ rerotate because position is random
                    PetRotateToChest(ClosestChest,CreativeProp)
                    Move_To_Chest(FortniteCharacter,ClosestChest,CreativeProp)
                    PetRotateToChest(ClosestChest,CreativeProp)
                    #-----------------Attack Chest
                    if(Cine:=Pets[PetIndex].AttackAnim):
                        Cine.Play(Agent)
                        Sleep(1.5) #Attack duration
                        Cine.Stop(Agent)
                    #----------------Crit
                    var PowerCrit:float=0.0
                    var CritRand:int = GetRandomInt(0,100)
                    if (CritRand>CustomPlayer.Pet_Crit_Value):
                        set PowerCrit=Power*CustomPlayer.BonusPowerMultiplicator*CustomPlayer.Bonus_Pet_Power
                    else:
                        set PowerCrit=Power*1.5*CustomPlayer.BonusPowerMultiplicator*CustomPlayer.Bonus_Pet_Power
                    #-------------------
                    Attack_Chest(ChestIndexFinal,PowerCrit)
                    Check_Chest_Gold(ChestIndexFinal,FortniteCharacter)
            
            ReleasePet(CreativeProp)

    FollowPlayerL4(FortniteCharacter : fort_character, CreativeProp : creative_prop, Power:float, PetIndex:int)<suspends>: void =
        #Print("Follow Player")
        #Future position of the pet
        Xplayer:float=-200.0
        Yplayer:float= 0.0
        if:
            Agent := FortniteCharacter.GetAgent[]
            PlayerObj := player[Agent]
            CustomPlayer:Custom_Player = PlayersMap[PlayerObj]
        then:
            var CheckValue:int = CustomPlayer.ButtonL4Pressed  
            set CustomPlayer.FollowActiveL4=1
            CustomPlayer.Update_Pets_Section()
            loop:
                if(CheckValue <> CustomPlayer.ButtonL4Pressed) {break}
                else if(CustomPlayer.PlayerInTreasureAreaFlag=0):
                    #AttackCinema.Stop()
                    PetRotateToPlayer(FortniteCharacter,CreativeProp)
                    Move_To_Player(FortniteCharacter,CreativeProp,Xplayer,Yplayer)
                else if(CustomPlayer.PlayerInTreasureAreaFlag=1):
                    #-----------------Find closest chest
                    var ClosestChest:creative_prop = creative_prop{}
                    var DistanceToMe:float = 100000000.0
                    var FoundChest:logic = false
                    var ChestIndexTemp:int = -1
                    var ChestIndexFinal:int = 0
                    var ClosestChestIndex:int=0
                    Allthechests:= ChestsL1+ChestsL2+ChestsL3+ChestsL4+ChestsL5+ChestsL6+ChestsL7+ChestsL8+ChestsL9+ChestsL10+ChestsL11+ChestsL12+ChestsL13+ChestsL14+ChestsL15+ChestsL16+ChestsL17+ChestsL18+ChestsL19+ChestsL20+ChestsL21+ChestsL22+ChestsL23+ChestsL24+ChestsL25+ChestsL26+ChestsL27+ChestsL28+ChestsL29+ChestsL30
                    for(Chestt:Allthechests):
                        set ChestIndexTemp+=1
                        MyPos:vector3 = FortniteCharacter.GetTransform().Translation
                        ChestPos:vector3 = Chestt.ChestProp.GetTransform().Translation
                        TotalDistance:float = Distance(MyPos, ChestPos) / 100.0
                        if(TotalDistance < DistanceToMe and Chestt.CurrentHP>0.0):
                            set ChestIndexFinal = ChestIndexTemp
                            #Print("Chest index{ChestIndexFinal}")
                            set FoundChest = true
                            set DistanceToMe = TotalDistance
                            set ClosestChest = Chestt.ChestProp
            
                    if(FoundChest = true):
                    #-----------------Rotate and move to chest+ rerotate because position is random
                    PetRotateToChest(ClosestChest,CreativeProp)
                    Move_To_Chest(FortniteCharacter,ClosestChest,CreativeProp)
                    PetRotateToChest(ClosestChest,CreativeProp)
                    #-----------------Attack Chest
                    if(Cine:=Pets[PetIndex].AttackAnim):
                        Cine.Play(Agent)
                        Sleep(1.5) #Attack duration
                        Cine.Stop(Agent)
                    #----------------Crit
                    var PowerCrit:float=0.0
                    var CritRand:int = GetRandomInt(0,100)
                    if (CritRand>CustomPlayer.Pet_Crit_Value):
                        set PowerCrit=Power*CustomPlayer.BonusPowerMultiplicator*CustomPlayer.Bonus_Pet_Power
                    else:
                        set PowerCrit=Power*1.5*CustomPlayer.BonusPowerMultiplicator*CustomPlayer.Bonus_Pet_Power
                    #-------------------
                    Attack_Chest(ChestIndexFinal,PowerCrit)
                    Check_Chest_Gold(ChestIndexFinal,FortniteCharacter)

            ReleasePet(CreativeProp)

    FollowPlayerL5(FortniteCharacter : fort_character, CreativeProp : creative_prop, Power:float, PetIndex:int)<suspends>: void =
        #Print("Follow Player")
        #Future position of the pet
        Xplayer:float=-200.0
        Yplayer:float=-100.0
        if:
            Agent := FortniteCharacter.GetAgent[]
            PlayerObj := player[Agent]
            CustomPlayer:Custom_Player = PlayersMap[PlayerObj]
        then:
            var CheckValue:int = CustomPlayer.ButtonL5Pressed  
            set CustomPlayer.FollowActiveL5=1
            CustomPlayer.Update_Pets_Section()
            loop:
                if(CheckValue <> CustomPlayer.ButtonL5Pressed) {break}
                else if(CustomPlayer.PlayerInTreasureAreaFlag=0):
                    #AttackCinema.Stop()
                    PetRotateToPlayer(FortniteCharacter,CreativeProp)
                    Move_To_Player(FortniteCharacter,CreativeProp,Xplayer,Yplayer)
                else if(CustomPlayer.PlayerInTreasureAreaFlag=1):
                    #-----------------Find closest chest
                    var ClosestChest:creative_prop = creative_prop{}
                    var DistanceToMe:float = 100000000.0
                    var FoundChest:logic = false
                    var ChestIndexTemp:int = -1
                    var ChestIndexFinal:int = 0
                    var ClosestChestIndex:int=0
                    Allthechests:= ChestsL1+ChestsL2+ChestsL3+ChestsL4+ChestsL5+ChestsL6+ChestsL7+ChestsL8+ChestsL9+ChestsL10+ChestsL11+ChestsL12+ChestsL13+ChestsL14+ChestsL15+ChestsL16+ChestsL17+ChestsL18+ChestsL19+ChestsL20+ChestsL21+ChestsL22+ChestsL23+ChestsL24+ChestsL25+ChestsL26+ChestsL27+ChestsL28+ChestsL29+ChestsL30
                    for(Chestt:Allthechests):
                        set ChestIndexTemp+=1
                        MyPos:vector3 = FortniteCharacter.GetTransform().Translation
                        ChestPos:vector3 = Chestt.ChestProp.GetTransform().Translation
                        TotalDistance:float = Distance(MyPos, ChestPos) / 100.0
                        if(TotalDistance < DistanceToMe and Chestt.CurrentHP>0.0):
                            set ChestIndexFinal = ChestIndexTemp
                            #Print("Chest index{ChestIndexFinal}")
                            set FoundChest = true
                            set DistanceToMe = TotalDistance
                            set ClosestChest = Chestt.ChestProp
            
                    if(FoundChest = true):
                    #-----------------Rotate and move to chest+ rerotate because position is random
                    PetRotateToChest(ClosestChest,CreativeProp)
                    Move_To_Chest(FortniteCharacter,ClosestChest,CreativeProp)
                    PetRotateToChest(ClosestChest,CreativeProp)
                    #-----------------Attack Chest
                    if(Cine:=Pets[PetIndex].AttackAnim):
                        Cine.Play(Agent)
                        Sleep(1.5) #Attack duration
                        Cine.Stop(Agent)
                    #----------------Crit
                    var PowerCrit:float=0.0
                    var CritRand:int = GetRandomInt(0,100)
                    if (CritRand>CustomPlayer.Pet_Crit_Value):
                        set PowerCrit=Power*CustomPlayer.BonusPowerMultiplicator*CustomPlayer.Bonus_Pet_Power
                    else:
                        set PowerCrit=Power*1.5*CustomPlayer.BonusPowerMultiplicator*CustomPlayer.Bonus_Pet_Power
                    #-------------------
                    Attack_Chest(ChestIndexFinal,PowerCrit)
                    Check_Chest_Gold(ChestIndexFinal,FortniteCharacter)

            ReleasePet(CreativeProp)

    FollowPlayerL6(FortniteCharacter : fort_character, CreativeProp : creative_prop, Power:float, PetIndex:int)<suspends>: void =
        #Print("Follow Player")
        #Future position of the pet
        Xplayer:float=-200.0
        Yplayer:float= 100.0
        if:
            Agent := FortniteCharacter.GetAgent[]
            PlayerObj := player[Agent]
            CustomPlayer:Custom_Player = PlayersMap[PlayerObj]
        then:
            var CheckValue:int = CustomPlayer.ButtonL6Pressed  
            set CustomPlayer.FollowActiveL6=1
            CustomPlayer.Update_Pets_Section()
            loop:
                if(CheckValue <> CustomPlayer.ButtonL6Pressed) {break}
                else if(CustomPlayer.PlayerInTreasureAreaFlag=0):
                    #AttackCinema.Stop()
                    PetRotateToPlayer(FortniteCharacter,CreativeProp)
                    Move_To_Player(FortniteCharacter,CreativeProp,Xplayer,Yplayer)
                else if(CustomPlayer.PlayerInTreasureAreaFlag=1):
                    #-----------------Find closest chest
                    var ClosestChest:creative_prop = creative_prop{}
                    var DistanceToMe:float = 100000000.0
                    var FoundChest:logic = false
                    var ChestIndexTemp:int = -1
                    var ChestIndexFinal:int = 0
                    var ClosestChestIndex:int=0
                    Allthechests:= ChestsL1+ChestsL2+ChestsL3+ChestsL4+ChestsL5+ChestsL6+ChestsL7+ChestsL8+ChestsL9+ChestsL10+ChestsL11+ChestsL12+ChestsL13+ChestsL14+ChestsL15+ChestsL16+ChestsL17+ChestsL18+ChestsL19+ChestsL20+ChestsL21+ChestsL22+ChestsL23+ChestsL24+ChestsL25+ChestsL26+ChestsL27+ChestsL28+ChestsL29+ChestsL30
                    for(Chestt:Allthechests):
                        set ChestIndexTemp+=1
                        MyPos:vector3 = FortniteCharacter.GetTransform().Translation
                        ChestPos:vector3 = Chestt.ChestProp.GetTransform().Translation
                        TotalDistance:float = Distance(MyPos, ChestPos) / 100.0
                        if(TotalDistance < DistanceToMe and Chestt.CurrentHP>0.0):
                            set ChestIndexFinal = ChestIndexTemp
                            #Print("Chest index{ChestIndexFinal}")
                            set FoundChest = true
                            set DistanceToMe = TotalDistance
                            set ClosestChest = Chestt.ChestProp
            
                    if(FoundChest = true):
                    #-----------------Rotate and move to chest+ rerotate because position is random
                    PetRotateToChest(ClosestChest,CreativeProp)
                    Move_To_Chest(FortniteCharacter,ClosestChest,CreativeProp)
                    PetRotateToChest(ClosestChest,CreativeProp)
                    #-----------------Attack Chest
                    if(Cine:=Pets[PetIndex].AttackAnim):
                        Cine.Play(Agent)
                        Sleep(1.5) #Attack duration
                        Cine.Stop(Agent)
                    #----------------Crit
                    var PowerCrit:float=0.0
                    var CritRand:int = GetRandomInt(0,100)
                    if (CritRand>CustomPlayer.Pet_Crit_Value):
                        set PowerCrit=Power*CustomPlayer.BonusPowerMultiplicator*CustomPlayer.Bonus_Pet_Power
                    else:
                        set PowerCrit=Power*1.5*CustomPlayer.BonusPowerMultiplicator*CustomPlayer.Bonus_Pet_Power
                    #-------------------
                    Attack_Chest(ChestIndexFinal,PowerCrit)
                    Check_Chest_Gold(ChestIndexFinal,FortniteCharacter)

            ReleasePet(CreativeProp)

    FollowPlayerL7(FortniteCharacter : fort_character, CreativeProp : creative_prop, Power:float, PetIndex:int)<suspends>: void =
        #Print("Follow Player")
        #Future position of the pet
        Xplayer:float=-200.0
        Yplayer:float=-200.0
        if:
            Agent := FortniteCharacter.GetAgent[]
            PlayerObj := player[Agent]
            CustomPlayer:Custom_Player = PlayersMap[PlayerObj]
        then:
            var CheckValue:int = CustomPlayer.ButtonL7Pressed  
            set CustomPlayer.FollowActiveL7=1
            CustomPlayer.Update_Pets_Section()
            loop:
                if(CheckValue <> CustomPlayer.ButtonL7Pressed) {break}
                else if(CustomPlayer.PlayerInTreasureAreaFlag=0):
                    #AttackCinema.Stop()
                    PetRotateToPlayer(FortniteCharacter,CreativeProp)
                    Move_To_Player(FortniteCharacter,CreativeProp,Xplayer,Yplayer)
                else if(CustomPlayer.PlayerInTreasureAreaFlag=1):
                    #-----------------Find closest chest
                    var ClosestChest:creative_prop = creative_prop{}
                    var DistanceToMe:float = 100000000.0
                    var FoundChest:logic = false
                    var ChestIndexTemp:int = -1
                    var ChestIndexFinal:int = 0
                    var ClosestChestIndex:int=0
                    Allthechests:= ChestsL1+ChestsL2+ChestsL3+ChestsL4+ChestsL5+ChestsL6+ChestsL7+ChestsL8+ChestsL9+ChestsL10+ChestsL11+ChestsL12+ChestsL13+ChestsL14+ChestsL15+ChestsL16+ChestsL17+ChestsL18+ChestsL19+ChestsL20+ChestsL21+ChestsL22+ChestsL23+ChestsL24+ChestsL25+ChestsL26+ChestsL27+ChestsL28+ChestsL29+ChestsL30
                    for(Chestt:Allthechests):
                        set ChestIndexTemp+=1
                        MyPos:vector3 = FortniteCharacter.GetTransform().Translation
                        ChestPos:vector3 = Chestt.ChestProp.GetTransform().Translation
                        TotalDistance:float = Distance(MyPos, ChestPos) / 100.0
                        if(TotalDistance < DistanceToMe and Chestt.CurrentHP>0.0):
                            set ChestIndexFinal = ChestIndexTemp
                            #Print("Chest index{ChestIndexFinal}")
                            set FoundChest = true
                            set DistanceToMe = TotalDistance
                            set ClosestChest = Chestt.ChestProp
            
                    if(FoundChest = true):
                    #-----------------Rotate and move to chest+ rerotate because position is random
                    PetRotateToChest(ClosestChest,CreativeProp)
                    Move_To_Chest(FortniteCharacter,ClosestChest,CreativeProp)
                    PetRotateToChest(ClosestChest,CreativeProp)
                    #-----------------Attack Chest
                    if(Cine:=Pets[PetIndex].AttackAnim):
                        Cine.Play(Agent)
                        Sleep(1.5) #Attack duration
                        Cine.Stop(Agent)
                    #----------------Crit
                    var PowerCrit:float=0.0
                    var CritRand:int = GetRandomInt(0,100)
                    if (CritRand>CustomPlayer.Pet_Crit_Value):
                        set PowerCrit=Power*CustomPlayer.BonusPowerMultiplicator*CustomPlayer.Bonus_Pet_Power
                    else:
                        set PowerCrit=Power*1.5*CustomPlayer.BonusPowerMultiplicator*CustomPlayer.Bonus_Pet_Power
                    #-------------------
                    Attack_Chest(ChestIndexFinal,PowerCrit)
                    Check_Chest_Gold(ChestIndexFinal,FortniteCharacter)

            ReleasePet(CreativeProp)

    FollowPlayerL8(FortniteCharacter : fort_character, CreativeProp : creative_prop, Power:float, PetIndex:int)<suspends>: void =
        #Print("Follow Player")
        #Future position of the pet
        Xplayer:float=-200.0
        Yplayer:float= 200.0
        if:
            Agent := FortniteCharacter.GetAgent[]
            PlayerObj := player[Agent]
            CustomPlayer:Custom_Player = PlayersMap[PlayerObj]
        then:
            var CheckValue:int = CustomPlayer.ButtonL8Pressed  
            set CustomPlayer.FollowActiveL8=1
            CustomPlayer.Update_Pets_Section()
            loop:
                if(CheckValue <> CustomPlayer.ButtonL8Pressed) {break}
                else if(CustomPlayer.PlayerInTreasureAreaFlag=0):
                    #AttackCinema.Stop()
                    PetRotateToPlayer(FortniteCharacter,CreativeProp)
                    Move_To_Player(FortniteCharacter,CreativeProp,Xplayer,Yplayer)
                else if(CustomPlayer.PlayerInTreasureAreaFlag=1):
                    #-----------------Find closest chest
                    var ClosestChest:creative_prop = creative_prop{}
                    var DistanceToMe:float = 100000000.0
                    var FoundChest:logic = false
                    var ChestIndexTemp:int = -1
                    var ChestIndexFinal:int = 0
                    var ClosestChestIndex:int=0
                    Allthechests:= ChestsL1+ChestsL2+ChestsL3+ChestsL4+ChestsL5+ChestsL6+ChestsL7+ChestsL8+ChestsL9+ChestsL10+ChestsL11+ChestsL12+ChestsL13+ChestsL14+ChestsL15+ChestsL16+ChestsL17+ChestsL18+ChestsL19+ChestsL20+ChestsL21+ChestsL22+ChestsL23+ChestsL24+ChestsL25+ChestsL26+ChestsL27+ChestsL28+ChestsL29+ChestsL30
                    for(Chestt:Allthechests):
                        set ChestIndexTemp+=1
                        MyPos:vector3 = FortniteCharacter.GetTransform().Translation
                        ChestPos:vector3 = Chestt.ChestProp.GetTransform().Translation
                        TotalDistance:float = Distance(MyPos, ChestPos) / 100.0
                        if(TotalDistance < DistanceToMe and Chestt.CurrentHP>0.0):
                            set ChestIndexFinal = ChestIndexTemp
                            #Print("Chest index{ChestIndexFinal}")
                            set FoundChest = true
                            set DistanceToMe = TotalDistance
                            set ClosestChest = Chestt.ChestProp
            
                    if(FoundChest = true):
                    #-----------------Rotate and move to chest+ rerotate because position is random
                    PetRotateToChest(ClosestChest,CreativeProp)
                    Move_To_Chest(FortniteCharacter,ClosestChest,CreativeProp)
                    PetRotateToChest(ClosestChest,CreativeProp)
                    #-----------------Attack Chest
                    if(Cine:=Pets[PetIndex].AttackAnim):
                        Cine.Play(Agent)
                        Sleep(1.5) #Attack duration
                        Cine.Stop(Agent)
                    #----------------Crit
                    var PowerCrit:float=0.0
                    var CritRand:int = GetRandomInt(0,100)
                    if (CritRand>CustomPlayer.Pet_Crit_Value):
                        set PowerCrit=Power*CustomPlayer.BonusPowerMultiplicator*CustomPlayer.Bonus_Pet_Power
                    else:
                        set PowerCrit=Power*1.5*CustomPlayer.BonusPowerMultiplicator*CustomPlayer.Bonus_Pet_Power
                    #-------------------
                    Attack_Chest(ChestIndexFinal,PowerCrit)
                    Check_Chest_Gold(ChestIndexFinal,FortniteCharacter)

            ReleasePet(CreativeProp)

    FollowPlayerL9(FortniteCharacter : fort_character, CreativeProp : creative_prop, Power:float, PetIndex:int)<suspends>: void =
        #Print("Follow Player")
        #Future position of the pet
        Xplayer:float=-300.0
        Yplayer:float= 0.0
        if:
            Agent := FortniteCharacter.GetAgent[]
            PlayerObj := player[Agent]
            CustomPlayer:Custom_Player = PlayersMap[PlayerObj]
        then:
            var CheckValue:int = CustomPlayer.ButtonL9Pressed  
            set CustomPlayer.FollowActiveL9=1
            CustomPlayer.Update_Pets_Section()
            loop:
                if(CheckValue <> CustomPlayer.ButtonL9Pressed) {break}
                else if(CustomPlayer.PlayerInTreasureAreaFlag=0):
                    #AttackCinema.Stop()
                    PetRotateToPlayer(FortniteCharacter,CreativeProp)
                    Move_To_Player(FortniteCharacter,CreativeProp,Xplayer,Yplayer)
                else if(CustomPlayer.PlayerInTreasureAreaFlag=1):
                    #-----------------Find closest chest
                    var ClosestChest:creative_prop = creative_prop{}
                    var DistanceToMe:float = 100000000.0
                    var FoundChest:logic = false
                    var ChestIndexTemp:int = -1
                    var ChestIndexFinal:int = 0
                    var ClosestChestIndex:int=0
                    Allthechests:= ChestsL1+ChestsL2+ChestsL3+ChestsL4+ChestsL5+ChestsL6+ChestsL7+ChestsL8+ChestsL9+ChestsL10+ChestsL11+ChestsL12+ChestsL13+ChestsL14+ChestsL15+ChestsL16+ChestsL17+ChestsL18+ChestsL19+ChestsL20+ChestsL21+ChestsL22+ChestsL23+ChestsL24+ChestsL25+ChestsL26+ChestsL27+ChestsL28+ChestsL29+ChestsL30
                    for(Chestt:Allthechests):
                        set ChestIndexTemp+=1
                        MyPos:vector3 = FortniteCharacter.GetTransform().Translation
                        ChestPos:vector3 = Chestt.ChestProp.GetTransform().Translation
                        TotalDistance:float = Distance(MyPos, ChestPos) / 100.0
                        if(TotalDistance < DistanceToMe and Chestt.CurrentHP>0.0):
                            set ChestIndexFinal = ChestIndexTemp
                            #Print("Chest index{ChestIndexFinal}")
                            set FoundChest = true
                            set DistanceToMe = TotalDistance
                            set ClosestChest = Chestt.ChestProp
            
                    if(FoundChest = true):
                    #-----------------Rotate and move to chest+ rerotate because position is random
                    PetRotateToChest(ClosestChest,CreativeProp)
                    Move_To_Chest(FortniteCharacter,ClosestChest,CreativeProp)
                    PetRotateToChest(ClosestChest,CreativeProp)
                    #-----------------Attack Chest
                    if(Cine:=Pets[PetIndex].AttackAnim):
                        Cine.Play(Agent)
                        Sleep(1.5) #Attack duration
                        Cine.Stop(Agent)
                    #----------------Crit
                    var PowerCrit:float=0.0
                    var CritRand:int = GetRandomInt(0,100)
                    if (CritRand>CustomPlayer.Pet_Crit_Value):
                        set PowerCrit=Power*CustomPlayer.BonusPowerMultiplicator*CustomPlayer.Bonus_Pet_Power
                    else:
                        set PowerCrit=Power*1.5*CustomPlayer.BonusPowerMultiplicator*CustomPlayer.Bonus_Pet_Power
                    #-------------------
                    Attack_Chest(ChestIndexFinal,PowerCrit)
                    Check_Chest_Gold(ChestIndexFinal,FortniteCharacter)

            ReleasePet(CreativeProp)

    FollowPlayerL10(FortniteCharacter : fort_character, CreativeProp : creative_prop, Power:float, PetIndex:int)<suspends>: void =
        #Print("Follow Player")
        #Future position of the pet
        Xplayer:float=-300.0
        Yplayer:float=-100.0
        if:
            Agent := FortniteCharacter.GetAgent[]
            PlayerObj := player[Agent]
            CustomPlayer:Custom_Player = PlayersMap[PlayerObj]
        then:
            var CheckValue:int = CustomPlayer.ButtonL10Pressed  
            set CustomPlayer.FollowActiveL10=1
            CustomPlayer.Update_Pets_Section()
            loop:
                if(CheckValue <> CustomPlayer.ButtonL10Pressed) {break}
                else if(CustomPlayer.PlayerInTreasureAreaFlag=0):
                    #AttackCinema.Stop()
                    PetRotateToPlayer(FortniteCharacter,CreativeProp)
                    Move_To_Player(FortniteCharacter,CreativeProp,Xplayer,Yplayer)
                else if(CustomPlayer.PlayerInTreasureAreaFlag=1):
                    #-----------------Find closest chest
                    var ClosestChest:creative_prop = creative_prop{}
                    var DistanceToMe:float = 100000000.0
                    var FoundChest:logic = false
                    var ChestIndexTemp:int = -1
                    var ChestIndexFinal:int = 0
                    var ClosestChestIndex:int=0
                    Allthechests:= ChestsL1+ChestsL2+ChestsL3+ChestsL4+ChestsL5+ChestsL6+ChestsL7+ChestsL8+ChestsL9+ChestsL10+ChestsL11+ChestsL12+ChestsL13+ChestsL14+ChestsL15+ChestsL16+ChestsL17+ChestsL18+ChestsL19+ChestsL20+ChestsL21+ChestsL22+ChestsL23+ChestsL24+ChestsL25+ChestsL26+ChestsL27+ChestsL28+ChestsL29+ChestsL30
                    for(Chestt:Allthechests):
                        set ChestIndexTemp+=1
                        MyPos:vector3 = FortniteCharacter.GetTransform().Translation
                        ChestPos:vector3 = Chestt.ChestProp.GetTransform().Translation
                        TotalDistance:float = Distance(MyPos, ChestPos) / 100.0
                        if(TotalDistance < DistanceToMe and Chestt.CurrentHP>0.0):
                            set ChestIndexFinal = ChestIndexTemp
                            #Print("Chest index{ChestIndexFinal}")
                            set FoundChest = true
                            set DistanceToMe = TotalDistance
                            set ClosestChest = Chestt.ChestProp
            
                    if(FoundChest = true):
                    #-----------------Rotate and move to chest+ rerotate because position is random
                    PetRotateToChest(ClosestChest,CreativeProp)
                    Move_To_Chest(FortniteCharacter,ClosestChest,CreativeProp)
                    PetRotateToChest(ClosestChest,CreativeProp)
                    #-----------------Attack Chest
                    if(Cine:=Pets[PetIndex].AttackAnim):
                        Cine.Play(Agent)
                        Sleep(1.5) #Attack duration
                        Cine.Stop(Agent)
                    #----------------Crit
                    var PowerCrit:float=0.0
                    var CritRand:int = GetRandomInt(0,100)
                    if (CritRand>CustomPlayer.Pet_Crit_Value):
                        set PowerCrit=Power*CustomPlayer.BonusPowerMultiplicator*CustomPlayer.Bonus_Pet_Power
                    else:
                        set PowerCrit=Power*1.5*CustomPlayer.BonusPowerMultiplicator*CustomPlayer.Bonus_Pet_Power
                    #-------------------
                    Attack_Chest(ChestIndexFinal,PowerCrit)
                    Check_Chest_Gold(ChestIndexFinal,FortniteCharacter)

            ReleasePet(CreativeProp)

    FollowPlayerL11(FortniteCharacter : fort_character, CreativeProp : creative_prop, Power:float, PetIndex:int)<suspends>: void =
        #Print("Follow Player")
        #Future position of the pet
        Xplayer:float=-300.0
        Yplayer:float= 100.0
        if:
            Agent := FortniteCharacter.GetAgent[]
            PlayerObj := player[Agent]
            CustomPlayer:Custom_Player = PlayersMap[PlayerObj]
        then:
            var CheckValue:int = CustomPlayer.ButtonL11Pressed  
            set CustomPlayer.FollowActiveL11=1
            CustomPlayer.Update_Pets_Section()
            loop:
                if(CheckValue <> CustomPlayer.ButtonL11Pressed) {break}
                else if(CustomPlayer.PlayerInTreasureAreaFlag=0):
                    #AttackCinema.Stop()
                    PetRotateToPlayer(FortniteCharacter,CreativeProp)
                    Move_To_Player(FortniteCharacter,CreativeProp,Xplayer,Yplayer)
                else if(CustomPlayer.PlayerInTreasureAreaFlag=1):
                    #-----------------Find closest chest
                    var ClosestChest:creative_prop = creative_prop{}
                    var DistanceToMe:float = 100000000.0
                    var FoundChest:logic = false
                    var ChestIndexTemp:int = -1
                    var ChestIndexFinal:int = 0
                    var ClosestChestIndex:int=0
                    Allthechests:= ChestsL1+ChestsL2+ChestsL3+ChestsL4+ChestsL5+ChestsL6+ChestsL7+ChestsL8+ChestsL9+ChestsL10+ChestsL11+ChestsL12+ChestsL13+ChestsL14+ChestsL15+ChestsL16+ChestsL17+ChestsL18+ChestsL19+ChestsL20+ChestsL21+ChestsL22+ChestsL23+ChestsL24+ChestsL25+ChestsL26+ChestsL27+ChestsL28+ChestsL29+ChestsL30
                    for(Chestt:Allthechests):
                        set ChestIndexTemp+=1
                        MyPos:vector3 = FortniteCharacter.GetTransform().Translation
                        ChestPos:vector3 = Chestt.ChestProp.GetTransform().Translation
                        TotalDistance:float = Distance(MyPos, ChestPos) / 100.0
                        if(TotalDistance < DistanceToMe and Chestt.CurrentHP>0.0):
                            set ChestIndexFinal = ChestIndexTemp
                            #Print("Chest index{ChestIndexFinal}")
                            set FoundChest = true
                            set DistanceToMe = TotalDistance
                            set ClosestChest = Chestt.ChestProp
            
                    if(FoundChest = true):
                    #-----------------Rotate and move to chest+ rerotate because position is random
                    PetRotateToChest(ClosestChest,CreativeProp)
                    Move_To_Chest(FortniteCharacter,ClosestChest,CreativeProp)
                    PetRotateToChest(ClosestChest,CreativeProp)
                    #-----------------Attack Chest
                    if(Cine:=Pets[PetIndex].AttackAnim):
                        Cine.Play(Agent)
                        Sleep(1.5) #Attack duration
                        Cine.Stop(Agent)
                    #----------------Crit
                    var PowerCrit:float=0.0
                    var CritRand:int = GetRandomInt(0,100)
                    if (CritRand>CustomPlayer.Pet_Crit_Value):
                        set PowerCrit=Power*CustomPlayer.BonusPowerMultiplicator*CustomPlayer.Bonus_Pet_Power
                    else:
                        set PowerCrit=Power*1.5*CustomPlayer.BonusPowerMultiplicator*CustomPlayer.Bonus_Pet_Power
                    #-------------------
                    Attack_Chest(ChestIndexFinal,PowerCrit)
                    Check_Chest_Gold(ChestIndexFinal,FortniteCharacter)

            ReleasePet(CreativeProp)

    FollowPlayerL12(FortniteCharacter : fort_character, CreativeProp : creative_prop, Power:float, PetIndex:int)<suspends>: void =
        #Print("Follow Player")
        #Future position of the pet
        Xplayer:float=-300.0
        Yplayer:float=-200.0
        if:
            Agent := FortniteCharacter.GetAgent[]
            PlayerObj := player[Agent]
            CustomPlayer:Custom_Player = PlayersMap[PlayerObj]
        then:
            var CheckValue:int = CustomPlayer.ButtonL12Pressed  
            set CustomPlayer.FollowActiveL12=1
            CustomPlayer.Update_Pets_Section()
            loop:
                if(CheckValue <> CustomPlayer.ButtonL12Pressed) {break}
                else if(CustomPlayer.PlayerInTreasureAreaFlag=0):
                    #AttackCinema.Stop()
                    PetRotateToPlayer(FortniteCharacter,CreativeProp)
                    Move_To_Player(FortniteCharacter,CreativeProp,Xplayer,Yplayer)
                else if(CustomPlayer.PlayerInTreasureAreaFlag=1):
                    #-----------------Find closest chest
                    var ClosestChest:creative_prop = creative_prop{}
                    var DistanceToMe:float = 100000000.0
                    var FoundChest:logic = false
                    var ChestIndexTemp:int = -1
                    var ChestIndexFinal:int = 0
                    var ClosestChestIndex:int=0
                    Allthechests:= ChestsL1+ChestsL2+ChestsL3+ChestsL4+ChestsL5+ChestsL6+ChestsL7+ChestsL8+ChestsL9+ChestsL10+ChestsL11+ChestsL12+ChestsL13+ChestsL14+ChestsL15+ChestsL16+ChestsL17+ChestsL18+ChestsL19+ChestsL20+ChestsL21+ChestsL22+ChestsL23+ChestsL24+ChestsL25+ChestsL26+ChestsL27+ChestsL28+ChestsL29+ChestsL30
                    for(Chestt:Allthechests):
                        set ChestIndexTemp+=1
                        MyPos:vector3 = FortniteCharacter.GetTransform().Translation
                        ChestPos:vector3 = Chestt.ChestProp.GetTransform().Translation
                        TotalDistance:float = Distance(MyPos, ChestPos) / 100.0
                        if(TotalDistance < DistanceToMe and Chestt.CurrentHP>0.0):
                            set ChestIndexFinal = ChestIndexTemp
                            #Print("Chest index{ChestIndexFinal}")
                            set FoundChest = true
                            set DistanceToMe = TotalDistance
                            set ClosestChest = Chestt.ChestProp
            
                    if(FoundChest = true):
                    #-----------------Rotate and move to chest+ rerotate because position is random
                    PetRotateToChest(ClosestChest,CreativeProp)
                    Move_To_Chest(FortniteCharacter,ClosestChest,CreativeProp)
                    PetRotateToChest(ClosestChest,CreativeProp)
                    #-----------------Attack Chest
                    if(Cine:=Pets[PetIndex].AttackAnim):
                        Cine.Play(Agent)
                        Sleep(1.5) #Attack duration
                        Cine.Stop(Agent)
                    #----------------Crit
                    var PowerCrit:float=0.0
                    var CritRand:int = GetRandomInt(0,100)
                    if (CritRand>CustomPlayer.Pet_Crit_Value):
                        set PowerCrit=Power*CustomPlayer.BonusPowerMultiplicator*CustomPlayer.Bonus_Pet_Power
                    else:
                        set PowerCrit=Power*1.5*CustomPlayer.BonusPowerMultiplicator*CustomPlayer.Bonus_Pet_Power
                    #-------------------
                    Attack_Chest(ChestIndexFinal,PowerCrit)
                    Check_Chest_Gold(ChestIndexFinal,FortniteCharacter)

            ReleasePet(CreativeProp)

    FollowPlayerL13(FortniteCharacter : fort_character, CreativeProp : creative_prop, Power:float, PetIndex:int)<suspends>: void =              
        #Print("Follow Player")
        #Future position of the pet
        Xplayer:float=-300.0
        Yplayer:float= 200.0
        if:
            Agent := FortniteCharacter.GetAgent[]
            PlayerObj := player[Agent]
            CustomPlayer:Custom_Player = PlayersMap[PlayerObj]
        then:
            var CheckValue:int = CustomPlayer.ButtonL13Pressed  
            set CustomPlayer.FollowActiveL13=1
            CustomPlayer.Update_Pets_Section()
            loop:
                if(CheckValue <> CustomPlayer.ButtonL13Pressed) {break}
                else if(CustomPlayer.PlayerInTreasureAreaFlag=0):
                    #AttackCinema.Stop()
                    PetRotateToPlayer(FortniteCharacter,CreativeProp)
                    Move_To_Player(FortniteCharacter,CreativeProp,Xplayer,Yplayer)
                else if(CustomPlayer.PlayerInTreasureAreaFlag=1):
                    #-----------------Find closest chest
                    var ClosestChest:creative_prop = creative_prop{}
                    var DistanceToMe:float = 100000000.0
                    var FoundChest:logic = false
                    var ChestIndexTemp:int = -1
                    var ChestIndexFinal:int = 0
                    var ClosestChestIndex:int=0
                    Allthechests:= ChestsL1+ChestsL2+ChestsL3+ChestsL4+ChestsL5+ChestsL6+ChestsL7+ChestsL8+ChestsL9+ChestsL10+ChestsL11+ChestsL12+ChestsL13+ChestsL14+ChestsL15+ChestsL16+ChestsL17+ChestsL18+ChestsL19+ChestsL20+ChestsL21+ChestsL22+ChestsL23+ChestsL24+ChestsL25+ChestsL26+ChestsL27+ChestsL28+ChestsL29+ChestsL30
                    for(Chestt:Allthechests):
                        set ChestIndexTemp+=1
                        MyPos:vector3 = FortniteCharacter.GetTransform().Translation
                        ChestPos:vector3 = Chestt.ChestProp.GetTransform().Translation
                        TotalDistance:float = Distance(MyPos, ChestPos) / 100.0
                        if(TotalDistance < DistanceToMe and Chestt.CurrentHP>0.0):
                            set ChestIndexFinal = ChestIndexTemp
                            #Print("Chest index{ChestIndexFinal}")
                            set FoundChest = true
                            set DistanceToMe = TotalDistance
                            set ClosestChest = Chestt.ChestProp
            
                    if(FoundChest = true):
                    #-----------------Rotate and move to chest+ rerotate because position is random
                    PetRotateToChest(ClosestChest,CreativeProp)
                    Move_To_Chest(FortniteCharacter,ClosestChest,CreativeProp)
                    PetRotateToChest(ClosestChest,CreativeProp)
                    #-----------------Attack Chest
                    if(Cine:=Pets[PetIndex].AttackAnim):
                        Cine.Play(Agent)
                        Sleep(1.5) #Attack duration
                        Cine.Stop(Agent)
                    #----------------Crit
                    var PowerCrit:float=0.0
                    var CritRand:int = GetRandomInt(0,100)
                    if (CritRand>CustomPlayer.Pet_Crit_Value):
                        set PowerCrit=Power*CustomPlayer.BonusPowerMultiplicator*CustomPlayer.Bonus_Pet_Power
                    else:
                        set PowerCrit=Power*1.5*CustomPlayer.BonusPowerMultiplicator*CustomPlayer.Bonus_Pet_Power
                    #-------------------
                    Attack_Chest(ChestIndexFinal,PowerCrit)
                    Check_Chest_Gold(ChestIndexFinal,FortniteCharacter)

            ReleasePet(CreativeProp)

    FollowPlayerL14(FortniteCharacter : fort_character, CreativeProp : creative_prop, Power:float, PetIndex:int)<suspends>: void =
        #Print("Follow Player")
        #Future position of the pet
        Xplayer:float=-300.0
        Yplayer:float=-300.0
        if:
            Agent := FortniteCharacter.GetAgent[]
            PlayerObj := player[Agent]
            CustomPlayer:Custom_Player = PlayersMap[PlayerObj]
        then:
            var CheckValue:int = CustomPlayer.ButtonL14Pressed  
            set CustomPlayer.FollowActiveL14=1
            CustomPlayer.Update_Pets_Section()
            loop:
                if(CheckValue <> CustomPlayer.ButtonL14Pressed) {break}
                else if(CustomPlayer.PlayerInTreasureAreaFlag=0):
                    #AttackCinema.Stop()
                    PetRotateToPlayer(FortniteCharacter,CreativeProp)
                    Move_To_Player(FortniteCharacter,CreativeProp,Xplayer,Yplayer)
                else if(CustomPlayer.PlayerInTreasureAreaFlag=1):
                    #-----------------Find closest chest
                    var ClosestChest:creative_prop = creative_prop{}
                    var DistanceToMe:float = 100000000.0
                    var FoundChest:logic = false
                    var ChestIndexTemp:int = -1
                    var ChestIndexFinal:int = 0
                    var ClosestChestIndex:int=0
                    Allthechests:= ChestsL1+ChestsL2+ChestsL3+ChestsL4+ChestsL5+ChestsL6+ChestsL7+ChestsL8+ChestsL9+ChestsL10+ChestsL11+ChestsL12+ChestsL13+ChestsL14+ChestsL15+ChestsL16+ChestsL17+ChestsL18+ChestsL19+ChestsL20+ChestsL21+ChestsL22+ChestsL23+ChestsL24+ChestsL25+ChestsL26+ChestsL27+ChestsL28+ChestsL29+ChestsL30
                    for(Chestt:Allthechests):
                        set ChestIndexTemp+=1
                        MyPos:vector3 = FortniteCharacter.GetTransform().Translation
                        ChestPos:vector3 = Chestt.ChestProp.GetTransform().Translation
                        TotalDistance:float = Distance(MyPos, ChestPos) / 100.0
                        if(TotalDistance < DistanceToMe and Chestt.CurrentHP>0.0):
                            set ChestIndexFinal = ChestIndexTemp
                            #Print("Chest index{ChestIndexFinal}")
                            set FoundChest = true
                            set DistanceToMe = TotalDistance
                            set ClosestChest = Chestt.ChestProp
            
                    if(FoundChest = true):
                    #-----------------Rotate and move to chest+ rerotate because position is random
                    PetRotateToChest(ClosestChest,CreativeProp)
                    Move_To_Chest(FortniteCharacter,ClosestChest,CreativeProp)
                    PetRotateToChest(ClosestChest,CreativeProp)
                    #-----------------Attack Chest
                    if(Cine:=Pets[PetIndex].AttackAnim):
                        Cine.Play(Agent)
                        Sleep(1.5) #Attack duration
                        Cine.Stop(Agent)
                    #----------------Crit
                    var PowerCrit:float=0.0
                    var CritRand:int = GetRandomInt(0,100)
                    if (CritRand>CustomPlayer.Pet_Crit_Value):
                        set PowerCrit=Power*CustomPlayer.BonusPowerMultiplicator*CustomPlayer.Bonus_Pet_Power
                    else:
                        set PowerCrit=Power*1.5*CustomPlayer.BonusPowerMultiplicator*CustomPlayer.Bonus_Pet_Power
                    #-------------------
                    Attack_Chest(ChestIndexFinal,PowerCrit)
                    Check_Chest_Gold(ChestIndexFinal,FortniteCharacter)

            ReleasePet(CreativeProp)

    FollowPlayerL15(FortniteCharacter : fort_character, CreativeProp : creative_prop, Power:float, PetIndex:int)<suspends>: void =
        #Print("Follow Player")
        #Future position of the pet
        Xplayer:float=-300.0
        Yplayer:float= 300.0
        if:
            Agent := FortniteCharacter.GetAgent[]
            PlayerObj := player[Agent]
            CustomPlayer:Custom_Player = PlayersMap[PlayerObj]
        then:
            var CheckValue:int = CustomPlayer.ButtonL15Pressed  
            set CustomPlayer.FollowActiveL15=1
            CustomPlayer.Update_Pets_Section()
            loop:
                if(CheckValue <> CustomPlayer.ButtonL15Pressed) {break}
                else if(CustomPlayer.PlayerInTreasureAreaFlag=0):
                    #AttackCinema.Stop()
                    PetRotateToPlayer(FortniteCharacter,CreativeProp)
                    Move_To_Player(FortniteCharacter,CreativeProp,Xplayer,Yplayer)
                else if(CustomPlayer.PlayerInTreasureAreaFlag=1):
                    #-----------------Find closest chest
                    var ClosestChest:creative_prop = creative_prop{}
                    var DistanceToMe:float = 100000000.0
                    var FoundChest:logic = false
                    var ChestIndexTemp:int = -1
                    var ChestIndexFinal:int = 0
                    var ClosestChestIndex:int=0
                    Allthechests:= ChestsL1+ChestsL2+ChestsL3+ChestsL4+ChestsL5+ChestsL6+ChestsL7+ChestsL8+ChestsL9+ChestsL10+ChestsL11+ChestsL12+ChestsL13+ChestsL14+ChestsL15+ChestsL16+ChestsL17+ChestsL18+ChestsL19+ChestsL20+ChestsL21+ChestsL22+ChestsL23+ChestsL24+ChestsL25+ChestsL26+ChestsL27+ChestsL28+ChestsL29+ChestsL30
                    for(Chestt:Allthechests):
                        set ChestIndexTemp+=1
                        MyPos:vector3 = FortniteCharacter.GetTransform().Translation
                        ChestPos:vector3 = Chestt.ChestProp.GetTransform().Translation
                        TotalDistance:float = Distance(MyPos, ChestPos) / 100.0
                        if(TotalDistance < DistanceToMe and Chestt.CurrentHP>0.0):
                            set ChestIndexFinal = ChestIndexTemp
                            #Print("Chest index{ChestIndexFinal}")
                            set FoundChest = true
                            set DistanceToMe = TotalDistance
                            set ClosestChest = Chestt.ChestProp
            
                    if(FoundChest = true):
                    #-----------------Rotate and move to chest+ rerotate because position is random
                    PetRotateToChest(ClosestChest,CreativeProp)
                    Move_To_Chest(FortniteCharacter,ClosestChest,CreativeProp)
                    PetRotateToChest(ClosestChest,CreativeProp)
                    #-----------------Attack Chest
                    if(Cine:=Pets[PetIndex].AttackAnim):
                        Cine.Play(Agent)
                        Sleep(1.5) #Attack duration
                        Cine.Stop(Agent)
                    #----------------Crit
                    var PowerCrit:float=0.0
                    var CritRand:int = GetRandomInt(0,100)
                    if (CritRand>CustomPlayer.Pet_Crit_Value):
                        set PowerCrit=Power*CustomPlayer.BonusPowerMultiplicator*CustomPlayer.Bonus_Pet_Power
                    else:
                        set PowerCrit=Power*1.5*CustomPlayer.BonusPowerMultiplicator*CustomPlayer.Bonus_Pet_Power
                    #-------------------
                    Attack_Chest(ChestIndexFinal,PowerCrit)
                    Check_Chest_Gold(ChestIndexFinal,FortniteCharacter)

            ReleasePet(CreativeProp)

    FollowPlayerL16(FortniteCharacter : fort_character, CreativeProp : creative_prop, Power:float, PetIndex:int)<suspends>: void =
        #Print("Follow Player")
        #Future position of the pet
        Xplayer:float=-400.0
        Yplayer:float= 0.0
        if:
            Agent := FortniteCharacter.GetAgent[]
            PlayerObj := player[Agent]
            CustomPlayer:Custom_Player = PlayersMap[PlayerObj]
        then:
            var CheckValue:int = CustomPlayer.ButtonL16Pressed  
            set CustomPlayer.FollowActiveL16=1
            CustomPlayer.Update_Pets_Section()
            loop:
                if(CheckValue <> CustomPlayer.ButtonL16Pressed) {break}
                else if(CustomPlayer.PlayerInTreasureAreaFlag=0):
                    #AttackCinema.Stop()
                    PetRotateToPlayer(FortniteCharacter,CreativeProp)
                    Move_To_Player(FortniteCharacter,CreativeProp,Xplayer,Yplayer)
                else if(CustomPlayer.PlayerInTreasureAreaFlag=1):
                    #-----------------Find closest chest
                    var ClosestChest:creative_prop = creative_prop{}
                    var DistanceToMe:float = 100000000.0
                    var FoundChest:logic = false
                    var ChestIndexTemp:int = -1
                    var ChestIndexFinal:int = 0
                    var ClosestChestIndex:int=0
                    Allthechests:= ChestsL1+ChestsL2+ChestsL3+ChestsL4+ChestsL5+ChestsL6+ChestsL7+ChestsL8+ChestsL9+ChestsL10+ChestsL11+ChestsL12+ChestsL13+ChestsL14+ChestsL15+ChestsL16+ChestsL17+ChestsL18+ChestsL19+ChestsL20+ChestsL21+ChestsL22+ChestsL23+ChestsL24+ChestsL25+ChestsL26+ChestsL27+ChestsL28+ChestsL29+ChestsL30
                    for(Chestt:Allthechests):
                        set ChestIndexTemp+=1
                        MyPos:vector3 = FortniteCharacter.GetTransform().Translation
                        ChestPos:vector3 = Chestt.ChestProp.GetTransform().Translation
                        TotalDistance:float = Distance(MyPos, ChestPos) / 100.0
                        if(TotalDistance < DistanceToMe and Chestt.CurrentHP>0.0):
                            set ChestIndexFinal = ChestIndexTemp
                            #Print("Chest index{ChestIndexFinal}")
                            set FoundChest = true
                            set DistanceToMe = TotalDistance
                            set ClosestChest = Chestt.ChestProp
            
                    if(FoundChest = true):
                    #-----------------Rotate and move to chest+ rerotate because position is random
                    PetRotateToChest(ClosestChest,CreativeProp)
                    Move_To_Chest(FortniteCharacter,ClosestChest,CreativeProp)
                    PetRotateToChest(ClosestChest,CreativeProp)
                    #-----------------Attack Chest
                    if(Cine:=Pets[PetIndex].AttackAnim):
                        Cine.Play(Agent)
                        Sleep(1.5) #Attack duration
                        Cine.Stop(Agent)
                    #----------------Crit
                    var PowerCrit:float=0.0
                    var CritRand:int = GetRandomInt(0,100)
                    if (CritRand>CustomPlayer.Pet_Crit_Value):
                        set PowerCrit=Power*CustomPlayer.BonusPowerMultiplicator*CustomPlayer.Bonus_Pet_Power
                    else:
                        set PowerCrit=Power*1.5*CustomPlayer.BonusPowerMultiplicator*CustomPlayer.Bonus_Pet_Power
                    #-------------------
                    Attack_Chest(ChestIndexFinal,PowerCrit)
                    Check_Chest_Gold(ChestIndexFinal,FortniteCharacter)

            ReleasePet(CreativeProp)

    FollowPlayerL17(FortniteCharacter : fort_character, CreativeProp : creative_prop, Power:float, PetIndex:int)<suspends>: void =
        #Print("Follow Player")
        #Future position of the pet
        Xplayer:float=-400.0
        Yplayer:float=-100.0
        if:
            Agent := FortniteCharacter.GetAgent[]
            PlayerObj := player[Agent]
            CustomPlayer:Custom_Player = PlayersMap[PlayerObj]
        then:
            var CheckValue:int = CustomPlayer.ButtonL17Pressed  
            set CustomPlayer.FollowActiveL17=1
            CustomPlayer.Update_Pets_Section()
            loop:
                if(CheckValue <> CustomPlayer.ButtonL17Pressed) {break}
                else if(CustomPlayer.PlayerInTreasureAreaFlag=0):
                    #AttackCinema.Stop()
                    PetRotateToPlayer(FortniteCharacter,CreativeProp)
                    Move_To_Player(FortniteCharacter,CreativeProp,Xplayer,Yplayer)
                else if(CustomPlayer.PlayerInTreasureAreaFlag=1):
                    #-----------------Find closest chest
                    var ClosestChest:creative_prop = creative_prop{}
                    var DistanceToMe:float = 100000000.0
                    var FoundChest:logic = false
                    var ChestIndexTemp:int = -1
                    var ChestIndexFinal:int = 0
                    var ClosestChestIndex:int=0
                    Allthechests:= ChestsL1+ChestsL2+ChestsL3+ChestsL4+ChestsL5+ChestsL6+ChestsL7+ChestsL8+ChestsL9+ChestsL10+ChestsL11+ChestsL12+ChestsL13+ChestsL14+ChestsL15+ChestsL16+ChestsL17+ChestsL18+ChestsL19+ChestsL20+ChestsL21+ChestsL22+ChestsL23+ChestsL24+ChestsL25+ChestsL26+ChestsL27+ChestsL28+ChestsL29+ChestsL30
                    for(Chestt:Allthechests):
                        set ChestIndexTemp+=1
                        MyPos:vector3 = FortniteCharacter.GetTransform().Translation
                        ChestPos:vector3 = Chestt.ChestProp.GetTransform().Translation
                        TotalDistance:float = Distance(MyPos, ChestPos) / 100.0
                        if(TotalDistance < DistanceToMe and Chestt.CurrentHP>0.0):
                            set ChestIndexFinal = ChestIndexTemp
                            #Print("Chest index{ChestIndexFinal}")
                            set FoundChest = true
                            set DistanceToMe = TotalDistance
                            set ClosestChest = Chestt.ChestProp
            
                    if(FoundChest = true):
                    #-----------------Rotate and move to chest+ rerotate because position is random
                    PetRotateToChest(ClosestChest,CreativeProp)
                    Move_To_Chest(FortniteCharacter,ClosestChest,CreativeProp)
                    PetRotateToChest(ClosestChest,CreativeProp)
                    #-----------------Attack Chest
                    if(Cine:=Pets[PetIndex].AttackAnim):
                        Cine.Play(Agent)
                        Sleep(1.5) #Attack duration
                        Cine.Stop(Agent)
                    #----------------Crit
                    var PowerCrit:float=0.0
                    var CritRand:int = GetRandomInt(0,100)
                    if (CritRand>CustomPlayer.Pet_Crit_Value):
                        set PowerCrit=Power*CustomPlayer.BonusPowerMultiplicator*CustomPlayer.Bonus_Pet_Power
                    else:
                        set PowerCrit=Power*1.5*CustomPlayer.BonusPowerMultiplicator*CustomPlayer.Bonus_Pet_Power
                    #-------------------
                    Attack_Chest(ChestIndexFinal,PowerCrit)
                    Check_Chest_Gold(ChestIndexFinal,FortniteCharacter)

            ReleasePet(CreativeProp)

    FollowPlayerL18(FortniteCharacter : fort_character, CreativeProp : creative_prop, Power:float, PetIndex:int)<suspends>: void =
        #Print("Follow Player")
        #Future position of the pet
        Xplayer:float=-400.0
        Yplayer:float= 100.0
        if:
            Agent := FortniteCharacter.GetAgent[]
            PlayerObj := player[Agent]
            CustomPlayer:Custom_Player = PlayersMap[PlayerObj]
        then:
            var CheckValue:int = CustomPlayer.ButtonL18Pressed  
            set CustomPlayer.FollowActiveL18=1
            CustomPlayer.Update_Pets_Section()
            loop:
                if(CheckValue <> CustomPlayer.ButtonL18Pressed) {break}
                else if(CustomPlayer.PlayerInTreasureAreaFlag=0):
                    #AttackCinema.Stop()
                    PetRotateToPlayer(FortniteCharacter,CreativeProp)
                    Move_To_Player(FortniteCharacter,CreativeProp,Xplayer,Yplayer)
                else if(CustomPlayer.PlayerInTreasureAreaFlag=1):
                    #-----------------Find closest chest
                    var ClosestChest:creative_prop = creative_prop{}
                    var DistanceToMe:float = 100000000.0
                    var FoundChest:logic = false
                    var ChestIndexTemp:int = -1
                    var ChestIndexFinal:int = 0
                    var ClosestChestIndex:int=0
                    Allthechests:= ChestsL1+ChestsL2+ChestsL3+ChestsL4+ChestsL5+ChestsL6+ChestsL7+ChestsL8+ChestsL9+ChestsL10+ChestsL11+ChestsL12+ChestsL13+ChestsL14+ChestsL15+ChestsL16+ChestsL17+ChestsL18+ChestsL19+ChestsL20+ChestsL21+ChestsL22+ChestsL23+ChestsL24+ChestsL25+ChestsL26+ChestsL27+ChestsL28+ChestsL29+ChestsL30
                    for(Chestt:Allthechests):
                        set ChestIndexTemp+=1
                        MyPos:vector3 = FortniteCharacter.GetTransform().Translation
                        ChestPos:vector3 = Chestt.ChestProp.GetTransform().Translation
                        TotalDistance:float = Distance(MyPos, ChestPos) / 100.0
                        if(TotalDistance < DistanceToMe and Chestt.CurrentHP>0.0):
                            set ChestIndexFinal = ChestIndexTemp
                            #Print("Chest index{ChestIndexFinal}")
                            set FoundChest = true
                            set DistanceToMe = TotalDistance
                            set ClosestChest = Chestt.ChestProp
            
                    if(FoundChest = true):
                    #-----------------Rotate and move to chest+ rerotate because position is random
                    PetRotateToChest(ClosestChest,CreativeProp)
                    Move_To_Chest(FortniteCharacter,ClosestChest,CreativeProp)
                    PetRotateToChest(ClosestChest,CreativeProp)
                    #-----------------Attack Chest
                    if(Cine:=Pets[PetIndex].AttackAnim):
                        Cine.Play(Agent)
                        Sleep(1.5) #Attack duration
                        Cine.Stop(Agent)
                    #----------------Crit
                    var PowerCrit:float=0.0
                    var CritRand:int = GetRandomInt(0,100)
                    if (CritRand>CustomPlayer.Pet_Crit_Value):
                        set PowerCrit=Power*CustomPlayer.BonusPowerMultiplicator*CustomPlayer.Bonus_Pet_Power
                    else:
                        set PowerCrit=Power*1.5*CustomPlayer.BonusPowerMultiplicator*CustomPlayer.Bonus_Pet_Power
                    #-------------------
                    Attack_Chest(ChestIndexFinal,PowerCrit)
                    Check_Chest_Gold(ChestIndexFinal,FortniteCharacter)

            ReleasePet(CreativeProp)

    FollowPlayerL19(FortniteCharacter : fort_character, CreativeProp : creative_prop, Power:float, PetIndex:int)<suspends>: void =
        #Print("Follow Player")
        #Future position of the pet
        Xplayer:float=-400.0
        Yplayer:float=-200.0
        if:
            Agent := FortniteCharacter.GetAgent[]
            PlayerObj := player[Agent]
            CustomPlayer:Custom_Player = PlayersMap[PlayerObj]
        then:
            var CheckValue:int = CustomPlayer.ButtonL19Pressed  
            set CustomPlayer.FollowActiveL19=1
            CustomPlayer.Update_Pets_Section()
            loop:
                if(CheckValue <> CustomPlayer.ButtonL19Pressed) {break}
                else if(CustomPlayer.PlayerInTreasureAreaFlag=0):
                    #AttackCinema.Stop()
                    PetRotateToPlayer(FortniteCharacter,CreativeProp)
                    Move_To_Player(FortniteCharacter,CreativeProp,Xplayer,Yplayer)
                else if(CustomPlayer.PlayerInTreasureAreaFlag=1):
                    #-----------------Find closest chest
                    var ClosestChest:creative_prop = creative_prop{}
                    var DistanceToMe:float = 100000000.0
                    var FoundChest:logic = false
                    var ChestIndexTemp:int = -1
                    var ChestIndexFinal:int = 0
                    var ClosestChestIndex:int=0
                    Allthechests:= ChestsL1+ChestsL2+ChestsL3+ChestsL4+ChestsL5+ChestsL6+ChestsL7+ChestsL8+ChestsL9+ChestsL10+ChestsL11+ChestsL12+ChestsL13+ChestsL14+ChestsL15+ChestsL16+ChestsL17+ChestsL18+ChestsL19+ChestsL20+ChestsL21+ChestsL22+ChestsL23+ChestsL24+ChestsL25+ChestsL26+ChestsL27+ChestsL28+ChestsL29+ChestsL30
                    for(Chestt:Allthechests):
                        set ChestIndexTemp+=1
                        MyPos:vector3 = FortniteCharacter.GetTransform().Translation
                        ChestPos:vector3 = Chestt.ChestProp.GetTransform().Translation
                        TotalDistance:float = Distance(MyPos, ChestPos) / 100.0
                        if(TotalDistance < DistanceToMe and Chestt.CurrentHP>0.0):
                            set ChestIndexFinal = ChestIndexTemp
                            #Print("Chest index{ChestIndexFinal}")
                            set FoundChest = true
                            set DistanceToMe = TotalDistance
                            set ClosestChest = Chestt.ChestProp
            
                    if(FoundChest = true):
                    #-----------------Rotate and move to chest+ rerotate because position is random
                    PetRotateToChest(ClosestChest,CreativeProp)
                    Move_To_Chest(FortniteCharacter,ClosestChest,CreativeProp)
                    PetRotateToChest(ClosestChest,CreativeProp)
                    #-----------------Attack Chest
                    if(Cine:=Pets[PetIndex].AttackAnim):
                        Cine.Play(Agent)
                        Sleep(1.5) #Attack duration
                        Cine.Stop(Agent)
                    #----------------Crit
                    var PowerCrit:float=0.0
                    var CritRand:int = GetRandomInt(0,100)
                    if (CritRand>CustomPlayer.Pet_Crit_Value):
                        set PowerCrit=Power*CustomPlayer.BonusPowerMultiplicator*CustomPlayer.Bonus_Pet_Power
                    else:
                        set PowerCrit=Power*1.5*CustomPlayer.BonusPowerMultiplicator*CustomPlayer.Bonus_Pet_Power
                    #-------------------
                    Attack_Chest(ChestIndexFinal,PowerCrit)
                    Check_Chest_Gold(ChestIndexFinal,FortniteCharacter)

            ReleasePet(CreativeProp)

    FollowPlayerL20(FortniteCharacter : fort_character, CreativeProp : creative_prop, Power:float, PetIndex:int)<suspends>: void =
        #Print("Follow Player")
        #Future position of the pet
        Xplayer:float=-400.0
        Yplayer:float= 200.0
        if:
            Agent := FortniteCharacter.GetAgent[]
            PlayerObj := player[Agent]
            CustomPlayer:Custom_Player = PlayersMap[PlayerObj]
        then:
            var CheckValue:int = CustomPlayer.ButtonL20Pressed  
            set CustomPlayer.FollowActiveL20=1
            CustomPlayer.Update_Pets_Section()
            loop:
                if(CheckValue <> CustomPlayer.ButtonL20Pressed) {break}
                else if(CustomPlayer.PlayerInTreasureAreaFlag=0):
                    #AttackCinema.Stop()
                    PetRotateToPlayer(FortniteCharacter,CreativeProp)
                    Move_To_Player(FortniteCharacter,CreativeProp,Xplayer,Yplayer)
                else if(CustomPlayer.PlayerInTreasureAreaFlag=1):
                    #-----------------Find closest chest
                    var ClosestChest:creative_prop = creative_prop{}
                    var DistanceToMe:float = 100000000.0
                    var FoundChest:logic = false
                    var ChestIndexTemp:int = -1
                    var ChestIndexFinal:int = 0
                    var ClosestChestIndex:int=0
                    Allthechests:= ChestsL1+ChestsL2+ChestsL3+ChestsL4+ChestsL5+ChestsL6+ChestsL7+ChestsL8+ChestsL9+ChestsL10+ChestsL11+ChestsL12+ChestsL13+ChestsL14+ChestsL15+ChestsL16+ChestsL17+ChestsL18+ChestsL19+ChestsL20+ChestsL21+ChestsL22+ChestsL23+ChestsL24+ChestsL25+ChestsL26+ChestsL27+ChestsL28+ChestsL29+ChestsL30
                    for(Chestt:Allthechests):
                        set ChestIndexTemp+=1
                        MyPos:vector3 = FortniteCharacter.GetTransform().Translation
                        ChestPos:vector3 = Chestt.ChestProp.GetTransform().Translation
                        TotalDistance:float = Distance(MyPos, ChestPos) / 100.0
                        if(TotalDistance < DistanceToMe and Chestt.CurrentHP>0.0):
                            set ChestIndexFinal = ChestIndexTemp
                            #Print("Chest index{ChestIndexFinal}")
                            set FoundChest = true
                            set DistanceToMe = TotalDistance
                            set ClosestChest = Chestt.ChestProp
            
                    if(FoundChest = true):
                    #-----------------Rotate and move to chest+ rerotate because position is random
                    PetRotateToChest(ClosestChest,CreativeProp)
                    Move_To_Chest(FortniteCharacter,ClosestChest,CreativeProp)
                    PetRotateToChest(ClosestChest,CreativeProp)
                    #-----------------Attack Chest
                    if(Cine:=Pets[PetIndex].AttackAnim):
                        Cine.Play(Agent)
                        Sleep(1.5) #Attack duration
                        Cine.Stop(Agent)
                    #----------------Crit
                    var PowerCrit:float=0.0
                    var CritRand:int = GetRandomInt(0,100)
                    if (CritRand>CustomPlayer.Pet_Crit_Value):
                        set PowerCrit=Power*CustomPlayer.BonusPowerMultiplicator*CustomPlayer.Bonus_Pet_Power
                    else:
                        set PowerCrit=Power*1.5*CustomPlayer.BonusPowerMultiplicator*CustomPlayer.Bonus_Pet_Power
                    #-------------------
                    Attack_Chest(ChestIndexFinal,PowerCrit)
                    Check_Chest_Gold(ChestIndexFinal,FortniteCharacter)

            ReleasePet(CreativeProp)

    FollowPlayerL21(FortniteCharacter : fort_character, CreativeProp : creative_prop, Power:float, PetIndex:int)<suspends>: void =
        #Print("Follow Player")
        #Future position of the pet
        Xplayer:float=-400.0
        Yplayer:float=-300.0
        if:
            Agent := FortniteCharacter.GetAgent[]
            PlayerObj := player[Agent]
            CustomPlayer:Custom_Player = PlayersMap[PlayerObj]
        then:
            var CheckValue:int = CustomPlayer.ButtonL21Pressed  
            set CustomPlayer.FollowActiveL21=1
            CustomPlayer.Update_Pets_Section()
            loop:
                if(CheckValue <> CustomPlayer.ButtonL21Pressed) {break}
                else if(CustomPlayer.PlayerInTreasureAreaFlag=0):
                    #AttackCinema.Stop()
                    PetRotateToPlayer(FortniteCharacter,CreativeProp)
                    Move_To_Player(FortniteCharacter,CreativeProp,Xplayer,Yplayer)
                else if(CustomPlayer.PlayerInTreasureAreaFlag=1):
                    #-----------------Find closest chest
                    var ClosestChest:creative_prop = creative_prop{}
                    var DistanceToMe:float = 100000000.0
                    var FoundChest:logic = false
                    var ChestIndexTemp:int = -1
                    var ChestIndexFinal:int = 0
                    var ClosestChestIndex:int=0
                    Allthechests:= ChestsL1+ChestsL2+ChestsL3+ChestsL4+ChestsL5+ChestsL6+ChestsL7+ChestsL8+ChestsL9+ChestsL10+ChestsL11+ChestsL12+ChestsL13+ChestsL14+ChestsL15+ChestsL16+ChestsL17+ChestsL18+ChestsL19+ChestsL20+ChestsL21+ChestsL22+ChestsL23+ChestsL24+ChestsL25+ChestsL26+ChestsL27+ChestsL28+ChestsL29+ChestsL30
                    for(Chestt:Allthechests):
                        set ChestIndexTemp+=1
                        MyPos:vector3 = FortniteCharacter.GetTransform().Translation
                        ChestPos:vector3 = Chestt.ChestProp.GetTransform().Translation
                        TotalDistance:float = Distance(MyPos, ChestPos) / 100.0
                        if(TotalDistance < DistanceToMe and Chestt.CurrentHP>0.0):
                            set ChestIndexFinal = ChestIndexTemp
                            #Print("Chest index{ChestIndexFinal}")
                            set FoundChest = true
                            set DistanceToMe = TotalDistance
                            set ClosestChest = Chestt.ChestProp
            
                    if(FoundChest = true):
                    #-----------------Rotate and move to chest+ rerotate because position is random
                    PetRotateToChest(ClosestChest,CreativeProp)
                    Move_To_Chest(FortniteCharacter,ClosestChest,CreativeProp)
                    PetRotateToChest(ClosestChest,CreativeProp)
                    #-----------------Attack Chest
                    if(Cine:=Pets[PetIndex].AttackAnim):
                        Cine.Play(Agent)
                        Sleep(1.5) #Attack duration
                        Cine.Stop(Agent)
                    #----------------Crit
                    var PowerCrit:float=0.0
                    var CritRand:int = GetRandomInt(0,100)
                    if (CritRand>CustomPlayer.Pet_Crit_Value):
                        set PowerCrit=Power*CustomPlayer.BonusPowerMultiplicator*CustomPlayer.Bonus_Pet_Power
                    else:
                        set PowerCrit=Power*1.5*CustomPlayer.BonusPowerMultiplicator*CustomPlayer.Bonus_Pet_Power
                    #-------------------
                    Attack_Chest(ChestIndexFinal,PowerCrit)
                    Check_Chest_Gold(ChestIndexFinal,FortniteCharacter)

            ReleasePet(CreativeProp)

    FollowPlayerL22(FortniteCharacter : fort_character, CreativeProp : creative_prop, Power:float, PetIndex:int)<suspends>: void =
        #Print("Follow Player")
        #Future position of the pet
        Xplayer:float=-400.0
        Yplayer:float= 300.0
        if:
            Agent := FortniteCharacter.GetAgent[]
            PlayerObj := player[Agent]
            CustomPlayer:Custom_Player = PlayersMap[PlayerObj]
        then:
            var CheckValue:int = CustomPlayer.ButtonL22Pressed  
            set CustomPlayer.FollowActiveL22=1
            CustomPlayer.Update_Pets_Section()
            loop:
                if(CheckValue <> CustomPlayer.ButtonL22Pressed) {break}
                else if(CustomPlayer.PlayerInTreasureAreaFlag=0):
                    #AttackCinema.Stop()
                    PetRotateToPlayer(FortniteCharacter,CreativeProp)
                    Move_To_Player(FortniteCharacter,CreativeProp,Xplayer,Yplayer)
                else if(CustomPlayer.PlayerInTreasureAreaFlag=1):
                    #-----------------Find closest chest
                    var ClosestChest:creative_prop = creative_prop{}
                    var DistanceToMe:float = 100000000.0
                    var FoundChest:logic = false
                    var ChestIndexTemp:int = -1
                    var ChestIndexFinal:int = 0
                    var ClosestChestIndex:int=0
                    Allthechests:= ChestsL1+ChestsL2+ChestsL3+ChestsL4+ChestsL5+ChestsL6+ChestsL7+ChestsL8+ChestsL9+ChestsL10+ChestsL11+ChestsL12+ChestsL13+ChestsL14+ChestsL15+ChestsL16+ChestsL17+ChestsL18+ChestsL19+ChestsL20+ChestsL21+ChestsL22+ChestsL23+ChestsL24+ChestsL25+ChestsL26+ChestsL27+ChestsL28+ChestsL29+ChestsL30
                    for(Chestt:Allthechests):
                        set ChestIndexTemp+=1
                        MyPos:vector3 = FortniteCharacter.GetTransform().Translation
                        ChestPos:vector3 = Chestt.ChestProp.GetTransform().Translation
                        TotalDistance:float = Distance(MyPos, ChestPos) / 100.0
                        if(TotalDistance < DistanceToMe and Chestt.CurrentHP>0.0):
                            set ChestIndexFinal = ChestIndexTemp
                            #Print("Chest index{ChestIndexFinal}")
                            set FoundChest = true
                            set DistanceToMe = TotalDistance
                            set ClosestChest = Chestt.ChestProp
            
                    if(FoundChest = true):
                    #-----------------Rotate and move to chest+ rerotate because position is random
                    PetRotateToChest(ClosestChest,CreativeProp)
                    Move_To_Chest(FortniteCharacter,ClosestChest,CreativeProp)
                    PetRotateToChest(ClosestChest,CreativeProp)
                    #-----------------Attack Chest
                    if(Cine:=Pets[PetIndex].AttackAnim):
                        Cine.Play(Agent)
                        Sleep(1.5) #Attack duration
                        Cine.Stop(Agent)
                    #----------------Crit
                    var PowerCrit:float=0.0
                    var CritRand:int = GetRandomInt(0,100)
                    if (CritRand>CustomPlayer.Pet_Crit_Value):
                        set PowerCrit=Power*CustomPlayer.BonusPowerMultiplicator*CustomPlayer.Bonus_Pet_Power
                    else:
                        set PowerCrit=Power*1.5*CustomPlayer.BonusPowerMultiplicator*CustomPlayer.Bonus_Pet_Power
                    #-------------------
                    Attack_Chest(ChestIndexFinal,PowerCrit)
                    Check_Chest_Gold(ChestIndexFinal,FortniteCharacter)

            ReleasePet(CreativeProp)

    FollowPlayerL23(FortniteCharacter : fort_character, CreativeProp : creative_prop, Power:float, PetIndex:int)<suspends>: void =
        #Print("Follow Player")
        #Future position of the pet
        Xplayer:float=-500.0
        Yplayer:float= 0.0
        if:
            Agent := FortniteCharacter.GetAgent[]
            PlayerObj := player[Agent]
            CustomPlayer:Custom_Player = PlayersMap[PlayerObj]
        then:
            var CheckValue:int = CustomPlayer.ButtonL23Pressed  
            set CustomPlayer.FollowActiveL23=1
            CustomPlayer.Update_Pets_Section()
            loop:
                if(CheckValue <> CustomPlayer.ButtonL23Pressed) {break}
                else if(CustomPlayer.PlayerInTreasureAreaFlag=0):
                    #AttackCinema.Stop()
                    PetRotateToPlayer(FortniteCharacter,CreativeProp)
                    Move_To_Player(FortniteCharacter,CreativeProp,Xplayer,Yplayer)
                else if(CustomPlayer.PlayerInTreasureAreaFlag=1):
                    #-----------------Find closest chest
                    var ClosestChest:creative_prop = creative_prop{}
                    var DistanceToMe:float = 100000000.0
                    var FoundChest:logic = false
                    var ChestIndexTemp:int = -1
                    var ChestIndexFinal:int = 0
                    var ClosestChestIndex:int=0
                    Allthechests:= ChestsL1+ChestsL2+ChestsL3+ChestsL4+ChestsL5+ChestsL6+ChestsL7+ChestsL8+ChestsL9+ChestsL10+ChestsL11+ChestsL12+ChestsL13+ChestsL14+ChestsL15+ChestsL16+ChestsL17+ChestsL18+ChestsL19+ChestsL20+ChestsL21+ChestsL22+ChestsL23+ChestsL24+ChestsL25+ChestsL26+ChestsL27+ChestsL28+ChestsL29+ChestsL30
                    for(Chestt:Allthechests):
                        set ChestIndexTemp+=1
                        MyPos:vector3 = FortniteCharacter.GetTransform().Translation
                        ChestPos:vector3 = Chestt.ChestProp.GetTransform().Translation
                        TotalDistance:float = Distance(MyPos, ChestPos) / 100.0
                        if(TotalDistance < DistanceToMe and Chestt.CurrentHP>0.0):
                            set ChestIndexFinal = ChestIndexTemp
                            #Print("Chest index{ChestIndexFinal}")
                            set FoundChest = true
                            set DistanceToMe = TotalDistance
                            set ClosestChest = Chestt.ChestProp
            
                    if(FoundChest = true):
                    #-----------------Rotate and move to chest+ rerotate because position is random
                    PetRotateToChest(ClosestChest,CreativeProp)
                    Move_To_Chest(FortniteCharacter,ClosestChest,CreativeProp)
                    PetRotateToChest(ClosestChest,CreativeProp)
                    #-----------------Attack Chest
                    if(Cine:=Pets[PetIndex].AttackAnim):
                        Cine.Play(Agent)
                        Sleep(1.5) #Attack duration
                        Cine.Stop(Agent)
                    #----------------Crit
                    var PowerCrit:float=0.0
                    var CritRand:int = GetRandomInt(0,100)
                    if (CritRand>CustomPlayer.Pet_Crit_Value):
                        set PowerCrit=Power*CustomPlayer.BonusPowerMultiplicator*CustomPlayer.Bonus_Pet_Power
                    else:
                        set PowerCrit=Power*1.5*CustomPlayer.BonusPowerMultiplicator*CustomPlayer.Bonus_Pet_Power
                    #-------------------
                    Attack_Chest(ChestIndexFinal,PowerCrit)
                    Check_Chest_Gold(ChestIndexFinal,FortniteCharacter)

            ReleasePet(CreativeProp)

    FollowPlayerL24(FortniteCharacter : fort_character, CreativeProp : creative_prop, Power:float, PetIndex:int)<suspends>: void =
        #Print("Follow Player")
        #Future position of the pet
        Xplayer:float=-500.0
        Yplayer:float=-100.0
        if:
            Agent := FortniteCharacter.GetAgent[]
            PlayerObj := player[Agent]
            CustomPlayer:Custom_Player = PlayersMap[PlayerObj]
        then:
            var CheckValue:int = CustomPlayer.ButtonL24Pressed  
            set CustomPlayer.FollowActiveL24=1
            CustomPlayer.Update_Pets_Section()
            loop:
                if(CheckValue <> CustomPlayer.ButtonL24Pressed) {break}
                else if(CustomPlayer.PlayerInTreasureAreaFlag=0):
                    #AttackCinema.Stop()
                    PetRotateToPlayer(FortniteCharacter,CreativeProp)
                    Move_To_Player(FortniteCharacter,CreativeProp,Xplayer,Yplayer)
                else if(CustomPlayer.PlayerInTreasureAreaFlag=1):
                    #-----------------Find closest chest
                    var ClosestChest:creative_prop = creative_prop{}
                    var DistanceToMe:float = 100000000.0
                    var FoundChest:logic = false
                    var ChestIndexTemp:int = -1
                    var ChestIndexFinal:int = 0
                    var ClosestChestIndex:int=0
                    Allthechests:= ChestsL1+ChestsL2+ChestsL3+ChestsL4+ChestsL5+ChestsL6+ChestsL7+ChestsL8+ChestsL9+ChestsL10+ChestsL11+ChestsL12+ChestsL13+ChestsL14+ChestsL15+ChestsL16+ChestsL17+ChestsL18+ChestsL19+ChestsL20+ChestsL21+ChestsL22+ChestsL23+ChestsL24+ChestsL25+ChestsL26+ChestsL27+ChestsL28+ChestsL29+ChestsL30
                    for(Chestt:Allthechests):
                        set ChestIndexTemp+=1
                        MyPos:vector3 = FortniteCharacter.GetTransform().Translation
                        ChestPos:vector3 = Chestt.ChestProp.GetTransform().Translation
                        TotalDistance:float = Distance(MyPos, ChestPos) / 100.0
                        if(TotalDistance < DistanceToMe and Chestt.CurrentHP>0.0):
                            set ChestIndexFinal = ChestIndexTemp
                            #Print("Chest index{ChestIndexFinal}")
                            set FoundChest = true
                            set DistanceToMe = TotalDistance
                            set ClosestChest = Chestt.ChestProp
            
                    if(FoundChest = true):
                    #-----------------Rotate and move to chest+ rerotate because position is random
                    PetRotateToChest(ClosestChest,CreativeProp)
                    Move_To_Chest(FortniteCharacter,ClosestChest,CreativeProp)
                    PetRotateToChest(ClosestChest,CreativeProp)
                    #-----------------Attack Chest
                    if(Cine:=Pets[PetIndex].AttackAnim):
                        Cine.Play(Agent)
                        Sleep(1.5) #Attack duration
                        Cine.Stop(Agent)
                    #----------------Crit
                    var PowerCrit:float=0.0
                    var CritRand:int = GetRandomInt(0,100)
                    if (CritRand>CustomPlayer.Pet_Crit_Value):
                        set PowerCrit=Power*CustomPlayer.BonusPowerMultiplicator*CustomPlayer.Bonus_Pet_Power
                    else:
                        set PowerCrit=Power*1.5*CustomPlayer.BonusPowerMultiplicator*CustomPlayer.Bonus_Pet_Power
                    #-------------------
                    Attack_Chest(ChestIndexFinal,PowerCrit)
                    Check_Chest_Gold(ChestIndexFinal,FortniteCharacter)

            ReleasePet(CreativeProp)

    FollowPlayerL25(FortniteCharacter : fort_character, CreativeProp : creative_prop, Power:float, PetIndex:int)<suspends>: void =
        #Print("Follow Player")
        #Future position of the pet
        Xplayer:float=-500.0
        Yplayer:float= 100.0
        if:
            Agent := FortniteCharacter.GetAgent[]
            PlayerObj := player[Agent]
            CustomPlayer:Custom_Player = PlayersMap[PlayerObj]
        then:
            var CheckValue:int = CustomPlayer.ButtonL25Pressed  
            set CustomPlayer.FollowActiveL25=1
            CustomPlayer.Update_Pets_Section()
            loop:
                if(CheckValue <> CustomPlayer.ButtonL25Pressed) {break}
                else if(CustomPlayer.PlayerInTreasureAreaFlag=0):
                    #AttackCinema.Stop()
                    PetRotateToPlayer(FortniteCharacter,CreativeProp)
                    Move_To_Player(FortniteCharacter,CreativeProp,Xplayer,Yplayer)
                else if(CustomPlayer.PlayerInTreasureAreaFlag=1):
                    #-----------------Find closest chest
                    var ClosestChest:creative_prop = creative_prop{}
                    var DistanceToMe:float = 100000000.0
                    var FoundChest:logic = false
                    var ChestIndexTemp:int = -1
                    var ChestIndexFinal:int = 0
                    var ClosestChestIndex:int=0
                    Allthechests:= ChestsL1+ChestsL2+ChestsL3+ChestsL4+ChestsL5+ChestsL6+ChestsL7+ChestsL8+ChestsL9+ChestsL10+ChestsL11+ChestsL12+ChestsL13+ChestsL14+ChestsL15+ChestsL16+ChestsL17+ChestsL18+ChestsL19+ChestsL20+ChestsL21+ChestsL22+ChestsL23+ChestsL24+ChestsL25+ChestsL26+ChestsL27+ChestsL28+ChestsL29+ChestsL30
                    for(Chestt:Allthechests):
                        set ChestIndexTemp+=1
                        MyPos:vector3 = FortniteCharacter.GetTransform().Translation
                        ChestPos:vector3 = Chestt.ChestProp.GetTransform().Translation
                        TotalDistance:float = Distance(MyPos, ChestPos) / 100.0
                        if(TotalDistance < DistanceToMe and Chestt.CurrentHP>0.0):
                            set ChestIndexFinal = ChestIndexTemp
                            #Print("Chest index{ChestIndexFinal}")
                            set FoundChest = true
                            set DistanceToMe = TotalDistance
                            set ClosestChest = Chestt.ChestProp
            
                    if(FoundChest = true):
                    #-----------------Rotate and move to chest+ rerotate because position is random
                    PetRotateToChest(ClosestChest,CreativeProp)
                    Move_To_Chest(FortniteCharacter,ClosestChest,CreativeProp)
                    PetRotateToChest(ClosestChest,CreativeProp)
                    #-----------------Attack Chest
                    if(Cine:=Pets[PetIndex].AttackAnim):
                        Cine.Play(Agent)
                        Sleep(1.5) #Attack duration
                        Cine.Stop(Agent)
                    #----------------Crit
                    var PowerCrit:float=0.0
                    var CritRand:int = GetRandomInt(0,100)
                    if (CritRand>CustomPlayer.Pet_Crit_Value):
                        set PowerCrit=Power*CustomPlayer.BonusPowerMultiplicator*CustomPlayer.Bonus_Pet_Power
                    else:
                        set PowerCrit=Power*1.5*CustomPlayer.BonusPowerMultiplicator*CustomPlayer.Bonus_Pet_Power
                    #-------------------
                    Attack_Chest(ChestIndexFinal,PowerCrit)
                    Check_Chest_Gold(ChestIndexFinal,FortniteCharacter)

            ReleasePet(CreativeProp)

    FollowPlayerL26(FortniteCharacter : fort_character, CreativeProp : creative_prop, Power:float, PetIndex:int)<suspends>: void =
        #Print("Follow Player")
        #Future position of the pet
        Xplayer:float=-500.0
        Yplayer:float=-200.0
        if:
            Agent := FortniteCharacter.GetAgent[]
            PlayerObj := player[Agent]
            CustomPlayer:Custom_Player = PlayersMap[PlayerObj]
        then:
            var CheckValue:int = CustomPlayer.ButtonL26Pressed  
            set CustomPlayer.FollowActiveL26=1
            CustomPlayer.Update_Pets_Section()
            loop:
                if(CheckValue <> CustomPlayer.ButtonL26Pressed) {break}
                else if(CustomPlayer.PlayerInTreasureAreaFlag=0):
                    #AttackCinema.Stop()
                    PetRotateToPlayer(FortniteCharacter,CreativeProp)
                    Move_To_Player(FortniteCharacter,CreativeProp,Xplayer,Yplayer)
                else if(CustomPlayer.PlayerInTreasureAreaFlag=1):
                    #-----------------Find closest chest
                    var ClosestChest:creative_prop = creative_prop{}
                    var DistanceToMe:float = 100000000.0
                    var FoundChest:logic = false
                    var ChestIndexTemp:int = -1
                    var ChestIndexFinal:int = 0
                    var ClosestChestIndex:int=0
                    Allthechests:= ChestsL1+ChestsL2+ChestsL3+ChestsL4+ChestsL5+ChestsL6+ChestsL7+ChestsL8+ChestsL9+ChestsL10+ChestsL11+ChestsL12+ChestsL13+ChestsL14+ChestsL15+ChestsL16+ChestsL17+ChestsL18+ChestsL19+ChestsL20+ChestsL21+ChestsL22+ChestsL23+ChestsL24+ChestsL25+ChestsL26+ChestsL27+ChestsL28+ChestsL29+ChestsL30
                    for(Chestt:Allthechests):
                        set ChestIndexTemp+=1
                        MyPos:vector3 = FortniteCharacter.GetTransform().Translation
                        ChestPos:vector3 = Chestt.ChestProp.GetTransform().Translation
                        TotalDistance:float = Distance(MyPos, ChestPos) / 100.0
                        if(TotalDistance < DistanceToMe and Chestt.CurrentHP>0.0):
                            set ChestIndexFinal = ChestIndexTemp
                            #Print("Chest index{ChestIndexFinal}")
                            set FoundChest = true
                            set DistanceToMe = TotalDistance
                            set ClosestChest = Chestt.ChestProp
            
                    if(FoundChest = true):
                    #-----------------Rotate and move to chest+ rerotate because position is random
                    PetRotateToChest(ClosestChest,CreativeProp)
                    Move_To_Chest(FortniteCharacter,ClosestChest,CreativeProp)
                    PetRotateToChest(ClosestChest,CreativeProp)
                    #-----------------Attack Chest
                    if(Cine:=Pets[PetIndex].AttackAnim):
                        Cine.Play(Agent)
                        Sleep(1.5) #Attack duration
                        Cine.Stop(Agent)
                    #----------------Crit
                    var PowerCrit:float=0.0
                    var CritRand:int = GetRandomInt(0,100)
                    if (CritRand>CustomPlayer.Pet_Crit_Value):
                        set PowerCrit=Power*CustomPlayer.BonusPowerMultiplicator*CustomPlayer.Bonus_Pet_Power
                    else:
                        set PowerCrit=Power*1.5*CustomPlayer.BonusPowerMultiplicator*CustomPlayer.Bonus_Pet_Power
                    #-------------------
                    Attack_Chest(ChestIndexFinal,PowerCrit)
                    Check_Chest_Gold(ChestIndexFinal,FortniteCharacter)

            ReleasePet(CreativeProp)

    FollowPlayerL27(FortniteCharacter : fort_character, CreativeProp : creative_prop, Power:float, PetIndex:int)<suspends>: void =
        #Print("Follow Player")
        #Future position of the pet
        Xplayer:float=-500.0
        Yplayer:float= 200.0
        if:
            Agent := FortniteCharacter.GetAgent[]
            PlayerObj := player[Agent]
            CustomPlayer:Custom_Player = PlayersMap[PlayerObj]
        then:
            var CheckValue:int = CustomPlayer.ButtonL27Pressed  
            set CustomPlayer.FollowActiveL27=1
            CustomPlayer.Update_Pets_Section()
            loop:
                if(CheckValue <> CustomPlayer.ButtonL27Pressed) {break}
                else if(CustomPlayer.PlayerInTreasureAreaFlag=0):
                    #AttackCinema.Stop()
                    PetRotateToPlayer(FortniteCharacter,CreativeProp)
                    Move_To_Player(FortniteCharacter,CreativeProp,Xplayer,Yplayer)
                else if(CustomPlayer.PlayerInTreasureAreaFlag=1):
                    #-----------------Find closest chest
                    var ClosestChest:creative_prop = creative_prop{}
                    var DistanceToMe:float = 100000000.0
                    var FoundChest:logic = false
                    var ChestIndexTemp:int = -1
                    var ChestIndexFinal:int = 0
                    var ClosestChestIndex:int=0
                    Allthechests:= ChestsL1+ChestsL2+ChestsL3+ChestsL4+ChestsL5+ChestsL6+ChestsL7+ChestsL8+ChestsL9+ChestsL10+ChestsL11+ChestsL12+ChestsL13+ChestsL14+ChestsL15+ChestsL16+ChestsL17+ChestsL18+ChestsL19+ChestsL20+ChestsL21+ChestsL22+ChestsL23+ChestsL24+ChestsL25+ChestsL26+ChestsL27+ChestsL28+ChestsL29+ChestsL30
                    for(Chestt:Allthechests):
                        set ChestIndexTemp+=1
                        MyPos:vector3 = FortniteCharacter.GetTransform().Translation
                        ChestPos:vector3 = Chestt.ChestProp.GetTransform().Translation
                        TotalDistance:float = Distance(MyPos, ChestPos) / 100.0
                        if(TotalDistance < DistanceToMe and Chestt.CurrentHP>0.0):
                            set ChestIndexFinal = ChestIndexTemp
                            #Print("Chest index{ChestIndexFinal}")
                            set FoundChest = true
                            set DistanceToMe = TotalDistance
                            set ClosestChest = Chestt.ChestProp
            
                    if(FoundChest = true):
                    #-----------------Rotate and move to chest+ rerotate because position is random
                    PetRotateToChest(ClosestChest,CreativeProp)
                    Move_To_Chest(FortniteCharacter,ClosestChest,CreativeProp)
                    PetRotateToChest(ClosestChest,CreativeProp)
                    #-----------------Attack Chest
                    if(Cine:=Pets[PetIndex].AttackAnim):
                        Cine.Play(Agent)
                        Sleep(1.5) #Attack duration
                        Cine.Stop(Agent)
                    #----------------Crit
                    var PowerCrit:float=0.0
                    var CritRand:int = GetRandomInt(0,100)
                    if (CritRand>CustomPlayer.Pet_Crit_Value):
                        set PowerCrit=Power*CustomPlayer.BonusPowerMultiplicator*CustomPlayer.Bonus_Pet_Power
                    else:
                        set PowerCrit=Power*1.5*CustomPlayer.BonusPowerMultiplicator*CustomPlayer.Bonus_Pet_Power
                    #-------------------
                    Attack_Chest(ChestIndexFinal,PowerCrit)
                    Check_Chest_Gold(ChestIndexFinal,FortniteCharacter)

            ReleasePet(CreativeProp)

    FollowPlayerL28(FortniteCharacter : fort_character, CreativeProp : creative_prop, Power:float, PetIndex:int)<suspends>: void =
        #Print("Follow Player")
        #Future position of the pet
        Xplayer:float=-500.0
        Yplayer:float=-300.0
        if:
            Agent := FortniteCharacter.GetAgent[]
            PlayerObj := player[Agent]
            CustomPlayer:Custom_Player = PlayersMap[PlayerObj]
        then:
            var CheckValue:int = CustomPlayer.ButtonL28Pressed  
            set CustomPlayer.FollowActiveL28=1
            CustomPlayer.Update_Pets_Section()
            loop:
                if(CheckValue <> CustomPlayer.ButtonL28Pressed) {break}
                else if(CustomPlayer.PlayerInTreasureAreaFlag=0):
                    #AttackCinema.Stop()
                    PetRotateToPlayer(FortniteCharacter,CreativeProp)
                    Move_To_Player(FortniteCharacter,CreativeProp,Xplayer,Yplayer)
                else if(CustomPlayer.PlayerInTreasureAreaFlag=1):
                    #-----------------Find closest chest
                    var ClosestChest:creative_prop = creative_prop{}
                    var DistanceToMe:float = 100000000.0
                    var FoundChest:logic = false
                    var ChestIndexTemp:int = -1
                    var ChestIndexFinal:int = 0
                    var ClosestChestIndex:int=0
                    Allthechests:= ChestsL1+ChestsL2+ChestsL3+ChestsL4+ChestsL5+ChestsL6+ChestsL7+ChestsL8+ChestsL9+ChestsL10+ChestsL11+ChestsL12+ChestsL13+ChestsL14+ChestsL15+ChestsL16+ChestsL17+ChestsL18+ChestsL19+ChestsL20+ChestsL21+ChestsL22+ChestsL23+ChestsL24+ChestsL25+ChestsL26+ChestsL27+ChestsL28+ChestsL29+ChestsL30
                    for(Chestt:Allthechests):
                        set ChestIndexTemp+=1
                        MyPos:vector3 = FortniteCharacter.GetTransform().Translation
                        ChestPos:vector3 = Chestt.ChestProp.GetTransform().Translation
                        TotalDistance:float = Distance(MyPos, ChestPos) / 100.0
                        if(TotalDistance < DistanceToMe and Chestt.CurrentHP>0.0):
                            set ChestIndexFinal = ChestIndexTemp
                            #Print("Chest index{ChestIndexFinal}")
                            set FoundChest = true
                            set DistanceToMe = TotalDistance
                            set ClosestChest = Chestt.ChestProp
            
                    if(FoundChest = true):
                    #-----------------Rotate and move to chest+ rerotate because position is random
                    PetRotateToChest(ClosestChest,CreativeProp)
                    Move_To_Chest(FortniteCharacter,ClosestChest,CreativeProp)
                    PetRotateToChest(ClosestChest,CreativeProp)
                    #-----------------Attack Chest
                    if(Cine:=Pets[PetIndex].AttackAnim):
                        Cine.Play(Agent)
                        Sleep(1.5) #Attack duration
                        Cine.Stop(Agent)
                    #----------------Crit
                    var PowerCrit:float=0.0
                    var CritRand:int = GetRandomInt(0,100)
                    if (CritRand>CustomPlayer.Pet_Crit_Value):
                        set PowerCrit=Power*CustomPlayer.BonusPowerMultiplicator*CustomPlayer.Bonus_Pet_Power
                    else:
                        set PowerCrit=Power*1.5*CustomPlayer.BonusPowerMultiplicator*CustomPlayer.Bonus_Pet_Power
                    #-------------------
                    Attack_Chest(ChestIndexFinal,PowerCrit)
                    Check_Chest_Gold(ChestIndexFinal,FortniteCharacter)

            ReleasePet(CreativeProp)

    FollowPlayerL29(FortniteCharacter : fort_character, CreativeProp : creative_prop, Power:float, PetIndex:int)<suspends>: void =
        #Print("Follow Player")
        #Future position of the pet
        Xplayer:float=-500.0
        Yplayer:float= 300.0
        if:
            Agent := FortniteCharacter.GetAgent[]
            PlayerObj := player[Agent]
            CustomPlayer:Custom_Player = PlayersMap[PlayerObj]
        then:
            var CheckValue:int = CustomPlayer.ButtonL29Pressed  
            set CustomPlayer.FollowActiveL29=1
            CustomPlayer.Update_Pets_Section()
            loop:
                if(CheckValue <> CustomPlayer.ButtonL29Pressed) {break}
                else if(CustomPlayer.PlayerInTreasureAreaFlag=0):
                    #AttackCinema.Stop()
                    PetRotateToPlayer(FortniteCharacter,CreativeProp)
                    Move_To_Player(FortniteCharacter,CreativeProp,Xplayer,Yplayer)
                else if(CustomPlayer.PlayerInTreasureAreaFlag=1):
                    #-----------------Find closest chest
                    var ClosestChest:creative_prop = creative_prop{}
                    var DistanceToMe:float = 100000000.0
                    var FoundChest:logic = false
                    var ChestIndexTemp:int = -1
                    var ChestIndexFinal:int = 0
                    var ClosestChestIndex:int=0
                    Allthechests:= ChestsL1+ChestsL2+ChestsL3+ChestsL4+ChestsL5+ChestsL6+ChestsL7+ChestsL8+ChestsL9+ChestsL10+ChestsL11+ChestsL12+ChestsL13+ChestsL14+ChestsL15+ChestsL16+ChestsL17+ChestsL18+ChestsL19+ChestsL20+ChestsL21+ChestsL22+ChestsL23+ChestsL24+ChestsL25+ChestsL26+ChestsL27+ChestsL28+ChestsL29+ChestsL30
                    for(Chestt:Allthechests):
                        set ChestIndexTemp+=1
                        MyPos:vector3 = FortniteCharacter.GetTransform().Translation
                        ChestPos:vector3 = Chestt.ChestProp.GetTransform().Translation
                        TotalDistance:float = Distance(MyPos, ChestPos) / 100.0
                        if(TotalDistance < DistanceToMe and Chestt.CurrentHP>0.0):
                            set ChestIndexFinal = ChestIndexTemp
                            #Print("Chest index{ChestIndexFinal}")
                            set FoundChest = true
                            set DistanceToMe = TotalDistance
                            set ClosestChest = Chestt.ChestProp
            
                    if(FoundChest = true):
                    #-----------------Rotate and move to chest+ rerotate because position is random
                    PetRotateToChest(ClosestChest,CreativeProp)
                    Move_To_Chest(FortniteCharacter,ClosestChest,CreativeProp)
                    PetRotateToChest(ClosestChest,CreativeProp)
                    #-----------------Attack Chest
                    if(Cine:=Pets[PetIndex].AttackAnim):
                        Cine.Play(Agent)
                        Sleep(1.5) #Attack duration
                        Cine.Stop(Agent)
                    #----------------Crit
                    var PowerCrit:float=0.0
                    var CritRand:int = GetRandomInt(0,100)
                    if (CritRand>CustomPlayer.Pet_Crit_Value):
                        set PowerCrit=Power*CustomPlayer.BonusPowerMultiplicator*CustomPlayer.Bonus_Pet_Power
                    else:
                        set PowerCrit=Power*1.5*CustomPlayer.BonusPowerMultiplicator*CustomPlayer.Bonus_Pet_Power
                    #-------------------
                    Attack_Chest(ChestIndexFinal,PowerCrit)
                    Check_Chest_Gold(ChestIndexFinal,FortniteCharacter)

            ReleasePet(CreativeProp)

    FollowPlayerL30(FortniteCharacter : fort_character, CreativeProp : creative_prop, Power:float, PetIndex:int)<suspends>: void =
        #Print("Follow Player")
        #Future position of the pet
        Xplayer:float=-600.0
        Yplayer:float= 0.0
        if:
            Agent := FortniteCharacter.GetAgent[]
            PlayerObj := player[Agent]
            CustomPlayer:Custom_Player = PlayersMap[PlayerObj]
        then:
            var CheckValue:int = CustomPlayer.ButtonL30Pressed  
            set CustomPlayer.FollowActiveL30=1
            CustomPlayer.Update_Pets_Section()
            loop:
                if(CheckValue <> CustomPlayer.ButtonL30Pressed) {break}
                else if(CustomPlayer.PlayerInTreasureAreaFlag=0):
                    #AttackCinema.Stop()
                    PetRotateToPlayer(FortniteCharacter,CreativeProp)
                    Move_To_Player(FortniteCharacter,CreativeProp,Xplayer,Yplayer)
                else if(CustomPlayer.PlayerInTreasureAreaFlag=1):
                    #-----------------Find closest chest
                    var ClosestChest:creative_prop = creative_prop{}
                    var DistanceToMe:float = 100000000.0
                    var FoundChest:logic = false
                    var ChestIndexTemp:int = -1
                    var ChestIndexFinal:int = 0
                    var ClosestChestIndex:int=0
                    Allthechests:= ChestsL1+ChestsL2+ChestsL3+ChestsL4+ChestsL5+ChestsL6+ChestsL7+ChestsL8+ChestsL9+ChestsL10+ChestsL11+ChestsL12+ChestsL13+ChestsL14+ChestsL15+ChestsL16+ChestsL17+ChestsL18+ChestsL19+ChestsL20+ChestsL21+ChestsL22+ChestsL23+ChestsL24+ChestsL25+ChestsL26+ChestsL27+ChestsL28+ChestsL29+ChestsL30
                    for(Chestt:Allthechests):
                        set ChestIndexTemp+=1
                        MyPos:vector3 = FortniteCharacter.GetTransform().Translation
                        ChestPos:vector3 = Chestt.ChestProp.GetTransform().Translation
                        TotalDistance:float = Distance(MyPos, ChestPos) / 100.0
                        if(TotalDistance < DistanceToMe and Chestt.CurrentHP>0.0):
                            set ChestIndexFinal = ChestIndexTemp
                            #Print("Chest index{ChestIndexFinal}")
                            set FoundChest = true
                            set DistanceToMe = TotalDistance
                            set ClosestChest = Chestt.ChestProp
            
                    if(FoundChest = true):
                    #-----------------Rotate and move to chest+ rerotate because position is random
                    PetRotateToChest(ClosestChest,CreativeProp)
                    Move_To_Chest(FortniteCharacter,ClosestChest,CreativeProp)
                    PetRotateToChest(ClosestChest,CreativeProp)
                    #-----------------Attack Chest
                    if(Cine:=Pets[PetIndex].AttackAnim):
                        Cine.Play(Agent)
                        Sleep(1.5) #Attack duration
                        Cine.Stop(Agent)
                    #----------------Crit
                    var PowerCrit:float=0.0
                    var CritRand:int = GetRandomInt(0,100)
                    if (CritRand>CustomPlayer.Pet_Crit_Value):
                        set PowerCrit=Power*CustomPlayer.BonusPowerMultiplicator*CustomPlayer.Bonus_Pet_Power
                    else:
                        set PowerCrit=Power*1.5*CustomPlayer.BonusPowerMultiplicator*CustomPlayer.Bonus_Pet_Power
                    #-------------------
                    Attack_Chest(ChestIndexFinal,PowerCrit)
                    Check_Chest_Gold(ChestIndexFinal,FortniteCharacter)

            ReleasePet(CreativeProp)

    FollowPlayerL31(FortniteCharacter : fort_character, CreativeProp : creative_prop, Power:float, PetIndex:int)<suspends>: void =
        #Future position of the pet
        Xplayer:float= 0.0
        Yplayer:float= -200.0
        if:
            Agent := FortniteCharacter.GetAgent[]
            PlayerObj := player[Agent]
            CustomPlayer:Custom_Player = PlayersMap[PlayerObj]
        then:
            var CheckValue:int = CustomPlayer.ButtonL31Pressed  
            set CustomPlayer.FollowActiveL31=1
            CustomPlayer.Update_Pets_Section()
            loop:
                if(CheckValue <> CustomPlayer.ButtonL31Pressed) {break}
                else if(CustomPlayer.PlayerInTreasureAreaFlag=0):
                    #AttackCinema.Stop()
                    PetRotateToPlayer(FortniteCharacter,CreativeProp)
                    Move_To_Player(FortniteCharacter,CreativeProp,Xplayer,Yplayer)
                else if(CustomPlayer.PlayerInTreasureAreaFlag=1):
                    #-----------------Find closest chest
                    var ClosestChest:creative_prop = creative_prop{}
                    var DistanceToMe:float = 100000000.0
                    var FoundChest:logic = false
                    var ChestIndexTemp:int = -1
                    var ChestIndexFinal:int = 0
                    var ClosestChestIndex:int=0
                    Allthechests:= ChestsL1+ChestsL2+ChestsL3+ChestsL4+ChestsL5+ChestsL6+ChestsL7+ChestsL8+ChestsL9+ChestsL10+ChestsL11+ChestsL12+ChestsL13+ChestsL14+ChestsL15+ChestsL16+ChestsL17+ChestsL18+ChestsL19+ChestsL20+ChestsL21+ChestsL22+ChestsL23+ChestsL24+ChestsL25+ChestsL26+ChestsL27+ChestsL28+ChestsL29+ChestsL30
                    for(Chestt:Allthechests):
                        set ChestIndexTemp+=1
                        MyPos:vector3 = FortniteCharacter.GetTransform().Translation
                        ChestPos:vector3 = Chestt.ChestProp.GetTransform().Translation
                        TotalDistance:float = Distance(MyPos, ChestPos) / 100.0
                        if(TotalDistance < DistanceToMe and Chestt.CurrentHP>0.0):
                            set ChestIndexFinal = ChestIndexTemp
                            #Print("Chest index{ChestIndexFinal}")
                            set FoundChest = true
                            set DistanceToMe = TotalDistance
                            set ClosestChest = Chestt.ChestProp
            
                    if(FoundChest = true):
                    #-----------------Rotate and move to chest+ rerotate because position is random
                    PetRotateToChest(ClosestChest,CreativeProp)
                    Move_To_Chest(FortniteCharacter,ClosestChest,CreativeProp)
                    PetRotateToChest(ClosestChest,CreativeProp)
                    #-----------------Attack Chest
                    if(Cine:=Pets[PetIndex].AttackAnim):
                        Cine.Play(Agent)
                        Sleep(1.5) #Attack duration
                        Cine.Stop(Agent)
                    #----------------Crit
                    var PowerCrit:float=0.0
                    var CritRand:int = GetRandomInt(0,100)
                    if (CritRand>CustomPlayer.Pet_Crit_Value):
                        set PowerCrit=Power*CustomPlayer.BonusPowerMultiplicator*CustomPlayer.Bonus_Pet_Power
                    else:
                        set PowerCrit=Power*1.5*CustomPlayer.BonusPowerMultiplicator*CustomPlayer.Bonus_Pet_Power
                    #-------------------
                    Attack_Chest(ChestIndexFinal,PowerCrit)
                    Check_Chest_Gold(ChestIndexFinal,FortniteCharacter)

            ReleasePet(CreativeProp)

    FollowPlayerL32(FortniteCharacter : fort_character, CreativeProp : creative_prop, Power:float, PetIndex:int)<suspends>: void =
        #Future position of the pet
        Xplayer:float= 0.0
        Yplayer:float= 200.0
        if:
            Agent := FortniteCharacter.GetAgent[]
            PlayerObj := player[Agent]
            CustomPlayer:Custom_Player = PlayersMap[PlayerObj]
        then:
            var CheckValue:int = CustomPlayer.ButtonL32Pressed  
            set CustomPlayer.FollowActiveL32=1
            CustomPlayer.Update_Pets_Section()
            loop:
                if(CheckValue <> CustomPlayer.ButtonL32Pressed) {break}
                else if(CustomPlayer.PlayerInTreasureAreaFlag=0):
                    #AttackCinema.Stop()
                    PetRotateToPlayer(FortniteCharacter,CreativeProp)
                    Move_To_Player(FortniteCharacter,CreativeProp,Xplayer,Yplayer)
                else if(CustomPlayer.PlayerInTreasureAreaFlag=1):
                    #-----------------Find closest chest
                    var ClosestChest:creative_prop = creative_prop{}
                    var DistanceToMe:float = 100000000.0
                    var FoundChest:logic = false
                    var ChestIndexTemp:int = -1
                    var ChestIndexFinal:int = 0
                    var ClosestChestIndex:int=0
                    Allthechests:= ChestsL1+ChestsL2+ChestsL3+ChestsL4+ChestsL5+ChestsL6+ChestsL7+ChestsL8+ChestsL9+ChestsL10+ChestsL11+ChestsL12+ChestsL13+ChestsL14+ChestsL15+ChestsL16+ChestsL17+ChestsL18+ChestsL19+ChestsL20+ChestsL21+ChestsL22+ChestsL23+ChestsL24+ChestsL25+ChestsL26+ChestsL27+ChestsL28+ChestsL29+ChestsL30
                    for(Chestt:Allthechests):
                        set ChestIndexTemp+=1
                        MyPos:vector3 = FortniteCharacter.GetTransform().Translation
                        ChestPos:vector3 = Chestt.ChestProp.GetTransform().Translation
                        TotalDistance:float = Distance(MyPos, ChestPos) / 100.0
                        if(TotalDistance < DistanceToMe and Chestt.CurrentHP>0.0):
                            set ChestIndexFinal = ChestIndexTemp
                            #Print("Chest index{ChestIndexFinal}")
                            set FoundChest = true
                            set DistanceToMe = TotalDistance
                            set ClosestChest = Chestt.ChestProp
            
                    if(FoundChest = true):
                    #-----------------Rotate and move to chest+ rerotate because position is random
                    PetRotateToChest(ClosestChest,CreativeProp)
                    Move_To_Chest(FortniteCharacter,ClosestChest,CreativeProp)
                    PetRotateToChest(ClosestChest,CreativeProp)
                    #-----------------Attack Chest
                    if(Cine:=Pets[PetIndex].AttackAnim):
                        Cine.Play(Agent)
                        Sleep(1.5) #Attack duration
                        Cine.Stop(Agent)
                    #----------------Crit
                    var PowerCrit:float=0.0
                    var CritRand:int = GetRandomInt(0,100)
                    if (CritRand>CustomPlayer.Pet_Crit_Value):
                        set PowerCrit=Power*CustomPlayer.BonusPowerMultiplicator*CustomPlayer.Bonus_Pet_Power
                    else:
                        set PowerCrit=Power*1.5*CustomPlayer.BonusPowerMultiplicator*CustomPlayer.Bonus_Pet_Power
                    #-------------------
                    Attack_Chest(ChestIndexFinal,PowerCrit)
                    Check_Chest_Gold(ChestIndexFinal,FortniteCharacter)

            ReleasePet(CreativeProp)

    FollowPlayerL33(FortniteCharacter : fort_character, CreativeProp : creative_prop, Power:float, PetIndex:int)<suspends>: void =
        #Future position of the pet
        Xplayer:float= 0.0
        Yplayer:float= -100.0
        if:
            Agent := FortniteCharacter.GetAgent[]
            PlayerObj := player[Agent]
            CustomPlayer:Custom_Player = PlayersMap[PlayerObj]
        then:
            var CheckValue:int = CustomPlayer.ButtonL33Pressed  
            set CustomPlayer.FollowActiveL33=1
            CustomPlayer.Update_Pets_Section()
            loop:
                if(CheckValue <> CustomPlayer.ButtonL33Pressed) {break}
                else if(CustomPlayer.PlayerInTreasureAreaFlag=0):
                    #AttackCinema.Stop()
                    PetRotateToPlayer(FortniteCharacter,CreativeProp)
                    Move_To_Player(FortniteCharacter,CreativeProp,Xplayer,Yplayer)
                else if(CustomPlayer.PlayerInTreasureAreaFlag=1):
                    #-----------------Find closest chest
                    var ClosestChest:creative_prop = creative_prop{}
                    var DistanceToMe:float = 100000000.0
                    var FoundChest:logic = false
                    var ChestIndexTemp:int = -1
                    var ChestIndexFinal:int = 0
                    var ClosestChestIndex:int=0
                    Allthechests:= ChestsL1+ChestsL2+ChestsL3+ChestsL4+ChestsL5+ChestsL6+ChestsL7+ChestsL8+ChestsL9+ChestsL10+ChestsL11+ChestsL12+ChestsL13+ChestsL14+ChestsL15+ChestsL16+ChestsL17+ChestsL18+ChestsL19+ChestsL20+ChestsL21+ChestsL22+ChestsL23+ChestsL24+ChestsL25+ChestsL26+ChestsL27+ChestsL28+ChestsL29+ChestsL30
                    for(Chestt:Allthechests):
                        set ChestIndexTemp+=1
                        MyPos:vector3 = FortniteCharacter.GetTransform().Translation
                        ChestPos:vector3 = Chestt.ChestProp.GetTransform().Translation
                        TotalDistance:float = Distance(MyPos, ChestPos) / 100.0
                        if(TotalDistance < DistanceToMe and Chestt.CurrentHP>0.0):
                            set ChestIndexFinal = ChestIndexTemp
                            #Print("Chest index{ChestIndexFinal}")
                            set FoundChest = true
                            set DistanceToMe = TotalDistance
                            set ClosestChest = Chestt.ChestProp
            
                    if(FoundChest = true):
                    #-----------------Rotate and move to chest+ rerotate because position is random
                    PetRotateToChest(ClosestChest,CreativeProp)
                    Move_To_Chest(FortniteCharacter,ClosestChest,CreativeProp)
                    PetRotateToChest(ClosestChest,CreativeProp)
                    #-----------------Attack Chest
                    if(Cine:=Pets[PetIndex].AttackAnim):
                        Cine.Play(Agent)
                        Sleep(1.5) #Attack duration
                        Cine.Stop(Agent)
                    #----------------Crit
                    var PowerCrit:float=0.0
                    var CritRand:int = GetRandomInt(0,100)
                    if (CritRand>CustomPlayer.Pet_Crit_Value):
                        set PowerCrit=Power*CustomPlayer.BonusPowerMultiplicator*CustomPlayer.Bonus_Pet_Power
                    else:
                        set PowerCrit=Power*1.5*CustomPlayer.BonusPowerMultiplicator*CustomPlayer.Bonus_Pet_Power
                    #-------------------
                    Attack_Chest(ChestIndexFinal,PowerCrit)
                    Check_Chest_Gold(ChestIndexFinal,FortniteCharacter)

            ReleasePet(CreativeProp)

    FollowPlayerL34(FortniteCharacter : fort_character, CreativeProp : creative_prop, Power:float, PetIndex:int)<suspends>: void =
        #Future position of the pet
        Xplayer:float= 0.0
        Yplayer:float= 100.0
        if:
            Agent := FortniteCharacter.GetAgent[]
            PlayerObj := player[Agent]
            CustomPlayer:Custom_Player = PlayersMap[PlayerObj]
        then:
            var CheckValue:int = CustomPlayer.ButtonL34Pressed  
            set CustomPlayer.FollowActiveL34=1
            CustomPlayer.Update_Pets_Section()
            loop:
                if(CheckValue <> CustomPlayer.ButtonL34Pressed) {break}
                else if(CustomPlayer.PlayerInTreasureAreaFlag=0):
                    #AttackCinema.Stop()
                    PetRotateToPlayer(FortniteCharacter,CreativeProp)
                    Move_To_Player(FortniteCharacter,CreativeProp,Xplayer,Yplayer)
                else if(CustomPlayer.PlayerInTreasureAreaFlag=1):
                    #-----------------Find closest chest
                    var ClosestChest:creative_prop = creative_prop{}
                    var DistanceToMe:float = 100000000.0
                    var FoundChest:logic = false
                    var ChestIndexTemp:int = -1
                    var ChestIndexFinal:int = 0
                    var ClosestChestIndex:int=0
                    Allthechests:= ChestsL1+ChestsL2+ChestsL3+ChestsL4+ChestsL5+ChestsL6+ChestsL7+ChestsL8+ChestsL9+ChestsL10+ChestsL11+ChestsL12+ChestsL13+ChestsL14+ChestsL15+ChestsL16+ChestsL17+ChestsL18+ChestsL19+ChestsL20+ChestsL21+ChestsL22+ChestsL23+ChestsL24+ChestsL25+ChestsL26+ChestsL27+ChestsL28+ChestsL29+ChestsL30
                    for(Chestt:Allthechests):
                        set ChestIndexTemp+=1
                        MyPos:vector3 = FortniteCharacter.GetTransform().Translation
                        ChestPos:vector3 = Chestt.ChestProp.GetTransform().Translation
                        TotalDistance:float = Distance(MyPos, ChestPos) / 100.0
                        if(TotalDistance < DistanceToMe and Chestt.CurrentHP>0.0):
                            set ChestIndexFinal = ChestIndexTemp
                            #Print("Chest index{ChestIndexFinal}")
                            set FoundChest = true
                            set DistanceToMe = TotalDistance
                            set ClosestChest = Chestt.ChestProp
            
                    if(FoundChest = true):
                    #-----------------Rotate and move to chest+ rerotate because position is random
                    PetRotateToChest(ClosestChest,CreativeProp)
                    Move_To_Chest(FortniteCharacter,ClosestChest,CreativeProp)
                    PetRotateToChest(ClosestChest,CreativeProp)
                    #-----------------Attack Chest
                    if(Cine:=Pets[PetIndex].AttackAnim):
                        Cine.Play(Agent)
                        Sleep(1.5) #Attack duration
                        Cine.Stop(Agent)
                    #----------------Crit
                    var PowerCrit:float=0.0
                    var CritRand:int = GetRandomInt(0,100)
                    if (CritRand>CustomPlayer.Pet_Crit_Value):
                        set PowerCrit=Power*CustomPlayer.BonusPowerMultiplicator*CustomPlayer.Bonus_Pet_Power
                    else:
                        set PowerCrit=Power*1.5*CustomPlayer.BonusPowerMultiplicator*CustomPlayer.Bonus_Pet_Power
                    #-------------------
                    Attack_Chest(ChestIndexFinal,PowerCrit)
                    Check_Chest_Gold(ChestIndexFinal,FortniteCharacter)

            ReleasePet(CreativeProp)

    Move_To_Player(FortniteCharacter : fort_character, CreativeProp : creative_prop, Xplayer : float, Yplayer : float)<suspends>:void=
        #Print("Move_To_Player")
        var Speed:float=1.0
        PlayerPos := FortniteCharacter.GetTransform().Translation
        PlayerBack := FortniteCharacter.GetViewRotation().GetLocalForward() * Xplayer
        PlayerRight := FortniteCharacter.GetViewRotation().GetLocalRight() * Yplayer
        PlayerPosFinal := PlayerPos+PlayerBack+PlayerRight
        PlayerPos_Allowed := vector3{ X := PlayerPosFinal.X , Y := PlayerPosFinal.Y, Z := PlayerPos.Z}

        if:
            Agent:=FortniteCharacter.GetAgent[]
            PlayerObj := player[Agent]
            CustomPlayer:Custom_Player = PlayersMap[PlayerObj]
        then:
            set Speed=CustomPlayer.Pet_Speed_Value

        CreativeProp.MoveTo(PlayerPos_Allowed, CreativeProp.GetTransform().Rotation, 1.0/Speed)

    Rand_Minus_One_And_One():int=
        var Result:int=GetRandomInt(0,1)
        if(Result=1):
            1
        else:
            -1

    Move_To_Chest(FortniteCharacter : fort_character,TargetChest :creative_prop , CreativeProp : creative_prop)<suspends>:void=
        #Print("Move_To_Chest")
        var Speed:float=1.0
        var RandX:float=GetRandomInt(10,15)*10.0*Rand_Minus_One_And_One()
        var RandY:float=GetRandomInt(10,15)*10.0*Rand_Minus_One_And_One()
        #Print("Rand X: {RandX}")
        #Print("Rand Y: {RandY}")
        ChestPos := TargetChest.GetTransform().Translation + vector3{X:=RandX} + vector3{Y:=RandY}
        ChestPos_Allowed := vector3{ X := ChestPos.X, Y := ChestPos.Y, Z := ChestPos.Z}

        if:
            Agent:=FortniteCharacter.GetAgent[]
            PlayerObj := player[Agent]
            CustomPlayer:Custom_Player = PlayersMap[PlayerObj]
        then:
            set Speed=CustomPlayer.Pet_Speed_Value

        CreativeProp.MoveTo(ChestPos_Allowed, CreativeProp.GetTransform().Rotation, 1.0/Speed)
        
    PetRotateToPlayer(FortniteCharacter : fort_character, CreativeProp : creative_prop)<suspends> : void =
        #Print("Pet Rotate To Player")
        PlayerLocation := FortniteCharacter.GetTransform().Translation
        PropLocation := CreativeProp.GetTransform().Translation
        
        if (LookDirection := (PlayerLocation - PropLocation).MakeUnitVector[]):

            Yaw := RadiansToDegrees(ArcTan(LookDirection.Y, LookDirection.X)) - 90.0
            Pitch := RadiansToDegrees(ArcTan(LookDirection.Z, Sqrt((LookDirection.X * LookDirection.X) + (LookDirection.Y * LookDirection.Y))))
            Roll := 0.0

            NewRotation := MakeRotationFromYawPitchRollDegrees(Yaw, Pitch, Roll)

            CreativeProp.MoveTo(CreativeProp.GetTransform().Translation, NewRotation, 0.5)

    PetRotateToChest(TargetChest : creative_prop, CreativeProp : creative_prop)<suspends> : void =
        #Print("Pet Rotate To Chest")
        ChestLocation := TargetChest.GetTransform().Translation
        PetLocation := CreativeProp.GetTransform().Translation
        
        if (LookDirection := (ChestLocation-PetLocation).MakeUnitVector[]):

            Yaw := RadiansToDegrees(ArcTan(LookDirection.Y, LookDirection.X)) - 90.0
            Pitch := RadiansToDegrees(ArcTan(LookDirection.Z, Sqrt((LookDirection.X * LookDirection.X) + (LookDirection.Y * LookDirection.Y))))
            Roll := 0.0

            NewRotation := MakeRotationFromYawPitchRollDegrees(Yaw, Pitch, Roll)

            CreativeProp.MoveTo(CreativeProp.GetTransform().Translation, NewRotation, 0.5)

    OnPlayerEnterTreasureArea(Agent:agent):void=
        #Print("Player enter Area")
        if:
            PlayerObj := player[Agent]
            CustomPlayer:Custom_Player = PlayersMap[PlayerObj]
        then:
            set CustomPlayer.PlayerInTreasureAreaFlag=1

    OnPlayerExitTreasureArea(Agent:agent):void=
        #Print("Player exit Area")
        if:
            PlayerObj := player[Agent]
            CustomPlayer:Custom_Player = PlayersMap[PlayerObj]
        then:
            set CustomPlayer.PlayerInTreasureAreaFlag=0

    Buy_B1(Agent:agent):void=
        #Already purchased?
        if:
            PlayerObj := player[Agent]
            CurrentPlayerStats := PlayerStatsManager.GetPlayerStats[Agent]
            CurrentPlayerStats.LevelUnlocked.CurrentValue>=1
        then:
            AlreadyPurchasedHUD.Show(Agent) 
        else:
            #Not enough gold ?
            if:
                PlayerObj := player[Agent]
                CustomPlayer:Custom_Player = PlayersMap[PlayerObj]
                CustomPlayer.PlayerGold<CostBarrier1
            then:
                NotEnoughGoldHUD.Show(Agent)
            else:
                NewEggHUD.Show(Agent)
                Accolade_B1B5.Award(Agent)
                UnlockBarrierB1(Agent)#Action
                AddMaxLevelUnlocked(Agent)#Save
                DecreaseGoldFunc(Agent,CostBarrier1)#Gold Save+UI

    Buy_B2(Agent:agent):void=
        #Already purchased?
        if:
            PlayerObj := player[Agent]
            CurrentPlayerStats := PlayerStatsManager.GetPlayerStats[Agent]
            CurrentPlayerStats.LevelUnlocked.CurrentValue>=2
        then:
            AlreadyPurchasedHUD.Show(Agent) 
        else:
            #Not enough gold ?
            if:
                PlayerObj := player[Agent]
                CustomPlayer:Custom_Player = PlayersMap[PlayerObj]
                CustomPlayer.PlayerGold<CostBarrier2
            then:
                NotEnoughGoldHUD.Show(Agent)
            else:
                Accolade_B1B5.Award(Agent)
                UnlockBarrierB2(Agent)#Action
                AddMaxLevelUnlocked(Agent)#Save
                DecreaseGoldFunc(Agent,CostBarrier2)#Gold Save+UI
                VillageUnlockedHUD.Show(Agent)
                if(Cine:=Cine_Unlock_Village[GetTeamIndexAgent[Agent]-1]):#Camera show village
                    Cine.Play(Agent)

    Buy_B3(Agent:agent):void=
        #Already purchased?
        if:
            PlayerObj := player[Agent]
            CurrentPlayerStats := PlayerStatsManager.GetPlayerStats[Agent]
            CurrentPlayerStats.LevelUnlocked.CurrentValue>=3
        then:
            AlreadyPurchasedHUD.Show(Agent) 
        else:
            #Not enough gold ?
            if:
                PlayerObj := player[Agent]
                CustomPlayer:Custom_Player = PlayersMap[PlayerObj]
                CustomPlayer.PlayerGold<CostBarrier3
            then:
                NotEnoughGoldHUD.Show(Agent)
            else:
                NewEggHUD.Show(Agent)
                Accolade_B1B5.Award(Agent)
                UnlockBarrierB3(Agent)#Action
                AddMaxLevelUnlocked(Agent)#Save
                DecreaseGoldFunc(Agent,CostBarrier3)#Gold Save+UI

    Buy_B4(Agent:agent):void=
        #Already purchased?
        if:
            PlayerObj := player[Agent]
            CurrentPlayerStats := PlayerStatsManager.GetPlayerStats[Agent]
            CurrentPlayerStats.LevelUnlocked.CurrentValue>=4
        then:
            AlreadyPurchasedHUD.Show(Agent) 
        else:
            #Not enough gold ?
            if:
                PlayerObj := player[Agent]
                CustomPlayer:Custom_Player = PlayersMap[PlayerObj]
                CustomPlayer.PlayerGold<CostBarrier4
            then:
                NotEnoughGoldHUD.Show(Agent)
            else:
                NewEggHUD.Show(Agent)
                Accolade_B1B5.Award(Agent)
                UnlockBarrierB4(Agent)#Action
                AddMaxLevelUnlocked(Agent)#Save
                DecreaseGoldFunc(Agent,CostBarrier4)#Gold Save+UI

    Buy_B5(Agent:agent):void=
        #Already purchased?
        if:
            PlayerObj := player[Agent]
            CurrentPlayerStats := PlayerStatsManager.GetPlayerStats[Agent]
            CurrentPlayerStats.LevelUnlocked.CurrentValue>=5
        then:
            AlreadyPurchasedHUD.Show(Agent) 
        else:
            #Not enough gold ?
            if:
                PlayerObj := player[Agent]
                CustomPlayer:Custom_Player = PlayersMap[PlayerObj]
                CustomPlayer.PlayerGold<CostBarrier5
            then:
                NotEnoughGoldHUD.Show(Agent)
            else:
                NewEggHUD.Show(Agent)
                Accolade_B1B5.Award(Agent)
                UnlockBarrierB5(Agent)#Action
                AddMaxLevelUnlocked(Agent)#Save
                DecreaseGoldFunc(Agent,CostBarrier5)#Gold Save+UI

    Buy_B6(Agent:agent):void=
        #Already purchased?
        if:
            PlayerObj := player[Agent]
            CurrentPlayerStats := PlayerStatsManager.GetPlayerStats[Agent]
            CurrentPlayerStats.LevelUnlocked.CurrentValue>=6
        then:
            AlreadyPurchasedHUD.Show(Agent) 
        else:
            #Not enough gold ?
            if:
                PlayerObj := player[Agent]
                CustomPlayer:Custom_Player = PlayersMap[PlayerObj]
                CustomPlayer.PlayerGold<CostBarrier6
            then:
                NotEnoughGoldHUD.Show(Agent)
            else:
                NewEggHUD.Show(Agent)
                Accolade_B6B10.Award(Agent)
                UnlockBarrierB6(Agent)#Action
                AddMaxLevelUnlocked(Agent)#Save
                DecreaseGoldFunc(Agent,CostBarrier6)#Gold Save+UI

    Buy_B7(Agent:agent):void=
        #Already purchased?
        if:
            PlayerObj := player[Agent]
            CurrentPlayerStats := PlayerStatsManager.GetPlayerStats[Agent]
            CurrentPlayerStats.LevelUnlocked.CurrentValue>=7
        then:
            AlreadyPurchasedHUD.Show(Agent) 
        else:
            #Not enough gold ?
            if:
                PlayerObj := player[Agent]
                CustomPlayer:Custom_Player = PlayersMap[PlayerObj]
                CustomPlayer.PlayerGold<CostBarrier7
            then:
                NotEnoughGoldHUD.Show(Agent)
            else:
                NewEggHUD.Show(Agent)
                Accolade_B6B10.Award(Agent)
                UnlockBarrierB7(Agent)#Action
                AddMaxLevelUnlocked(Agent)#Save
                DecreaseGoldFunc(Agent,CostBarrier7)#Gold Save+UI

    Buy_B8(Agent:agent):void=
        #Already purchased?
        if:
            PlayerObj := player[Agent]
            CurrentPlayerStats := PlayerStatsManager.GetPlayerStats[Agent]
            CurrentPlayerStats.LevelUnlocked.CurrentValue>=8
        then:
            AlreadyPurchasedHUD.Show(Agent) 
        else:
            #Not enough gold ?
            if:
                PlayerObj := player[Agent]
                CustomPlayer:Custom_Player = PlayersMap[PlayerObj]
                CustomPlayer.PlayerGold<CostBarrier8
            then:
                NotEnoughGoldHUD.Show(Agent)
            else:
                NewEggHUD.Show(Agent)
                Accolade_B6B10.Award(Agent)
                UnlockBarrierB8(Agent)#Action
                AddMaxLevelUnlocked(Agent)#Save
                DecreaseGoldFunc(Agent,CostBarrier8)#Gold Save+UI

    Buy_B9(Agent:agent):void=
        #Already purchased?
        if:
            PlayerObj := player[Agent]
            CurrentPlayerStats := PlayerStatsManager.GetPlayerStats[Agent]
            CurrentPlayerStats.LevelUnlocked.CurrentValue>=9
        then:
            AlreadyPurchasedHUD.Show(Agent) 
        else:
            #Not enough gold ?
            if:
                PlayerObj := player[Agent]
                CustomPlayer:Custom_Player = PlayersMap[PlayerObj]
                CustomPlayer.PlayerGold<CostBarrier9
            then:
                NotEnoughGoldHUD.Show(Agent)
            else:
                NewEggHUD.Show(Agent)
                Accolade_B6B10.Award(Agent)
                UnlockBarrierB9(Agent)#Action
                AddMaxLevelUnlocked(Agent)#Save
                DecreaseGoldFunc(Agent,CostBarrier9)#Gold Save+UI

    Buy_B10(Agent:agent):void=
        #Already purchased?
        if:
            PlayerObj := player[Agent]
            CurrentPlayerStats := PlayerStatsManager.GetPlayerStats[Agent]
            CurrentPlayerStats.LevelUnlocked.CurrentValue>=10
        then:
            AlreadyPurchasedHUD.Show(Agent) 
        else:
            #Not enough gold ?
            if:
                PlayerObj := player[Agent]
                CustomPlayer:Custom_Player = PlayersMap[PlayerObj]
                CustomPlayer.PlayerGold<CostBarrier10
            then:
                NotEnoughGoldHUD.Show(Agent)
            else:
                NewEggHUD.Show(Agent)
                Accolade_B6B10.Award(Agent)
                UnlockBarrierB10(Agent)#Action
                AddMaxLevelUnlocked(Agent)#Save
                DecreaseGoldFunc(Agent,CostBarrier10)#Gold Save+UI

    Buy_B11(Agent:agent):void=
        #Already purchased?
        if:
            PlayerObj := player[Agent]
            CurrentPlayerStats := PlayerStatsManager.GetPlayerStats[Agent]
            CurrentPlayerStats.LevelUnlocked.CurrentValue>=11
        then:
            AlreadyPurchasedHUD.Show(Agent) 
        else:
            #Not enough gold ?
            if:
                PlayerObj := player[Agent]
                CustomPlayer:Custom_Player = PlayersMap[PlayerObj]
                CustomPlayer.PlayerGold<CostBarrier11
            then:
                NotEnoughGoldHUD.Show(Agent)
            else:
                NewEggHUD.Show(Agent)
                Accolade_B11B15.Award(Agent)
                UnlockBarrierB11(Agent)#Action
                AddMaxLevelUnlocked(Agent)#Save
                DecreaseGoldFunc(Agent,CostBarrier11)#Gold Save+UI

    Buy_B12(Agent:agent):void=
        #Already purchased?
        if:
            PlayerObj := player[Agent]
            CurrentPlayerStats := PlayerStatsManager.GetPlayerStats[Agent]
            CurrentPlayerStats.LevelUnlocked.CurrentValue>=12
        then:
            AlreadyPurchasedHUD.Show(Agent) 
        else:
            #Not enough gold ?
            if:
                PlayerObj := player[Agent]
                CustomPlayer:Custom_Player = PlayersMap[PlayerObj]
                CustomPlayer.PlayerGold<CostBarrier12
            then:
                NotEnoughGoldHUD.Show(Agent)
            else:
                NewEggHUD.Show(Agent)
                Accolade_B11B15.Award(Agent)
                UnlockBarrierB12(Agent)#Action
                AddMaxLevelUnlocked(Agent)#Save
                DecreaseGoldFunc(Agent,CostBarrier12)#Gold Save+UI

    Buy_B13(Agent:agent):void=
        #Already purchased?
        if:
            PlayerObj := player[Agent]
            CurrentPlayerStats := PlayerStatsManager.GetPlayerStats[Agent]
            CurrentPlayerStats.LevelUnlocked.CurrentValue>=13
        then:
            AlreadyPurchasedHUD.Show(Agent) 
        else:
            #Not enough gold ?
            if:
                PlayerObj := player[Agent]
                CustomPlayer:Custom_Player = PlayersMap[PlayerObj]
                CustomPlayer.PlayerGold<CostBarrier13
            then:
                NotEnoughGoldHUD.Show(Agent)
            else:
                NewEggHUD.Show(Agent)
                Accolade_B11B15.Award(Agent)
                UnlockBarrierB13(Agent)#Action
                AddMaxLevelUnlocked(Agent)#Save
                DecreaseGoldFunc(Agent,CostBarrier13)#Gold Save+UI

    Buy_B14(Agent:agent):void=
        #Already purchased?
        if:
            PlayerObj := player[Agent]
            CurrentPlayerStats := PlayerStatsManager.GetPlayerStats[Agent]
            CurrentPlayerStats.LevelUnlocked.CurrentValue>=14
        then:
            AlreadyPurchasedHUD.Show(Agent) 
        else:
            #Not enough gold ?
            if:
                PlayerObj := player[Agent]
                CustomPlayer:Custom_Player = PlayersMap[PlayerObj]
                CustomPlayer.PlayerGold<CostBarrier14
            then:
                NotEnoughGoldHUD.Show(Agent)
            else:
                NewEggHUD.Show(Agent)
                Accolade_B11B15.Award(Agent)
                UnlockBarrierB14(Agent)#Action
                AddMaxLevelUnlocked(Agent)#Save
                DecreaseGoldFunc(Agent,CostBarrier14)#Gold Save+UI

    Buy_B15(Agent:agent):void=
        #Already purchased?
        if:
            PlayerObj := player[Agent]
            CurrentPlayerStats := PlayerStatsManager.GetPlayerStats[Agent]
            CurrentPlayerStats.LevelUnlocked.CurrentValue>=15
        then:
            AlreadyPurchasedHUD.Show(Agent) 
        else:
            #Not enough gold ?
            if:
                PlayerObj := player[Agent]
                CustomPlayer:Custom_Player = PlayersMap[PlayerObj]
                CustomPlayer.PlayerGold<CostBarrier15
            then:
                NotEnoughGoldHUD.Show(Agent)
            else:
                NewEggHUD.Show(Agent)
                Accolade_B11B15.Award(Agent)
                UnlockBarrierB15(Agent)#Action
                AddMaxLevelUnlocked(Agent)#Save
                DecreaseGoldFunc(Agent,CostBarrier15)#Gold Save+UI

    Buy_B16(Agent:agent):void=
        #Already purchased?
        if:
            PlayerObj := player[Agent]
            CurrentPlayerStats := PlayerStatsManager.GetPlayerStats[Agent]
            CurrentPlayerStats.LevelUnlocked.CurrentValue>=16
        then:
            AlreadyPurchasedHUD.Show(Agent) 
        else:
            #Not enough gold ?
            if:
                PlayerObj := player[Agent]
                CustomPlayer:Custom_Player = PlayersMap[PlayerObj]
                CustomPlayer.PlayerGold<CostBarrier16
            then:
                NotEnoughGoldHUD.Show(Agent)
            else:
                NewEggHUD.Show(Agent)
                Accolade_B16B20.Award(Agent)
                UnlockBarrierB16(Agent)#Action
                AddMaxLevelUnlocked(Agent)#Save
                DecreaseGoldFunc(Agent,CostBarrier16)#Gold Save+UI

    Buy_B17(Agent:agent):void=
        #Already purchased?
        if:
            PlayerObj := player[Agent]
            CurrentPlayerStats := PlayerStatsManager.GetPlayerStats[Agent]
            CurrentPlayerStats.LevelUnlocked.CurrentValue>=17
        then:
            AlreadyPurchasedHUD.Show(Agent) 
        else:
            #Not enough gold ?
            if:
                PlayerObj := player[Agent]
                CustomPlayer:Custom_Player = PlayersMap[PlayerObj]
                CustomPlayer.PlayerGold<CostBarrier17
            then:
                NotEnoughGoldHUD.Show(Agent)
            else:
                NewEggHUD.Show(Agent)
                Accolade_B16B20.Award(Agent)
                UnlockBarrierB17(Agent)#Action
                AddMaxLevelUnlocked(Agent)#Save
                DecreaseGoldFunc(Agent,CostBarrier17)#Gold Save+UI

    Buy_B18(Agent:agent):void=
        #Already purchased?
        if:
            PlayerObj := player[Agent]
            CurrentPlayerStats := PlayerStatsManager.GetPlayerStats[Agent]
            CurrentPlayerStats.LevelUnlocked.CurrentValue>=18
        then:
            AlreadyPurchasedHUD.Show(Agent) 
        else:
            #Not enough gold ?
            if:
                PlayerObj := player[Agent]
                CustomPlayer:Custom_Player = PlayersMap[PlayerObj]
                CustomPlayer.PlayerGold<CostBarrier18
            then:
                NotEnoughGoldHUD.Show(Agent)
            else:
                NewEggHUD.Show(Agent)
                Accolade_B16B20.Award(Agent)
                UnlockBarrierB18(Agent)#Action
                AddMaxLevelUnlocked(Agent)#Save
                DecreaseGoldFunc(Agent,CostBarrier18)#Gold Save+UI

    Buy_B19(Agent:agent):void=
        #Already purchased?
        if:
            PlayerObj := player[Agent]
            CurrentPlayerStats := PlayerStatsManager.GetPlayerStats[Agent]
            CurrentPlayerStats.LevelUnlocked.CurrentValue>=19
        then:
            AlreadyPurchasedHUD.Show(Agent) 
        else:
            #Not enough gold ?
            if:
                PlayerObj := player[Agent]
                CustomPlayer:Custom_Player = PlayersMap[PlayerObj]
                CustomPlayer.PlayerGold<CostBarrier19
            then:
                NotEnoughGoldHUD.Show(Agent)
            else:
                NewEggHUD.Show(Agent)
                Accolade_B16B20.Award(Agent)
                UnlockBarrierB19(Agent)#Action
                AddMaxLevelUnlocked(Agent)#Save
                DecreaseGoldFunc(Agent,CostBarrier19)#Gold Save+UI

    Buy_B20(Agent:agent):void=
        #Already purchased?
        if:
            PlayerObj := player[Agent]
            CurrentPlayerStats := PlayerStatsManager.GetPlayerStats[Agent]
            CurrentPlayerStats.LevelUnlocked.CurrentValue>=20
        then:
            AlreadyPurchasedHUD.Show(Agent) 
        else:
            #Not enough gold ?
            if:
                PlayerObj := player[Agent]
                CustomPlayer:Custom_Player = PlayersMap[PlayerObj]
                CustomPlayer.PlayerGold<CostBarrier20
            then:
                NotEnoughGoldHUD.Show(Agent)
            else:
                Accolade_B16B20.Award(Agent)
                UnlockBarrierB20(Agent)#Action
                AddMaxLevelUnlocked(Agent)#Save
                DecreaseGoldFunc(Agent,CostBarrier20)#Gold Save+UI

    Buy_B21(Agent:agent):void=
        #Already purchased?
        if:
            PlayerObj := player[Agent]
            CurrentPlayerStats := PlayerStatsManager.GetPlayerStats[Agent]
            CurrentPlayerStats.LevelUnlocked.CurrentValue>=21
        then:
            AlreadyPurchasedHUD.Show(Agent) 
        else:
            #Not enough gold ?
            if:
                PlayerObj := player[Agent]
                CustomPlayer:Custom_Player = PlayersMap[PlayerObj]
                CustomPlayer.PlayerGold<CostBarrier21
            then:
                NotEnoughGoldHUD.Show(Agent)
            else:
                Accolade_B21B25.Award(Agent)
                UnlockBarrierB21(Agent)#Action
                AddMaxLevelUnlocked(Agent)#Save
                DecreaseGoldFunc(Agent,CostBarrier21)#Gold Save+UI

    Buy_B22(Agent:agent):void=
        #Already purchased?
        if:
            PlayerObj := player[Agent]
            CurrentPlayerStats := PlayerStatsManager.GetPlayerStats[Agent]
            CurrentPlayerStats.LevelUnlocked.CurrentValue>=22
        then:
            AlreadyPurchasedHUD.Show(Agent) 
        else:
            #Not enough gold ?
            if:
                PlayerObj := player[Agent]
                CustomPlayer:Custom_Player = PlayersMap[PlayerObj]
                CustomPlayer.PlayerGold<CostBarrier22
            then:
                NotEnoughGoldHUD.Show(Agent)
            else:
                Accolade_B21B25.Award(Agent)
                UnlockBarrierB22(Agent)#Action
                AddMaxLevelUnlocked(Agent)#Save
                DecreaseGoldFunc(Agent,CostBarrier22)#Gold Save+UI

    Buy_B23(Agent:agent):void=
        #Already purchased?
        if:
            PlayerObj := player[Agent]
            CurrentPlayerStats := PlayerStatsManager.GetPlayerStats[Agent]
            CurrentPlayerStats.LevelUnlocked.CurrentValue>=23
        then:
            AlreadyPurchasedHUD.Show(Agent) 
        else:
            #Not enough gold ?
            if:
                PlayerObj := player[Agent]
                CustomPlayer:Custom_Player = PlayersMap[PlayerObj]
                CustomPlayer.PlayerGold<CostBarrier23
            then:
                NotEnoughGoldHUD.Show(Agent)
            else:
                Accolade_B21B25.Award(Agent)
                UnlockBarrierB23(Agent)#Action
                AddMaxLevelUnlocked(Agent)#Save
                DecreaseGoldFunc(Agent,CostBarrier23)#Gold Save+UI

    Buy_B24(Agent:agent):void=
        #Already purchased?
        if:
            PlayerObj := player[Agent]
            CurrentPlayerStats := PlayerStatsManager.GetPlayerStats[Agent]
            CurrentPlayerStats.LevelUnlocked.CurrentValue>=24
        then:
            AlreadyPurchasedHUD.Show(Agent) 
        else:
            #Not enough gold ?
            if:
                PlayerObj := player[Agent]
                CustomPlayer:Custom_Player = PlayersMap[PlayerObj]
                CustomPlayer.PlayerGold<CostBarrier24
            then:
                NotEnoughGoldHUD.Show(Agent)
            else:
                Accolade_B21B25.Award(Agent)
                UnlockBarrierB24(Agent)#Action
                AddMaxLevelUnlocked(Agent)#Save
                DecreaseGoldFunc(Agent,CostBarrier24)#Gold Save+UI

    Buy_B25(Agent:agent):void=
        #Already purchased?
        if:
            PlayerObj := player[Agent]
            CurrentPlayerStats := PlayerStatsManager.GetPlayerStats[Agent]
            CurrentPlayerStats.LevelUnlocked.CurrentValue>=25
        then:
            AlreadyPurchasedHUD.Show(Agent) 
        else:
            #Not enough gold ?
            if:
                PlayerObj := player[Agent]
                CustomPlayer:Custom_Player = PlayersMap[PlayerObj]
                CustomPlayer.PlayerGold<CostBarrier25
            then:
                NotEnoughGoldHUD.Show(Agent)
            else:
                Accolade_B21B25.Award(Agent)
                UnlockBarrierB25(Agent)#Action
                AddMaxLevelUnlocked(Agent)#Save
                DecreaseGoldFunc(Agent,CostBarrier25)#Gold Save+UI

    Buy_B26(Agent:agent):void=
        #Already purchased?
        if:
            PlayerObj := player[Agent]
            CurrentPlayerStats := PlayerStatsManager.GetPlayerStats[Agent]
            CurrentPlayerStats.LevelUnlocked.CurrentValue>=26
        then:
            AlreadyPurchasedHUD.Show(Agent) 
        else:
            #Not enough gold ?
            if:
                PlayerObj := player[Agent]
                CustomPlayer:Custom_Player = PlayersMap[PlayerObj]
                CustomPlayer.PlayerGold<CostBarrier26
            then:
                NotEnoughGoldHUD.Show(Agent)
            else:
                Accolade_B26B30.Award(Agent)
                UnlockBarrierB26(Agent)#Action
                AddMaxLevelUnlocked(Agent)#Save
                DecreaseGoldFunc(Agent,CostBarrier26)#Gold Save+UI

    Buy_B27(Agent:agent):void=
        #Already purchased?
        if:
            PlayerObj := player[Agent]
            CurrentPlayerStats := PlayerStatsManager.GetPlayerStats[Agent]
            CurrentPlayerStats.LevelUnlocked.CurrentValue>=27
        then:
            AlreadyPurchasedHUD.Show(Agent) 
        else:
            #Not enough gold ?
            if:
                PlayerObj := player[Agent]
                CustomPlayer:Custom_Player = PlayersMap[PlayerObj]
                CustomPlayer.PlayerGold<CostBarrier27
            then:
                NotEnoughGoldHUD.Show(Agent)
            else:
                Accolade_B26B30.Award(Agent)
                UnlockBarrierB27(Agent)#Action
                AddMaxLevelUnlocked(Agent)#Save
                DecreaseGoldFunc(Agent,CostBarrier27)#Gold Save+UI

    Buy_B28(Agent:agent):void=
        #Already purchased?
        if:
            PlayerObj := player[Agent]
            CurrentPlayerStats := PlayerStatsManager.GetPlayerStats[Agent]
            CurrentPlayerStats.LevelUnlocked.CurrentValue>=28
        then:
            AlreadyPurchasedHUD.Show(Agent) 
        else:
            #Not enough gold ?
            if:
                PlayerObj := player[Agent]
                CustomPlayer:Custom_Player = PlayersMap[PlayerObj]
                CustomPlayer.PlayerGold<CostBarrier28
            then:
                NotEnoughGoldHUD.Show(Agent)
            else:
                Accolade_B26B30.Award(Agent)
                UnlockBarrierB28(Agent)#Action
                AddMaxLevelUnlocked(Agent)#Save
                DecreaseGoldFunc(Agent,CostBarrier28)#Gold Save+UI

    Buy_B29(Agent:agent):void=
        #Already purchased?
        if:
            PlayerObj := player[Agent]
            CurrentPlayerStats := PlayerStatsManager.GetPlayerStats[Agent]
            CurrentPlayerStats.LevelUnlocked.CurrentValue>=29
        then:
            AlreadyPurchasedHUD.Show(Agent) 
        else:
            #Not enough gold ?
            if:
                PlayerObj := player[Agent]
                CustomPlayer:Custom_Player = PlayersMap[PlayerObj]
                CustomPlayer.PlayerGold<CostBarrier29
            then:
                NotEnoughGoldHUD.Show(Agent)
            else:
                Accolade_B26B30.Award(Agent)
                UnlockBarrierB29(Agent)#Action
                AddMaxLevelUnlocked(Agent)#Save
                DecreaseGoldFunc(Agent,CostBarrier29)#Gold Save+UI

    Buy_B30(Agent:agent):void=
        #Already purchased?
        if:
            PlayerObj := player[Agent]
            CurrentPlayerStats := PlayerStatsManager.GetPlayerStats[Agent]
            CurrentPlayerStats.LevelUnlocked.CurrentValue>=30
        then:
            AlreadyPurchasedHUD.Show(Agent) 
        else:
            #Not enough gold ?
            if:
                PlayerObj := player[Agent]
                CustomPlayer:Custom_Player = PlayersMap[PlayerObj]
                CustomPlayer.PlayerGold<CostBarrier30
            then:
                NotEnoughGoldHUD.Show(Agent)
            else:
                Accolade_B26B30.Award(Agent)
                UnlockBarrierB30(Agent)#Action
                AddMaxLevelUnlocked(Agent)#Save
                DecreaseGoldFunc(Agent,CostBarrier30)#Gold Save+UI
                
    #Crit Purchase buttons
    UpgradeCrit_1(Agent:agent):void=
        #Already purchased?
        if:
            PlayerObj := player[Agent]
            CustomPlayer:Custom_Player = PlayersMap[PlayerObj]
            CustomPlayer.Pet_Crit_Value >=5
        then:
            AlreadyPurchasedHUD.Show(Agent) 
        else:
            #Previous Level bought?
            if:
                PlayerObj := player[Agent]
                CustomPlayer:Custom_Player = PlayersMap[PlayerObj]
                CustomPlayer.Pet_Crit_Value <> 0
            then:
                RequirePreviousLevelHUD.Show(Agent) 
            else:
                #Not enough gold ?
                if:
                    PlayerObj := player[Agent]
                    CustomPlayer:Custom_Player = PlayersMap[PlayerObj]
                    CustomPlayer.PlayerGold<CostCrit1
                then:
                    NotEnoughGoldHUD.Show(Agent)

        if:
            PlayerObj := player[Agent]
            CustomPlayer:Custom_Player = PlayersMap[PlayerObj]
            CurrentPlayerStats := PlayerStatsManager.GetPlayerStats[Agent]
            CustomPlayer.Pet_Crit_Value = 0
            CustomPlayer.PlayerGold>=CostCrit1
        then:
            Upgrade_Crit1_HUD.Show(Agent)
            CustomPlayer.SetCritBy(5)
            SetCrit(Agent,1)
            DecreaseGoldFunc(Agent,CostCrit1)
            Accolade_Upgrades.Award(Agent)
 

    UpgradeCrit_2(Agent:agent):void=
        #Already purchased?
        if:
            PlayerObj := player[Agent]
            CustomPlayer:Custom_Player = PlayersMap[PlayerObj]
            CustomPlayer.Pet_Crit_Value >=10
        then:
            AlreadyPurchasedHUD.Show(Agent) 
        else:
            #Previous Level bought?
            if:
                PlayerObj := player[Agent]
                CustomPlayer:Custom_Player = PlayersMap[PlayerObj]
                CustomPlayer.Pet_Crit_Value <> 5
            then:
                RequirePreviousLevelHUD.Show(Agent) 
            else:
                #Not enough gold ?
                if:
                    PlayerObj := player[Agent]
                    CustomPlayer:Custom_Player = PlayersMap[PlayerObj]
                    CustomPlayer.PlayerGold<CostCrit2
                then:
                    NotEnoughGoldHUD.Show(Agent)

        if:
            PlayerObj := player[Agent]
            CustomPlayer:Custom_Player = PlayersMap[PlayerObj]
            CurrentPlayerStats := PlayerStatsManager.GetPlayerStats[Agent]
            CustomPlayer.Pet_Crit_Value = 5
            CustomPlayer.PlayerGold>=CostCrit2
        then:
            Upgrade_Crit2_HUD.Show(Agent)
            CustomPlayer.SetCritBy(10)
            SetCrit(Agent,2)
            DecreaseGoldFunc(Agent,CostCrit2)
            Accolade_Upgrades.Award(Agent)

    UpgradeCrit_3(Agent:agent):void=
        #Already purchased?
        if:
            PlayerObj := player[Agent]
            CustomPlayer:Custom_Player = PlayersMap[PlayerObj]
            CustomPlayer.Pet_Crit_Value >=15
        then:
            AlreadyPurchasedHUD.Show(Agent) 
        else:
            #Previous Level bought?
            if:
                PlayerObj := player[Agent]
                CustomPlayer:Custom_Player = PlayersMap[PlayerObj]
                CustomPlayer.Pet_Crit_Value <> 10
            then:
                RequirePreviousLevelHUD.Show(Agent) 
            else:
                #Not enough gold ?
                if:
                    PlayerObj := player[Agent]
                    CustomPlayer:Custom_Player = PlayersMap[PlayerObj]
                    CustomPlayer.PlayerGold<CostCrit3
                then:
                    NotEnoughGoldHUD.Show(Agent)

        if:
            PlayerObj := player[Agent]
            CustomPlayer:Custom_Player = PlayersMap[PlayerObj]
            CurrentPlayerStats := PlayerStatsManager.GetPlayerStats[Agent]
            CustomPlayer.Pet_Crit_Value = 10
            CustomPlayer.PlayerGold>=CostCrit3
        then:
            Upgrade_Crit3_HUD.Show(Agent)
            CustomPlayer.SetCritBy(15)
            SetCrit(Agent,3)
            DecreaseGoldFunc(Agent,CostCrit3)
            Accolade_Upgrades.Award(Agent)

    UpgradeCrit_4(Agent:agent):void=
        #Already purchased?
        if:
            PlayerObj := player[Agent]
            CustomPlayer:Custom_Player = PlayersMap[PlayerObj]
            CustomPlayer.Pet_Crit_Value >=20
        then:
            AlreadyPurchasedHUD.Show(Agent) 
        else:
            #Previous Level bought?
            if:
                PlayerObj := player[Agent]
                CustomPlayer:Custom_Player = PlayersMap[PlayerObj]
                CustomPlayer.Pet_Crit_Value <> 15
            then:
                RequirePreviousLevelHUD.Show(Agent) 
            else:
                #Not enough gold ?
                if:
                    PlayerObj := player[Agent]
                    CustomPlayer:Custom_Player = PlayersMap[PlayerObj]
                    CustomPlayer.PlayerGold<CostCrit4
                then:
                    NotEnoughGoldHUD.Show(Agent)

        if:
            PlayerObj := player[Agent]
            CustomPlayer:Custom_Player = PlayersMap[PlayerObj]
            CurrentPlayerStats := PlayerStatsManager.GetPlayerStats[Agent]
            CustomPlayer.Pet_Crit_Value = 15
            CustomPlayer.PlayerGold>=CostCrit4
        then:
            Upgrade_Crit4_HUD.Show(Agent)
            CustomPlayer.SetCritBy(20)
            SetCrit(Agent,4)
            DecreaseGoldFunc(Agent,CostCrit4)
            Accolade_Upgrades.Award(Agent)

    UpgradeCrit_5(Agent:agent):void=
        #Already purchased?
        if:
            PlayerObj := player[Agent]
            CustomPlayer:Custom_Player = PlayersMap[PlayerObj]
            CustomPlayer.Pet_Crit_Value >=25
        then:
            AlreadyPurchasedHUD.Show(Agent) 
        else:
            #Previous Level bought?
            if:
                PlayerObj := player[Agent]
                CustomPlayer:Custom_Player = PlayersMap[PlayerObj]
                CustomPlayer.Pet_Crit_Value <> 20
            then:
                RequirePreviousLevelHUD.Show(Agent) 
            else:
                #Not enough gold ?
                if:
                    PlayerObj := player[Agent]
                    CustomPlayer:Custom_Player = PlayersMap[PlayerObj]
                    CustomPlayer.PlayerGold<CostCrit5
                then:
                    NotEnoughGoldHUD.Show(Agent)

        if:
            PlayerObj := player[Agent]
            CustomPlayer:Custom_Player = PlayersMap[PlayerObj]
            CurrentPlayerStats := PlayerStatsManager.GetPlayerStats[Agent]
            CustomPlayer.Pet_Crit_Value = 20
            CustomPlayer.PlayerGold>=CostCrit5
        then:
            Upgrade_Crit5_HUD.Show(Agent)
            CustomPlayer.SetCritBy(25)
            SetCrit(Agent,5)
            DecreaseGoldFunc(Agent,CostCrit5)
            Accolade_Upgrades.Award(Agent)
    
    #Speed Purchase buttons
    UpgradeSpeed_1(Agent:agent):void=
        #Already purchased?
        if:
            PlayerObj := player[Agent]
            CustomPlayer:Custom_Player = PlayersMap[PlayerObj]
            CustomPlayer.Pet_Speed_Value >=1.2
        then:
            AlreadyPurchasedHUD.Show(Agent) 
        else:
            #Previous Level bought?
            if:
                PlayerObj := player[Agent]
                CustomPlayer:Custom_Player = PlayersMap[PlayerObj]
                CustomPlayer.Pet_Speed_Value <> 1.0
            then:
                RequirePreviousLevelHUD.Show(Agent) 
            else:
                #Not enough gold ?
                if:
                    PlayerObj := player[Agent]
                    CustomPlayer:Custom_Player = PlayersMap[PlayerObj]
                    CustomPlayer.PlayerGold<CostSpeed1
                then:
                    NotEnoughGoldHUD.Show(Agent)

        if:
            PlayerObj := player[Agent]
            CustomPlayer:Custom_Player = PlayersMap[PlayerObj]
            CustomPlayer.Pet_Speed_Value = 1.0
            CustomPlayer.PlayerGold>=CostSpeed1
        then:
            Upgrade_Speed1_HUD.Show(Agent)
            CustomPlayer.SetSpeedBy(1.2)
            SetSpeed(Agent,1)
            DecreaseGoldFunc(Agent,CostSpeed1)
            Accolade_Upgrades.Award(Agent)

    UpgradeSpeed_2(Agent:agent):void=
        #Already purchased?
        if:
            PlayerObj := player[Agent]
            CustomPlayer:Custom_Player = PlayersMap[PlayerObj]
            CustomPlayer.Pet_Speed_Value >=1.3
        then:
            AlreadyPurchasedHUD.Show(Agent) 
        else:
            #Previous Level bought?
            if:
                PlayerObj := player[Agent]
                CustomPlayer:Custom_Player = PlayersMap[PlayerObj]
                CustomPlayer.Pet_Speed_Value <> 1.2
            then:
                RequirePreviousLevelHUD.Show(Agent) 
            else:
                #Not enough gold ?
                if:
                    PlayerObj := player[Agent]
                    CustomPlayer:Custom_Player = PlayersMap[PlayerObj]
                    CustomPlayer.PlayerGold<CostSpeed2
                then:
                    NotEnoughGoldHUD.Show(Agent)
        if:
            PlayerObj := player[Agent]
            CustomPlayer:Custom_Player = PlayersMap[PlayerObj]
            CustomPlayer.Pet_Speed_Value = 1.2
            CustomPlayer.PlayerGold>=CostSpeed2
        then:
            Upgrade_Speed2_HUD.Show(Agent)
            CustomPlayer.SetSpeedBy(1.3)
            SetSpeed(Agent,2)
            DecreaseGoldFunc(Agent,CostSpeed2)
            Accolade_Upgrades.Award(Agent)

    UpgradeSpeed_3(Agent:agent):void=
        #Already purchased?
        if:
            PlayerObj := player[Agent]
            CustomPlayer:Custom_Player = PlayersMap[PlayerObj]
            CustomPlayer.Pet_Speed_Value >=1.5
        then:
            AlreadyPurchasedHUD.Show(Agent) 
        else:
            #Previous Level bought?
            if:
                PlayerObj := player[Agent]
                CustomPlayer:Custom_Player = PlayersMap[PlayerObj]
                CustomPlayer.Pet_Speed_Value <> 1.3
            then:
                RequirePreviousLevelHUD.Show(Agent) 
            else:
                #Not enough gold ?
                if:
                    PlayerObj := player[Agent]
                    CustomPlayer:Custom_Player = PlayersMap[PlayerObj]
                    CustomPlayer.PlayerGold<CostSpeed3
                then:
                    NotEnoughGoldHUD.Show(Agent)

        if:
            PlayerObj := player[Agent]
            CustomPlayer:Custom_Player = PlayersMap[PlayerObj]
            CustomPlayer.Pet_Speed_Value = 1.3
            CustomPlayer.PlayerGold>=CostSpeed3
        then:
            Upgrade_Speed3_HUD.Show(Agent)
            CustomPlayer.SetSpeedBy(1.5)
            SetSpeed(Agent,3)
            DecreaseGoldFunc(Agent,CostSpeed3)
            Accolade_Upgrades.Award(Agent)

    UpgradeSpeed_4(Agent:agent):void=
        #Already purchased?
        if:
            PlayerObj := player[Agent]
            CustomPlayer:Custom_Player = PlayersMap[PlayerObj]
            CustomPlayer.Pet_Speed_Value >=2.0
        then:
            AlreadyPurchasedHUD.Show(Agent) 
        else:
            #Previous Level bought?
            if:
                PlayerObj := player[Agent]
                CustomPlayer:Custom_Player = PlayersMap[PlayerObj]
                CustomPlayer.Pet_Speed_Value <> 1.5
            then:
                RequirePreviousLevelHUD.Show(Agent) 
            else:
                #Not enough gold ?
                if:
                    PlayerObj := player[Agent]
                    CustomPlayer:Custom_Player = PlayersMap[PlayerObj]
                    CustomPlayer.PlayerGold<CostSpeed4
                then:
                    NotEnoughGoldHUD.Show(Agent)

        if:
            PlayerObj := player[Agent]
            CustomPlayer:Custom_Player = PlayersMap[PlayerObj]
            CustomPlayer.Pet_Speed_Value = 1.5
            CustomPlayer.PlayerGold>=CostSpeed4
        then:
            Upgrade_Speed4_HUD.Show(Agent)
            CustomPlayer.SetSpeedBy(2.0)
            SetSpeed(Agent,4)
            DecreaseGoldFunc(Agent,CostSpeed4)
            Accolade_Upgrades.Award(Agent)

    UpgradeSpeed_5(Agent:agent):void=
        #Already purchased?
        if:
            PlayerObj := player[Agent]
            CustomPlayer:Custom_Player = PlayersMap[PlayerObj]
            CustomPlayer.Pet_Speed_Value >=2.5
        then:
            AlreadyPurchasedHUD.Show(Agent) 
        else:
            #Previous Level bought?
            if:
                PlayerObj := player[Agent]
                CustomPlayer:Custom_Player = PlayersMap[PlayerObj]
                CustomPlayer.Pet_Speed_Value <> 2.0
            then:
                RequirePreviousLevelHUD.Show(Agent) 
            else:
                #Not enough gold ?
                if:
                    PlayerObj := player[Agent]
                    CustomPlayer:Custom_Player = PlayersMap[PlayerObj]
                    CustomPlayer.PlayerGold<CostSpeed5
                then:
                    NotEnoughGoldHUD.Show(Agent)

        if:
            PlayerObj := player[Agent]
            CustomPlayer:Custom_Player = PlayersMap[PlayerObj]
            CustomPlayer.Pet_Speed_Value = 2.0
            CustomPlayer.PlayerGold>=CostSpeed5
        then:
            Upgrade_Speed5_HUD.Show(Agent)
            CustomPlayer.SetSpeedBy(2.5)
            SetSpeed(Agent,5)
            DecreaseGoldFunc(Agent,CostSpeed5)
            Accolade_Upgrades.Award(Agent)

    #Player speed
    UpgradePlayerSpeed_1(Agent:agent):void=
        #Already purchased?
        if:
            PlayerObj := player[Agent]
            CustomPlayer:Custom_Player = PlayersMap[PlayerObj]
            CustomPlayer.Player_Speed_Value >=2.0
        then:
            AlreadyPurchasedHUD.Show(Agent) 
        else:
            #Previous Level bought?
            if:
                PlayerObj := player[Agent]
                CustomPlayer:Custom_Player = PlayersMap[PlayerObj]
                CustomPlayer.Player_Speed_Value <> 1.0
            then:
                RequirePreviousLevelHUD.Show(Agent) 
            else:
                #Not enough gold ?
                if:
                    PlayerObj := player[Agent]
                    CustomPlayer:Custom_Player = PlayersMap[PlayerObj]
                    CustomPlayer.PlayerGold<CostPlayerSpeed1
                then:
                    NotEnoughGoldHUD.Show(Agent)

        if:
            PlayerObj := player[Agent]
            CustomPlayer:Custom_Player = PlayersMap[PlayerObj]
            CustomPlayer.Player_Speed_Value = 1.0
            CustomPlayer.PlayerGold>=CostPlayerSpeed1
        then:
            ClassSelector1.ChangeClass(Agent)
            Upgrade_PlayerSpeed1_HUD.Show(Agent)
            CustomPlayer.SetPlayerSpeedBy(2.0)
            SetPlayerSpeed(Agent,1)
            DecreaseGoldFunc(Agent,CostPlayerSpeed1)
            Accolade_Upgrades.Award(Agent)

    UpgradePlayerSpeed_2(Agent:agent):void=
        #Already purchased?
        if:
            PlayerObj := player[Agent]
            CustomPlayer:Custom_Player = PlayersMap[PlayerObj]
            CustomPlayer.Player_Speed_Value >=3.0
        then:
            AlreadyPurchasedHUD.Show(Agent) 
        else:
            #Previous Level bought?
            if:
                PlayerObj := player[Agent]
                CustomPlayer:Custom_Player = PlayersMap[PlayerObj]
                CustomPlayer.Player_Speed_Value <> 2.0
            then:
                RequirePreviousLevelHUD.Show(Agent) 
            else:
                #Not enough gold ?
                if:
                    PlayerObj := player[Agent]
                    CustomPlayer:Custom_Player = PlayersMap[PlayerObj]
                    CustomPlayer.PlayerGold<CostPlayerSpeed2
                then:
                    NotEnoughGoldHUD.Show(Agent)

        if:
            PlayerObj := player[Agent]
            CustomPlayer:Custom_Player = PlayersMap[PlayerObj]
            CustomPlayer.Player_Speed_Value = 2.0
            CustomPlayer.PlayerGold>=CostPlayerSpeed2
        then:
            ClassSelector2.ChangeClass(Agent)
            Upgrade_PlayerSpeed2_HUD.Show(Agent)
            CustomPlayer.SetPlayerSpeedBy(3.0)
            SetPlayerSpeed(Agent,2)
            DecreaseGoldFunc(Agent,CostPlayerSpeed2)
            Accolade_Upgrades.Award(Agent)

    UpgradePlayerSpeed_3(Agent:agent):void=
        #Already purchased?
        if:
            PlayerObj := player[Agent]
            CustomPlayer:Custom_Player = PlayersMap[PlayerObj]
            CustomPlayer.Player_Speed_Value >=4.0
        then:
            AlreadyPurchasedHUD.Show(Agent) 
        else:
            #Previous Level bought?
            if:
                PlayerObj := player[Agent]
                CustomPlayer:Custom_Player = PlayersMap[PlayerObj]
                CustomPlayer.Player_Speed_Value <> 3.0
            then:
                RequirePreviousLevelHUD.Show(Agent) 
            else:
                #Not enough gold ?
                if:
                    PlayerObj := player[Agent]
                    CustomPlayer:Custom_Player = PlayersMap[PlayerObj]
                    CustomPlayer.PlayerGold<CostPlayerSpeed3
                then:
                    NotEnoughGoldHUD.Show(Agent)

        if:
            PlayerObj := player[Agent]
            CustomPlayer:Custom_Player = PlayersMap[PlayerObj]
            CustomPlayer.Player_Speed_Value = 3.0
            CustomPlayer.PlayerGold>=CostPlayerSpeed3
        then:
            ClassSelector3.ChangeClass(Agent)
            Upgrade_PlayerSpeed3_HUD.Show(Agent)
            CustomPlayer.SetPlayerSpeedBy(4.0)
            SetPlayerSpeed(Agent,3)
            DecreaseGoldFunc(Agent,CostPlayerSpeed3)
            Accolade_Upgrades.Award(Agent)

    UpgradePlayerSpeed_4(Agent:agent):void=
        #Already purchased?
        if:
            PlayerObj := player[Agent]
            CustomPlayer:Custom_Player = PlayersMap[PlayerObj]
            CustomPlayer.Player_Speed_Value >=5.0
        then:
            AlreadyPurchasedHUD.Show(Agent) 
        else:
            #Previous Level bought?
            if:
                PlayerObj := player[Agent]
                CustomPlayer:Custom_Player = PlayersMap[PlayerObj]
                CustomPlayer.Player_Speed_Value <> 4.0
            then:
                RequirePreviousLevelHUD.Show(Agent) 
            else:
                #Not enough gold ?
                if:
                    PlayerObj := player[Agent]
                    CustomPlayer:Custom_Player = PlayersMap[PlayerObj]
                    CustomPlayer.PlayerGold<CostPlayerSpeed4
                then:
                    NotEnoughGoldHUD.Show(Agent)

        if:
            PlayerObj := player[Agent]
            CustomPlayer:Custom_Player = PlayersMap[PlayerObj]
            CustomPlayer.Player_Speed_Value = 4.0
            CustomPlayer.PlayerGold>=CostPlayerSpeed4
        then:
            ClassSelector4.ChangeClass(Agent)
            Upgrade_PlayerSpeed4_HUD.Show(Agent)
            CustomPlayer.SetPlayerSpeedBy(5.0)
            SetPlayerSpeed(Agent,4)
            DecreaseGoldFunc(Agent,CostPlayerSpeed4)
            Accolade_Upgrades.Award(Agent)

    UpgradePlayerSpeed_5(Agent:agent):void=
        #Already purchased?
        if:
            PlayerObj := player[Agent]
            CustomPlayer:Custom_Player = PlayersMap[PlayerObj]
            CustomPlayer.Player_Speed_Value >=6.0
        then:
            AlreadyPurchasedHUD.Show(Agent) 
        else:
            #Previous Level bought?
            if:
                PlayerObj := player[Agent]
                CustomPlayer:Custom_Player = PlayersMap[PlayerObj]
                CustomPlayer.Player_Speed_Value <> 5.0
            then:
                RequirePreviousLevelHUD.Show(Agent) 
            else:
                #Not enough gold ?
                if:
                    PlayerObj := player[Agent]
                    CustomPlayer:Custom_Player = PlayersMap[PlayerObj]
                    CustomPlayer.PlayerGold<CostPlayerSpeed5
                then:
                    NotEnoughGoldHUD.Show(Agent)

        if:
            PlayerObj := player[Agent]
            CustomPlayer:Custom_Player = PlayersMap[PlayerObj]
            CustomPlayer.Player_Speed_Value = 5.0
            CustomPlayer.PlayerGold>=CostPlayerSpeed5
        then:
            ClassSelector5.ChangeClass(Agent)
            Upgrade_PlayerSpeed5_HUD.Show(Agent)
            CustomPlayer.SetPlayerSpeedBy(6.0)
            SetPlayerSpeed(Agent,5)
            DecreaseGoldFunc(Agent,CostPlayerSpeed5)
            Accolade_Upgrades.Award(Agent)

    #Luck Purchase buttons
    UpgradeLuck_1(Agent:agent):void=
        #Already purchased?
        if:
            PlayerObj := player[Agent]
            CustomPlayer:Custom_Player = PlayersMap[PlayerObj]
            CustomPlayer.Pet_Luck_Value >=1
        then:
            AlreadyPurchasedHUD.Show(Agent) 
        else:
            #Previous Level bought?
            if:
                PlayerObj := player[Agent]
                CustomPlayer:Custom_Player = PlayersMap[PlayerObj]
                CustomPlayer.Pet_Luck_Value <> 0
            then:
                RequirePreviousLevelHUD.Show(Agent) 
            else:
                #Not enough gold ?
                if:
                    PlayerObj := player[Agent]
                    CustomPlayer:Custom_Player = PlayersMap[PlayerObj]
                    CustomPlayer.PlayerGold<CostLuck1
                then:
                    NotEnoughGoldHUD.Show(Agent)

        if:
            PlayerObj := player[Agent]
            CustomPlayer:Custom_Player = PlayersMap[PlayerObj]
            CustomPlayer.Pet_Luck_Value = 0
            CustomPlayer.PlayerGold>=CostLuck1
        then:
            Upgrade_Luck1_HUD.Show(Agent)
            CustomPlayer.SetLuckBy(1)
            SetLuck(Agent,1)
            DecreaseGoldFunc(Agent,CostLuck1)
            Accolade_Upgrades.Award(Agent)

    UpgradeLuck_2(Agent:agent):void=
        #Already purchased?
        if:
            PlayerObj := player[Agent]
            CustomPlayer:Custom_Player = PlayersMap[PlayerObj]
            CustomPlayer.Pet_Luck_Value >=2
        then:
            AlreadyPurchasedHUD.Show(Agent) 
        else:
            #Previous Level bought?
            if:
                PlayerObj := player[Agent]
                CustomPlayer:Custom_Player = PlayersMap[PlayerObj]
                CustomPlayer.Pet_Luck_Value <> 1
            then:
                RequirePreviousLevelHUD.Show(Agent) 
            else:
                #Not enough gold ?
                if:
                    PlayerObj := player[Agent]
                    CustomPlayer:Custom_Player = PlayersMap[PlayerObj]
                    CustomPlayer.PlayerGold<CostLuck2
                then:
                    NotEnoughGoldHUD.Show(Agent)
        if:
            PlayerObj := player[Agent]
            CustomPlayer:Custom_Player = PlayersMap[PlayerObj]
            CustomPlayer.Pet_Luck_Value = 1
            CustomPlayer.PlayerGold>=CostLuck2
        then:
            Upgrade_Luck2_HUD.Show(Agent)
            CustomPlayer.SetLuckBy(2)
            SetLuck(Agent,2)
            DecreaseGoldFunc(Agent,CostLuck2)
            Accolade_Upgrades.Award(Agent)

    UpgradeLuck_3(Agent:agent):void=
        #Already purchased?
        if:
            PlayerObj := player[Agent]
            CustomPlayer:Custom_Player = PlayersMap[PlayerObj]
            CustomPlayer.Pet_Luck_Value >=3
        then:
            AlreadyPurchasedHUD.Show(Agent) 
        else:
            #Previous Level bought?
            if:
                PlayerObj := player[Agent]
                CustomPlayer:Custom_Player = PlayersMap[PlayerObj]
                CustomPlayer.Pet_Luck_Value <> 2
            then:
                RequirePreviousLevelHUD.Show(Agent) 
            else:
                #Not enough gold ?
                if:
                    PlayerObj := player[Agent]
                    CustomPlayer:Custom_Player = PlayersMap[PlayerObj]
                    CustomPlayer.PlayerGold<CostLuck3
                then:
                    NotEnoughGoldHUD.Show(Agent)

        if:
            PlayerObj := player[Agent]
            CustomPlayer:Custom_Player = PlayersMap[PlayerObj]
            CustomPlayer.Pet_Luck_Value = 2
            CustomPlayer.PlayerGold>=CostLuck3
        then:
            Upgrade_Luck3_HUD.Show(Agent)
            CustomPlayer.SetLuckBy(3)
            SetLuck(Agent,3)
            DecreaseGoldFunc(Agent,CostLuck3)
            Accolade_Upgrades.Award(Agent)

    UpgradeLuck_4(Agent:agent):void=
        #Already purchased?
        if:
            PlayerObj := player[Agent]
            CustomPlayer:Custom_Player = PlayersMap[PlayerObj]
            CustomPlayer.Pet_Luck_Value >=5
        then:
            AlreadyPurchasedHUD.Show(Agent) 
        else:
            #Previous Level bought?
            if:
                PlayerObj := player[Agent]
                CustomPlayer:Custom_Player = PlayersMap[PlayerObj]
                CustomPlayer.Pet_Luck_Value <> 3
            then:
                RequirePreviousLevelHUD.Show(Agent) 
            else:
                #Not enough gold ?
                if:
                    PlayerObj := player[Agent]
                    CustomPlayer:Custom_Player = PlayersMap[PlayerObj]
                    CustomPlayer.PlayerGold<CostLuck4
                then:
                    NotEnoughGoldHUD.Show(Agent)

        if:
            PlayerObj := player[Agent]
            CustomPlayer:Custom_Player = PlayersMap[PlayerObj]
            CustomPlayer.Pet_Luck_Value = 3
            CustomPlayer.PlayerGold>=CostLuck4
        then:
            Upgrade_Luck4_HUD.Show(Agent)
            CustomPlayer.SetLuckBy(5)
            SetLuck(Agent,4)
            DecreaseGoldFunc(Agent,CostLuck4)
            Accolade_Upgrades.Award(Agent)

    UpgradeLuck_5(Agent:agent):void=
        #Already purchased?
        if:
            PlayerObj := player[Agent]
            CustomPlayer:Custom_Player = PlayersMap[PlayerObj]
            CustomPlayer.Pet_Luck_Value >=7
        then:
            AlreadyPurchasedHUD.Show(Agent) 
        else:
            #Previous Level bought?
            if:
                PlayerObj := player[Agent]
                CustomPlayer:Custom_Player = PlayersMap[PlayerObj]
                CustomPlayer.Pet_Luck_Value <> 5
            then:
                RequirePreviousLevelHUD.Show(Agent) 
            else:
                #Not enough gold ?
                if:
                    PlayerObj := player[Agent]
                    CustomPlayer:Custom_Player = PlayersMap[PlayerObj]
                    CustomPlayer.PlayerGold<CostLuck5
                then:
                    NotEnoughGoldHUD.Show(Agent)

        if:
            PlayerObj := player[Agent]
            CustomPlayer:Custom_Player = PlayersMap[PlayerObj]
            CustomPlayer.Pet_Luck_Value = 5
            CustomPlayer.PlayerGold>=CostLuck5
        then:
            Upgrade_Luck5_HUD.Show(Agent)
            CustomPlayer.SetLuckBy(7)
            SetLuck(Agent,5)
            DecreaseGoldFunc(Agent,CostLuck5)
            Accolade_Upgrades.Award(Agent)
    
    UpgradeLuck_100(Agent:agent):void=
        #Montage video
        if:
            PlayerObj := player[Agent]
            CustomPlayer:Custom_Player = PlayersMap[PlayerObj]  
        then:
            set CustomPlayer.Pet_Luck_Value = 1000000
           
    ReleasePet(CreativeProp : creative_prop)<suspends>:void=
        #Print("Release Pet")
        Grave_Yard_Pos := vector3{ X := 160.0, Y := -14000.0, Z := 30.0}
        CreativeProp.MoveTo(Grave_Yard_Pos, CreativeProp.GetTransform().Rotation, 0.5)
    
    Attack_Chest(Index : int, Damage : float):void =
        Allthechests:= ChestsL1+ChestsL2+ChestsL3+ChestsL4+ChestsL5+ChestsL6+ChestsL7+ChestsL8+ChestsL9+ChestsL10+ChestsL11+ChestsL12+ChestsL13+ChestsL14+ChestsL15+ChestsL16+ChestsL17+ChestsL18+ChestsL19+ChestsL20+ChestsL21+ChestsL22+ChestsL23+ChestsL24+ChestsL25+ChestsL26+ChestsL27+ChestsL28+ChestsL29+ChestsL30
        if (Chestt:=Allthechests[Index]):
            #Print("Attack Successful")
            spawn:
                Chestt.TakeDamage(Damage)

    Check_Chest_Gold(Index : int, FortniteCharacter : fort_character):void =
        #Print("Checking HP")
        Allthechests:= ChestsL1+ChestsL2+ChestsL3+ChestsL4+ChestsL5+ChestsL6+ChestsL7+ChestsL8+ChestsL9+ChestsL10+ChestsL11+ChestsL12+ChestsL13+ChestsL14+ChestsL15+ChestsL16+ChestsL17+ChestsL18+ChestsL19+ChestsL20+ChestsL21+ChestsL22+ChestsL23+ChestsL24+ChestsL25+ChestsL26+ChestsL27+ChestsL28+ChestsL29+ChestsL30
        if (Chestt:=Allthechests[Index]):
            if(Chestt.CurrentHP<=0.0 and Chestt.CanGrantGold=1):
                set Chestt.CanGrantGold=0
                set Chestt.CanVFX=0
                if (NewAgent := FortniteCharacter.GetAgent[]):
                    #Print("Chest destroyed index: {Index}")
                    if(Index<=15): #-------14 chest + tuto chest
                        GrantGoldChestL1(NewAgent)
                    if(Index>15 and Index<=(15*2)):
                        GrantGoldChestL2(NewAgent)
                    if(Index>15*2 and Index<=(15*3)):
                        GrantGoldChestL3(NewAgent)
                    if(Index>15*3 and Index<=(15*4)):
                        GrantGoldChestL4(NewAgent)
                    if(Index>15*4 and Index<=(15*5)):
                        GrantGoldChestL5(NewAgent)
                    if(Index>15*5 and Index<=(15*6)):
                        GrantGoldChestL6(NewAgent)
                    if(Index>15*6 and Index<=(15*7)):
                        GrantGoldChestL7(NewAgent)
                    if(Index>15*7 and Index<=(15*8)):
                        GrantGoldChestL8(NewAgent)
                    if(Index>15*8 and Index<=(15*9)):
                        GrantGoldChestL9(NewAgent)
                    if(Index>15*9 and Index<=(15*10)):
                        GrantGoldChestL10(NewAgent)
                    if(Index>15*10 and Index<=(15*11)):
                        GrantGoldChestL11(NewAgent)
                    if(Index>15*11 and Index<=(15*12)):
                        GrantGoldChestL12(NewAgent)
                    if(Index>15*12 and Index<=(15*13)):
                        GrantGoldChestL13(NewAgent)
                    if(Index>15*13 and Index<=(15*14)):
                        GrantGoldChestL14(NewAgent)
                    if(Index>15*14 and Index<=(15*15)):
                        GrantGoldChestL15(NewAgent)
                    if(Index>15*15 and Index<=(15*16)):
                        GrantGoldChestL16(NewAgent)
                    if(Index>15*16 and Index<=(15*17)):
                        GrantGoldChestL17(NewAgent)
                    if(Index>15*17 and Index<=(15*18)):
                        GrantGoldChestL18(NewAgent)
                    if(Index>15*18 and Index<=(15*19)):
                        GrantGoldChestL19(NewAgent)
                    if(Index>15*19 and Index<=(15*20)):
                        GrantGoldChestL20(NewAgent)
                    if(Index>15*20 and Index<=(15*21)):
                        GrantGoldChestL21(NewAgent)
                    if(Index>15*21 and Index<=(15*22)):
                        GrantGoldChestL22(NewAgent)
                    if(Index>15*22 and Index<=(15*23)):
                        GrantGoldChestL23(NewAgent)
                    if(Index>15*23 and Index<=(15*24)):
                        GrantGoldChestL24(NewAgent)
                    if(Index>15*24 and Index<=(15*25)):
                        GrantGoldChestL25(NewAgent)
                    if(Index>15*25 and Index<=(15*26)):
                        GrantGoldChestL26(NewAgent)
                    if(Index>15*26 and Index<=(15*27)):
                        GrantGoldChestL27(NewAgent)
                    if(Index>15*27 and Index<=(15*28)):
                        GrantGoldChestL28(NewAgent)
                    if(Index>15*28 and Index<=(15*29)):
                        GrantGoldChestL29(NewAgent)
                    if(Index>15*29 and Index<=(15*30)):
                        GrantGoldChestL30(NewAgent)
                        

    Increasegoldevent(Agent : agent):void =
        #Print("Increase gold save by 100k")
        if:
            PlayerObj := player[Agent]
            CustomPlayer:Custom_Player = PlayersMap[PlayerObj]
            CurrentPlayerStats := PlayerStatsManager.GetPlayerStats[Agent]
        then:
            Print("CustomPlayer.Bonus_Gold:{CustomPlayer.Bonus_Gold}")
            Value_Float:float=10000000000*CustomPlayer.Bonus_Gold
            Print("Value_Float: {Value_Float}")
            if(Value:int=Round[Value_Float]):
                Print("{Value}")
                CustomPlayer.IncreaseGoldBy(Value)
                AddGoldBalance(Agent,Value)

    Decreasegoldevent(Agent : agent):void =
        #Print("Decrease gold save by 100k")
        if:
            PlayerObj := player[Agent]
            CustomPlayer:Custom_Player = PlayersMap[PlayerObj]
        then:
            if(Value:int=Round[100000*CustomPlayer.Bonus_Gold]):
                Print("{Value}")
                CustomPlayer.DecreaseGoldBy(Value)
                DecreaseGoldBalance(Agent,Value)

    Increasegoldbonusevent(Agent : agent):void =
        if:
            PlayerObj := player[Agent]
            CustomPlayer:Custom_Player = PlayersMap[PlayerObj]
        then:
            set CustomPlayer.Bonus_Gold = CustomPlayer.Potion_Bonus_Gold3
            Print("CustomPlayer.Bonus_Gold: {CustomPlayer.Bonus_Gold}")

    Increasediamondevent(Agent : agent):void =
        if:
            PlayerObj := player[Agent]
            CustomPlayer:Custom_Player = PlayersMap[PlayerObj]
            CurrentPlayerStats := PlayerStatsManager.GetPlayerStats[Agent]
        then:
            Value_Float:float=1000.0
            if(Value:int=Round[Value_Float]):
                CustomPlayer.IncreaseDiamondBy(Value)
                AddDiamondBalance(Agent,Value)

    Decreasediamondevent(Agent : agent):void =
        if:
            PlayerObj := player[Agent]
            CustomPlayer:Custom_Player = PlayersMap[PlayerObj]
        then:
            if(Value:int=Round[100.0]):
                CustomPlayer.DecreaseDiamondBy(Value)
                DecreaseDiamondBalance(Agent,Value)

    IncreaseGoldFunc(Agent : agent, Value:int):void =
        if:
            PlayerObj := player[Agent]
            CustomPlayer:Custom_Player = PlayersMap[PlayerObj]
        then:
            CustomPlayer.IncreaseGoldBy(Value)#UI
            AddGoldBalance(Agent,Value)#Save

    DecreaseGoldFunc(Agent : agent , Value:int):void =
        if:
            PlayerObj := player[Agent]
            CustomPlayer:Custom_Player = PlayersMap[PlayerObj]
        then:
            CustomPlayer.DecreaseGoldBy(Value)#UI
            DecreaseGoldBalance(Agent,Value)#Save

    GrantGoldChestL1(Agent : agent):void =
        if:
            PlayerObj := player[Agent]
            CustomPlayer:Custom_Player = PlayersMap[PlayerObj]
        then:
            if(Value:int=Round[GoldChestL1Value*CustomPlayer.Bonus_Gold*CustomPlayer.Bonus_Gold_Perma]):
                CustomPlayer.IncreaseGoldBy(Value)
                AddGoldBalance(Agent,Value)
                GoldChestL1.GrantItem(Agent)
                PotionDrop(Agent)

    GrantGoldChestL2(Agent : agent):void =
        if:
            PlayerObj := player[Agent]
            CustomPlayer:Custom_Player = PlayersMap[PlayerObj]
        then:
            if(Value:int=Round[GoldChestL2Value*CustomPlayer.Bonus_Gold*CustomPlayer.Bonus_Gold_Perma]):
                CustomPlayer.IncreaseGoldBy(Value)
                AddGoldBalance(Agent,Value)
                GoldChestL2.GrantItem(Agent)
                PotionDrop(Agent)

    GrantGoldChestL3(Agent : agent):void =
        if:
            PlayerObj := player[Agent]
            CustomPlayer:Custom_Player = PlayersMap[PlayerObj]
        then:
            if(Value:int=Round[GoldChestL3Value*CustomPlayer.Bonus_Gold*CustomPlayer.Bonus_Gold_Perma]):
                CustomPlayer.IncreaseGoldBy(Value)
                AddGoldBalance(Agent,Value)
                GoldChestL3.GrantItem(Agent)
                PotionDrop(Agent)

    GrantGoldChestL4(Agent : agent):void =
        if:
            PlayerObj := player[Agent]
            CustomPlayer:Custom_Player = PlayersMap[PlayerObj]
        then:
            if(Value:int=Round[GoldChestL4Value*CustomPlayer.Bonus_Gold*CustomPlayer.Bonus_Gold_Perma]):
                CustomPlayer.IncreaseGoldBy(Value)
                AddGoldBalance(Agent,Value)
                GoldChestL4.GrantItem(Agent)
                PotionDrop(Agent)
    
    GrantGoldChestL5(Agent : agent):void =
        if:
            PlayerObj := player[Agent]
            CustomPlayer:Custom_Player = PlayersMap[PlayerObj]
        then:
            if(Value:int=Round[GoldChestL5Value*CustomPlayer.Bonus_Gold*CustomPlayer.Bonus_Gold_Perma]):
                CustomPlayer.IncreaseGoldBy(Value)
                AddGoldBalance(Agent,Value)
                GoldChestL5.GrantItem(Agent)
                PotionDrop(Agent)

    GrantGoldChestL6(Agent : agent):void =
        if:
            PlayerObj := player[Agent]
            CustomPlayer:Custom_Player = PlayersMap[PlayerObj]
        then:
            if(Value:int=Round[GoldChestL6Value*CustomPlayer.Bonus_Gold*CustomPlayer.Bonus_Gold_Perma]):
                CustomPlayer.IncreaseGoldBy(Value)
                AddGoldBalance(Agent,Value)
                GoldChestL6.GrantItem(Agent)
                PotionDrop(Agent)

    GrantGoldChestL7(Agent : agent):void =
        if:
            PlayerObj := player[Agent]
            CustomPlayer:Custom_Player = PlayersMap[PlayerObj]
        then:
            if(Value:int=Round[GoldChestL7Value*CustomPlayer.Bonus_Gold*CustomPlayer.Bonus_Gold_Perma]):
                CustomPlayer.IncreaseGoldBy(Value)
                AddGoldBalance(Agent,Value)
                GoldChestL7.GrantItem(Agent)
                PotionDrop(Agent)

    GrantGoldChestL8(Agent : agent):void =
        if:
            PlayerObj := player[Agent]
            CustomPlayer:Custom_Player = PlayersMap[PlayerObj]
        then:
            if(Value:int=Round[GoldChestL8Value*CustomPlayer.Bonus_Gold*CustomPlayer.Bonus_Gold_Perma]):
                CustomPlayer.IncreaseGoldBy(Value)
                AddGoldBalance(Agent,Value)
                GoldChestL8.GrantItem(Agent)
                PotionDrop(Agent)

    GrantGoldChestL9(Agent : agent):void =
        if:
            PlayerObj := player[Agent]
            CustomPlayer:Custom_Player = PlayersMap[PlayerObj]
        then:
            if(Value:int=Round[GoldChestL9Value*CustomPlayer.Bonus_Gold*CustomPlayer.Bonus_Gold_Perma]):
                CustomPlayer.IncreaseGoldBy(Value)
                AddGoldBalance(Agent,Value)
                GoldChestL9.GrantItem(Agent)
                PotionDrop(Agent)

    GrantGoldChestL10(Agent : agent):void =
        if:
            PlayerObj := player[Agent]
            CustomPlayer:Custom_Player = PlayersMap[PlayerObj]
        then:
            if(Value:int=Round[GoldChestL10Value*CustomPlayer.Bonus_Gold*CustomPlayer.Bonus_Gold_Perma]):
                CustomPlayer.IncreaseGoldBy(Value)
                AddGoldBalance(Agent,Value)
                GoldChestL10.GrantItem(Agent)
                PotionDrop(Agent)

    GrantGoldChestL11(Agent : agent):void =
        if:
            PlayerObj := player[Agent]
            CustomPlayer:Custom_Player = PlayersMap[PlayerObj]
        then:
            if(Value:int=Round[GoldChestL11Value*CustomPlayer.Bonus_Gold*CustomPlayer.Bonus_Gold_Perma]):
                CustomPlayer.IncreaseGoldBy(Value)
                AddGoldBalance(Agent,Value)
                GoldChestL11.GrantItem(Agent)
                PotionDrop(Agent)

    GrantGoldChestL12(Agent : agent):void =
        if:
            PlayerObj := player[Agent]
            CustomPlayer:Custom_Player = PlayersMap[PlayerObj]
        then:
            if(Value:int=Round[GoldChestL12Value*CustomPlayer.Bonus_Gold*CustomPlayer.Bonus_Gold_Perma]):
                CustomPlayer.IncreaseGoldBy(Value)
                AddGoldBalance(Agent,Value)
                GoldChestL12.GrantItem(Agent)
                PotionDrop(Agent)

    GrantGoldChestL13(Agent : agent):void =
        if:
            PlayerObj := player[Agent]
            CustomPlayer:Custom_Player = PlayersMap[PlayerObj]
        then:
            if(Value:int=Round[GoldChestL13Value*CustomPlayer.Bonus_Gold*CustomPlayer.Bonus_Gold_Perma]):
                CustomPlayer.IncreaseGoldBy(Value)
                AddGoldBalance(Agent,Value)
                GoldChestL13.GrantItem(Agent)
                PotionDrop(Agent)

    GrantGoldChestL14(Agent : agent):void =
        if:
            PlayerObj := player[Agent]
            CustomPlayer:Custom_Player = PlayersMap[PlayerObj]
        then:
            if(Value:int=Round[GoldChestL14Value*CustomPlayer.Bonus_Gold*CustomPlayer.Bonus_Gold_Perma]):
                CustomPlayer.IncreaseGoldBy(Value)
                AddGoldBalance(Agent,Value)
                GoldChestL14.GrantItem(Agent)
                PotionDrop(Agent)

    GrantGoldChestL15(Agent : agent):void =
        if:
            PlayerObj := player[Agent]
            CustomPlayer:Custom_Player = PlayersMap[PlayerObj]
        then:
            if(Value:int=Round[GoldChestL15Value*CustomPlayer.Bonus_Gold*CustomPlayer.Bonus_Gold_Perma]):
                CustomPlayer.IncreaseGoldBy(Value)
                AddGoldBalance(Agent,Value)
                GoldChestL15.GrantItem(Agent)
                PotionDrop(Agent)

    GrantGoldChestL16(Agent : agent):void =
        if:
            PlayerObj := player[Agent]
            CustomPlayer:Custom_Player = PlayersMap[PlayerObj]
        then:
            if(Value:int=Round[GoldChestL16Value*CustomPlayer.Bonus_Gold*CustomPlayer.Bonus_Gold_Perma]):
                CustomPlayer.IncreaseGoldBy(Value)
                AddGoldBalance(Agent,Value)
                GoldChestL16.GrantItem(Agent)
                PotionDrop(Agent)

    GrantGoldChestL17(Agent : agent):void =
        if:
            PlayerObj := player[Agent]
            CustomPlayer:Custom_Player = PlayersMap[PlayerObj]
        then:
            if(Value:int=Round[GoldChestL17Value*CustomPlayer.Bonus_Gold*CustomPlayer.Bonus_Gold_Perma]):
                CustomPlayer.IncreaseGoldBy(Value)
                AddGoldBalance(Agent,Value)
                GoldChestL17.GrantItem(Agent)
                PotionDrop(Agent)

    GrantGoldChestL18(Agent : agent):void =
        if:
            PlayerObj := player[Agent]
            CustomPlayer:Custom_Player = PlayersMap[PlayerObj]
        then:
            if(Value:int=Round[GoldChestL18Value*CustomPlayer.Bonus_Gold*CustomPlayer.Bonus_Gold_Perma]):
                CustomPlayer.IncreaseGoldBy(Value)
                AddGoldBalance(Agent,Value)
                GoldChestL18.GrantItem(Agent)
                PotionDrop(Agent)

    GrantGoldChestL19(Agent : agent):void =
        if:
            PlayerObj := player[Agent]
            CustomPlayer:Custom_Player = PlayersMap[PlayerObj]
        then:
            if(Value:int=Round[GoldChestL19Value*CustomPlayer.Bonus_Gold*CustomPlayer.Bonus_Gold_Perma]):
                CustomPlayer.IncreaseGoldBy(Value)
                AddGoldBalance(Agent,Value)
                GoldChestL19.GrantItem(Agent)
                PotionDrop(Agent)

    GrantGoldChestL20(Agent : agent):void =
        if:
            PlayerObj := player[Agent]
            CustomPlayer:Custom_Player = PlayersMap[PlayerObj]
        then:
            if(Value:int=Round[GoldChestL20Value*CustomPlayer.Bonus_Gold*CustomPlayer.Bonus_Gold_Perma]):
                CustomPlayer.IncreaseGoldBy(Value)
                AddGoldBalance(Agent,Value)
                GoldChestL20.GrantItem(Agent)
                PotionDrop(Agent)

    GrantGoldChestL21(Agent : agent):void =
        if:
            PlayerObj := player[Agent]
            CustomPlayer:Custom_Player = PlayersMap[PlayerObj]
        then:
            if(Value:int=Round[GoldChestL21Value*CustomPlayer.Bonus_Gold*CustomPlayer.Bonus_Gold_Perma]):
                CustomPlayer.IncreaseGoldBy(Value)
                AddGoldBalance(Agent,Value)
                GoldChestL21.GrantItem(Agent)
                PotionDrop(Agent)

    GrantGoldChestL22(Agent : agent):void =
        if:
            PlayerObj := player[Agent]
            CustomPlayer:Custom_Player = PlayersMap[PlayerObj]
        then:
            if(Value:int=Round[GoldChestL22Value*CustomPlayer.Bonus_Gold*CustomPlayer.Bonus_Gold_Perma]):
                CustomPlayer.IncreaseGoldBy(Value)
                AddGoldBalance(Agent,Value)
                GoldChestL22.GrantItem(Agent)
                PotionDrop(Agent)

    GrantGoldChestL23(Agent : agent):void =
        if:
            PlayerObj := player[Agent]
            CustomPlayer:Custom_Player = PlayersMap[PlayerObj]
        then:
            if(Value:int=Round[GoldChestL23Value*CustomPlayer.Bonus_Gold*CustomPlayer.Bonus_Gold_Perma]):
                CustomPlayer.IncreaseGoldBy(Value)
                AddGoldBalance(Agent,Value)
                GoldChestL23.GrantItem(Agent)
                PotionDrop(Agent)

    GrantGoldChestL24(Agent : agent):void =
        if:
            PlayerObj := player[Agent]
            CustomPlayer:Custom_Player = PlayersMap[PlayerObj]
        then:
            if(Value:int=Round[GoldChestL24Value*CustomPlayer.Bonus_Gold*CustomPlayer.Bonus_Gold_Perma]):
                CustomPlayer.IncreaseGoldBy(Value)
                AddGoldBalance(Agent,Value)
                GoldChestL24.GrantItem(Agent)
                PotionDrop(Agent)

    GrantGoldChestL25(Agent : agent):void =
        if:
            PlayerObj := player[Agent]
            CustomPlayer:Custom_Player = PlayersMap[PlayerObj]
        then:
            if(Value:int=Round[GoldChestL25Value*CustomPlayer.Bonus_Gold*CustomPlayer.Bonus_Gold_Perma]):
                CustomPlayer.IncreaseGoldBy(Value)
                AddGoldBalance(Agent,Value)
                GoldChestL25.GrantItem(Agent)
                PotionDrop(Agent)

    GrantGoldChestL26(Agent : agent):void =
        if:
            PlayerObj := player[Agent]
            CustomPlayer:Custom_Player = PlayersMap[PlayerObj]
        then:
            if(Value:int=Round[GoldChestL26Value*CustomPlayer.Bonus_Gold*CustomPlayer.Bonus_Gold_Perma]):
                CustomPlayer.IncreaseGoldBy(Value)
                AddGoldBalance(Agent,Value)
                GoldChestL26.GrantItem(Agent)
                PotionDrop(Agent)

    GrantGoldChestL27(Agent : agent):void =
        if:
            PlayerObj := player[Agent]
            CustomPlayer:Custom_Player = PlayersMap[PlayerObj]
        then:
            if(Value:int=Round[GoldChestL27Value*CustomPlayer.Bonus_Gold*CustomPlayer.Bonus_Gold_Perma]):
                CustomPlayer.IncreaseGoldBy(Value)
                AddGoldBalance(Agent,Value)
                GoldChestL27.GrantItem(Agent)
                PotionDrop(Agent)

    GrantGoldChestL28(Agent : agent):void =
        if:
            PlayerObj := player[Agent]
            CustomPlayer:Custom_Player = PlayersMap[PlayerObj]
        then:
            if(Value:int=Round[GoldChestL28Value*CustomPlayer.Bonus_Gold*CustomPlayer.Bonus_Gold_Perma]):
                CustomPlayer.IncreaseGoldBy(Value)
                AddGoldBalance(Agent,Value)
                GoldChestL28.GrantItem(Agent)
                PotionDrop(Agent)

    GrantGoldChestL29(Agent : agent):void =
        if:
            PlayerObj := player[Agent]
            CustomPlayer:Custom_Player = PlayersMap[PlayerObj]
        then:
            if(Value:int=Round[GoldChestL29Value*CustomPlayer.Bonus_Gold*CustomPlayer.Bonus_Gold_Perma]):
                CustomPlayer.IncreaseGoldBy(Value)
                AddGoldBalance(Agent,Value)
                GoldChestL29.GrantItem(Agent)
                PotionDrop(Agent)

    GrantGoldChestL30(Agent : agent):void =
        if:
            PlayerObj := player[Agent]
            CustomPlayer:Custom_Player = PlayersMap[PlayerObj]
        then:
            if(Value:int=Round[GoldChestL30Value*CustomPlayer.Bonus_Gold*CustomPlayer.Bonus_Gold_Perma]):
                CustomPlayer.IncreaseGoldBy(Value)
                AddGoldBalance(Agent,Value)
                GoldChestL30.GrantItem(Agent)
                PotionDrop(Agent)

    #BIG Side chests

    GrantGoldChestBIG1(Agent : agent):void =
        if:
            PlayerObj := player[Agent]
            CustomPlayer:Custom_Player = PlayersMap[PlayerObj]
        then:
            CustomPlayer.IncreaseGoldBy(GoldChestBIG1Value)
            AddGoldBalance(Agent,GoldChestBIG1Value)
            GoldChestBIG1.GrantItem(Agent)

    GrantGoldChestBIG2(Agent : agent):void =
        if:
            PlayerObj := player[Agent]
            CustomPlayer:Custom_Player = PlayersMap[PlayerObj]
        then:
            CustomPlayer.IncreaseGoldBy(GoldChestBIG2Value)
            AddGoldBalance(Agent,GoldChestBIG2Value)
            GoldChestBIG2.GrantItem(Agent)

    GrantGoldChestBIG3(Agent : agent):void =
        if:
            PlayerObj := player[Agent]
            CustomPlayer:Custom_Player = PlayersMap[PlayerObj]
        then:
            CustomPlayer.IncreaseGoldBy(GoldChestBIG3Value)
            AddGoldBalance(Agent,GoldChestBIG3Value)
            GoldChestBIG3.GrantItem(Agent)

    GrantGoldChestBIG4(Agent : agent):void =
        if:
            PlayerObj := player[Agent]
            CustomPlayer:Custom_Player = PlayersMap[PlayerObj]
        then:
            CustomPlayer.IncreaseGoldBy(GoldChestBIG4Value)
            AddGoldBalance(Agent,GoldChestBIG4Value)
            GoldChestBIG4.GrantItem(Agent)

    GrantGoldChestBIG5(Agent : agent):void =
        if:
            PlayerObj := player[Agent]
            CustomPlayer:Custom_Player = PlayersMap[PlayerObj]
        then:
            CustomPlayer.IncreaseGoldBy(GoldChestBIG5Value)
            AddGoldBalance(Agent,GoldChestBIG5Value)
            GoldChestBIG5.GrantItem(Agent)

    GrantGoldChestBIG6(Agent : agent):void =
        if:
            PlayerObj := player[Agent]
            CustomPlayer:Custom_Player = PlayersMap[PlayerObj]
        then:
            CustomPlayer.IncreaseGoldBy(GoldChestBIG6Value)
            AddGoldBalance(Agent,GoldChestBIG6Value)
            GoldChestBIG6.GrantItem(Agent)

    GrantGoldChestBIG7(Agent : agent):void =
        if:
            PlayerObj := player[Agent]
            CustomPlayer:Custom_Player = PlayersMap[PlayerObj]
        then:
            CustomPlayer.IncreaseGoldBy(GoldChestBIG7Value)
            AddGoldBalance(Agent,GoldChestBIG7Value)
            GoldChestBIG7.GrantItem(Agent)

    GrantGoldChestBIG8(Agent : agent):void =
        if:
            PlayerObj := player[Agent]
            CustomPlayer:Custom_Player = PlayersMap[PlayerObj]
        then:
            CustomPlayer.IncreaseGoldBy(GoldChestBIG8Value)
            AddGoldBalance(Agent,GoldChestBIG8Value)
            GoldChestBIG8.GrantItem(Agent)

    #------------Gift Playtime
    GrantGift1(Agent : agent):void =
        if:
            PlayerObj := player[Agent]
            CustomPlayer:Custom_Player = PlayersMap[PlayerObj]
        then:
            CustomPlayer.IncreaseGoldBy(GoldGift1Value)
            AddGoldBalance(Agent,GoldGift1Value)     

    GrantGift2(Agent : agent):void =
        if:
            PlayerObj := player[Agent]
            CustomPlayer:Custom_Player = PlayersMap[PlayerObj]
        then:
            CustomPlayer.IncreaseGoldBy(GoldGift2Value)
            AddGoldBalance(Agent,GoldGift2Value)

    GrantGift3(Agent : agent):void =
        if:
            PlayerObj := player[Agent]
            CustomPlayer:Custom_Player = PlayersMap[PlayerObj]
        then:
            CustomPlayer.IncreaseGoldBy(GoldGift3Value)
            AddGoldBalance(Agent,GoldGift3Value)

    GrantGift4(Agent : agent):void =
        if:
            PlayerObj := player[Agent]
            CustomPlayer:Custom_Player = PlayersMap[PlayerObj]
        then:
            CustomPlayer.IncreaseGoldBy(GoldGift4Value)
            AddGoldBalance(Agent,GoldGift4Value)

    GrantGift5(Agent : agent):void =
        if:
            PlayerObj := player[Agent]
            CustomPlayer:Custom_Player = PlayersMap[PlayerObj]
        then:
            CustomPlayer.IncreaseGoldBy(GoldGift5Value)
            AddGoldBalance(Agent,GoldGift5Value)

    GrantGift6(Agent : agent):void =
        if:
            PlayerObj := player[Agent]
            CustomPlayer:Custom_Player = PlayersMap[PlayerObj]
        then:
            CustomPlayer.IncreaseGoldBy(GoldGift6Value)
            AddGoldBalance(Agent,GoldGift6Value)

    GrantGift7(Agent : agent):void =
        if:
            PlayerObj := player[Agent]
            CustomPlayer:Custom_Player = PlayersMap[PlayerObj]
        then:
            CustomPlayer.IncreaseGoldBy(GoldGift7Value)
            AddGoldBalance(Agent,GoldGift7Value)

    GrantGift8(Agent : agent):void =
        if:
            PlayerObj := player[Agent]
            CustomPlayer:Custom_Player = PlayersMap[PlayerObj]
        then:
            CustomPlayer.IncreaseGoldBy(GoldGift8Value)
            AddGoldBalance(Agent,GoldGift8Value)

    GrantGift9(Agent : agent):void =
        if:
            PlayerObj := player[Agent]
            CustomPlayer:Custom_Player = PlayersMap[PlayerObj]
        then:
            CustomPlayer.IncreaseGoldBy(GoldGift9Value)
            AddGoldBalance(Agent,GoldGift9Value)
    
    #---------------------HIT
    GrantGoldHitL1(Agent : agent):void =
        #Print("GrantUI Hit L1")
        if:
            PlayerObj := player[Agent]
            CustomPlayer:Custom_Player = PlayersMap[PlayerObj]
        then:
            CustomPlayer.IncreaseGoldBy(GoldHitL1Value)
            AddGoldBalance(Agent,GoldHitL1Value)

    GrantGoldHitL2(Agent : agent):void =
        #Print("GrantUI Hit L2")
        if:
            PlayerObj := player[Agent]
            CustomPlayer:Custom_Player = PlayersMap[PlayerObj]
        then:
            CustomPlayer.IncreaseGoldBy(GoldHitL2Value)
            AddGoldBalance(Agent,GoldHitL2Value)

    GrantGoldHitL3(Agent : agent):void =
        if:
            PlayerObj := player[Agent]
            CustomPlayer:Custom_Player = PlayersMap[PlayerObj]
        then:
            CustomPlayer.IncreaseGoldBy(GoldHitL3Value)
            AddGoldBalance(Agent,GoldHitL3Value)

    GrantGoldHitL4(Agent : agent):void =
        if:
            PlayerObj := player[Agent]
            CustomPlayer:Custom_Player = PlayersMap[PlayerObj]
        then:
            CustomPlayer.IncreaseGoldBy(GoldHitL4Value)
            AddGoldBalance(Agent,GoldHitL4Value)
    
    GrantGoldHitL5(Agent : agent):void =
        if:
            PlayerObj := player[Agent]
            CustomPlayer:Custom_Player = PlayersMap[PlayerObj]
        then:
            CustomPlayer.IncreaseGoldBy(GoldHitL5Value)
            AddGoldBalance(Agent,GoldHitL5Value)

    GrantGoldHitL6(Agent : agent):void =
        if:
            PlayerObj := player[Agent]
            CustomPlayer:Custom_Player = PlayersMap[PlayerObj]
        then:
            CustomPlayer.IncreaseGoldBy(GoldHitL6Value)
            AddGoldBalance(Agent,GoldHitL6Value)

    GrantGoldHitL7(Agent : agent):void =
        if:
            PlayerObj := player[Agent]
            CustomPlayer:Custom_Player = PlayersMap[PlayerObj]
        then:
            CustomPlayer.IncreaseGoldBy(GoldHitL7Value)
            AddGoldBalance(Agent,GoldHitL7Value)

    GrantGoldHitL8(Agent : agent):void =
        if:
            PlayerObj := player[Agent]
            CustomPlayer:Custom_Player = PlayersMap[PlayerObj]
        then:
            CustomPlayer.IncreaseGoldBy(GoldHitL8Value)
            AddGoldBalance(Agent,GoldHitL8Value)

    GrantGoldHitL9(Agent : agent):void =
        if:
            PlayerObj := player[Agent]
            CustomPlayer:Custom_Player = PlayersMap[PlayerObj]
        then:
            CustomPlayer.IncreaseGoldBy(GoldHitL9Value)
            AddGoldBalance(Agent,GoldHitL9Value)

    GrantGoldHitL10(Agent : agent):void =
        if:
            PlayerObj := player[Agent]
            CustomPlayer:Custom_Player = PlayersMap[PlayerObj]
        then:
            CustomPlayer.IncreaseGoldBy(GoldHitL10Value)
            AddGoldBalance(Agent,GoldHitL10Value)

    GrantGoldHitL11(Agent : agent):void =
        if:
            PlayerObj := player[Agent]
            CustomPlayer:Custom_Player = PlayersMap[PlayerObj]
        then:
            CustomPlayer.IncreaseGoldBy(GoldHitL11Value)
            AddGoldBalance(Agent,GoldHitL11Value)

    GrantGoldHitL12(Agent : agent):void =
        if:
            PlayerObj := player[Agent]
            CustomPlayer:Custom_Player = PlayersMap[PlayerObj]
        then:
            CustomPlayer.IncreaseGoldBy(GoldHitL12Value)
            AddGoldBalance(Agent,GoldHitL12Value)

    GrantGoldHitL13(Agent : agent):void =
        if:
            PlayerObj := player[Agent]
            CustomPlayer:Custom_Player = PlayersMap[PlayerObj]
        then:
            CustomPlayer.IncreaseGoldBy(GoldHitL13Value)
            AddGoldBalance(Agent,GoldHitL13Value)

    GrantGoldHitL14(Agent : agent):void =
        if:
            PlayerObj := player[Agent]
            CustomPlayer:Custom_Player = PlayersMap[PlayerObj]
        then:
            CustomPlayer.IncreaseGoldBy(GoldHitL14Value)
            AddGoldBalance(Agent,GoldHitL14Value)

    GrantGoldHitL15(Agent : agent):void =
        if:
            PlayerObj := player[Agent]
            CustomPlayer:Custom_Player = PlayersMap[PlayerObj]
        then:
            CustomPlayer.IncreaseGoldBy(GoldHitL15Value)
            AddGoldBalance(Agent,GoldHitL15Value)

    GrantGoldHitL16(Agent : agent):void =
        if:
            PlayerObj := player[Agent]
            CustomPlayer:Custom_Player = PlayersMap[PlayerObj]
        then:
            CustomPlayer.IncreaseGoldBy(GoldHitL16Value)
            AddGoldBalance(Agent,GoldHitL16Value)

    GrantGoldHitL17(Agent : agent):void =
        if:
            PlayerObj := player[Agent]
            CustomPlayer:Custom_Player = PlayersMap[PlayerObj]
        then:
            CustomPlayer.IncreaseGoldBy(GoldHitL17Value)
            AddGoldBalance(Agent,GoldHitL17Value)

    GrantGoldHitL18(Agent : agent):void =
        if:
            PlayerObj := player[Agent]
            CustomPlayer:Custom_Player = PlayersMap[PlayerObj]
        then:
            CustomPlayer.IncreaseGoldBy(GoldHitL18Value)
            AddGoldBalance(Agent,GoldHitL18Value)

    GrantGoldHitL19(Agent : agent):void =
        if:
            PlayerObj := player[Agent]
            CustomPlayer:Custom_Player = PlayersMap[PlayerObj]
        then:
            CustomPlayer.IncreaseGoldBy(GoldHitL19Value)
            AddGoldBalance(Agent,GoldHitL19Value)

    GrantGoldHitL20(Agent : agent):void =
        if:
            PlayerObj := player[Agent]
            CustomPlayer:Custom_Player = PlayersMap[PlayerObj]
        then:
            CustomPlayer.IncreaseGoldBy(GoldHitL20Value)
            AddGoldBalance(Agent,GoldHitL20Value)

    GrantGoldHitL21(Agent : agent):void =
        if:
            PlayerObj := player[Agent]
            CustomPlayer:Custom_Player = PlayersMap[PlayerObj]
        then:
            CustomPlayer.IncreaseGoldBy(GoldHitL21Value)
            AddGoldBalance(Agent,GoldHitL21Value)

    GrantGoldHitL22(Agent : agent):void =
        if:
            PlayerObj := player[Agent]
            CustomPlayer:Custom_Player = PlayersMap[PlayerObj]
        then:
            CustomPlayer.IncreaseGoldBy(GoldHitL22Value)
            AddGoldBalance(Agent,GoldHitL22Value)

    GrantGoldHitL23(Agent : agent):void =
        if:
            PlayerObj := player[Agent]
            CustomPlayer:Custom_Player = PlayersMap[PlayerObj]
        then:
            CustomPlayer.IncreaseGoldBy(GoldHitL23Value)
            AddGoldBalance(Agent,GoldHitL23Value)

    GrantGoldHitL24(Agent : agent):void =
        if:
            PlayerObj := player[Agent]
            CustomPlayer:Custom_Player = PlayersMap[PlayerObj]
        then:
            CustomPlayer.IncreaseGoldBy(GoldHitL24Value)
            AddGoldBalance(Agent,GoldHitL24Value)

    GrantGoldHitL25(Agent : agent):void =
        if:
            PlayerObj := player[Agent]
            CustomPlayer:Custom_Player = PlayersMap[PlayerObj]
        then:
            CustomPlayer.IncreaseGoldBy(GoldHitL25Value)
            AddGoldBalance(Agent,GoldHitL25Value)

    GrantGoldHitL26(Agent : agent):void =
        if:
            PlayerObj := player[Agent]
            CustomPlayer:Custom_Player = PlayersMap[PlayerObj]
        then:
            CustomPlayer.IncreaseGoldBy(GoldHitL26Value)
            AddGoldBalance(Agent,GoldHitL26Value)

    GrantGoldHitL27(Agent : agent):void =
        if:
            PlayerObj := player[Agent]
            CustomPlayer:Custom_Player = PlayersMap[PlayerObj]
        then:
            CustomPlayer.IncreaseGoldBy(GoldHitL27Value)
            AddGoldBalance(Agent,GoldHitL27Value)

    GrantGoldHitL28(Agent : agent):void =
        if:
            PlayerObj := player[Agent]
            CustomPlayer:Custom_Player = PlayersMap[PlayerObj]
        then:
            CustomPlayer.IncreaseGoldBy(GoldHitL28Value)
            AddGoldBalance(Agent,GoldHitL28Value)

    GrantGoldHitL29(Agent : agent):void =
        if:
            PlayerObj := player[Agent]
            CustomPlayer:Custom_Player = PlayersMap[PlayerObj]
        then:
            CustomPlayer.IncreaseGoldBy(GoldHitL29Value)
            AddGoldBalance(Agent,GoldHitL29Value)

    GrantGoldHitL30(Agent : agent):void =
        if:
            PlayerObj := player[Agent]
            CustomPlayer:Custom_Player = PlayersMap[PlayerObj]
        then:
            CustomPlayer.IncreaseGoldBy(GoldHitL30Value)
            AddGoldBalance(Agent,GoldHitL30Value)

    UnlockBarrierB1(Agent : agent):void = 
        BarrierL1.AddToIgnoreList(Agent)
        #BarrierTUTO.AddToIgnoreList(Agent)
        if(Cine:=Cine_Unlock_L2_Egg[GetTeamIndexAgent[Agent]-1]):#Swap eggs to unlock + Barrier Box removed
            Cine.Play(Agent)
        if(Cine:=Cine_Unlock_L2[GetTeamIndexAgent[Agent]-1]):#Show new egg ready
            Cine.Play(Agent) 
            Print("Cinematique Swapp Egg 2")
        VFX_B1.Restart()
        if:
            PlayerObj := player[Agent]
            CustomPlayer:Custom_Player = PlayersMap[PlayerObj]
        then:
            set CustomPlayer.B1_Unlocked=1
            CustomPlayer.Update_Pets_Section()

    UnlockBarrierB2(Agent : agent):void = 
        BarrierL2.AddToIgnoreList(Agent)
        BarrierVILLAGE.AddToIgnoreList(Agent)
        if(Cine:=Cine_Unlock_L3[GetTeamIndexAgent[Agent]-1]):#Swap eggs to unlock + Barrier Box removed
            Cine.Play(Agent)
            Print("Cinematique Swapp Egg 3")
        VFX_B2.Restart()
        if:
            PlayerObj := player[Agent]
            CustomPlayer:Custom_Player = PlayersMap[PlayerObj]
        then:
            set CustomPlayer.B2_Unlocked=1
            CustomPlayer.Update_Pets_Section()

    UnlockBarrierB3(Agent : agent):void = 
        BarrierL3.AddToIgnoreList(Agent)
        if(Cine:=Cine_Unlock_L4[GetTeamIndexAgent[Agent]-1]):#Swap eggs to unlock + Barrier Box removed
            Cine.Play(Agent)
            Print("Cinematique Swapp Egg 4")
        VFX_B3.Restart()
        if:
            PlayerObj := player[Agent]
            CustomPlayer:Custom_Player = PlayersMap[PlayerObj]
        then:
            set CustomPlayer.B3_Unlocked=1
            CustomPlayer.Update_Pets_Section()

    UnlockBarrierB4(Agent : agent):void = 
        BarrierL4.AddToIgnoreList(Agent)
        if(Cine:=Cine_Unlock_L5[GetTeamIndexAgent[Agent]-1]):#Swap eggs to unlock + Barrier Box removed
            Cine.Play(Agent)
        VFX_B4.Restart()
        if:
            PlayerObj := player[Agent]
            CustomPlayer:Custom_Player = PlayersMap[PlayerObj]
        then:
            set CustomPlayer.B4_Unlocked=1
            CustomPlayer.Update_Pets_Section()

    UnlockBarrierB5(Agent : agent):void = 
        BarrierL5.AddToIgnoreList(Agent)
        if(Cine:=Cine_Unlock_L6[GetTeamIndexAgent[Agent]-1]):#Swap eggs to unlock + Barrier Box removed
            Cine.Play(Agent)
        VFX_B5.Restart()
        if:
            PlayerObj := player[Agent]
            CustomPlayer:Custom_Player = PlayersMap[PlayerObj]
        then:
            set CustomPlayer.B5_Unlocked=1
            CustomPlayer.Update_Pets_Section()

    UnlockBarrierB6(Agent : agent):void = 
        BarrierL6.AddToIgnoreList(Agent)
        if(Cine:=Cine_Unlock_L7[GetTeamIndexAgent[Agent]-1]):#Swap eggs to unlock + Barrier Box removed
            Cine.Play(Agent)
        VFX_B6.Restart()
        if:
            PlayerObj := player[Agent]
            CustomPlayer:Custom_Player = PlayersMap[PlayerObj]
        then:
            set CustomPlayer.B6_Unlocked=1
            CustomPlayer.Update_Pets_Section()

    UnlockBarrierB7(Agent : agent):void = 
        BarrierL7.AddToIgnoreList(Agent)
        if(Cine:=Cine_Unlock_L8[GetTeamIndexAgent[Agent]-1]):#Swap eggs to unlock + Barrier Box removed
            Cine.Play(Agent)
        VFX_B7.Restart()
        if:
            PlayerObj := player[Agent]
            CustomPlayer:Custom_Player = PlayersMap[PlayerObj]
        then:
            set CustomPlayer.B7_Unlocked=1
            CustomPlayer.Update_Pets_Section()

    UnlockBarrierB8(Agent : agent):void = 
        BarrierL8.AddToIgnoreList(Agent)
        if(Cine:=Cine_Unlock_L9[GetTeamIndexAgent[Agent]-1]):#Swap eggs to unlock + Barrier Box removed
            Cine.Play(Agent)
        VFX_B8.Restart()
        if:
            PlayerObj := player[Agent]
            CustomPlayer:Custom_Player = PlayersMap[PlayerObj]
        then:
            set CustomPlayer.B8_Unlocked=1
            CustomPlayer.Update_Pets_Section()

    UnlockBarrierB9(Agent : agent):void = 
        BarrierL9.AddToIgnoreList(Agent)
        if(Cine:=Cine_Unlock_L10[GetTeamIndexAgent[Agent]-1]):#Swap eggs to unlock + Barrier Box removed
            Cine.Play(Agent)
        VFX_B9.Restart()
        if:
            PlayerObj := player[Agent]
            CustomPlayer:Custom_Player = PlayersMap[PlayerObj]
        then:
            set CustomPlayer.B9_Unlocked=1
            CustomPlayer.Update_Pets_Section()

    UnlockBarrierB10(Agent : agent):void = 
        BarrierL10.AddToIgnoreList(Agent)
        if(Cine:=Cine_Unlock_L11[GetTeamIndexAgent[Agent]-1]):#Swap eggs to unlock + Barrier Box removed
            Cine.Play(Agent)
        VFX_B10.Restart()
        if:
            PlayerObj := player[Agent]
            CustomPlayer:Custom_Player = PlayersMap[PlayerObj]
        then:
            set CustomPlayer.B10_Unlocked=1
            CustomPlayer.Update_Pets_Section()

    UnlockBarrierB11(Agent : agent):void = 
        BarrierL11.AddToIgnoreList(Agent)
        if(Cine:=Cine_Unlock_L12[GetTeamIndexAgent[Agent]-1]):#Swap eggs to unlock + Barrier Box removed
            Cine.Play(Agent)
        VFX_B11.Restart()
        if:
            PlayerObj := player[Agent]
            CustomPlayer:Custom_Player = PlayersMap[PlayerObj]
        then:
            set CustomPlayer.B11_Unlocked=1
            CustomPlayer.Update_Pets_Section()

    UnlockBarrierB12(Agent : agent):void = 
        BarrierL12.AddToIgnoreList(Agent)
        if(Cine:=Cine_Unlock_L13[GetTeamIndexAgent[Agent]-1]):#Swap eggs to unlock + Barrier Box removed
            Cine.Play(Agent)
        VFX_B12.Restart()
        if:
            PlayerObj := player[Agent]
            CustomPlayer:Custom_Player = PlayersMap[PlayerObj]
        then:
            set CustomPlayer.B12_Unlocked=1
            CustomPlayer.Update_Pets_Section()

    UnlockBarrierB13(Agent : agent):void = 
        BarrierL13.AddToIgnoreList(Agent)
        if(Cine:=Cine_Unlock_L14[GetTeamIndexAgent[Agent]-1]):#Swap eggs to unlock + Barrier Box removed
            Cine.Play(Agent)
        VFX_B13.Restart()
        if:
            PlayerObj := player[Agent]
            CustomPlayer:Custom_Player = PlayersMap[PlayerObj]
        then:
            set CustomPlayer.B13_Unlocked=1
            CustomPlayer.Update_Pets_Section()

    UnlockBarrierB14(Agent : agent):void = 
        BarrierL14.AddToIgnoreList(Agent)
        if(Cine:=Cine_Unlock_L15[GetTeamIndexAgent[Agent]-1]):#Swap eggs to unlock + Barrier Box removed
            Cine.Play(Agent)
        VFX_B14.Restart()
        if:
            PlayerObj := player[Agent]
            CustomPlayer:Custom_Player = PlayersMap[PlayerObj]
        then:
            set CustomPlayer.B14_Unlocked=1
            CustomPlayer.Update_Pets_Section()

    UnlockBarrierB15(Agent : agent):void = 
        BarrierL15.AddToIgnoreList(Agent)
        if(Cine:=Cine_Unlock_L16[GetTeamIndexAgent[Agent]-1]):#Swap eggs to unlock + Barrier Box removed
            Cine.Play(Agent)
        VFX_B15.Restart()
        if:
            PlayerObj := player[Agent]
            CustomPlayer:Custom_Player = PlayersMap[PlayerObj]
        then:
            set CustomPlayer.B15_Unlocked=1
            CustomPlayer.Update_Pets_Section()

    UnlockBarrierB16(Agent : agent):void = 
        BarrierL16.AddToIgnoreList(Agent)
        if(Cine:=Cine_Unlock_L17[GetTeamIndexAgent[Agent]-1]):#Swap eggs to unlock + Barrier Box removed
            Cine.Play(Agent)
        VFX_B16.Restart()
        if:
            PlayerObj := player[Agent]
            CustomPlayer:Custom_Player = PlayersMap[PlayerObj]
        then:
            set CustomPlayer.B16_Unlocked=1
            CustomPlayer.Update_Pets_Section()

    UnlockBarrierB17(Agent : agent):void = 
        BarrierL17.AddToIgnoreList(Agent)
        if(Cine:=Cine_Unlock_L18[GetTeamIndexAgent[Agent]-1]):#Swap eggs to unlock + Barrier Box removed
            Cine.Play(Agent)
        VFX_B17.Restart()
        if:
            PlayerObj := player[Agent]
            CustomPlayer:Custom_Player = PlayersMap[PlayerObj]
        then:
            set CustomPlayer.B17_Unlocked=1
            CustomPlayer.Update_Pets_Section()

    UnlockBarrierB18(Agent : agent):void = 
        BarrierL18.AddToIgnoreList(Agent)
        if(Cine:=Cine_Unlock_L19[GetTeamIndexAgent[Agent]-1]):#Swap eggs to unlock + Barrier Box removed
            Cine.Play(Agent)
        VFX_B18.Restart()
        if:
            PlayerObj := player[Agent]
            CustomPlayer:Custom_Player = PlayersMap[PlayerObj]
        then:
            set CustomPlayer.B18_Unlocked=1
            CustomPlayer.Update_Pets_Section()

    UnlockBarrierB19(Agent : agent):void = 
        BarrierL19.AddToIgnoreList(Agent)
        if(Cine:=Cine_Unlock_L20[GetTeamIndexAgent[Agent]-1]):#Swap eggs to unlock + Barrier Box removed
            Cine.Play(Agent)
        VFX_B19.Restart()
        if:
            PlayerObj := player[Agent]
            CustomPlayer:Custom_Player = PlayersMap[PlayerObj]
        then:
            set CustomPlayer.B19_Unlocked=1
            CustomPlayer.Update_Pets_Section()

    UnlockBarrierB20(Agent : agent):void = 
        BarrierL20.AddToIgnoreList(Agent)
        if(Cine:=Cine_Unlock_L21[GetTeamIndexAgent[Agent]-1]):#Swap eggs to unlock + Barrier Box removed
            Cine.Play(Agent)
        VFX_B20.Restart()
        if:
            PlayerObj := player[Agent]
            CustomPlayer:Custom_Player = PlayersMap[PlayerObj]
        then:
            set CustomPlayer.B20_Unlocked=1
            CustomPlayer.Update_Pets_Section()

    UnlockBarrierB21(Agent : agent):void = 
        BarrierL21.AddToIgnoreList(Agent)
        if(Cine:=Cine_Unlock_L22[GetTeamIndexAgent[Agent]-1]):#Swap eggs to unlock + Barrier Box removed
            Cine.Play(Agent)
        VFX_B21.Restart()
        if:
            PlayerObj := player[Agent]
            CustomPlayer:Custom_Player = PlayersMap[PlayerObj]
        then:
            set CustomPlayer.B21_Unlocked=1
            CustomPlayer.Update_Pets_Section()

    UnlockBarrierB22(Agent : agent):void = 
        BarrierL22.AddToIgnoreList(Agent)
        if(Cine:=Cine_Unlock_L23[GetTeamIndexAgent[Agent]-1]):#Swap eggs to unlock + Barrier Box removed
            Cine.Play(Agent)
        VFX_B22.Restart()
        if:
            PlayerObj := player[Agent]
            CustomPlayer:Custom_Player = PlayersMap[PlayerObj]
        then:
            set CustomPlayer.B22_Unlocked=1
            CustomPlayer.Update_Pets_Section()

    UnlockBarrierB23(Agent : agent):void = 
        BarrierL23.AddToIgnoreList(Agent)
        if(Cine:=Cine_Unlock_L24[GetTeamIndexAgent[Agent]-1]):#Swap eggs to unlock + Barrier Box removed
            Cine.Play(Agent)
        VFX_B23.Restart()
        if:
            PlayerObj := player[Agent]
            CustomPlayer:Custom_Player = PlayersMap[PlayerObj]
        then:
            set CustomPlayer.B23_Unlocked=1
            CustomPlayer.Update_Pets_Section()

    UnlockBarrierB24(Agent : agent):void = 
        BarrierL24.AddToIgnoreList(Agent)
        if(Cine:=Cine_Unlock_L25[GetTeamIndexAgent[Agent]-1]):#Swap eggs to unlock + Barrier Box removed
            Cine.Play(Agent)
        VFX_B24.Restart()
        if:
            PlayerObj := player[Agent]
            CustomPlayer:Custom_Player = PlayersMap[PlayerObj]
        then:
            set CustomPlayer.B24_Unlocked=1
            CustomPlayer.Update_Pets_Section()

    UnlockBarrierB25(Agent : agent):void = 
        BarrierL25.AddToIgnoreList(Agent)
        if(Cine:=Cine_Unlock_L26[GetTeamIndexAgent[Agent]-1]):#Swap eggs to unlock + Barrier Box removed
            Cine.Play(Agent)
        VFX_B25.Restart()
        if:
            PlayerObj := player[Agent]
            CustomPlayer:Custom_Player = PlayersMap[PlayerObj]
        then:
            set CustomPlayer.B25_Unlocked=1
            CustomPlayer.Update_Pets_Section()

    UnlockBarrierB26(Agent : agent):void = 
        BarrierL26.AddToIgnoreList(Agent)
        if(Cine:=Cine_Unlock_L27[GetTeamIndexAgent[Agent]-1]):#Swap eggs to unlock + Barrier Box removed
            Cine.Play(Agent)
        VFX_B26.Restart()
        if:
            PlayerObj := player[Agent]
            CustomPlayer:Custom_Player = PlayersMap[PlayerObj]
        then:
            set CustomPlayer.B26_Unlocked=1
            CustomPlayer.Update_Pets_Section()

    UnlockBarrierB27(Agent : agent):void = 
        BarrierL27.AddToIgnoreList(Agent)
        if(Cine:=Cine_Unlock_L28[GetTeamIndexAgent[Agent]-1]):#Swap eggs to unlock + Barrier Box removed
            Cine.Play(Agent)
        VFX_B27.Restart()
        if:
            PlayerObj := player[Agent]
            CustomPlayer:Custom_Player = PlayersMap[PlayerObj]
        then:
            set CustomPlayer.B27_Unlocked=1
            CustomPlayer.Update_Pets_Section()

    UnlockBarrierB28(Agent : agent):void = 
        BarrierL28.AddToIgnoreList(Agent)
        if(Cine:=Cine_Unlock_L29[GetTeamIndexAgent[Agent]-1]):#Swap eggs to unlock + Barrier Box removed
            Cine.Play(Agent)
        VFX_B28.Restart()
        if:
            PlayerObj := player[Agent]
            CustomPlayer:Custom_Player = PlayersMap[PlayerObj]
        then:
            set CustomPlayer.B28_Unlocked=1
            CustomPlayer.Update_Pets_Section()

    UnlockBarrierB29(Agent : agent):void = 
        BarrierL29.AddToIgnoreList(Agent)
        if(Cine:=Cine_Unlock_L30[GetTeamIndexAgent[Agent]-1]):#Swap eggs to unlock + Barrier Box removed
            Cine.Play(Agent)
        VFX_B29.Restart()
        if:
            PlayerObj := player[Agent]
            CustomPlayer:Custom_Player = PlayersMap[PlayerObj]
        then:
            set CustomPlayer.B29_Unlocked=1
            CustomPlayer.Update_Pets_Section()

    #NOT USED END OF MAP
    UnlockBarrierB30(Agent : agent):void = 
        BarrierL30.AddToIgnoreList(Agent)
        if(Cine:=Cine_Unlock_L30[GetTeamIndexAgent[Agent]-1]):
            Cine.Play(Agent)
        VFX_B30.Restart()
        if:
            PlayerObj := player[Agent]
            CustomPlayer:Custom_Player = PlayersMap[PlayerObj]
        then:
            set CustomPlayer.B30_Unlocked=1
            CustomPlayer.Update_Pets_Section()

    #REBIRTH 1
    UnlockBarrierB31(Agent : agent):void = 
        BarrierL31.AddToIgnoreList(Agent)
        if(Cine:=Cine_RebirthBox1[GetTeamIndexAgent[Agent]-1]):#Barrier Box removed
            Cine.Play(Agent)
        if:
            PlayerObj := player[Agent]
            CustomPlayer:Custom_Player = PlayersMap[PlayerObj]
        then:
            set CustomPlayer.B31_Unlocked=1
            CustomPlayer.Update_Pets_Section()
    
    #REBIRTH 2
    UnlockBarrierB32(Agent : agent):void = 
        BarrierL32.AddToIgnoreList(Agent)
        if(Cine:=Cine_RebirthBox2[GetTeamIndexAgent[Agent]-1]):#Barrier Box removed
            Cine.Play(Agent)
        if:
            PlayerObj := player[Agent]
            CustomPlayer:Custom_Player = PlayersMap[PlayerObj]
        then:
            set CustomPlayer.B32_Unlocked=1
            CustomPlayer.Update_Pets_Section()

    #REBIRTH 3
    UnlockBarrierB33(Agent : agent):void = 
        BarrierL33.AddToIgnoreList(Agent)
        if(Cine:=Cine_RebirthBox3[GetTeamIndexAgent[Agent]-1]):#Barrier Box removed
            Cine.Play(Agent)
        if:
            PlayerObj := player[Agent]
            CustomPlayer:Custom_Player = PlayersMap[PlayerObj]
        then:
            set CustomPlayer.B33_Unlocked=1
            CustomPlayer.Update_Pets_Section()

    ####################################### GAME SAVE RELATED #######################################

    InitializeL1Pets(Agent:agent)<suspends>:void=
        Sleep(0.1)
        var FirstPet:creative_prop = creative_prop{}
        var SecondPet:creative_prop = creative_prop{}
        var ThirdPet:creative_prop = creative_prop{}
        var ForthPet:creative_prop = creative_prop{}

        var FirstPetIndex:int = 0
        var SecondPetIndex:int = 0
        var ThirdPetIndex:int = 0
        var ForthPetIndex:int = 0

        var IndexBase:int = 0
        #Select right pets for each player
        if(GetTeamIndexAgent[Agent]=1):
            set FirstPetIndex =  IndexBase
            set SecondPetIndex = IndexBase+1
            set ThirdPetIndex =  IndexBase+2
            set ForthPetIndex =  IndexBase+3
        if(GetTeamIndexAgent[Agent]=2):
            set FirstPetIndex =  IndexBase+4
            set SecondPetIndex = IndexBase+5
            set ThirdPetIndex =  IndexBase+6
            set ForthPetIndex =  IndexBase+7
        if(GetTeamIndexAgent[Agent]=3):
            set FirstPetIndex =  IndexBase+8
            set SecondPetIndex = IndexBase+9
            set ThirdPetIndex =  IndexBase+10
            set ForthPetIndex =  IndexBase+11
        
        if (Mypet:=Pets[FirstPetIndex]):
            set FirstPet = Mypet.PetProp
        if (Mypet:=Pets[SecondPetIndex]):
            set SecondPet = Mypet.PetProp
        if (Mypet:=Pets[ThirdPetIndex]):
            set ThirdPet = Mypet.PetProp
        if (Mypet:=Pets[ForthPetIndex]):
            set ForthPet = Mypet.PetProp

        if:
            CurrentPlayerStats := PlayerStatsManager.GetPlayerStats[Agent]
            PlayerObj := player[Agent]
            CustomPlayer:Custom_Player = PlayersMap[PlayerObj]
        then:
            if(CurrentPlayerStats.Ownership_L1_Pet4.CurrentValue=1 and CustomPlayer.FollowActiveL1=0):
                if (Fort := Agent.GetFortCharacter[]):
                    spawn:
                        FollowPlayerL1(Fort, ForthPet, Power_L1P4*1.0, ForthPetIndex)
            if(CurrentPlayerStats.Ownership_L1_Pet3.CurrentValue=1 and CustomPlayer.FollowActiveL1=0):
                if (Fort := Agent.GetFortCharacter[]):
                    spawn:
                        FollowPlayerL1(Fort, ThirdPet, Power_L1P3*1.0, ThirdPetIndex)
            if(CurrentPlayerStats.Ownership_L1_Pet2.CurrentValue=1 and CustomPlayer.FollowActiveL1=0):
                if (Fort := Agent.GetFortCharacter[]):
                    spawn:
                        FollowPlayerL1(Fort, SecondPet, Power_L1P2*1.0, SecondPetIndex)
            if(CurrentPlayerStats.Ownership_L1_Pet1.CurrentValue=1 and CustomPlayer.FollowActiveL1=0):
                if (Fort := Agent.GetFortCharacter[]):
                    spawn:
                        FollowPlayerL1(Fort, FirstPet, Power_L1P1*1.0, FirstPetIndex)
                
    InitializeL2Pets(Agent:agent)<suspends>:void=
        Sleep(0.2)
        var FirstPet:creative_prop = creative_prop{}
        var SecondPet:creative_prop = creative_prop{}
        var ThirdPet:creative_prop = creative_prop{}
        var ForthPet:creative_prop = creative_prop{}

        var FirstPetIndex:int = 0
        var SecondPetIndex:int = 0
        var ThirdPetIndex:int = 0
        var ForthPetIndex:int = 0

        var IndexBase:int = 12
        #Select right pets for each player
        if(GetTeamIndexAgent[Agent]=1):
            set FirstPetIndex =  IndexBase
            set SecondPetIndex = IndexBase+1
            set ThirdPetIndex =  IndexBase+2
            set ForthPetIndex =  IndexBase+3
        if(GetTeamIndexAgent[Agent]=2):
            set FirstPetIndex =  IndexBase+4
            set SecondPetIndex = IndexBase+5
            set ThirdPetIndex =  IndexBase+6
            set ForthPetIndex =  IndexBase+7
        if(GetTeamIndexAgent[Agent]=3):
            set FirstPetIndex =  IndexBase+8
            set SecondPetIndex = IndexBase+9
            set ThirdPetIndex =  IndexBase+10
            set ForthPetIndex =  IndexBase+11
        
        if (Mypet:=Pets[FirstPetIndex]):
            set FirstPet = Mypet.PetProp
        if (Mypet:=Pets[SecondPetIndex]):
            set SecondPet = Mypet.PetProp
        if (Mypet:=Pets[ThirdPetIndex]):
            set ThirdPet = Mypet.PetProp
        if (Mypet:=Pets[ForthPetIndex]):
            set ForthPet = Mypet.PetProp

        if:
            CurrentPlayerStats := PlayerStatsManager.GetPlayerStats[Agent]
            PlayerObj := player[Agent]
            CustomPlayer:Custom_Player = PlayersMap[PlayerObj]
        then:
            if(CurrentPlayerStats.Ownership_L2_Pet4.CurrentValue=1 and CustomPlayer.FollowActiveL2=0):
                if (Fort := Agent.GetFortCharacter[]):
                    spawn:
                        FollowPlayerL2(Fort, ForthPet, Power_L2P4*1.0, ForthPetIndex)
            if(CurrentPlayerStats.Ownership_L2_Pet3.CurrentValue=1 and CustomPlayer.FollowActiveL2=0):
                if (Fort := Agent.GetFortCharacter[]):
                    spawn:
                        FollowPlayerL2(Fort, ThirdPet, Power_L2P3*1.0, ThirdPetIndex)
            if(CurrentPlayerStats.Ownership_L2_Pet2.CurrentValue=1 and CustomPlayer.FollowActiveL2=0):
                if (Fort := Agent.GetFortCharacter[]):
                    spawn:
                        FollowPlayerL2(Fort, SecondPet, Power_L2P2*1.0, SecondPetIndex)
            if(CurrentPlayerStats.Ownership_L2_Pet1.CurrentValue=1 and CustomPlayer.FollowActiveL2=0):
                if (Fort := Agent.GetFortCharacter[]):
                    spawn:
                        FollowPlayerL2(Fort, FirstPet, Power_L2P1*1.0, FirstPetIndex)

    InitializeL3Pets(Agent:agent)<suspends>:void=
        Sleep(0.3)
        var FirstPet:creative_prop = creative_prop{}
        var SecondPet:creative_prop = creative_prop{}
        var ThirdPet:creative_prop = creative_prop{}
        var ForthPet:creative_prop = creative_prop{}

        var FirstPetIndex:int = 0
        var SecondPetIndex:int = 0
        var ThirdPetIndex:int = 0
        var ForthPetIndex:int = 0

        var IndexBase:int = 12*2
        #Select right pets for each player
        if(GetTeamIndexAgent[Agent]=1):
            set FirstPetIndex =  IndexBase
            set SecondPetIndex = IndexBase+1
            set ThirdPetIndex =  IndexBase+2
            set ForthPetIndex =  IndexBase+3
        if(GetTeamIndexAgent[Agent]=2):
            set FirstPetIndex =  IndexBase+4
            set SecondPetIndex = IndexBase+5
            set ThirdPetIndex =  IndexBase+6
            set ForthPetIndex =  IndexBase+7
        if(GetTeamIndexAgent[Agent]=3):
            set FirstPetIndex =  IndexBase+8
            set SecondPetIndex = IndexBase+9
            set ThirdPetIndex =  IndexBase+10
            set ForthPetIndex =  IndexBase+11
        
        if (Mypet:=Pets[FirstPetIndex]):
            set FirstPet = Mypet.PetProp
        if (Mypet:=Pets[SecondPetIndex]):
            set SecondPet = Mypet.PetProp
        if (Mypet:=Pets[ThirdPetIndex]):
            set ThirdPet = Mypet.PetProp
        if (Mypet:=Pets[ForthPetIndex]):
            set ForthPet = Mypet.PetProp

        if:
            CurrentPlayerStats := PlayerStatsManager.GetPlayerStats[Agent]
            PlayerObj := player[Agent]
            CustomPlayer:Custom_Player = PlayersMap[PlayerObj]
        then:
            if(CurrentPlayerStats.Ownership_L3_Pet4.CurrentValue=1 and CustomPlayer.FollowActiveL3=0):
                if (Fort := Agent.GetFortCharacter[]):
                    spawn:
                        FollowPlayerL3(Fort, ForthPet, Power_L3P4*1.0, ForthPetIndex)
            if(CurrentPlayerStats.Ownership_L3_Pet3.CurrentValue=1 and CustomPlayer.FollowActiveL3=0):
                if (Fort := Agent.GetFortCharacter[]):
                    spawn:
                        FollowPlayerL3(Fort, ThirdPet, Power_L3P3*1.0, ThirdPetIndex)
            if(CurrentPlayerStats.Ownership_L3_Pet2.CurrentValue=1 and CustomPlayer.FollowActiveL3=0):
                if (Fort := Agent.GetFortCharacter[]):
                    spawn:
                        FollowPlayerL3(Fort, SecondPet, Power_L3P2*1.0, SecondPetIndex)
            if(CurrentPlayerStats.Ownership_L3_Pet1.CurrentValue=1 and CustomPlayer.FollowActiveL3=0):
                if (Fort := Agent.GetFortCharacter[]):
                    spawn:
                        FollowPlayerL3(Fort, FirstPet, Power_L3P1*1.0, FirstPetIndex)

    InitializeL4Pets(Agent:agent)<suspends>:void=
        Sleep(0.4)
        var FirstPet:creative_prop = creative_prop{}
        var SecondPet:creative_prop = creative_prop{}
        var ThirdPet:creative_prop = creative_prop{}
        var ForthPet:creative_prop = creative_prop{}

        var FirstPetIndex:int = 0
        var SecondPetIndex:int = 0
        var ThirdPetIndex:int = 0
        var ForthPetIndex:int = 0

        var IndexBase:int = 12*3
        #Select right pets for each player
        if(GetTeamIndexAgent[Agent]=1):
            set FirstPetIndex =  IndexBase
            set SecondPetIndex = IndexBase+1
            set ThirdPetIndex =  IndexBase+2
            set ForthPetIndex =  IndexBase+3
        if(GetTeamIndexAgent[Agent]=2):
            set FirstPetIndex =  IndexBase+4
            set SecondPetIndex = IndexBase+5
            set ThirdPetIndex =  IndexBase+6
            set ForthPetIndex =  IndexBase+7
        if(GetTeamIndexAgent[Agent]=3):
            set FirstPetIndex =  IndexBase+8
            set SecondPetIndex = IndexBase+9
            set ThirdPetIndex =  IndexBase+10
            set ForthPetIndex =  IndexBase+11
        
        if (Mypet:=Pets[FirstPetIndex]):
            set FirstPet = Mypet.PetProp
        if (Mypet:=Pets[SecondPetIndex]):
            set SecondPet = Mypet.PetProp
        if (Mypet:=Pets[ThirdPetIndex]):
            set ThirdPet = Mypet.PetProp
        if (Mypet:=Pets[ForthPetIndex]):
            set ForthPet = Mypet.PetProp

        if:
            CurrentPlayerStats := PlayerStatsManager.GetPlayerStats[Agent]
            PlayerObj := player[Agent]
            CustomPlayer:Custom_Player = PlayersMap[PlayerObj]
        then:
            if(CurrentPlayerStats.Ownership_L4_Pet4.CurrentValue=1 and CustomPlayer.FollowActiveL4=0):
                if (Fort := Agent.GetFortCharacter[]):
                    spawn:
                        FollowPlayerL4(Fort, ForthPet, Power_L4P4*1.0, ForthPetIndex)
            if(CurrentPlayerStats.Ownership_L4_Pet3.CurrentValue=1 and CustomPlayer.FollowActiveL4=0):
                if (Fort := Agent.GetFortCharacter[]):
                    spawn:
                        FollowPlayerL4(Fort, ThirdPet, Power_L4P3*1.0, ThirdPetIndex)
            if(CurrentPlayerStats.Ownership_L4_Pet2.CurrentValue=1 and CustomPlayer.FollowActiveL4=0):
                if (Fort := Agent.GetFortCharacter[]):
                    spawn:
                        FollowPlayerL4(Fort, SecondPet, Power_L4P2*1.0, SecondPetIndex)
            if(CurrentPlayerStats.Ownership_L4_Pet1.CurrentValue=1 and CustomPlayer.FollowActiveL4=0):
                if (Fort := Agent.GetFortCharacter[]):
                    spawn:
                        FollowPlayerL4(Fort, FirstPet, Power_L4P1*1.0, FirstPetIndex)

    InitializeL5Pets(Agent:agent)<suspends>:void=
        Sleep(0.5)
        var FirstPet:creative_prop = creative_prop{}
        var SecondPet:creative_prop = creative_prop{}
        var ThirdPet:creative_prop = creative_prop{}
        var ForthPet:creative_prop = creative_prop{}

        var FirstPetIndex:int = 0
        var SecondPetIndex:int = 0
        var ThirdPetIndex:int = 0
        var ForthPetIndex:int = 0

        var IndexBase:int = 12*4
        #Select right pets for each player
        if(GetTeamIndexAgent[Agent]=1):
            set FirstPetIndex =  IndexBase
            set SecondPetIndex = IndexBase+1
            set ThirdPetIndex =  IndexBase+2
            set ForthPetIndex =  IndexBase+3
        if(GetTeamIndexAgent[Agent]=2):
            set FirstPetIndex =  IndexBase+4
            set SecondPetIndex = IndexBase+5
            set ThirdPetIndex =  IndexBase+6
            set ForthPetIndex =  IndexBase+7
        if(GetTeamIndexAgent[Agent]=3):
            set FirstPetIndex =  IndexBase+8
            set SecondPetIndex = IndexBase+9
            set ThirdPetIndex =  IndexBase+10
            set ForthPetIndex =  IndexBase+11
        
        if (Mypet:=Pets[FirstPetIndex]):
            set FirstPet = Mypet.PetProp
        if (Mypet:=Pets[SecondPetIndex]):
            set SecondPet = Mypet.PetProp
        if (Mypet:=Pets[ThirdPetIndex]):
            set ThirdPet = Mypet.PetProp
        if (Mypet:=Pets[ForthPetIndex]):
            set ForthPet = Mypet.PetProp

        if:
            CurrentPlayerStats := PlayerStatsManager.GetPlayerStats[Agent]
            PlayerObj := player[Agent]
            CustomPlayer:Custom_Player = PlayersMap[PlayerObj]
        then:
            if(CurrentPlayerStats.Ownership_L5_Pet4.CurrentValue=1 and CustomPlayer.FollowActiveL5=0):
                if (Fort := Agent.GetFortCharacter[]):
                    spawn:
                        FollowPlayerL5(Fort, ForthPet, Power_L5P4*1.0, ForthPetIndex)
            if(CurrentPlayerStats.Ownership_L5_Pet3.CurrentValue=1 and CustomPlayer.FollowActiveL5=0):
                if (Fort := Agent.GetFortCharacter[]):
                    spawn:
                        FollowPlayerL5(Fort, ThirdPet, Power_L5P3*1.0, ThirdPetIndex)
            if(CurrentPlayerStats.Ownership_L5_Pet2.CurrentValue=1 and CustomPlayer.FollowActiveL5=0):
                if (Fort := Agent.GetFortCharacter[]):
                    spawn:
                        FollowPlayerL5(Fort, SecondPet, Power_L5P2*1.0, SecondPetIndex)
            if(CurrentPlayerStats.Ownership_L5_Pet1.CurrentValue=1 and CustomPlayer.FollowActiveL5=0):
                if (Fort := Agent.GetFortCharacter[]):
                    spawn:
                        FollowPlayerL5(Fort, FirstPet, Power_L5P1*1.0, FirstPetIndex)

    InitializeL6Pets(Agent:agent)<suspends>:void=
        Sleep(0.6)
        var FirstPet:creative_prop = creative_prop{}
        var SecondPet:creative_prop = creative_prop{}
        var ThirdPet:creative_prop = creative_prop{}
        var ForthPet:creative_prop = creative_prop{}

        var FirstPetIndex:int = 0
        var SecondPetIndex:int = 0
        var ThirdPetIndex:int = 0
        var ForthPetIndex:int = 0

        var IndexBase:int = 12*5
        #Select right pets for each player
        if(GetTeamIndexAgent[Agent]=1):
            set FirstPetIndex =  IndexBase
            set SecondPetIndex = IndexBase+1
            set ThirdPetIndex =  IndexBase+2
            set ForthPetIndex =  IndexBase+3
        if(GetTeamIndexAgent[Agent]=2):
            set FirstPetIndex =  IndexBase+4
            set SecondPetIndex = IndexBase+5
            set ThirdPetIndex =  IndexBase+6
            set ForthPetIndex =  IndexBase+7
        if(GetTeamIndexAgent[Agent]=3):
            set FirstPetIndex =  IndexBase+8
            set SecondPetIndex = IndexBase+9
            set ThirdPetIndex =  IndexBase+10
            set ForthPetIndex =  IndexBase+11
        
        if (Mypet:=Pets[FirstPetIndex]):
            set FirstPet = Mypet.PetProp
        if (Mypet:=Pets[SecondPetIndex]):
            set SecondPet = Mypet.PetProp
        if (Mypet:=Pets[ThirdPetIndex]):
            set ThirdPet = Mypet.PetProp
        if (Mypet:=Pets[ForthPetIndex]):
            set ForthPet = Mypet.PetProp

        if:
            CurrentPlayerStats := PlayerStatsManager.GetPlayerStats[Agent]
            PlayerObj := player[Agent]
            CustomPlayer:Custom_Player = PlayersMap[PlayerObj]
        then:
            if(CurrentPlayerStats.Ownership_L6_Pet4.CurrentValue=1 and CustomPlayer.FollowActiveL6=0):
                if (Fort := Agent.GetFortCharacter[]):
                    spawn:
                        FollowPlayerL6(Fort, ForthPet, Power_L6P4*1.0, ForthPetIndex)
            if(CurrentPlayerStats.Ownership_L6_Pet3.CurrentValue=1 and CustomPlayer.FollowActiveL6=0):
                if (Fort := Agent.GetFortCharacter[]):
                    spawn:
                        FollowPlayerL6(Fort, ThirdPet, Power_L6P3*1.0, ThirdPetIndex)
            if(CurrentPlayerStats.Ownership_L6_Pet2.CurrentValue=1 and CustomPlayer.FollowActiveL6=0):
                if (Fort := Agent.GetFortCharacter[]):
                    spawn:
                        FollowPlayerL6(Fort, SecondPet, Power_L6P2*1.0, SecondPetIndex)
            if(CurrentPlayerStats.Ownership_L6_Pet1.CurrentValue=1 and CustomPlayer.FollowActiveL6=0):
                if (Fort := Agent.GetFortCharacter[]):
                    spawn:
                        FollowPlayerL6(Fort, FirstPet, Power_L6P1*1.0, FirstPetIndex)

    InitializeL7Pets(Agent:agent)<suspends>:void=
        Sleep(0.7)
        var FirstPet:creative_prop = creative_prop{}
        var SecondPet:creative_prop = creative_prop{}
        var ThirdPet:creative_prop = creative_prop{}
        var ForthPet:creative_prop = creative_prop{}

        var FirstPetIndex:int = 0
        var SecondPetIndex:int = 0
        var ThirdPetIndex:int = 0
        var ForthPetIndex:int = 0

        var IndexBase:int = 12*6
        #Select right pets for each player
        if(GetTeamIndexAgent[Agent]=1):
            set FirstPetIndex =  IndexBase
            set SecondPetIndex = IndexBase+1
            set ThirdPetIndex =  IndexBase+2
            set ForthPetIndex =  IndexBase+3
        if(GetTeamIndexAgent[Agent]=2):
            set FirstPetIndex =  IndexBase+4
            set SecondPetIndex = IndexBase+5
            set ThirdPetIndex =  IndexBase+6
            set ForthPetIndex =  IndexBase+7
        if(GetTeamIndexAgent[Agent]=3):
            set FirstPetIndex =  IndexBase+8
            set SecondPetIndex = IndexBase+9
            set ThirdPetIndex =  IndexBase+10
            set ForthPetIndex =  IndexBase+11
        
        if (Mypet:=Pets[FirstPetIndex]):
            set FirstPet = Mypet.PetProp
        if (Mypet:=Pets[SecondPetIndex]):
            set SecondPet = Mypet.PetProp
        if (Mypet:=Pets[ThirdPetIndex]):
            set ThirdPet = Mypet.PetProp
        if (Mypet:=Pets[ForthPetIndex]):
            set ForthPet = Mypet.PetProp

        if:
            CurrentPlayerStats := PlayerStatsManager.GetPlayerStats[Agent]
            PlayerObj := player[Agent]
            CustomPlayer:Custom_Player = PlayersMap[PlayerObj]
        then:
            if(CurrentPlayerStats.Ownership_L7_Pet4.CurrentValue=1 and CustomPlayer.FollowActiveL7=0):
                if (Fort := Agent.GetFortCharacter[]):
                    spawn:
                        FollowPlayerL7(Fort, ForthPet, Power_L7P4*1.0, ForthPetIndex)
            if(CurrentPlayerStats.Ownership_L7_Pet3.CurrentValue=1 and CustomPlayer.FollowActiveL7=0):
                if (Fort := Agent.GetFortCharacter[]):
                    spawn:
                        FollowPlayerL7(Fort, ThirdPet, Power_L7P3*1.0, ThirdPetIndex)
            if(CurrentPlayerStats.Ownership_L7_Pet2.CurrentValue=1 and CustomPlayer.FollowActiveL7=0):
                if (Fort := Agent.GetFortCharacter[]):
                    spawn:
                        FollowPlayerL7(Fort, SecondPet, Power_L7P2*1.0, SecondPetIndex)
            if(CurrentPlayerStats.Ownership_L7_Pet1.CurrentValue=1 and CustomPlayer.FollowActiveL7=0):
                if (Fort := Agent.GetFortCharacter[]):
                    spawn:
                        FollowPlayerL7(Fort, FirstPet, Power_L7P1*1.0, FirstPetIndex)

    InitializeL8Pets(Agent:agent)<suspends>:void=
        Sleep(0.8)
        var FirstPet:creative_prop = creative_prop{}
        var SecondPet:creative_prop = creative_prop{}
        var ThirdPet:creative_prop = creative_prop{}
        var ForthPet:creative_prop = creative_prop{}

        var FirstPetIndex:int = 0
        var SecondPetIndex:int = 0
        var ThirdPetIndex:int = 0
        var ForthPetIndex:int = 0

        var IndexBase:int = 12*7
        #Select right pets for each player
        if(GetTeamIndexAgent[Agent]=1):
            set FirstPetIndex =  IndexBase
            set SecondPetIndex = IndexBase+1
            set ThirdPetIndex =  IndexBase+2
            set ForthPetIndex =  IndexBase+3
        if(GetTeamIndexAgent[Agent]=2):
            set FirstPetIndex =  IndexBase+4
            set SecondPetIndex = IndexBase+5
            set ThirdPetIndex =  IndexBase+6
            set ForthPetIndex =  IndexBase+7
        if(GetTeamIndexAgent[Agent]=3):
            set FirstPetIndex =  IndexBase+8
            set SecondPetIndex = IndexBase+9
            set ThirdPetIndex =  IndexBase+10
            set ForthPetIndex =  IndexBase+11
        
        if (Mypet:=Pets[FirstPetIndex]):
            set FirstPet = Mypet.PetProp
        if (Mypet:=Pets[SecondPetIndex]):
            set SecondPet = Mypet.PetProp
        if (Mypet:=Pets[ThirdPetIndex]):
            set ThirdPet = Mypet.PetProp
        if (Mypet:=Pets[ForthPetIndex]):
            set ForthPet = Mypet.PetProp

        if:
            CurrentPlayerStats := PlayerStatsManager.GetPlayerStats[Agent]
            PlayerObj := player[Agent]
            CustomPlayer:Custom_Player = PlayersMap[PlayerObj]
        then:
            if(CurrentPlayerStats.Ownership_L8_Pet4.CurrentValue=1 and CustomPlayer.FollowActiveL8=0):
                if (Fort := Agent.GetFortCharacter[]):
                    spawn:
                        FollowPlayerL8(Fort, ForthPet, Power_L8P4*1.0, ForthPetIndex)
            if(CurrentPlayerStats.Ownership_L8_Pet3.CurrentValue=1 and CustomPlayer.FollowActiveL8=0):
                if (Fort := Agent.GetFortCharacter[]):
                    spawn:
                        FollowPlayerL8(Fort, ThirdPet, Power_L8P3*1.0, ThirdPetIndex)
            if(CurrentPlayerStats.Ownership_L8_Pet2.CurrentValue=1 and CustomPlayer.FollowActiveL8=0):
                if (Fort := Agent.GetFortCharacter[]):
                    spawn:
                        FollowPlayerL8(Fort, SecondPet, Power_L8P2*1.0, SecondPetIndex)
            if(CurrentPlayerStats.Ownership_L8_Pet1.CurrentValue=1 and CustomPlayer.FollowActiveL8=0):
                if (Fort := Agent.GetFortCharacter[]):
                    spawn:
                        FollowPlayerL8(Fort, FirstPet, Power_L8P1*1.0, FirstPetIndex)

    InitializeL9Pets(Agent:agent)<suspends>:void=
        Sleep(0.9)
        var FirstPet:creative_prop = creative_prop{}
        var SecondPet:creative_prop = creative_prop{}
        var ThirdPet:creative_prop = creative_prop{}
        var ForthPet:creative_prop = creative_prop{}

        var FirstPetIndex:int = 0
        var SecondPetIndex:int = 0
        var ThirdPetIndex:int = 0
        var ForthPetIndex:int = 0

        var IndexBase:int = 12*8
        #Select right pets for each player
        if(GetTeamIndexAgent[Agent]=1):
            set FirstPetIndex =  IndexBase
            set SecondPetIndex = IndexBase+1
            set ThirdPetIndex =  IndexBase+2
            set ForthPetIndex =  IndexBase+3
        if(GetTeamIndexAgent[Agent]=2):
            set FirstPetIndex =  IndexBase+4
            set SecondPetIndex = IndexBase+5
            set ThirdPetIndex =  IndexBase+6
            set ForthPetIndex =  IndexBase+7
        if(GetTeamIndexAgent[Agent]=3):
            set FirstPetIndex =  IndexBase+8
            set SecondPetIndex = IndexBase+9
            set ThirdPetIndex =  IndexBase+10
            set ForthPetIndex =  IndexBase+11
        
        if (Mypet:=Pets[FirstPetIndex]):
            set FirstPet = Mypet.PetProp
        if (Mypet:=Pets[SecondPetIndex]):
            set SecondPet = Mypet.PetProp
        if (Mypet:=Pets[ThirdPetIndex]):
            set ThirdPet = Mypet.PetProp
        if (Mypet:=Pets[ForthPetIndex]):
            set ForthPet = Mypet.PetProp

        if:
            CurrentPlayerStats := PlayerStatsManager.GetPlayerStats[Agent]
            PlayerObj := player[Agent]
            CustomPlayer:Custom_Player = PlayersMap[PlayerObj]
        then:
            if(CurrentPlayerStats.Ownership_L9_Pet4.CurrentValue=1 and CustomPlayer.FollowActiveL9=0):
                if (Fort := Agent.GetFortCharacter[]):
                    spawn:
                        FollowPlayerL9(Fort, ForthPet, Power_L9P4*1.0, ForthPetIndex)
            if(CurrentPlayerStats.Ownership_L9_Pet3.CurrentValue=1 and CustomPlayer.FollowActiveL9=0):
                if (Fort := Agent.GetFortCharacter[]):
                    spawn:
                        FollowPlayerL9(Fort, ThirdPet, Power_L9P3*1.0, ThirdPetIndex)
            if(CurrentPlayerStats.Ownership_L9_Pet2.CurrentValue=1 and CustomPlayer.FollowActiveL9=0):
                if (Fort := Agent.GetFortCharacter[]):
                    spawn:
                        FollowPlayerL9(Fort, SecondPet, Power_L9P2*1.0, SecondPetIndex)
            if(CurrentPlayerStats.Ownership_L9_Pet1.CurrentValue=1 and CustomPlayer.FollowActiveL9=0):
                if (Fort := Agent.GetFortCharacter[]):
                    spawn:
                        FollowPlayerL9(Fort, FirstPet, Power_L9P1*1.0, FirstPetIndex)

    InitializeL10Pets(Agent:agent)<suspends>:void=
        Sleep(1.0)
        var FirstPet:creative_prop = creative_prop{}
        var SecondPet:creative_prop = creative_prop{}
        var ThirdPet:creative_prop = creative_prop{}
        var ForthPet:creative_prop = creative_prop{}

        var FirstPetIndex:int = 0
        var SecondPetIndex:int = 0
        var ThirdPetIndex:int = 0
        var ForthPetIndex:int = 0

        var IndexBase:int = 12*9
        #Select right pets for each player
        if(GetTeamIndexAgent[Agent]=1):
            set FirstPetIndex =  IndexBase
            set SecondPetIndex = IndexBase+1
            set ThirdPetIndex =  IndexBase+2
            set ForthPetIndex =  IndexBase+3
        if(GetTeamIndexAgent[Agent]=2):
            set FirstPetIndex =  IndexBase+4
            set SecondPetIndex = IndexBase+5
            set ThirdPetIndex =  IndexBase+6
            set ForthPetIndex =  IndexBase+7
        if(GetTeamIndexAgent[Agent]=3):
            set FirstPetIndex =  IndexBase+8
            set SecondPetIndex = IndexBase+9
            set ThirdPetIndex =  IndexBase+10
            set ForthPetIndex =  IndexBase+11
        
        if (Mypet:=Pets[FirstPetIndex]):
            set FirstPet = Mypet.PetProp
        if (Mypet:=Pets[SecondPetIndex]):
            set SecondPet = Mypet.PetProp
        if (Mypet:=Pets[ThirdPetIndex]):
            set ThirdPet = Mypet.PetProp
        if (Mypet:=Pets[ForthPetIndex]):
            set ForthPet = Mypet.PetProp

        if:
            CurrentPlayerStats := PlayerStatsManager.GetPlayerStats[Agent]
            PlayerObj := player[Agent]
            CustomPlayer:Custom_Player = PlayersMap[PlayerObj]
        then:
            if(CurrentPlayerStats.Ownership_L10_Pet4.CurrentValue=1 and CustomPlayer.FollowActiveL10=0):
                if (Fort := Agent.GetFortCharacter[]):
                    spawn:
                        FollowPlayerL10(Fort, ForthPet, Power_L10P4*1.0, ForthPetIndex)
            if(CurrentPlayerStats.Ownership_L10_Pet3.CurrentValue=1 and CustomPlayer.FollowActiveL10=0):
                if (Fort := Agent.GetFortCharacter[]):
                    spawn:
                        FollowPlayerL10(Fort, ThirdPet, Power_L10P3*1.0, ThirdPetIndex)
            if(CurrentPlayerStats.Ownership_L10_Pet2.CurrentValue=1 and CustomPlayer.FollowActiveL10=0):
                if (Fort := Agent.GetFortCharacter[]):
                    spawn:
                        FollowPlayerL10(Fort, SecondPet, Power_L10P2*1.0, SecondPetIndex)
            if(CurrentPlayerStats.Ownership_L10_Pet1.CurrentValue=1 and CustomPlayer.FollowActiveL10=0):
                if (Fort := Agent.GetFortCharacter[]):
                    spawn:
                        FollowPlayerL10(Fort, FirstPet, Power_L10P1*1.0, FirstPetIndex)

    InitializeL11Pets(Agent:agent)<suspends>:void=
        Sleep(1.1)
        var FirstPet:creative_prop = creative_prop{}
        var SecondPet:creative_prop = creative_prop{}
        var ThirdPet:creative_prop = creative_prop{}
        var ForthPet:creative_prop = creative_prop{}

        var FirstPetIndex:int = 0
        var SecondPetIndex:int = 0
        var ThirdPetIndex:int = 0
        var ForthPetIndex:int = 0

        var IndexBase:int = 12*10
        #Select right pets for each player
        if(GetTeamIndexAgent[Agent]=1):
            set FirstPetIndex =  IndexBase
            set SecondPetIndex = IndexBase+1
            set ThirdPetIndex =  IndexBase+2
            set ForthPetIndex =  IndexBase+3
        if(GetTeamIndexAgent[Agent]=2):
            set FirstPetIndex =  IndexBase+4
            set SecondPetIndex = IndexBase+5
            set ThirdPetIndex =  IndexBase+6
            set ForthPetIndex =  IndexBase+7
        if(GetTeamIndexAgent[Agent]=3):
            set FirstPetIndex =  IndexBase+8
            set SecondPetIndex = IndexBase+9
            set ThirdPetIndex =  IndexBase+10
            set ForthPetIndex =  IndexBase+11
        
        if (Mypet:=Pets[FirstPetIndex]):
            set FirstPet = Mypet.PetProp
        if (Mypet:=Pets[SecondPetIndex]):
            set SecondPet = Mypet.PetProp
        if (Mypet:=Pets[ThirdPetIndex]):
            set ThirdPet = Mypet.PetProp
        if (Mypet:=Pets[ForthPetIndex]):
            set ForthPet = Mypet.PetProp

        if:
            CurrentPlayerStats := PlayerStatsManager.GetPlayerStats[Agent]
            PlayerObj := player[Agent]
            CustomPlayer:Custom_Player = PlayersMap[PlayerObj]
        then:
            if(CurrentPlayerStats.Ownership_L11_Pet4.CurrentValue=1 and CustomPlayer.FollowActiveL11=0):
                if (Fort := Agent.GetFortCharacter[]):
                    spawn:
                        FollowPlayerL11(Fort, ForthPet, Power_L11P4*1.0, ForthPetIndex)
            if(CurrentPlayerStats.Ownership_L11_Pet3.CurrentValue=1 and CustomPlayer.FollowActiveL11=0):
                if (Fort := Agent.GetFortCharacter[]):
                    spawn:
                        FollowPlayerL11(Fort, ThirdPet, Power_L11P3*1.0, ThirdPetIndex)
            if(CurrentPlayerStats.Ownership_L11_Pet2.CurrentValue=1 and CustomPlayer.FollowActiveL11=0):
                if (Fort := Agent.GetFortCharacter[]):
                    spawn:
                        FollowPlayerL11(Fort, SecondPet, Power_L11P2*1.0, SecondPetIndex)
            if(CurrentPlayerStats.Ownership_L11_Pet1.CurrentValue=1 and CustomPlayer.FollowActiveL11=0):
                if (Fort := Agent.GetFortCharacter[]):
                    spawn:
                        FollowPlayerL11(Fort, FirstPet, Power_L11P1*1.0, FirstPetIndex)

    InitializeL12Pets(Agent:agent)<suspends>:void=
        Sleep(1.2)
        var FirstPet:creative_prop = creative_prop{}
        var SecondPet:creative_prop = creative_prop{}
        var ThirdPet:creative_prop = creative_prop{}
        var ForthPet:creative_prop = creative_prop{}

        var FirstPetIndex:int = 0
        var SecondPetIndex:int = 0
        var ThirdPetIndex:int = 0
        var ForthPetIndex:int = 0

        var IndexBase:int = 12*11
        #Select right pets for each player
        if(GetTeamIndexAgent[Agent]=1):
            set FirstPetIndex =  IndexBase
            set SecondPetIndex = IndexBase+1
            set ThirdPetIndex =  IndexBase+2
            set ForthPetIndex =  IndexBase+3
        if(GetTeamIndexAgent[Agent]=2):
            set FirstPetIndex =  IndexBase+4
            set SecondPetIndex = IndexBase+5
            set ThirdPetIndex =  IndexBase+6
            set ForthPetIndex =  IndexBase+7
        if(GetTeamIndexAgent[Agent]=3):
            set FirstPetIndex =  IndexBase+8
            set SecondPetIndex = IndexBase+9
            set ThirdPetIndex =  IndexBase+10
            set ForthPetIndex =  IndexBase+11
        
        if (Mypet:=Pets[FirstPetIndex]):
            set FirstPet = Mypet.PetProp
        if (Mypet:=Pets[SecondPetIndex]):
            set SecondPet = Mypet.PetProp
        if (Mypet:=Pets[ThirdPetIndex]):
            set ThirdPet = Mypet.PetProp
        if (Mypet:=Pets[ForthPetIndex]):
            set ForthPet = Mypet.PetProp

        if:
            CurrentPlayerStats := PlayerStatsManager.GetPlayerStats[Agent]
            PlayerObj := player[Agent]
            CustomPlayer:Custom_Player = PlayersMap[PlayerObj]
        then:
            if(CurrentPlayerStats.Ownership_L12_Pet4.CurrentValue=1 and CustomPlayer.FollowActiveL12=0):
                if (Fort := Agent.GetFortCharacter[]):
                    spawn:
                        FollowPlayerL12(Fort, ForthPet, Power_L12P4*1.0, ForthPetIndex)
            if(CurrentPlayerStats.Ownership_L12_Pet3.CurrentValue=1 and CustomPlayer.FollowActiveL12=0):
                if (Fort := Agent.GetFortCharacter[]):
                    spawn:
                        FollowPlayerL12(Fort, ThirdPet, Power_L12P3*1.0, ThirdPetIndex)
            if(CurrentPlayerStats.Ownership_L12_Pet2.CurrentValue=1 and CustomPlayer.FollowActiveL12=0):
                if (Fort := Agent.GetFortCharacter[]):
                    spawn:
                        FollowPlayerL12(Fort, SecondPet, Power_L12P2*1.0, SecondPetIndex)
            if(CurrentPlayerStats.Ownership_L12_Pet1.CurrentValue=1 and CustomPlayer.FollowActiveL12=0):
                if (Fort := Agent.GetFortCharacter[]):
                    spawn:
                        FollowPlayerL12(Fort, FirstPet, Power_L12P1*1.0, FirstPetIndex)

    InitializeL13Pets(Agent:agent)<suspends>:void=
        Sleep(1.3)
        var FirstPet:creative_prop = creative_prop{}
        var SecondPet:creative_prop = creative_prop{}
        var ThirdPet:creative_prop = creative_prop{}
        var ForthPet:creative_prop = creative_prop{}

        var FirstPetIndex:int = 0
        var SecondPetIndex:int = 0
        var ThirdPetIndex:int = 0
        var ForthPetIndex:int = 0

        var IndexBase:int = 12*12
        #Select right pets for each player
        if(GetTeamIndexAgent[Agent]=1):
            set FirstPetIndex =  IndexBase
            set SecondPetIndex = IndexBase+1
            set ThirdPetIndex =  IndexBase+2
            set ForthPetIndex =  IndexBase+3
        if(GetTeamIndexAgent[Agent]=2):
            set FirstPetIndex =  IndexBase+4
            set SecondPetIndex = IndexBase+5
            set ThirdPetIndex =  IndexBase+6
            set ForthPetIndex =  IndexBase+7
        if(GetTeamIndexAgent[Agent]=3):
            set FirstPetIndex =  IndexBase+8
            set SecondPetIndex = IndexBase+9
            set ThirdPetIndex =  IndexBase+10
            set ForthPetIndex =  IndexBase+11
        

        if (Mypet:=Pets[FirstPetIndex]):
            set FirstPet = Mypet.PetProp
        if (Mypet:=Pets[SecondPetIndex]):
            set SecondPet = Mypet.PetProp
        if (Mypet:=Pets[ThirdPetIndex]):
            set ThirdPet = Mypet.PetProp
        if (Mypet:=Pets[ForthPetIndex]):
            set ForthPet = Mypet.PetProp

        if:
            CurrentPlayerStats := PlayerStatsManager.GetPlayerStats[Agent]
            PlayerObj := player[Agent]
            CustomPlayer:Custom_Player = PlayersMap[PlayerObj]
        then:
            if(CurrentPlayerStats.Ownership_L13_Pet4.CurrentValue=1 and CustomPlayer.FollowActiveL13=0):
                if (Fort := Agent.GetFortCharacter[]):
                    spawn:
                        FollowPlayerL13(Fort, ForthPet, Power_L13P4*1.0, ForthPetIndex)
            if(CurrentPlayerStats.Ownership_L13_Pet3.CurrentValue=1 and CustomPlayer.FollowActiveL13=0):
                if (Fort := Agent.GetFortCharacter[]):
                    spawn:
                        FollowPlayerL13(Fort, ThirdPet, Power_L13P3*1.0, ThirdPetIndex)
            if(CurrentPlayerStats.Ownership_L13_Pet2.CurrentValue=1 and CustomPlayer.FollowActiveL13=0):
                if (Fort := Agent.GetFortCharacter[]):
                    spawn:
                        FollowPlayerL13(Fort, SecondPet, Power_L13P2*1.0, SecondPetIndex)
            if(CurrentPlayerStats.Ownership_L13_Pet1.CurrentValue=1 and CustomPlayer.FollowActiveL13=0):
                if (Fort := Agent.GetFortCharacter[]):
                    spawn:
                        FollowPlayerL13(Fort, FirstPet, Power_L13P1*1.0, FirstPetIndex)

    InitializeL14Pets(Agent:agent)<suspends>:void=
        Sleep(1.4)
        var FirstPet:creative_prop = creative_prop{}
        var SecondPet:creative_prop = creative_prop{}
        var ThirdPet:creative_prop = creative_prop{}
        var ForthPet:creative_prop = creative_prop{}

        var FirstPetIndex:int = 0
        var SecondPetIndex:int = 0
        var ThirdPetIndex:int = 0
        var ForthPetIndex:int = 0

        var IndexBase:int = 12*13
        #Select right pets for each player
        if(GetTeamIndexAgent[Agent]=1):
            set FirstPetIndex =  IndexBase
            set SecondPetIndex = IndexBase+1
            set ThirdPetIndex =  IndexBase+2
            set ForthPetIndex =  IndexBase+3
        if(GetTeamIndexAgent[Agent]=2):
            set FirstPetIndex =  IndexBase+4
            set SecondPetIndex = IndexBase+5
            set ThirdPetIndex =  IndexBase+6
            set ForthPetIndex =  IndexBase+7
        if(GetTeamIndexAgent[Agent]=3):
            set FirstPetIndex =  IndexBase+8
            set SecondPetIndex = IndexBase+9
            set ThirdPetIndex =  IndexBase+10
            set ForthPetIndex =  IndexBase+11
        

        if (Mypet:=Pets[FirstPetIndex]):
            set FirstPet = Mypet.PetProp
        if (Mypet:=Pets[SecondPetIndex]):
            set SecondPet = Mypet.PetProp
        if (Mypet:=Pets[ThirdPetIndex]):
            set ThirdPet = Mypet.PetProp
        if (Mypet:=Pets[ForthPetIndex]):
            set ForthPet = Mypet.PetProp

        if:
            CurrentPlayerStats := PlayerStatsManager.GetPlayerStats[Agent]
            PlayerObj := player[Agent]
            CustomPlayer:Custom_Player = PlayersMap[PlayerObj]
        then:
            if(CurrentPlayerStats.Ownership_L14_Pet4.CurrentValue=1 and CustomPlayer.FollowActiveL14=0):
                if (Fort := Agent.GetFortCharacter[]):
                    spawn:
                        FollowPlayerL14(Fort, ForthPet, Power_L14P4*1.0, ForthPetIndex)
            if(CurrentPlayerStats.Ownership_L14_Pet3.CurrentValue=1 and CustomPlayer.FollowActiveL14=0):
                if (Fort := Agent.GetFortCharacter[]):
                    spawn:
                        FollowPlayerL14(Fort, ThirdPet, Power_L14P3*1.0, ThirdPetIndex)
            if(CurrentPlayerStats.Ownership_L14_Pet2.CurrentValue=1 and CustomPlayer.FollowActiveL14=0):
                if (Fort := Agent.GetFortCharacter[]):
                    spawn:
                        FollowPlayerL14(Fort, SecondPet, Power_L14P2*1.0, SecondPetIndex)
            if(CurrentPlayerStats.Ownership_L14_Pet1.CurrentValue=1 and CustomPlayer.FollowActiveL14=0):
                if (Fort := Agent.GetFortCharacter[]):
                    spawn:
                        FollowPlayerL14(Fort, FirstPet, Power_L14P1*1.0, FirstPetIndex)

    InitializeL15Pets(Agent:agent)<suspends>:void=
        Sleep(1.5)
        var FirstPet:creative_prop = creative_prop{}
        var SecondPet:creative_prop = creative_prop{}
        var ThirdPet:creative_prop = creative_prop{}
        var ForthPet:creative_prop = creative_prop{}

        var FirstPetIndex:int = 0
        var SecondPetIndex:int = 0
        var ThirdPetIndex:int = 0
        var ForthPetIndex:int = 0

        var IndexBase:int = 12*14
        #Select right pets for each player
        if(GetTeamIndexAgent[Agent]=1):
            set FirstPetIndex =  IndexBase
            set SecondPetIndex = IndexBase+1
            set ThirdPetIndex =  IndexBase+2
            set ForthPetIndex =  IndexBase+3
        if(GetTeamIndexAgent[Agent]=2):
            set FirstPetIndex =  IndexBase+4
            set SecondPetIndex = IndexBase+5
            set ThirdPetIndex =  IndexBase+6
            set ForthPetIndex =  IndexBase+7
        if(GetTeamIndexAgent[Agent]=3):
            set FirstPetIndex =  IndexBase+8
            set SecondPetIndex = IndexBase+9
            set ThirdPetIndex =  IndexBase+10
            set ForthPetIndex =  IndexBase+11
        
        if (Mypet:=Pets[FirstPetIndex]):
            set FirstPet = Mypet.PetProp
        if (Mypet:=Pets[SecondPetIndex]):
            set SecondPet = Mypet.PetProp
        if (Mypet:=Pets[ThirdPetIndex]):
            set ThirdPet = Mypet.PetProp
        if (Mypet:=Pets[ForthPetIndex]):
            set ForthPet = Mypet.PetProp

        if:
            CurrentPlayerStats := PlayerStatsManager.GetPlayerStats[Agent]
            PlayerObj := player[Agent]
            CustomPlayer:Custom_Player = PlayersMap[PlayerObj]
        then:
            if(CurrentPlayerStats.Ownership_L15_Pet4.CurrentValue=1 and CustomPlayer.FollowActiveL15=0):
                if (Fort := Agent.GetFortCharacter[]):
                    spawn:
                        FollowPlayerL15(Fort, ForthPet, Power_L15P4*1.0, ForthPetIndex)
            if(CurrentPlayerStats.Ownership_L15_Pet3.CurrentValue=1 and CustomPlayer.FollowActiveL15=0):
                if (Fort := Agent.GetFortCharacter[]):
                    spawn:
                        FollowPlayerL15(Fort, ThirdPet, Power_L15P3*1.0, ThirdPetIndex)
            if(CurrentPlayerStats.Ownership_L15_Pet2.CurrentValue=1 and CustomPlayer.FollowActiveL15=0):
                if (Fort := Agent.GetFortCharacter[]):
                    spawn:
                        FollowPlayerL15(Fort, SecondPet, Power_L15P2*1.0, SecondPetIndex)
            if(CurrentPlayerStats.Ownership_L15_Pet1.CurrentValue=1 and CustomPlayer.FollowActiveL15=0):
                if (Fort := Agent.GetFortCharacter[]):
                    spawn:
                        FollowPlayerL15(Fort, FirstPet, Power_L15P1*1.0, FirstPetIndex)

    InitializeL16Pets(Agent:agent)<suspends>:void=
        Sleep(1.6)
        var FirstPet:creative_prop = creative_prop{}
        var SecondPet:creative_prop = creative_prop{}
        var ThirdPet:creative_prop = creative_prop{}
        var ForthPet:creative_prop = creative_prop{}

        var FirstPetIndex:int = 0
        var SecondPetIndex:int = 0
        var ThirdPetIndex:int = 0
        var ForthPetIndex:int = 0

        var IndexBase:int = 12*15
        #Select right pets for each player
        if(GetTeamIndexAgent[Agent]=1):
            set FirstPetIndex =  IndexBase
            set SecondPetIndex = IndexBase+1
            set ThirdPetIndex =  IndexBase+2
            set ForthPetIndex =  IndexBase+3
        if(GetTeamIndexAgent[Agent]=2):
            set FirstPetIndex =  IndexBase+4
            set SecondPetIndex = IndexBase+5
            set ThirdPetIndex =  IndexBase+6
            set ForthPetIndex =  IndexBase+7
        if(GetTeamIndexAgent[Agent]=3):
            set FirstPetIndex =  IndexBase+8
            set SecondPetIndex = IndexBase+9
            set ThirdPetIndex =  IndexBase+10
            set ForthPetIndex =  IndexBase+11
        
        if (Mypet:=Pets[FirstPetIndex]):
            set FirstPet = Mypet.PetProp
        if (Mypet:=Pets[SecondPetIndex]):
            set SecondPet = Mypet.PetProp
        if (Mypet:=Pets[ThirdPetIndex]):
            set ThirdPet = Mypet.PetProp
        if (Mypet:=Pets[ForthPetIndex]):
            set ForthPet = Mypet.PetProp

        if:
            CurrentPlayerStats := PlayerStatsManager.GetPlayerStats[Agent]
            PlayerObj := player[Agent]
            CustomPlayer:Custom_Player = PlayersMap[PlayerObj]
        then:
            if(CurrentPlayerStats.Ownership_L16_Pet4.CurrentValue=1 and CustomPlayer.FollowActiveL16=0):
                if (Fort := Agent.GetFortCharacter[]):
                    spawn:
                        FollowPlayerL16(Fort, ForthPet, Power_L16P4*1.0, ForthPetIndex)
            if(CurrentPlayerStats.Ownership_L16_Pet3.CurrentValue=1 and CustomPlayer.FollowActiveL16=0):
                if (Fort := Agent.GetFortCharacter[]):
                    spawn:
                        FollowPlayerL16(Fort, ThirdPet, Power_L16P3*1.0, ThirdPetIndex)
            if(CurrentPlayerStats.Ownership_L16_Pet2.CurrentValue=1 and CustomPlayer.FollowActiveL16=0):
                if (Fort := Agent.GetFortCharacter[]):
                    spawn:
                        FollowPlayerL16(Fort, SecondPet, Power_L16P2*1.0, SecondPetIndex)
            if(CurrentPlayerStats.Ownership_L16_Pet1.CurrentValue=1 and CustomPlayer.FollowActiveL16=0):
                if (Fort := Agent.GetFortCharacter[]):
                    spawn:
                        FollowPlayerL16(Fort, FirstPet, Power_L16P1*1.0, FirstPetIndex)

    InitializeL17Pets(Agent:agent)<suspends>:void=
        Sleep(1.7)
        var FirstPet:creative_prop = creative_prop{}
        var SecondPet:creative_prop = creative_prop{}
        var ThirdPet:creative_prop = creative_prop{}
        var ForthPet:creative_prop = creative_prop{}

        var FirstPetIndex:int = 0
        var SecondPetIndex:int = 0
        var ThirdPetIndex:int = 0
        var ForthPetIndex:int = 0

        var IndexBase:int = 12*16
        #Select right pets for each player
        if(GetTeamIndexAgent[Agent]=1):
            set FirstPetIndex =  IndexBase
            set SecondPetIndex = IndexBase+1
            set ThirdPetIndex =  IndexBase+2
            set ForthPetIndex =  IndexBase+3
        if(GetTeamIndexAgent[Agent]=2):
            set FirstPetIndex =  IndexBase+4
            set SecondPetIndex = IndexBase+5
            set ThirdPetIndex =  IndexBase+6
            set ForthPetIndex =  IndexBase+7
        if(GetTeamIndexAgent[Agent]=3):
            set FirstPetIndex =  IndexBase+8
            set SecondPetIndex = IndexBase+9
            set ThirdPetIndex =  IndexBase+10
            set ForthPetIndex =  IndexBase+11
        
        if (Mypet:=Pets[FirstPetIndex]):
            set FirstPet = Mypet.PetProp
        if (Mypet:=Pets[SecondPetIndex]):
            set SecondPet = Mypet.PetProp
        if (Mypet:=Pets[ThirdPetIndex]):
            set ThirdPet = Mypet.PetProp
        if (Mypet:=Pets[ForthPetIndex]):
            set ForthPet = Mypet.PetProp

        if:
            CurrentPlayerStats := PlayerStatsManager.GetPlayerStats[Agent]
            PlayerObj := player[Agent]
            CustomPlayer:Custom_Player = PlayersMap[PlayerObj]
        then:
            if(CurrentPlayerStats.Ownership_L17_Pet4.CurrentValue=1 and CustomPlayer.FollowActiveL17=0):
                if (Fort := Agent.GetFortCharacter[]):
                    spawn:
                        FollowPlayerL17(Fort, ForthPet, Power_L17P4*1.0, ForthPetIndex)
            if(CurrentPlayerStats.Ownership_L17_Pet3.CurrentValue=1 and CustomPlayer.FollowActiveL17=0):
                if (Fort := Agent.GetFortCharacter[]):
                    spawn:
                        FollowPlayerL17(Fort, ThirdPet, Power_L17P3*1.0, ThirdPetIndex)
            if(CurrentPlayerStats.Ownership_L17_Pet2.CurrentValue=1 and CustomPlayer.FollowActiveL17=0):
                if (Fort := Agent.GetFortCharacter[]):
                    spawn:
                        FollowPlayerL17(Fort, SecondPet, Power_L17P2*1.0, SecondPetIndex)
            if(CurrentPlayerStats.Ownership_L17_Pet1.CurrentValue=1 and CustomPlayer.FollowActiveL17=0):
                if (Fort := Agent.GetFortCharacter[]):
                    spawn:
                        FollowPlayerL17(Fort, FirstPet, Power_L17P1*1.0, FirstPetIndex)

    InitializeL18Pets(Agent:agent)<suspends>:void=
        Sleep(1.8)
        var FirstPet:creative_prop = creative_prop{}
        var SecondPet:creative_prop = creative_prop{}
        var ThirdPet:creative_prop = creative_prop{}
        var ForthPet:creative_prop = creative_prop{}

        var FirstPetIndex:int = 0
        var SecondPetIndex:int = 0
        var ThirdPetIndex:int = 0
        var ForthPetIndex:int = 0

        var IndexBase:int = 12*17
        #Select right pets for each player
        if(GetTeamIndexAgent[Agent]=1):
            set FirstPetIndex =  IndexBase
            set SecondPetIndex = IndexBase+1
            set ThirdPetIndex =  IndexBase+2
            set ForthPetIndex =  IndexBase+3
        if(GetTeamIndexAgent[Agent]=2):
            set FirstPetIndex =  IndexBase+4
            set SecondPetIndex = IndexBase+5
            set ThirdPetIndex =  IndexBase+6
            set ForthPetIndex =  IndexBase+7
        if(GetTeamIndexAgent[Agent]=3):
            set FirstPetIndex =  IndexBase+8
            set SecondPetIndex = IndexBase+9
            set ThirdPetIndex =  IndexBase+10
            set ForthPetIndex =  IndexBase+11
        
        if (Mypet:=Pets[FirstPetIndex]):
            set FirstPet = Mypet.PetProp
        if (Mypet:=Pets[SecondPetIndex]):
            set SecondPet = Mypet.PetProp
        if (Mypet:=Pets[ThirdPetIndex]):
            set ThirdPet = Mypet.PetProp
        if (Mypet:=Pets[ForthPetIndex]):
            set ForthPet = Mypet.PetProp

        if:
            CurrentPlayerStats := PlayerStatsManager.GetPlayerStats[Agent]
            PlayerObj := player[Agent]
            CustomPlayer:Custom_Player = PlayersMap[PlayerObj]
        then:
            if(CurrentPlayerStats.Ownership_L18_Pet4.CurrentValue=1 and CustomPlayer.FollowActiveL18=0):
                if (Fort := Agent.GetFortCharacter[]):
                    spawn:
                        FollowPlayerL18(Fort, ForthPet, Power_L18P4*1.0, ForthPetIndex)
            if(CurrentPlayerStats.Ownership_L18_Pet3.CurrentValue=1 and CustomPlayer.FollowActiveL18=0):
                if (Fort := Agent.GetFortCharacter[]):
                    spawn:
                        FollowPlayerL18(Fort, ThirdPet, Power_L18P3*1.0, ThirdPetIndex)
            if(CurrentPlayerStats.Ownership_L18_Pet2.CurrentValue=1 and CustomPlayer.FollowActiveL18=0):
                if (Fort := Agent.GetFortCharacter[]):
                    spawn:
                        FollowPlayerL18(Fort, SecondPet, Power_L18P2*1.0, SecondPetIndex)
            if(CurrentPlayerStats.Ownership_L18_Pet1.CurrentValue=1 and CustomPlayer.FollowActiveL18=0):
                if (Fort := Agent.GetFortCharacter[]):
                    spawn:
                        FollowPlayerL18(Fort, FirstPet, Power_L18P1*1.0, FirstPetIndex)

    InitializeL19Pets(Agent:agent)<suspends>:void=
        Sleep(1.9)
        var FirstPet:creative_prop = creative_prop{}
        var SecondPet:creative_prop = creative_prop{}
        var ThirdPet:creative_prop = creative_prop{}
        var ForthPet:creative_prop = creative_prop{}

        var FirstPetIndex:int = 0
        var SecondPetIndex:int = 0
        var ThirdPetIndex:int = 0
        var ForthPetIndex:int = 0

        var IndexBase:int = 12*18
        #Select right pets for each player
        if(GetTeamIndexAgent[Agent]=1):
            set FirstPetIndex =  IndexBase
            set SecondPetIndex = IndexBase+1
            set ThirdPetIndex =  IndexBase+2
            set ForthPetIndex =  IndexBase+3
        if(GetTeamIndexAgent[Agent]=2):
            set FirstPetIndex =  IndexBase+4
            set SecondPetIndex = IndexBase+5
            set ThirdPetIndex =  IndexBase+6
            set ForthPetIndex =  IndexBase+7
        if(GetTeamIndexAgent[Agent]=3):
            set FirstPetIndex =  IndexBase+8
            set SecondPetIndex = IndexBase+9
            set ThirdPetIndex =  IndexBase+10
            set ForthPetIndex =  IndexBase+11

        if (Mypet:=Pets[FirstPetIndex]):
            set FirstPet = Mypet.PetProp
        if (Mypet:=Pets[SecondPetIndex]):
            set SecondPet = Mypet.PetProp
        if (Mypet:=Pets[ThirdPetIndex]):
            set ThirdPet = Mypet.PetProp
        if (Mypet:=Pets[ForthPetIndex]):
            set ForthPet = Mypet.PetProp

        if:
            CurrentPlayerStats := PlayerStatsManager.GetPlayerStats[Agent]
            PlayerObj := player[Agent]
            CustomPlayer:Custom_Player = PlayersMap[PlayerObj]
        then:
            if(CurrentPlayerStats.Ownership_L19_Pet4.CurrentValue=1 and CustomPlayer.FollowActiveL19=0):
                if (Fort := Agent.GetFortCharacter[]):
                    spawn:
                        FollowPlayerL19(Fort, ForthPet, Power_L19P4*1.0, ForthPetIndex)
            if(CurrentPlayerStats.Ownership_L19_Pet3.CurrentValue=1 and CustomPlayer.FollowActiveL19=0):
                if (Fort := Agent.GetFortCharacter[]):
                    spawn:
                        FollowPlayerL19(Fort, ThirdPet, Power_L19P3*1.0, ThirdPetIndex)
            if(CurrentPlayerStats.Ownership_L19_Pet2.CurrentValue=1 and CustomPlayer.FollowActiveL19=0):
                if (Fort := Agent.GetFortCharacter[]):
                    spawn:
                        FollowPlayerL19(Fort, SecondPet, Power_L19P2*1.0, SecondPetIndex)
            if(CurrentPlayerStats.Ownership_L19_Pet1.CurrentValue=1 and CustomPlayer.FollowActiveL19=0):
                if (Fort := Agent.GetFortCharacter[]):
                    spawn:
                        FollowPlayerL19(Fort, FirstPet, Power_L19P1*1.0, FirstPetIndex)

    InitializeL20Pets(Agent:agent)<suspends>:void=
        Sleep(2.0)
        var FirstPet:creative_prop = creative_prop{}
        var SecondPet:creative_prop = creative_prop{}
        var ThirdPet:creative_prop = creative_prop{}
        var ForthPet:creative_prop = creative_prop{}

        var FirstPetIndex:int = 0
        var SecondPetIndex:int = 0
        var ThirdPetIndex:int = 0
        var ForthPetIndex:int = 0

        var IndexBase:int = 12*19
        #Select right pets for each player
        if(GetTeamIndexAgent[Agent]=1):
            set FirstPetIndex =  IndexBase
            set SecondPetIndex = IndexBase+1
            set ThirdPetIndex =  IndexBase+2
            set ForthPetIndex =  IndexBase+3
        if(GetTeamIndexAgent[Agent]=2):
            set FirstPetIndex =  IndexBase+4
            set SecondPetIndex = IndexBase+5
            set ThirdPetIndex =  IndexBase+6
            set ForthPetIndex =  IndexBase+7
        if(GetTeamIndexAgent[Agent]=3):
            set FirstPetIndex =  IndexBase+8
            set SecondPetIndex = IndexBase+9
            set ThirdPetIndex =  IndexBase+10
            set ForthPetIndex =  IndexBase+11
        
        if (Mypet:=Pets[FirstPetIndex]):
            set FirstPet = Mypet.PetProp
        if (Mypet:=Pets[SecondPetIndex]):
            set SecondPet = Mypet.PetProp
        if (Mypet:=Pets[ThirdPetIndex]):
            set ThirdPet = Mypet.PetProp
        if (Mypet:=Pets[ForthPetIndex]):
            set ForthPet = Mypet.PetProp

        if:
            CurrentPlayerStats := PlayerStatsManager.GetPlayerStats[Agent]
            PlayerObj := player[Agent]
            CustomPlayer:Custom_Player = PlayersMap[PlayerObj]
        then:
            if(CurrentPlayerStats.Ownership_L20_Pet4.CurrentValue=1 and CustomPlayer.FollowActiveL20=0):
                if (Fort := Agent.GetFortCharacter[]):
                    spawn:
                        FollowPlayerL20(Fort, ForthPet, Power_L20P4*1.0, ForthPetIndex)
            if(CurrentPlayerStats.Ownership_L20_Pet3.CurrentValue=1 and CustomPlayer.FollowActiveL20=0):
                if (Fort := Agent.GetFortCharacter[]):
                    spawn:
                        FollowPlayerL20(Fort, ThirdPet, Power_L20P3*1.0, ThirdPetIndex)
            if(CurrentPlayerStats.Ownership_L20_Pet2.CurrentValue=1 and CustomPlayer.FollowActiveL20=0):
                if (Fort := Agent.GetFortCharacter[]):
                    spawn:
                        FollowPlayerL20(Fort, SecondPet, Power_L20P2*1.0, SecondPetIndex)
            if(CurrentPlayerStats.Ownership_L20_Pet1.CurrentValue=1 and CustomPlayer.FollowActiveL20=0):
                if (Fort := Agent.GetFortCharacter[]):
                    spawn:
                        FollowPlayerL20(Fort, FirstPet, Power_L20P1*1.0, FirstPetIndex)

    InitializeL21Pets(Agent:agent)<suspends>:void=
        Sleep(2.1)
        var FirstPet:creative_prop = creative_prop{}
        var SecondPet:creative_prop = creative_prop{}
        var ThirdPet:creative_prop = creative_prop{}
        var ForthPet:creative_prop = creative_prop{}

        var FirstPetIndex:int = 0
        var SecondPetIndex:int = 0
        var ThirdPetIndex:int = 0
        var ForthPetIndex:int = 0

        var IndexBase:int = 12*20
        #Select right pets for each player
        if(GetTeamIndexAgent[Agent]=1):
            set FirstPetIndex =  IndexBase
            set SecondPetIndex = IndexBase+1
            set ThirdPetIndex =  IndexBase+2
            set ForthPetIndex =  IndexBase+3
        if(GetTeamIndexAgent[Agent]=2):
            set FirstPetIndex =  IndexBase+4
            set SecondPetIndex = IndexBase+5
            set ThirdPetIndex =  IndexBase+6
            set ForthPetIndex =  IndexBase+7
        if(GetTeamIndexAgent[Agent]=3):
            set FirstPetIndex =  IndexBase+8
            set SecondPetIndex = IndexBase+9
            set ThirdPetIndex =  IndexBase+10
            set ForthPetIndex =  IndexBase+11
        
        if (Mypet:=Pets[FirstPetIndex]):
            set FirstPet = Mypet.PetProp
        if (Mypet:=Pets[SecondPetIndex]):
            set SecondPet = Mypet.PetProp
        if (Mypet:=Pets[ThirdPetIndex]):
            set ThirdPet = Mypet.PetProp
        if (Mypet:=Pets[ForthPetIndex]):
            set ForthPet = Mypet.PetProp

        if:
            CurrentPlayerStats := PlayerStatsManager.GetPlayerStats[Agent]
            PlayerObj := player[Agent]
            CustomPlayer:Custom_Player = PlayersMap[PlayerObj]
        then:
            if(CurrentPlayerStats.Ownership_L21_Pet4.CurrentValue=1 and CustomPlayer.FollowActiveL21=0):
                if (Fort := Agent.GetFortCharacter[]):
                    spawn:
                        FollowPlayerL21(Fort, ForthPet, Power_L21P4*1.0, ForthPetIndex)
            if(CurrentPlayerStats.Ownership_L21_Pet3.CurrentValue=1 and CustomPlayer.FollowActiveL21=0):
                if (Fort := Agent.GetFortCharacter[]):
                    spawn:
                        FollowPlayerL21(Fort, ThirdPet, Power_L21P3*1.0, ThirdPetIndex)
            if(CurrentPlayerStats.Ownership_L21_Pet2.CurrentValue=1 and CustomPlayer.FollowActiveL21=0):
                if (Fort := Agent.GetFortCharacter[]):
                    spawn:
                        FollowPlayerL21(Fort, SecondPet, Power_L21P2*1.0, SecondPetIndex)
            if(CurrentPlayerStats.Ownership_L21_Pet1.CurrentValue=1 and CustomPlayer.FollowActiveL21=0):
                if (Fort := Agent.GetFortCharacter[]):
                    spawn:
                        FollowPlayerL21(Fort, FirstPet, Power_L21P1*1.0, FirstPetIndex)

    InitializeL22Pets(Agent:agent)<suspends>:void=
        Sleep(2.2)
        var FirstPet:creative_prop = creative_prop{}
        var SecondPet:creative_prop = creative_prop{}
        var ThirdPet:creative_prop = creative_prop{}
        var ForthPet:creative_prop = creative_prop{}

        var FirstPetIndex:int = 0
        var SecondPetIndex:int = 0
        var ThirdPetIndex:int = 0
        var ForthPetIndex:int = 0

        var IndexBase:int = 12*21
        #Select right pets for each player
        if(GetTeamIndexAgent[Agent]=1):
            set FirstPetIndex =  IndexBase
            set SecondPetIndex = IndexBase+1
            set ThirdPetIndex =  IndexBase+2
            set ForthPetIndex =  IndexBase+3
        if(GetTeamIndexAgent[Agent]=2):
            set FirstPetIndex =  IndexBase+4
            set SecondPetIndex = IndexBase+5
            set ThirdPetIndex =  IndexBase+6
            set ForthPetIndex =  IndexBase+7
        if(GetTeamIndexAgent[Agent]=3):
            set FirstPetIndex =  IndexBase+8
            set SecondPetIndex = IndexBase+9
            set ThirdPetIndex =  IndexBase+10
            set ForthPetIndex =  IndexBase+11
        
        if (Mypet:=Pets[FirstPetIndex]):
            set FirstPet = Mypet.PetProp
        if (Mypet:=Pets[SecondPetIndex]):
            set SecondPet = Mypet.PetProp
        if (Mypet:=Pets[ThirdPetIndex]):
            set ThirdPet = Mypet.PetProp
        if (Mypet:=Pets[ForthPetIndex]):
            set ForthPet = Mypet.PetProp

        if:
            CurrentPlayerStats := PlayerStatsManager.GetPlayerStats[Agent]
            PlayerObj := player[Agent]
            CustomPlayer:Custom_Player = PlayersMap[PlayerObj]
        then:
            if(CurrentPlayerStats.Ownership_L22_Pet4.CurrentValue=1 and CustomPlayer.FollowActiveL22=0):
                if (Fort := Agent.GetFortCharacter[]):
                    spawn:
                        FollowPlayerL22(Fort, ForthPet, Power_L22P4*1.0, ForthPetIndex)
            if(CurrentPlayerStats.Ownership_L22_Pet3.CurrentValue=1 and CustomPlayer.FollowActiveL22=0):
                if (Fort := Agent.GetFortCharacter[]):
                    spawn:
                        FollowPlayerL22(Fort, ThirdPet, Power_L22P3*1.0, ThirdPetIndex)
            if(CurrentPlayerStats.Ownership_L22_Pet2.CurrentValue=1 and CustomPlayer.FollowActiveL22=0):
                if (Fort := Agent.GetFortCharacter[]):
                    spawn:
                        FollowPlayerL22(Fort, SecondPet, Power_L22P2*1.0, SecondPetIndex)
            if(CurrentPlayerStats.Ownership_L22_Pet1.CurrentValue=1 and CustomPlayer.FollowActiveL22=0):
                if (Fort := Agent.GetFortCharacter[]):
                    spawn:
                        FollowPlayerL22(Fort, FirstPet, Power_L22P1*1.0, FirstPetIndex)

    InitializeL23Pets(Agent:agent)<suspends>:void=
        Sleep(2.3)
        var FirstPet:creative_prop = creative_prop{}
        var SecondPet:creative_prop = creative_prop{}
        var ThirdPet:creative_prop = creative_prop{}
        var ForthPet:creative_prop = creative_prop{}

        var FirstPetIndex:int = 0
        var SecondPetIndex:int = 0
        var ThirdPetIndex:int = 0
        var ForthPetIndex:int = 0

        var IndexBase:int = 12*22
        #Select right pets for each player
        if(GetTeamIndexAgent[Agent]=1):
            set FirstPetIndex =  IndexBase
            set SecondPetIndex = IndexBase+1
            set ThirdPetIndex =  IndexBase+2
            set ForthPetIndex =  IndexBase+3
        if(GetTeamIndexAgent[Agent]=2):
            set FirstPetIndex =  IndexBase+4
            set SecondPetIndex = IndexBase+5
            set ThirdPetIndex =  IndexBase+6
            set ForthPetIndex =  IndexBase+7
        if(GetTeamIndexAgent[Agent]=3):
            set FirstPetIndex =  IndexBase+8
            set SecondPetIndex = IndexBase+9
            set ThirdPetIndex =  IndexBase+10
            set ForthPetIndex =  IndexBase+11
        
        if (Mypet:=Pets[FirstPetIndex]):
            set FirstPet = Mypet.PetProp
        if (Mypet:=Pets[SecondPetIndex]):
            set SecondPet = Mypet.PetProp
        if (Mypet:=Pets[ThirdPetIndex]):
            set ThirdPet = Mypet.PetProp
        if (Mypet:=Pets[ForthPetIndex]):
            set ForthPet = Mypet.PetProp

        if:
            CurrentPlayerStats := PlayerStatsManager.GetPlayerStats[Agent]
            PlayerObj := player[Agent]
            CustomPlayer:Custom_Player = PlayersMap[PlayerObj]
        then:
            if(CurrentPlayerStats.Ownership_L23_Pet4.CurrentValue=1 and CustomPlayer.FollowActiveL23=0):
                if (Fort := Agent.GetFortCharacter[]):
                    spawn:
                        FollowPlayerL23(Fort, ForthPet, Power_L23P4*1.0, ForthPetIndex)
            if(CurrentPlayerStats.Ownership_L23_Pet3.CurrentValue=1 and CustomPlayer.FollowActiveL23=0):
                if (Fort := Agent.GetFortCharacter[]):
                    spawn:
                        FollowPlayerL23(Fort, ThirdPet, Power_L23P3*1.0, ThirdPetIndex)
            if(CurrentPlayerStats.Ownership_L23_Pet2.CurrentValue=1 and CustomPlayer.FollowActiveL23=0):
                if (Fort := Agent.GetFortCharacter[]):
                    spawn:
                        FollowPlayerL23(Fort, SecondPet, Power_L23P2*1.0, SecondPetIndex)
            if(CurrentPlayerStats.Ownership_L23_Pet1.CurrentValue=1 and CustomPlayer.FollowActiveL23=0):
                if (Fort := Agent.GetFortCharacter[]):
                    spawn:
                        FollowPlayerL23(Fort, FirstPet, Power_L23P1*1.0, FirstPetIndex)

    InitializeL24Pets(Agent:agent)<suspends>:void=
        Sleep(2.4)
        var FirstPet:creative_prop = creative_prop{}
        var SecondPet:creative_prop = creative_prop{}
        var ThirdPet:creative_prop = creative_prop{}
        var ForthPet:creative_prop = creative_prop{}

        var FirstPetIndex:int = 0
        var SecondPetIndex:int = 0
        var ThirdPetIndex:int = 0
        var ForthPetIndex:int = 0

        var IndexBase:int = 12*23
        #Select right pets for each player
        if(GetTeamIndexAgent[Agent]=1):
            set FirstPetIndex =  IndexBase
            set SecondPetIndex = IndexBase+1
            set ThirdPetIndex =  IndexBase+2
            set ForthPetIndex =  IndexBase+3
        if(GetTeamIndexAgent[Agent]=2):
            set FirstPetIndex =  IndexBase+4
            set SecondPetIndex = IndexBase+5
            set ThirdPetIndex =  IndexBase+6
            set ForthPetIndex =  IndexBase+7
        if(GetTeamIndexAgent[Agent]=3):
            set FirstPetIndex =  IndexBase+8
            set SecondPetIndex = IndexBase+9
            set ThirdPetIndex =  IndexBase+10
            set ForthPetIndex =  IndexBase+11
        
        if (Mypet:=Pets[FirstPetIndex]):
            set FirstPet = Mypet.PetProp
        if (Mypet:=Pets[SecondPetIndex]):
            set SecondPet = Mypet.PetProp
        if (Mypet:=Pets[ThirdPetIndex]):
            set ThirdPet = Mypet.PetProp
        if (Mypet:=Pets[ForthPetIndex]):
            set ForthPet = Mypet.PetProp

        if:
            CurrentPlayerStats := PlayerStatsManager.GetPlayerStats[Agent]
            PlayerObj := player[Agent]
            CustomPlayer:Custom_Player = PlayersMap[PlayerObj]
        then:
            if(CurrentPlayerStats.Ownership_L24_Pet4.CurrentValue=1 and CustomPlayer.FollowActiveL24=0):
                if (Fort := Agent.GetFortCharacter[]):
                    spawn:
                        FollowPlayerL24(Fort, ForthPet, Power_L24P4*1.0, ForthPetIndex)
            if(CurrentPlayerStats.Ownership_L24_Pet3.CurrentValue=1 and CustomPlayer.FollowActiveL24=0):
                if (Fort := Agent.GetFortCharacter[]):
                    spawn:
                        FollowPlayerL24(Fort, ThirdPet, Power_L24P3*1.0, ThirdPetIndex)
            if(CurrentPlayerStats.Ownership_L24_Pet2.CurrentValue=1 and CustomPlayer.FollowActiveL24=0):
                if (Fort := Agent.GetFortCharacter[]):
                    spawn:
                        FollowPlayerL24(Fort, SecondPet, Power_L24P2*1.0, SecondPetIndex)
            if(CurrentPlayerStats.Ownership_L24_Pet1.CurrentValue=1 and CustomPlayer.FollowActiveL24=0):
                if (Fort := Agent.GetFortCharacter[]):
                    spawn:
                        FollowPlayerL24(Fort, FirstPet, Power_L24P1*1.0, FirstPetIndex)

    InitializeL25Pets(Agent:agent)<suspends>:void=
        Sleep(2.5)
        var FirstPet:creative_prop = creative_prop{}
        var SecondPet:creative_prop = creative_prop{}
        var ThirdPet:creative_prop = creative_prop{}
        var ForthPet:creative_prop = creative_prop{}

        var FirstPetIndex:int = 0
        var SecondPetIndex:int = 0
        var ThirdPetIndex:int = 0
        var ForthPetIndex:int = 0

        var IndexBase:int = 12*24
        #Select right pets for each player
        if(GetTeamIndexAgent[Agent]=1):
            set FirstPetIndex =  IndexBase
            set SecondPetIndex = IndexBase+1
            set ThirdPetIndex =  IndexBase+2
            set ForthPetIndex =  IndexBase+3
        if(GetTeamIndexAgent[Agent]=2):
            set FirstPetIndex =  IndexBase+4
            set SecondPetIndex = IndexBase+5
            set ThirdPetIndex =  IndexBase+6
            set ForthPetIndex =  IndexBase+7
        if(GetTeamIndexAgent[Agent]=3):
            set FirstPetIndex =  IndexBase+8
            set SecondPetIndex = IndexBase+9
            set ThirdPetIndex =  IndexBase+10
            set ForthPetIndex =  IndexBase+11
        
        if (Mypet:=Pets[FirstPetIndex]):
            set FirstPet = Mypet.PetProp
        if (Mypet:=Pets[SecondPetIndex]):
            set SecondPet = Mypet.PetProp
        if (Mypet:=Pets[ThirdPetIndex]):
            set ThirdPet = Mypet.PetProp
        if (Mypet:=Pets[ForthPetIndex]):
            set ForthPet = Mypet.PetProp

        if:
            CurrentPlayerStats := PlayerStatsManager.GetPlayerStats[Agent]
            PlayerObj := player[Agent]
            CustomPlayer:Custom_Player = PlayersMap[PlayerObj]
        then:
            if(CurrentPlayerStats.Ownership_L25_Pet4.CurrentValue=1 and CustomPlayer.FollowActiveL25=0):
                if (Fort := Agent.GetFortCharacter[]):
                    spawn:
                        FollowPlayerL25(Fort, ForthPet, Power_L25P4*1.0, ForthPetIndex)
            if(CurrentPlayerStats.Ownership_L25_Pet3.CurrentValue=1 and CustomPlayer.FollowActiveL25=0):
                if (Fort := Agent.GetFortCharacter[]):
                    spawn:
                        FollowPlayerL25(Fort, ThirdPet, Power_L25P3*1.0, ThirdPetIndex)
            if(CurrentPlayerStats.Ownership_L25_Pet2.CurrentValue=1 and CustomPlayer.FollowActiveL25=0):
                if (Fort := Agent.GetFortCharacter[]):
                    spawn:
                        FollowPlayerL25(Fort, SecondPet, Power_L25P2*1.0, SecondPetIndex)
            if(CurrentPlayerStats.Ownership_L25_Pet1.CurrentValue=1 and CustomPlayer.FollowActiveL25=0):
                if (Fort := Agent.GetFortCharacter[]):
                    spawn:
                        FollowPlayerL25(Fort, FirstPet, Power_L25P1*1.0, FirstPetIndex)

    InitializeL26Pets(Agent:agent)<suspends>:void=
        Sleep(2.6)
        var FirstPet:creative_prop = creative_prop{}
        var SecondPet:creative_prop = creative_prop{}
        var ThirdPet:creative_prop = creative_prop{}
        var ForthPet:creative_prop = creative_prop{}

        var FirstPetIndex:int = 0
        var SecondPetIndex:int = 0
        var ThirdPetIndex:int = 0
        var ForthPetIndex:int = 0

        var IndexBase:int = 12*25
        #Select right pets for each player
        if(GetTeamIndexAgent[Agent]=1):
            set FirstPetIndex =  IndexBase
            set SecondPetIndex = IndexBase+1
            set ThirdPetIndex =  IndexBase+2
            set ForthPetIndex =  IndexBase+3
        if(GetTeamIndexAgent[Agent]=2):
            set FirstPetIndex =  IndexBase+4
            set SecondPetIndex = IndexBase+5
            set ThirdPetIndex =  IndexBase+6
            set ForthPetIndex =  IndexBase+7
        if(GetTeamIndexAgent[Agent]=3):
            set FirstPetIndex =  IndexBase+8
            set SecondPetIndex = IndexBase+9
            set ThirdPetIndex =  IndexBase+10
            set ForthPetIndex =  IndexBase+11
        
        if (Mypet:=Pets[FirstPetIndex]):
            set FirstPet = Mypet.PetProp
        if (Mypet:=Pets[SecondPetIndex]):
            set SecondPet = Mypet.PetProp
        if (Mypet:=Pets[ThirdPetIndex]):
            set ThirdPet = Mypet.PetProp
        if (Mypet:=Pets[ForthPetIndex]):
            set ForthPet = Mypet.PetProp

        if:
            CurrentPlayerStats := PlayerStatsManager.GetPlayerStats[Agent]
            PlayerObj := player[Agent]
            CustomPlayer:Custom_Player = PlayersMap[PlayerObj]
        then:
            if(CurrentPlayerStats.Ownership_L26_Pet4.CurrentValue=1 and CustomPlayer.FollowActiveL26=0):
                if (Fort := Agent.GetFortCharacter[]):
                    spawn:
                        FollowPlayerL26(Fort, ForthPet, Power_L26P4*1.0, ForthPetIndex)
            if(CurrentPlayerStats.Ownership_L26_Pet3.CurrentValue=1 and CustomPlayer.FollowActiveL26=0):
                if (Fort := Agent.GetFortCharacter[]):
                    spawn:
                        FollowPlayerL26(Fort, ThirdPet, Power_L26P3*1.0, ThirdPetIndex)
            if(CurrentPlayerStats.Ownership_L26_Pet2.CurrentValue=1 and CustomPlayer.FollowActiveL26=0):
                if (Fort := Agent.GetFortCharacter[]):
                    spawn:
                        FollowPlayerL26(Fort, SecondPet, Power_L26P2*1.0, SecondPetIndex)
            if(CurrentPlayerStats.Ownership_L26_Pet1.CurrentValue=1 and CustomPlayer.FollowActiveL26=0):
                if (Fort := Agent.GetFortCharacter[]):
                    spawn:
                        FollowPlayerL26(Fort, FirstPet, Power_L26P1*1.0, FirstPetIndex)

    InitializeL27Pets(Agent:agent)<suspends>:void=
        Sleep(2.7)
        var FirstPet:creative_prop = creative_prop{}
        var SecondPet:creative_prop = creative_prop{}
        var ThirdPet:creative_prop = creative_prop{}
        var ForthPet:creative_prop = creative_prop{}

        var FirstPetIndex:int = 0
        var SecondPetIndex:int = 0
        var ThirdPetIndex:int = 0
        var ForthPetIndex:int = 0

        var IndexBase:int = 12*26
        #Select right pets for each player
        if(GetTeamIndexAgent[Agent]=1):
            set FirstPetIndex =  IndexBase
            set SecondPetIndex = IndexBase+1
            set ThirdPetIndex =  IndexBase+2
            set ForthPetIndex =  IndexBase+3
        if(GetTeamIndexAgent[Agent]=2):
            set FirstPetIndex =  IndexBase+4
            set SecondPetIndex = IndexBase+5
            set ThirdPetIndex =  IndexBase+6
            set ForthPetIndex =  IndexBase+7
        if(GetTeamIndexAgent[Agent]=3):
            set FirstPetIndex =  IndexBase+8
            set SecondPetIndex = IndexBase+9
            set ThirdPetIndex =  IndexBase+10
            set ForthPetIndex =  IndexBase+11
        
        if (Mypet:=Pets[FirstPetIndex]):
            set FirstPet = Mypet.PetProp
        if (Mypet:=Pets[SecondPetIndex]):
            set SecondPet = Mypet.PetProp
        if (Mypet:=Pets[ThirdPetIndex]):
            set ThirdPet = Mypet.PetProp
        if (Mypet:=Pets[ForthPetIndex]):
            set ForthPet = Mypet.PetProp

        if:
            CurrentPlayerStats := PlayerStatsManager.GetPlayerStats[Agent]
            PlayerObj := player[Agent]
            CustomPlayer:Custom_Player = PlayersMap[PlayerObj]
        then:
            if(CurrentPlayerStats.Ownership_L27_Pet4.CurrentValue=1 and CustomPlayer.FollowActiveL27=0):
                if (Fort := Agent.GetFortCharacter[]):
                    spawn:
                        FollowPlayerL27(Fort, ForthPet, Power_L27P4*1.0, ForthPetIndex)
            if(CurrentPlayerStats.Ownership_L27_Pet3.CurrentValue=1 and CustomPlayer.FollowActiveL27=0):
                if (Fort := Agent.GetFortCharacter[]):
                    spawn:
                        FollowPlayerL27(Fort, ThirdPet, Power_L27P3*1.0, ThirdPetIndex)
            if(CurrentPlayerStats.Ownership_L27_Pet2.CurrentValue=1 and CustomPlayer.FollowActiveL27=0):
                if (Fort := Agent.GetFortCharacter[]):
                    spawn:
                        FollowPlayerL27(Fort, SecondPet, Power_L27P2*1.0, SecondPetIndex)
            if(CurrentPlayerStats.Ownership_L27_Pet1.CurrentValue=1 and CustomPlayer.FollowActiveL27=0):
                if (Fort := Agent.GetFortCharacter[]):
                    spawn:
                        FollowPlayerL27(Fort, FirstPet, Power_L27P1*1.0, FirstPetIndex)

    InitializeL28Pets(Agent:agent)<suspends>:void=
        Sleep(2.8)
        var FirstPet:creative_prop = creative_prop{}
        var SecondPet:creative_prop = creative_prop{}
        var ThirdPet:creative_prop = creative_prop{}
        var ForthPet:creative_prop = creative_prop{}

        var FirstPetIndex:int = 0
        var SecondPetIndex:int = 0
        var ThirdPetIndex:int = 0
        var ForthPetIndex:int = 0

        var IndexBase:int = 12*27
        #Select right pets for each player
        if(GetTeamIndexAgent[Agent]=1):
            set FirstPetIndex =  IndexBase
            set SecondPetIndex = IndexBase+1
            set ThirdPetIndex =  IndexBase+2
            set ForthPetIndex =  IndexBase+3
        if(GetTeamIndexAgent[Agent]=2):
            set FirstPetIndex =  IndexBase+4
            set SecondPetIndex = IndexBase+5
            set ThirdPetIndex =  IndexBase+6
            set ForthPetIndex =  IndexBase+7
        if(GetTeamIndexAgent[Agent]=3):
            set FirstPetIndex =  IndexBase+8
            set SecondPetIndex = IndexBase+9
            set ThirdPetIndex =  IndexBase+10
            set ForthPetIndex =  IndexBase+11
        
        if (Mypet:=Pets[FirstPetIndex]):
            set FirstPet = Mypet.PetProp
        if (Mypet:=Pets[SecondPetIndex]):
            set SecondPet = Mypet.PetProp
        if (Mypet:=Pets[ThirdPetIndex]):
            set ThirdPet = Mypet.PetProp
        if (Mypet:=Pets[ForthPetIndex]):
            set ForthPet = Mypet.PetProp

        if:
            CurrentPlayerStats := PlayerStatsManager.GetPlayerStats[Agent]
            PlayerObj := player[Agent]
            CustomPlayer:Custom_Player = PlayersMap[PlayerObj]
        then:
            if(CurrentPlayerStats.Ownership_L28_Pet4.CurrentValue=1 and CustomPlayer.FollowActiveL28=0):
                if (Fort := Agent.GetFortCharacter[]):
                    spawn:
                        FollowPlayerL28(Fort, ForthPet, Power_L28P4*1.0, ForthPetIndex)
            if(CurrentPlayerStats.Ownership_L28_Pet3.CurrentValue=1 and CustomPlayer.FollowActiveL28=0):
                if (Fort := Agent.GetFortCharacter[]):
                    spawn:
                        FollowPlayerL28(Fort, ThirdPet, Power_L28P3*1.0, ThirdPetIndex)
            if(CurrentPlayerStats.Ownership_L28_Pet2.CurrentValue=1 and CustomPlayer.FollowActiveL28=0):
                if (Fort := Agent.GetFortCharacter[]):
                    spawn:
                        FollowPlayerL28(Fort, SecondPet, Power_L28P2*1.0, SecondPetIndex)
            if(CurrentPlayerStats.Ownership_L28_Pet1.CurrentValue=1 and CustomPlayer.FollowActiveL28=0):
                if (Fort := Agent.GetFortCharacter[]):
                    spawn:
                        FollowPlayerL28(Fort, FirstPet, Power_L28P1*1.0, FirstPetIndex)

    InitializeL29Pets(Agent:agent)<suspends>:void=
        Sleep(2.9)
        var FirstPet:creative_prop = creative_prop{}
        var SecondPet:creative_prop = creative_prop{}
        var ThirdPet:creative_prop = creative_prop{}
        var ForthPet:creative_prop = creative_prop{}

        var FirstPetIndex:int = 0
        var SecondPetIndex:int = 0
        var ThirdPetIndex:int = 0
        var ForthPetIndex:int = 0

        var IndexBase:int = 12*28
        #Select right pets for each player
        if(GetTeamIndexAgent[Agent]=1):
            set FirstPetIndex =  IndexBase
            set SecondPetIndex = IndexBase+1
            set ThirdPetIndex =  IndexBase+2
            set ForthPetIndex =  IndexBase+3
        if(GetTeamIndexAgent[Agent]=2):
            set FirstPetIndex =  IndexBase+4
            set SecondPetIndex = IndexBase+5
            set ThirdPetIndex =  IndexBase+6
            set ForthPetIndex =  IndexBase+7
        if(GetTeamIndexAgent[Agent]=3):
            set FirstPetIndex =  IndexBase+8
            set SecondPetIndex = IndexBase+9
            set ThirdPetIndex =  IndexBase+10
            set ForthPetIndex =  IndexBase+11
        
        if (Mypet:=Pets[FirstPetIndex]):
            set FirstPet = Mypet.PetProp
        if (Mypet:=Pets[SecondPetIndex]):
            set SecondPet = Mypet.PetProp
        if (Mypet:=Pets[ThirdPetIndex]):
            set ThirdPet = Mypet.PetProp
        if (Mypet:=Pets[ForthPetIndex]):
            set ForthPet = Mypet.PetProp

        if:
            CurrentPlayerStats := PlayerStatsManager.GetPlayerStats[Agent]
            PlayerObj := player[Agent]
            CustomPlayer:Custom_Player = PlayersMap[PlayerObj]
        then:
            if(CurrentPlayerStats.Ownership_L29_Pet4.CurrentValue=1 and CustomPlayer.FollowActiveL29=0):
                if (Fort := Agent.GetFortCharacter[]):
                    spawn:
                        FollowPlayerL29(Fort, ForthPet, Power_L29P4*1.0, ForthPetIndex)
            if(CurrentPlayerStats.Ownership_L29_Pet3.CurrentValue=1 and CustomPlayer.FollowActiveL29=0):
                if (Fort := Agent.GetFortCharacter[]):
                    spawn:
                        FollowPlayerL29(Fort, ThirdPet, Power_L29P3*1.0, ThirdPetIndex)
            if(CurrentPlayerStats.Ownership_L29_Pet2.CurrentValue=1 and CustomPlayer.FollowActiveL29=0):
                if (Fort := Agent.GetFortCharacter[]):
                    spawn:
                        FollowPlayerL29(Fort, SecondPet, Power_L29P2*1.0, SecondPetIndex)
            if(CurrentPlayerStats.Ownership_L29_Pet1.CurrentValue=1 and CustomPlayer.FollowActiveL29=0):
                if (Fort := Agent.GetFortCharacter[]):
                    spawn:
                        FollowPlayerL29(Fort, FirstPet, Power_L29P1*1.0, FirstPetIndex)

    InitializeL30Pets(Agent:agent)<suspends>:void=
        Sleep(3.0)
        var FirstPet:creative_prop = creative_prop{}
        var SecondPet:creative_prop = creative_prop{}
        var ThirdPet:creative_prop = creative_prop{}
        var ForthPet:creative_prop = creative_prop{}

        var FirstPetIndex:int = 0
        var SecondPetIndex:int = 0
        var ThirdPetIndex:int = 0
        var ForthPetIndex:int = 0

        var IndexBase:int = 12*29
        #Select right pets for each player
        if(GetTeamIndexAgent[Agent]=1):
            set FirstPetIndex =  IndexBase
            set SecondPetIndex = IndexBase+1
            set ThirdPetIndex =  IndexBase+2
            set ForthPetIndex =  IndexBase+3
        if(GetTeamIndexAgent[Agent]=2):
            set FirstPetIndex =  IndexBase+4
            set SecondPetIndex = IndexBase+5
            set ThirdPetIndex =  IndexBase+6
            set ForthPetIndex =  IndexBase+7
        if(GetTeamIndexAgent[Agent]=3):
            set FirstPetIndex =  IndexBase+8
            set SecondPetIndex = IndexBase+9
            set ThirdPetIndex =  IndexBase+10
            set ForthPetIndex =  IndexBase+11
        
        if (Mypet:=Pets[FirstPetIndex]):
            set FirstPet = Mypet.PetProp
        if (Mypet:=Pets[SecondPetIndex]):
            set SecondPet = Mypet.PetProp
        if (Mypet:=Pets[ThirdPetIndex]):
            set ThirdPet = Mypet.PetProp
        if (Mypet:=Pets[ForthPetIndex]):
            set ForthPet = Mypet.PetProp

        if:
            CurrentPlayerStats := PlayerStatsManager.GetPlayerStats[Agent]
            PlayerObj := player[Agent]
            CustomPlayer:Custom_Player = PlayersMap[PlayerObj]
        then:
            if(CurrentPlayerStats.Ownership_L30_Pet4.CurrentValue=1 and CustomPlayer.FollowActiveL30=0):
                if (Fort := Agent.GetFortCharacter[]):
                    spawn:
                        FollowPlayerL30(Fort, ForthPet, Power_L30P4*1.0, ForthPetIndex)
            if(CurrentPlayerStats.Ownership_L30_Pet3.CurrentValue=1 and CustomPlayer.FollowActiveL30=0):
                if (Fort := Agent.GetFortCharacter[]):
                    spawn:
                        FollowPlayerL30(Fort, ThirdPet, Power_L30P3*1.0, ThirdPetIndex)
            if(CurrentPlayerStats.Ownership_L30_Pet2.CurrentValue=1 and CustomPlayer.FollowActiveL30=0):
                if (Fort := Agent.GetFortCharacter[]):
                    spawn:
                        FollowPlayerL30(Fort, SecondPet, Power_L30P2*1.0, SecondPetIndex)
            if(CurrentPlayerStats.Ownership_L30_Pet1.CurrentValue=1 and CustomPlayer.FollowActiveL30=0):
                if (Fort := Agent.GetFortCharacter[]):
                    spawn:
                        FollowPlayerL30(Fort, FirstPet, Power_L30P1*1.0, FirstPetIndex)

    InitializeL31Pets(Agent:agent)<suspends>:void=
        Sleep(3.1)
        var FirstPet:creative_prop = creative_prop{}
        var FirstPetIndex:int = 0
        var IndexBase:int = 360
        #Select right pets for each player
        if(GetTeamIndexAgent[Agent]=1):
            set FirstPetIndex =  IndexBase
        if(GetTeamIndexAgent[Agent]=2):
            set FirstPetIndex =  IndexBase+1
        if(GetTeamIndexAgent[Agent]=3):
            set FirstPetIndex =  IndexBase+2
        
        if (Mypet:=Pets[FirstPetIndex]):
            set FirstPet = Mypet.PetProp
        if:
            CurrentPlayerStats := PlayerStatsManager.GetPlayerStats[Agent]
            PlayerObj := player[Agent]
            CustomPlayer:Custom_Player = PlayersMap[PlayerObj]
        then:
            if(CurrentPlayerStats.Ownership_L31_Pet1.CurrentValue=1 and CustomPlayer.FollowActiveL31=0):
                if (Fort := Agent.GetFortCharacter[]):
                    spawn:
                        FollowPlayerL31(Fort, FirstPet, 0.0, FirstPetIndex)
            
    InitializeL32Pets(Agent:agent)<suspends>:void=
        Sleep(3.2)
        var FirstPet:creative_prop = creative_prop{}
        var FirstPetIndex:int = 0
        var IndexBase:int = 360+3
        #Select right pets for each player
        if(GetTeamIndexAgent[Agent]=1):
            set FirstPetIndex =  IndexBase
        if(GetTeamIndexAgent[Agent]=2):
            set FirstPetIndex =  IndexBase+1
        if(GetTeamIndexAgent[Agent]=3):
            set FirstPetIndex =  IndexBase+2
        
        if (Mypet:=Pets[FirstPetIndex]):
            set FirstPet = Mypet.PetProp
        if:
            CurrentPlayerStats := PlayerStatsManager.GetPlayerStats[Agent]
            PlayerObj := player[Agent]
            CustomPlayer:Custom_Player = PlayersMap[PlayerObj]
        then:
            if(CurrentPlayerStats.Ownership_L32_Pet1.CurrentValue=1 and CustomPlayer.FollowActiveL32=0):
                if (Fort := Agent.GetFortCharacter[]):
                    spawn:
                        FollowPlayerL32(Fort, FirstPet, 0.0, FirstPetIndex)

    InitializeL33Pets(Agent:agent)<suspends>:void=
        Sleep(3.3)
        var FirstPet:creative_prop = creative_prop{}
        var FirstPetIndex:int = 0
        var IndexBase:int = 360+6
        #Select right pets for each player
        if(GetTeamIndexAgent[Agent]=1):
            set FirstPetIndex =  IndexBase
        if(GetTeamIndexAgent[Agent]=2):
            set FirstPetIndex =  IndexBase+1
        if(GetTeamIndexAgent[Agent]=3):
            set FirstPetIndex =  IndexBase+2
        
        if (Mypet:=Pets[FirstPetIndex]):
            set FirstPet = Mypet.PetProp
        if:
            CurrentPlayerStats := PlayerStatsManager.GetPlayerStats[Agent]
            PlayerObj := player[Agent]
            CustomPlayer:Custom_Player = PlayersMap[PlayerObj]
        then:
            if(CurrentPlayerStats.Ownership_L33_Pet1.CurrentValue=1 and CustomPlayer.FollowActiveL33=0):
                if (Fort := Agent.GetFortCharacter[]):
                    spawn:
                        FollowPlayerL33(Fort, FirstPet, 0.0, FirstPetIndex)

    InitializeL34Pets(Agent:agent)<suspends>:void=
        Sleep(3.4)
        var FirstPet:creative_prop = creative_prop{}
        var FirstPetIndex:int = 0
        var IndexBase:int = 360+9
        #Select right pets for each player
        if(GetTeamIndexAgent[Agent]=1):
            set FirstPetIndex =  IndexBase
        if(GetTeamIndexAgent[Agent]=2):
            set FirstPetIndex =  IndexBase+1
        if(GetTeamIndexAgent[Agent]=3):
            set FirstPetIndex =  IndexBase+2
        
        if (Mypet:=Pets[FirstPetIndex]):
            set FirstPet = Mypet.PetProp
        if:
            CurrentPlayerStats := PlayerStatsManager.GetPlayerStats[Agent]
            PlayerObj := player[Agent]
            CustomPlayer:Custom_Player = PlayersMap[PlayerObj]
        then:
            if(CurrentPlayerStats.Ownership_L34_Pet1.CurrentValue=1 and CustomPlayer.FollowActiveL34=0):
                if (Fort := Agent.GetFortCharacter[]):
                    spawn:
                        FollowPlayerL34(Fort, FirstPet, 500.0, FirstPetIndex)

    InitializePetOwnerShipUI(Agent:agent):void=
        if:
            PlayerObj := player[Agent]
            CurrentPlayerStats := PlayerStatsManager.GetPlayerStats[Agent]
            CustomPlayer:Custom_Player = PlayersMap[PlayerObj]
        then:
            #L1
            if(CurrentPlayerStats.Ownership_L1_Pet1.CurrentValue=1):
                set CustomPlayer.Ownership_L1_Pet1=1
                set CustomPlayer.Pet_Unlocked_L1=1
                CustomPlayer.Update_Pet_PowerL1(Power_L1P1)
            if(CurrentPlayerStats.Ownership_L1_Pet2.CurrentValue=1):
                set CustomPlayer.Ownership_L1_Pet2=1
                set CustomPlayer.Pet_Unlocked_L1=2
                CustomPlayer.Update_Pet_PowerL1(Power_L1P2)
            if(CurrentPlayerStats.Ownership_L1_Pet3.CurrentValue=1):
                set CustomPlayer.Ownership_L1_Pet3=1
                set CustomPlayer.Pet_Unlocked_L1=3
                CustomPlayer.Update_Pet_PowerL1(Power_L1P3)
            if(CurrentPlayerStats.Ownership_L1_Pet4.CurrentValue=1):
                set CustomPlayer.Ownership_L1_Pet4=1
                set CustomPlayer.Pet_Unlocked_L1=4
                CustomPlayer.Update_Pet_PowerL1(Power_L1P4)

            #L2
            if(CurrentPlayerStats.Ownership_L2_Pet1.CurrentValue=1):
                set CustomPlayer.Ownership_L2_Pet1=1
                set CustomPlayer.Pet_Unlocked_L2=1
                CustomPlayer.Update_Pet_PowerL2(Power_L2P1)
            if(CurrentPlayerStats.Ownership_L2_Pet2.CurrentValue=1):
                set CustomPlayer.Ownership_L2_Pet2=1
                set CustomPlayer.Pet_Unlocked_L2=2
                CustomPlayer.Update_Pet_PowerL2(Power_L2P2)
            if(CurrentPlayerStats.Ownership_L2_Pet3.CurrentValue=1):
                set CustomPlayer.Ownership_L2_Pet3=1
                set CustomPlayer.Pet_Unlocked_L2=3
                CustomPlayer.Update_Pet_PowerL2(Power_L2P3)
            if(CurrentPlayerStats.Ownership_L2_Pet4.CurrentValue=1):
                set CustomPlayer.Ownership_L2_Pet4=1
                set CustomPlayer.Pet_Unlocked_L2=4
                CustomPlayer.Update_Pet_PowerL2(Power_L2P4)

            #L3
            if(CurrentPlayerStats.Ownership_L3_Pet1.CurrentValue=1):
                set CustomPlayer.Ownership_L3_Pet1=1
                set CustomPlayer.Pet_Unlocked_L3=1
                CustomPlayer.Update_Pet_PowerL3(Power_L3P1)
            if(CurrentPlayerStats.Ownership_L3_Pet2.CurrentValue=1):
                set CustomPlayer.Ownership_L3_Pet2=1
                set CustomPlayer.Pet_Unlocked_L3=2
                CustomPlayer.Update_Pet_PowerL3(Power_L3P2)
            if(CurrentPlayerStats.Ownership_L3_Pet3.CurrentValue=1):
                set CustomPlayer.Ownership_L3_Pet3=1
                set CustomPlayer.Pet_Unlocked_L3=3
                CustomPlayer.Update_Pet_PowerL3(Power_L3P3)
            if(CurrentPlayerStats.Ownership_L3_Pet4.CurrentValue=1):
                set CustomPlayer.Ownership_L3_Pet4=1
                set CustomPlayer.Pet_Unlocked_L3=4
                CustomPlayer.Update_Pet_PowerL3(Power_L3P4)

            #L4
            if(CurrentPlayerStats.Ownership_L4_Pet1.CurrentValue=1):
                set CustomPlayer.Ownership_L4_Pet1=1
                set CustomPlayer.Pet_Unlocked_L4=1
                CustomPlayer.Update_Pet_PowerL4(Power_L4P1)
            if(CurrentPlayerStats.Ownership_L4_Pet2.CurrentValue=1):
                set CustomPlayer.Ownership_L4_Pet2=1
                set CustomPlayer.Pet_Unlocked_L4=2
                CustomPlayer.Update_Pet_PowerL4(Power_L4P2)
            if(CurrentPlayerStats.Ownership_L4_Pet3.CurrentValue=1):
                set CustomPlayer.Ownership_L4_Pet3=1
                set CustomPlayer.Pet_Unlocked_L4=3
                CustomPlayer.Update_Pet_PowerL4(Power_L4P3)
            if(CurrentPlayerStats.Ownership_L4_Pet4.CurrentValue=1):
                set CustomPlayer.Ownership_L4_Pet4=1
                set CustomPlayer.Pet_Unlocked_L4=4
                CustomPlayer.Update_Pet_PowerL4(Power_L4P4)

            #L5
            if(CurrentPlayerStats.Ownership_L5_Pet1.CurrentValue=1):
                set CustomPlayer.Ownership_L5_Pet1=1
                set CustomPlayer.Pet_Unlocked_L5=1
                CustomPlayer.Update_Pet_PowerL5(Power_L5P1)
            if(CurrentPlayerStats.Ownership_L5_Pet2.CurrentValue=1):
                set CustomPlayer.Ownership_L5_Pet2=1
                set CustomPlayer.Pet_Unlocked_L5=2
                CustomPlayer.Update_Pet_PowerL5(Power_L5P2)
            if(CurrentPlayerStats.Ownership_L5_Pet3.CurrentValue=1):
                set CustomPlayer.Ownership_L5_Pet3=1
                set CustomPlayer.Pet_Unlocked_L5=3
                CustomPlayer.Update_Pet_PowerL5(Power_L5P3)
            if(CurrentPlayerStats.Ownership_L5_Pet4.CurrentValue=1):
                set CustomPlayer.Ownership_L5_Pet4=1
                set CustomPlayer.Pet_Unlocked_L5=4
                CustomPlayer.Update_Pet_PowerL5(Power_L5P4)

            #L6
            if(CurrentPlayerStats.Ownership_L6_Pet1.CurrentValue=1):
                set CustomPlayer.Ownership_L6_Pet1=1
                set CustomPlayer.Pet_Unlocked_L6=1
                CustomPlayer.Update_Pet_PowerL6(Power_L6P1)
            if(CurrentPlayerStats.Ownership_L6_Pet2.CurrentValue=1):
                set CustomPlayer.Ownership_L6_Pet2=1
                set CustomPlayer.Pet_Unlocked_L6=2
                CustomPlayer.Update_Pet_PowerL6(Power_L6P2)
            if(CurrentPlayerStats.Ownership_L6_Pet3.CurrentValue=1):
                set CustomPlayer.Ownership_L6_Pet3=1
                set CustomPlayer.Pet_Unlocked_L6=3
                CustomPlayer.Update_Pet_PowerL6(Power_L6P3)
            if(CurrentPlayerStats.Ownership_L6_Pet4.CurrentValue=1):
                set CustomPlayer.Ownership_L6_Pet4=1
                set CustomPlayer.Pet_Unlocked_L6=4
                CustomPlayer.Update_Pet_PowerL6(Power_L6P4)

            #L7
            if(CurrentPlayerStats.Ownership_L7_Pet1.CurrentValue=1):
                set CustomPlayer.Ownership_L7_Pet1=1
                set CustomPlayer.Pet_Unlocked_L7=1
                CustomPlayer.Update_Pet_PowerL7(Power_L7P1)
            if(CurrentPlayerStats.Ownership_L7_Pet2.CurrentValue=1):
                set CustomPlayer.Ownership_L7_Pet2=1
                set CustomPlayer.Pet_Unlocked_L7=2
                CustomPlayer.Update_Pet_PowerL7(Power_L7P2)
            if(CurrentPlayerStats.Ownership_L7_Pet3.CurrentValue=1):
                set CustomPlayer.Ownership_L7_Pet3=1
                set CustomPlayer.Pet_Unlocked_L7=3
                CustomPlayer.Update_Pet_PowerL7(Power_L7P3)
            if(CurrentPlayerStats.Ownership_L7_Pet4.CurrentValue=1):
                set CustomPlayer.Ownership_L7_Pet4=1
                set CustomPlayer.Pet_Unlocked_L7=4
                CustomPlayer.Update_Pet_PowerL7(Power_L7P4)

            #L8
            if(CurrentPlayerStats.Ownership_L8_Pet1.CurrentValue=1):
                set CustomPlayer.Ownership_L8_Pet1=1
                set CustomPlayer.Pet_Unlocked_L8=1
                CustomPlayer.Update_Pet_PowerL8(Power_L8P1)
            if(CurrentPlayerStats.Ownership_L8_Pet2.CurrentValue=1):
                set CustomPlayer.Ownership_L8_Pet2=1
                set CustomPlayer.Pet_Unlocked_L8=2
                CustomPlayer.Update_Pet_PowerL8(Power_L8P2)
            if(CurrentPlayerStats.Ownership_L8_Pet3.CurrentValue=1):
                set CustomPlayer.Ownership_L8_Pet3=1
                set CustomPlayer.Pet_Unlocked_L8=3
                CustomPlayer.Update_Pet_PowerL8(Power_L8P3)
            if(CurrentPlayerStats.Ownership_L8_Pet4.CurrentValue=1):
                set CustomPlayer.Ownership_L8_Pet4=1
                set CustomPlayer.Pet_Unlocked_L8=4
                CustomPlayer.Update_Pet_PowerL8(Power_L8P4)

            #L9
            if(CurrentPlayerStats.Ownership_L9_Pet1.CurrentValue=1):
                set CustomPlayer.Ownership_L9_Pet1=1
                set CustomPlayer.Pet_Unlocked_L9=1
                CustomPlayer.Update_Pet_PowerL9(Power_L9P1)
            if(CurrentPlayerStats.Ownership_L9_Pet2.CurrentValue=1):
                set CustomPlayer.Ownership_L9_Pet2=1
                set CustomPlayer.Pet_Unlocked_L9=2
                CustomPlayer.Update_Pet_PowerL9(Power_L9P2)
            if(CurrentPlayerStats.Ownership_L9_Pet3.CurrentValue=1):
                set CustomPlayer.Ownership_L9_Pet3=1
                set CustomPlayer.Pet_Unlocked_L9=3
                CustomPlayer.Update_Pet_PowerL9(Power_L9P3)
            if(CurrentPlayerStats.Ownership_L9_Pet4.CurrentValue=1):
                set CustomPlayer.Ownership_L9_Pet4=1
                set CustomPlayer.Pet_Unlocked_L9=4
                CustomPlayer.Update_Pet_PowerL9(Power_L9P4)

            #L10
            if(CurrentPlayerStats.Ownership_L10_Pet1.CurrentValue=1):
                set CustomPlayer.Ownership_L10_Pet1=1
                set CustomPlayer.Pet_Unlocked_L10=1
                CustomPlayer.Update_Pet_PowerL10(Power_L10P1)
            if(CurrentPlayerStats.Ownership_L10_Pet2.CurrentValue=1):
                set CustomPlayer.Ownership_L10_Pet2=1
                set CustomPlayer.Pet_Unlocked_L10=2
                CustomPlayer.Update_Pet_PowerL10(Power_L10P2)
            if(CurrentPlayerStats.Ownership_L10_Pet3.CurrentValue=1):
                set CustomPlayer.Ownership_L10_Pet3=1
                set CustomPlayer.Pet_Unlocked_L10=3
                CustomPlayer.Update_Pet_PowerL10(Power_L10P3)
            if(CurrentPlayerStats.Ownership_L10_Pet4.CurrentValue=1):
                set CustomPlayer.Ownership_L10_Pet4=1
                set CustomPlayer.Pet_Unlocked_L10=4
                CustomPlayer.Update_Pet_PowerL10(Power_L10P4)

            #L11
            if(CurrentPlayerStats.Ownership_L11_Pet1.CurrentValue=1):
                set CustomPlayer.Ownership_L11_Pet1=1
                set CustomPlayer.Pet_Unlocked_L11=1
                CustomPlayer.Update_Pet_PowerL11(Power_L11P1)
            if(CurrentPlayerStats.Ownership_L11_Pet2.CurrentValue=1):
                set CustomPlayer.Ownership_L11_Pet2=1
                set CustomPlayer.Pet_Unlocked_L11=2
                CustomPlayer.Update_Pet_PowerL11(Power_L11P2)
            if(CurrentPlayerStats.Ownership_L11_Pet3.CurrentValue=1):
                set CustomPlayer.Ownership_L11_Pet3=1
                set CustomPlayer.Pet_Unlocked_L11=3
                CustomPlayer.Update_Pet_PowerL11(Power_L11P3)
            if(CurrentPlayerStats.Ownership_L11_Pet4.CurrentValue=1):
                set CustomPlayer.Ownership_L11_Pet4=1
                set CustomPlayer.Pet_Unlocked_L11=4
                CustomPlayer.Update_Pet_PowerL11(Power_L11P4)

            #L12
            if(CurrentPlayerStats.Ownership_L12_Pet1.CurrentValue=1):
                set CustomPlayer.Ownership_L12_Pet1=1
                set CustomPlayer.Pet_Unlocked_L12=1
                CustomPlayer.Update_Pet_PowerL12(Power_L12P1)
            if(CurrentPlayerStats.Ownership_L12_Pet2.CurrentValue=1):
                set CustomPlayer.Ownership_L12_Pet2=1
                set CustomPlayer.Pet_Unlocked_L12=2
                CustomPlayer.Update_Pet_PowerL12(Power_L12P2)
            if(CurrentPlayerStats.Ownership_L12_Pet3.CurrentValue=1):
                set CustomPlayer.Ownership_L12_Pet3=1
                set CustomPlayer.Pet_Unlocked_L12=3
                CustomPlayer.Update_Pet_PowerL12(Power_L12P3)
            if(CurrentPlayerStats.Ownership_L12_Pet4.CurrentValue=1):
                set CustomPlayer.Ownership_L12_Pet4=1
                set CustomPlayer.Pet_Unlocked_L12=4
                CustomPlayer.Update_Pet_PowerL12(Power_L12P4)

            #L13
            if(CurrentPlayerStats.Ownership_L13_Pet1.CurrentValue=1):
                set CustomPlayer.Ownership_L13_Pet1=1
                set CustomPlayer.Pet_Unlocked_L13=1
                CustomPlayer.Update_Pet_PowerL13(Power_L13P1)
            if(CurrentPlayerStats.Ownership_L13_Pet2.CurrentValue=1):
                set CustomPlayer.Ownership_L13_Pet2=1
                set CustomPlayer.Pet_Unlocked_L13=2
                CustomPlayer.Update_Pet_PowerL13(Power_L13P2)
            if(CurrentPlayerStats.Ownership_L13_Pet3.CurrentValue=1):
                set CustomPlayer.Ownership_L13_Pet3=1
                set CustomPlayer.Pet_Unlocked_L13=3
                CustomPlayer.Update_Pet_PowerL13(Power_L13P3)
            if(CurrentPlayerStats.Ownership_L13_Pet4.CurrentValue=1):
                set CustomPlayer.Ownership_L13_Pet4=1
                set CustomPlayer.Pet_Unlocked_L13=4
                CustomPlayer.Update_Pet_PowerL13(Power_L13P4)

            #L14
            if(CurrentPlayerStats.Ownership_L14_Pet1.CurrentValue=1):
                set CustomPlayer.Ownership_L14_Pet1=1
                set CustomPlayer.Pet_Unlocked_L14=1
                CustomPlayer.Update_Pet_PowerL14(Power_L14P1)
            if(CurrentPlayerStats.Ownership_L14_Pet2.CurrentValue=1):
                set CustomPlayer.Ownership_L14_Pet2=1
                set CustomPlayer.Pet_Unlocked_L14=2
                CustomPlayer.Update_Pet_PowerL14(Power_L14P2)
            if(CurrentPlayerStats.Ownership_L14_Pet3.CurrentValue=1):
                set CustomPlayer.Ownership_L14_Pet3=1
                set CustomPlayer.Pet_Unlocked_L14=3
                CustomPlayer.Update_Pet_PowerL14(Power_L14P3)
            if(CurrentPlayerStats.Ownership_L14_Pet4.CurrentValue=1):
                set CustomPlayer.Ownership_L14_Pet4=1
                set CustomPlayer.Pet_Unlocked_L14=4
                CustomPlayer.Update_Pet_PowerL14(Power_L14P4)

            #L15
            if(CurrentPlayerStats.Ownership_L15_Pet1.CurrentValue=1):
                set CustomPlayer.Ownership_L15_Pet1=1
                set CustomPlayer.Pet_Unlocked_L15=1
                CustomPlayer.Update_Pet_PowerL15(Power_L15P1)
            if(CurrentPlayerStats.Ownership_L15_Pet2.CurrentValue=1):
                set CustomPlayer.Ownership_L15_Pet2=1
                set CustomPlayer.Pet_Unlocked_L15=2
                CustomPlayer.Update_Pet_PowerL15(Power_L15P2)
            if(CurrentPlayerStats.Ownership_L15_Pet3.CurrentValue=1):
                set CustomPlayer.Ownership_L15_Pet3=1
                set CustomPlayer.Pet_Unlocked_L15=3
                CustomPlayer.Update_Pet_PowerL15(Power_L15P3)
            if(CurrentPlayerStats.Ownership_L15_Pet4.CurrentValue=1):
                set CustomPlayer.Ownership_L15_Pet4=1
                set CustomPlayer.Pet_Unlocked_L15=4
                CustomPlayer.Update_Pet_PowerL15(Power_L15P4)

            #L16
            if(CurrentPlayerStats.Ownership_L16_Pet1.CurrentValue=1):
                set CustomPlayer.Ownership_L16_Pet1=1
                set CustomPlayer.Pet_Unlocked_L16=1
                CustomPlayer.Update_Pet_PowerL16(Power_L16P1)
            if(CurrentPlayerStats.Ownership_L16_Pet2.CurrentValue=1):
                set CustomPlayer.Ownership_L16_Pet2=1
                set CustomPlayer.Pet_Unlocked_L16=2
                CustomPlayer.Update_Pet_PowerL16(Power_L16P2)
            if(CurrentPlayerStats.Ownership_L16_Pet3.CurrentValue=1):
                set CustomPlayer.Ownership_L16_Pet3=1
                set CustomPlayer.Pet_Unlocked_L16=3
                CustomPlayer.Update_Pet_PowerL16(Power_L16P3)
            if(CurrentPlayerStats.Ownership_L16_Pet4.CurrentValue=1):
                set CustomPlayer.Ownership_L16_Pet4=1
                set CustomPlayer.Pet_Unlocked_L16=4
                CustomPlayer.Update_Pet_PowerL16(Power_L16P4)

            #L17
            if(CurrentPlayerStats.Ownership_L17_Pet1.CurrentValue=1):
                set CustomPlayer.Ownership_L17_Pet1=1
                set CustomPlayer.Pet_Unlocked_L17=1
                CustomPlayer.Update_Pet_PowerL17(Power_L17P1)
            if(CurrentPlayerStats.Ownership_L17_Pet2.CurrentValue=1):
                set CustomPlayer.Ownership_L17_Pet2=1
                set CustomPlayer.Pet_Unlocked_L17=2
                CustomPlayer.Update_Pet_PowerL17(Power_L17P2)
            if(CurrentPlayerStats.Ownership_L17_Pet3.CurrentValue=1):
                set CustomPlayer.Ownership_L17_Pet3=1
                set CustomPlayer.Pet_Unlocked_L17=3
                CustomPlayer.Update_Pet_PowerL17(Power_L17P3)
            if(CurrentPlayerStats.Ownership_L17_Pet4.CurrentValue=1):
                set CustomPlayer.Ownership_L17_Pet4=1
                set CustomPlayer.Pet_Unlocked_L17=4
                CustomPlayer.Update_Pet_PowerL17(Power_L17P4)

            #L18
            if(CurrentPlayerStats.Ownership_L18_Pet1.CurrentValue=1):
                set CustomPlayer.Ownership_L18_Pet1=1
                set CustomPlayer.Pet_Unlocked_L18=1
                CustomPlayer.Update_Pet_PowerL18(Power_L18P1)
            if(CurrentPlayerStats.Ownership_L18_Pet2.CurrentValue=1):
                set CustomPlayer.Ownership_L18_Pet2=1
                set CustomPlayer.Pet_Unlocked_L18=2
                CustomPlayer.Update_Pet_PowerL18(Power_L18P2)
            if(CurrentPlayerStats.Ownership_L18_Pet3.CurrentValue=1):
                set CustomPlayer.Ownership_L18_Pet3=1
                set CustomPlayer.Pet_Unlocked_L18=3
                CustomPlayer.Update_Pet_PowerL18(Power_L18P3)
            if(CurrentPlayerStats.Ownership_L18_Pet4.CurrentValue=1):
                set CustomPlayer.Ownership_L18_Pet4=1
                set CustomPlayer.Pet_Unlocked_L18=4
                CustomPlayer.Update_Pet_PowerL18(Power_L18P4)

            #L19
            if(CurrentPlayerStats.Ownership_L19_Pet1.CurrentValue=1):
                set CustomPlayer.Ownership_L19_Pet1=1
                set CustomPlayer.Pet_Unlocked_L19=1
                CustomPlayer.Update_Pet_PowerL19(Power_L19P1)
            if(CurrentPlayerStats.Ownership_L19_Pet2.CurrentValue=1):
                set CustomPlayer.Ownership_L19_Pet2=1
                set CustomPlayer.Pet_Unlocked_L19=2
                CustomPlayer.Update_Pet_PowerL19(Power_L19P2)
            if(CurrentPlayerStats.Ownership_L19_Pet3.CurrentValue=1):
                set CustomPlayer.Ownership_L19_Pet3=1
                set CustomPlayer.Pet_Unlocked_L19=3
                CustomPlayer.Update_Pet_PowerL19(Power_L19P3)
            if(CurrentPlayerStats.Ownership_L19_Pet4.CurrentValue=1):
                set CustomPlayer.Ownership_L19_Pet4=1
                set CustomPlayer.Pet_Unlocked_L19=4
                CustomPlayer.Update_Pet_PowerL19(Power_L19P4)

            #L20
            if(CurrentPlayerStats.Ownership_L20_Pet1.CurrentValue=1):
                set CustomPlayer.Ownership_L20_Pet1=1
                set CustomPlayer.Pet_Unlocked_L20=1
                CustomPlayer.Update_Pet_PowerL20(Power_L20P1)
            if(CurrentPlayerStats.Ownership_L20_Pet2.CurrentValue=1):
                set CustomPlayer.Ownership_L20_Pet2=1
                set CustomPlayer.Pet_Unlocked_L20=2
                CustomPlayer.Update_Pet_PowerL20(Power_L20P2)
            if(CurrentPlayerStats.Ownership_L20_Pet3.CurrentValue=1):
                set CustomPlayer.Ownership_L20_Pet3=1
                set CustomPlayer.Pet_Unlocked_L20=3
                CustomPlayer.Update_Pet_PowerL20(Power_L20P3)
            if(CurrentPlayerStats.Ownership_L20_Pet4.CurrentValue=1):
                set CustomPlayer.Ownership_L20_Pet4=1
                set CustomPlayer.Pet_Unlocked_L20=4
                CustomPlayer.Update_Pet_PowerL20(Power_L20P4)

            #L21
            if(CurrentPlayerStats.Ownership_L21_Pet1.CurrentValue=1):
                set CustomPlayer.Ownership_L21_Pet1=1
                set CustomPlayer.Pet_Unlocked_L21=1
                CustomPlayer.Update_Pet_PowerL21(Power_L21P1)
            if(CurrentPlayerStats.Ownership_L21_Pet2.CurrentValue=1):
                set CustomPlayer.Ownership_L21_Pet2=1
                set CustomPlayer.Pet_Unlocked_L21=2
                CustomPlayer.Update_Pet_PowerL21(Power_L21P2)
            if(CurrentPlayerStats.Ownership_L21_Pet3.CurrentValue=1):
                set CustomPlayer.Ownership_L21_Pet3=1
                set CustomPlayer.Pet_Unlocked_L21=3
                CustomPlayer.Update_Pet_PowerL21(Power_L21P3)
            if(CurrentPlayerStats.Ownership_L21_Pet4.CurrentValue=1):
                set CustomPlayer.Ownership_L21_Pet4=1
                set CustomPlayer.Pet_Unlocked_L21=4
                CustomPlayer.Update_Pet_PowerL21(Power_L21P4)

            #L22
            if(CurrentPlayerStats.Ownership_L22_Pet1.CurrentValue=1):
                set CustomPlayer.Ownership_L22_Pet1=1
                set CustomPlayer.Pet_Unlocked_L22=1
                CustomPlayer.Update_Pet_PowerL22(Power_L22P1)
            if(CurrentPlayerStats.Ownership_L22_Pet2.CurrentValue=1):
                set CustomPlayer.Ownership_L22_Pet2=1
                set CustomPlayer.Pet_Unlocked_L22=2
                CustomPlayer.Update_Pet_PowerL22(Power_L22P2)
            if(CurrentPlayerStats.Ownership_L22_Pet3.CurrentValue=1):
                set CustomPlayer.Ownership_L22_Pet3=1
                set CustomPlayer.Pet_Unlocked_L22=3
                CustomPlayer.Update_Pet_PowerL22(Power_L22P3)
            if(CurrentPlayerStats.Ownership_L22_Pet4.CurrentValue=1):
                set CustomPlayer.Ownership_L22_Pet4=1
                set CustomPlayer.Pet_Unlocked_L22=4
                CustomPlayer.Update_Pet_PowerL22(Power_L22P4)

            #L23
            if(CurrentPlayerStats.Ownership_L23_Pet1.CurrentValue=1):
                set CustomPlayer.Ownership_L23_Pet1=1
                set CustomPlayer.Pet_Unlocked_L23=1
                CustomPlayer.Update_Pet_PowerL23(Power_L23P1)
            if(CurrentPlayerStats.Ownership_L23_Pet2.CurrentValue=1):
                set CustomPlayer.Ownership_L23_Pet2=1
                set CustomPlayer.Pet_Unlocked_L23=2
                CustomPlayer.Update_Pet_PowerL23(Power_L23P2)
            if(CurrentPlayerStats.Ownership_L23_Pet3.CurrentValue=1):
                set CustomPlayer.Ownership_L23_Pet3=1
                set CustomPlayer.Pet_Unlocked_L23=3
                CustomPlayer.Update_Pet_PowerL23(Power_L23P3)
            if(CurrentPlayerStats.Ownership_L23_Pet4.CurrentValue=1):
                set CustomPlayer.Ownership_L23_Pet4=1
                set CustomPlayer.Pet_Unlocked_L23=4
                CustomPlayer.Update_Pet_PowerL23(Power_L23P4)

            #L24
            if(CurrentPlayerStats.Ownership_L24_Pet1.CurrentValue=1):
                set CustomPlayer.Ownership_L24_Pet1=1
                set CustomPlayer.Pet_Unlocked_L24=1
                CustomPlayer.Update_Pet_PowerL24(Power_L24P1)
            if(CurrentPlayerStats.Ownership_L24_Pet2.CurrentValue=1):
                set CustomPlayer.Ownership_L24_Pet2=1
                set CustomPlayer.Pet_Unlocked_L24=2
                CustomPlayer.Update_Pet_PowerL24(Power_L24P2)
            if(CurrentPlayerStats.Ownership_L24_Pet3.CurrentValue=1):
                set CustomPlayer.Ownership_L24_Pet3=1
                set CustomPlayer.Pet_Unlocked_L24=3
                CustomPlayer.Update_Pet_PowerL24(Power_L24P3)
            if(CurrentPlayerStats.Ownership_L24_Pet4.CurrentValue=1):
                set CustomPlayer.Ownership_L24_Pet4=1
                set CustomPlayer.Pet_Unlocked_L24=4
                CustomPlayer.Update_Pet_PowerL24(Power_L24P4)
                            
            #L25
            if(CurrentPlayerStats.Ownership_L25_Pet1.CurrentValue=1):
                set CustomPlayer.Ownership_L25_Pet1=1
                set CustomPlayer.Pet_Unlocked_L25=1
                CustomPlayer.Update_Pet_PowerL25(Power_L25P1)
            if(CurrentPlayerStats.Ownership_L25_Pet2.CurrentValue=1):
                set CustomPlayer.Ownership_L25_Pet2=1
                set CustomPlayer.Pet_Unlocked_L25=2
                CustomPlayer.Update_Pet_PowerL25(Power_L25P2)
            if(CurrentPlayerStats.Ownership_L25_Pet3.CurrentValue=1):
                set CustomPlayer.Ownership_L25_Pet3=1
                set CustomPlayer.Pet_Unlocked_L25=3
                CustomPlayer.Update_Pet_PowerL25(Power_L25P3)
            if(CurrentPlayerStats.Ownership_L25_Pet4.CurrentValue=1):
                set CustomPlayer.Ownership_L25_Pet4=1
                set CustomPlayer.Pet_Unlocked_L25=4
                CustomPlayer.Update_Pet_PowerL25(Power_L25P4)

            #L26
            if(CurrentPlayerStats.Ownership_L26_Pet1.CurrentValue=1):
                set CustomPlayer.Ownership_L26_Pet1=1
                set CustomPlayer.Pet_Unlocked_L26=1
                CustomPlayer.Update_Pet_PowerL26(Power_L26P1)
            if(CurrentPlayerStats.Ownership_L26_Pet2.CurrentValue=1):
                set CustomPlayer.Ownership_L26_Pet2=1
                set CustomPlayer.Pet_Unlocked_L26=2
                CustomPlayer.Update_Pet_PowerL26(Power_L26P2)
            if(CurrentPlayerStats.Ownership_L26_Pet3.CurrentValue=1):
                set CustomPlayer.Ownership_L26_Pet3=1
                set CustomPlayer.Pet_Unlocked_L26=3
                CustomPlayer.Update_Pet_PowerL26(Power_L26P3)
            if(CurrentPlayerStats.Ownership_L26_Pet4.CurrentValue=1):
                set CustomPlayer.Ownership_L26_Pet4=1
                set CustomPlayer.Pet_Unlocked_L26=4
                CustomPlayer.Update_Pet_PowerL26(Power_L26P4)

            #L27
            if(CurrentPlayerStats.Ownership_L27_Pet1.CurrentValue=1):
                set CustomPlayer.Ownership_L27_Pet1=1
                set CustomPlayer.Pet_Unlocked_L27=1
                CustomPlayer.Update_Pet_PowerL27(Power_L27P1)
            if(CurrentPlayerStats.Ownership_L27_Pet2.CurrentValue=1):
                set CustomPlayer.Ownership_L27_Pet2=1
                set CustomPlayer.Pet_Unlocked_L27=2
                CustomPlayer.Update_Pet_PowerL27(Power_L27P2)
            if(CurrentPlayerStats.Ownership_L27_Pet3.CurrentValue=1):
                set CustomPlayer.Ownership_L27_Pet3=1
                set CustomPlayer.Pet_Unlocked_L27=3
                CustomPlayer.Update_Pet_PowerL27(Power_L27P3)
            if(CurrentPlayerStats.Ownership_L27_Pet4.CurrentValue=1):
                set CustomPlayer.Ownership_L27_Pet4=1
                set CustomPlayer.Pet_Unlocked_L27=4
                CustomPlayer.Update_Pet_PowerL27(Power_L27P4)
                            
            #L28
            if(CurrentPlayerStats.Ownership_L28_Pet1.CurrentValue=1):
                set CustomPlayer.Ownership_L28_Pet1=1
                set CustomPlayer.Pet_Unlocked_L28=1
                CustomPlayer.Update_Pet_PowerL28(Power_L28P1)
            if(CurrentPlayerStats.Ownership_L28_Pet2.CurrentValue=1):
                set CustomPlayer.Ownership_L28_Pet2=1
                set CustomPlayer.Pet_Unlocked_L28=2
                CustomPlayer.Update_Pet_PowerL28(Power_L28P2)
            if(CurrentPlayerStats.Ownership_L28_Pet3.CurrentValue=1):
                set CustomPlayer.Ownership_L28_Pet3=1
                set CustomPlayer.Pet_Unlocked_L28=3
                CustomPlayer.Update_Pet_PowerL28(Power_L28P3)
            if(CurrentPlayerStats.Ownership_L28_Pet4.CurrentValue=1):
                set CustomPlayer.Ownership_L28_Pet4=1
                set CustomPlayer.Pet_Unlocked_L28=4
                CustomPlayer.Update_Pet_PowerL28(Power_L28P4)

            #L29
            if(CurrentPlayerStats.Ownership_L29_Pet1.CurrentValue=1):
                set CustomPlayer.Ownership_L29_Pet1=1
                set CustomPlayer.Pet_Unlocked_L29=1
                CustomPlayer.Update_Pet_PowerL29(Power_L29P1)
            if(CurrentPlayerStats.Ownership_L29_Pet2.CurrentValue=1):
                set CustomPlayer.Ownership_L29_Pet2=1
                set CustomPlayer.Pet_Unlocked_L29=2
                CustomPlayer.Update_Pet_PowerL29(Power_L29P2)
            if(CurrentPlayerStats.Ownership_L29_Pet3.CurrentValue=1):
                set CustomPlayer.Ownership_L29_Pet3=1
                set CustomPlayer.Pet_Unlocked_L29=3
                CustomPlayer.Update_Pet_PowerL29(Power_L29P3)
            if(CurrentPlayerStats.Ownership_L29_Pet4.CurrentValue=1):
                set CustomPlayer.Ownership_L29_Pet4=1
                set CustomPlayer.Pet_Unlocked_L29=4
                CustomPlayer.Update_Pet_PowerL29(Power_L29P4)
                            
            #L30
            if(CurrentPlayerStats.Ownership_L30_Pet1.CurrentValue=1):
                set CustomPlayer.Ownership_L30_Pet1=1
                set CustomPlayer.Pet_Unlocked_L30=1
                CustomPlayer.Update_Pet_PowerL30(Power_L30P1)
            if(CurrentPlayerStats.Ownership_L30_Pet2.CurrentValue=1):
                set CustomPlayer.Ownership_L30_Pet2=1
                set CustomPlayer.Pet_Unlocked_L30=2
                CustomPlayer.Update_Pet_PowerL30(Power_L30P2)
            if(CurrentPlayerStats.Ownership_L30_Pet3.CurrentValue=1):
                set CustomPlayer.Ownership_L30_Pet3=1
                set CustomPlayer.Pet_Unlocked_L30=3
                CustomPlayer.Update_Pet_PowerL30(Power_L30P3)
            if(CurrentPlayerStats.Ownership_L30_Pet4.CurrentValue=1):
                set CustomPlayer.Ownership_L30_Pet4=1
                set CustomPlayer.Pet_Unlocked_L30=4
                CustomPlayer.Update_Pet_PowerL30(Power_L30P4)

            #Secret Pets
            if(CurrentPlayerStats.Ownership_L31_Pet1.CurrentValue=1):
                set CustomPlayer.Ownership_Secret_Pet1=1
            if(CurrentPlayerStats.Ownership_L32_Pet1.CurrentValue=1):
                set CustomPlayer.Ownership_Secret_Pet2=1
            if(CurrentPlayerStats.Ownership_L33_Pet1.CurrentValue=1):
                set CustomPlayer.Ownership_Secret_Pet3=1
            if(CurrentPlayerStats.Ownership_L34_Pet1.CurrentValue=1):
                set CustomPlayer.Ownership_Secret_Pet4=1

    InitializeStatUI(Agent:agent):void=
        if:
            PlayerObj := player[Agent]
            CurrentPlayerStats := PlayerStatsManager.GetPlayerStats[Agent]
            CustomPlayer:Custom_Player = PlayersMap[PlayerObj]
        then:
            #Crit
            if(CurrentPlayerStats.CritScore.CurrentValue=1):
                CustomPlayer.SetCritBy(5)
            if(CurrentPlayerStats.CritScore.CurrentValue=2):
                CustomPlayer.SetCritBy(10)
            if(CurrentPlayerStats.CritScore.CurrentValue=3):
                CustomPlayer.SetCritBy(15)
            if(CurrentPlayerStats.CritScore.CurrentValue=4):
                CustomPlayer.SetCritBy(20)
            if(CurrentPlayerStats.CritScore.CurrentValue=5):
                CustomPlayer.SetCritBy(25)
            #Speed
            if(CurrentPlayerStats.SpeedScore.CurrentValue=1):
                CustomPlayer.SetSpeedBy(1.2)
            if(CurrentPlayerStats.SpeedScore.CurrentValue=2):
                CustomPlayer.SetSpeedBy(1.3)
            if(CurrentPlayerStats.SpeedScore.CurrentValue=3):
                CustomPlayer.SetSpeedBy(1.5)
            if(CurrentPlayerStats.SpeedScore.CurrentValue=4):
                CustomPlayer.SetSpeedBy(2.0)
            if(CurrentPlayerStats.SpeedScore.CurrentValue=5):
                CustomPlayer.SetSpeedBy(2.5)
            #PlayerSpeed
            if(CurrentPlayerStats.PlayerSpeedScore.CurrentValue=1): 
                ClassSelector1.ChangeClass(Agent)
                CustomPlayer.SetPlayerSpeedBy(2.0)
            if(CurrentPlayerStats.PlayerSpeedScore.CurrentValue=2):
                ClassSelector2.ChangeClass(Agent)
                CustomPlayer.SetPlayerSpeedBy(3.0)
            if(CurrentPlayerStats.PlayerSpeedScore.CurrentValue=3):
                ClassSelector3.ChangeClass(Agent)
                CustomPlayer.SetPlayerSpeedBy(4.0)
            if(CurrentPlayerStats.PlayerSpeedScore.CurrentValue=4):
                ClassSelector4.ChangeClass(Agent)
                CustomPlayer.SetPlayerSpeedBy(5.0)
            if(CurrentPlayerStats.PlayerSpeedScore.CurrentValue=5):
                ClassSelector5.ChangeClass(Agent)
                CustomPlayer.SetPlayerSpeedBy(6.0)
            #Luck
            if(CurrentPlayerStats.LuckScore.CurrentValue=1):
                CustomPlayer.SetLuckBy(1)
            if(CurrentPlayerStats.LuckScore.CurrentValue=2):
                CustomPlayer.SetLuckBy(2)
            if(CurrentPlayerStats.LuckScore.CurrentValue=3):
                CustomPlayer.SetLuckBy(3)
            if(CurrentPlayerStats.LuckScore.CurrentValue=4):
                CustomPlayer.SetLuckBy(5)
            if(CurrentPlayerStats.LuckScore.CurrentValue=5):
                CustomPlayer.SetLuckBy(7)
            #Gold
            CustomPlayer.SetGoldBy(CurrentPlayerStats.GoldBalance.CurrentValue)
            #Diamond
            CustomPlayer.SetDiamondBy(CurrentPlayerStats.DiamondBalance.CurrentValue)
            #Inventory
            CustomPlayer.SetLuck1By(CurrentPlayerStats.Ownership_Potion_Luck1.CurrentValue)
            CustomPlayer.SetLuck2By(CurrentPlayerStats.Ownership_Potion_Luck2.CurrentValue)
            CustomPlayer.SetLuck3By(CurrentPlayerStats.Ownership_Potion_Luck3.CurrentValue)
            CustomPlayer.SetGold1By(CurrentPlayerStats.Ownership_Potion_Gold1.CurrentValue)
            CustomPlayer.SetGold2By(CurrentPlayerStats.Ownership_Potion_Gold2.CurrentValue)
            CustomPlayer.SetGold3By(CurrentPlayerStats.Ownership_Potion_Gold3.CurrentValue)
            CustomPlayer.SetPetPower1By(CurrentPlayerStats.Ownership_Potion_PetPower1.CurrentValue)
            CustomPlayer.SetPetPower2By(CurrentPlayerStats.Ownership_Potion_PetPower2.CurrentValue)
            CustomPlayer.SetPetPower3By(CurrentPlayerStats.Ownership_Potion_PetPower3.CurrentValue)
            CustomPlayer.SetPlayerPower1By(CurrentPlayerStats.Ownership_Potion_PlayerPower1.CurrentValue)
            CustomPlayer.SetPlayerPower2By(CurrentPlayerStats.Ownership_Potion_PlayerPower2.CurrentValue)
            CustomPlayer.SetPlayerPower3By(CurrentPlayerStats.Ownership_Potion_PlayerPower3.CurrentValue)

            #Perma gold bonus
            if(CurrentPlayerStats.Code4Entered.CurrentValue=1):
                set CustomPlayer.Bonus_Gold_Perma=1.05

            #Code opening 
            if(CurrentPlayerStats.Code1Entered.CurrentValue=1):
                set CustomPlayer.Code1Opened=1
            if(CurrentPlayerStats.Code2Entered.CurrentValue=1):
                set CustomPlayer.Code2Opened=1
            if(CurrentPlayerStats.Code3Entered.CurrentValue=1):
                set CustomPlayer.Code3Opened=1
            if(CurrentPlayerStats.Code4Entered.CurrentValue=1):
                set CustomPlayer.Code4Opened=1

    <#
    InitializeBarriers(Agent:agent):void=
        if:
            PlayerObj := player[Agent]
            CurrentPlayerStats := PlayerStatsManager.GetPlayerStats[Agent]
        then:
            if(CurrentPlayerStats.LevelUnlocked.CurrentValue>=1):
                Print("Attempt to unlock B1")
                UnlockBarrierB1(Agent)
            if(CurrentPlayerStats.LevelUnlocked.CurrentValue>=2):
                Print("Attempt to unlock B2")
                UnlockBarrierB2(Agent)
            if(CurrentPlayerStats.LevelUnlocked.CurrentValue>=3):
                Print("Attempt to unlock B3")
                UnlockBarrierB3(Agent)
            if(CurrentPlayerStats.LevelUnlocked.CurrentValue>=4):
                Print("Attempt to unlock B4")
                UnlockBarrierB4(Agent)
            if(CurrentPlayerStats.LevelUnlocked.CurrentValue>=5):
                UnlockBarrierB5(Agent)
            if(CurrentPlayerStats.LevelUnlocked.CurrentValue>=6):
                UnlockBarrierB6(Agent)
            if(CurrentPlayerStats.LevelUnlocked.CurrentValue>=7):
                UnlockBarrierB7(Agent)
            if(CurrentPlayerStats.LevelUnlocked.CurrentValue>=8):
                UnlockBarrierB8(Agent)
            if(CurrentPlayerStats.LevelUnlocked.CurrentValue>=9):
                UnlockBarrierB9(Agent)
            if(CurrentPlayerStats.LevelUnlocked.CurrentValue>=10):
                UnlockBarrierB10(Agent)
            if(CurrentPlayerStats.LevelUnlocked.CurrentValue>=11):
                UnlockBarrierB11(Agent)
            if(CurrentPlayerStats.LevelUnlocked.CurrentValue>=12):
                UnlockBarrierB12(Agent)
            if(CurrentPlayerStats.LevelUnlocked.CurrentValue>=13):
                UnlockBarrierB13(Agent)
            if(CurrentPlayerStats.LevelUnlocked.CurrentValue>=14):
                UnlockBarrierB14(Agent)
            if(CurrentPlayerStats.LevelUnlocked.CurrentValue>=15):
                UnlockBarrierB15(Agent)
            if(CurrentPlayerStats.LevelUnlocked.CurrentValue>=16):
                UnlockBarrierB16(Agent)
            if(CurrentPlayerStats.LevelUnlocked.CurrentValue>=17):
                UnlockBarrierB17(Agent)
            if(CurrentPlayerStats.LevelUnlocked.CurrentValue>=18):
                UnlockBarrierB18(Agent)
            if(CurrentPlayerStats.LevelUnlocked.CurrentValue>=19):
                UnlockBarrierB19(Agent)
    #>

    InitializeRebirthBonus(Agent:agent):void=
        if:
            PlayerObj := player[Agent]
            CustomPlayer:Custom_Player = PlayersMap[PlayerObj]
            CurrentPlayerStats := PlayerStatsManager.GetPlayerStats[Agent]
        then:
            if(CurrentPlayerStats.RebirthUnlocked.CurrentValue=1):
                UnlockBarrierB31(Agent)
                CustomPlayer.SetBonusPowerMultiplicator(2.0)
                CustomPlayer.Update_Pet_Power()
                CustomPlayer.InitiateUI()
                if(HUDDevice:=Rebirth0_HUD[GetTeamIndexAgent[Agent]-1]):
                    HUDDevice.Hide(Agent)
                if(HUDDevice:=Rebirth1_HUD[GetTeamIndexAgent[Agent]-1]):
                    HUDDevice.Show(Agent)
            if(CurrentPlayerStats.RebirthUnlocked.CurrentValue=2):
                UnlockBarrierB32(Agent)
                CustomPlayer.SetBonusPowerMultiplicator(3.0)
                CustomPlayer.Update_Pet_Power()
                CustomPlayer.InitiateUI()
                if(HUDDevice:=Rebirth0_HUD[GetTeamIndexAgent[Agent]-1]):
                    HUDDevice.Hide(Agent)
                if(HUDDevice:=Rebirth2_HUD[GetTeamIndexAgent[Agent]-1]):
                    HUDDevice.Show(Agent)
            if(CurrentPlayerStats.RebirthUnlocked.CurrentValue=3):
                UnlockBarrierB33(Agent)
                CustomPlayer.SetBonusPowerMultiplicator(4.0)
                CustomPlayer.Update_Pet_Power()
                CustomPlayer.InitiateUI()
                if(HUDDevice:=Rebirth0_HUD[GetTeamIndexAgent[Agent]-1]):
                    HUDDevice.Hide(Agent)
                if(HUDDevice:=Rebirth3_HUD[GetTeamIndexAgent[Agent]-1]):
                    HUDDevice.Show(Agent)
            if(CurrentPlayerStats.RebirthUnlocked.CurrentValue=4):
                CustomPlayer.SetBonusPowerMultiplicator(5.0)
                CustomPlayer.Update_Pet_Power()
                CustomPlayer.InitiateUI()
                if(HUDDevice:=Rebirth0_HUD[GetTeamIndexAgent[Agent]-1]):
                    HUDDevice.Hide(Agent)
                if(HUDDevice:=Rebirth4_HUD[GetTeamIndexAgent[Agent]-1]):
                    HUDDevice.Show(Agent)
            if(CurrentPlayerStats.RebirthUnlocked.CurrentValue=5):
                CustomPlayer.SetBonusPowerMultiplicator(6.0)
                CustomPlayer.Update_Pet_Power()
                CustomPlayer.InitiateUI()
                if(HUDDevice:=Rebirth0_HUD[GetTeamIndexAgent[Agent]-1]):
                    HUDDevice.Hide(Agent)
                if(HUDDevice:=Rebirth5_HUD[GetTeamIndexAgent[Agent]-1]):
                    HUDDevice.Show(Agent)
        
    InitializeB1(Agent:agent)<suspends>:void=
        Sleep(0.5)
        if:
            PlayerObj := player[Agent]
            CurrentPlayerStats := PlayerStatsManager.GetPlayerStats[Agent]
        then:
            if(CurrentPlayerStats.LevelUnlocked.CurrentValue>=1):
                Print("Attempt to unlock B1")
                UnlockBarrierB1(Agent)

    InitializeB2(Agent:agent)<suspends>:void=
        Sleep(0.6)
        if:
            PlayerObj := player[Agent]
            CurrentPlayerStats := PlayerStatsManager.GetPlayerStats[Agent]
        then:
            if(CurrentPlayerStats.LevelUnlocked.CurrentValue>=2):
                Print("Attempt to unlock B2")
                UnlockBarrierB2(Agent)

    InitializeB3(Agent:agent)<suspends>:void=
        Sleep(0.7)
        if:
            PlayerObj := player[Agent]
            CurrentPlayerStats := PlayerStatsManager.GetPlayerStats[Agent]
        then:
            if(CurrentPlayerStats.LevelUnlocked.CurrentValue>=3):
                Print("Attempt to unlock B3")
                UnlockBarrierB3(Agent)
                            
    InitializeB4(Agent:agent)<suspends>:void=
        Sleep(0.8)
        if:
            PlayerObj := player[Agent]
            CurrentPlayerStats := PlayerStatsManager.GetPlayerStats[Agent]
        then:
            if(CurrentPlayerStats.LevelUnlocked.CurrentValue>=4):
                Print("Attempt to unlock B4")
                UnlockBarrierB4(Agent)

    InitializeB5(Agent:agent)<suspends>:void=
        Sleep(0.9)
        if:
            PlayerObj := player[Agent]
            CurrentPlayerStats := PlayerStatsManager.GetPlayerStats[Agent]
        then:
            if(CurrentPlayerStats.LevelUnlocked.CurrentValue>=5):
                Print("Attempt to unlock B5")
                UnlockBarrierB5(Agent)

    InitializeB6(Agent:agent)<suspends>:void=
        Sleep(1.0)
        if:
            PlayerObj := player[Agent]
            CurrentPlayerStats := PlayerStatsManager.GetPlayerStats[Agent]
        then:
            if(CurrentPlayerStats.LevelUnlocked.CurrentValue>=6):
                Print("Attempt to unlock B6")
                UnlockBarrierB6(Agent)

    InitializeB7(Agent:agent)<suspends>:void=
        Sleep(1.1)
        if:
            PlayerObj := player[Agent]
            CurrentPlayerStats := PlayerStatsManager.GetPlayerStats[Agent]
        then:
            if(CurrentPlayerStats.LevelUnlocked.CurrentValue>=7):
                Print("Attempt to unlock B7")
                UnlockBarrierB7(Agent)

    InitializeB8(Agent:agent)<suspends>:void=
        Sleep(1.2)
        if:
            PlayerObj := player[Agent]
            CurrentPlayerStats := PlayerStatsManager.GetPlayerStats[Agent]
        then:
            if(CurrentPlayerStats.LevelUnlocked.CurrentValue>=8):
                Print("Attempt to unlock B8")
                UnlockBarrierB8(Agent)

    InitializeB9(Agent:agent)<suspends>:void=
        Sleep(1.3)
        if:
            PlayerObj := player[Agent]
            CurrentPlayerStats := PlayerStatsManager.GetPlayerStats[Agent]
        then:
            if(CurrentPlayerStats.LevelUnlocked.CurrentValue>=9):
                Print("Attempt to unlock B9")
                UnlockBarrierB9(Agent)

    InitializeB10(Agent:agent)<suspends>:void=
        Sleep(1.4)
        if:
            PlayerObj := player[Agent]
            CurrentPlayerStats := PlayerStatsManager.GetPlayerStats[Agent]
        then:
            if(CurrentPlayerStats.LevelUnlocked.CurrentValue>=10):
                Print("Attempt to unlock B10")
                UnlockBarrierB10(Agent)

    InitializeB11(Agent:agent)<suspends>:void=
        Sleep(1.5)
        if:
            PlayerObj := player[Agent]
            CurrentPlayerStats := PlayerStatsManager.GetPlayerStats[Agent]
        then:
            if(CurrentPlayerStats.LevelUnlocked.CurrentValue>=11):
                Print("Attempt to unlock B11")
                UnlockBarrierB11(Agent)

    InitializeB12(Agent:agent)<suspends>:void=
        Sleep(1.6)
        if:
            PlayerObj := player[Agent]
            CurrentPlayerStats := PlayerStatsManager.GetPlayerStats[Agent]
        then:
            if(CurrentPlayerStats.LevelUnlocked.CurrentValue>=12):
                Print("Attempt to unlock B12")
                UnlockBarrierB12(Agent)

    InitializeB13(Agent:agent)<suspends>:void=
        Sleep(1.7)
        if:
            PlayerObj := player[Agent]
            CurrentPlayerStats := PlayerStatsManager.GetPlayerStats[Agent]
        then:
            if(CurrentPlayerStats.LevelUnlocked.CurrentValue>=13):
                Print("Attempt to unlock B13")
                UnlockBarrierB13(Agent)

    InitializeB14(Agent:agent)<suspends>:void=
        Sleep(1.8)
        if:
            PlayerObj := player[Agent]
            CurrentPlayerStats := PlayerStatsManager.GetPlayerStats[Agent]
        then:
            if(CurrentPlayerStats.LevelUnlocked.CurrentValue>=14):
                Print("Attempt to unlock B14")
                UnlockBarrierB14(Agent)

    InitializeB15(Agent:agent)<suspends>:void=
        Sleep(1.9)
        if:
            PlayerObj := player[Agent]
            CurrentPlayerStats := PlayerStatsManager.GetPlayerStats[Agent]
        then:
            if(CurrentPlayerStats.LevelUnlocked.CurrentValue>=15):
                Print("Attempt to unlock B15")
                UnlockBarrierB15(Agent)

    InitializeB16(Agent:agent)<suspends>:void=
        Sleep(2.0)
        if:
            PlayerObj := player[Agent]
            CurrentPlayerStats := PlayerStatsManager.GetPlayerStats[Agent]
        then:
            if(CurrentPlayerStats.LevelUnlocked.CurrentValue>=16):
                Print("Attempt to unlock B16")
                UnlockBarrierB16(Agent)

    InitializeB17(Agent:agent)<suspends>:void=
        Sleep(2.1)
        if:
            PlayerObj := player[Agent]
            CurrentPlayerStats := PlayerStatsManager.GetPlayerStats[Agent]
        then:
            if(CurrentPlayerStats.LevelUnlocked.CurrentValue>=17):
                Print("Attempt to unlock B17")
             UnlockBarrierB17(Agent)

    InitializeB18(Agent:agent)<suspends>:void=
        Sleep(2.2)
        if:
            PlayerObj := player[Agent]
            CurrentPlayerStats := PlayerStatsManager.GetPlayerStats[Agent]
        then:
            if(CurrentPlayerStats.LevelUnlocked.CurrentValue>=18):
                Print("Attempt to unlock B18")
                UnlockBarrierB18(Agent)

    InitializeB19(Agent:agent)<suspends>:void=
        Sleep(2.3)
        if:
            PlayerObj := player[Agent]
            CurrentPlayerStats := PlayerStatsManager.GetPlayerStats[Agent]
        then:
            if(CurrentPlayerStats.LevelUnlocked.CurrentValue>=19):
                Print("Attempt to unlock B19")
                UnlockBarrierB19(Agent)

    InitializeB20(Agent:agent)<suspends>:void=
        Sleep(2.4)
        if:
            PlayerObj := player[Agent]
            CurrentPlayerStats := PlayerStatsManager.GetPlayerStats[Agent]
        then:
            if(CurrentPlayerStats.LevelUnlocked.CurrentValue>=20):
                Print("Attempt to unlock B20")
                UnlockBarrierB20(Agent)

    InitializeB21(Agent:agent)<suspends>:void=
        Sleep(2.5)
        if:
            PlayerObj := player[Agent]
            CurrentPlayerStats := PlayerStatsManager.GetPlayerStats[Agent]
        then:
            if(CurrentPlayerStats.LevelUnlocked.CurrentValue>=21):
                Print("Attempt to unlock B21")
                UnlockBarrierB21(Agent)

    InitializeB22(Agent:agent)<suspends>:void=
        Sleep(2.6)
        if:
            PlayerObj := player[Agent]
            CurrentPlayerStats := PlayerStatsManager.GetPlayerStats[Agent]
        then:
            if(CurrentPlayerStats.LevelUnlocked.CurrentValue>=22):
                Print("Attempt to unlock B22")
                UnlockBarrierB22(Agent)

    InitializeB23(Agent:agent)<suspends>:void=
        Sleep(2.7)
        if:
            PlayerObj := player[Agent]
            CurrentPlayerStats := PlayerStatsManager.GetPlayerStats[Agent]
        then:
            if(CurrentPlayerStats.LevelUnlocked.CurrentValue>=23):
                Print("Attempt to unlock B23")
                UnlockBarrierB23(Agent)

    InitializeB24(Agent:agent)<suspends>:void=
        Sleep(2.8)
        if:
            PlayerObj := player[Agent]
            CurrentPlayerStats := PlayerStatsManager.GetPlayerStats[Agent]
        then:
            if(CurrentPlayerStats.LevelUnlocked.CurrentValue>=24):
                Print("Attempt to unlock B24")
                UnlockBarrierB24(Agent)

    InitializeB25(Agent:agent)<suspends>:void=
        Sleep(2.9)
        if:
            PlayerObj := player[Agent]
            CurrentPlayerStats := PlayerStatsManager.GetPlayerStats[Agent]
        then:
            if(CurrentPlayerStats.LevelUnlocked.CurrentValue>=25):
                Print("Attempt to unlock B25")
                UnlockBarrierB25(Agent)

    InitializeB26(Agent:agent)<suspends>:void=
        Sleep(3.0)
        if:
            PlayerObj := player[Agent]
            CurrentPlayerStats := PlayerStatsManager.GetPlayerStats[Agent]
        then:
            if(CurrentPlayerStats.LevelUnlocked.CurrentValue>=26):
                Print("Attempt to unlock B26")
                UnlockBarrierB26(Agent)

    InitializeB27(Agent:agent)<suspends>:void=
        Sleep(3.1)
        if:
            PlayerObj := player[Agent]
            CurrentPlayerStats := PlayerStatsManager.GetPlayerStats[Agent]
        then:
            if(CurrentPlayerStats.LevelUnlocked.CurrentValue>=27):
                Print("Attempt to unlock B27")
                UnlockBarrierB27(Agent)

    InitializeB28(Agent:agent)<suspends>:void=
        Sleep(3.2)
        if:
            PlayerObj := player[Agent]
            CurrentPlayerStats := PlayerStatsManager.GetPlayerStats[Agent]
        then:
            if(CurrentPlayerStats.LevelUnlocked.CurrentValue>=28):
                Print("Attempt to unlock B28")
                UnlockBarrierB28(Agent)

    InitializeB29(Agent:agent)<suspends>:void=
        Sleep(3.3)
        if:
            PlayerObj := player[Agent]
            CurrentPlayerStats := PlayerStatsManager.GetPlayerStats[Agent]
        then:
            if(CurrentPlayerStats.LevelUnlocked.CurrentValue>=29):
                Print("Attempt to unlock B29")
                UnlockBarrierB29(Agent)

    #NOT USED
    InitializeB30(Agent:agent)<suspends>:void=
        Sleep(3.4)
        if:
            PlayerObj := player[Agent]
            CurrentPlayerStats := PlayerStatsManager.GetPlayerStats[Agent]
        then:
            if(CurrentPlayerStats.LevelUnlocked.CurrentValue>=30):
                Print("Attempt to unlock B30")
                UnlockBarrierB30(Agent)

    InitializeB31(Agent:agent)<suspends>:void=
        Sleep(3.5)
        if:
            PlayerObj := player[Agent]
            CurrentPlayerStats := PlayerStatsManager.GetPlayerStats[Agent]
        then:
            if(CurrentPlayerStats.RebirthUnlocked.CurrentValue>=1):
                Print("Attempt to unlock B30")
                UnlockBarrierB31(Agent)

    InitializeB32(Agent:agent)<suspends>:void=
        Sleep(3.6)
        if:
            PlayerObj := player[Agent]
            CurrentPlayerStats := PlayerStatsManager.GetPlayerStats[Agent]
        then:
            if(CurrentPlayerStats.RebirthUnlocked.CurrentValue>=2):
                Print("Attempt to unlock B30")
                UnlockBarrierB32(Agent)

    InitializeB33(Agent:agent)<suspends>:void=
        Sleep(3.7)
        if:
            PlayerObj := player[Agent]
            CurrentPlayerStats := PlayerStatsManager.GetPlayerStats[Agent]
        then:
            if(CurrentPlayerStats.RebirthUnlocked.CurrentValue>=3):
                Print("Attempt to unlock B30")
                UnlockBarrierB33(Agent)

    # Adds to the given player's score and updates both their stats table in PlayerStatsManager 
    AddScore(Agent:agent):void=
        if:
            CurrentPlayerStats := PlayerStatsManager.GetPlayerStats[Agent]
            CurrentScore := CurrentPlayerStats.Score.CurrentValue
        then:
            PlayerStatsManager.RecordPlayerStat(Agent, StatType.Score, ?Score := CurrentScore + 1)

    DecreaseScore(Agent:agent):void=
        if:
            CurrentPlayerStats := PlayerStatsManager.GetPlayerStats[Agent]
            CurrentScore := CurrentPlayerStats.Score.CurrentValue
        then:
            PlayerStatsManager.RecordPlayerStat(Agent, StatType.Score, ?Score := CurrentScore - 1)

    ResetScore(Agent:agent):void=
        if:
            CurrentPlayerStats := PlayerStatsManager.GetPlayerStats[Agent]
            CurrentScore := CurrentPlayerStats.Score.CurrentValue
        then:
            PlayerStatsManager.RecordPlayerStat(Agent, StatType.Score, ?Score := 0)

    #Max Level Unlocked
    AddMaxLevelUnlocked(Agent:agent):void=
        if:
            CurrentPlayerStats := PlayerStatsManager.GetPlayerStats[Agent]
            CurrentLevel := CurrentPlayerStats.LevelUnlocked.CurrentValue
        then:
            PlayerStatsManager.RecordMaxLevelUnlocked(Agent, StatType.LevelUnlocked, ?LevelUnlocked := CurrentLevel + 1)

    #Max Level Unlocked Reset
    MaxLevelReset(Agent:agent):void=
        if:
            CurrentPlayerStats := PlayerStatsManager.GetPlayerStats[Agent]
            CurrentLevel := CurrentPlayerStats.LevelUnlocked.CurrentValue
        then:
            PlayerStatsManager.RecordMaxLevelUnlocked(Agent, StatType.LevelUnlocked, ?LevelUnlocked := 0)

    #Max Rebirth Unlocked
    AddMaxRebirthUnlocked(Agent:agent):void=
        if:
            CurrentPlayerStats := PlayerStatsManager.GetPlayerStats[Agent]
            CurrentRebirth := CurrentPlayerStats.RebirthUnlocked.CurrentValue
        then:
            PlayerStatsManager.RecordMaxRebirthUnlocked(Agent, StatType.RebirthUnlocked, ?RebirthUnlocked := CurrentRebirth + 1)
    
    #Gold
    AddGoldBalance(Agent:agent, NewGold:int):void=
        if:
            CurrentPlayerStats := PlayerStatsManager.GetPlayerStats[Agent]
            CurrentGold := CurrentPlayerStats.GoldBalance.CurrentValue
        then:
            PlayerStatsManager.RecordGold(Agent, StatType.GoldBalance, ?GoldBalance := CurrentGold + NewGold)

    DecreaseGoldBalance(Agent:agent, NewGold:int):void=
        if:
            CurrentPlayerStats := PlayerStatsManager.GetPlayerStats[Agent]
            CurrentGold := CurrentPlayerStats.GoldBalance.CurrentValue
        then:
            PlayerStatsManager.RecordGold(Agent, StatType.GoldBalance, ?GoldBalance := CurrentGold - NewGold)

    #Diamond
    AddDiamondBalance(Agent:agent, NewDiamond:int):void=
        if:
            CurrentPlayerStats := PlayerStatsManager.GetPlayerStats[Agent]
            CurrentDiamond := CurrentPlayerStats.DiamondBalance.CurrentValue
        then:
            PlayerStatsManager.RecordDiamond(Agent, StatType.DiamondBalance, ?DiamondBalance := CurrentDiamond + NewDiamond)

    DecreaseDiamondBalance(Agent:agent, NewDiamond:int):void=
        if:
            CurrentPlayerStats := PlayerStatsManager.GetPlayerStats[Agent]
            CurrentDiamond := CurrentPlayerStats.DiamondBalance.CurrentValue
        then:
            PlayerStatsManager.RecordDiamond(Agent, StatType.DiamondBalance, ?DiamondBalance := CurrentDiamond - NewDiamond)
            
    #Crit
    SetCrit(Agent:agent, NewCrit:int):void=
        if:
            CurrentPlayerStats := PlayerStatsManager.GetPlayerStats[Agent]
            FinalCrit := NewCrit
        then:
            PlayerStatsManager.RecordCrit(Agent, StatType.CritScore, ?CritScore := FinalCrit)

    #Speed
    SetSpeed(Agent:agent, NewSpeed:int):void=
        if:
            CurrentPlayerStats := PlayerStatsManager.GetPlayerStats[Agent]
            FinalSpeed := NewSpeed
        then:
            PlayerStatsManager.RecordSpeed(Agent, StatType.SpeedScore, ?SpeedScore := FinalSpeed)

    #PlayerSpeed
    SetPlayerSpeed(Agent:agent, NewPlayerSpeed:int):void=
        if:
            CurrentPlayerStats := PlayerStatsManager.GetPlayerStats[Agent]
            FinalPlayerSpeed := NewPlayerSpeed
        then:
            PlayerStatsManager.RecordPlayerSpeed(Agent, StatType.PlayerSpeedScore, ?PlayerSpeedScore := FinalPlayerSpeed)

    #Luck
    SetLuck(Agent:agent, NewLuck:int):void=
        if:
            CurrentPlayerStats := PlayerStatsManager.GetPlayerStats[Agent]
            FinalLuck := NewLuck
        then:
            PlayerStatsManager.RecordLuck(Agent, StatType.LuckScore, ?LuckScore := FinalLuck)

    #Record Code 1 entered
    SetRecordCode1(Agent:agent, NewRecord:int):void=
        if:
            CurrentPlayerStats := PlayerStatsManager.GetPlayerStats[Agent]
            Record := NewRecord
        then:
            PlayerStatsManager.RecordCode1Entered(Agent, StatType.Code1Entered, ?Code1Entered := Record)

    #Record Code 2 entered
    SetRecordCode2(Agent:agent, NewRecord:int):void=
        if:
            CurrentPlayerStats := PlayerStatsManager.GetPlayerStats[Agent]
            Record := NewRecord
        then:
            PlayerStatsManager.RecordCode2Entered(Agent, StatType.Code2Entered, ?Code2Entered := Record)

    #Record Code 3 entered
    SetRecordCode3(Agent:agent, NewRecord:int):void=
        if:
            CurrentPlayerStats := PlayerStatsManager.GetPlayerStats[Agent]
            Record := NewRecord
        then:
            PlayerStatsManager.RecordCode3Entered(Agent, StatType.Code3Entered, ?Code3Entered := Record)

    #Record Code 4 entered
    SetRecordCode4(Agent:agent, NewRecord:int):void=
        if:
            CurrentPlayerStats := PlayerStatsManager.GetPlayerStats[Agent]
            Record := NewRecord
        then:
            PlayerStatsManager.RecordCode4Entered(Agent, StatType.Code4Entered, ?Code4Entered := Record)
    
    #Potions

    #Luck1
    AddPotionLuck1(Agent:agent, NewPotion:int):void=
        if:
            CurrentPlayerStats := PlayerStatsManager.GetPlayerStats[Agent]
            CurrentPotion := CurrentPlayerStats.Ownership_Potion_Luck1.CurrentValue
        then:
            PlayerStatsManager.RecordOwnershipPotionLuck1(Agent, StatType.Ownership_Potion_Luck1, ?Ownership_Potion_Luck1 := CurrentPotion + NewPotion)

    DecreasePotionLuck1(Agent:agent, NewPotion:int):void=
        if:
            CurrentPlayerStats := PlayerStatsManager.GetPlayerStats[Agent]
            CurrentPotion := CurrentPlayerStats.Ownership_Potion_Luck1.CurrentValue
        then:
            PlayerStatsManager.RecordOwnershipPotionLuck1(Agent, StatType.Ownership_Potion_Luck1, ?Ownership_Potion_Luck1 := CurrentPotion - NewPotion)
    #Luck2
    AddPotionLuck2(Agent:agent, NewPotion:int):void=
        if:
            CurrentPlayerStats := PlayerStatsManager.GetPlayerStats[Agent]
            CurrentPotion := CurrentPlayerStats.Ownership_Potion_Luck2.CurrentValue
        then:
            PlayerStatsManager.RecordOwnershipPotionLuck2(Agent, StatType.Ownership_Potion_Luck2, ?Ownership_Potion_Luck2 := CurrentPotion + NewPotion)

    DecreasePotionLuck2(Agent:agent, NewPotion:int):void=
        if:
            CurrentPlayerStats := PlayerStatsManager.GetPlayerStats[Agent]
            CurrentPotion := CurrentPlayerStats.Ownership_Potion_Luck2.CurrentValue
        then:
            PlayerStatsManager.RecordOwnershipPotionLuck2(Agent, StatType.Ownership_Potion_Luck2, ?Ownership_Potion_Luck2 := CurrentPotion - NewPotion)
    #Luck3
    AddPotionLuck3(Agent:agent, NewPotion:int):void=
        if:
            CurrentPlayerStats := PlayerStatsManager.GetPlayerStats[Agent]
            CurrentPotion := CurrentPlayerStats.Ownership_Potion_Luck3.CurrentValue
        then:
            PlayerStatsManager.RecordOwnershipPotionLuck3(Agent, StatType.Ownership_Potion_Luck3, ?Ownership_Potion_Luck3 := CurrentPotion + NewPotion)

    DecreasePotionLuck3(Agent:agent, NewPotion:int):void=
        if:
            CurrentPlayerStats := PlayerStatsManager.GetPlayerStats[Agent]
            CurrentPotion := CurrentPlayerStats.Ownership_Potion_Luck3.CurrentValue
        then:
            PlayerStatsManager.RecordOwnershipPotionLuck3(Agent, StatType.Ownership_Potion_Luck3, ?Ownership_Potion_Luck3 := CurrentPotion - NewPotion)

    #Gold1
    AddPotionGold1(Agent:agent, NewPotion:int):void=
        if:
            CurrentPlayerStats := PlayerStatsManager.GetPlayerStats[Agent]
            CurrentPotion := CurrentPlayerStats.Ownership_Potion_Gold1.CurrentValue
        then:
            PlayerStatsManager.RecordOwnershipPotionGold1(Agent, StatType.Ownership_Potion_Gold1, ?Ownership_Potion_Gold1 := CurrentPotion + NewPotion)

    DecreasePotionGold1(Agent:agent, NewPotion:int):void=
        if:
            CurrentPlayerStats := PlayerStatsManager.GetPlayerStats[Agent]
            CurrentPotion := CurrentPlayerStats.Ownership_Potion_Gold1.CurrentValue
        then:
            PlayerStatsManager.RecordOwnershipPotionGold1(Agent, StatType.Ownership_Potion_Gold1, ?Ownership_Potion_Gold1 := CurrentPotion - NewPotion)
    #Gold2
    AddPotionGold2(Agent:agent, NewPotion:int):void=
        if:
            CurrentPlayerStats := PlayerStatsManager.GetPlayerStats[Agent]
            CurrentPotion := CurrentPlayerStats.Ownership_Potion_Gold2.CurrentValue
        then:
            PlayerStatsManager.RecordOwnershipPotionGold2(Agent, StatType.Ownership_Potion_Gold2, ?Ownership_Potion_Gold2 := CurrentPotion + NewPotion)

    DecreasePotionGold2(Agent:agent, NewPotion:int):void=
        if:
            CurrentPlayerStats := PlayerStatsManager.GetPlayerStats[Agent]
            CurrentPotion := CurrentPlayerStats.Ownership_Potion_Gold2.CurrentValue
        then:
            PlayerStatsManager.RecordOwnershipPotionGold2(Agent, StatType.Ownership_Potion_Gold2, ?Ownership_Potion_Gold2 := CurrentPotion - NewPotion)
    #Gold3
    AddPotionGold3(Agent:agent, NewPotion:int):void=
        if:
            CurrentPlayerStats := PlayerStatsManager.GetPlayerStats[Agent]
            CurrentPotion := CurrentPlayerStats.Ownership_Potion_Gold3.CurrentValue
        then:
            PlayerStatsManager.RecordOwnershipPotionGold3(Agent, StatType.Ownership_Potion_Gold3, ?Ownership_Potion_Gold3 := CurrentPotion + NewPotion)

    DecreasePotionGold3(Agent:agent, NewPotion:int):void=
        if:
            CurrentPlayerStats := PlayerStatsManager.GetPlayerStats[Agent]
            CurrentPotion := CurrentPlayerStats.Ownership_Potion_Gold3.CurrentValue
        then:
            PlayerStatsManager.RecordOwnershipPotionGold3(Agent, StatType.Ownership_Potion_Gold3, ?Ownership_Potion_Gold3 := CurrentPotion - NewPotion)

    #PetPower1
    AddPotionPetPower1(Agent:agent, NewPotion:int):void=
        if:
            CurrentPlayerStats := PlayerStatsManager.GetPlayerStats[Agent]
            CurrentPotion := CurrentPlayerStats.Ownership_Potion_PetPower1.CurrentValue
        then:
            PlayerStatsManager.RecordOwnershipPotionPetPower1(Agent, StatType.Ownership_Potion_PetPower1, ?Ownership_Potion_PetPower1 := CurrentPotion + NewPotion)

    DecreasePotionPetPower1(Agent:agent, NewPotion:int):void=
        if:
            CurrentPlayerStats := PlayerStatsManager.GetPlayerStats[Agent]
            CurrentPotion := CurrentPlayerStats.Ownership_Potion_PetPower1.CurrentValue
        then:
            PlayerStatsManager.RecordOwnershipPotionPetPower1(Agent, StatType.Ownership_Potion_PetPower1, ?Ownership_Potion_PetPower1 := CurrentPotion - NewPotion)
    #PetPower2
    AddPotionPetPower2(Agent:agent, NewPotion:int):void=
        if:
            CurrentPlayerStats := PlayerStatsManager.GetPlayerStats[Agent]
            CurrentPotion := CurrentPlayerStats.Ownership_Potion_PetPower2.CurrentValue
        then:
            PlayerStatsManager.RecordOwnershipPotionPetPower2(Agent, StatType.Ownership_Potion_PetPower2, ?Ownership_Potion_PetPower2 := CurrentPotion + NewPotion)

    DecreasePotionPetPower2(Agent:agent, NewPotion:int):void=
        if:
            CurrentPlayerStats := PlayerStatsManager.GetPlayerStats[Agent]
            CurrentPotion := CurrentPlayerStats.Ownership_Potion_PetPower2.CurrentValue
        then:
            PlayerStatsManager.RecordOwnershipPotionPetPower2(Agent, StatType.Ownership_Potion_PetPower2, ?Ownership_Potion_PetPower2 := CurrentPotion - NewPotion)
    #PetPower3
    AddPotionPetPower3(Agent:agent, NewPotion:int):void=
        if:
            CurrentPlayerStats := PlayerStatsManager.GetPlayerStats[Agent]
            CurrentPotion := CurrentPlayerStats.Ownership_Potion_PetPower3.CurrentValue
        then:
            PlayerStatsManager.RecordOwnershipPotionPetPower3(Agent, StatType.Ownership_Potion_PetPower3, ?Ownership_Potion_PetPower3 := CurrentPotion + NewPotion)

    DecreasePotionPetPower3(Agent:agent, NewPotion:int):void=
        if:
            CurrentPlayerStats := PlayerStatsManager.GetPlayerStats[Agent]
            CurrentPotion := CurrentPlayerStats.Ownership_Potion_PetPower3.CurrentValue
        then:
            PlayerStatsManager.RecordOwnershipPotionPetPower3(Agent, StatType.Ownership_Potion_PetPower3, ?Ownership_Potion_PetPower3 := CurrentPotion - NewPotion)

    #PlayerPower1
    AddPotionPlayerPower1(Agent:agent, NewPotion:int):void=
        if:
            CurrentPlayerStats := PlayerStatsManager.GetPlayerStats[Agent]
            CurrentPotion := CurrentPlayerStats.Ownership_Potion_PlayerPower1.CurrentValue
        then:
            PlayerStatsManager.RecordOwnershipPotionPlayerPower1(Agent, StatType.Ownership_Potion_PlayerPower1, ?Ownership_Potion_PlayerPower1 := CurrentPotion + NewPotion)

    DecreasePotionPlayerPower1(Agent:agent, NewPotion:int):void=
        if:
            CurrentPlayerStats := PlayerStatsManager.GetPlayerStats[Agent]
            CurrentPotion := CurrentPlayerStats.Ownership_Potion_PlayerPower1.CurrentValue
        then:
            PlayerStatsManager.RecordOwnershipPotionPlayerPower1(Agent, StatType.Ownership_Potion_PlayerPower1, ?Ownership_Potion_PlayerPower1 := CurrentPotion - NewPotion)
    #PlayerPower2
    AddPotionPlayerPower2(Agent:agent, NewPotion:int):void=
        if:
            CurrentPlayerStats := PlayerStatsManager.GetPlayerStats[Agent]
            CurrentPotion := CurrentPlayerStats.Ownership_Potion_PlayerPower2.CurrentValue
        then:
            PlayerStatsManager.RecordOwnershipPotionPlayerPower2(Agent, StatType.Ownership_Potion_PlayerPower2, ?Ownership_Potion_PlayerPower2 := CurrentPotion + NewPotion)

    DecreasePotionPlayerPower2(Agent:agent, NewPotion:int):void=
        if:
            CurrentPlayerStats := PlayerStatsManager.GetPlayerStats[Agent]
            CurrentPotion := CurrentPlayerStats.Ownership_Potion_PlayerPower2.CurrentValue
        then:
            PlayerStatsManager.RecordOwnershipPotionPlayerPower2(Agent, StatType.Ownership_Potion_PlayerPower2, ?Ownership_Potion_PlayerPower2 := CurrentPotion - NewPotion)
    #PlayerPower3
    AddPotionPlayerPower3(Agent:agent, NewPotion:int):void=
        if:
            CurrentPlayerStats := PlayerStatsManager.GetPlayerStats[Agent]
            CurrentPotion := CurrentPlayerStats.Ownership_Potion_PlayerPower3.CurrentValue
        then:
            PlayerStatsManager.RecordOwnershipPotionPlayerPower3(Agent, StatType.Ownership_Potion_PlayerPower3, ?Ownership_Potion_PlayerPower3 := CurrentPotion + NewPotion)

    DecreasePotionPlayerPower3(Agent:agent, NewPotion:int):void=
        if:
            CurrentPlayerStats := PlayerStatsManager.GetPlayerStats[Agent]
            CurrentPotion := CurrentPlayerStats.Ownership_Potion_PlayerPower3.CurrentValue
        then:
            PlayerStatsManager.RecordOwnershipPotionPlayerPower3(Agent, StatType.Ownership_Potion_PlayerPower3, ?Ownership_Potion_PlayerPower3 := CurrentPotion - NewPotion)
    #---------------------------------------L1
    #Switch Ownership of L1_Pet1 to 1
    L1_Pet1_Own(Agent:agent):void=
        Print("Ownership L1P1 Saved")
        if(CurrentPlayerStats := PlayerStatsManager.GetPlayerStats[Agent]):
            PlayerStatsManager.RecordOwnershipL1P1(Agent, StatType.Ownership_L1_Pet1, ?Ownership_L1_Pet1 := 1)
            

    #Switch Ownership of L1_Pet1 to 0
    L1_Pet1_DesOwn(Agent:agent):void=
        Print("DesOwnership L1P1 Saved")
        if(CurrentPlayerStats := PlayerStatsManager.GetPlayerStats[Agent]):
            PlayerStatsManager.RemoveOwnershipL1P1(Agent, StatType.Ownership_L1_Pet1, ?Ownership_L1_Pet1 := 0)
            

    #Switch Ownership of L1_Pet2 to 1
    L1_Pet2_Own(Agent:agent):void=
        Print("Ownership L1P2 Saved")
        if(CurrentPlayerStats := PlayerStatsManager.GetPlayerStats[Agent]):
            PlayerStatsManager.RecordOwnershipL1P2(Agent, StatType.Ownership_L1_Pet2, ?Ownership_L1_Pet2 := 1)
            

    #Switch Ownership of L1_Pet2 to 0
    L1_Pet2_DesOwn(Agent:agent):void=
        Print("DesOwnership L1P2 Saved")
        if(CurrentPlayerStats := PlayerStatsManager.GetPlayerStats[Agent]):
            PlayerStatsManager.RemoveOwnershipL1P2(Agent, StatType.Ownership_L1_Pet2, ?Ownership_L1_Pet2 := 0)
            

    #Switch Ownership of L1_Pet3 to 1
    L1_Pet3_Own(Agent:agent):void=
        Print("Ownership L1P3 Saved")
        if(CurrentPlayerStats := PlayerStatsManager.GetPlayerStats[Agent]):
            PlayerStatsManager.RecordOwnershipL1P3(Agent, StatType.Ownership_L1_Pet3, ?Ownership_L1_Pet3 := 1)
            

    #Switch Ownership of L1_Pet3 to 0
    L1_Pet3_DesOwn(Agent:agent):void=
        Print("DesOwnership L1P3 Saved")
        if(CurrentPlayerStats := PlayerStatsManager.GetPlayerStats[Agent]):
            PlayerStatsManager.RemoveOwnershipL1P3(Agent, StatType.Ownership_L1_Pet3, ?Ownership_L1_Pet3 := 0)
            

    #Switch Ownership of L1_Pet4 to 1
    L1_Pet4_Own(Agent:agent):void=
        Print("Ownership L1P4 Saved")
        if(CurrentPlayerStats := PlayerStatsManager.GetPlayerStats[Agent]):
            PlayerStatsManager.RecordOwnershipL1P4(Agent, StatType.Ownership_L1_Pet4, ?Ownership_L1_Pet4 := 1)
            

    #Switch Ownership of L1_Pet4 to 0
    L1_Pet4_DesOwn(Agent:agent):void=
        Print("DesOwnership L1P4 Saved")
        if(CurrentPlayerStats := PlayerStatsManager.GetPlayerStats[Agent]):
            PlayerStatsManager.RemoveOwnershipL1P4(Agent, StatType.Ownership_L1_Pet4, ?Ownership_L1_Pet4 := 0)
            
    #---------------------------------------L2
    #Switch Ownership of L2_Pet1 to 1
    L2_Pet1_Own(Agent:agent):void=
        Print("Ownership L2P1 Saved")
        if(CurrentPlayerStats := PlayerStatsManager.GetPlayerStats[Agent]):
            PlayerStatsManager.RecordOwnershipL2P1(Agent, StatType.Ownership_L2_Pet1, ?Ownership_L2_Pet1 := 1)
            

    #Switch Ownership of L2_Pet1 to 0
    L2_Pet1_DesOwn(Agent:agent):void=
        Print("DesOwnership L2P1 Saved")
        if(CurrentPlayerStats := PlayerStatsManager.GetPlayerStats[Agent]):
            PlayerStatsManager.RemoveOwnershipL2P1(Agent, StatType.Ownership_L2_Pet1, ?Ownership_L2_Pet1 := 0)
            

    #Switch Ownership of L2_Pet2 to 1
    L2_Pet2_Own(Agent:agent):void=
        Print("Ownership L2P2 Saved")
        if(CurrentPlayerStats := PlayerStatsManager.GetPlayerStats[Agent]):
            PlayerStatsManager.RecordOwnershipL2P2(Agent, StatType.Ownership_L2_Pet2, ?Ownership_L2_Pet2 := 1)
            

    #Switch Ownership of L2_Pet2 to 0
    L2_Pet2_DesOwn(Agent:agent):void=
        Print("DesOwnership L2P2 Saved")
        if(CurrentPlayerStats := PlayerStatsManager.GetPlayerStats[Agent]):
            PlayerStatsManager.RemoveOwnershipL2P2(Agent, StatType.Ownership_L2_Pet2, ?Ownership_L2_Pet2 := 0)
            

    #Switch Ownership of L2_Pet3 to 1
    L2_Pet3_Own(Agent:agent):void=
        Print("Ownership L2P3 Saved")
        if(CurrentPlayerStats := PlayerStatsManager.GetPlayerStats[Agent]):
            PlayerStatsManager.RecordOwnershipL2P3(Agent, StatType.Ownership_L2_Pet3, ?Ownership_L2_Pet3 := 1)
            

    #Switch Ownership of L2_Pet3 to 0
    L2_Pet3_DesOwn(Agent:agent):void=
        Print("DesOwnership L2P3 Saved")
        if(CurrentPlayerStats := PlayerStatsManager.GetPlayerStats[Agent]):
            PlayerStatsManager.RemoveOwnershipL2P3(Agent, StatType.Ownership_L2_Pet3, ?Ownership_L2_Pet3 := 0)
            

    #Switch Ownership of L2_Pet4 to 1
    L2_Pet4_Own(Agent:agent):void=
        Print("Ownership L2P4 Saved")
        if(CurrentPlayerStats := PlayerStatsManager.GetPlayerStats[Agent]):
            PlayerStatsManager.RecordOwnershipL2P4(Agent, StatType.Ownership_L2_Pet4, ?Ownership_L2_Pet4 := 1)
            

    #Switch Ownership of L2_Pet4 to 0
    L2_Pet4_DesOwn(Agent:agent):void=
        Print("DesOwnership L2P4 Saved")
        if(CurrentPlayerStats := PlayerStatsManager.GetPlayerStats[Agent]):
            PlayerStatsManager.RemoveOwnershipL2P4(Agent, StatType.Ownership_L2_Pet4, ?Ownership_L2_Pet4 := 0)    
    
    #---------------------------------------L3
    #Switch Ownership of L3_Pet1 to 1
    L3_Pet1_Own(Agent:agent):void=
        Print("Ownership L3P1 Saved")
        if(CurrentPlayerStats := PlayerStatsManager.GetPlayerStats[Agent]):
            PlayerStatsManager.RecordOwnershipL3P1(Agent, StatType.Ownership_L3_Pet1, ?Ownership_L3_Pet1 := 1)
            

    #Switch Ownership of L3_Pet1 to 0
    L3_Pet1_DesOwn(Agent:agent):void=
        Print("DesOwnership L3P1 Saved")
        if(CurrentPlayerStats := PlayerStatsManager.GetPlayerStats[Agent]):
            PlayerStatsManager.RemoveOwnershipL3P1(Agent, StatType.Ownership_L3_Pet1, ?Ownership_L3_Pet1 := 0)
            

    #Switch Ownership of L3_Pet2 to 1
    L3_Pet2_Own(Agent:agent):void=
        Print("Ownership L3P2 Saved")
        if(CurrentPlayerStats := PlayerStatsManager.GetPlayerStats[Agent]):
            PlayerStatsManager.RecordOwnershipL3P2(Agent, StatType.Ownership_L3_Pet2, ?Ownership_L3_Pet2 := 1)
            

    #Switch Ownership of L3_Pet2 to 0
    L3_Pet2_DesOwn(Agent:agent):void=
        Print("DesOwnership L3P2 Saved")
        if(CurrentPlayerStats := PlayerStatsManager.GetPlayerStats[Agent]):
            PlayerStatsManager.RemoveOwnershipL3P2(Agent, StatType.Ownership_L3_Pet2, ?Ownership_L3_Pet2 := 0)
            

    #Switch Ownership of L3_Pet3 to 1
    L3_Pet3_Own(Agent:agent):void=
        Print("Ownership L3P3 Saved")
        if(CurrentPlayerStats := PlayerStatsManager.GetPlayerStats[Agent]):
            PlayerStatsManager.RecordOwnershipL3P3(Agent, StatType.Ownership_L3_Pet3, ?Ownership_L3_Pet3 := 1)
            

    #Switch Ownership of L3_Pet3 to 0
    L3_Pet3_DesOwn(Agent:agent):void=
        Print("DesOwnership L3P3 Saved")
        if(CurrentPlayerStats := PlayerStatsManager.GetPlayerStats[Agent]):
            PlayerStatsManager.RemoveOwnershipL3P3(Agent, StatType.Ownership_L3_Pet3, ?Ownership_L3_Pet3 := 0)
            

    #Switch Ownership of L3_Pet4 to 1
    L3_Pet4_Own(Agent:agent):void=
        Print("Ownership L3P4 Saved")
        if(CurrentPlayerStats := PlayerStatsManager.GetPlayerStats[Agent]):
            PlayerStatsManager.RecordOwnershipL3P4(Agent, StatType.Ownership_L3_Pet4, ?Ownership_L3_Pet4 := 1)
            

    #Switch Ownership of L3_Pet4 to 0
    L3_Pet4_DesOwn(Agent:agent):void=
        Print("DesOwnership L3P4 Saved")
        if(CurrentPlayerStats := PlayerStatsManager.GetPlayerStats[Agent]):
            PlayerStatsManager.RemoveOwnershipL3P4(Agent, StatType.Ownership_L3_Pet4, ?Ownership_L3_Pet4 := 0) 

    #---------------------------------------L4
    #Switch Ownership of L4_Pet1 to 1
    L4_Pet1_Own(Agent:agent):void=
        Print("Ownership L4P1 Saved")
        if(CurrentPlayerStats := PlayerStatsManager.GetPlayerStats[Agent]):
            PlayerStatsManager.RecordOwnershipL4P1(Agent, StatType.Ownership_L4_Pet1, ?Ownership_L4_Pet1 := 1)
            

    #Switch Ownership of L4_Pet1 to 0
    L4_Pet1_DesOwn(Agent:agent):void=
        Print("DesOwnership L4P1 Saved")
        if(CurrentPlayerStats := PlayerStatsManager.GetPlayerStats[Agent]):
            PlayerStatsManager.RemoveOwnershipL4P1(Agent, StatType.Ownership_L4_Pet1, ?Ownership_L4_Pet1 := 0)
            

    #Switch Ownership of L4_Pet2 to 1
    L4_Pet2_Own(Agent:agent):void=
        Print("Ownership L4P2 Saved")
        if(CurrentPlayerStats := PlayerStatsManager.GetPlayerStats[Agent]):
            PlayerStatsManager.RecordOwnershipL4P2(Agent, StatType.Ownership_L4_Pet2, ?Ownership_L4_Pet2 := 1)
            

    #Switch Ownership of L4_Pet2 to 0
    L4_Pet2_DesOwn(Agent:agent):void=
        Print("DesOwnership L4P2 Saved")
        if(CurrentPlayerStats := PlayerStatsManager.GetPlayerStats[Agent]):
            PlayerStatsManager.RemoveOwnershipL4P2(Agent, StatType.Ownership_L4_Pet2, ?Ownership_L4_Pet2 := 0)
            

    #Switch Ownership of L4_Pet3 to 1
    L4_Pet3_Own(Agent:agent):void=
        Print("Ownership L4P3 Saved")
        if(CurrentPlayerStats := PlayerStatsManager.GetPlayerStats[Agent]):
            PlayerStatsManager.RecordOwnershipL4P3(Agent, StatType.Ownership_L4_Pet3, ?Ownership_L4_Pet3 := 1)
            

    #Switch Ownership of L4_Pet3 to 0
    L4_Pet3_DesOwn(Agent:agent):void=
        Print("DesOwnership L4P3 Saved")
        if(CurrentPlayerStats := PlayerStatsManager.GetPlayerStats[Agent]):
            PlayerStatsManager.RemoveOwnershipL4P3(Agent, StatType.Ownership_L4_Pet3, ?Ownership_L4_Pet3 := 0)
            

    #Switch Ownership of L4_Pet4 to 1
    L4_Pet4_Own(Agent:agent):void=
        Print("Ownership L4P4 Saved")
        if(CurrentPlayerStats := PlayerStatsManager.GetPlayerStats[Agent]):
            PlayerStatsManager.RecordOwnershipL4P4(Agent, StatType.Ownership_L4_Pet4, ?Ownership_L4_Pet4 := 1)
            

    #Switch Ownership of L4_Pet4 to 0
    L4_Pet4_DesOwn(Agent:agent):void=
        Print("DesOwnership L4P4 Saved")
        if(CurrentPlayerStats := PlayerStatsManager.GetPlayerStats[Agent]):
            PlayerStatsManager.RemoveOwnershipL4P4(Agent, StatType.Ownership_L4_Pet4, ?Ownership_L4_Pet4 := 0) 

    #---------------------------------------L5
    #Switch Ownership of L5_Pet1 to 1
    L5_Pet1_Own(Agent:agent):void=
        Print("Ownership L5P1 Saved")
        if(CurrentPlayerStats := PlayerStatsManager.GetPlayerStats[Agent]):
            PlayerStatsManager.RecordOwnershipL5P1(Agent, StatType.Ownership_L5_Pet1, ?Ownership_L5_Pet1 := 1)
            

    #Switch Ownership of L5_Pet1 to 0
    L5_Pet1_DesOwn(Agent:agent):void=
        Print("DesOwnership L5P1 Saved")
        if(CurrentPlayerStats := PlayerStatsManager.GetPlayerStats[Agent]):
            PlayerStatsManager.RemoveOwnershipL5P1(Agent, StatType.Ownership_L5_Pet1, ?Ownership_L5_Pet1 := 0)
            

    #Switch Ownership of L5_Pet2 to 1
    L5_Pet2_Own(Agent:agent):void=
        Print("Ownership L5P2 Saved")
        if(CurrentPlayerStats := PlayerStatsManager.GetPlayerStats[Agent]):
            PlayerStatsManager.RecordOwnershipL5P2(Agent, StatType.Ownership_L5_Pet2, ?Ownership_L5_Pet2 := 1)
            

    #Switch Ownership of L5_Pet2 to 0
    L5_Pet2_DesOwn(Agent:agent):void=
        Print("DesOwnership L5P2 Saved")
        if(CurrentPlayerStats := PlayerStatsManager.GetPlayerStats[Agent]):
            PlayerStatsManager.RemoveOwnershipL5P2(Agent, StatType.Ownership_L5_Pet2, ?Ownership_L5_Pet2 := 0)
            

    #Switch Ownership of L5_Pet3 to 1
    L5_Pet3_Own(Agent:agent):void=
        Print("Ownership L5P3 Saved")
        if(CurrentPlayerStats := PlayerStatsManager.GetPlayerStats[Agent]):
            PlayerStatsManager.RecordOwnershipL5P3(Agent, StatType.Ownership_L5_Pet3, ?Ownership_L5_Pet3 := 1)
            

    #Switch Ownership of L5_Pet3 to 0
    L5_Pet3_DesOwn(Agent:agent):void=
        Print("DesOwnership L5P3 Saved")
        if(CurrentPlayerStats := PlayerStatsManager.GetPlayerStats[Agent]):
            PlayerStatsManager.RemoveOwnershipL5P3(Agent, StatType.Ownership_L5_Pet3, ?Ownership_L5_Pet3 := 0)
            

    #Switch Ownership of L5_Pet4 to 1
    L5_Pet4_Own(Agent:agent):void=
        Print("Ownership L5P4 Saved")
        if(CurrentPlayerStats := PlayerStatsManager.GetPlayerStats[Agent]):
            PlayerStatsManager.RecordOwnershipL5P4(Agent, StatType.Ownership_L5_Pet4, ?Ownership_L5_Pet4 := 1)
            

    #Switch Ownership of L5_Pet4 to 0
    L5_Pet4_DesOwn(Agent:agent):void=
        Print("DesOwnership L5P4 Saved")
        if(CurrentPlayerStats := PlayerStatsManager.GetPlayerStats[Agent]):
            PlayerStatsManager.RemoveOwnershipL5P4(Agent, StatType.Ownership_L5_Pet4, ?Ownership_L5_Pet4 := 0) 

    #---------------------------------------L6
    #Switch Ownership of L6_Pet1 to 1
    L6_Pet1_Own(Agent:agent):void=
        Print("Ownership L6P1 Saved")
        if(CurrentPlayerStats := PlayerStatsManager.GetPlayerStats[Agent]):
            PlayerStatsManager.RecordOwnershipL6P1(Agent, StatType.Ownership_L6_Pet1, ?Ownership_L6_Pet1 := 1)
            

    #Switch Ownership of L6_Pet1 to 0
    L6_Pet1_DesOwn(Agent:agent):void=
        Print("DesOwnership L6P1 Saved")
        if(CurrentPlayerStats := PlayerStatsManager.GetPlayerStats[Agent]):
            PlayerStatsManager.RemoveOwnershipL6P1(Agent, StatType.Ownership_L6_Pet1, ?Ownership_L6_Pet1 := 0)
            

    #Switch Ownership of L6_Pet2 to 1
    L6_Pet2_Own(Agent:agent):void=
        Print("Ownership L6P2 Saved")
        if(CurrentPlayerStats := PlayerStatsManager.GetPlayerStats[Agent]):
            PlayerStatsManager.RecordOwnershipL6P2(Agent, StatType.Ownership_L6_Pet2, ?Ownership_L6_Pet2 := 1)
            

    #Switch Ownership of L6_Pet2 to 0
    L6_Pet2_DesOwn(Agent:agent):void=
        Print("DesOwnership L6P2 Saved")
        if(CurrentPlayerStats := PlayerStatsManager.GetPlayerStats[Agent]):
            PlayerStatsManager.RemoveOwnershipL6P2(Agent, StatType.Ownership_L6_Pet2, ?Ownership_L6_Pet2 := 0)
            

    #Switch Ownership of L6_Pet3 to 1
    L6_Pet3_Own(Agent:agent):void=
        Print("Ownership L6P3 Saved")
        if(CurrentPlayerStats := PlayerStatsManager.GetPlayerStats[Agent]):
            PlayerStatsManager.RecordOwnershipL6P3(Agent, StatType.Ownership_L6_Pet3, ?Ownership_L6_Pet3 := 1)
            

    #Switch Ownership of L6_Pet3 to 0
    L6_Pet3_DesOwn(Agent:agent):void=
        Print("DesOwnership L6P3 Saved")
        if(CurrentPlayerStats := PlayerStatsManager.GetPlayerStats[Agent]):
            PlayerStatsManager.RemoveOwnershipL6P3(Agent, StatType.Ownership_L6_Pet3, ?Ownership_L6_Pet3 := 0)
            

    #Switch Ownership of L6_Pet4 to 1
    L6_Pet4_Own(Agent:agent):void=
        Print("Ownership L6P4 Saved")
        if(CurrentPlayerStats := PlayerStatsManager.GetPlayerStats[Agent]):
            PlayerStatsManager.RecordOwnershipL6P4(Agent, StatType.Ownership_L6_Pet4, ?Ownership_L6_Pet4 := 1)
            

    #Switch Ownership of L6_Pet4 to 0
    L6_Pet4_DesOwn(Agent:agent):void=
        Print("DesOwnership L6P4 Saved")
        if(CurrentPlayerStats := PlayerStatsManager.GetPlayerStats[Agent]):
            PlayerStatsManager.RemoveOwnershipL6P4(Agent, StatType.Ownership_L6_Pet4, ?Ownership_L6_Pet4 := 0) 

    #---------------------------------------L7
    #Switch Ownership of L7_Pet1 to 1
    L7_Pet1_Own(Agent:agent):void=
        Print("Ownership L7P1 Saved")
        if(CurrentPlayerStats := PlayerStatsManager.GetPlayerStats[Agent]):
            PlayerStatsManager.RecordOwnershipL7P1(Agent, StatType.Ownership_L7_Pet1, ?Ownership_L7_Pet1 := 1)
            

    #Switch Ownership of L7_Pet1 to 0
    L7_Pet1_DesOwn(Agent:agent):void=
        Print("DesOwnership L7P1 Saved")
        if(CurrentPlayerStats := PlayerStatsManager.GetPlayerStats[Agent]):
            PlayerStatsManager.RemoveOwnershipL7P1(Agent, StatType.Ownership_L7_Pet1, ?Ownership_L7_Pet1 := 0)
            

    #Switch Ownership of L7_Pet2 to 1
    L7_Pet2_Own(Agent:agent):void=
        Print("Ownership L7P2 Saved")
        if(CurrentPlayerStats := PlayerStatsManager.GetPlayerStats[Agent]):
            PlayerStatsManager.RecordOwnershipL7P2(Agent, StatType.Ownership_L7_Pet2, ?Ownership_L7_Pet2 := 1)
            

    #Switch Ownership of L7_Pet2 to 0
    L7_Pet2_DesOwn(Agent:agent):void=
        Print("DesOwnership L7P2 Saved")
        if(CurrentPlayerStats := PlayerStatsManager.GetPlayerStats[Agent]):
            PlayerStatsManager.RemoveOwnershipL7P2(Agent, StatType.Ownership_L7_Pet2, ?Ownership_L7_Pet2 := 0)
            

    #Switch Ownership of L7_Pet3 to 1
    L7_Pet3_Own(Agent:agent):void=
        Print("Ownership L7P3 Saved")
        if(CurrentPlayerStats := PlayerStatsManager.GetPlayerStats[Agent]):
            PlayerStatsManager.RecordOwnershipL7P3(Agent, StatType.Ownership_L7_Pet3, ?Ownership_L7_Pet3 := 1)
            

    #Switch Ownership of L7_Pet3 to 0
    L7_Pet3_DesOwn(Agent:agent):void=
        Print("DesOwnership L7P3 Saved")
        if(CurrentPlayerStats := PlayerStatsManager.GetPlayerStats[Agent]):
            PlayerStatsManager.RemoveOwnershipL7P3(Agent, StatType.Ownership_L7_Pet3, ?Ownership_L7_Pet3 := 0)
            

    #Switch Ownership of L7_Pet4 to 1
    L7_Pet4_Own(Agent:agent):void=
        Print("Ownership L7P4 Saved")
        if(CurrentPlayerStats := PlayerStatsManager.GetPlayerStats[Agent]):
            PlayerStatsManager.RecordOwnershipL7P4(Agent, StatType.Ownership_L7_Pet4, ?Ownership_L7_Pet4 := 1)
            

    #Switch Ownership of L7_Pet4 to 0
    L7_Pet4_DesOwn(Agent:agent):void=
        Print("DesOwnership L7P4 Saved")
        if(CurrentPlayerStats := PlayerStatsManager.GetPlayerStats[Agent]):
            PlayerStatsManager.RemoveOwnershipL7P4(Agent, StatType.Ownership_L7_Pet4, ?Ownership_L7_Pet4 := 0) 

    #---------------------------------------L8
    #Switch Ownership of L8_Pet1 to 1
    L8_Pet1_Own(Agent:agent):void=
        Print("Ownership L8P1 Saved")
        if(CurrentPlayerStats := PlayerStatsManager.GetPlayerStats[Agent]):
            PlayerStatsManager.RecordOwnershipL8P1(Agent, StatType.Ownership_L8_Pet1, ?Ownership_L8_Pet1 := 1)
            

    #Switch Ownership of L8_Pet1 to 0
    L8_Pet1_DesOwn(Agent:agent):void=
        Print("DesOwnership L8P1 Saved")
        if(CurrentPlayerStats := PlayerStatsManager.GetPlayerStats[Agent]):
            PlayerStatsManager.RemoveOwnershipL8P1(Agent, StatType.Ownership_L8_Pet1, ?Ownership_L8_Pet1 := 0)
            

    #Switch Ownership of L8_Pet2 to 1
    L8_Pet2_Own(Agent:agent):void=
        Print("Ownership L8P2 Saved")
        if(CurrentPlayerStats := PlayerStatsManager.GetPlayerStats[Agent]):
            PlayerStatsManager.RecordOwnershipL8P2(Agent, StatType.Ownership_L8_Pet2, ?Ownership_L8_Pet2 := 1)
            

    #Switch Ownership of L8_Pet2 to 0
    L8_Pet2_DesOwn(Agent:agent):void=
        Print("DesOwnership L8P2 Saved")
        if(CurrentPlayerStats := PlayerStatsManager.GetPlayerStats[Agent]):
            PlayerStatsManager.RemoveOwnershipL8P2(Agent, StatType.Ownership_L8_Pet2, ?Ownership_L8_Pet2 := 0)
            

    #Switch Ownership of L8_Pet3 to 1
    L8_Pet3_Own(Agent:agent):void=
        Print("Ownership L8P3 Saved")
        if(CurrentPlayerStats := PlayerStatsManager.GetPlayerStats[Agent]):
            PlayerStatsManager.RecordOwnershipL8P3(Agent, StatType.Ownership_L8_Pet3, ?Ownership_L8_Pet3 := 1)
            

    #Switch Ownership of L8_Pet3 to 0
    L8_Pet3_DesOwn(Agent:agent):void=
        Print("DesOwnership L8P3 Saved")
        if(CurrentPlayerStats := PlayerStatsManager.GetPlayerStats[Agent]):
            PlayerStatsManager.RemoveOwnershipL8P3(Agent, StatType.Ownership_L8_Pet3, ?Ownership_L8_Pet3 := 0)
            

    #Switch Ownership of L8_Pet4 to 1
    L8_Pet4_Own(Agent:agent):void=
        Print("Ownership L8P4 Saved")
        if(CurrentPlayerStats := PlayerStatsManager.GetPlayerStats[Agent]):
            PlayerStatsManager.RecordOwnershipL8P4(Agent, StatType.Ownership_L8_Pet4, ?Ownership_L8_Pet4 := 1)
            

    #Switch Ownership of L8_Pet4 to 0
    L8_Pet4_DesOwn(Agent:agent):void=
        Print("DesOwnership L8P4 Saved")
        if(CurrentPlayerStats := PlayerStatsManager.GetPlayerStats[Agent]):
            PlayerStatsManager.RemoveOwnershipL8P4(Agent, StatType.Ownership_L8_Pet4, ?Ownership_L8_Pet4 := 0) 

    #---------------------------------------L9
    #Switch Ownership of L9_Pet1 to 1
    L9_Pet1_Own(Agent:agent):void=
        Print("Ownership L9P1 Saved")
        if(CurrentPlayerStats := PlayerStatsManager.GetPlayerStats[Agent]):
            PlayerStatsManager.RecordOwnershipL9P1(Agent, StatType.Ownership_L9_Pet1, ?Ownership_L9_Pet1 := 1)
            

    #Switch Ownership of L9_Pet1 to 0
    L9_Pet1_DesOwn(Agent:agent):void=
        Print("DesOwnership L9P1 Saved")
        if(CurrentPlayerStats := PlayerStatsManager.GetPlayerStats[Agent]):
            PlayerStatsManager.RemoveOwnershipL9P1(Agent, StatType.Ownership_L9_Pet1, ?Ownership_L9_Pet1 := 0)
            

    #Switch Ownership of L9_Pet2 to 1
    L9_Pet2_Own(Agent:agent):void=
        Print("Ownership L9P2 Saved")
        if(CurrentPlayerStats := PlayerStatsManager.GetPlayerStats[Agent]):
            PlayerStatsManager.RecordOwnershipL9P2(Agent, StatType.Ownership_L9_Pet2, ?Ownership_L9_Pet2 := 1)
            

    #Switch Ownership of L9_Pet2 to 0
    L9_Pet2_DesOwn(Agent:agent):void=
        Print("DesOwnership L9P2 Saved")
        if(CurrentPlayerStats := PlayerStatsManager.GetPlayerStats[Agent]):
            PlayerStatsManager.RemoveOwnershipL9P2(Agent, StatType.Ownership_L9_Pet2, ?Ownership_L9_Pet2 := 0)
            

    #Switch Ownership of L9_Pet3 to 1
    L9_Pet3_Own(Agent:agent):void=
        Print("Ownership L9P3 Saved")
        if(CurrentPlayerStats := PlayerStatsManager.GetPlayerStats[Agent]):
            PlayerStatsManager.RecordOwnershipL9P3(Agent, StatType.Ownership_L9_Pet3, ?Ownership_L9_Pet3 := 1)
            

    #Switch Ownership of L9_Pet3 to 0
    L9_Pet3_DesOwn(Agent:agent):void=
        Print("DesOwnership L9P3 Saved")
        if(CurrentPlayerStats := PlayerStatsManager.GetPlayerStats[Agent]):
            PlayerStatsManager.RemoveOwnershipL9P3(Agent, StatType.Ownership_L9_Pet3, ?Ownership_L9_Pet3 := 0)
            

    #Switch Ownership of L9_Pet4 to 1
    L9_Pet4_Own(Agent:agent):void=
        Print("Ownership L9P4 Saved")
        if(CurrentPlayerStats := PlayerStatsManager.GetPlayerStats[Agent]):
            PlayerStatsManager.RecordOwnershipL9P4(Agent, StatType.Ownership_L9_Pet4, ?Ownership_L9_Pet4 := 1)
            

    #Switch Ownership of L9_Pet4 to 0
    L9_Pet4_DesOwn(Agent:agent):void=
        Print("DesOwnership L9P4 Saved")
        if(CurrentPlayerStats := PlayerStatsManager.GetPlayerStats[Agent]):
            PlayerStatsManager.RemoveOwnershipL9P4(Agent, StatType.Ownership_L9_Pet4, ?Ownership_L9_Pet4 := 0) 

    #---------------------------------------L10
    #Switch Ownership of L10_Pet1 to 1
    L10_Pet1_Own(Agent:agent):void=
        Print("Ownership L10P1 Saved")
        if(CurrentPlayerStats := PlayerStatsManager.GetPlayerStats[Agent]):
            PlayerStatsManager.RecordOwnershipL10P1(Agent, StatType.Ownership_L10_Pet1, ?Ownership_L10_Pet1 := 1)
            

    #Switch Ownership of L10_Pet1 to 0
    L10_Pet1_DesOwn(Agent:agent):void=
        Print("DesOwnership L10P1 Saved")
        if(CurrentPlayerStats := PlayerStatsManager.GetPlayerStats[Agent]):
            PlayerStatsManager.RemoveOwnershipL10P1(Agent, StatType.Ownership_L10_Pet1, ?Ownership_L10_Pet1 := 0)
            

    #Switch Ownership of L10_Pet2 to 1
    L10_Pet2_Own(Agent:agent):void=
        Print("Ownership L10P2 Saved")
        if(CurrentPlayerStats := PlayerStatsManager.GetPlayerStats[Agent]):
            PlayerStatsManager.RecordOwnershipL10P2(Agent, StatType.Ownership_L10_Pet2, ?Ownership_L10_Pet2 := 1)
            

    #Switch Ownership of L10_Pet2 to 0
    L10_Pet2_DesOwn(Agent:agent):void=
        Print("DesOwnership L10P2 Saved")
        if(CurrentPlayerStats := PlayerStatsManager.GetPlayerStats[Agent]):
            PlayerStatsManager.RemoveOwnershipL10P2(Agent, StatType.Ownership_L10_Pet2, ?Ownership_L10_Pet2 := 0)
            

    #Switch Ownership of L10_Pet3 to 1
    L10_Pet3_Own(Agent:agent):void=
        Print("Ownership L10P3 Saved")
        if(CurrentPlayerStats := PlayerStatsManager.GetPlayerStats[Agent]):
            PlayerStatsManager.RecordOwnershipL10P3(Agent, StatType.Ownership_L10_Pet3, ?Ownership_L10_Pet3 := 1)
            

    #Switch Ownership of L10_Pet3 to 0
    L10_Pet3_DesOwn(Agent:agent):void=
        Print("DesOwnership L10P3 Saved")
        if(CurrentPlayerStats := PlayerStatsManager.GetPlayerStats[Agent]):
            PlayerStatsManager.RemoveOwnershipL10P3(Agent, StatType.Ownership_L10_Pet3, ?Ownership_L10_Pet3 := 0)
            

    #Switch Ownership of L10_Pet4 to 1
    L10_Pet4_Own(Agent:agent):void=
        Print("Ownership L10P4 Saved")
        if(CurrentPlayerStats := PlayerStatsManager.GetPlayerStats[Agent]):
            PlayerStatsManager.RecordOwnershipL10P4(Agent, StatType.Ownership_L10_Pet4, ?Ownership_L10_Pet4 := 1)
            

    #Switch Ownership of L10_Pet4 to 0
    L10_Pet4_DesOwn(Agent:agent):void=
        Print("DesOwnership L10P4 Saved")
        if(CurrentPlayerStats := PlayerStatsManager.GetPlayerStats[Agent]):
            PlayerStatsManager.RemoveOwnershipL10P4(Agent, StatType.Ownership_L10_Pet4, ?Ownership_L10_Pet4 := 0) 

    #---------------------------------------L11
    #Switch Ownership of L11_Pet1 to 1
    L11_Pet1_Own(Agent:agent):void=
        Print("Ownership L11P1 Saved")
        if(CurrentPlayerStats := PlayerStatsManager.GetPlayerStats[Agent]):
            PlayerStatsManager.RecordOwnershipL11P1(Agent, StatType.Ownership_L11_Pet1, ?Ownership_L11_Pet1 := 1)
            

    #Switch Ownership of L11_Pet1 to 0
    L11_Pet1_DesOwn(Agent:agent):void=
        Print("DesOwnership L11P1 Saved")
        if(CurrentPlayerStats := PlayerStatsManager.GetPlayerStats[Agent]):
            PlayerStatsManager.RemoveOwnershipL11P1(Agent, StatType.Ownership_L11_Pet1, ?Ownership_L11_Pet1 := 0)
            

    #Switch Ownership of L11_Pet2 to 1
    L11_Pet2_Own(Agent:agent):void=
        Print("Ownership L11P2 Saved")
        if(CurrentPlayerStats := PlayerStatsManager.GetPlayerStats[Agent]):
            PlayerStatsManager.RecordOwnershipL11P2(Agent, StatType.Ownership_L11_Pet2, ?Ownership_L11_Pet2 := 1)
            

    #Switch Ownership of L11_Pet2 to 0
    L11_Pet2_DesOwn(Agent:agent):void=
        Print("DesOwnership L11P2 Saved")
        if(CurrentPlayerStats := PlayerStatsManager.GetPlayerStats[Agent]):
            PlayerStatsManager.RemoveOwnershipL11P2(Agent, StatType.Ownership_L11_Pet2, ?Ownership_L11_Pet2 := 0)
            

    #Switch Ownership of L11_Pet3 to 1
    L11_Pet3_Own(Agent:agent):void=
        Print("Ownership L11P3 Saved")
        if(CurrentPlayerStats := PlayerStatsManager.GetPlayerStats[Agent]):
            PlayerStatsManager.RecordOwnershipL11P3(Agent, StatType.Ownership_L11_Pet3, ?Ownership_L11_Pet3 := 1)
            

    #Switch Ownership of L11_Pet3 to 0
    L11_Pet3_DesOwn(Agent:agent):void=
        Print("DesOwnership L11P3 Saved")
        if(CurrentPlayerStats := PlayerStatsManager.GetPlayerStats[Agent]):
            PlayerStatsManager.RemoveOwnershipL11P3(Agent, StatType.Ownership_L11_Pet3, ?Ownership_L11_Pet3 := 0)
            

    #Switch Ownership of L11_Pet4 to 1
    L11_Pet4_Own(Agent:agent):void=
        Print("Ownership L11P4 Saved")
        if(CurrentPlayerStats := PlayerStatsManager.GetPlayerStats[Agent]):
            PlayerStatsManager.RecordOwnershipL11P4(Agent, StatType.Ownership_L11_Pet4, ?Ownership_L11_Pet4 := 1)
            

    #Switch Ownership of L11_Pet4 to 0
    L11_Pet4_DesOwn(Agent:agent):void=
        Print("DesOwnership L11P4 Saved")
        if(CurrentPlayerStats := PlayerStatsManager.GetPlayerStats[Agent]):
            PlayerStatsManager.RemoveOwnershipL11P4(Agent, StatType.Ownership_L11_Pet4, ?Ownership_L11_Pet4 := 0) 

    #---------------------------------------L12
    #Switch Ownership of L12_Pet1 to 1
    L12_Pet1_Own(Agent:agent):void=
        Print("Ownership L12P1 Saved")
        if(CurrentPlayerStats := PlayerStatsManager.GetPlayerStats[Agent]):
            PlayerStatsManager.RecordOwnershipL12P1(Agent, StatType.Ownership_L12_Pet1, ?Ownership_L12_Pet1 := 1)
            

    #Switch Ownership of L12_Pet1 to 0
    L12_Pet1_DesOwn(Agent:agent):void=
        Print("DesOwnership L12P1 Saved")
        if(CurrentPlayerStats := PlayerStatsManager.GetPlayerStats[Agent]):
            PlayerStatsManager.RemoveOwnershipL12P1(Agent, StatType.Ownership_L12_Pet1, ?Ownership_L12_Pet1 := 0)
            

    #Switch Ownership of L12_Pet2 to 1
    L12_Pet2_Own(Agent:agent):void=
        Print("Ownership L12P2 Saved")
        if(CurrentPlayerStats := PlayerStatsManager.GetPlayerStats[Agent]):
            PlayerStatsManager.RecordOwnershipL12P2(Agent, StatType.Ownership_L12_Pet2, ?Ownership_L12_Pet2 := 1)
            

    #Switch Ownership of L12_Pet2 to 0
    L12_Pet2_DesOwn(Agent:agent):void=
        Print("DesOwnership L12P2 Saved")
        if(CurrentPlayerStats := PlayerStatsManager.GetPlayerStats[Agent]):
            PlayerStatsManager.RemoveOwnershipL12P2(Agent, StatType.Ownership_L12_Pet2, ?Ownership_L12_Pet2 := 0)
            

    #Switch Ownership of L12_Pet3 to 1
    L12_Pet3_Own(Agent:agent):void=
        Print("Ownership L12P3 Saved")
        if(CurrentPlayerStats := PlayerStatsManager.GetPlayerStats[Agent]):
            PlayerStatsManager.RecordOwnershipL12P3(Agent, StatType.Ownership_L12_Pet3, ?Ownership_L12_Pet3 := 1)
            

    #Switch Ownership of L12_Pet3 to 0
    L12_Pet3_DesOwn(Agent:agent):void=
        Print("DesOwnership L12P3 Saved")
        if(CurrentPlayerStats := PlayerStatsManager.GetPlayerStats[Agent]):
            PlayerStatsManager.RemoveOwnershipL12P3(Agent, StatType.Ownership_L12_Pet3, ?Ownership_L12_Pet3 := 0)
            

    #Switch Ownership of L12_Pet4 to 1
    L12_Pet4_Own(Agent:agent):void=
        Print("Ownership L12P4 Saved")
        if(CurrentPlayerStats := PlayerStatsManager.GetPlayerStats[Agent]):
            PlayerStatsManager.RecordOwnershipL12P4(Agent, StatType.Ownership_L12_Pet4, ?Ownership_L12_Pet4 := 1)
            

    #Switch Ownership of L12_Pet4 to 0
    L12_Pet4_DesOwn(Agent:agent):void=
        Print("DesOwnership L12P4 Saved")
        if(CurrentPlayerStats := PlayerStatsManager.GetPlayerStats[Agent]):
            PlayerStatsManager.RemoveOwnershipL12P4(Agent, StatType.Ownership_L12_Pet4, ?Ownership_L12_Pet4 := 0) 

    #---------------------------------------L13
    #Switch Ownership of L13_Pet1 to 1
    L13_Pet1_Own(Agent:agent):void=
        Print("Ownership L13P1 Saved")
        if(CurrentPlayerStats := PlayerStatsManager.GetPlayerStats[Agent]):
            PlayerStatsManager.RecordOwnershipL13P1(Agent, StatType.Ownership_L13_Pet1, ?Ownership_L13_Pet1 := 1)
            

    #Switch Ownership of L13_Pet1 to 0
    L13_Pet1_DesOwn(Agent:agent):void=
        Print("DesOwnership L13P1 Saved")
        if(CurrentPlayerStats := PlayerStatsManager.GetPlayerStats[Agent]):
            PlayerStatsManager.RemoveOwnershipL13P1(Agent, StatType.Ownership_L13_Pet1, ?Ownership_L13_Pet1 := 0)
            

    #Switch Ownership of L13_Pet2 to 1
    L13_Pet2_Own(Agent:agent):void=
        Print("Ownership L13P2 Saved")
        if(CurrentPlayerStats := PlayerStatsManager.GetPlayerStats[Agent]):
            PlayerStatsManager.RecordOwnershipL13P2(Agent, StatType.Ownership_L13_Pet2, ?Ownership_L13_Pet2 := 1)
            

    #Switch Ownership of L13_Pet2 to 0
    L13_Pet2_DesOwn(Agent:agent):void=
        Print("DesOwnership L13P2 Saved")
        if(CurrentPlayerStats := PlayerStatsManager.GetPlayerStats[Agent]):
            PlayerStatsManager.RemoveOwnershipL13P2(Agent, StatType.Ownership_L13_Pet2, ?Ownership_L13_Pet2 := 0)
            

    #Switch Ownership of L13_Pet3 to 1
    L13_Pet3_Own(Agent:agent):void=
        Print("Ownership L13P3 Saved")
        if(CurrentPlayerStats := PlayerStatsManager.GetPlayerStats[Agent]):
            PlayerStatsManager.RecordOwnershipL13P3(Agent, StatType.Ownership_L13_Pet3, ?Ownership_L13_Pet3 := 1)
            

    #Switch Ownership of L13_Pet3 to 0
    L13_Pet3_DesOwn(Agent:agent):void=
        Print("DesOwnership L13P3 Saved")
        if(CurrentPlayerStats := PlayerStatsManager.GetPlayerStats[Agent]):
            PlayerStatsManager.RemoveOwnershipL13P3(Agent, StatType.Ownership_L13_Pet3, ?Ownership_L13_Pet3 := 0)
            

    #Switch Ownership of L13_Pet4 to 1
    L13_Pet4_Own(Agent:agent):void=
        Print("Ownership L13P4 Saved")
        if(CurrentPlayerStats := PlayerStatsManager.GetPlayerStats[Agent]):
            PlayerStatsManager.RecordOwnershipL13P4(Agent, StatType.Ownership_L13_Pet4, ?Ownership_L13_Pet4 := 1)
            

    #Switch Ownership of L13_Pet4 to 0
    L13_Pet4_DesOwn(Agent:agent):void=
        Print("DesOwnership L13P4 Saved")
        if(CurrentPlayerStats := PlayerStatsManager.GetPlayerStats[Agent]):
            PlayerStatsManager.RemoveOwnershipL13P4(Agent, StatType.Ownership_L13_Pet4, ?Ownership_L13_Pet4 := 0) 

    #---------------------------------------L14
    #Switch Ownership of L14_Pet1 to 1
    L14_Pet1_Own(Agent:agent):void=
        Print("Ownership L14P1 Saved")
        if(CurrentPlayerStats := PlayerStatsManager.GetPlayerStats[Agent]):
            PlayerStatsManager.RecordOwnershipL14P1(Agent, StatType.Ownership_L14_Pet1, ?Ownership_L14_Pet1 := 1)
            

    #Switch Ownership of L14_Pet1 to 0
    L14_Pet1_DesOwn(Agent:agent):void=
        Print("DesOwnership L14P1 Saved")
        if(CurrentPlayerStats := PlayerStatsManager.GetPlayerStats[Agent]):
            PlayerStatsManager.RemoveOwnershipL14P1(Agent, StatType.Ownership_L14_Pet1, ?Ownership_L14_Pet1 := 0)
            

    #Switch Ownership of L14_Pet2 to 1
    L14_Pet2_Own(Agent:agent):void=
        Print("Ownership L14P2 Saved")
        if(CurrentPlayerStats := PlayerStatsManager.GetPlayerStats[Agent]):
            PlayerStatsManager.RecordOwnershipL14P2(Agent, StatType.Ownership_L14_Pet2, ?Ownership_L14_Pet2 := 1)
            

    #Switch Ownership of L14_Pet2 to 0
    L14_Pet2_DesOwn(Agent:agent):void=
        Print("DesOwnership L14P2 Saved")
        if(CurrentPlayerStats := PlayerStatsManager.GetPlayerStats[Agent]):
            PlayerStatsManager.RemoveOwnershipL14P2(Agent, StatType.Ownership_L14_Pet2, ?Ownership_L14_Pet2 := 0)
            

    #Switch Ownership of L14_Pet3 to 1
    L14_Pet3_Own(Agent:agent):void=
        Print("Ownership L14P3 Saved")
        if(CurrentPlayerStats := PlayerStatsManager.GetPlayerStats[Agent]):
            PlayerStatsManager.RecordOwnershipL14P3(Agent, StatType.Ownership_L14_Pet3, ?Ownership_L14_Pet3 := 1)
            

    #Switch Ownership of L14_Pet3 to 0
    L14_Pet3_DesOwn(Agent:agent):void=
        Print("DesOwnership L14P3 Saved")
        if(CurrentPlayerStats := PlayerStatsManager.GetPlayerStats[Agent]):
            PlayerStatsManager.RemoveOwnershipL14P3(Agent, StatType.Ownership_L14_Pet3, ?Ownership_L14_Pet3 := 0)
            

    #Switch Ownership of L14_Pet4 to 1
    L14_Pet4_Own(Agent:agent):void=
        Print("Ownership L14P4 Saved")
        if(CurrentPlayerStats := PlayerStatsManager.GetPlayerStats[Agent]):
            PlayerStatsManager.RecordOwnershipL14P4(Agent, StatType.Ownership_L14_Pet4, ?Ownership_L14_Pet4 := 1)
            

    #Switch Ownership of L14_Pet4 to 0
    L14_Pet4_DesOwn(Agent:agent):void=
        Print("DesOwnership L14P4 Saved")
        if(CurrentPlayerStats := PlayerStatsManager.GetPlayerStats[Agent]):
            PlayerStatsManager.RemoveOwnershipL14P4(Agent, StatType.Ownership_L14_Pet4, ?Ownership_L14_Pet4 := 0) 

    #---------------------------------------L15
    #Switch Ownership of L15_Pet1 to 1
    L15_Pet1_Own(Agent:agent):void=
        Print("Ownership L15P1 Saved")
        if(CurrentPlayerStats := PlayerStatsManager.GetPlayerStats[Agent]):
            PlayerStatsManager.RecordOwnershipL15P1(Agent, StatType.Ownership_L15_Pet1, ?Ownership_L15_Pet1 := 1)
            

    #Switch Ownership of L15_Pet1 to 0
    L15_Pet1_DesOwn(Agent:agent):void=
        Print("DesOwnership L15P1 Saved")
        if(CurrentPlayerStats := PlayerStatsManager.GetPlayerStats[Agent]):
            PlayerStatsManager.RemoveOwnershipL15P1(Agent, StatType.Ownership_L15_Pet1, ?Ownership_L15_Pet1 := 0)
            

    #Switch Ownership of L15_Pet2 to 1
    L15_Pet2_Own(Agent:agent):void=
        Print("Ownership L15P2 Saved")
        if(CurrentPlayerStats := PlayerStatsManager.GetPlayerStats[Agent]):
            PlayerStatsManager.RecordOwnershipL15P2(Agent, StatType.Ownership_L15_Pet2, ?Ownership_L15_Pet2 := 1)
            

    #Switch Ownership of L15_Pet2 to 0
    L15_Pet2_DesOwn(Agent:agent):void=
        Print("DesOwnership L15P2 Saved")
        if(CurrentPlayerStats := PlayerStatsManager.GetPlayerStats[Agent]):
            PlayerStatsManager.RemoveOwnershipL15P2(Agent, StatType.Ownership_L15_Pet2, ?Ownership_L15_Pet2 := 0)
            

    #Switch Ownership of L15_Pet3 to 1
    L15_Pet3_Own(Agent:agent):void=
        Print("Ownership L15P3 Saved")
        if(CurrentPlayerStats := PlayerStatsManager.GetPlayerStats[Agent]):
            PlayerStatsManager.RecordOwnershipL15P3(Agent, StatType.Ownership_L15_Pet3, ?Ownership_L15_Pet3 := 1)
            

    #Switch Ownership of L15_Pet3 to 0
    L15_Pet3_DesOwn(Agent:agent):void=
        Print("DesOwnership L15P3 Saved")
        if(CurrentPlayerStats := PlayerStatsManager.GetPlayerStats[Agent]):
            PlayerStatsManager.RemoveOwnershipL15P3(Agent, StatType.Ownership_L15_Pet3, ?Ownership_L15_Pet3 := 0)
            

    #Switch Ownership of L15_Pet4 to 1
    L15_Pet4_Own(Agent:agent):void=
        Print("Ownership L15P4 Saved")
        if(CurrentPlayerStats := PlayerStatsManager.GetPlayerStats[Agent]):
            PlayerStatsManager.RecordOwnershipL15P4(Agent, StatType.Ownership_L15_Pet4, ?Ownership_L15_Pet4 := 1)
            

    #Switch Ownership of L15_Pet4 to 0
    L15_Pet4_DesOwn(Agent:agent):void=
        Print("DesOwnership L15P4 Saved")
        if(CurrentPlayerStats := PlayerStatsManager.GetPlayerStats[Agent]):
            PlayerStatsManager.RemoveOwnershipL15P4(Agent, StatType.Ownership_L15_Pet4, ?Ownership_L15_Pet4 := 0) 

    #---------------------------------------L16
    #Switch Ownership of L16_Pet1 to 1
    L16_Pet1_Own(Agent:agent):void=
        Print("Ownership L16P1 Saved")
        if(CurrentPlayerStats := PlayerStatsManager.GetPlayerStats[Agent]):
            PlayerStatsManager.RecordOwnershipL16P1(Agent, StatType.Ownership_L16_Pet1, ?Ownership_L16_Pet1 := 1)
            

    #Switch Ownership of L16_Pet1 to 0
    L16_Pet1_DesOwn(Agent:agent):void=
        Print("DesOwnership L16P1 Saved")
        if(CurrentPlayerStats := PlayerStatsManager.GetPlayerStats[Agent]):
            PlayerStatsManager.RemoveOwnershipL16P1(Agent, StatType.Ownership_L16_Pet1, ?Ownership_L16_Pet1 := 0)
            

    #Switch Ownership of L16_Pet2 to 1
    L16_Pet2_Own(Agent:agent):void=
        Print("Ownership L16P2 Saved")
        if(CurrentPlayerStats := PlayerStatsManager.GetPlayerStats[Agent]):
            PlayerStatsManager.RecordOwnershipL16P2(Agent, StatType.Ownership_L16_Pet2, ?Ownership_L16_Pet2 := 1)
            

    #Switch Ownership of L16_Pet2 to 0
    L16_Pet2_DesOwn(Agent:agent):void=
        Print("DesOwnership L16P2 Saved")
        if(CurrentPlayerStats := PlayerStatsManager.GetPlayerStats[Agent]):
            PlayerStatsManager.RemoveOwnershipL16P2(Agent, StatType.Ownership_L16_Pet2, ?Ownership_L16_Pet2 := 0)
            

    #Switch Ownership of L16_Pet3 to 1
    L16_Pet3_Own(Agent:agent):void=
        Print("Ownership L16P3 Saved")
        if(CurrentPlayerStats := PlayerStatsManager.GetPlayerStats[Agent]):
            PlayerStatsManager.RecordOwnershipL16P3(Agent, StatType.Ownership_L16_Pet3, ?Ownership_L16_Pet3 := 1)
            

    #Switch Ownership of L16_Pet3 to 0
    L16_Pet3_DesOwn(Agent:agent):void=
        Print("DesOwnership L16P3 Saved")
        if(CurrentPlayerStats := PlayerStatsManager.GetPlayerStats[Agent]):
            PlayerStatsManager.RemoveOwnershipL16P3(Agent, StatType.Ownership_L16_Pet3, ?Ownership_L16_Pet3 := 0)
            

    #Switch Ownership of L16_Pet4 to 1
    L16_Pet4_Own(Agent:agent):void=
        Print("Ownership L16P4 Saved")
        if(CurrentPlayerStats := PlayerStatsManager.GetPlayerStats[Agent]):
            PlayerStatsManager.RecordOwnershipL16P4(Agent, StatType.Ownership_L16_Pet4, ?Ownership_L16_Pet4 := 1)
            

    #Switch Ownership of L16_Pet4 to 0
    L16_Pet4_DesOwn(Agent:agent):void=
        Print("DesOwnership L16P4 Saved")
        if(CurrentPlayerStats := PlayerStatsManager.GetPlayerStats[Agent]):
            PlayerStatsManager.RemoveOwnershipL16P4(Agent, StatType.Ownership_L16_Pet4, ?Ownership_L16_Pet4 := 0) 

    #---------------------------------------L17
    #Switch Ownership of L17_Pet1 to 1
    L17_Pet1_Own(Agent:agent):void=
        Print("Ownership L17P1 Saved")
        if(CurrentPlayerStats := PlayerStatsManager.GetPlayerStats[Agent]):
            PlayerStatsManager.RecordOwnershipL17P1(Agent, StatType.Ownership_L17_Pet1, ?Ownership_L17_Pet1 := 1)
            

    #Switch Ownership of L17_Pet1 to 0
    L17_Pet1_DesOwn(Agent:agent):void=
        Print("DesOwnership L17P1 Saved")
        if(CurrentPlayerStats := PlayerStatsManager.GetPlayerStats[Agent]):
            PlayerStatsManager.RemoveOwnershipL17P1(Agent, StatType.Ownership_L17_Pet1, ?Ownership_L17_Pet1 := 0)
            

    #Switch Ownership of L17_Pet2 to 1
    L17_Pet2_Own(Agent:agent):void=
        Print("Ownership L17P2 Saved")
        if(CurrentPlayerStats := PlayerStatsManager.GetPlayerStats[Agent]):
            PlayerStatsManager.RecordOwnershipL17P2(Agent, StatType.Ownership_L17_Pet2, ?Ownership_L17_Pet2 := 1)
            

    #Switch Ownership of L17_Pet2 to 0
    L17_Pet2_DesOwn(Agent:agent):void=
        Print("DesOwnership L17P2 Saved")
        if(CurrentPlayerStats := PlayerStatsManager.GetPlayerStats[Agent]):
            PlayerStatsManager.RemoveOwnershipL17P2(Agent, StatType.Ownership_L17_Pet2, ?Ownership_L17_Pet2 := 0)
            

    #Switch Ownership of L17_Pet3 to 1
    L17_Pet3_Own(Agent:agent):void=
        Print("Ownership L17P3 Saved")
        if(CurrentPlayerStats := PlayerStatsManager.GetPlayerStats[Agent]):
            PlayerStatsManager.RecordOwnershipL17P3(Agent, StatType.Ownership_L17_Pet3, ?Ownership_L17_Pet3 := 1)
            

    #Switch Ownership of L17_Pet3 to 0
    L17_Pet3_DesOwn(Agent:agent):void=
        Print("DesOwnership L17P3 Saved")
        if(CurrentPlayerStats := PlayerStatsManager.GetPlayerStats[Agent]):
            PlayerStatsManager.RemoveOwnershipL17P3(Agent, StatType.Ownership_L17_Pet3, ?Ownership_L17_Pet3 := 0)
            

    #Switch Ownership of L17_Pet4 to 1
    L17_Pet4_Own(Agent:agent):void=
        Print("Ownership L17P4 Saved")
        if(CurrentPlayerStats := PlayerStatsManager.GetPlayerStats[Agent]):
            PlayerStatsManager.RecordOwnershipL17P4(Agent, StatType.Ownership_L17_Pet4, ?Ownership_L17_Pet4 := 1)
            

    #Switch Ownership of L17_Pet4 to 0
    L17_Pet4_DesOwn(Agent:agent):void=
        Print("DesOwnership L17P4 Saved")
        if(CurrentPlayerStats := PlayerStatsManager.GetPlayerStats[Agent]):
            PlayerStatsManager.RemoveOwnershipL17P4(Agent, StatType.Ownership_L17_Pet4, ?Ownership_L17_Pet4 := 0) 

    #---------------------------------------L18
    #Switch Ownership of L18_Pet1 to 1
    L18_Pet1_Own(Agent:agent):void=
        Print("Ownership L18P1 Saved")
        if(CurrentPlayerStats := PlayerStatsManager.GetPlayerStats[Agent]):
            PlayerStatsManager.RecordOwnershipL18P1(Agent, StatType.Ownership_L18_Pet1, ?Ownership_L18_Pet1 := 1)
            

    #Switch Ownership of L18_Pet1 to 0
    L18_Pet1_DesOwn(Agent:agent):void=
        Print("DesOwnership L18P1 Saved")
        if(CurrentPlayerStats := PlayerStatsManager.GetPlayerStats[Agent]):
            PlayerStatsManager.RemoveOwnershipL18P1(Agent, StatType.Ownership_L18_Pet1, ?Ownership_L18_Pet1 := 0)
            

    #Switch Ownership of L18_Pet2 to 1
    L18_Pet2_Own(Agent:agent):void=
        Print("Ownership L18P2 Saved")
        if(CurrentPlayerStats := PlayerStatsManager.GetPlayerStats[Agent]):
            PlayerStatsManager.RecordOwnershipL18P2(Agent, StatType.Ownership_L18_Pet2, ?Ownership_L18_Pet2 := 1)
            

    #Switch Ownership of L18_Pet2 to 0
    L18_Pet2_DesOwn(Agent:agent):void=
        Print("DesOwnership L18P2 Saved")
        if(CurrentPlayerStats := PlayerStatsManager.GetPlayerStats[Agent]):
            PlayerStatsManager.RemoveOwnershipL18P2(Agent, StatType.Ownership_L18_Pet2, ?Ownership_L18_Pet2 := 0)
            

    #Switch Ownership of L18_Pet3 to 1
    L18_Pet3_Own(Agent:agent):void=
        Print("Ownership L18P3 Saved")
        if(CurrentPlayerStats := PlayerStatsManager.GetPlayerStats[Agent]):
            PlayerStatsManager.RecordOwnershipL18P3(Agent, StatType.Ownership_L18_Pet3, ?Ownership_L18_Pet3 := 1)
            

    #Switch Ownership of L18_Pet3 to 0
    L18_Pet3_DesOwn(Agent:agent):void=
        Print("DesOwnership L18P3 Saved")
        if(CurrentPlayerStats := PlayerStatsManager.GetPlayerStats[Agent]):
            PlayerStatsManager.RemoveOwnershipL18P3(Agent, StatType.Ownership_L18_Pet3, ?Ownership_L18_Pet3 := 0)
            

    #Switch Ownership of L18_Pet4 to 1
    L18_Pet4_Own(Agent:agent):void=
        Print("Ownership L18P4 Saved")
        if(CurrentPlayerStats := PlayerStatsManager.GetPlayerStats[Agent]):
            PlayerStatsManager.RecordOwnershipL18P4(Agent, StatType.Ownership_L18_Pet4, ?Ownership_L18_Pet4 := 1)
            

    #Switch Ownership of L18_Pet4 to 0
    L18_Pet4_DesOwn(Agent:agent):void=
        Print("DesOwnership L18P4 Saved")
        if(CurrentPlayerStats := PlayerStatsManager.GetPlayerStats[Agent]):
            PlayerStatsManager.RemoveOwnershipL18P4(Agent, StatType.Ownership_L18_Pet4, ?Ownership_L18_Pet4 := 0) 

    #---------------------------------------L19
    #Switch Ownership of L19_Pet1 to 1
    L19_Pet1_Own(Agent:agent):void=
        Print("Ownership L19P1 Saved")
        if(CurrentPlayerStats := PlayerStatsManager.GetPlayerStats[Agent]):
            PlayerStatsManager.RecordOwnershipL19P1(Agent, StatType.Ownership_L19_Pet1, ?Ownership_L19_Pet1 := 1)
            

    #Switch Ownership of L19_Pet1 to 0
    L19_Pet1_DesOwn(Agent:agent):void=
        Print("DesOwnership L19P1 Saved")
        if(CurrentPlayerStats := PlayerStatsManager.GetPlayerStats[Agent]):
            PlayerStatsManager.RemoveOwnershipL19P1(Agent, StatType.Ownership_L19_Pet1, ?Ownership_L19_Pet1 := 0)
            

    #Switch Ownership of L19_Pet2 to 1
    L19_Pet2_Own(Agent:agent):void=
        Print("Ownership L19P2 Saved")
        if(CurrentPlayerStats := PlayerStatsManager.GetPlayerStats[Agent]):
            PlayerStatsManager.RecordOwnershipL19P2(Agent, StatType.Ownership_L19_Pet2, ?Ownership_L19_Pet2 := 1)
            

    #Switch Ownership of L19_Pet2 to 0
    L19_Pet2_DesOwn(Agent:agent):void=
        Print("DesOwnership L19P2 Saved")
        if(CurrentPlayerStats := PlayerStatsManager.GetPlayerStats[Agent]):
            PlayerStatsManager.RemoveOwnershipL19P2(Agent, StatType.Ownership_L19_Pet2, ?Ownership_L19_Pet2 := 0)
            

    #Switch Ownership of L19_Pet3 to 1
    L19_Pet3_Own(Agent:agent):void=
        Print("Ownership L19P3 Saved")
        if(CurrentPlayerStats := PlayerStatsManager.GetPlayerStats[Agent]):
            PlayerStatsManager.RecordOwnershipL19P3(Agent, StatType.Ownership_L19_Pet3, ?Ownership_L19_Pet3 := 1)
            

    #Switch Ownership of L19_Pet3 to 0
    L19_Pet3_DesOwn(Agent:agent):void=
        Print("DesOwnership L19P3 Saved")
        if(CurrentPlayerStats := PlayerStatsManager.GetPlayerStats[Agent]):
            PlayerStatsManager.RemoveOwnershipL19P3(Agent, StatType.Ownership_L19_Pet3, ?Ownership_L19_Pet3 := 0)
            

    #Switch Ownership of L19_Pet4 to 1
    L19_Pet4_Own(Agent:agent):void=
        Print("Ownership L19P4 Saved")
        if(CurrentPlayerStats := PlayerStatsManager.GetPlayerStats[Agent]):
            PlayerStatsManager.RecordOwnershipL19P4(Agent, StatType.Ownership_L19_Pet4, ?Ownership_L19_Pet4 := 1)
            

    #Switch Ownership of L19_Pet4 to 0
    L19_Pet4_DesOwn(Agent:agent):void=
        Print("DesOwnership L19P4 Saved")
        if(CurrentPlayerStats := PlayerStatsManager.GetPlayerStats[Agent]):
            PlayerStatsManager.RemoveOwnershipL19P4(Agent, StatType.Ownership_L19_Pet4, ?Ownership_L19_Pet4 := 0) 
    
    #---------------------------------------L20
    #Switch Ownership of L20_Pet1 to 1
    L20_Pet1_Own(Agent:agent):void=
        Print("Ownership L20P1 Saved")
        if(CurrentPlayerStats := PlayerStatsManager.GetPlayerStats[Agent]):
            PlayerStatsManager.RecordOwnershipL20P1(Agent, StatType.Ownership_L20_Pet1, ?Ownership_L20_Pet1 := 1)
            

    #Switch Ownership of L20_Pet1 to 0
    L20_Pet1_DesOwn(Agent:agent):void=
        Print("DesOwnership L20P1 Saved")
        if(CurrentPlayerStats := PlayerStatsManager.GetPlayerStats[Agent]):
            PlayerStatsManager.RemoveOwnershipL20P1(Agent, StatType.Ownership_L20_Pet1, ?Ownership_L20_Pet1 := 0)
            

    #Switch Ownership of L20_Pet2 to 1
    L20_Pet2_Own(Agent:agent):void=
        Print("Ownership L20P2 Saved")
        if(CurrentPlayerStats := PlayerStatsManager.GetPlayerStats[Agent]):
            PlayerStatsManager.RecordOwnershipL20P2(Agent, StatType.Ownership_L20_Pet2, ?Ownership_L20_Pet2 := 1)
            

    #Switch Ownership of L20_Pet2 to 0
    L20_Pet2_DesOwn(Agent:agent):void=
        Print("DesOwnership L20P2 Saved")
        if(CurrentPlayerStats := PlayerStatsManager.GetPlayerStats[Agent]):
            PlayerStatsManager.RemoveOwnershipL20P2(Agent, StatType.Ownership_L20_Pet2, ?Ownership_L20_Pet2 := 0)
            

    #Switch Ownership of L20_Pet3 to 1
    L20_Pet3_Own(Agent:agent):void=
        Print("Ownership L20P3 Saved")
        if(CurrentPlayerStats := PlayerStatsManager.GetPlayerStats[Agent]):
            PlayerStatsManager.RecordOwnershipL20P3(Agent, StatType.Ownership_L20_Pet3, ?Ownership_L20_Pet3 := 1)
            

    #Switch Ownership of L20_Pet3 to 0
    L20_Pet3_DesOwn(Agent:agent):void=
        Print("DesOwnership L20P3 Saved")
        if(CurrentPlayerStats := PlayerStatsManager.GetPlayerStats[Agent]):
            PlayerStatsManager.RemoveOwnershipL20P3(Agent, StatType.Ownership_L20_Pet3, ?Ownership_L20_Pet3 := 0)
            

    #Switch Ownership of L20_Pet4 to 1
    L20_Pet4_Own(Agent:agent):void=
        Print("Ownership L20P4 Saved")
        if(CurrentPlayerStats := PlayerStatsManager.GetPlayerStats[Agent]):
            PlayerStatsManager.RecordOwnershipL20P4(Agent, StatType.Ownership_L20_Pet4, ?Ownership_L20_Pet4 := 1)
            

    #Switch Ownership of L20_Pet4 to 0
    L20_Pet4_DesOwn(Agent:agent):void=
        Print("DesOwnership L20P4 Saved")
        if(CurrentPlayerStats := PlayerStatsManager.GetPlayerStats[Agent]):
            PlayerStatsManager.RemoveOwnershipL20P4(Agent, StatType.Ownership_L20_Pet4, ?Ownership_L20_Pet4 := 0) 

    #---------------------------------------L21
    #Switch Ownership of L21_Pet1 to 1
    L21_Pet1_Own(Agent:agent):void=
        Print("Ownership L21P1 Saved")
        if(CurrentPlayerStats := PlayerStatsManager.GetPlayerStats[Agent]):
            PlayerStatsManager.RecordOwnershipL21P1(Agent, StatType.Ownership_L21_Pet1, ?Ownership_L21_Pet1 := 1)
            

    #Switch Ownership of L21_Pet1 to 0
    L21_Pet1_DesOwn(Agent:agent):void=
        Print("DesOwnership L21P1 Saved")
        if(CurrentPlayerStats := PlayerStatsManager.GetPlayerStats[Agent]):
            PlayerStatsManager.RemoveOwnershipL21P1(Agent, StatType.Ownership_L21_Pet1, ?Ownership_L21_Pet1 := 0)
            

    #Switch Ownership of L21_Pet2 to 1
    L21_Pet2_Own(Agent:agent):void=
        Print("Ownership L21P2 Saved")
        if(CurrentPlayerStats := PlayerStatsManager.GetPlayerStats[Agent]):
            PlayerStatsManager.RecordOwnershipL21P2(Agent, StatType.Ownership_L21_Pet2, ?Ownership_L21_Pet2 := 1)
            

    #Switch Ownership of L21_Pet2 to 0
    L21_Pet2_DesOwn(Agent:agent):void=
        Print("DesOwnership L21P2 Saved")
        if(CurrentPlayerStats := PlayerStatsManager.GetPlayerStats[Agent]):
            PlayerStatsManager.RemoveOwnershipL21P2(Agent, StatType.Ownership_L21_Pet2, ?Ownership_L21_Pet2 := 0)
            

    #Switch Ownership of L21_Pet3 to 1
    L21_Pet3_Own(Agent:agent):void=
        Print("Ownership L21P3 Saved")
        if(CurrentPlayerStats := PlayerStatsManager.GetPlayerStats[Agent]):
            PlayerStatsManager.RecordOwnershipL21P3(Agent, StatType.Ownership_L21_Pet3, ?Ownership_L21_Pet3 := 1)
            

    #Switch Ownership of L21_Pet3 to 0
    L21_Pet3_DesOwn(Agent:agent):void=
        Print("DesOwnership L21P3 Saved")
        if(CurrentPlayerStats := PlayerStatsManager.GetPlayerStats[Agent]):
            PlayerStatsManager.RemoveOwnershipL21P3(Agent, StatType.Ownership_L21_Pet3, ?Ownership_L21_Pet3 := 0)
            

    #Switch Ownership of L21_Pet4 to 1
    L21_Pet4_Own(Agent:agent):void=
        Print("Ownership L21P4 Saved")
        if(CurrentPlayerStats := PlayerStatsManager.GetPlayerStats[Agent]):
            PlayerStatsManager.RecordOwnershipL21P4(Agent, StatType.Ownership_L21_Pet4, ?Ownership_L21_Pet4 := 1)
            

    #Switch Ownership of L21_Pet4 to 0
    L21_Pet4_DesOwn(Agent:agent):void=
        Print("DesOwnership L21P4 Saved")
        if(CurrentPlayerStats := PlayerStatsManager.GetPlayerStats[Agent]):
            PlayerStatsManager.RemoveOwnershipL21P4(Agent, StatType.Ownership_L21_Pet4, ?Ownership_L21_Pet4 := 0) 

    #---------------------------------------L22
    #Switch Ownership of L22_Pet1 to 1
    L22_Pet1_Own(Agent:agent):void=
        Print("Ownership L22P1 Saved")
        if(CurrentPlayerStats := PlayerStatsManager.GetPlayerStats[Agent]):
            PlayerStatsManager.RecordOwnershipL22P1(Agent, StatType.Ownership_L22_Pet1, ?Ownership_L22_Pet1 := 1)
            

    #Switch Ownership of L22_Pet1 to 0
    L22_Pet1_DesOwn(Agent:agent):void=
        Print("DesOwnership L22P1 Saved")
        if(CurrentPlayerStats := PlayerStatsManager.GetPlayerStats[Agent]):
            PlayerStatsManager.RemoveOwnershipL22P1(Agent, StatType.Ownership_L22_Pet1, ?Ownership_L22_Pet1 := 0)
            

    #Switch Ownership of L22_Pet2 to 1
    L22_Pet2_Own(Agent:agent):void=
        Print("Ownership L22P2 Saved")
        if(CurrentPlayerStats := PlayerStatsManager.GetPlayerStats[Agent]):
            PlayerStatsManager.RecordOwnershipL22P2(Agent, StatType.Ownership_L22_Pet2, ?Ownership_L22_Pet2 := 1)
            

    #Switch Ownership of L22_Pet2 to 0
    L22_Pet2_DesOwn(Agent:agent):void=
        Print("DesOwnership L22P2 Saved")
        if(CurrentPlayerStats := PlayerStatsManager.GetPlayerStats[Agent]):
            PlayerStatsManager.RemoveOwnershipL22P2(Agent, StatType.Ownership_L22_Pet2, ?Ownership_L22_Pet2 := 0)
            

    #Switch Ownership of L22_Pet3 to 1
    L22_Pet3_Own(Agent:agent):void=
        Print("Ownership L22P3 Saved")
        if(CurrentPlayerStats := PlayerStatsManager.GetPlayerStats[Agent]):
            PlayerStatsManager.RecordOwnershipL22P3(Agent, StatType.Ownership_L22_Pet3, ?Ownership_L22_Pet3 := 1)
            

    #Switch Ownership of L22_Pet3 to 0
    L22_Pet3_DesOwn(Agent:agent):void=
        Print("DesOwnership L22P3 Saved")
        if(CurrentPlayerStats := PlayerStatsManager.GetPlayerStats[Agent]):
            PlayerStatsManager.RemoveOwnershipL22P3(Agent, StatType.Ownership_L22_Pet3, ?Ownership_L22_Pet3 := 0)
            

    #Switch Ownership of L22_Pet4 to 1
    L22_Pet4_Own(Agent:agent):void=
        Print("Ownership L22P4 Saved")
        if(CurrentPlayerStats := PlayerStatsManager.GetPlayerStats[Agent]):
            PlayerStatsManager.RecordOwnershipL22P4(Agent, StatType.Ownership_L22_Pet4, ?Ownership_L22_Pet4 := 1)
            

    #Switch Ownership of L22_Pet4 to 0
    L22_Pet4_DesOwn(Agent:agent):void=
        Print("DesOwnership L22P4 Saved")
        if(CurrentPlayerStats := PlayerStatsManager.GetPlayerStats[Agent]):
            PlayerStatsManager.RemoveOwnershipL22P4(Agent, StatType.Ownership_L22_Pet4, ?Ownership_L22_Pet4 := 0) 

    #---------------------------------------L23
    #Switch Ownership of L23_Pet1 to 1
    L23_Pet1_Own(Agent:agent):void=
        Print("Ownership L23P1 Saved")
        if(CurrentPlayerStats := PlayerStatsManager.GetPlayerStats[Agent]):
            PlayerStatsManager.RecordOwnershipL23P1(Agent, StatType.Ownership_L23_Pet1, ?Ownership_L23_Pet1 := 1)
            

    #Switch Ownership of L23_Pet1 to 0
    L23_Pet1_DesOwn(Agent:agent):void=
        Print("DesOwnership L23P1 Saved")
        if(CurrentPlayerStats := PlayerStatsManager.GetPlayerStats[Agent]):
            PlayerStatsManager.RemoveOwnershipL23P1(Agent, StatType.Ownership_L23_Pet1, ?Ownership_L23_Pet1 := 0)
            

    #Switch Ownership of L23_Pet2 to 1
    L23_Pet2_Own(Agent:agent):void=
        Print("Ownership L23P2 Saved")
        if(CurrentPlayerStats := PlayerStatsManager.GetPlayerStats[Agent]):
            PlayerStatsManager.RecordOwnershipL23P2(Agent, StatType.Ownership_L23_Pet2, ?Ownership_L23_Pet2 := 1)
            

    #Switch Ownership of L23_Pet2 to 0
    L23_Pet2_DesOwn(Agent:agent):void=
        Print("DesOwnership L23P2 Saved")
        if(CurrentPlayerStats := PlayerStatsManager.GetPlayerStats[Agent]):
            PlayerStatsManager.RemoveOwnershipL23P2(Agent, StatType.Ownership_L23_Pet2, ?Ownership_L23_Pet2 := 0)
            

    #Switch Ownership of L23_Pet3 to 1
    L23_Pet3_Own(Agent:agent):void=
        Print("Ownership L23P3 Saved")
        if(CurrentPlayerStats := PlayerStatsManager.GetPlayerStats[Agent]):
            PlayerStatsManager.RecordOwnershipL23P3(Agent, StatType.Ownership_L23_Pet3, ?Ownership_L23_Pet3 := 1)
            

    #Switch Ownership of L23_Pet3 to 0
    L23_Pet3_DesOwn(Agent:agent):void=
        Print("DesOwnership L23P3 Saved")
        if(CurrentPlayerStats := PlayerStatsManager.GetPlayerStats[Agent]):
            PlayerStatsManager.RemoveOwnershipL23P3(Agent, StatType.Ownership_L23_Pet3, ?Ownership_L23_Pet3 := 0)
            

    #Switch Ownership of L23_Pet4 to 1
    L23_Pet4_Own(Agent:agent):void=
        Print("Ownership L23P4 Saved")
        if(CurrentPlayerStats := PlayerStatsManager.GetPlayerStats[Agent]):
            PlayerStatsManager.RecordOwnershipL23P4(Agent, StatType.Ownership_L23_Pet4, ?Ownership_L23_Pet4 := 1)
            

    #Switch Ownership of L23_Pet4 to 0
    L23_Pet4_DesOwn(Agent:agent):void=
        Print("DesOwnership L23P4 Saved")
        if(CurrentPlayerStats := PlayerStatsManager.GetPlayerStats[Agent]):
            PlayerStatsManager.RemoveOwnershipL23P4(Agent, StatType.Ownership_L23_Pet4, ?Ownership_L23_Pet4 := 0) 

    #---------------------------------------L24
    #Switch Ownership of L24_Pet1 to 1
    L24_Pet1_Own(Agent:agent):void=
        Print("Ownership L24P1 Saved")
        if(CurrentPlayerStats := PlayerStatsManager.GetPlayerStats[Agent]):
            PlayerStatsManager.RecordOwnershipL24P1(Agent, StatType.Ownership_L24_Pet1, ?Ownership_L24_Pet1 := 1)
            

    #Switch Ownership of L24_Pet1 to 0
    L24_Pet1_DesOwn(Agent:agent):void=
        Print("DesOwnership L24P1 Saved")
        if(CurrentPlayerStats := PlayerStatsManager.GetPlayerStats[Agent]):
            PlayerStatsManager.RemoveOwnershipL24P1(Agent, StatType.Ownership_L24_Pet1, ?Ownership_L24_Pet1 := 0)
            

    #Switch Ownership of L24_Pet2 to 1
    L24_Pet2_Own(Agent:agent):void=
        Print("Ownership L24P2 Saved")
        if(CurrentPlayerStats := PlayerStatsManager.GetPlayerStats[Agent]):
            PlayerStatsManager.RecordOwnershipL24P2(Agent, StatType.Ownership_L24_Pet2, ?Ownership_L24_Pet2 := 1)
            

    #Switch Ownership of L24_Pet2 to 0
    L24_Pet2_DesOwn(Agent:agent):void=
        Print("DesOwnership L24P2 Saved")
        if(CurrentPlayerStats := PlayerStatsManager.GetPlayerStats[Agent]):
            PlayerStatsManager.RemoveOwnershipL24P2(Agent, StatType.Ownership_L24_Pet2, ?Ownership_L24_Pet2 := 0)
            

    #Switch Ownership of L24_Pet3 to 1
    L24_Pet3_Own(Agent:agent):void=
        Print("Ownership L24P3 Saved")
        if(CurrentPlayerStats := PlayerStatsManager.GetPlayerStats[Agent]):
            PlayerStatsManager.RecordOwnershipL24P3(Agent, StatType.Ownership_L24_Pet3, ?Ownership_L24_Pet3 := 1)
            

    #Switch Ownership of L24_Pet3 to 0
    L24_Pet3_DesOwn(Agent:agent):void=
        Print("DesOwnership L24P3 Saved")
        if(CurrentPlayerStats := PlayerStatsManager.GetPlayerStats[Agent]):
            PlayerStatsManager.RemoveOwnershipL24P3(Agent, StatType.Ownership_L24_Pet3, ?Ownership_L24_Pet3 := 0)
            

    #Switch Ownership of L24_Pet4 to 1
    L24_Pet4_Own(Agent:agent):void=
        Print("Ownership L24P4 Saved")
        if(CurrentPlayerStats := PlayerStatsManager.GetPlayerStats[Agent]):
            PlayerStatsManager.RecordOwnershipL24P4(Agent, StatType.Ownership_L24_Pet4, ?Ownership_L24_Pet4 := 1)
            

    #Switch Ownership of L24_Pet4 to 0
    L24_Pet4_DesOwn(Agent:agent):void=
        Print("DesOwnership L24P4 Saved")
        if(CurrentPlayerStats := PlayerStatsManager.GetPlayerStats[Agent]):
            PlayerStatsManager.RemoveOwnershipL24P4(Agent, StatType.Ownership_L24_Pet4, ?Ownership_L24_Pet4 := 0) 

    #---------------------------------------L25
    #Switch Ownership of L25_Pet1 to 1
    L25_Pet1_Own(Agent:agent):void=
        Print("Ownership L25P1 Saved")
        if(CurrentPlayerStats := PlayerStatsManager.GetPlayerStats[Agent]):
            PlayerStatsManager.RecordOwnershipL25P1(Agent, StatType.Ownership_L25_Pet1, ?Ownership_L25_Pet1 := 1)
            

    #Switch Ownership of L25_Pet1 to 0
    L25_Pet1_DesOwn(Agent:agent):void=
        Print("DesOwnership L25P1 Saved")
        if(CurrentPlayerStats := PlayerStatsManager.GetPlayerStats[Agent]):
            PlayerStatsManager.RemoveOwnershipL25P1(Agent, StatType.Ownership_L25_Pet1, ?Ownership_L25_Pet1 := 0)
            

    #Switch Ownership of L25_Pet2 to 1
    L25_Pet2_Own(Agent:agent):void=
        Print("Ownership L25P2 Saved")
        if(CurrentPlayerStats := PlayerStatsManager.GetPlayerStats[Agent]):
            PlayerStatsManager.RecordOwnershipL25P2(Agent, StatType.Ownership_L25_Pet2, ?Ownership_L25_Pet2 := 1)
            

    #Switch Ownership of L25_Pet2 to 0
    L25_Pet2_DesOwn(Agent:agent):void=
        Print("DesOwnership L25P2 Saved")
        if(CurrentPlayerStats := PlayerStatsManager.GetPlayerStats[Agent]):
            PlayerStatsManager.RemoveOwnershipL25P2(Agent, StatType.Ownership_L25_Pet2, ?Ownership_L25_Pet2 := 0)
            

    #Switch Ownership of L25_Pet3 to 1
    L25_Pet3_Own(Agent:agent):void=
        Print("Ownership L25P3 Saved")
        if(CurrentPlayerStats := PlayerStatsManager.GetPlayerStats[Agent]):
            PlayerStatsManager.RecordOwnershipL25P3(Agent, StatType.Ownership_L25_Pet3, ?Ownership_L25_Pet3 := 1)
            

    #Switch Ownership of L25_Pet3 to 0
    L25_Pet3_DesOwn(Agent:agent):void=
        Print("DesOwnership L25P3 Saved")
        if(CurrentPlayerStats := PlayerStatsManager.GetPlayerStats[Agent]):
            PlayerStatsManager.RemoveOwnershipL25P3(Agent, StatType.Ownership_L25_Pet3, ?Ownership_L25_Pet3 := 0)
            

    #Switch Ownership of L25_Pet4 to 1
    L25_Pet4_Own(Agent:agent):void=
        Print("Ownership L25P4 Saved")
        if(CurrentPlayerStats := PlayerStatsManager.GetPlayerStats[Agent]):
            PlayerStatsManager.RecordOwnershipL25P4(Agent, StatType.Ownership_L25_Pet4, ?Ownership_L25_Pet4 := 1)
            

    #Switch Ownership of L25_Pet4 to 0
    L25_Pet4_DesOwn(Agent:agent):void=
        Print("DesOwnership L25P4 Saved")
        if(CurrentPlayerStats := PlayerStatsManager.GetPlayerStats[Agent]):
            PlayerStatsManager.RemoveOwnershipL25P4(Agent, StatType.Ownership_L25_Pet4, ?Ownership_L25_Pet4 := 0) 

    #---------------------------------------L26
    #Switch Ownership of L26_Pet1 to 1
    L26_Pet1_Own(Agent:agent):void=
        Print("Ownership L26P1 Saved")
        if(CurrentPlayerStats := PlayerStatsManager.GetPlayerStats[Agent]):
            PlayerStatsManager.RecordOwnershipL26P1(Agent, StatType.Ownership_L26_Pet1, ?Ownership_L26_Pet1 := 1)
            

    #Switch Ownership of L26_Pet1 to 0
    L26_Pet1_DesOwn(Agent:agent):void=
        Print("DesOwnership L26P1 Saved")
        if(CurrentPlayerStats := PlayerStatsManager.GetPlayerStats[Agent]):
            PlayerStatsManager.RemoveOwnershipL26P1(Agent, StatType.Ownership_L26_Pet1, ?Ownership_L26_Pet1 := 0)
            

    #Switch Ownership of L26_Pet2 to 1
    L26_Pet2_Own(Agent:agent):void=
        Print("Ownership L26P2 Saved")
        if(CurrentPlayerStats := PlayerStatsManager.GetPlayerStats[Agent]):
            PlayerStatsManager.RecordOwnershipL26P2(Agent, StatType.Ownership_L26_Pet2, ?Ownership_L26_Pet2 := 1)
            

    #Switch Ownership of L26_Pet2 to 0
    L26_Pet2_DesOwn(Agent:agent):void=
        Print("DesOwnership L26P2 Saved")
        if(CurrentPlayerStats := PlayerStatsManager.GetPlayerStats[Agent]):
            PlayerStatsManager.RemoveOwnershipL26P2(Agent, StatType.Ownership_L26_Pet2, ?Ownership_L26_Pet2 := 0)
            

    #Switch Ownership of L26_Pet3 to 1
    L26_Pet3_Own(Agent:agent):void=
        Print("Ownership L26P3 Saved")
        if(CurrentPlayerStats := PlayerStatsManager.GetPlayerStats[Agent]):
            PlayerStatsManager.RecordOwnershipL26P3(Agent, StatType.Ownership_L26_Pet3, ?Ownership_L26_Pet3 := 1)
            

    #Switch Ownership of L26_Pet3 to 0
    L26_Pet3_DesOwn(Agent:agent):void=
        Print("DesOwnership L26P3 Saved")
        if(CurrentPlayerStats := PlayerStatsManager.GetPlayerStats[Agent]):
            PlayerStatsManager.RemoveOwnershipL26P3(Agent, StatType.Ownership_L26_Pet3, ?Ownership_L26_Pet3 := 0)
            

    #Switch Ownership of L26_Pet4 to 1
    L26_Pet4_Own(Agent:agent):void=
        Print("Ownership L26P4 Saved")
        if(CurrentPlayerStats := PlayerStatsManager.GetPlayerStats[Agent]):
            PlayerStatsManager.RecordOwnershipL26P4(Agent, StatType.Ownership_L26_Pet4, ?Ownership_L26_Pet4 := 1)
            

    #Switch Ownership of L26_Pet4 to 0
    L26_Pet4_DesOwn(Agent:agent):void=
        Print("DesOwnership L26P4 Saved")
        if(CurrentPlayerStats := PlayerStatsManager.GetPlayerStats[Agent]):
            PlayerStatsManager.RemoveOwnershipL26P4(Agent, StatType.Ownership_L26_Pet4, ?Ownership_L26_Pet4 := 0) 

    #---------------------------------------L27
    #Switch Ownership of L27_Pet1 to 1
    L27_Pet1_Own(Agent:agent):void=
        Print("Ownership L27P1 Saved")
        if(CurrentPlayerStats := PlayerStatsManager.GetPlayerStats[Agent]):
            PlayerStatsManager.RecordOwnershipL27P1(Agent, StatType.Ownership_L27_Pet1, ?Ownership_L27_Pet1 := 1)
            

    #Switch Ownership of L27_Pet1 to 0
    L27_Pet1_DesOwn(Agent:agent):void=
        Print("DesOwnership L27P1 Saved")
        if(CurrentPlayerStats := PlayerStatsManager.GetPlayerStats[Agent]):
            PlayerStatsManager.RemoveOwnershipL27P1(Agent, StatType.Ownership_L27_Pet1, ?Ownership_L27_Pet1 := 0)
            

    #Switch Ownership of L27_Pet2 to 1
    L27_Pet2_Own(Agent:agent):void=
        Print("Ownership L27P2 Saved")
        if(CurrentPlayerStats := PlayerStatsManager.GetPlayerStats[Agent]):
            PlayerStatsManager.RecordOwnershipL27P2(Agent, StatType.Ownership_L27_Pet2, ?Ownership_L27_Pet2 := 1)
            

    #Switch Ownership of L27_Pet2 to 0
    L27_Pet2_DesOwn(Agent:agent):void=
        Print("DesOwnership L27P2 Saved")
        if(CurrentPlayerStats := PlayerStatsManager.GetPlayerStats[Agent]):
            PlayerStatsManager.RemoveOwnershipL27P2(Agent, StatType.Ownership_L27_Pet2, ?Ownership_L27_Pet2 := 0)
            

    #Switch Ownership of L27_Pet3 to 1
    L27_Pet3_Own(Agent:agent):void=
        Print("Ownership L27P3 Saved")
        if(CurrentPlayerStats := PlayerStatsManager.GetPlayerStats[Agent]):
            PlayerStatsManager.RecordOwnershipL27P3(Agent, StatType.Ownership_L27_Pet3, ?Ownership_L27_Pet3 := 1)
            

    #Switch Ownership of L27_Pet3 to 0
    L27_Pet3_DesOwn(Agent:agent):void=
        Print("DesOwnership L27P3 Saved")
        if(CurrentPlayerStats := PlayerStatsManager.GetPlayerStats[Agent]):
            PlayerStatsManager.RemoveOwnershipL27P3(Agent, StatType.Ownership_L27_Pet3, ?Ownership_L27_Pet3 := 0)
            

    #Switch Ownership of L27_Pet4 to 1
    L27_Pet4_Own(Agent:agent):void=
        Print("Ownership L27P4 Saved")
        if(CurrentPlayerStats := PlayerStatsManager.GetPlayerStats[Agent]):
            PlayerStatsManager.RecordOwnershipL27P4(Agent, StatType.Ownership_L27_Pet4, ?Ownership_L27_Pet4 := 1)
            

    #Switch Ownership of L27_Pet4 to 0
    L27_Pet4_DesOwn(Agent:agent):void=
        Print("DesOwnership L27P4 Saved")
        if(CurrentPlayerStats := PlayerStatsManager.GetPlayerStats[Agent]):
            PlayerStatsManager.RemoveOwnershipL27P4(Agent, StatType.Ownership_L27_Pet4, ?Ownership_L27_Pet4 := 0) 

    #---------------------------------------L28
    #Switch Ownership of L28_Pet1 to 1
    L28_Pet1_Own(Agent:agent):void=
        Print("Ownership L28P1 Saved")
        if(CurrentPlayerStats := PlayerStatsManager.GetPlayerStats[Agent]):
            PlayerStatsManager.RecordOwnershipL28P1(Agent, StatType.Ownership_L28_Pet1, ?Ownership_L28_Pet1 := 1)
            

    #Switch Ownership of L28_Pet1 to 0
    L28_Pet1_DesOwn(Agent:agent):void=
        Print("DesOwnership L28P1 Saved")
        if(CurrentPlayerStats := PlayerStatsManager.GetPlayerStats[Agent]):
            PlayerStatsManager.RemoveOwnershipL28P1(Agent, StatType.Ownership_L28_Pet1, ?Ownership_L28_Pet1 := 0)
            

    #Switch Ownership of L28_Pet2 to 1
    L28_Pet2_Own(Agent:agent):void=
        Print("Ownership L28P2 Saved")
        if(CurrentPlayerStats := PlayerStatsManager.GetPlayerStats[Agent]):
            PlayerStatsManager.RecordOwnershipL28P2(Agent, StatType.Ownership_L28_Pet2, ?Ownership_L28_Pet2 := 1)
            

    #Switch Ownership of L28_Pet2 to 0
    L28_Pet2_DesOwn(Agent:agent):void=
        Print("DesOwnership L28P2 Saved")
        if(CurrentPlayerStats := PlayerStatsManager.GetPlayerStats[Agent]):
            PlayerStatsManager.RemoveOwnershipL28P2(Agent, StatType.Ownership_L28_Pet2, ?Ownership_L28_Pet2 := 0)
            

    #Switch Ownership of L28_Pet3 to 1
    L28_Pet3_Own(Agent:agent):void=
        Print("Ownership L28P3 Saved")
        if(CurrentPlayerStats := PlayerStatsManager.GetPlayerStats[Agent]):
            PlayerStatsManager.RecordOwnershipL28P3(Agent, StatType.Ownership_L28_Pet3, ?Ownership_L28_Pet3 := 1)
            

    #Switch Ownership of L28_Pet3 to 0
    L28_Pet3_DesOwn(Agent:agent):void=
        Print("DesOwnership L28P3 Saved")
        if(CurrentPlayerStats := PlayerStatsManager.GetPlayerStats[Agent]):
            PlayerStatsManager.RemoveOwnershipL28P3(Agent, StatType.Ownership_L28_Pet3, ?Ownership_L28_Pet3 := 0)
            

    #Switch Ownership of L28_Pet4 to 1
    L28_Pet4_Own(Agent:agent):void=
        Print("Ownership L28P4 Saved")
        if(CurrentPlayerStats := PlayerStatsManager.GetPlayerStats[Agent]):
            PlayerStatsManager.RecordOwnershipL28P4(Agent, StatType.Ownership_L28_Pet4, ?Ownership_L28_Pet4 := 1)
            

    #Switch Ownership of L28_Pet4 to 0
    L28_Pet4_DesOwn(Agent:agent):void=
        Print("DesOwnership L28P4 Saved")
        if(CurrentPlayerStats := PlayerStatsManager.GetPlayerStats[Agent]):
            PlayerStatsManager.RemoveOwnershipL28P4(Agent, StatType.Ownership_L28_Pet4, ?Ownership_L28_Pet4 := 0) 

    #---------------------------------------L29
    #Switch Ownership of L29_Pet1 to 1
    L29_Pet1_Own(Agent:agent):void=
        Print("Ownership L29P1 Saved")
        if(CurrentPlayerStats := PlayerStatsManager.GetPlayerStats[Agent]):
            PlayerStatsManager.RecordOwnershipL29P1(Agent, StatType.Ownership_L29_Pet1, ?Ownership_L29_Pet1 := 1)
            

    #Switch Ownership of L29_Pet1 to 0
    L29_Pet1_DesOwn(Agent:agent):void=
        Print("DesOwnership L29P1 Saved")
        if(CurrentPlayerStats := PlayerStatsManager.GetPlayerStats[Agent]):
            PlayerStatsManager.RemoveOwnershipL29P1(Agent, StatType.Ownership_L29_Pet1, ?Ownership_L29_Pet1 := 0)
            

    #Switch Ownership of L29_Pet2 to 1
    L29_Pet2_Own(Agent:agent):void=
        Print("Ownership L29P2 Saved")
        if(CurrentPlayerStats := PlayerStatsManager.GetPlayerStats[Agent]):
            PlayerStatsManager.RecordOwnershipL29P2(Agent, StatType.Ownership_L29_Pet2, ?Ownership_L29_Pet2 := 1)
            

    #Switch Ownership of L29_Pet2 to 0
    L29_Pet2_DesOwn(Agent:agent):void=
        Print("DesOwnership L29P2 Saved")
        if(CurrentPlayerStats := PlayerStatsManager.GetPlayerStats[Agent]):
            PlayerStatsManager.RemoveOwnershipL29P2(Agent, StatType.Ownership_L29_Pet2, ?Ownership_L29_Pet2 := 0)
            

    #Switch Ownership of L29_Pet3 to 1
    L29_Pet3_Own(Agent:agent):void=
        Print("Ownership L29P3 Saved")
        if(CurrentPlayerStats := PlayerStatsManager.GetPlayerStats[Agent]):
            PlayerStatsManager.RecordOwnershipL29P3(Agent, StatType.Ownership_L29_Pet3, ?Ownership_L29_Pet3 := 1)
            

    #Switch Ownership of L29_Pet3 to 0
    L29_Pet3_DesOwn(Agent:agent):void=
        Print("DesOwnership L29P3 Saved")
        if(CurrentPlayerStats := PlayerStatsManager.GetPlayerStats[Agent]):
            PlayerStatsManager.RemoveOwnershipL29P3(Agent, StatType.Ownership_L29_Pet3, ?Ownership_L29_Pet3 := 0)
            

    #Switch Ownership of L29_Pet4 to 1
    L29_Pet4_Own(Agent:agent):void=
        Print("Ownership L29P4 Saved")
        if(CurrentPlayerStats := PlayerStatsManager.GetPlayerStats[Agent]):
            PlayerStatsManager.RecordOwnershipL29P4(Agent, StatType.Ownership_L29_Pet4, ?Ownership_L29_Pet4 := 1)
            

    #Switch Ownership of L29_Pet4 to 0
    L29_Pet4_DesOwn(Agent:agent):void=
        Print("DesOwnership L29P4 Saved")
        if(CurrentPlayerStats := PlayerStatsManager.GetPlayerStats[Agent]):
            PlayerStatsManager.RemoveOwnershipL29P4(Agent, StatType.Ownership_L29_Pet4, ?Ownership_L29_Pet4 := 0) 

    #---------------------------------------L30
    #Switch Ownership of L30_Pet1 to 1
    L30_Pet1_Own(Agent:agent):void=
        Print("Ownership L30P1 Saved")
        if(CurrentPlayerStats := PlayerStatsManager.GetPlayerStats[Agent]):
            PlayerStatsManager.RecordOwnershipL30P1(Agent, StatType.Ownership_L30_Pet1, ?Ownership_L30_Pet1 := 1)
            

    #Switch Ownership of L30_Pet1 to 0
    L30_Pet1_DesOwn(Agent:agent):void=
        Print("DesOwnership L30P1 Saved")
        if(CurrentPlayerStats := PlayerStatsManager.GetPlayerStats[Agent]):
            PlayerStatsManager.RemoveOwnershipL30P1(Agent, StatType.Ownership_L30_Pet1, ?Ownership_L30_Pet1 := 0)
            

    #Switch Ownership of L30_Pet2 to 1
    L30_Pet2_Own(Agent:agent):void=
        Print("Ownership L30P2 Saved")
        if(CurrentPlayerStats := PlayerStatsManager.GetPlayerStats[Agent]):
            PlayerStatsManager.RecordOwnershipL30P2(Agent, StatType.Ownership_L30_Pet2, ?Ownership_L30_Pet2 := 1)
            

    #Switch Ownership of L30_Pet2 to 0
    L30_Pet2_DesOwn(Agent:agent):void=
        Print("DesOwnership L30P2 Saved")
        if(CurrentPlayerStats := PlayerStatsManager.GetPlayerStats[Agent]):
            PlayerStatsManager.RemoveOwnershipL30P2(Agent, StatType.Ownership_L30_Pet2, ?Ownership_L30_Pet2 := 0)
            

    #Switch Ownership of L30_Pet3 to 1
    L30_Pet3_Own(Agent:agent):void=
        Print("Ownership L30P3 Saved")
        if(CurrentPlayerStats := PlayerStatsManager.GetPlayerStats[Agent]):
            PlayerStatsManager.RecordOwnershipL30P3(Agent, StatType.Ownership_L30_Pet3, ?Ownership_L30_Pet3 := 1)
            

    #Switch Ownership of L30_Pet3 to 0
    L30_Pet3_DesOwn(Agent:agent):void=
        Print("DesOwnership L30P3 Saved")
        if(CurrentPlayerStats := PlayerStatsManager.GetPlayerStats[Agent]):
            PlayerStatsManager.RemoveOwnershipL30P3(Agent, StatType.Ownership_L30_Pet3, ?Ownership_L30_Pet3 := 0)
            

    #Switch Ownership of L30_Pet4 to 1
    L30_Pet4_Own(Agent:agent):void=
        Print("Ownership L30P4 Saved")
        if(CurrentPlayerStats := PlayerStatsManager.GetPlayerStats[Agent]):
            PlayerStatsManager.RecordOwnershipL30P4(Agent, StatType.Ownership_L30_Pet4, ?Ownership_L30_Pet4 := 1)
            

    #Switch Ownership of L30_Pet4 to 0
    L30_Pet4_DesOwn(Agent:agent):void=
        Print("DesOwnership L30P4 Saved")
        if(CurrentPlayerStats := PlayerStatsManager.GetPlayerStats[Agent]):
            PlayerStatsManager.RemoveOwnershipL30P4(Agent, StatType.Ownership_L30_Pet4, ?Ownership_L30_Pet4 := 0) 

    #---------------------------------------L31
    #Switch Ownership of L31_Pet1 to 1
    L31_Pet1_Own(Agent:agent):void=
        Print("Ownership L31P1 Saved")
        if(CurrentPlayerStats := PlayerStatsManager.GetPlayerStats[Agent]):
            PlayerStatsManager.RecordOwnershipL31P1(Agent, StatType.Ownership_L31_Pet1, ?Ownership_L31_Pet1 := 1)

    #---------------------------------------L32
    #Switch Ownership of L32_Pet1 to 1
    L32_Pet1_Own(Agent:agent):void=
        Print("Ownership L32P1 Saved")
        if(CurrentPlayerStats := PlayerStatsManager.GetPlayerStats[Agent]):
            PlayerStatsManager.RecordOwnershipL32P1(Agent, StatType.Ownership_L32_Pet1, ?Ownership_L32_Pet1 := 1)

    #---------------------------------------L33
    #Switch Ownership of L33_Pet1 to 1
    L33_Pet1_Own(Agent:agent):void=
        Print("Ownership L33P1 Saved")
        if(CurrentPlayerStats := PlayerStatsManager.GetPlayerStats[Agent]):
            PlayerStatsManager.RecordOwnershipL33P1(Agent, StatType.Ownership_L33_Pet1, ?Ownership_L33_Pet1 := 1)

    #---------------------------------------L34
    #Switch Ownership of L34_Pet1 to 1
    L34_Pet1_Own(Agent:agent):void=
        Print("Ownership L34P1 Saved")
        if(CurrentPlayerStats := PlayerStatsManager.GetPlayerStats[Agent]):
            PlayerStatsManager.RecordOwnershipL34P1(Agent, StatType.Ownership_L34_Pet1, ?Ownership_L34_Pet1 := 1)

    #Tuto Wheel
    #Max Rebirth Unlocked
    RegisterWheelshowed(Agent:agent):void=
        Print("Wheel showed saved")
        if(CurrentPlayerStats := PlayerStatsManager.GetPlayerStats[Agent]):
            PlayerStatsManager.RecordWheelshowed(Agent, StatType.Wheelshowed, ?Wheelshowed := 1)

    #Rebirth related
    ResetAllSave(Agent:agent):void=
        #Make sure button can't be triggered more than once in 10 sec
        if:
            PlayerObj := player[Agent]
            CurrentPlayerStats := PlayerStatsManager.GetPlayerStats[Agent]
            CustomPlayer:Custom_Player = PlayersMap[PlayerObj]
        then:
            #Remove all gold
            var Goldtoremove:int = CurrentPlayerStats.GoldBalance.CurrentValue
            DecreaseGoldBalance(Agent,Goldtoremove)
            CustomPlayer.DecreaseGoldBy(Goldtoremove)

            #Stop all pet follwing:

            <#
            CustomPlayer.IncreaseButtonL1() #close all the loops of all pets active
            CustomPlayer.IncreaseButtonL2()
            CustomPlayer.IncreaseButtonL3()
            CustomPlayer.IncreaseButtonL4()
            CustomPlayer.IncreaseButtonL5()
            CustomPlayer.IncreaseButtonL6()
            CustomPlayer.IncreaseButtonL7()
            CustomPlayer.IncreaseButtonL8()
            CustomPlayer.IncreaseButtonL9()
            CustomPlayer.IncreaseButtonL10()
            CustomPlayer.IncreaseButtonL11()
            CustomPlayer.IncreaseButtonL12()
            CustomPlayer.IncreaseButtonL13()
            CustomPlayer.IncreaseButtonL14()
            CustomPlayer.IncreaseButtonL15()
            CustomPlayer.IncreaseButtonL16()
            CustomPlayer.IncreaseButtonL17()
            CustomPlayer.IncreaseButtonL18()
            CustomPlayer.IncreaseButtonL19()
            CustomPlayer.IncreaseButtonL20()
            #>

            #Initialize Customer player pet variables

            <#
            set CustomPlayer.FollowActiveL1=0
            set CustomPlayer.FollowActiveL2=0
            set CustomPlayer.FollowActiveL3=0
            set CustomPlayer.FollowActiveL4=0
            set CustomPlayer.FollowActiveL5=0
            set CustomPlayer.FollowActiveL6=0
            set CustomPlayer.FollowActiveL7=0
            set CustomPlayer.FollowActiveL8=0
            set CustomPlayer.FollowActiveL9=0
            set CustomPlayer.FollowActiveL10=0
            set CustomPlayer.FollowActiveL11=0
            set CustomPlayer.FollowActiveL12=0
            set CustomPlayer.FollowActiveL13=0
            set CustomPlayer.FollowActiveL14=0
            set CustomPlayer.FollowActiveL15=0
            set CustomPlayer.FollowActiveL16=0
            set CustomPlayer.FollowActiveL17=0
            set CustomPlayer.FollowActiveL18=0
            set CustomPlayer.FollowActiveL19=0
            set CustomPlayer.FollowActiveL20=0

            set CustomPlayer.Pet_Power_ValueL1 = 0
            set CustomPlayer.Pet_Power_ValueL2 = 0
            set CustomPlayer.Pet_Power_ValueL3 = 0
            set CustomPlayer.Pet_Power_ValueL4 = 0
            set CustomPlayer.Pet_Power_ValueL5 = 0
            set CustomPlayer.Pet_Power_ValueL6 = 0
            set CustomPlayer.Pet_Power_ValueL7 = 0
            set CustomPlayer.Pet_Power_ValueL8 = 0
            set CustomPlayer.Pet_Power_ValueL9 = 0
            set CustomPlayer.Pet_Power_ValueL10 = 0
            set CustomPlayer.Pet_Power_ValueL11 = 0
            set CustomPlayer.Pet_Power_ValueL12 = 0
            set CustomPlayer.Pet_Power_ValueL13 = 0
            set CustomPlayer.Pet_Power_ValueL14 = 0
            set CustomPlayer.Pet_Power_ValueL15 = 0
            set CustomPlayer.Pet_Power_ValueL16 = 0
            set CustomPlayer.Pet_Power_ValueL17 = 0
            set CustomPlayer.Pet_Power_ValueL18 = 0
            set CustomPlayer.Pet_Power_ValueL19 = 0
            set CustomPlayer.Pet_Power_ValueL20 = 0

            set CustomPlayer.Ownership_L1_Pet1 = 0
            set CustomPlayer.Ownership_L1_Pet2 = 0
            set CustomPlayer.Ownership_L1_Pet3 = 0
            set CustomPlayer.Ownership_L1_Pet4 = 0
            
            set CustomPlayer.Ownership_L2_Pet1 = 0
            set CustomPlayer.Ownership_L2_Pet2 = 0
            set CustomPlayer.Ownership_L2_Pet3 = 0
            set CustomPlayer.Ownership_L2_Pet4 = 0
        
            set CustomPlayer.Ownership_L3_Pet1 = 0
            set CustomPlayer.Ownership_L3_Pet2 = 0
            set CustomPlayer.Ownership_L3_Pet3 = 0
            set CustomPlayer.Ownership_L3_Pet4 = 0
        
            set CustomPlayer.Ownership_L4_Pet1 = 0
            set CustomPlayer.Ownership_L4_Pet2 = 0
            set CustomPlayer.Ownership_L4_Pet3 = 0
            set CustomPlayer.Ownership_L4_Pet4 = 0
        
            set CustomPlayer.Ownership_L5_Pet1 = 0
            set CustomPlayer.Ownership_L5_Pet2 = 0
            set CustomPlayer.Ownership_L5_Pet3 = 0
            set CustomPlayer.Ownership_L5_Pet4 = 0
        
            set CustomPlayer.Ownership_L6_Pet1 = 0
            set CustomPlayer.Ownership_L6_Pet2 = 0
            set CustomPlayer.Ownership_L6_Pet3 = 0
            set CustomPlayer.Ownership_L6_Pet4 = 0
        
            set CustomPlayer.Ownership_L7_Pet1 = 0
            set CustomPlayer.Ownership_L7_Pet2 = 0
            set CustomPlayer.Ownership_L7_Pet3 = 0
            set CustomPlayer.Ownership_L7_Pet4 = 0
        
            set CustomPlayer.Ownership_L8_Pet1 = 0
            set CustomPlayer.Ownership_L8_Pet2 = 0
            set CustomPlayer.Ownership_L8_Pet3 = 0
            set CustomPlayer.Ownership_L8_Pet4 = 0
        
            set CustomPlayer.Ownership_L9_Pet1 = 0
            set CustomPlayer.Ownership_L9_Pet2 = 0
            set CustomPlayer.Ownership_L9_Pet3 = 0
            set CustomPlayer.Ownership_L9_Pet4 = 0
        
            set CustomPlayer.Ownership_L10_Pet1 = 0
            set CustomPlayer.Ownership_L10_Pet2 = 0
            set CustomPlayer.Ownership_L10_Pet3 = 0
            set CustomPlayer.Ownership_L10_Pet4 = 0
        
            set CustomPlayer.Ownership_L11_Pet1 = 0
            set CustomPlayer.Ownership_L11_Pet2 = 0
            set CustomPlayer.Ownership_L11_Pet3 = 0
            set CustomPlayer.Ownership_L11_Pet4 = 0
        
            set CustomPlayer.Ownership_L12_Pet1 = 0
            set CustomPlayer.Ownership_L12_Pet2 = 0
            set CustomPlayer.Ownership_L12_Pet3 = 0
            set CustomPlayer.Ownership_L12_Pet4 = 0
        
            set CustomPlayer.Ownership_L13_Pet1 = 0
            set CustomPlayer.Ownership_L13_Pet2 = 0
            set CustomPlayer.Ownership_L13_Pet3 = 0
            set CustomPlayer.Ownership_L13_Pet4 = 0
        
            set CustomPlayer.Ownership_L14_Pet1 = 0
            set CustomPlayer.Ownership_L14_Pet2 = 0
            set CustomPlayer.Ownership_L14_Pet3 = 0
            set CustomPlayer.Ownership_L14_Pet4 = 0
        
            set CustomPlayer.Ownership_L15_Pet1 = 0
            set CustomPlayer.Ownership_L15_Pet2 = 0
            set CustomPlayer.Ownership_L15_Pet3 = 0
            set CustomPlayer.Ownership_L15_Pet4 = 0
        
            set CustomPlayer.Ownership_L16_Pet1 = 0
            set CustomPlayer.Ownership_L16_Pet2 = 0
            set CustomPlayer.Ownership_L16_Pet3 = 0
            set CustomPlayer.Ownership_L16_Pet4 = 0
        
            set CustomPlayer.Ownership_L17_Pet1 = 0
            set CustomPlayer.Ownership_L17_Pet2 = 0
            set CustomPlayer.Ownership_L17_Pet3 = 0
            set CustomPlayer.Ownership_L17_Pet4 = 0
            
            set CustomPlayer.Ownership_L18_Pet1 = 0
            set CustomPlayer.Ownership_L18_Pet2 = 0
            set CustomPlayer.Ownership_L18_Pet3 = 0
            set CustomPlayer.Ownership_L18_Pet4 = 0
            
            set CustomPlayer.Ownership_L19_Pet1 = 0
            set CustomPlayer.Ownership_L19_Pet2 = 0
            set CustomPlayer.Ownership_L19_Pet3 = 0
            set CustomPlayer.Ownership_L19_Pet4 = 0
        
            set CustomPlayer.Ownership_L20_Pet1 = 0
            set CustomPlayer.Ownership_L20_Pet2 = 0
            set CustomPlayer.Ownership_L20_Pet3 = 0
            set CustomPlayer.Ownership_L20_Pet4 = 0

            set CustomPlayer.Pet_Unlocked_L1=0
            set CustomPlayer.Pet_Unlocked_L2=0
            set CustomPlayer.Pet_Unlocked_L3=0
            set CustomPlayer.Pet_Unlocked_L4=0
            set CustomPlayer.Pet_Unlocked_L5=0
            set CustomPlayer.Pet_Unlocked_L6=0
            set CustomPlayer.Pet_Unlocked_L7=0
            set CustomPlayer.Pet_Unlocked_L8=0
            set CustomPlayer.Pet_Unlocked_L9=0
            set CustomPlayer.Pet_Unlocked_L10=0
            set CustomPlayer.Pet_Unlocked_L11=0
            set CustomPlayer.Pet_Unlocked_L12=0
            set CustomPlayer.Pet_Unlocked_L13=0
            set CustomPlayer.Pet_Unlocked_L14=0
            set CustomPlayer.Pet_Unlocked_L15=0
            set CustomPlayer.Pet_Unlocked_L16=0
            set CustomPlayer.Pet_Unlocked_L17=0
            set CustomPlayer.Pet_Unlocked_L18=0
            set CustomPlayer.Pet_Unlocked_L19=0
            set CustomPlayer.Pet_Unlocked_L20=0
            #>

            #Initialize persistance data for pet ownership

            <#
            L1_Pet1_DesOwn(Agent)
            L1_Pet2_DesOwn(Agent)
            L1_Pet3_DesOwn(Agent)
            L1_Pet4_DesOwn(Agent)

            L2_Pet1_DesOwn(Agent)
            L2_Pet2_DesOwn(Agent)
            L2_Pet3_DesOwn(Agent)
            L2_Pet4_DesOwn(Agent)

            L3_Pet1_DesOwn(Agent)
            L3_Pet2_DesOwn(Agent)
            L3_Pet3_DesOwn(Agent)
            L3_Pet4_DesOwn(Agent)

            L4_Pet1_DesOwn(Agent)
            L4_Pet2_DesOwn(Agent)
            L4_Pet3_DesOwn(Agent)
            L4_Pet4_DesOwn(Agent)

            L5_Pet1_DesOwn(Agent)
            L5_Pet2_DesOwn(Agent)
            L5_Pet3_DesOwn(Agent)
            L5_Pet4_DesOwn(Agent)

            L6_Pet1_DesOwn(Agent)
            L6_Pet2_DesOwn(Agent)
            L6_Pet3_DesOwn(Agent)
            L6_Pet4_DesOwn(Agent)

            L7_Pet1_DesOwn(Agent)
            L7_Pet2_DesOwn(Agent)
            L7_Pet3_DesOwn(Agent)
            L7_Pet4_DesOwn(Agent)

            L8_Pet1_DesOwn(Agent)
            L8_Pet2_DesOwn(Agent)
            L8_Pet3_DesOwn(Agent)
            L8_Pet4_DesOwn(Agent)

            L9_Pet1_DesOwn(Agent)
            L9_Pet2_DesOwn(Agent)
            L9_Pet3_DesOwn(Agent)
            L9_Pet4_DesOwn(Agent)

            L10_Pet1_DesOwn(Agent)
            L10_Pet2_DesOwn(Agent)
            L10_Pet3_DesOwn(Agent)
            L10_Pet4_DesOwn(Agent)

            L11_Pet1_DesOwn(Agent)
            L11_Pet2_DesOwn(Agent)
            L11_Pet3_DesOwn(Agent)
            L11_Pet4_DesOwn(Agent)

            L12_Pet1_DesOwn(Agent)
            L12_Pet2_DesOwn(Agent)
            L12_Pet3_DesOwn(Agent)
            L12_Pet4_DesOwn(Agent)

            L13_Pet1_DesOwn(Agent)
            L13_Pet2_DesOwn(Agent)
            L13_Pet3_DesOwn(Agent)
            L13_Pet4_DesOwn(Agent)

            L14_Pet1_DesOwn(Agent)
            L14_Pet2_DesOwn(Agent)
            L14_Pet3_DesOwn(Agent)
            L14_Pet4_DesOwn(Agent)

            L15_Pet1_DesOwn(Agent)
            L15_Pet2_DesOwn(Agent)
            L15_Pet3_DesOwn(Agent)
            L15_Pet4_DesOwn(Agent)

            L16_Pet1_DesOwn(Agent)
            L16_Pet2_DesOwn(Agent)
            L16_Pet3_DesOwn(Agent)
            L16_Pet4_DesOwn(Agent)

            L17_Pet1_DesOwn(Agent)
            L17_Pet2_DesOwn(Agent)
            L17_Pet3_DesOwn(Agent)
            L17_Pet4_DesOwn(Agent)

            L18_Pet1_DesOwn(Agent)
            L18_Pet2_DesOwn(Agent)
            L18_Pet3_DesOwn(Agent)
            L18_Pet4_DesOwn(Agent)

            L19_Pet1_DesOwn(Agent)
            L19_Pet2_DesOwn(Agent)
            L19_Pet3_DesOwn(Agent)
            L19_Pet4_DesOwn(Agent)

            L20_Pet1_DesOwn(Agent)
            L20_Pet2_DesOwn(Agent)
            L20_Pet3_DesOwn(Agent)
            L20_Pet4_DesOwn(Agent)
            #>

            #Stats related

            #CustomPlayer.SetCritBy(0)       #UI
            #CustomPlayer.SetSpeedBy(1.0)    #UI
            #CustomPlayer.SetLuckBy(0)       #UI
            #SetCrit(Agent,0)                #Persistance
            #SetSpeed(Agent,0)               #Persistance
            #SetLuck(Agent,0)                #Persistance

            #Barrier related

            BarrierL1.RemoveFromIgnoreList(Agent)
            BarrierL2.RemoveFromIgnoreList(Agent)
            BarrierL3.RemoveFromIgnoreList(Agent)
            BarrierL4.RemoveFromIgnoreList(Agent)
            BarrierL5.RemoveFromIgnoreList(Agent)
            BarrierL6.RemoveFromIgnoreList(Agent)
            BarrierL7.RemoveFromIgnoreList(Agent)
            BarrierL8.RemoveFromIgnoreList(Agent)
            BarrierL9.RemoveFromIgnoreList(Agent)
            BarrierL10.RemoveFromIgnoreList(Agent)
            BarrierL11.RemoveFromIgnoreList(Agent)
            BarrierL12.RemoveFromIgnoreList(Agent)
            BarrierL13.RemoveFromIgnoreList(Agent)
            BarrierL14.RemoveFromIgnoreList(Agent)
            BarrierL15.RemoveFromIgnoreList(Agent)
            BarrierL16.RemoveFromIgnoreList(Agent)
            BarrierL17.RemoveFromIgnoreList(Agent)
            BarrierL18.RemoveFromIgnoreList(Agent)
            BarrierL19.RemoveFromIgnoreList(Agent)
            BarrierL20.RemoveFromIgnoreList(Agent)
            BarrierL21.RemoveFromIgnoreList(Agent)
            BarrierL22.RemoveFromIgnoreList(Agent)
            BarrierL23.RemoveFromIgnoreList(Agent)
            BarrierL24.RemoveFromIgnoreList(Agent)
            BarrierL25.RemoveFromIgnoreList(Agent)
            BarrierL26.RemoveFromIgnoreList(Agent)
            BarrierL27.RemoveFromIgnoreList(Agent)
            BarrierL28.RemoveFromIgnoreList(Agent)
            BarrierL29.RemoveFromIgnoreList(Agent)
            BarrierL30.RemoveFromIgnoreList(Agent)

            if(Cine:=Cine_Unlock_L2[GetTeamIndexAgent[Agent]-1]):#Swap eggs to unlock + Reset Box
                Cine.PlayReverse(Agent)

            if(Cine:=Cine_Unlock_L3[GetTeamIndexAgent[Agent]-1]):#Swap eggs to unlock + Reset Box
                Cine.PlayReverse(Agent)

            if(Cine:=Cine_Unlock_L4[GetTeamIndexAgent[Agent]-1]):#Swap eggs to unlock + Reset Box
                Cine.PlayReverse(Agent)

            if(Cine:=Cine_Unlock_L5[GetTeamIndexAgent[Agent]-1]):#Swap eggs to unlock + Reset Box
                Cine.PlayReverse(Agent)

            if(Cine:=Cine_Unlock_L6[GetTeamIndexAgent[Agent]-1]):#Swap eggs to unlock + Reset Box
                Cine.PlayReverse(Agent)

            if(Cine:=Cine_Unlock_L7[GetTeamIndexAgent[Agent]-1]):#Swap eggs to unlock + Reset Box
                Cine.PlayReverse(Agent)

            if(Cine:=Cine_Unlock_L8[GetTeamIndexAgent[Agent]-1]):#Swap eggs to unlock + Reset Box
                Cine.PlayReverse(Agent)

            if(Cine:=Cine_Unlock_L9[GetTeamIndexAgent[Agent]-1]):#Swap eggs to unlock + Reset Box
                Cine.PlayReverse(Agent)

            if(Cine:=Cine_Unlock_L10[GetTeamIndexAgent[Agent]-1]):#Swap eggs to unlock + Reset Box
                Cine.PlayReverse(Agent)

            if(Cine:=Cine_Unlock_L11[GetTeamIndexAgent[Agent]-1]):#Swap eggs to unlock + Reset Box
                Cine.PlayReverse(Agent)

            if(Cine:=Cine_Unlock_L12[GetTeamIndexAgent[Agent]-1]):#Swap eggs to unlock + Reset Box
                Cine.PlayReverse(Agent)

            if(Cine:=Cine_Unlock_L13[GetTeamIndexAgent[Agent]-1]):#Swap eggs to unlock + Reset Box
                Cine.PlayReverse(Agent)

            if(Cine:=Cine_Unlock_L14[GetTeamIndexAgent[Agent]-1]):#Swap eggs to unlock + Reset Box
                Cine.PlayReverse(Agent)

            if(Cine:=Cine_Unlock_L15[GetTeamIndexAgent[Agent]-1]):#Swap eggs to unlock + Reset Box
                Cine.PlayReverse(Agent)

            if(Cine:=Cine_Unlock_L16[GetTeamIndexAgent[Agent]-1]):#Swap eggs to unlock + Reset Box
                Cine.PlayReverse(Agent)

            if(Cine:=Cine_Unlock_L17[GetTeamIndexAgent[Agent]-1]):#Swap eggs to unlock + Reset Box
                Cine.PlayReverse(Agent)

            if(Cine:=Cine_Unlock_L18[GetTeamIndexAgent[Agent]-1]):#Swap eggs to unlock + Reset Box
                Cine.PlayReverse(Agent)

            if(Cine:=Cine_Unlock_L19[GetTeamIndexAgent[Agent]-1]):#Swap eggs to unlock + Reset Box
                Cine.PlayReverse(Agent)

            if(Cine:=Cine_Unlock_L20[GetTeamIndexAgent[Agent]-1]):#Swap eggs to unlock + Reset Box
                Cine.PlayReverse(Agent)

            if(Cine:=Cine_Unlock_L21[GetTeamIndexAgent[Agent]-1]):#Swap eggs to unlock + Reset Box
                Cine.PlayReverse(Agent)

            if(Cine:=Cine_Unlock_L22[GetTeamIndexAgent[Agent]-1]):#Swap eggs to unlock + Reset Box
                Cine.PlayReverse(Agent)

            if(Cine:=Cine_Unlock_L23[GetTeamIndexAgent[Agent]-1]):#Swap eggs to unlock + Reset Box
                Cine.PlayReverse(Agent)

            if(Cine:=Cine_Unlock_L24[GetTeamIndexAgent[Agent]-1]):#Swap eggs to unlock + Reset Box
                Cine.PlayReverse(Agent)

            if(Cine:=Cine_Unlock_L25[GetTeamIndexAgent[Agent]-1]):#Swap eggs to unlock + Reset Box
                Cine.PlayReverse(Agent)

            if(Cine:=Cine_Unlock_L26[GetTeamIndexAgent[Agent]-1]):#Swap eggs to unlock + Reset Box
                Cine.PlayReverse(Agent)

            if(Cine:=Cine_Unlock_L27[GetTeamIndexAgent[Agent]-1]):#Swap eggs to unlock + Reset Box
                Cine.PlayReverse(Agent)

            if(Cine:=Cine_Unlock_L28[GetTeamIndexAgent[Agent]-1]):#Swap eggs to unlock + Reset Box
                Cine.PlayReverse(Agent)

            if(Cine:=Cine_Unlock_L29[GetTeamIndexAgent[Agent]-1]):#Swap eggs to unlock + Reset Box
             Cine.PlayReverse(Agent)


            set CustomPlayer.B1_Unlocked=0
            set CustomPlayer.B2_Unlocked=0
            set CustomPlayer.B3_Unlocked=0
            set CustomPlayer.B4_Unlocked=0
            set CustomPlayer.B5_Unlocked=0
            set CustomPlayer.B6_Unlocked=0
            set CustomPlayer.B7_Unlocked=0
            set CustomPlayer.B8_Unlocked=0
            set CustomPlayer.B9_Unlocked=0
            set CustomPlayer.B10_Unlocked=0
            set CustomPlayer.B11_Unlocked=0
            set CustomPlayer.B12_Unlocked=0
            set CustomPlayer.B13_Unlocked=0
            set CustomPlayer.B14_Unlocked=0
            set CustomPlayer.B15_Unlocked=0
            set CustomPlayer.B16_Unlocked=0
            set CustomPlayer.B17_Unlocked=0
            set CustomPlayer.B18_Unlocked=0
            set CustomPlayer.B19_Unlocked=0
            set CustomPlayer.B20_Unlocked=0
            set CustomPlayer.B21_Unlocked=0
            set CustomPlayer.B22_Unlocked=0
            set CustomPlayer.B23_Unlocked=0
            set CustomPlayer.B24_Unlocked=0
            set CustomPlayer.B25_Unlocked=0
            set CustomPlayer.B26_Unlocked=0
            set CustomPlayer.B27_Unlocked=0
            set CustomPlayer.B28_Unlocked=0
            set CustomPlayer.B29_Unlocked=0
            set CustomPlayer.B30_Unlocked=0

            MaxLevelReset(Agent)

            #Finally reloading the UI

            CustomPlayer.Update_Pets_Section()
            CustomPlayer.Update_Pet_Power()
            #CustomPlayer.Remove_UI()
            #CustomPlayer.InitiateUI()

    Rebirth1_Func(Agent:agent):void=
        if:
            PlayerObj := player[Agent]
            CurrentPlayerStats := PlayerStatsManager.GetPlayerStats[Agent]
            CustomPlayer:Custom_Player = PlayersMap[PlayerObj]
            CurrentPlayerStats.RebirthUnlocked.CurrentValue>=1
        then:
            AlreadyUnlockedHUD.Show(Agent)
        else:
            if:
                PlayerObj := player[Agent]
                CustomPlayer:Custom_Player = PlayersMap[PlayerObj]
            then:
                CustomPlayer.SetBonusPowerMultiplicator(2.0)
                ResetAllSave(Agent)
                AddMaxRebirthUnlocked(Agent)
                set CustomPlayer.RebirthCount = 1
                if(Cine:=Cine_RebirthCamera1[GetTeamIndexAgent[Agent]-1]):
                    Cine.Play(Agent)
                TeleporterRebirth.Teleport(Agent)
                spawn:
                    Rebirth_End_Sound()
                if(HUDDevice:=Rebirth0_HUD[GetTeamIndexAgent[Agent]-1]):
                    HUDDevice.Hide(Agent)
                if(HUDDevice:=Rebirth1_HUD[GetTeamIndexAgent[Agent]-1]):
                    HUDDevice.Show(Agent)
                Accolade_Final_Level.Award(Agent)
                UnlockBarrierB31(Agent)

    Rebirth2_Func(Agent:agent):void=
        if:
            CurrentPlayerStats := PlayerStatsManager.GetPlayerStats[Agent]
            CurrentPlayerStats.RebirthUnlocked.CurrentValue<1
        then:
            RequirePreviousRebirthHUD.Show(Agent)
        else:
            if:
                PlayerObj := player[Agent]
                CurrentPlayerStats := PlayerStatsManager.GetPlayerStats[Agent]
                CustomPlayer:Custom_Player = PlayersMap[PlayerObj]
                CurrentPlayerStats.RebirthUnlocked.CurrentValue>=2
            then:
                AlreadyUnlockedHUD.Show(Agent)
            else:
                if:
                    PlayerObj := player[Agent]
                    CurrentPlayerStats := PlayerStatsManager.GetPlayerStats[Agent]
                    CustomPlayer:Custom_Player = PlayersMap[PlayerObj]
                    CurrentPlayerStats.RebirthUnlocked.CurrentValue=1
                then:
                    CustomPlayer.SetBonusPowerMultiplicator(3.0)
                    ResetAllSave(Agent)
                    AddMaxRebirthUnlocked(Agent)
                    set CustomPlayer.RebirthCount = 1
                    if(Cine:=Cine_RebirthCamera2[GetTeamIndexAgent[Agent]-1]):
                        Cine.Play(Agent)
                    TeleporterRebirth.Teleport(Agent)
                    spawn:
                        Rebirth_End_Sound()
                    if(HUDDevice:=Rebirth1_HUD[GetTeamIndexAgent[Agent]-1]):
                        HUDDevice.Hide(Agent)
                    if(HUDDevice:=Rebirth2_HUD[GetTeamIndexAgent[Agent]-1]):
                        HUDDevice.Show(Agent)
                    Accolade_Final_Level.Award(Agent)
                    UnlockBarrierB32(Agent)

    Rebirth3_Func(Agent:agent):void=
        if:
            CurrentPlayerStats := PlayerStatsManager.GetPlayerStats[Agent]
            CurrentPlayerStats.RebirthUnlocked.CurrentValue<2
        then:
            RequirePreviousRebirthHUD.Show(Agent)
        else:
            if:
                PlayerObj := player[Agent]
                CurrentPlayerStats := PlayerStatsManager.GetPlayerStats[Agent]
                CustomPlayer:Custom_Player = PlayersMap[PlayerObj]
                CurrentPlayerStats.RebirthUnlocked.CurrentValue>=3
            then:
                AlreadyUnlockedHUD.Show(Agent)
            else:
                if:
                    PlayerObj := player[Agent]
                    CurrentPlayerStats := PlayerStatsManager.GetPlayerStats[Agent]
                    CustomPlayer:Custom_Player = PlayersMap[PlayerObj]
                    CurrentPlayerStats.RebirthUnlocked.CurrentValue=2
                then:
                    CustomPlayer.SetBonusPowerMultiplicator(4.0)
                    ResetAllSave(Agent)
                    AddMaxRebirthUnlocked(Agent)
                    set CustomPlayer.RebirthCount = 1
                    if(Cine:=Cine_RebirthCamera3[GetTeamIndexAgent[Agent]-1]):
                        Cine.Play(Agent)
                    TeleporterRebirth.Teleport(Agent)
                    spawn:
                        Rebirth_End_Sound()
                    if(HUDDevice:=Rebirth2_HUD[GetTeamIndexAgent[Agent]-1]):
                        HUDDevice.Hide(Agent)
                    if(HUDDevice:=Rebirth3_HUD[GetTeamIndexAgent[Agent]-1]):
                        HUDDevice.Show(Agent)
                    Accolade_Final_Level.Award(Agent)
                    UnlockBarrierB33(Agent)

    Rebirth4_Func(Agent:agent):void=
        Print("")
    
    Rebirth5_Func(Agent:agent):void=
        Print("")

    Rebirth5_Func_Code(Agent:agent):void=
        if:
            CurrentPlayerStats := PlayerStatsManager.GetPlayerStats[Agent]
            CurrentPlayerStats.RebirthUnlocked.CurrentValue<5
        then:
            RequirePreviousRebirthHUD.Show(Agent)
        else:
            if:
                PlayerObj := player[Agent]
                CurrentPlayerStats := PlayerStatsManager.GetPlayerStats[Agent]
                CustomPlayer:Custom_Player = PlayersMap[PlayerObj]
                CurrentPlayerStats.RebirthUnlocked.CurrentValue>=5
            then:
                Code_Rebirth5.Show(Agent)

    Rebirth_End_Sound()<suspends>:void=
        Sleep(0.5)
        RebirthSFX_End.Play()

    Test_Func(Agent:agent):void=
        if:
            CurrentPlayerStats := PlayerStatsManager.GetPlayerStats[Agent]
            PlayerObj := player[Agent]
            CustomPlayer:Custom_Player = PlayersMap[PlayerObj]
        then:
            #AddMaxRebirthUnlocked(Agent)
            #Print("Rebirth: {CurrentPlayerStats.RebirthUnlocked.CurrentValue}")
            AddPotionLuck1(Agent,1)
            set CustomPlayer.Potion_Owned_Luck1 += 1
            AddPotionLuck2(Agent,1)
            set CustomPlayer.Potion_Owned_Luck2 += 1
            AddPotionLuck3(Agent,1)
            set CustomPlayer.Potion_Owned_Luck3 += 1
            AddPotionGold1(Agent,1)
            set CustomPlayer.Potion_Owned_Gold1 += 1
            AddPotionGold2(Agent,1)
            set CustomPlayer.Potion_Owned_Gold2 += 1
            AddPotionGold3(Agent,1)
            set CustomPlayer.Potion_Owned_Gold3 += 1
            AddPotionPetPower1(Agent,1)
            set CustomPlayer.Potion_Owned_Pet_Power1 += 1
            AddPotionPetPower2(Agent,1)
            set CustomPlayer.Potion_Owned_Pet_Power2 += 1
            AddPotionPetPower3(Agent,1)
            set CustomPlayer.Potion_Owned_Pet_Power3 += 1
            AddPotionPlayerPower1(Agent,1)
            set CustomPlayer.Potion_Owned_Player_Power1 += 1
            AddPotionPlayerPower2(Agent,1)
            set CustomPlayer.Potion_Owned_Player_Power2 += 1
            AddPotionPlayerPower3(Agent,1)
            set CustomPlayer.Potion_Owned_Player_Power3 += 1

    ##########################################################################################################################################################
    #------------------INVENTORY
    ##########################################################################################################################################################
    EnumToTexture(PurchaseType : purchase_type)<transacts>:texture=
        case(PurchaseType):
            purchase_type.Luck1 => Inventory_Textures.inv_luck_1
            purchase_type.Luck2 => Inventory_Textures.inv_luck_2
            purchase_type.Luck3 => Inventory_Textures.inv_luck_3
            purchase_type.Gold1 => Inventory_Textures.inv_gold_1
            purchase_type.Gold2 => Inventory_Textures.inv_gold_2
            purchase_type.Gold3 => Inventory_Textures.inv_gold_3
            purchase_type.Pet_Power1 => Inventory_Textures.inv_pet_power_1
            purchase_type.Pet_Power2 => Inventory_Textures.inv_pet_power_2
            purchase_type.Pet_Power3 => Inventory_Textures.inv_pet_power_3
            purchase_type.Player_Power1 => Inventory_Textures.inv_player_power_1
            purchase_type.Player_Power2 => Inventory_Textures.inv_player_power_2
            purchase_type.Player_Power3 => Inventory_Textures.inv_player_power_3
            purchase_type.Empty => Inventory_Textures.inv_luck_1

    CalculateMaxPages()<transacts>:int=
        MaxPages : int = if (Num := Ceil[(ShopPurchases.Length*1.0) / (PurchasePerPage*1.0)])
        then Num
        else -1

        return MaxPages

    GenerateFrameStrings(PowerupValue : float, PurchasePrice : int, PurchaseType : purchase_type,
    bPurchased : logic, bPurchaseOnce : logic):tuple(string , string)=
        var PVInt : int = if (Num := Int[PowerupValue])
        then Num
        else -1

        var PVIntFraction : int = 0
        if (PVM100Int := Int[PowerupValue*100.0]):
            set PVIntFraction = PVM100Int - PVInt*100

        var PriceString : string = "{PVInt}."
        if (PVIntFraction < 10):
            set PriceString += "0"
        set PriceString += "{PVIntFraction}"
        
        var ItemName: string = ""

        case(PurchaseType):
            purchase_type.Luck1 => set ItemName ="Luck lvl 1"
            purchase_type.Luck2 => set ItemName ="Luck lvl 2"
            purchase_type.Luck3 => set ItemName ="Luck lvl 3"
            purchase_type.Gold1 => set ItemName ="Gold lvl 1"
            purchase_type.Gold2 => set ItemName ="Gold lvl 2"
            purchase_type.Gold3 => set ItemName ="Gold lvl 3"
            purchase_type.Pet_Power1 => set ItemName ="Pet Power lvl 1"
            purchase_type.Pet_Power2 => set ItemName ="Pet Power lvl 2"
            purchase_type.Pet_Power3 => set ItemName ="Pet Power lvl 3"
            purchase_type.Player_Power1 => set ItemName ="Player Power lvl 1"
            purchase_type.Player_Power2 => set ItemName ="Player Power lvl 2"
            purchase_type.Player_Power3 => set ItemName ="Player Power lvl 3"
            purchase_type.Empty => set ItemName =""
        
        PurchaseString : string = if (PurchaseType <> purchase_type.Empty)
        then "{ItemName}"
        else ""

        var ButtonString : string = "USE"

        return (PurchaseString, ButtonString)

    OnShopOpen(Agent : agent):void=
        if (AgentData := AgentMap[Agent], AgentData.bShopOpen = false,Player := player[Agent], PlayerUI := GetPlayerUI[Player], CustomPlayer:Custom_Player = PlayersMap[Player]):
            set AgentData.bShopOpen = true
            set AgentData.CurrentPageIndex = 0

            ShopCanvas := CreateShopCanvas()
            set AgentData.MaybeShopCanvas = option{ ShopCanvas }
            UpdatePurchaseFrames(Agent)

            PlayerUI.AddWidget(ShopCanvas, player_ui_slot{ InputMode := ui_input_mode.All })

    OnShopOpen2(MaybeAgent:?agent):void=
        if (Agent := MaybeAgent?):
            if (AgentData := AgentMap[Agent], AgentData.bShopOpen = false,Player := player[Agent], PlayerUI := GetPlayerUI[Player], CustomPlayer:Custom_Player = PlayersMap[Player]):
                set AgentData.bShopOpen = true
                set AgentData.CurrentPageIndex = 0

                ShopCanvas := CreateShopCanvas()
                set AgentData.MaybeShopCanvas = option{ ShopCanvas }
                UpdatePurchaseFrames(Agent)

                PlayerUI.AddWidget(ShopCanvas, player_ui_slot{ InputMode := ui_input_mode.All })

    OnPagePrevClicked(WidgetMessage : widget_message):void=
        if (AgentData := AgentMap[WidgetMessage.Player], AgentData.CurrentPageIndex > 0):
            set AgentData.CurrentPageIndex -= 1
            UpdatePurchaseFrames(WidgetMessage.Player)

    OnPageNextClicked(WidgetMessage : widget_message):void=
        if (AgentData := AgentMap[WidgetMessage.Player], AgentData.CurrentPageIndex < CalculateMaxPages()-1):
            set AgentData.CurrentPageIndex += 1
            UpdatePurchaseFrames(WidgetMessage.Player)

    OnPageCloseClicked(WidgetMessage : widget_message):void=
        if (AgentData := AgentMap[WidgetMessage.Player], PlayerUI := GetPlayerUI[WidgetMessage.Player],
        ShopCanvas := AgentData.MaybeShopCanvas?):
            set AgentData.bShopOpen = false
            set AgentData.MaybeShopCanvas = false

            PlayerUI.RemoveWidget(ShopCanvas)

    OnPurchaseClicked0(WidgetMessage : widget_message):void=
        OnPurchaseClicked(WidgetMessage, 0)

    OnPurchaseClicked1(WidgetMessage : widget_message):void=
        OnPurchaseClicked(WidgetMessage, 1)

    OnPurchaseClicked2(WidgetMessage : widget_message):void=
        OnPurchaseClicked(WidgetMessage, 2)

    OnPurchaseClicked(WidgetMessage : widget_message, PurchaseIndex : int):void=
        if (AgentData := AgentMap[WidgetMessage.Player], ShopCanvas := AgentData.MaybeShopCanvas?, PlayerObj := WidgetMessage.Player, CustomPlayer:Custom_Player = PlayersMap[PlayerObj], Agent:=agent[WidgetMessage.Player]):

            TotalPurchaseIndex := (AgentData.CurrentPageIndex * PurchasePerPage) + PurchaseIndex
            if (PurchaseData := ShopPurchases[TotalPurchaseIndex], CastedPrice := int[PurchaseData.PurchasePrice],
            PriceScaler := float[PurchaseData.PriceScalerAfterEachPurchase]):

                if (PurchaseData.PurchaseType = purchase_type.Empty):
                    Print("Error: Empty Purchase!")
                    return

                PurchaseCount : int = if (Num := AgentData.Purchases[TotalPurchaseIndex])
                then Num
                else 0

                var OwnedStock :int = 0

                case(PurchaseData.PurchaseType):
                    purchase_type.Luck1 => set OwnedStock = CustomPlayer.Potion_Owned_Luck1
                    purchase_type.Luck2 => set OwnedStock = CustomPlayer.Potion_Owned_Luck2
                    purchase_type.Luck3 => set OwnedStock = CustomPlayer.Potion_Owned_Luck3
                    purchase_type.Gold1 => set OwnedStock = CustomPlayer.Potion_Owned_Gold1
                    purchase_type.Gold2 => set OwnedStock = CustomPlayer.Potion_Owned_Gold2
                    purchase_type.Gold3 => set OwnedStock = CustomPlayer.Potion_Owned_Gold3
                    purchase_type.Pet_Power1 => set OwnedStock = CustomPlayer.Potion_Owned_Pet_Power1
                    purchase_type.Pet_Power2 => set OwnedStock = CustomPlayer.Potion_Owned_Pet_Power2
                    purchase_type.Pet_Power3 => set OwnedStock = CustomPlayer.Potion_Owned_Pet_Power3
                    purchase_type.Player_Power1 => set OwnedStock = CustomPlayer.Potion_Owned_Player_Power1 
                    purchase_type.Player_Power2 => set OwnedStock = CustomPlayer.Potion_Owned_Player_Power2
                    purchase_type.Player_Power3 => set OwnedStock = CustomPlayer.Potion_Owned_Player_Power3
                    purchase_type.Empty => set OwnedStock = 0

                if (OwnedStock < 1):
                    Print("Error: Not Enough item!")
                    return

                Print("Purchase Successful.")

                if (set AgentData.Purchases[TotalPurchaseIndex] += 1){}

                case(PurchaseData.PurchaseType):
                    purchase_type.Luck1 => block{set CustomPlayer.Potion_Owned_Luck1 -= 1, DecreasePotionLuck1(Agent,1), set CustomPlayer.Potion_Active_Luck1_Sec +=600}
                    purchase_type.Luck2 => block{set CustomPlayer.Potion_Owned_Luck2 -= 1, DecreasePotionLuck2(Agent,1),set CustomPlayer.Potion_Active_Luck2_Sec +=1800}
                    purchase_type.Luck3 => block{set CustomPlayer.Potion_Owned_Luck3 -= 1, DecreasePotionLuck3(Agent,1),set CustomPlayer.Potion_Active_Luck3_Sec +=3600}
                    purchase_type.Gold1 => block{set CustomPlayer.Potion_Owned_Gold1 -= 1, DecreasePotionGold1(Agent,1),set CustomPlayer.Potion_Active_Gold1_Sec +=600}
                    purchase_type.Gold2 => block{set CustomPlayer.Potion_Owned_Gold2 -= 1, DecreasePotionGold2(Agent,1),set CustomPlayer.Potion_Active_Gold2_Sec +=1800}
                    purchase_type.Gold3 => block{set CustomPlayer.Potion_Owned_Gold3 -= 1, DecreasePotionGold3(Agent,1),set CustomPlayer.Potion_Active_Gold3_Sec +=3600}
                    purchase_type.Pet_Power1 => block{set CustomPlayer.Potion_Owned_Pet_Power1 -= 1, DecreasePotionPetPower1(Agent,1),set CustomPlayer.Potion_Active_PetPower1_Sec +=600}
                    purchase_type.Pet_Power2 => block{set CustomPlayer.Potion_Owned_Pet_Power2 -= 1, DecreasePotionPetPower2(Agent,1),set CustomPlayer.Potion_Active_PetPower2_Sec +=1800}
                    purchase_type.Pet_Power3 => block{set CustomPlayer.Potion_Owned_Pet_Power3 -= 1, DecreasePotionPetPower3(Agent,1),set CustomPlayer.Potion_Active_PetPower3_Sec +=3600}
                    purchase_type.Player_Power1 => block{set CustomPlayer.Potion_Owned_Player_Power1 -= 1, DecreasePotionPlayerPower1(Agent,1),set CustomPlayer.Potion_Active_PlayerPower1_Sec +=600}
                    purchase_type.Player_Power2 => block{set CustomPlayer.Potion_Owned_Player_Power2 -= 1, DecreasePotionPlayerPower2(Agent,1),set CustomPlayer.Potion_Active_PlayerPower2_Sec +=1800}
                    purchase_type.Player_Power3 => block{set CustomPlayer.Potion_Owned_Player_Power3 -= 1, DecreasePotionPlayerPower3(Agent,1),set CustomPlayer.Potion_Active_PlayerPower3_Sec +=3600}
                    purchase_type.Empty => #Empty On Purpose

                if (FC := WidgetMessage.Player.GetFortCharacter[]):
                    #FC.SetMaxHealth(AgentData.MaxHP)
                    #FC.SetMaxShield(AgentData.MaxShield)

                UpdatePurchaseFrames(WidgetMessage.Player)

    UpdatePurchaseFrames(Agent : agent):void=

        PageTextIndex : int = 4
        PurchaseFrameStartIndex : int = 5

        if (AgentData := AgentMap[Agent], ShopCanvas := AgentData.MaybeShopCanvas?,
        ShopCanvasInner := canvas[ShopCanvas.Slots[0].Widget],
        PageTextBlock := text_block[ShopCanvasInner.Slots[PageTextIndex].Widget]):

            PageTextBlock.SetText(TextForUI("{AgentData.CurrentPageIndex+1}/{CalculateMaxPages()}"))

            FrameTextIndex := 0
            FrameTextureIndex := 1
            FrameButtonIndex := 2

            for (I := 0 .. PurchasePerPage-1,
            PurchaseFrameCanvas := canvas[ShopCanvasInner.Slots[PurchaseFrameStartIndex+I].Widget],
            TextBlock := text_block[PurchaseFrameCanvas.Slots[FrameTextIndex].Widget],
            TextureBlock := texture_block[PurchaseFrameCanvas.Slots[FrameTextureIndex].Widget],
            ButtonRegular := button_regular[PurchaseFrameCanvas.Slots[FrameButtonIndex].Widget]):

                var PowerupValue : float = 0.0
                var PurchaseType : purchase_type = purchase_type.Empty
                var bPurchaseOnce : logic = true
                var PriceScaler : float = 0.0
                var CastedPrice : int = 0
                if (PurchaseData := ShopPurchases[(AgentData.CurrentPageIndex*PurchasePerPage) + I]):
                    set PowerupValue = PurchaseData.PowerupValue
                    set PurchaseType = PurchaseData.PurchaseType
                    set bPurchaseOnce = PurchaseData.bOneTimePurchase

                    if (ImmPriceScaler := float[PurchaseData.PriceScalerAfterEachPurchase]):
                        set PriceScaler = ImmPriceScaler

                    if (ImmCastedPrice := int[PurchaseData.PurchasePrice]):
                        set CastedPrice = ImmCastedPrice

                PurchaseCount : int = if (Num := AgentData.Purchases[(AgentData.CurrentPageIndex*PurchasePerPage) + I])
                then Num
                else 0

                Price := CastedPrice 
                bPurchased : logic = logic{ PurchaseCount >= 1 }

                StringsTuple := GenerateFrameStrings(PowerupValue, Price, PurchaseType, bPurchased, bPurchaseOnce)
                PurchaseString : string = StringsTuple(0)
                ButtonString : string = StringsTuple(1)

                TextBlock.SetText(TextForUI(PurchaseString))
                ButtonRegular.SetText(TextForUI(ButtonString))

                TextureBlock.SetImage(EnumToTexture(PurchaseType))

                Visibility : widget_visibility = if (PurchaseType = purchase_type.Empty)
                then widget_visibility.Hidden
                else widget_visibility.Visible

                TextBlock.SetVisibility(Visibility)
                ButtonRegular.SetVisibility(Visibility)
                TextureBlock.SetVisibility(Visibility)

    CreateShopCanvas():canvas=
        var CanvasSlots : []canvas_slot = array{}

        set CanvasSlots += array
        {
            canvas_slot:
                Anchors := anchors{ Minimum := vector2{ X := 0.5, Y := 0.5 }, Maximum := vector2{ X := 0.5, Y := 0.5 } }
                Offsets := margin{ Top := 0.0, Left := 0.0, Right := 0.0, Bottom := 0.0 }
                Alignment := vector2{ X := 0.5, Y := 0.5 }
                SizeToContent := true
                Widget := CreateShopCanvasInner()
        }

        Canvas : canvas = canvas:
            Slots := CanvasSlots

        return Canvas

    CreateShopCanvasInner():canvas=
        var CanvasSlots : []canvas_slot = array{}

        MaxPages := CalculateMaxPages()

        BGSize := vector2{ X := 1280.0, Y := 720.0 }
        set CanvasSlots += array
        {
            CanvasSlot0 := canvas_slot:
                Anchors := anchors{ Minimum := vector2{ X := 0.0, Y := 0.0 }, Maximum := vector2{ X := 0.0, Y := 0.0 } }
                Offsets := margin{ Top := 0.0, Left := 0.0, Right := 0.0, Bottom := 0.0 }
                Alignment := vector2{ X := 0.0, Y := 0.0 }
                SizeToContent := true
                Widget := texture_block{ DefaultImage := Inventory_Textures.T_Inventory_BG, DefaultDesiredSize := BGSize }
        }

        var PageButtonString : string = ""
        WhiteSpaceCount := 14
        for (I := 1 .. WhiteSpaceCount):
            set PageButtonString += " "
        set PageButtonString += "\n"
        for (I := 1 .. WhiteSpaceCount):
            set PageButtonString += " "
        set CanvasSlots += array
        {
            CanvasSlot1 := canvas_slot:
                Anchors := anchors{ Minimum := vector2{ X := 0.0, Y := 1.0 }, Maximum := vector2{ X := 0.0, Y := 1.0 } }
                Offsets := margin{ Top := -26.0, Left := 76.0, Right := 0.0, Bottom := 0.0 }
                Alignment := vector2{ X := 0.0, Y := 1.0 }
                SizeToContent := true
                Widget := button_regular{ DefaultText := TextForUI(PageButtonString) }
        }
        if (ButtonRegular := button_regular[CanvasSlots[CanvasSlots.Length-1].Widget]):
            ButtonRegular.OnClick().Subscribe(OnPagePrevClicked)

        set CanvasSlots += array
        {
            CanvasSlot2 := canvas_slot:
                Anchors := anchors{ Minimum := vector2{ X := 1.0, Y := 1.0 }, Maximum := vector2{ X := 1.0, Y := 1.0 } }
                Offsets := margin{ Top := -26.0, Left := -76.0, Right := 0.0, Bottom := 0.0 }
                Alignment := vector2{ X := 1.0, Y := 1.0 }
                SizeToContent := true
                Widget := button_regular{ DefaultText := TextForUI(PageButtonString) }
        }
        if (ButtonRegular := button_regular[CanvasSlots[CanvasSlots.Length-1].Widget]):
            ButtonRegular.OnClick().Subscribe(OnPageNextClicked)

        var CloseButtonString : string = ""
        WhiteSpaceCount2 := 12
        for (I := 1 .. WhiteSpaceCount2):
            set CloseButtonString += " "
        set CloseButtonString += "\n"
        for (I := 1 .. WhiteSpaceCount2):
            set CloseButtonString += " "
        set CanvasSlots += array
        {
            CanvasSlot3 := canvas_slot:
                Anchors := anchors{ Minimum := vector2{ X := 1.0, Y := 0.0 }, Maximum := vector2{ X := 1.0, Y := 0.0 } }
                Offsets := margin{ Top := 12.0, Left := -76.0, Right := 0.0, Bottom := 0.0 }
                Alignment := vector2{ X := 1.0, Y := 0.0 }
                SizeToContent := true
                Widget := button_regular{ DefaultText := TextForUI(CloseButtonString) }
        }
        if (ButtonRegular := button_regular[CanvasSlots[CanvasSlots.Length-1].Widget]):
            ButtonRegular.OnClick().Subscribe(OnPageCloseClicked)

        set CanvasSlots += array
        {
            CanvasSlot4 := canvas_slot:
                Anchors := anchors{ Minimum := vector2{ X := 0.5, Y := 1.0 }, Maximum := vector2{ X := 0.5, Y := 1.0 } }
                Offsets := margin{ Top := -68.0, Left := 0.0, Right := 0.0, Bottom := 0.0 }
                Alignment := vector2{ X := 0.5, Y := 1.0 }
                SizeToContent := true
                Widget := text_block{ DefaultText := TextForUI("1/{MaxPages}"), DefaultShadowColor := NamedColors.Black,
                DefaultShadowOpacity := 1.0, DefaultShadowOffset := option{ vector2{ X := 2.0, Y := 2.0 }}, DefaultTextColor := NamedColors.White }
        }

        for (I := 0 .. PurchasePerPage-1):

            XAnchorOffset : float = if (I = 1)
            then 0.0
            else 0.325

            XAnchorOffsetSign : float = if (I = 0)
            then -1.0
            else 1.0

            TotalXAnchorOffset := XAnchorOffset * XAnchorOffsetSign

            set CanvasSlots += array
            {
                CanvasSlot5 := canvas_slot:
                    Anchors := anchors{ Minimum := vector2{ X := 0.5+TotalXAnchorOffset, Y := 0.45 }, Maximum := vector2{ X := 0.5+TotalXAnchorOffset, Y := 0.45 } }
                    Offsets := margin{ Top := 0.0, Left := 0.0, Right := 0.0, Bottom := 0.0 }
                    Alignment := vector2{ X := 0.5, Y := 0.5 }
                    SizeToContent := true
                    Widget := CreateFrameCanvas(0.0, 0, purchase_type.Empty, false, false, I)
            }

        Canvas : canvas = canvas:
            Slots := CanvasSlots

        return Canvas

    CreateFrameCanvas(PowerupValue : float, PurchasePrice : int, PurchaseType : purchase_type,
    bPurchased : logic, bPurchaseOnce : logic, FrameIndex : int):canvas=
        var CanvasSlots : []canvas_slot = array{}

        StringsTuple := GenerateFrameStrings(PowerupValue, PurchasePrice, PurchaseType, bPurchased, bPurchaseOnce)
        PurchaseString : string = StringsTuple(0)
        ButtonString : string = StringsTuple(1)

        FrameTextureSize := vector2{ X := 192.0, Y := 192.0 }

        set CanvasSlots += array
        {
            canvas_slot:
                Anchors := anchors{ Minimum := vector2{ X := 0.5, Y := 0.5 }, Maximum := vector2{ X := 0.5, Y := 0.5 } }
                Offsets := margin{ Top := -(FrameTextureSize.Y/2.0 + 46.0), Left := 0.0, Right := 0.0, Bottom := 0.0 }
                Alignment := vector2{ X := 0.5, Y := 1.0 }
                SizeToContent := true
                Widget := text_block{ DefaultText := TextForUI(PurchaseString), DefaultShadowColor := NamedColors.White,
                DefaultShadowOpacity := 1.0, DefaultShadowOffset := option{ vector2{ X := 2.0, Y := 2.0 }}, DefaultTextColor := NamedColors.Black }
        }

        set CanvasSlots += array
        {
            canvas_slot:
                Anchors := anchors{ Minimum := vector2{ X := 0.5, Y := 0.5 }, Maximum := vector2{ X := 0.5, Y := 0.5 } }
                Offsets := margin{ Top := 0.0, Left := 0.0, Right := 0.0, Bottom := 0.0 }
                Alignment := vector2{ X := 0.5, Y := 0.5 }
                SizeToContent := true
                Widget := texture_block{ DefaultImage := EnumToTexture(PurchaseType), DefaultDesiredSize := FrameTextureSize }
        }

        set CanvasSlots += array
        {
            canvas_slot:
                Anchors := anchors{ Minimum := vector2{ X := 0.5, Y := 0.5 }, Maximum := vector2{ X := 0.5, Y := 0.5 } }
                Offsets := margin{ Top := (FrameTextureSize.Y/2.0 + 72.0), Left := 0.0, Right := 0.0, Bottom := 0.0 }
                Alignment := vector2{ X := 0.5, Y := 0.0 }
                SizeToContent := true
                Widget := button_regular{ DefaultText := TextForUI(ButtonString) }
        }
        if (ButtonRegular := button_regular[CanvasSlots[CanvasSlots.Length-1].Widget]):
            if (FrameIndex = 0):
                ButtonRegular.OnClick().Subscribe(OnPurchaseClicked0)
            else if (FrameIndex = 1):
                ButtonRegular.OnClick().Subscribe(OnPurchaseClicked1)
            else if (FrameIndex = 2):
                ButtonRegular.OnClick().Subscribe(OnPurchaseClicked2)

        Canvas : canvas = canvas:
            Slots := CanvasSlots

        return Canvas

    #Check how many seconds of bonus are remaining for each potion and apply the corresponding bonus. Loops infinitely every 1 sec.
    ManagePotions(Agent : agent)<suspends>: void = 
        loop:
            if(1 <> 1) {break}
            else:
                ManagePotions_loop(Agent)
                Sleep(1.0)

    ManagePotions_loop(Agent : agent): void = 
        if:
            PlayerObj := player[Agent]
            CustomPlayer:Custom_Player = PlayersMap[PlayerObj]
        then:
            if(CustomPlayer.Potion_Active_Luck3_Sec>0):
                Print("Reducing luck 3 by 1 sec")
                set CustomPlayer.Potion_Active_Luck3_Sec -= 1
                set CustomPlayer.Bonus_Luck = CustomPlayer.Potion_Bonus_Luck3
            else if(CustomPlayer.Potion_Active_Luck2_Sec>0):
                Print("Reducing luck 2 by 1 sec")
                set CustomPlayer.Potion_Active_Luck2_Sec -= 1
                set CustomPlayer.Bonus_Luck = CustomPlayer.Potion_Bonus_Luck2
            else if(CustomPlayer.Potion_Active_Luck1_Sec>0):
                Print("Reducing luck 1 by 1 sec")
                set CustomPlayer.Potion_Active_Luck1_Sec -= 1
                set CustomPlayer.Bonus_Luck = CustomPlayer.Potion_Bonus_Luck1
            else set CustomPlayer.Bonus_Luck = 0

            if(CustomPlayer.Potion_Active_Gold3_Sec>0):
                set CustomPlayer.Potion_Active_Gold3_Sec -= 1
                set CustomPlayer.Bonus_Gold = CustomPlayer.Potion_Bonus_Gold3
            else if(CustomPlayer.Potion_Active_Gold2_Sec>0):
                set CustomPlayer.Potion_Active_Gold2_Sec -= 1
                set CustomPlayer.Bonus_Gold = CustomPlayer.Potion_Bonus_Gold2
            else if(CustomPlayer.Potion_Active_Gold1_Sec>0):
                set CustomPlayer.Potion_Active_Gold1_Sec -= 1
                set CustomPlayer.Bonus_Gold = CustomPlayer.Potion_Bonus_Gold1
            else set CustomPlayer.Bonus_Gold = 1.0

            if(CustomPlayer.Potion_Active_PetPower3_Sec>0):
                set CustomPlayer.Potion_Active_PetPower3_Sec -= 1
                set CustomPlayer.Bonus_Pet_Power = CustomPlayer.Potion_Bonus_PetPower3
            else if(CustomPlayer.Potion_Active_PetPower2_Sec>0):
                set CustomPlayer.Potion_Active_PetPower2_Sec -= 1
                set CustomPlayer.Bonus_Pet_Power = CustomPlayer.Potion_Bonus_PetPower2
            else if(CustomPlayer.Potion_Active_PetPower1_Sec>0):
                set CustomPlayer.Potion_Active_PetPower1_Sec -= 1
                set CustomPlayer.Bonus_Pet_Power = CustomPlayer.Potion_Bonus_PetPower1
            else set CustomPlayer.Bonus_Pet_Power = 1.0

            if(CustomPlayer.Potion_Active_PlayerPower3_Sec>0):
                set CustomPlayer.Potion_Active_PlayerPower3_Sec -= 1
                set CustomPlayer.Bonus_Player_Power = CustomPlayer.Potion_Bonus_PlayerPower3
            else if(CustomPlayer.Potion_Active_PlayerPower2_Sec>0):
                set CustomPlayer.Potion_Active_PlayerPower2_Sec -= 1
                set CustomPlayer.Bonus_Player_Power = CustomPlayer.Potion_Bonus_PlayerPower2
            else if(CustomPlayer.Potion_Active_PlayerPower1_Sec>0):
                set CustomPlayer.Potion_Active_PlayerPower1_Sec -= 1
                set CustomPlayer.Bonus_Player_Power = CustomPlayer.Potion_Bonus_PlayerPower1
            else set CustomPlayer.Bonus_Player_Power = 1.0
            
            CustomPlayer.Update_Pet_Power()
            CustomPlayer.Update_Luck()
            CustomPlayer.Update_Inventory_Count()
            CustomPlayer.Update_Inventory_Timings()

    PotionDrop(Agent : agent):void =
        Print("Drop attempt")
        if:
            PlayerObj := player[Agent]
            CustomPlayer:Custom_Player = PlayersMap[PlayerObj]
        then:
            #Type of potion:
            var RandResult_Type:int = GetRandomInt(1,4)
            if (RandResult_Type=1):#Luck
                Print("Luck")
                #Level of potion
                var RandResult_Level:int = GetRandomInt(0,10000)
                Print("Rand:{RandResult_Level}")
                if (RandResult_Level>9995):
                    set CustomPlayer.Potion_Owned_Luck3 += 1
                    AddPotionLuck3(Agent,1)
                    Found_PotionHUD.Show(Agent)
                else if (RandResult_Level>9980):
                    set CustomPlayer.Potion_Owned_Luck2 += 1
                    AddPotionLuck2(Agent,1)
                    Found_PotionHUD.Show(Agent)
                else if (RandResult_Level>9875):
                    set CustomPlayer.Potion_Owned_Luck1 += 1
                    AddPotionLuck1(Agent,1)
                    Found_PotionHUD.Show(Agent)
            if (RandResult_Type=2):#Gold
                Print("Gold")
                #Level of potion
                var RandResult_Level:int = GetRandomInt(0,10000)
                Print("Rand:{RandResult_Level}")
                if (RandResult_Level>9995):
                    set CustomPlayer.Potion_Owned_Gold3 += 1
                    AddPotionGold3(Agent,1)
                    Found_PotionHUD.Show(Agent)
                else if (RandResult_Level>9980):
                    set CustomPlayer.Potion_Owned_Gold2 += 1
                    AddPotionGold2(Agent,1)
                    Found_PotionHUD.Show(Agent)
                else if (RandResult_Level>9875):
                    set CustomPlayer.Potion_Owned_Gold1 += 1
                    AddPotionGold1(Agent,1)
                    Found_PotionHUD.Show(Agent)
            if (RandResult_Type=3):#Pet Power
                Print("Pet Power")
                #Level of potion
                var RandResult_Level:int = GetRandomInt(0,10000)
                Print("Rand:{RandResult_Level}")
                if (RandResult_Level>9995):
                    set CustomPlayer.Potion_Owned_Pet_Power3 += 1
                    AddPotionPetPower3(Agent,1)
                    Found_PotionHUD.Show(Agent)
                else if (RandResult_Level>9980):
                    set CustomPlayer.Potion_Owned_Pet_Power2 += 1
                    AddPotionPetPower2(Agent,1)
                    Found_PotionHUD.Show(Agent)
                else if (RandResult_Level>9875):
                    set CustomPlayer.Potion_Owned_Pet_Power1 += 1
                    AddPotionPetPower1(Agent,1)
                    Found_PotionHUD.Show(Agent)
            if (RandResult_Type=4):#Player Power  
                Print("Player Power")
                #Level of potion
                var RandResult_Level:int = GetRandomInt(0,10000)
                Print("Rand:{RandResult_Level}")
                if (RandResult_Level>9995):
                    set CustomPlayer.Potion_Owned_Player_Power3 += 1
                    AddPotionPlayerPower3(Agent,1)
                    Found_PotionHUD.Show(Agent)
                else if (RandResult_Level>9980):
                    set CustomPlayer.Potion_Owned_Player_Power2 += 1
                    AddPotionPlayerPower2(Agent,1)
                    Found_PotionHUD.Show(Agent)
                else if (RandResult_Level>9875):
                    set CustomPlayer.Potion_Owned_Player_Power1 += 1
                    AddPotionPlayerPower1(Agent,1)
                    Found_PotionHUD.Show(Agent)
    
    #WHEEL REALTED:
    AwaitButtonInteraction(LuckyMaterials1 : Materials.MI_LuckyWheelScreen_1_material)<suspends>:void=
        loop:
            Agent := WheelButton.InteractedWithEvent.Await()
            OnButtonPressed(Agent, LuckyMaterials1)
            Sleep(0.1)

    OnButtonPressed(Agent : agent, LuckyMaterials1 : Materials.MI_LuckyWheelScreen_1_material):void=
        if:
            PlayerObj := player[Agent]
            CustomPlayer:Custom_Player = PlayersMap[PlayerObj]
            CustomPlayer.PlayerDiamond<100
        then:   
            NotEnoughDiamondHUD.Show(Agent)
        else:
            if:
                PlayerObj := player[Agent]
                CustomPlayer:Custom_Player = PlayersMap[PlayerObj]
                CustomPlayer.PlayerDiamond>=100
            then:   
                RadioWheel.Play()
                CustomPlayer.DecreaseDiamondBy(100)
                DecreaseDiamondBalance(Agent,100)
                WheelButton.Disable()
                spawn {OnLuckyWheelActivated(Agent, LuckyMaterials1)}

    OnLuckyWheelActivated(Agent : agent, LuckyMaterials1 : Materials.MI_LuckyWheelScreen_1_material)<suspends>:void=
        var SpinSpeed : float = GetRandomFloat(0.5, 2.0)
        loop:
            if (WheelID = 0):
                set LuckyMaterials1.WinSlot = 0.0
                set LuckyMaterials1.Rotation += SpinSpeed
                if (LuckyMaterials1.Rotation >= (SlotsNumber - 1) * 1.0):
                    set LuckyMaterials1.Rotation = 0.0
                if (SpinSpeed > 0.0):
                    set SpinSpeed -= 0.01
                else:
                    if:
                        WinSlotIDX := Round[LuckyMaterials1.Rotation]
                    then:
                        spawn {FinishRotation(WinSlotIDX, Agent, LuckyMaterials1)}
                        break
            Sleep(0.01)

    FinishRotation(WinSlotIDX : int, Agent : agent, LuckyMaterials1 : Materials.MI_LuckyWheelScreen_1_material)<suspends>:void=
        loop:
            if (WheelID = 0):
                if (LuckyMaterials1.Rotation > WinSlotIDX * 1.0):
                    set LuckyMaterials1.Rotation -= 0.01
                else:
                    set LuckyMaterials1.Rotation += 0.01
                if (WinSlotIDX * 1.0 >= LuckyMaterials1.Rotation - 0.1 and WinSlotIDX * 1.0 <= LuckyMaterials1.Rotation + 0.1):
                    set LuckyMaterials1.WinSlot = (WinSlotIDX + 1) * 1.0
                    OnRewardGranted(WinSlotIDX, Agent)
                    Sleep(5.0)
                    set LuckyMaterials1.WinSlot = 0.0
                    break
            Sleep(0.01)

    OnRewardGranted(WinSlotIDX : int, Agent : agent):void=        
        Print("Winning slot: {WinSlotIDX}")
        if:
            PlayerObj := player[Agent]
            CustomPlayer:Custom_Player = PlayersMap[PlayerObj]
        then:
            if(WinSlotIDX=0):
                set CustomPlayer.Potion_Owned_Gold1 += 1
                AddPotionGold1(Agent,1)
            else if(WinSlotIDX=1):
                set CustomPlayer.Potion_Owned_Luck1 += 1
                AddPotionLuck1(Agent,1)
            else if(WinSlotIDX=2):
                set CustomPlayer.Potion_Owned_Player_Power1 += 1
                AddPotionPlayerPower1(Agent,1)
            else if(WinSlotIDX=3):
                set CustomPlayer.Potion_Owned_Pet_Power1 += 1
                AddPotionPetPower1(Agent,1)
            else if(WinSlotIDX=4):
                set CustomPlayer.Potion_Owned_Gold1 += 1
                AddPotionGold1(Agent,1)
            else if(WinSlotIDX=5):
                set CustomPlayer.Potion_Owned_Luck1 += 1
                AddPotionLuck1(Agent,1)
            else if(WinSlotIDX=6):
                set CustomPlayer.Potion_Owned_Player_Power1 += 1
                AddPotionPlayerPower1(Agent,1)
            else if(WinSlotIDX=7):
                set CustomPlayer.Potion_Owned_Pet_Power1 += 1
                AddPotionPetPower1(Agent,1)
            else if(WinSlotIDX=8):
                set CustomPlayer.Potion_Owned_Gold2 += 1
                AddPotionGold2(Agent,1)
            else if(WinSlotIDX=9):
                set CustomPlayer.Potion_Owned_Pet_Power2 += 1
                AddPotionPetPower2(Agent,1)
            else if(WinSlotIDX=10):
                set CustomPlayer.Potion_Owned_Player_Power2 += 1
                AddPotionPlayerPower2(Agent,1)
            else if(WinSlotIDX=11):
                set CustomPlayer.Potion_Owned_Luck2 += 1
                AddPotionLuck2(Agent,1)
            else if(WinSlotIDX=12):
                set CustomPlayer.Potion_Owned_Gold3 += 1
                AddPotionGold3(Agent,1)
            else if(WinSlotIDX=13):
                set CustomPlayer.Potion_Owned_Pet_Power3 += 1
                AddPotionPetPower3(Agent,1)
            else if(WinSlotIDX=14):
                set CustomPlayer.Potion_Owned_Player_Power3 += 1
                AddPotionPlayerPower3(Agent,1)
            else if(WinSlotIDX=15):
                set CustomPlayer.Potion_Owned_Luck3 += 1
                AddPotionLuck3(Agent,1)

        WheelButton.Enable()

    #Diamond loots
    CollectDiamondLoot(Agent : agent):void =
        Print("CollectDiamondLoot")
        if:
            PlayerObj := player[Agent]
            CustomPlayer:Custom_Player = PlayersMap[PlayerObj]
            CurrentPlayerStats := PlayerStatsManager.GetPlayerStats[Agent]
        then:
            Value_Float:float=10.0
            if(Value:int=Round[Value_Float]):
                CustomPlayer.IncreaseDiamondBy(Value)
                AddDiamondBalance(Agent,Value)
                CheckDiamonds(Agent)
    
    DiamondObbyNormal(Agent : agent):void =
        if:
            PlayerObj := player[Agent]
            CustomPlayer:Custom_Player = PlayersMap[PlayerObj]
            CurrentPlayerStats := PlayerStatsManager.GetPlayerStats[Agent]
        then:
            Value_Float:float=100.0
            if(Value:int=Round[Value_Float]):
                CustomPlayer.IncreaseDiamondBy(Value)
                AddDiamondBalance(Agent,Value)
                CheckDiamonds(Agent)

    DiamondObbyOnTime(Agent : agent):void =
        if:
            PlayerObj := player[Agent]
            CustomPlayer:Custom_Player = PlayersMap[PlayerObj]
            CurrentPlayerStats := PlayerStatsManager.GetPlayerStats[Agent]
        then:
            Value_Float:float=200.0
            if(Value:int=Round[Value_Float]):
                CustomPlayer.IncreaseDiamondBy(Value)
                AddDiamondBalance(Agent,Value)
                CheckDiamonds(Agent)

    BarrelLoot(Agent : agent):void =
        Print("BarrelLoot")
        if:
            PlayerObj := player[Agent]
            CustomPlayer:Custom_Player = PlayersMap[PlayerObj]
        then:
            var RandResult_Type:int = GetRandomInt(1,4)
            if (RandResult_Type=1):#Luck
                Print("Luck")
                #Level of potion
                var RandResult_Level:int = GetRandomInt(1,100)
                Print("Rand:{RandResult_Level}")
                if (RandResult_Level>80):
                    set CustomPlayer.Potion_Owned_Luck1 += 1
                    AddPotionLuck1(Agent,1)
                    Barrel_Found_SomethingHUD.Show(Agent)
                else:
                    Barrel_EmptyHUD.Show(Agent)
            if (RandResult_Type=2):#Gold
                Print("Gold")
                #Level of potion
                var RandResult_Level:int = GetRandomInt(1,100)
                Print("Rand:{RandResult_Level}")
                if (RandResult_Level>80):
                    set CustomPlayer.Potion_Owned_Gold1 += 1
                    AddPotionGold1(Agent,1)
                    Barrel_Found_SomethingHUD.Show(Agent)
                else:
                    Barrel_EmptyHUD.Show(Agent)
            if (RandResult_Type=3):#Pet Power
                Print("Pet Power")
                #Level of potion
                var RandResult_Level:int = GetRandomInt(1,100)
                Print("Rand:{RandResult_Level}")
                if (RandResult_Level>80):
                    set CustomPlayer.Potion_Owned_Pet_Power1 += 1
                    AddPotionPetPower1(Agent,1)  
                    Barrel_Found_SomethingHUD.Show(Agent)
                else:
                    Barrel_EmptyHUD.Show(Agent)
            if (RandResult_Type=4):#Player Power  
                Print("Player Power")
                #Level of potion
                var RandResult_Level:int = GetRandomInt(1,100)
                Print("Rand:{RandResult_Level}")
                if (RandResult_Level>80):
                    set CustomPlayer.Potion_Owned_Player_Power1 += 1
                    AddPotionPlayerPower1(Agent,1)
                    Barrel_Found_SomethingHUD.Show(Agent)
                else:
                    Barrel_EmptyHUD.Show(Agent)

    PopupButton1(Agent : agent, IntNum : int):void =
        if(IntNum=0):
            if (AgentData := AgentMap[Agent], AgentData.bShopOpen = false,Player := player[Agent], PlayerUI := GetPlayerUI[Player], CustomPlayer:Custom_Player = PlayersMap[Player]):
                set AgentData.bShopOpen = true
                set AgentData.CurrentPageIndex = 0
    
                ShopCanvas := CreateShopCanvas()
                set AgentData.MaybeShopCanvas = option{ ShopCanvas }
                UpdatePurchaseFrames(Agent)
    
                PlayerUI.AddWidget(ShopCanvas, player_ui_slot{ InputMode := ui_input_mode.All })
        if(IntNum=2):
            if:
                PlayerObj := player[Agent]
                CustomPlayer:Custom_Player = PlayersMap[PlayerObj]
            then:
                set CustomPlayer.bShopOpen = true
                CustomPlayer.InitiateUI_PetDex()

    CheckDiamonds(Agent : agent):void =
        if:
            CurrentPlayerStats := PlayerStatsManager.GetPlayerStats[Agent]
            PlayerObj := player[Agent]
            CustomPlayer:Custom_Player = PlayersMap[PlayerObj]
            CurrentPlayerStats.Wheelshowed.CurrentValue=1
        then:
            Print("Nothing")
        else:
            if:
                CurrentPlayerStats := PlayerStatsManager.GetPlayerStats[Agent]
                PlayerObj := player[Agent]
                CustomPlayer:Custom_Player = PlayersMap[PlayerObj]
                CurrentPlayerStats.Wheelshowed.CurrentValue=0
                CustomPlayer.PlayerDiamond>=100
            then:
                RegisterWheelshowed(Agent)
                Objective_Wheel.ActivateObjectivePulse(Agent)
                Try_Wheel_HUD.Show(Agent)

    Entered_Wheel(Agent : agent):void =
        Objective_Wheel.DeactivateObjectivePulse(Agent)