using { /Fortnite.com/Devices }
using { /Verse.org/Simulation }
using { /UnrealEngine.com/Temporary/UI}
using { /UnrealEngine.com/Temporary/SpatialMath}
using { /Verse.org/Colors/NamedColors }
using { /Fortnite.com/UI }
using { /Verse.org/Assets }

S2M<localizes>(S:string):message="{S}"

timed_reward<public> := class<concrete><unique>:

    @editable    RewardTrigger<private>:trigger_device = trigger_device{}
    @editable    Timer<private>:timer_device = timer_device{}
    @editable    var AdditionalSeconds:int = 0

    var WidgetPerPlayer<private>:[agent]widget = map{}

    HUDUpdateNecesaryEvent:event(agent) = event(agent){}

    ClaimedEvent:event(agent) = event(agent){}

    GetTime<public>(Agent:agent)<transacts>:float=Timer.GetActiveDuration(Agent) + AdditionalSeconds*1.0

    OnTimerEnd<private>(MaybeAgent:?agent):void=

        Message := S2M("CLAIM")

        if:
            Agent := MaybeAgent?
        then:
            HUDUpdateNecesaryEvent.Signal(Agent)
            ClaimButton := button_quiet:
                DefaultText := Message
            ClaimButton.OnClick().Subscribe(OnClaimed)
            if:
                set WidgetPerPlayer[Agent] = ClaimButton

    OnClaimed<private>(WData:widget_message):void=

        #Print("Claimed in class")
        Message := S2M("CLAIMED")
        Agent := WData.Player
        RewardTrigger.Trigger(Agent)
        ClaimedText:= text_block:
            DefaultText := Message
        HUDUpdateNecesaryEvent.Signal(Agent)
        if:
            set WidgetPerPlayer[Agent] = ClaimedText
        if(Player:=player[Agent], PlayerUI:=GetPlayerUI[Player]):
            if(MyUI:=WidgetPerPlayer[Agent]):
                PlayerUI.RemoveWidget(MyUI)
        ClaimedEvent.Signal(Agent)

    ToStopWatchFormat<public>(F:float)<decides><transacts>:[]char=

        F2 := F*1.00000001

        Hours := Floor[F2/3600.0]
        Minutes := Mod[Floor[F2/60.0], 60]
        Seconds := Mod[Floor[F2], 60]
        var MinutesString:[]char = "{Minutes}"
        var SecondsString:[]char = "{Seconds}"
        var HoursString:[]char = "{Hours}"
        if ( Minutes < 10 ):
            set MinutesString = "0{Minutes}"
        if ( Seconds < 10 ):
            set SecondsString = "0{Seconds}"
        if ( Hours < 10 ):
            set HoursString = "0{Hours}"
        if ( Minutes = 1 ):
            set MinutesString = "00"
        return "{HoursString}:{MinutesString}:{SecondsString}"

    Start<public>(Agent:agent)<suspends>:void=

        Timer.Start(Agent)
        Timer.SuccessEvent.Subscribe(OnTimerEnd)
        TimerText := text_block:
        if:
            set WidgetPerPlayer[Agent] = TimerText
        loop:
            Time := GetTime(Agent)
            if (TimeString := ToStopWatchFormat[Time]):
                TimerText.SetText(S2M(TimeString))
            if:
                TimerDuration := Timer.GetActiveDuration(Agent)
                TimerDuration < Time
            then:
                Timer.SetActiveDuration(TimerDuration+1.0, Agent)
                set AdditionalSeconds -= 1
            Sleep(1.0)

    GetRewardStackBox<public>(Agent:agent):stack_box=

        Time := GetTime(Agent)

        var StackBox:stack_box = stack_box:
            Orientation := orientation.Vertical

        if (CurrentWidget:widget = WidgetPerPlayer[Agent]):

            set StackBox = stack_box:
                Orientation := orientation.Vertical
                Slots := array:
                    stack_box_slot:
                        HorizontalAlignment:=horizontal_alignment.Left
                        VerticalAlignment:=vertical_alignment.Top
                        Padding:=margin{Bottom:=5.0}
                        Widget := texture_block:
                            DefaultImage := RewardUIImages.presentnew
                            DefaultDesiredSize := vector2{X:=150.0, Y:=150.0}
                    stack_box_slot:
                        HorizontalAlignment:=horizontal_alignment.Center
                        VerticalAlignment:=vertical_alignment.Bottom
                        Widget := CurrentWidget
            <#
            set StackBox = stack_box:
                Orientation := orientation.Vertical
                Slots := array:
                    stack_box_slot:
                        HorizontalAlignment:=horizontal_alignment.Center
                        VerticalAlignment:=vertical_alignment.Top
                        Padding:=margin{Bottom:=5.0}
                        Widget := texture_block:
                            DefaultImage := RewardUIImages.presentnew
                            DefaultDesiredSize := vector2{X:=150.0, Y:=150.0}
                    stack_box_slot:
                        HorizontalAlignment:=horizontal_alignment.Center
                        VerticalAlignment:=vertical_alignment.Bottom
                        Widget := CurrentWidget   
            #>

        else:
            Print("Can't get widget")

        return StackBox

playtime_reward_ui := class(creative_device):

    @editable   HUD_Trigger:trigger_device = trigger_device{}
    @editable   TimedRewards:[]timed_reward = array{}
    @editable   Spawners:[]player_spawner_device = array:

    var PlayersMap:[agent]timer_reward_player = map{}
    var OptionalUIPerPlayer:[agent]?canvas = map{}
    var LowestTimePerPlayer:[agent]tuple(text_block, text_block) = map{}

    RemoveHUDEvent:event(agent) = event(agent){}

    OnBegin<override>()<suspends>:void=

        HUD_Trigger.TriggeredEvent.Subscribe(ShowRewardUI)
        for(Spawner:Spawners):
            Spawner.SpawnedEvent.Subscribe(OnPlayerAdded)
        spawn:
            UpdateAllHUDs()
        spawn:
            AwaitAllClaimEvents()

    ShowRewardUI(MaybeAgent:?agent):void=

        spawn { ShowRewardUI_Async(MaybeAgent) }

    ShowRewardUI_Async(MaybeAgent:?agent)<suspends>:void=

        if ( Agent := MaybeAgent? ):

            UI2 := MakeRewardCanvas(Agent)
            ShowHUD(option{Agent}, UI2)
            loop:
                SignaledAgent := RemoveHUDEvent.Await()
                if (SignaledAgent = Agent):
                    break

            RemoveHUD(Agent)

    RewardClaimedEvent:event(agent) = event(agent){}
    
    AwaitClaimEvent(TimedReward:timed_reward)<suspends>:void=
        Agent := TimedReward.ClaimedEvent.Await()
        Print("Claimed Event")
        RewardClaimedEvent.Signal(Agent)

    AwaitAllClaimEvents()<suspends>:void=
        for(TimedReward:TimedRewards):
            spawn:
                AwaitClaimEvent(TimedReward)

    UpdateLowestTime(Agent:agent)<suspends>:void=

        var GiftsAvailible:int = 0

        Default := S2M("START✔")      
        
        branch:
            loop:
                ClaimerAgent := RewardClaimedEvent.Await()
                #Print("Claimed")
                if (Agent = ClaimerAgent):
                    set GiftsAvailible -= 1

        if:
            set LowestTimePerPlayer[Agent] = (
                text_block{
                    DefaultText := Default
                    DefaultTextColor := Black
                    DefaultJustification := text_justification.Center
            },
                text_block{
                    DefaultText := Default
                    DefaultTextColor := Black
                    DefaultJustification := text_justification.Center
            })
            

        var ActiveTimedRewards:[]timed_reward = TimedRewards

        loop:

            var LowestTime:float = 1000000.0

            var LowestTimeMessage:message = S2M("")

            for (TimedReward:ActiveTimedRewards):

                if:
                    RewardTime := TimedReward.GetTime(Agent) 
                    RewardTime <= LowestTime
                    set LowestTime = RewardTime
                    TimeString := TimedReward.ToStopWatchFormat[LowestTime]
                    CurrentTBs := LowestTimePerPlayer[Agent]

                then:
                    set LowestTimeMessage = S2M(TimeString)
                    TimeTB := CurrentTBs(0)
                    TimeTB.SetText(LowestTimeMessage)
                    if (RewardTime <= 1.0):
                        set ActiveTimedRewards = for (Reward:ActiveTimedRewards, Reward <> TimedReward){Reward}
                        set GiftsAvailible += 1
                    if:
                        GiftsAvailible > 0
                    then:
                        TimeTB.SetText(S2M("Ready!"))

                    CurrentTBs(1).SetText(S2M("{GiftsAvailible}"))

            Sleep(1.0)

    MakeClosestRewardCanvas(Agent:agent)<decides><transacts>:canvas=

        LowestTimeMessage := LowestTimePerPlayer[Agent]

        ClosestRewardOverlay:=overlay:
            

            Slots := array:
                    overlay_slot:
                        Widget := texture_block:
                            DefaultDesiredSize := vector2{X:=55.0, Y:=55.0}
                            DefaultImage := RewardUIImages.closest_rewardnew
                    <#
                    overlay_slot:
                        Padding := margin{Right := 11.0}
                            Widget := LowestTimeMessage(1)
                        HorizontalAlignment:=horizontal_alignment.Right
                        VerticalAlignment:=vertical_alignment.Top
                    overlay_slot:
                        Widget := LowestTimeMessage(0)
                        HorizontalAlignment:=horizontal_alignment.Center
                        VerticalAlignment:=vertical_alignment.Bottom
                    #>

                
        canvas:
            Slots := array:
                canvas_slot:
                    ZOrder := 1
                    Anchors:=anchors{Minimum:=vector2{X:=1.0,Y:=0.5},Maximum:=vector2{X:=1.0,Y:=0.5}}
                    SizeToContent:=true
                    Widget := ClosestRewardOverlay
                    Offsets:=margin{Left:=-55.0,Top:=0.0}
                    Alignment:=vector2{X:=2.0, Y:=-5.4}

    MakeRewardCanvas(Agent:agent):canvas=

        var Row1StackBoxSlots:[]stack_box_slot = array{}
        var Row2StackBoxSlots:[]stack_box_slot = array{}
        var Row3StackBoxSlots:[]stack_box_slot = array{}

        for (Index := 0..8):

            var HorizontalPadding:float = 40.0
            if ( Index = 8 or Index = 5 or Index = 2 ):
                set HorizontalPadding = 0.0

            if (TimedReward := TimedRewards[Index]):

                RewardStackBoxSlot := stack_box_slot:
                    HorizontalAlignment:=horizontal_alignment.Fill
                    VerticalAlignment:=vertical_alignment.Fill
                    Padding:=margin{Right := HorizontalPadding}
                    Widget := TimedReward.GetRewardStackBox(Agent)
                    <#
                    HorizontalAlignment:=horizontal_alignment.Fill
                    VerticalAlignment:=vertical_alignment.Fill
                    Padding:=margin{Right := HorizontalPadding}
                    Widget := TimedReward.GetRewardStackBox(Agent)
                    #>

                if ( 0 <= Index <= 2 ):

                    set Row1StackBoxSlots += array{RewardStackBoxSlot}

                if ( 3 <= Index <= 5 ):

                    set Row2StackBoxSlots += array{RewardStackBoxSlot}

                if (6 <= Index <= 8 ):
                    
                    set Row3StackBoxSlots += array{RewardStackBoxSlot}

        ExitButton:=button_quiet:
            DefaultText := S2M("❌  CLOSE")
        ExitButton.OnClick().Subscribe(RemoveHUDWM)

        canvas:
            Slots := array:
                canvas_slot:
                    ZOrder := 1
                    Anchors:=anchors{Minimum:=vector2{X:=0.5,Y:=0.4},Maximum:=vector2{X:=0.5,Y:=0.4}}
                    Alignment:=vector2{X:=0.5,Y:=0.4}
                    SizeToContent:=true
                    Widget :=  overlay:
                        Slots := array:
                            overlay_slot:
                                Widget := texture_block:
                                    DefaultImage := RewardUIImages.present_background
                                    DefaultDesiredSize := vector2{X:=1300.0, Y:=1300.0}
                            overlay_slot:
                                Padding := margin{Top := 50.0}
                                Widget := stack_box:
                                    Orientation := orientation.Vertical
                                    Slots := array:
                                        stack_box_slot:
                                            Distribution := option{0.6}
                                            Widget := stack_box:
                                                Orientation := orientation.Horizontal
                                                Slots := Row1StackBoxSlots
                                        stack_box_slot:
                                            Distribution := option{0.6}
                                            Widget := stack_box:
                                                Orientation := orientation.Horizontal
                                                Slots := Row2StackBoxSlots 
                                        stack_box_slot:
                                            Distribution := option{0.6}
                                            Widget := stack_box:
                                                Orientation := orientation.Horizontal
                                                Slots := Row3StackBoxSlots
                                        stack_box_slot:
                                            Widget := ExitButton
                                            Padding := margin {Top := 15.0}

    ShowHUD(Agent:?agent, Canvas:canvas):void=

        #Print("Show HUD")
        
        if(Player:=player[Agent?], PlayerUI:=GetPlayerUI[Player]):
      
            if(PlayersCurrentUI:=OptionalUIPerPlayer[Player]?):
      
                PlayerUI.RemoveWidget(PlayersCurrentUI)
                var PlayersNewCanvas:canvas = Canvas 
                PlayerUI.AddWidget(PlayersNewCanvas, player_ui_slot{InputMode:=ui_input_mode.All})
                if(set OptionalUIPerPlayer[Player] = option{PlayersNewCanvas}) {}
                                                            
            else:                                                   
                var PlayersNewCanvas:canvas = Canvas 
                PlayerUI.AddWidget(PlayersNewCanvas, player_ui_slot{InputMode:=ui_input_mode.All})
                if(set OptionalUIPerPlayer[Player] = option{PlayersNewCanvas}) {}

    RemoveHUD(Agent:agent):void=
        #Print("Removing HUD")
        if(Player:=player[Agent], PlayerUI:=GetPlayerUI[Player]):
            if(MyUI:=OptionalUIPerPlayer[Player]?):
                PlayerUI.RemoveWidget(MyUI)
                if(set OptionalUIPerPlayer[Player] = false) {}
            else:
                Print("Err 2")
        else:
            Print("Err 1")

    RemoveHUDWM(WData:widget_message):void=

        #Print("Removing HUD")
        if(Player:=player[WData.Player], PlayerUI:=GetPlayerUI[Player]):
            if(MyUI:=OptionalUIPerPlayer[Player]?):
                PlayerUI.RemoveWidget(MyUI)
                RemoveHUDEvent.Signal(Player)
                if(set OptionalUIPerPlayer[Player] = false) {}
            else:
                Print("Err 2")
        else:
            Print("Err 1")

    OnPlayerAdded(Agent:agent):void=

        #Print("Player Added")

        if ( CustomPlayer:timer_reward_player = PlayersMap[Agent] ) {}
        else if:
            set PlayersMap[Agent] = timer_reward_player{AgentObj := Agent}
        then:
            for ( Reward : TimedRewards ):
                spawn:
                    Reward.Start(Agent)
            spawn:
                UpdateLowestTime(Agent)

            if (Player := player[Agent], PlayerUI := GetPlayerUI[Player], ClosestRewardCanvas := MakeClosestRewardCanvas[Agent]):
                PlayerUI.AddWidget(ClosestRewardCanvas)
            else:
                #Print("Cant Get Reward Canvas")

                

    UpdateAllHUDs()<suspends>:void=

        for(TimedReward : TimedRewards):
            spawn:
                UpdateHUD(TimedReward)

    UpdateHUD(TimedReward:timed_reward)<suspends>:void=

        loop:
            Agent := TimedReward.HUDUpdateNecesaryEvent.Await()
            Sleep(0.1)
            if:
                not OptionalUIPerPlayer[Agent] = false
                OptionalUIPerPlayer[Agent]
            then:
                ShowHUD(option{Agent}, MakeRewardCanvas(Agent))

timer_reward_player := class:

    AgentObj:agent